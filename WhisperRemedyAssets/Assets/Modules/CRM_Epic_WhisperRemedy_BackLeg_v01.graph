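# Control Rig module that rigs a back leg with a hock, from Hip Socket through Knee, Ankle and Ball to Toe Tip connectors.
# Construction resolves connectors and bone lengths, then spawns hip, IK, pole vector, roll, toe and offset controls, plus space switches and visibility switching.
# IK is layered: a main hip-to-foot aim carries the leg, and lower-leg functions bend knee and ankle with triangle-angle math.
# The IK control gets knee offset, hock offset and twist channels; side flags mirror pole vectors.
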ControlRigBlueprint /WhisperRemedyAssets/Assets/Modules/CRM_Epic_WhisperRemedy_BackLeg_v01
graphs [
 Add_6_ContainedGraph {nodes [
   n1 = RigVMFunctionEntryNode()
   RigVMFunctionReturnNode(Result=&RigVMFunction_MathDoubleAdd(A=&RigVMFunction_MathDoubleAdd(A=@n1.A, B=@n1.B).Result, B=@n1.C).Result)
  ]}
 ConstructionGraph {nodes [
   Create Main Leg Ik System = RigVMFunctionReferenceNode(ExecuteContext=())
   > Create Lower Leg Ik System = RigVMFunctionReferenceNode(Input Hip Socket Transform=@"Create Main Leg Ik System".Hip Socket Transform)
   > Create Upper Leg Ik System = RigVMFunctionReferenceNode(
    Input Ankle Transform=@"Create Lower Leg Ik System".Ankle Transform
    Input Hip Socket Transform=@"Create Lower Leg Ik System".Hip Socket Transform
    Input Knee Transform=@"Create Lower Leg Ik System".Knee Transform
   )
   > Create Local Transforms = RigVMFunctionReferenceNode()
   Main Leg Ik System = RigVMFunctionReferenceNode(ExecuteContext=(), Initial=true)
   > Knee Offset Ik System = RigVMFunctionReferenceNode(Initial=true)
   > Ankle Offset Ik System = RigVMFunctionReferenceNode(Initial=true)
   RigUnit_PrepareForExecution = &RigUnit_PrepareForExecution()
   RigVMFunction_Sequence = RigVMAggregateNode(ExecuteContext=@RigUnit_PrepareForExecution.ExecutePin)
   Get Connectors = RigVMFunctionReferenceNode(ExecuteContext=@RigVMFunction_Sequence.A)
   > Get Bone Lengths = RigVMFunctionReferenceNode()
   Create Hip Socket Control = RigVMFunctionReferenceNode(ExecuteContext=@RigVMFunction_Sequence.B)
   Create Ik Control = RigVMFunctionReferenceNode(ExecuteContext=@RigVMFunction_Sequence.C)
   > Create Rig Up Vector = RigVMFunctionReferenceNode()
   Create Roll Controls = RigVMFunctionReferenceNode(ExecuteContext=@RigVMFunction_Sequence.D)
   > Create Toe Controls = RigVMFunctionReferenceNode()
   Create Ankle Offset Ik Control = RigVMFunctionReferenceNode(ExecuteContext=@RigVMFunction_Sequence.E)
   > Create Knee Twist And PV Control = RigVMFunctionReferenceNode()
   Create Local Transforms_1 = RigVMFunctionReferenceNode(ExecuteContext=@RigVMFunction_Sequence.F)
   Set Metadata = RigVMFunctionReferenceNode(ExecuteContext=@RigVMFunction_Sequence.G)
   > Create Space Switches = RigVMFunctionReferenceNode()
   > Create Sec Control Visibility Switch = RigVMFunctionReferenceNode()
   Create Rig Joints = RigVMFunctionReferenceNode(ExecuteContext=())
  ]}
 New Function_ContainedGraph {nodes [
   n1 = RigVMFunctionEntryNode()
   RigVMFunction_Sequence = &RigVMFunction_Sequence(ExecuteContext=@n1.ExecuteContext)
   RigVMFunction_Sequence_1 = RigVMAggregateNode(ExecuteContext=@RigVMFunction_Sequence.A)
   RigVMVariableNode(
    Value=/Script/ControlRig.RigUnit_ResolveConnector(Connector=(Type=Connector,Name="Hip Socket"), SkipSocket=False) {is=struct RigVMUnitNode}.Result
    ExecuteContext=@RigVMFunction_Sequence_1.A
   ) {variable=Hip Socket Connector}
   > RigVMVariableNode(Value=&RigUnit_ResolveConnector(Connector=(Type=Connector,Name="Knee"), SkipSocket=False).Result) {variable=Knee Connector}
   RigVMVariableNode(
    Value=&RigUnit_ResolveConnector(Connector=(Type=Connector,Name="Ankle"), SkipSocket=False).Result
    ExecuteContext=@RigVMFunction_Sequence_1.B
   ) {variable=Ankle Connector}
   > RigVMVariableNode(Value=&RigUnit_ResolveConnector(Connector=(Type=Connector,Name="Ball"), SkipSocket=False).Result) {variable=Ball Connector}
   RigVMVariableNode(
    Value=&RigUnit_ResolveConnector(Connector=(Type=Connector,Name="Toe"), SkipSocket=False).Result
    ExecuteContext=@RigVMFunction_Sequence_1.C
   ) {variable=Toe Connector}
   > RigVMVariableNode(Value=&RigUnit_ResolveConnector(Connector=(Type=Connector,Name="Toe Tip"), SkipSocket=False).Result) {variable=Toe Tip Connector}
   RigVMFunctionReturnNode(ExecuteContext=@RigVMFunction_Sequence.B)
  ]}
 New Function_ContainedGraph {nodes [
   n1 = RigVMFunctionEntryNode()
   RigVMFunction_Sequence = &RigVMFunction_Sequence(ExecuteContext=@n1.ExecuteContext)
   RigVMFunctionReturnNode(ExecuteContext=@RigVMFunction_Sequence.B)
  ]}
 New Function_ContainedGraph {nodes [
   n1 = RigVMFunctionEntryNode()
   RigVMFunction_Sequence = &RigVMFunction_Sequence(ExecuteContext=@n1.ExecuteContext)
   RigVMFunctionReturnNode(ExecuteContext=@RigVMFunction_Sequence.B)
   RigVMFunction_DebugLineNoSpace = /Script/RigVM.RigVMFunction_DebugLineNoSpace(
    Color=(R=1,G=0,B=0,A=1)
    DebugDrawSettings=(DepthPriority=SDPG_Foreground,Lifetime=-1)
    Thickness=0.5
    WorldOffset=(Rotation=(X=0,Y=0,Z=0,W=1),Translation=(X=0,Y=0,Z=0),Scale3D=(X=1,Y=1,Z=1))
    bEnabled=true
    A=&RigUnit_GetTransform(bInitial=False, Item=RigVMVariableNode() {variable=Hip Socket Connector}.Value).Transform.Translation
    B=&RigUnit_GetTransform(bInitial=False, Item=RigVMVariableNode() {variable=Ball Connector}.Value).Transform.Translation
    ExecutePin=@RigVMFunction_Sequence.A
   ) {is=struct RigVMUnitNode}
  ]}
 New Function_ContainedGraph {nodes [
   n1 = RigVMFunctionEntryNode()
   RigVMFunction_Sequence = &RigVMFunction_Sequence(ExecuteContext=@n1.ExecuteContext)
   RigVMFunctionReturnNode(ExecuteContext=@RigVMFunction_Sequence.B)
   RigVMFunction_ControlFlowBranch = &RigVMFunction_ControlFlowBranch(Condition=RigVMVariableNode() {variable=DebugOnOff}.Value, ExecuteContext=@RigVMFunction_Sequence.A)
   n2 = &RigVMFunction_Sequence(ExecuteContext=@RigVMFunction_ControlFlowBranch.True)
   Debug Ik Line = RigVMFunctionReferenceNode(ExecuteContext=@n2.A)
   > Debug Main Leg Ik Line = RigVMFunctionReferenceNode()
   > Debug Lower Leg Ik Line = RigVMFunctionReferenceNode()
   Debug Lower Leg Ik Pv Line = RigVMFunctionReferenceNode(ExecuteContext=@n2.B)
   > Debug Upper Leg Ik PV Line = RigVMFunctionReferenceNode()
  ]}
 New Function_ContainedGraph {nodes [
   n1 = RigVMFunctionEntryNode()
   RigVMFunction_Sequence = &RigVMFunction_Sequence(ExecuteContext=@n1.ExecuteContext)
   n2 = &RigVMFunction_Sequence(ExecuteContext=@RigVMFunction_Sequence.A)
   Create Name = RigVMFunctionReferenceNode(Tokens=("Bk Rig Up Vector Null"), ExecuteContext=@n2.A)
   n3 = RigVMDispatchNode(
    True=1
    False=RigVMDispatchNode(False=1, True=-1, Condition=RigVMVariableNode() {variable=Right Side}.Value) {template="DISPATCH_RigVMDispatch_If(in Condition,in True,in False,out Result)"}.Result
    Condition=RigVMVariableNode() {variable=Left Side}.Value
   ) {template="DISPATCH_RigVMDispatch_If(in Condition,in True,in False,out Result)"}
   Add = /Script/RigVM.RigVMFunction_MathVectorAdd(
    A=&RigUnit_GetTransform(bInitial=False, Item=RigVMVariableNode() {variable=Hip Socket Connector}.Value).Transform.Translation
    B=/Script/RigVM.RigVMFunction_MathVectorMul(B=(X=1,Y=1,Z=1), B.X=@n3.Result, B.Y=@n3.Result, B.Z=@n3.Result, A=RigVMVariableNode() {variable=Rig Up Vector}.Value) {is=struct RigVMUnitNode, template="Multiply::Execute(in A,in B,out Result)"}.Result
   ) {is=struct RigVMUnitNode, template="Add::Execute(in A,in B,out Result)"}
   > n4 = &RigUnit_HierarchyAddNull(
    Space=GlobalSpace
    Transform=(Rotation=(X=0,Y=0,Z=0,W=1),Translation=(X=0,Y=0,Z=0),Scale3D=(X=1,Y=1,Z=1))
    Transform.Translation=@Add.Result
    Parent=RigVMVariableNode() {variable=Hip Socket Control}.Value
    Name=@"Create Name".Name
   )
   > /Script/ControlRig.RigUnit_SetScale(Scale=(X=1,Y=1,Z=1), Space=LocalSpace, Weight=1, bInitial=true, bPropagateToChildren=True, Item=@n4.Item) {is=struct RigVMUnitNode}
   > Create Name_1 = RigVMFunctionReferenceNode(Tokens=("Bk Rig Up Vector"))
   n5 = &RigVMFunction_MathDoubleMul(B=0.333, A=RigVMVariableNode() {variable=Control Size}.Value)
   > SpawnControl = /Script/ControlRig.RigUnit_HierarchyAddControlTransform(
    InitialValue=(Rotation=(X=0,Y=0,Z=0,W=1),Translation=(X=0,Y=0,Z=0),Scale3D=(X=1,Y=1,Z=1))
    OffsetSpace=LocalSpace
    OffsetTransform=(Rotation=(X=0,Y=0,Z=0,W=1),Translation=(X=0,Y=0,Z=0),Scale3D=(X=1,Y=1,Z=1))
    Settings="(InitialSpace=LocalSpace,bUsePreferredRotationOrder=False,PreferredRotationOrder=YZX,Limits=(LimitTranslationX=(bMinimum=False,bMaximum=False),LimitTranslationY=(bMinimum=False,bMaximum=False),LimitTranslationZ=(bMinimum=False,bMaximum=False),LimitPitch=(b ...(truncated)"
    Name=@"Create Name_1".Name
    Settings.Shape.Transform.Scale3D.X=@n5.Result
    Settings.Shape.Transform.Scale3D.Y=@n5.Result
    Settings.Shape.Transform.Scale3D.Z=@n5.Result
    Parent=@n4.Item
   ) {
    is=struct RigVMUnitNode
    template="SpawnControl::Execute(in InitialValue,in Settings,in OffsetTransform,in OffsetSpace,in Parent,in Name,out Item)"
   }
   > RigVMVariableNode(Value=@SpawnControl.Item) {variable=Main PV Control}
   RigVMFunctionReturnNode(ExecuteContext=@RigVMFunction_Sequence.B)
  ]}
 New Function_ContainedGraph {nodes [
   n1 = RigVMFunctionEntryNode()
   RigVMFunction_Sequence = &RigVMFunction_Sequence(ExecuteContext=@n1.ExecuteContext)
   RigVMFunction_Sequence_1 = RigVMAggregateNode(ExecuteContext=@RigVMFunction_Sequence.A)
   Get Main Leg PV = RigVMFunctionReferenceNode(Initial=true)
   n2 = /Script/ControlRig.RigUnit_AimBoneMath(
    DebugSettings=(bEnabled=False,Scale=10,WorldOffset=(Rotation=(X=0,Y=0,Z=0,W=1),Translation=(X=0,Y=0,Z=0),Scale3D=(X=1,Y=1,Z=1)))
    Primary=(Weight=1,Axis=(X=1,Y=0,Z=0),Target=(X=1,Y=0,Z=0),Kind=Location,Space=(Type=Bone,Name="None"))
    Secondary=(Weight=1,Axis=(X=0,Y=0,Z=1),Target=(X=1,Y=0,Z=0),Kind=Direction,Space=(Type=Bone,Name="None"))
    Weight=1
    Secondary.Target=@"Get Main Leg PV".Pole Vector
    InputTransform=&RigUnit_GetTransform(bInitial=true, Item=RigVMVariableNode() {variable=Hip Socket Connector}.Value).Transform
    Primary.Target=&RigUnit_GetTransform(bInitial=true, Item=RigVMVariableNode() {variable=Ball Connector}.Value).Transform.Translation
   ) {is=struct RigVMUnitNode}
   RigVMVariableNode(Value=@n2.Result, ExecuteContext=@RigVMFunction_Sequence_1.B) {variable=Hip Socket Rig Transform}
   Get Main Leg PV_1 = RigVMFunctionReferenceNode(Initial=true)
   n3 = &RigUnit_AimBoneMath(
    DebugSettings=(bEnabled=False,Scale=10,WorldOffset=(Rotation=(X=0,Y=0,Z=0,W=1),Translation=(X=0,Y=0,Z=0),Scale3D=(X=1,Y=1,Z=1)))
    Primary=(Weight=1,Axis=(X=1,Y=0,Z=0),Target=(X=1,Y=0,Z=0),Kind=Location,Space=(Type=Bone,Name="None"))
    Secondary=(Weight=1,Axis=(X=0,Y=0,Z=1),Target=(X=1,Y=0,Z=0),Kind=Direction,Space=(Type=Bone,Name="None"))
    Weight=1
    Secondary.Target=@"Get Main Leg PV_1".Pole Vector
    InputTransform=&RigUnit_GetTransform(bInitial=true, Item=RigVMVariableNode() {variable=Hip Socket Connector}.Value).Transform
    Primary.Target=&RigUnit_GetTransform(bInitial=true, Item=RigVMVariableNode() {variable=Knee Connector}.Value).Transform.Translation
   )
   n4 = &RigVMFunction_MathTransformMakeRelative(Global=@n3.Result, Parent=RigVMVariableNode() {variable=Hip Socket Rig Transform}.Value)
   > RigVMVariableNode(Value=@n4.Local) {variable=Hip Socket Rig Local Transform}
   > RigVMVariableNode(Value=@n4.Local) {variable=Main Hip Socket Local Transform}
   RigVMFunctionReturnNode(ExecuteContext=@RigVMFunction_Sequence.B, Hip Socket Transform=RigVMVariableNode() {variable=Hip Socket Rig Transform}.Value)
   n5 = RigVMRerouteNode(Value=())
   Get Distance Between = RigVMFunctionReferenceNode(
    Initial=true
    ExecuteContext=@n5.Value
    First Object=RigVMVariableNode() {variable=Hip Socket Connector}.Value
    Second Object=RigVMVariableNode() {variable=Knee Connector}.Value
   )
   > RigVMVariableNode(Value=@"Get Distance Between".Distance) {variable=Hip Socket Bone Length}
   n6 = RigVMRerouteNode(Value=())
   n7 = RigVMVariableNode() {variable=Knee Connector}
   n8 = &RigUnit_GetTransform(bInitial=true, Item=RigVMVariableNode() {variable=Hip Socket Connector}.Value)
   RigVMVariableNode(
    Value=&RigVMFunction_MathTransformMakeRelative(
     Global=(Rotation=(X=0,Y=0,Z=0,W=1),Translation=(X=0,Y=0,Z=0),Scale3D=(X=1,Y=1,Z=1))
     Global.Translation=&RigUnit_GetTransform(bInitial=true, Item=@n7.Value).Transform.Translation
     Global.Rotation=@n8.Transform.Rotation
     Global.Scale3D=@n8.Transform.Scale3D
     Parent=RigVMVariableNode() {variable=Hip Socket Rig Transform}.Value
    ).Local
    ExecuteContext=@n6.Value
   ) {variable=First Bone Tip Local Offset}
   &RigVMFunction_MathTransformMakeRelative(
    Global=&RigUnit_GetTransform(bInitial=true, Item=RigVMVariableNode() {variable=Hip Socket Connector}.Value).Transform
    Parent=RigVMVariableNode() {variable=Hip Socket Rig Transform}.Value
   )
   RigVMRerouteNode(Value=())
  ]}
 New Function_ContainedGraph {nodes [
   n1 = RigVMFunctionEntryNode()
   RigVMFunction_Sequence = &RigVMFunction_Sequence(ExecuteContext=@n1.ExecuteContext)
   RigVMFunction_Sequence_3 = RigVMAggregateNode(ExecuteContext=@RigVMFunction_Sequence.A)
   Get Main Leg PV_1 = RigVMFunctionReferenceNode(Initial=true)
   n2 = &RigUnit_AimBoneMath(
    DebugSettings=(bEnabled=False,Scale=10,WorldOffset=(Rotation=(X=0,Y=0,Z=0,W=1),Translation=(X=0,Y=0,Z=0),Scale3D=(X=1,Y=1,Z=1)))
    InputTransform=(Rotation=(X=0,Y=0,Z=0,W=1),Translation=(X=0,Y=0,Z=0),Scale3D=(X=1,Y=1,Z=1))
    Primary=(Weight=1,Axis=(X=1,Y=0,Z=0),Target=(X=1,Y=0,Z=0),Kind=Location,Space=(Type=Bone,Name="None"))
    Secondary=(Weight=1,Axis=(X=0,Y=0,Z=1),Target=(X=1,Y=0,Z=0),Kind=Direction,Space=(Type=Bone,Name="None"))
    Weight=1
    Secondary.Target=@"Get Main Leg PV_1".Pole Vector
    Primary.Target=&RigUnit_GetTransform(bInitial=true, Item=RigVMVariableNode() {variable=Ball Connector}.Value).Transform.Translation
    InputTransform.Translation=&RigUnit_GetTransform(bInitial=true, Item=RigVMVariableNode() {variable=Ankle Connector}.Value).Transform.Translation
   )
   RigVMVariableNode(Value=@n2.Result, ExecuteContext=@RigVMFunction_Sequence_3.D) {variable=Initial Ankle Rig Transform}
   Get Relative Transform = RigVMFunctionReferenceNode(
    Initial=true
    Parent=RigVMVariableNode() {variable=Initial Ankle Rig Transform}.Value
    RigNode=RigVMVariableNode() {variable=Ball Connector}.Value
   )
   RigVMVariableNode(
    Value=/Script/RigVM.RigVMFunction_MathTransformMakeAbsolute(Local=@"Get Relative Transform".Relative, Parent=RigVMVariableNode() {variable=Ankle Rig Transform}.Value) {is=struct RigVMUnitNode, template="Make Absolute::Execute(in Local,in Parent,out Global)"}.Global
    ExecuteContext=@RigVMFunction_Sequence_3.E
   ) {variable=Ball Rig Transform}
   n3 = RigVMVariableNode() {variable=Ik Point}
   Get Ik Angles = RigVMFunctionReferenceNode(
    "B 180 "=false
    C 180=true
    Negate A=false
    Negate B=true
    Negate C=true
    Length C=&RigVMFunction_MathVectorDistance(
     A=&RigUnit_GetTransform(bInitial=False, Item=@n3.Value).Transform.Translation
     B=RigVMVariableNode() {variable=Knee Rig Transform}.Value.Translation
    ).Result
    Length A=RigVMVariableNode() {variable=Knee Bone Length}.Value
    Length B=RigVMVariableNode() {variable=Foot Bone Length}.Value
   )
   RigVMVariableNode(Value=@"Get Ik Angles".Angle B, ExecuteContext=@RigVMFunction_Sequence_3.A) {variable=Knee Rig Angle}
   > RigVMVariableNode(Value=RigVMDispatchNode(
     True=0
     Condition=RigVMDispatchNode(B=180, A=@"Get Ik Angles".Angle C) {template="DISPATCH_RigVMDispatch_CoreEquals(in A,in B,out Result)"}.Result
     False=@"Get Ik Angles".Angle C
    ) {template="DISPATCH_RigVMDispatch_If(in Condition,in True,in False,out Result)"}.Result) {variable=Foot Rig Angle}
   Get Relative Transform_1 = RigVMFunctionReferenceNode(
    Initial=true
    Parent=RigVMVariableNode() {variable=Initial Ankle Rig Transform}.Value
    RigNode=RigVMVariableNode() {variable=Toe Connector}.Value
   )
   RigVMVariableNode(
    Value=&RigVMFunction_MathTransformMakeAbsolute(Local=@"Get Relative Transform_1".Relative, Parent=RigVMVariableNode() {variable=Ankle Rig Transform}.Value) {template="Make Absolute::Execute(in Local,in Parent,out Global)"}.Global
    ExecuteContext=@RigVMFunction_Sequence_3.F
   ) {variable=Toe Rig Transform}
   Get Main Leg PV = RigVMFunctionReferenceNode(Initial=False)
   n4 = &RigUnit_AimBoneMath(
    DebugSettings=(bEnabled=False,Scale=10,WorldOffset=(Rotation=(X=0,Y=0,Z=0,W=1),Translation=(X=0,Y=0,Z=0),Scale3D=(X=1,Y=1,Z=1)))
    Primary=(Weight=1,Axis=(X=1,Y=0,Z=0),Target=(X=1,Y=0,Z=0),Kind=Location,Space=(Type=Bone,Name="None"))
    Secondary=(Weight=1,Axis=(X=0,Y=0,Z=1),Target=(X=1,Y=0,Z=0),Kind=Direction,Space=(Type=Bone,Name="None"))
    Weight=1
    Secondary.Target=@"Get Main Leg PV".Pole Vector
    Primary.Target=&RigUnit_GetTransform(bInitial=False, Item=RigVMVariableNode() {variable=Ik Point}.Value).Transform.Translation
    InputTransform=RigVMVariableNode() {variable=Knee Rig Transform}.Value
   )
   n5 = RigVMVariableNode() {variable=Knee Rig Angle}
   RigVMVariableNode(
    Value=&RigVMFunction_MathTransformMakeAbsolute(
     Local=(Rotation=(X=0,Y=0,Z=0,W=1),Translation=(X=0,Y=0,Z=0),Scale3D=(X=1,Y=1,Z=1))
     Parent=@n4.Result
     Local.Rotation=&RigVMFunction_MathQuaternionFromEuler(Euler=(X=0,Y=0,Z=0), Euler.Z=@n5.Value).Result
    ) {template="Make Absolute::Execute(in Local,in Parent,out Global)"}.Global
    ExecuteContext=@RigVMFunction_Sequence_3.B
   ) {variable=Knee Rig Transform}
   n6 = RigVMVariableNode() {variable=Foot Rig Angle}
   n7 = RigVMVariableNode() {variable=Knee Bone Length}
   n8 = RigVMVariableNode() {variable=Knee Rig Transform}
   RigVMVariableNode(
    Value=&RigVMFunction_MathTransformMakeAbsolute(
     Local=(Rotation=(X=0,Y=0,Z=0,W=1),Translation=(X=0,Y=0,Z=0),Scale3D=(X=1,Y=1,Z=1))
     Local.Rotation=&RigVMFunction_MathQuaternionFromEuler(Euler=(X=0,Y=0,Z=0), Euler.Z=@n6.Value).Result
     Parent=&RigVMFunction_MathTransformMakeAbsolute(Local=(Rotation=(X=0,Y=0,Z=0,W=1),Translation=(X=0,Y=0,Z=0),Scale3D=(X=1,Y=1,Z=1)), Local.Translation.X=@n7.Value, Parent=@n8.Value) {template="Make Absolute::Execute(in Local,in Parent,out Global)"}.Global
    ) {template="Make Absolute::Execute(in Local,in Parent,out Global)"}.Global
    ExecuteContext=@RigVMFunction_Sequence_3.C
   ) {variable=Ankle Rig Transform}
   RigVMFunction_DebugLineStripNoSpace = /Script/RigVM.RigVMFunction_DebugLineStripNoSpace(
    Color=(R=1,G=0,B=0,A=1)
    DebugDrawSettings=(DepthPriority=SDPG_Foreground,Lifetime=-1)
    Points=((X=0,Y=0,Z=0),(X=0,Y=0,Z=0),(X=0,Y=0,Z=0))
    Thickness=1
    WorldOffset=(Rotation=(X=0,Y=0,Z=0,W=1),Translation=(X=0,Y=0,Z=0),Scale3D=(X=1,Y=1,Z=1))
    Points.1=RigVMVariableNode() {variable=Ankle Rig Transform}.Value.Translation
    bEnabled=RigVMVariableNode() {variable=DebugOnOff}.Value
    Points.0=RigVMVariableNode() {variable=Knee Rig Transform}.Value.Translation
    Points.2=RigVMVariableNode() {variable=Ball Rig Transform}.Value.Translation
    ExecutePin=@RigVMFunction_Sequence_3.H
   ) {is=struct RigVMUnitNode}
   Get Relative Transform_2 = RigVMFunctionReferenceNode(
    Initial=true
    Parent=RigVMVariableNode() {variable=Initial Ankle Rig Transform}.Value
    RigNode=RigVMVariableNode() {variable=Toe Tip Connector}.Value
   )
   RigVMVariableNode(
    Value=&RigVMFunction_MathTransformMakeAbsolute(Local=@"Get Relative Transform_2".Relative, Parent=RigVMVariableNode() {variable=Ankle Rig Transform}.Value) {template="Make Absolute::Execute(in Local,in Parent,out Global)"}.Global
    ExecuteContext=@RigVMFunction_Sequence_3.G
   ) {variable=Toe Tip Rig Transform}
   RigVMFunctionReturnNode(ExecuteContext=@RigVMFunction_Sequence.B)
   Multiply = &RigVMFunction_MathDoubleMul(
    B=RigVMDispatchNode(False=0, True=-1, Condition=RigVMVariableNode() {variable=Right Side}.Value) {template="DISPATCH_RigVMDispatch_If(in Condition,in True,in False,out Result)"}.Result
    A=@n5.Value
   )
   RigVMVariableNode() {variable=Ik Control}
   RigVMVariableNode() {variable=Ik Control}
  ]}
 New Function_ContainedGraph {nodes [
   n1 = RigVMFunctionEntryNode()
   RigVMFunction_Sequence = &RigVMFunction_Sequence(ExecuteContext=@n1.ExecuteContext)
   RigVMFunctionReturnNode(ExecuteContext=@RigVMFunction_Sequence.B, Distance=RigVMVariableNode() {variable=Current Distance}.Value)
   n2 = RigVMVariableNode() {variable=Second Object}
   n3 = RigVMVariableNode() {variable=Initial}
   n4 = RigVMVariableNode() {variable=First Object}
   n5 = RigVMVariableNode(
    Value=&RigVMFunction_MathVectorDistance(
     B=&RigUnit_GetTransform(Item=@n2.Value, bInitial=@n3.Value).Transform.Translation
     A=&RigUnit_GetTransform(Item=@n4.Value, bInitial=@n3.Value).Transform.Translation
    ).Result
    ExecuteContext=@RigVMFunction_Sequence.A
   ) {variable=Current Distance}
   RigVMFunction_ControlFlowBranch = &RigVMFunction_ControlFlowBranch(Condition=RigVMVariableNode() {variable=Use Max Distance}.Value, ExecuteContext=@n5.ExecuteContext)
   Greater = /Script/RigVM.RigVMFunction_MathDoubleGreater(B=RigVMVariableNode() {variable=Max Distance}.Value, A=RigVMVariableNode() {variable=Current Distance}.Value) {is=struct RigVMUnitNode, template="Greater::Execute(in A,in B,out Result)"}
   RigVMVariableNode(
    Value=RigVMDispatchNode(
     Condition=@Greater.Result
     True=RigVMVariableNode() {variable=Max Distance}.Value
     False=RigVMVariableNode() {variable=Current Distance}.Value
    ) {template="DISPATCH_RigVMDispatch_If(in Condition,in True,in False,out Result)"}.Result
    ExecuteContext=@RigVMFunction_ControlFlowBranch.True
   ) {variable=Current Distance}
  ]}
 New Function_ContainedGraph {nodes [
   RigVMFunctionEntryNode()
   Subtract = /Script/RigVM.RigVMFunction_MathVectorSub(
    B=&RigUnit_GetTransform(bInitial=False, Item=RigVMVariableNode() {variable=Root Object}.Value).Transform.Translation
    A=&RigUnit_GetTransform(bInitial=False, Item=RigVMVariableNode() {variable=Tip Object}.Value).Transform.Translation
   ) {is=struct RigVMUnitNode, template="Subtract::Execute(in A,in B,out Result)"}
   > RigVMFunctionReturnNode(Vector=@Subtract.Result)
  ]}
 New Function_ContainedGraph {nodes [
   n1 = RigVMFunctionEntryNode()
   RigVMFunction_Sequence = &RigVMFunction_Sequence(ExecuteContext=@n1.ExecuteContext)
   RigVMFunction_Sequence_1 = RigVMAggregateNode(ExecuteContext=@RigVMFunction_Sequence.A)
   Get Main Leg PV_1 = RigVMFunctionReferenceNode(Initial=true)
   RigVMVariableNode(Value=@"Get Main Leg PV_1".Pole Vector, ExecuteContext=@RigVMFunction_Sequence_1.B) {variable=Pole Vector}
   RigVMFunction_MathVectorDistance = &RigVMFunction_MathVectorDistance(
    B=&RigUnit_GetTransform(bInitial=False, Item=RigVMVariableNode() {variable=Ik Control}.Value).Transform.Translation
    A=&RigUnit_GetTransform(bInitial=true, Item=RigVMVariableNode() {variable=Knee Connector}.Value).Transform.Translation
   )
   RigVMVariableNode(Value=@RigVMFunction_MathVectorDistance.Result, ExecuteContext=@RigVMFunction_Sequence_1.C) {variable=Ik Length}
   LawOfCosine = /Script/RigVM.RigVMFunction_MathDoubleLawOfCosine(
    C=RigVMVariableNode() {variable=Ik Length}.Value
    A=RigVMVariableNode() {variable=Knee Bone Length}.Value
    B=RigVMVariableNode() {variable=Foot Bone Length}.Value
   ) {is=struct RigVMUnitNode, template="LawOfCosine::Execute(in A,in B,in C,out AlphaAngle,out BetaAngle,out GammaAngle,out bValid)"}
   Negate = /Script/RigVM.RigVMFunction_MathDoubleNegate(Value=@LawOfCosine.BetaAngle) {is=struct RigVMUnitNode, template="Negate::Execute(in Value,out Result)"}
   Degrees = /Script/RigVM.RigVMFunction_MathDoubleDeg(Value=@Negate.Result) {is=struct RigVMUnitNode, template="Degrees::Execute(in Value,out Result)"}
   RigVMVariableNode(Value=@Degrees.Result, ExecuteContext=@RigVMFunction_Sequence_1.D) {variable=First Bone Angle}
   n2 = &RigVMFunction_MathDoubleNegate(Value=@LawOfCosine.GammaAngle) {template="Negate::Execute(in Value,out Result)"}
   > RigVMVariableNode(Value=&RigVMFunction_MathDoubleAdd(B=180, A=&RigVMFunction_MathDoubleDeg(Value=@n2.Result) {template="Degrees::Execute(in Value,out Result)"}.Result).Result) {variable=Second Bone Angle}
   Get Lower Leg Ik PV = RigVMFunctionReferenceNode()
   n3 = &RigUnit_AimBoneMath(
    DebugSettings=(bEnabled=False,Scale=10,WorldOffset=(Rotation=(X=0,Y=0,Z=0,W=1),Translation=(X=0,Y=0,Z=0),Scale3D=(X=1,Y=1,Z=1)))
    Primary=(Weight=1,Axis=(X=1,Y=0,Z=0),Target=(X=1,Y=0,Z=0),Kind=Location,Space=(Type=Bone,Name="None"))
    Secondary=(Weight=1,Axis=(X=0,Y=0,Z=1),Target=(X=1,Y=0,Z=0),Kind=Direction,Space=(Type=Bone,Name="None"))
    Weight=1
    Secondary.Target=@"Get Lower Leg Ik PV".Pole Vector
    InputTransform=&RigUnit_GetTransform(bInitial=true, Item=RigVMVariableNode() {variable=Knee Connector}.Value).Transform
    Primary.Target=&RigUnit_GetTransform(bInitial=true, Item=RigVMVariableNode() {variable=Ik Control}.Value).Transform.Translation
   )
   > RigVMVariableNode(Value=@n3.Result) {variable=Ik Aim Transform}
   RigVMFunction_MathQuaternionFromEuler = &RigVMFunction_MathQuaternionFromEuler(Euler=(X=0,Y=0,Z=0), Euler.Z=RigVMVariableNode() {variable=First Bone Angle}.Value)
   RigVMVariableNode(
    Value=&RigVMFunction_MathTransformMakeAbsolute(
     Local=(Rotation=(X=0,Y=0,Z=0,W=1),Translation=(X=0,Y=0,Z=0),Scale3D=(X=1,Y=1,Z=1))
     Local.Rotation=@RigVMFunction_MathQuaternionFromEuler.Result
     Parent=RigVMVariableNode() {variable=Ik Aim Transform}.Value
    ) {template="Make Absolute::Execute(in Local,in Parent,out Global)"}.Global
    ExecuteContext=@RigVMFunction_Sequence_1.E
   ) {variable=Knee Rig Transform}
   n4 = RigVMVariableNode() {variable=Knee Connector}
   > RigVMVariableNode(Value=&RigVMFunction_MathTransformMakeRelative(
     Global=&RigUnit_GetTransform(bInitial=true, Item=@n4.Value).Transform
     Parent=RigVMVariableNode() {variable=Knee Rig Transform}.Value
    ).Local) {variable=Knee Local Transform}
   n5 = RigVMVariableNode() {variable=Second Bone Angle}
   > RigVMVariableNode(Value=&RigVMFunction_MathTransformMakeAbsolute(
     Local=(Rotation=(X=0,Y=0,Z=0,W=1),Translation=(X=0,Y=0,Z=0),Scale3D=(X=1,Y=1,Z=1))
     Local.Rotation=&RigVMFunction_MathQuaternionFromEuler(Euler=(X=0,Y=0,Z=0), Euler.Z=@n5.Value).Result
     Parent=RigVMVariableNode() {variable=Knee Rig Transform}.Value
     Local.Translation.X=RigVMVariableNode() {variable=Knee Bone Length}.Value
    ) {template="Make Absolute::Execute(in Local,in Parent,out Global)"}.Global) {variable=Ankle Rig Transform}
   n6 = RigVMVariableNode() {variable=Ankle Connector}
   > RigVMVariableNode(Value=&RigVMFunction_MathTransformMakeRelative(
     Global=&RigUnit_GetTransform(bInitial=true, Item=@n6.Value).Transform
     Parent=RigVMVariableNode() {variable=Ankle Rig Transform}.Value
    ).Local) {variable=Ankle Local Transform}
   Find Up Vector = RigVMFunctionReferenceNode(
    Length=50
    First Bone=RigVMVariableNode() {variable=Knee Connector}.Value
    Second Bone=RigVMVariableNode() {variable=Ankle Connector}.Value
    Tip Bone=RigVMVariableNode() {variable=Ball Connector}.Value
    ExecuteContext=@RigVMFunction_Sequence_1.A
   )
   n7 = RigVMVariableNode() {variable=Hip Socket Connector}
   > RigVMVariableNode(Value=&RigVMFunction_MathTransformMakeRelative(
     Global=(Rotation=(X=0,Y=0,Z=0,W=1),Translation=(X=0,Y=0,Z=0),Scale3D=(X=1,Y=1,Z=1))
     Global.Translation=@"Find Up Vector".Up Vector
     Parent=&RigUnit_GetTransform(bInitial=true, Item=@n7.Value).Transform
    ).Local) {variable=Lower Ik PV Local Transform}
   > Find Up Vector_1 = RigVMFunctionReferenceNode(
    Length=-50
    Second Bone=RigVMVariableNode() {variable=Knee Connector}.Value
    Tip Bone=RigVMVariableNode() {variable=Ankle Connector}.Value
    First Bone=RigVMVariableNode() {variable=Hip Socket Connector}.Value
   )
   n8 = /Script/ControlRig.RigUnit_HierarchyGetParent(bDefaultParent=True, Child=RigVMVariableNode() {variable=Hip Socket Connector}.Value) {is=struct RigVMUnitNode}
   > RigVMVariableNode(Value=&RigVMFunction_MathTransformMakeRelative(
     Global=(Rotation=(X=0,Y=0,Z=0,W=1),Translation=(X=0,Y=0,Z=0),Scale3D=(X=1,Y=1,Z=1))
     Global.Translation=@"Find Up Vector_1".Up Vector
     Parent=&RigUnit_GetTransform(bInitial=true, Item=@n8.Parent).Transform
    ).Local) {variable=Upper Ik PV Local Transform}
   RigVMFunctionReturnNode(
    ExecuteContext=@RigVMFunction_Sequence.B
    Hip Socket Transform=RigVMVariableNode() {variable=Input Hip Socket Transform}.Value
    Knee Transform=RigVMVariableNode() {variable=Knee Rig Transform}.Value
    Ankle Transform=RigVMVariableNode() {variable=Ankle Rig Transform}.Value
   )
   Get Distance Between = RigVMFunctionReferenceNode(
    ExecuteContext=()
    Initial=true
    First Object=RigVMVariableNode() {variable=Knee Connector}.Value
    Second Object=RigVMVariableNode() {variable=Ankle Connector}.Value
   )
   > RigVMVariableNode(Value=@"Get Distance Between".Distance) {variable=First Bone Length}
   > RigVMVariableNode(Value=@"Get Distance Between".Distance) {variable=Knee Bone Length}
   > Get Distance Between_1 = RigVMFunctionReferenceNode(
    Initial=true
    First Object=RigVMVariableNode() {variable=Ankle Connector}.Value
    Second Object=RigVMVariableNode() {variable=Ball Connector}.Value
   )
   > RigVMVariableNode(Value=@"Get Distance Between_1".Distance) {variable=Second Bone Length}
   > RigVMVariableNode(Value=@"Get Distance Between_1".Distance) {variable=Foot Bone Length}
   Add = &RigVMFunction_MathDoubleAdd(A=RigVMVariableNode() {variable=Knee Bone Length}.Value, B=RigVMVariableNode() {variable=Foot Bone Length}.Value)
   Get Distance Between_2 = RigVMFunctionReferenceNode(
    ExecuteContext=()
    Initial=true
    Use Max Distance=false
    Max Distance=@Add.Result
    Second Object=RigVMVariableNode() {variable=Ik Control}.Value
    First Object=RigVMVariableNode() {variable=Knee Connector}.Value
   )
   > RigVMVariableNode(Value=@"Get Distance Between_2".Distance) {variable=Ik Length}
   n9 = /Script/ControlRig.RigUnit_HierarchyAddBone(
    ExecutePin=()
    Name=NewBone
    Parent=(Type=Bone,Name="None")
    Space=LocalSpace
    Transform=(Rotation=(X=0,Y=0,Z=0,W=1),Translation=(X=0,Y=0,Z=0),Scale3D=(X=1,Y=1,Z=1))
   ) {is=struct RigVMUnitNode}
   > &RigUnit_HierarchyAddBone(Name=NewBone, Space=LocalSpace, Transform=(Rotation=(X=0,Y=0,Z=0,W=1),Translation=(X=20,Y=0,Z=0),Scale3D=(X=1,Y=1,Z=1)), Parent=@n9.Item)
   Get Main Leg PV = RigVMFunctionReferenceNode(Initial=true)
   RigVMVariableNode() {variable=Ankle Rig Transform}
   RigVMVariableNode() {variable=Ankle Connector}
   RigVMVariableNode() {variable=Knee Rig Transform}
  ]}
 New Function_ContainedGraph {nodes [
   Create Name_2 = RigVMFunctionReferenceNode(ExecuteContext=(), Tokens=("Bk IkPoint"))
   > n1 = &RigUnit_HierarchyAddNull(
    Space=LocalSpace
    Transform=(Rotation=(X=0,Y=0,Z=0,W=1),Translation=(X=0,Y=0,Z=0),Scale3D=(X=1,Y=1,Z=1))
    Name=@"Create Name_2".Name
    Parent=RigVMVariableNode() {variable=Ik Control}.Value
   )
   > RigVMVariableNode(Value=@n1.Item) {variable=Ik Point}
   n2 = RigVMFunctionEntryNode()
   RigVMFunction_Sequence = &RigVMFunction_Sequence(ExecuteContext=@n2.ExecuteContext)
   RigVMFunction_Sequence_1 = RigVMAggregateNode(ExecuteContext=@RigVMFunction_Sequence.A)
   Create Name = RigVMFunctionReferenceNode(Tokens=("Bk Ik Null"), ExecuteContext=@RigVMFunction_Sequence_1.B)
   > n3 = &RigUnit_HierarchyAddNull(
    Space=GlobalSpace
    Transform=(Rotation=(X=0,Y=0,Z=0,W=1),Translation=(X=0,Y=0,Z=0),Scale3D=(X=1,Y=1,Z=1))
    Transform.Translation=&RigUnit_GetTransform(bInitial=true, Item=RigVMVariableNode() {variable=Ball Connector}.Value).Transform.Translation
    Parent=RigVMVariableNode() {variable=Ik Parent}.Value
    Name=@"Create Name".Name
   )
   > Set Side Color = RigVMFunctionReferenceNode()
   > Create Name_1 = RigVMFunctionReferenceNode(Tokens=("Bk Ik"))
   n4 = RigVMVariableNode() {variable=Control Size}
   n5 = RigVMVariableNode() {variable=Ik Control Transform}
   Multiply = &RigVMFunction_MathDoubleMul(A=@n4.Value, B=@n5.Value.Scale3D.X)
   > SpawnControl = &RigUnit_HierarchyAddControlTransform(
    InitialValue=(Rotation=(X=0,Y=0,Z=0,W=1),Translation=(X=0,Y=0,Z=0),Scale3D=(X=1,Y=1,Z=1))
    OffsetSpace=LocalSpace
    OffsetTransform=(Rotation=(X=0,Y=0,Z=0,W=1),Translation=(X=0,Y=0,Z=0),Scale3D=(X=1,Y=1,Z=1))
    Settings="(InitialSpace=LocalSpace,bUsePreferredRotationOrder=False,PreferredRotationOrder=YZX,Limits=(LimitTranslationX=(bMinimum=False,bMaximum=False),LimitTranslationY=(bMinimum=False,bMaximum=False),LimitTranslationZ=(bMinimum=False,bMaximum=False),LimitPitch=(b ...(truncated)"
    Name=@"Create Name_1".Name
    Settings.Shape.Transform.Scale3D.X=@Multiply.Result
    Settings.Shape.Transform.Scale3D.Y=&RigVMFunction_MathDoubleMul(A=@n4.Value, B=@n5.Value.Scale3D.Y).Result
    Settings.Shape.Transform.Scale3D.Z=&RigVMFunction_MathDoubleMul(A=@n4.Value, B=@n5.Value.Scale3D.Z).Result
    Settings.Shape.Transform.Rotation=@n5.Value.Rotation
    Settings.Shape.Transform.Translation=@n5.Value.Translation
    Parent=@n3.Item
    Settings.Shape.Color=@"Set Side Color".Color
   ) {template="SpawnControl::Execute(in InitialValue,in Settings,in OffsetTransform,in OffsetSpace,in Parent,in Name,out Item)"}
   > RigVMVariableNode(Value=@SpawnControl.Item) {variable=Ik Control}
   n6 = &RigVMFunction_Sequence(ExecuteContext=@RigVMFunction_Sequence_1.C)
   SpawnAnimationChannel = /Script/ControlRig.RigUnit_HierarchyAddAnimationChannelFloat(
    InitialValue=0
    LimitsEnabled=(Enabled=(bMinimum=false,bMaximum=false))
    MaximumValue=45
    MinimumValue=-45
    Name=Knee Offset
    Parent=RigVMVariableNode() {variable=Ik Control}.Value
    ExecutePin=@n6.A
   ) {
    is=struct RigVMUnitNode
    template="SpawnAnimationChannel::Execute(in InitialValue,in MinimumValue,in MaximumValue,in LimitsEnabled,in Parent,in Name,out Item)"
   }
   > RigVMVariableNode(Value=@SpawnAnimationChannel.Item) {variable=Knee Offset}
   > n7 = &RigUnit_HierarchyAddAnimationChannelFloat(
    InitialValue=0
    LimitsEnabled=(Enabled=(bMinimum=false,bMaximum=false))
    MaximumValue=45
    MinimumValue=-45
    Name=Hock Offset
    Parent=RigVMVariableNode() {variable=Ik Control}.Value
   ) {template="SpawnAnimationChannel::Execute(in InitialValue,in MinimumValue,in MaximumValue,in LimitsEnabled,in Parent,in Name,out Item)"}
   > RigVMVariableNode(Value=@n7.Item) {variable=Hock Offset}
   n8 = &RigUnit_HierarchyAddAnimationChannelFloat(
    InitialValue=0
    LimitsEnabled=(Enabled=(bMinimum=false,bMaximum=false))
    MaximumValue=45
    MinimumValue=-45
    Name=Knee Twist
    Parent=RigVMVariableNode() {variable=Ik Control}.Value
    ExecutePin=@n6.B
   ) {template="SpawnAnimationChannel::Execute(in InitialValue,in MinimumValue,in MaximumValue,in LimitsEnabled,in Parent,in Name,out Item)"}
   > RigVMVariableNode(Value=@n8.Item) {variable=Knee Twist}
   RigVMVariableNode(
    Value=RigVMDispatchNode(Default=(Type=Bone,Name="None"), Name=Offset Control, NameSpace=Root) {template="DISPATCH_RigDispatch_GetModuleMetadata(in Name,in NameSpace,in Default,out Value,out Found)"}.Value
    ExecuteContext=@RigVMFunction_Sequence_1.A
   ) {variable=Ik Parent}
   RigVMFunctionReturnNode(ExecuteContext=@RigVMFunction_Sequence.B)
   n9 = &RigUnit_HierarchyAddAnimationChannelFloat(
    ExecutePin=()
    InitialValue=0
    LimitsEnabled=(Enabled=(bMinimum=false,bMaximum=false))
    MaximumValue=45
    MinimumValue=-45
    Name=Ankle Twist
    Parent=RigVMVariableNode() {variable=Ik Control}.Value
   ) {template="SpawnAnimationChannel::Execute(in InitialValue,in MinimumValue,in MaximumValue,in LimitsEnabled,in Parent,in Name,out Item)"}
   > RigVMVariableNode(Value=@n9.Item) {variable=Ankle Twist}
  ]}
 New Function_ContainedGraph {nodes [
   n1 = RigVMFunctionEntryNode()
   n2 = RigVMDispatchNode(
    True=1
    False=RigVMDispatchNode(False=1, True=-1, Condition=RigVMVariableNode() {variable=Right Side}.Value) {template="DISPATCH_RigVMDispatch_If(in Condition,in True,in False,out Result)"}.Result
    Condition=RigVMVariableNode() {variable=Left Side}.Value
   ) {template="DISPATCH_RigVMDispatch_If(in Condition,in True,in False,out Result)"}
   Subtract = &RigVMFunction_MathVectorSub(
    B=&RigUnit_GetTransform(Item=RigVMVariableNode() {variable=Hip Socket Control}.Value, bInitial=@n1.Initial).Transform.Translation
    A=&RigUnit_GetTransform(Item=RigVMVariableNode() {variable=Main PV Control}.Value, bInitial=@n1.Initial).Transform.Translation
   ) {template="Subtract::Execute(in A,in B,out Result)"}
   Multiply = &RigVMFunction_MathVectorMul(B=(X=1,Y=1,Z=1), B.X=@n2.Result, B.Y=@n2.Result, B.Z=@n2.Result, A=@Subtract.Result) {template="Multiply::Execute(in A,in B,out Result)"}
   > RigVMFunctionReturnNode(Pole Vector=@Multiply.Result)
   RigVMVariableNode() {variable=Hip Socket Connector}
  ]}
 New Function_ContainedGraph {nodes [
   n1 = RigVMFunctionEntryNode()
   RigVMFunction_Sequence = &RigVMFunction_Sequence(ExecuteContext=@n1.ExecuteContext)
   RigVMFunction_Sequence_4 = RigVMAggregateNode(ExecuteContext=@RigVMFunction_Sequence.A)
   RigVMFunction_Sequence_3 = RigVMAggregateNode(ExecuteContext=@RigVMFunction_Sequence_4.C)
   Get Relative Transform_4 = RigVMFunctionReferenceNode(
    Initial=true
    RigNode=RigVMVariableNode() {variable=Hip Socket Connector}.Value
    Parent=RigVMVariableNode() {variable=Initial Hip To Ik Control Aim}.Value
   )
   RigVMVariableNode(
    Value=&RigVMFunction_MathTransformMakeAbsolute(Local=@"Get Relative Transform_4".Relative, Parent=RigVMVariableNode() {variable=Hip To Ik Control Aim}.Value) {template="Make Absolute::Execute(in Local,in Parent,out Global)"}.Global
    ExecuteContext=@RigVMFunction_Sequence_3.B
   ) {variable=Hip Socket Rig Transform}
   Get Relative Transform_8 = RigVMFunctionReferenceNode(
    Initial=true
    RigNode=RigVMVariableNode() {variable=Toe Connector}.Value
    Parent=RigVMVariableNode() {variable=Initial Hip To Ik Control Aim}.Value
   )
   RigVMVariableNode(
    Value=&RigVMFunction_MathTransformMakeAbsolute(Local=@"Get Relative Transform_8".Relative, Parent=RigVMVariableNode() {variable=Hip To Ik Control Aim}.Value) {template="Make Absolute::Execute(in Local,in Parent,out Global)"}.Global
    ExecuteContext=@RigVMFunction_Sequence_3.F
   ) {variable=Toe Rig Transform}
   Get Relative Transform_5 = RigVMFunctionReferenceNode(
    Initial=true
    RigNode=RigVMVariableNode() {variable=Knee Connector}.Value
    Parent=RigVMVariableNode() {variable=Initial Hip To Ik Control Aim}.Value
   )
   RigVMVariableNode(
    Value=&RigVMFunction_MathTransformMakeAbsolute(Local=@"Get Relative Transform_5".Relative, Parent=RigVMVariableNode() {variable=Hip To Ik Control Aim}.Value) {template="Make Absolute::Execute(in Local,in Parent,out Global)"}.Global
    ExecuteContext=@RigVMFunction_Sequence_3.C
   ) {variable=Knee Rig Transform}
   Get Relative Transform_9 = RigVMFunctionReferenceNode(
    Initial=true
    Parent=RigVMVariableNode() {variable=Initial Hip To Ik Control Aim}.Value
    RigNode=RigVMVariableNode() {variable=Toe Tip Connector}.Value
   )
   RigVMVariableNode(
    Value=&RigVMFunction_MathTransformMakeAbsolute(Local=@"Get Relative Transform_9".Relative, Parent=RigVMVariableNode() {variable=Hip To Ik Control Aim}.Value) {template="Make Absolute::Execute(in Local,in Parent,out Global)"}.Global
    ExecuteContext=@RigVMFunction_Sequence_3.G
   ) {variable=Toe Tip Rig Transform}
   Get Relative Transform_6 = RigVMFunctionReferenceNode(
    Initial=true
    RigNode=RigVMVariableNode() {variable=Ankle Connector}.Value
    Parent=RigVMVariableNode() {variable=Initial Hip To Ik Control Aim}.Value
   )
   RigVMVariableNode(
    Value=&RigVMFunction_MathTransformMakeAbsolute(Local=@"Get Relative Transform_6".Relative, Parent=RigVMVariableNode() {variable=Hip To Ik Control Aim}.Value) {template="Make Absolute::Execute(in Local,in Parent,out Global)"}.Global
    ExecuteContext=@RigVMFunction_Sequence_3.D
   ) {variable=Ankle Rig Transform}
   Get Relative Transform_7 = RigVMFunctionReferenceNode(
    Initial=true
    RigNode=RigVMVariableNode() {variable=Ball Connector}.Value
    Parent=RigVMVariableNode() {variable=Initial Hip To Ik Control Aim}.Value
   )
   RigVMVariableNode(
    Value=&RigVMFunction_MathTransformMakeAbsolute(Local=@"Get Relative Transform_7".Relative, Parent=RigVMVariableNode() {variable=Hip To Ik Control Aim}.Value) {template="Make Absolute::Execute(in Local,in Parent,out Global)"}.Global
    ExecuteContext=@RigVMFunction_Sequence_3.E
   ) {variable=Ball Rig Transform}
   RigVMFunction_DebugLineStripNoSpace = &RigVMFunction_DebugLineStripNoSpace(
    Color=(R=1,G=1,B=0.0064,A=1)
    DebugDrawSettings=(DepthPriority=SDPG_Foreground,Lifetime=-1)
    Points=((X=0,Y=0,Z=0),(X=0,Y=0,Z=0))
    Thickness=1
    WorldOffset=(Rotation=(X=0,Y=0,Z=0,W=1),Translation=(X=0,Y=0,Z=0),Scale3D=(X=1,Y=1,Z=1))
    bEnabled=RigVMVariableNode() {variable=DebugOnOff}.Value
    Points.0=RigVMVariableNode() {variable=Hip Socket Rig Transform}.Value.Translation
    Points.1=RigVMVariableNode() {variable=Knee Rig Transform}.Value.Translation
    ExecutePin=@RigVMFunction_Sequence_4.D
   )
   > RigVMFunction_DebugLineNoSpace = &RigVMFunction_DebugLineNoSpace(
    Color=(R=1,G=0,B=0,A=1)
    DebugDrawSettings=(DepthPriority=SDPG_Foreground,Lifetime=-1)
    Thickness=0
    WorldOffset=(Rotation=(X=0,Y=0,Z=0,W=1),Translation=(X=0,Y=0,Z=0),Scale3D=(X=1,Y=1,Z=1))
    B=&RigUnit_GetTransform(bInitial=false, Item=RigVMVariableNode() {variable=Ik Control}.Value).Transform.Translation
    A=&RigUnit_GetTransform(bInitial=false, Item=RigVMVariableNode() {variable=Hip Socket Connector}.Value).Transform.Translation
    bEnabled=RigVMVariableNode() {variable=DebugOnOff}.Value
   )
   Get Hip Socket Control Rotation = RigVMFunctionReferenceNode()
   Get Hip To Ik Control Aim_1 = RigVMFunctionReferenceNode()
   RigVMVariableNode(
    Value=&RigVMFunction_MathTransformMakeAbsolute(
     Local=(Rotation=(X=0,Y=0,Z=0,W=1),Translation=(X=0,Y=0,Z=0),Scale3D=(X=1,Y=1,Z=1))
     Local.Rotation=@"Get Hip Socket Control Rotation".Rotation
     Parent=@"Get Hip To Ik Control Aim_1".Aim
    ) {template="Make Absolute::Execute(in Local,in Parent,out Global)"}.Global
    ExecuteContext=@RigVMFunction_Sequence_4.B
   ) {variable=Hip To Ik Control Aim}
   Get Hip To Ik Control Aim_2 = RigVMFunctionReferenceNode(Initial=true)
   RigVMVariableNode(Value=@"Get Hip To Ik Control Aim_2".Aim, ExecuteContext=@RigVMFunction_Sequence_4.A) {variable=Initial Hip To Ik Control Aim}
   > &RigVMFunction_DebugLineNoSpace(
    Color=(R=1,G=0,B=0,A=1)
    DebugDrawSettings=(DepthPriority=SDPG_Foreground,Lifetime=-1)
    Thickness=0.1
    WorldOffset=(Rotation=(X=0,Y=0,Z=0,W=1),Translation=(X=0,Y=0,Z=0),Scale3D=(X=1,Y=1,Z=1))
    bEnabled=false
    A=&RigUnit_GetTransform(bInitial=False, Item=RigVMVariableNode() {variable=Hip Socket Control}.Value).Transform.Translation
    B=&RigUnit_GetTransform(bInitial=False, Item=RigVMVariableNode() {variable=Main PV Control}.Value).Transform.Translation
   )
   RigVMFunctionReturnNode(ExecuteContext=@RigVMFunction_Sequence.B)
  ]}
 New Function_ContainedGraph {nodes [
   n1 = RigVMFunctionEntryNode()
   Get Ik Vector = RigVMFunctionReferenceNode(Initial=@n1.Initial)
   RigVMFunction_MathVectorDistance = &RigVMFunction_MathVectorDistance(
    B=@"Get Ik Vector".Vector
    A=&RigUnit_GetTransform(Item=RigVMVariableNode() {variable=Hip Socket Connector}.Value, bInitial=@n1.Initial).Transform.Translation
   )
   > RigVMFunctionReturnNode(Distance=@RigVMFunction_MathVectorDistance.Result)
  ]}
 New Function_ContainedGraph {nodes [
   n1 = RigVMFunctionEntryNode()
   RigVMFunction_Sequence = &RigVMFunction_Sequence(ExecuteContext=@n1.ExecuteContext)
   RigVMFunctionReturnNode(ExecuteContext=@RigVMFunction_Sequence.B)
   Get Ik Vector = RigVMFunctionReferenceNode(Initial=RigVMVariableNode() {variable=Initial}.Value)
   n2 = RigVMVariableNode() {variable=Ik Control Transform}
   RigVMVariableNode(
    Value=(Rotation=(X=0,Y=0,Z=0,W=1),Translation=(X=0,Y=0,Z=0),Scale3D=(X=1,Y=1,Z=1))
    Value.Translation=@"Get Ik Vector".Vector
    Value.Scale3D=@n2.Value.Scale3D
    Value.Rotation=@n2.Value.Rotation
    ExecuteContext=@RigVMFunction_Sequence.A
   ) {variable=Ik Control Transform}
  ]}
 New Function_ContainedGraph {nodes [
   n1 = RigVMFunctionEntryNode()
   n2 = &RigUnit_GetTransform(Item=RigVMVariableNode() {variable=Hip Socket Connector}.Value, bInitial=@n1.Initial)
   Subtract = &RigVMFunction_MathVectorSub(
    B=@n2.Transform.Translation
    A=&RigUnit_GetTransform(Item=RigVMVariableNode() {variable=Ik Point}.Value, bInitial=@n1.Initial).Transform.Translation
   ) {template="Subtract::Execute(in A,in B,out Result)"}
   Add = &RigVMFunction_MathVectorAdd(A=@Subtract.Result, B=@n2.Transform.Translation) {template="Add::Execute(in A,in B,out Result)"}
   > RigVMFunctionReturnNode(Vector=@Add.Result)
   RigVMFunction_MathVectorClampLength = /Script/RigVM.RigVMFunction_MathVectorClampLength(
    Value=@Subtract.Result
    MinimumLength=RigVMVariableNode() {variable=Ik Min Distance}.Value
    MaximumLength=RigVMVariableNode() {variable=Ik Max Distance}.Value
   ) {is=struct RigVMUnitNode, template="ClampLength::Execute(in Value,in MinimumLength,in MaximumLength,out Result)"}
   RigVMVariableNode() {variable=Ik Control}
  ]}
 New Function_ContainedGraph {nodes [
   n1 = RigVMFunctionEntryNode()
   > n2 = RigVMVariableNode(Value=@n1.Tokens) {variable=Name Parts}
   RigVMFunction_Sequence = &RigVMFunction_Sequence(ExecuteContext=@n2.ExecuteContext)
   RigVMFunction_StringJoin = /Script/RigVM.RigVMFunction_StringJoin(Separator=" ", Values=RigVMVariableNode() {variable=Name Parts}.Value) {is=struct RigVMUnitNode}
   From_String = RigVMDispatchNode(String=@RigVMFunction_StringJoin.Result) {template="DISPATCH_RigDispatch_FromString(in String,out Result)"}
   RigVMFunctionReturnNode(Name=@From_String.Result, ExecuteContext=@RigVMFunction_Sequence.B)
   RigVMFunction_ControlFlowBranch = &RigVMFunction_ControlFlowBranch(Condition=RigVMVariableNode() {variable=Left Side}.Value, ExecuteContext=@RigVMFunction_Sequence.A)
   n3 = &RigVMFunction_ControlFlowBranch(ExecuteContext=@RigVMFunction_ControlFlowBranch.False, Condition=RigVMVariableNode() {variable=Right Side}.Value)
   n4 = &RigVMFunction_ControlFlowBranch(ExecuteContext=@n3.False, Condition=RigVMVariableNode() {variable=Custom Side}.Value)
   n5 = RigVMDispatchNode(
    Index=0
    Element=RigVMVariableNode() {variable=Custom Side Name}.Value
    Array=RigVMVariableNode() {variable=Name Parts}.Value
    ExecuteContext=@n4.True
   ) {template="DISPATCH_RigVMDispatch_ArrayInsert(io Array,in Index,in Element)"}
   > RigVMVariableNode(Value=@n5.Array) {variable=Name Parts}
   n6 = RigVMDispatchNode(Element=R, Index=0, Array=RigVMVariableNode() {variable=Name Parts}.Value, ExecuteContext=@n3.True) {template="DISPATCH_RigVMDispatch_ArrayInsert(io Array,in Index,in Element)"}
   > RigVMVariableNode(Value=@n6.Array) {variable=Name Parts}
   n7 = RigVMDispatchNode(Element=L, Index=0, Array=RigVMVariableNode() {variable=Name Parts}.Value, ExecuteContext=@RigVMFunction_ControlFlowBranch.True) {template="DISPATCH_RigVMDispatch_ArrayInsert(io Array,in Index,in Element)"}
   > RigVMVariableNode(Value=@n7.Array) {variable=Name Parts}
  ]}
 New Function_ContainedGraph {nodes [
   n1 = RigVMFunctionEntryNode()
   RigVMFunction_Sequence = &RigVMFunction_Sequence(ExecuteContext=@n1.ExecuteContext)
   RigVMFunctionReturnNode(ExecuteContext=@RigVMFunction_Sequence.B)
   RigVMFunction_DebugLineStripNoSpace = &RigVMFunction_DebugLineStripNoSpace(
    Color=(R=1,G=0.6443,B=0,A=1)
    DebugDrawSettings=(DepthPriority=SDPG_Foreground,Lifetime=-1)
    Thickness=2
    WorldOffset=(Rotation=(X=0,Y=0,Z=0,W=1),Translation=(X=0,Y=0,Z=0),Scale3D=(X=1,Y=1,Z=1))
    bEnabled=true
    Points=RigVMVariableNode() {variable=Debug Main Leg Ik Points}.Value
    ExecutePin=@RigVMFunction_Sequence.A
   )
  ]}
 New Function_ContainedGraph {nodes [
   n1 = RigVMFunctionEntryNode()
   RigVMFunction_Sequence = &RigVMFunction_Sequence(ExecuteContext=@n1.ExecuteContext)
   RigVMFunctionReturnNode(ExecuteContext=@RigVMFunction_Sequence.B)
   RigVMFunction_DebugLineStripNoSpace = &RigVMFunction_DebugLineStripNoSpace(
    Color=(R=0,G=0.2664,B=1,A=1)
    DebugDrawSettings=(DepthPriority=SDPG_Foreground,Lifetime=-1)
    Thickness=2
    WorldOffset=(Rotation=(X=0,Y=0,Z=0,W=1),Translation=(X=0,Y=0,Z=0),Scale3D=(X=1,Y=1,Z=1))
    bEnabled=true
    Points=RigVMVariableNode() {variable=Debug Lower Leg Ik Points}.Value
    ExecutePin=@RigVMFunction_Sequence.A
   )
   > &RigVMFunction_DebugLineStripNoSpace(
    Color=(R=0,G=0.2664,B=1,A=1)
    DebugDrawSettings=(DepthPriority=SDPG_Foreground,Lifetime=-1)
    Thickness=0
    WorldOffset=(Rotation=(X=0,Y=0,Z=0,W=1),Translation=(X=0,Y=0,Z=0),Scale3D=(X=1,Y=1,Z=1))
    bEnabled=true
    Points=RigVMVariableNode() {variable=Debug Lower Leg Ik Aim Points}.Value
   )
  ]}
 New Function_ContainedGraph {nodes [
   n1 = RigVMFunctionEntryNode()
   RigVMFunction_Sequence = &RigVMFunction_Sequence(ExecuteContext=@n1.ExecuteContext)
   RigVMFunction_Sequence_1 = RigVMAggregateNode(ExecuteContext=@RigVMFunction_Sequence.A)
   RigVMVariableNode(
    Value=&RigUnit_GetTransform(bInitial=False, Item=RigVMVariableNode() {variable=Ankle Rotate Aim Point}.Value).Transform
    ExecuteContext=@RigVMFunction_Sequence_1.B
   ) {variable=Ankle Rotate Point}
   n2 = RigVMVariableNode() {variable=Hip Socket Rig Transform}
   Subtract = &RigVMFunction_MathVectorSub(B=@n2.Value.Translation, A=RigVMVariableNode() {variable=Knee Rig Transform}.Value.Translation) {template="Subtract::Execute(in A,in B,out Result)"}
   RigVMFunction_MathVectorCross = /Script/RigVM.RigVMFunction_MathVectorCross(
    B=@Subtract.Result
    A=&RigVMFunction_MathVectorSub(B=@n2.Value.Translation, A=RigVMVariableNode() {variable=Ankle Rig Transform}.Value.Translation) {template="Subtract::Execute(in A,in B,out Result)"}.Result
   ) {is=struct RigVMUnitNode, template="Cross::Execute(in A,in B,out Result)"}
   RigVMFunction_MathVectorSetLength = /Script/RigVM.RigVMFunction_MathVectorSetLength(Length=10, Value=@RigVMFunction_MathVectorCross.Result) {is=struct RigVMUnitNode, template="SetLength::Execute(in Value,in Length,out Result)"}
   n3 = RigVMVariableNode() {variable=Hip Socket Rig Transform}
   Add = &RigVMFunction_MathVectorAdd(B=@RigVMFunction_MathVectorSetLength.Result, A=@n3.Value.Translation) {template="Add::Execute(in A,in B,out Result)"}
   &RigVMFunction_DebugLineNoSpace(
    Color=(R=0,G=1,B=0.2948,A=1)
    DebugDrawSettings=(DepthPriority=SDPG_Foreground,Lifetime=-1)
    Thickness=1
    WorldOffset=(Rotation=(X=0,Y=0,Z=0,W=1),Translation=(X=0,Y=0,Z=0),Scale3D=(X=1,Y=1,Z=1))
    bEnabled=false
    B=@Add.Result
    A=@n3.Value.Translation
    ExecutePin=@RigVMFunction_Sequence_1.D
   )
   > RigVMVariableNode(Value=@RigVMFunction_MathVectorSetLength.Result) {variable=Cross PV}
   RigUnit_GetFloatAnimationChannel = /Script/ControlRig.RigUnit_GetFloatAnimationChannel(Channel=Ankle Offset, bInitial=False, Control=RigVMVariableNode() {variable=Ik Control}.Value.Name) {is=struct RigVMUnitNode, template="GetAnimationChannel::Execute(out Value,in Control,in Channel,in bInitial)"}
   Multiply = &RigVMFunction_MathDoubleMul(B=-1, A=@RigUnit_GetFloatAnimationChannel.Value)
   Set Transform_1 = &RigUnit_SetRotation(
    Weight=1
    bInitial=False
    Value=&RigVMFunction_MathQuaternionFromEuler(Euler=(X=0,Y=0,Z=0), Euler.X=@Multiply.Result).Result
    Item=RigVMVariableNode() {variable=Hock Rotate Offset}.Value
    ExecutePin=@RigVMFunction_Sequence_1.A
   )
   n4 = &RigUnit_AimBoneMath(
    DebugSettings=(bEnabled=False,Scale=10,WorldOffset=(Rotation=(X=0,Y=0,Z=0,W=1),Translation=(X=0,Y=0,Z=0),Scale3D=(X=1,Y=1,Z=1)))
    InputTransform=(Rotation=(X=0,Y=0,Z=0,W=1),Translation=(X=0,Y=0,Z=0),Scale3D=(X=1,Y=1,Z=1))
    Primary=(Weight=1,Axis=(X=0,Y=0,Z=1),Target=(X=1,Y=0,Z=0),Kind=Location,Space=(Type=Bone,Name="None"))
    Secondary=(Weight=1,Axis=(X=1,Y=0,Z=0),Target=(X=1,Y=0,Z=0),Kind=Direction,Space=(Type=Bone,Name="None"))
    Weight=1
    InputTransform.Translation=&RigUnit_GetTransform(bInitial=False, Item=RigVMVariableNode() {variable=Ik Point}.Value).Transform.Translation
    Secondary.Space=RigVMVariableNode() {variable=Ik Control}.Value
    Primary.Target=RigVMVariableNode() {variable=Ankle Rig Transform}.Value.Translation
   )
   > &RigUnit_SetTransform(Weight=1, bInitial=False, Value=@n4.Result, Item=RigVMVariableNode() {variable=Hock Rotate Aim Null}.Value)
   Get Main Leg PV_2 = RigVMFunctionReferenceNode(Initial=False)
   n5 = &RigUnit_AimBoneMath(
    DebugSettings=(bEnabled=False,Scale=10,WorldOffset=(Rotation=(X=0,Y=0,Z=0,W=1),Translation=(X=0,Y=0,Z=0),Scale3D=(X=1,Y=1,Z=1)))
    InputTransform=(Rotation=(X=0,Y=0,Z=0,W=1),Translation=(X=0,Y=0,Z=0),Scale3D=(X=1,Y=1,Z=1))
    Primary=(Weight=1,Axis=(X=1,Y=0,Z=0),Target=(X=1,Y=0,Z=0),Kind=Location,Space=(Type=Bone,Name="None"))
    Secondary=(Weight=1,Axis=(X=0,Y=0,Z=1),Target=(X=1,Y=0,Z=0),Kind=Direction,Space=(Type=Bone,Name="None"))
    Weight=1
    Secondary.Target=@"Get Main Leg PV_2".Pole Vector
    Primary.Target=&RigUnit_GetTransform(bInitial=False, Item=RigVMVariableNode() {variable=Hock Rotate Control}.Value).Transform.Translation
    InputTransform.Translation=RigVMVariableNode() {variable=Ankle Point}.Value.Translation
   )
   RigVMVariableNode(Value=@n5.Result, ExecuteContext=@RigVMFunction_Sequence_1.H) {variable=Ankle Rig Transform}
   > RigVMVariableNode(Value=&RigVMFunction_MathTransformMakeAbsolute(
     Local=(Rotation=(X=0,Y=0,Z=0,W=1),Translation=(X=0,Y=0,Z=0),Scale3D=(X=1,Y=1,Z=1))
     Local.Translation.X=RigVMVariableNode() {variable=Foot Bone Length}.Value
     Parent=@n5.Result
    ) {template="Make Absolute::Execute(in Local,in Parent,out Global)"}.Global) {variable=Ball Rig Transform}
   n6 = &RigUnit_AimBoneMath(
    DebugSettings=(bEnabled=False,Scale=10,WorldOffset=(Rotation=(X=0,Y=0,Z=0,W=1),Translation=(X=0,Y=0,Z=0),Scale3D=(X=1,Y=1,Z=1)))
    InputTransform=(Rotation=(X=0,Y=0,Z=0,W=1),Translation=(X=0,Y=0,Z=0),Scale3D=(X=1,Y=1,Z=1))
    Primary=(Weight=1,Axis=(X=1,Y=0,Z=0),Target=(X=1,Y=0,Z=0),Kind=Location,Space=(Type=Bone,Name="None"))
    Secondary=(Weight=1,Axis=(X=0,Y=0,Z=1),Target=(X=1,Y=0,Z=0),Kind=Direction,Space=(Type=Bone,Name="None"))
    Weight=1
    Primary.Target=RigVMVariableNode() {variable=Ankle Rotate Point}.Value.Translation
    Secondary.Target=RigVMVariableNode() {variable=Cross PV}.Value
    InputTransform.Translation=RigVMVariableNode() {variable=Hip Socket Rig Transform}.Value.Translation
   )
   RigVMVariableNode(Value=@n6.Result, ExecuteContext=@RigVMFunction_Sequence_1.E) {variable=Hip Socket Aim Transform}
   Get Ik Angles = RigVMFunctionReferenceNode(
    C 180=true
    Length C=&RigVMFunction_MathVectorDistance(
     B=RigVMVariableNode() {variable=Ankle Rotate Point}.Value.Translation
     A=RigVMVariableNode() {variable=Hip Socket Rig Transform}.Value.Translation
    ).Result
    Length B=RigVMVariableNode() {variable=Knee Bone Length}.Value
    Length A=RigVMVariableNode() {variable=Hip Socket Bone Length}.Value
   )
   RigVMVariableNode(Value=@"Get Ik Angles".Angle B, ExecuteContext=@RigVMFunction_Sequence_1.C) {variable=Hip Socket Bone Angle}
   > RigVMVariableNode(Value=RigVMDispatchNode(
     True=0
     Condition=RigVMDispatchNode(B=180, A=@"Get Ik Angles".Angle C) {template="DISPATCH_RigVMDispatch_CoreEquals(in A,in B,out Result)"}.Result
     False=@"Get Ik Angles".Angle C
    ) {template="DISPATCH_RigVMDispatch_If(in Condition,in True,in False,out Result)"}.Result) {variable=Knee Bone Angle}
   n7 = RigVMVariableNode() {variable=Hip Socket Bone Angle}
   RigVMVariableNode(
    Value=&RigVMFunction_MathTransformMakeAbsolute(
     Local=(Rotation=(X=0,Y=0,Z=0,W=1),Translation=(X=0,Y=0,Z=0),Scale3D=(X=1,Y=1,Z=1))
     Local.Rotation=&RigVMFunction_MathQuaternionFromEuler(Euler=(X=0,Y=0,Z=0), Euler.Z=@n7.Value).Result
     Parent=RigVMVariableNode() {variable=Hip Socket Aim Transform}.Value
    ) {template="Make Absolute::Execute(in Local,in Parent,out Global)"}.Global
    ExecuteContext=@RigVMFunction_Sequence_1.F
   ) {variable=Hip Socket Rig Transform}
   RigVMFunction_DebugLineNoSpace = &RigVMFunction_DebugLineNoSpace(
    Color=(R=1,G=0,B=0,A=1)
    DebugDrawSettings=(DepthPriority=SDPG_Foreground,Lifetime=-1)
    Thickness=0
    WorldOffset=(Rotation=(X=0,Y=0,Z=0,W=1),Translation=(X=0,Y=0,Z=0),Scale3D=(X=1,Y=1,Z=1))
    A=RigVMVariableNode() {variable=Ankle Rig Transform}.Value.Translation
    B=RigVMVariableNode() {variable=Ball Rig Transform}.Value.Translation
    bEnabled=RigVMVariableNode() {variable=DebugOnOff}.Value
    ExecutePin=@RigVMFunction_Sequence_1.J
   )
   > RigVMFunction_DebugLineStripNoSpace = &RigVMFunction_DebugLineStripNoSpace(
    Color=(R=1,G=0,B=0,A=1)
    DebugDrawSettings=(DepthPriority=SDPG_Foreground,Lifetime=-1)
    Points=((X=0,Y=0,Z=0),(X=0,Y=0,Z=0),(X=0,Y=0,Z=0))
    Thickness=0
    WorldOffset=(Rotation=(X=0,Y=0,Z=0,W=1),Translation=(X=0,Y=0,Z=0),Scale3D=(X=1,Y=1,Z=1))
    Points.0=RigVMVariableNode() {variable=Hip Socket Rig Transform}.Value.Translation
    Points.1=RigVMVariableNode() {variable=Knee Rig Transform}.Value.Translation
    Points.2=RigVMVariableNode() {variable=Ankle Rig Transform}.Value.Translation
    bEnabled=RigVMVariableNode() {variable=DebugOnOff}.Value
   )
   n8 = RigVMVariableNode() {variable=Hip Socket Bone Length}
   n9 = RigVMVariableNode() {variable=Hip Socket Rig Transform}
   n10 = RigVMVariableNode() {variable=Knee Bone Angle}
   RigVMVariableNode(
    Value=&RigVMFunction_MathTransformMakeAbsolute(
     Local=(Rotation=(X=0,Y=0,Z=0,W=1),Translation=(X=0,Y=0,Z=0),Scale3D=(X=1,Y=1,Z=1))
     Parent=&RigVMFunction_MathTransformMakeAbsolute(Local=(Rotation=(X=0,Y=0,Z=0,W=1),Translation=(X=0,Y=0,Z=0),Scale3D=(X=1,Y=1,Z=1)), Local.Translation.X=@n8.Value, Parent=@n9.Value) {template="Make Absolute::Execute(in Local,in Parent,out Global)"}.Global
     Local.Rotation=&RigVMFunction_MathQuaternionFromEuler(Euler=(X=0,Y=0,Z=0), Euler.Z=@n10.Value).Result
    ) {template="Make Absolute::Execute(in Local,in Parent,out Global)"}.Global
    ExecuteContext=@RigVMFunction_Sequence_1.G
   ) {variable=Knee Rig Transform}
   > RigVMVariableNode(Value=&RigVMFunction_MathTransformMakeAbsolute(
     Local=(Rotation=(X=0,Y=0,Z=0,W=1),Translation=(X=0,Y=0,Z=0),Scale3D=(X=1,Y=1,Z=1))
     Parent=RigVMVariableNode() {variable=Knee Rig Transform}.Value
     Local.Translation.X=RigVMVariableNode() {variable=Knee Bone Length}.Value
    ) {template="Make Absolute::Execute(in Local,in Parent,out Global)"}.Global) {variable=Ankle Point}
   RigVMFunctionReturnNode(ExecuteContext=@RigVMFunction_Sequence.B)
   Get Main Leg PV = RigVMFunctionReferenceNode(Initial=False)
   n11 = &RigUnit_AimBoneMath(
    DebugSettings=(bEnabled=False,Scale=10,WorldOffset=(Rotation=(X=0,Y=0,Z=0,W=1),Translation=(X=0,Y=0,Z=0),Scale3D=(X=1,Y=1,Z=1)))
    InputTransform=(Rotation=(X=0,Y=0,Z=0,W=1),Translation=(X=0,Y=0,Z=0),Scale3D=(X=1,Y=1,Z=1))
    Primary=(Weight=1,Axis=(X=1,Y=0,Z=0),Target=(X=1,Y=0,Z=0),Kind=Location,Space=(Type=Bone,Name="None"))
    Secondary=(Weight=1,Axis=(X=0,Y=0,Z=1),Target=(X=1,Y=0,Z=0),Kind=Direction,Space=(Type=Bone,Name="None"))
    Weight=1
    Secondary.Target=@"Get Main Leg PV".Pole Vector
    InputTransform.Translation=RigVMVariableNode() {variable=Ball Rig Transform}.Value.Translation
    Primary.Target=RigVMVariableNode() {variable=Ankle Rig Transform}.Value.Translation
   )
   RigVMVariableNode(ExecuteContext=(), Value=@n11.Result) {variable=Ball Rotate Transform}
   RigVMFunction_MathQuaternionToEuler = &RigVMFunction_MathQuaternionToEuler(
    RotationOrder=ZYX
    Value=&RigUnit_GetTransform(Space=LocalSpace, bInitial=False, Item=RigVMVariableNode() {variable=Hock Rotate Control}.Value).Transform.Rotation
   )
   RigVMFunction_MathQuaternionFromEuler = &RigVMFunction_MathQuaternionFromEuler(Euler=(X=0,Y=0,Z=0), Euler.Z=@RigVMFunction_MathQuaternionToEuler.Result.X)
   n12 = RigVMVariableNode() {variable=Ball Rotate Transform}
   > RigVMVariableNode(Value=&RigVMFunction_MathTransformMakeAbsolute(
     Local=(Rotation=(X=0,Y=0,Z=0,W=1),Translation=(X=0,Y=0,Z=0),Scale3D=(X=1,Y=1,Z=1))
     Parent=&RigVMFunction_MathTransformMakeAbsolute(
      Local=(Rotation=(X=0,Y=0,Z=0,W=1),Translation=(X=0,Y=0,Z=0),Scale3D=(X=1,Y=1,Z=1))
      Local.Rotation=@RigVMFunction_MathQuaternionFromEuler.Result
      Parent=@n12.Value
     ) {template="Make Absolute::Execute(in Local,in Parent,out Global)"}.Global
     Local.Translation.X=RigVMVariableNode() {variable=Foot Bone Length}.Value
    ) {template="Make Absolute::Execute(in Local,in Parent,out Global)"}.Global) {variable=Ankle Rotate Point}
   RigVMVariableNode(ExecuteContext=(), Value=((X=0,Y=0,Z=0),(X=0,Y=0,Z=0))) {variable=Debug Upper Leg Ik Aim Points}
  ]}
 New Function_ContainedGraph {nodes [
   n1 = RigVMFunctionEntryNode()
   RigVMFunction_Sequence = &RigVMFunction_Sequence(ExecuteContext=@n1.ExecuteContext)
   RigVMFunctionReturnNode(ExecuteContext=@RigVMFunction_Sequence.B, Up Vector=RigVMVariableNode() {variable=PV}.Value)
   Subtract = &RigVMFunction_MathVectorSub(
    B=&RigUnit_GetTransform(bInitial=False, Item=RigVMVariableNode() {variable=First Bone}.Value).Transform.Translation
    A=&RigUnit_GetTransform(bInitial=False, Item=RigVMVariableNode() {variable=Second Bone}.Value).Transform.Translation
   ) {template="Subtract::Execute(in A,in B,out Result)"}
   n2 = RigVMVariableNode() {variable=Tip Bone}
   n3 = RigVMVariableNode() {variable=First Bone}
   RigVMFunction_MathVectorCross = &RigVMFunction_MathVectorCross(
    A=@Subtract.Result
    B=&RigVMFunction_MathVectorSub(
     A=&RigUnit_GetTransform(bInitial=False, Item=@n2.Value).Transform.Translation
     B=&RigUnit_GetTransform(bInitial=False, Item=@n3.Value).Transform.Translation
    ) {template="Subtract::Execute(in A,in B,out Result)"}.Result
   ) {template="Cross::Execute(in A,in B,out Result)"}
   RigVMFunction_MathVectorSetLength = &RigVMFunction_MathVectorSetLength(Value=@RigVMFunction_MathVectorCross.Result, Length=RigVMVariableNode() {variable=Length}.Value) {template="SetLength::Execute(in Value,in Length,out Result)"}
   Add = &RigVMFunction_MathVectorAdd(
    B=&RigUnit_GetTransform(bInitial=False, Item=RigVMVariableNode() {variable=First Bone}.Value).Transform.Translation
    A=@RigVMFunction_MathVectorSetLength.Result
   ) {template="Add::Execute(in A,in B,out Result)"}
   RigVMVariableNode(Value=@Add.Result, ExecuteContext=@RigVMFunction_Sequence.A) {variable=PV}
  ]}
 New Function_ContainedGraph {nodes [
   n1 = RigVMFunctionEntryNode()
   RigVMFunction_Sequence = &RigVMFunction_Sequence(ExecuteContext=@n1.ExecuteContext)
   RigVMFunctionReturnNode(ExecuteContext=@RigVMFunction_Sequence.B)
   RigVMFunction_DebugLineStripNoSpace = &RigVMFunction_DebugLineStripNoSpace(
    Color=(R=1,G=0,B=0,A=1)
    DebugDrawSettings=(DepthPriority=SDPG_Foreground,Lifetime=-1)
    Thickness=0
    WorldOffset=(Rotation=(X=0,Y=0,Z=0,W=1),Translation=(X=0,Y=0,Z=0),Scale3D=(X=1,Y=1,Z=1))
    bEnabled=true
    Points=RigVMVariableNode() {variable=Debug Lower Ik PV Points}.Value
    ExecutePin=@RigVMFunction_Sequence.A
   )
  ]}
 New Function_ContainedGraph {nodes [
   RigVMFunctionEntryNode()
   n1 = RigVMVariableNode() {variable=Hip Socket Connector}
   n2 = RigVMVariableNode() {variable=Initial}
   Subtract = &RigVMFunction_MathVectorSub(
    B=&RigUnit_GetTransform(Item=RigVMVariableNode() {variable=Knee Connector}.Value, bInitial=RigVMVariableNode() {variable=Initial}.Value).Transform.Translation
    A=&RigVMFunction_MathTransformMakeAbsolute(
     Parent=&RigUnit_GetTransform(Item=@n1.Value, bInitial=@n2.Value).Transform
     Local=RigVMVariableNode() {variable=Lower Ik PV Local Transform}.Value
    ) {template="Make Absolute::Execute(in Local,in Parent,out Global)"}.Global.Translation
   ) {template="Subtract::Execute(in A,in B,out Result)"}
   > RigVMFunctionReturnNode(Pole Vector=@Subtract.Result)
  ]}
 New Function_ContainedGraph {nodes [
   RigVMFunctionEntryNode()
   n1 = RigVMVariableNode() {variable=Hip Socket Connector}
   n2 = &RigUnit_HierarchyGetParent(bDefaultParent=True, Child=RigVMVariableNode() {variable=Hip Socket Connector}.Value)
   n3 = RigVMVariableNode() {variable=Initial}
   Subtract = &RigVMFunction_MathVectorSub(
    B=&RigUnit_GetTransform(
     Item=&RigUnit_HierarchyGetParent(bDefaultParent=True, Child=@n1.Value).Parent
     bInitial=RigVMVariableNode() {variable=Initial}.Value
    ).Transform.Translation
    A=&RigVMFunction_MathTransformMakeAbsolute(
     Parent=&RigUnit_GetTransform(Item=@n2.Parent, bInitial=@n3.Value).Transform
     Local=RigVMVariableNode() {variable=Upper Ik PV Local Transform}.Value
    ) {template="Make Absolute::Execute(in Local,in Parent,out Global)"}.Global.Translation
   ) {template="Subtract::Execute(in A,in B,out Result)"}
   > RigVMFunctionReturnNode(Vector=@Subtract.Result)
   Find Up Vector From Transforms = RigVMFunctionReferenceNode(
    Mid Transform=(Rotation=(X=0,Y=0,Z=0,W=1),Translation=(X=0,Y=0,Z=0),Scale3D=(X=1,Y=1,Z=1))
    Root Transform=(Rotation=(X=0,Y=0,Z=0,W=1),Translation=(X=0,Y=0,Z=0),Scale3D=(X=1,Y=1,Z=1))
    Tip Transform=(Rotation=(X=0,Y=0,Z=0,W=1),Translation=(X=0,Y=0,Z=0),Scale3D=(X=1,Y=1,Z=1))
   )
  ]}
 New Function_ContainedGraph {nodes [
   n1 = RigVMFunctionEntryNode()
   RigVMFunction_Sequence = &RigVMFunction_Sequence(ExecuteContext=@n1.ExecuteContext)
   n2 = &RigVMFunction_Sequence(ExecuteContext=@RigVMFunction_Sequence.A)
   RigVMFunction_DebugLineStripNoSpace = &RigVMFunction_DebugLineStripNoSpace(
    Color=(R=1,G=1,B=0,A=1)
    DebugDrawSettings=(DepthPriority=SDPG_Foreground,Lifetime=-1)
    Thickness=2.3
    WorldOffset=(Rotation=(X=0,Y=0,Z=0,W=1),Translation=(X=0,Y=0,Z=0),Scale3D=(X=1,Y=1,Z=1))
    bEnabled=true
    Points=RigVMVariableNode() {variable=Debug Upper Leg Ik Foot Points}.Value
    ExecutePin=@n2.A
   )
   > &RigVMFunction_DebugLineStripNoSpace(
    Color=(R=0,G=0.7031,B=0,A=1)
    DebugDrawSettings=(DepthPriority=SDPG_Foreground,Lifetime=-1)
    Thickness=0
    WorldOffset=(Rotation=(X=0,Y=0,Z=0,W=1),Translation=(X=0,Y=0,Z=0),Scale3D=(X=1,Y=1,Z=1))
    bEnabled=true
    Points=RigVMVariableNode() {variable=Debug Upper Leg Ik Aim Points}.Value
   )
   > &RigVMFunction_DebugLineStripNoSpace(
    Color=(R=0,G=0.7031,B=0,A=1)
    DebugDrawSettings=(DepthPriority=SDPG_Foreground,Lifetime=-1)
    Thickness=1
    WorldOffset=(Rotation=(X=0,Y=0,Z=0,W=1),Translation=(X=0,Y=0,Z=0),Scale3D=(X=1,Y=1,Z=1))
    bEnabled=true
    Points=RigVMVariableNode() {variable=Debug Upper Leg Ik Points}.Value
   )
   &RigVMFunction_DebugLineStripNoSpace(
    Color=(R=1,G=0,B=0,A=1)
    DebugDrawSettings=(DepthPriority=SDPG_Foreground,Lifetime=-1)
    Thickness=0
    WorldOffset=(Rotation=(X=0,Y=0,Z=0,W=1),Translation=(X=0,Y=0,Z=0),Scale3D=(X=1,Y=1,Z=1))
    bEnabled=true
    Points=RigVMVariableNode() {variable=Debug Upper Ik PV Points}.Value
    ExecutePin=@n2.B
   )
   RigVMFunctionReturnNode(ExecuteContext=@RigVMFunction_Sequence.B)
  ]}
 New Function_ContainedGraph {nodes [
   n1 = RigVMFunctionEntryNode()
   RigVMFunction_Sequence = &RigVMFunction_Sequence(ExecuteContext=@n1.ExecuteContext)
   RigVMFunction_Sequence_1 = RigVMAggregateNode(ExecuteContext=@RigVMFunction_Sequence.A)
   &RigUnit_SetTransform(
    Weight=1
    bInitial=False
    Value=&RigVMFunction_MathTransformMakeAbsolute(Parent=RigVMVariableNode() {variable=Hip Socket Rig Transform}.Value, Local=RigVMVariableNode() {variable=Hip Socket Local Transform}.Value) {template="Make Absolute::Execute(in Local,in Parent,out Global)"}.Global
    Item=RigVMVariableNode() {variable=Hip Socket Connector}.Value
    ExecutePin=@RigVMFunction_Sequence_1.A
   )
   n2 = RigVMVariableNode() {variable=Ik Fk Bones}
   > &RigUnit_SetTransform(
    Weight=1
    bInitial=False
    Item=RigVMDispatchNode(
     Index=0
     Array=RigVMDispatchNode(Index=0, Array=@n2.Value) {template="DISPATCH_RigVMDispatch_ArrayGetAtIndex(in Array,in Index,out Element)"}.Element.Keys
    ) {template="DISPATCH_RigVMDispatch_ArrayGetAtIndex(in Array,in Index,out Element)"}.Element
    Value=RigVMVariableNode() {variable=Hip Socket Rig Transform}.Value
   )
   &RigUnit_SetTransform(
    Weight=1
    bInitial=False
    Value=&RigVMFunction_MathTransformMakeAbsolute(Parent=RigVMVariableNode() {variable=Knee Rig Transform}.Value, Local=RigVMVariableNode() {variable=Knee Local Transform}.Value) {template="Make Absolute::Execute(in Local,in Parent,out Global)"}.Global
    Item=RigVMVariableNode() {variable=Knee Connector}.Value
    ExecutePin=@RigVMFunction_Sequence_1.B
   )
   n3 = RigVMVariableNode() {variable=Ik Fk Bones}
   > &RigUnit_SetTransform(
    Weight=1
    bInitial=False
    Item=RigVMDispatchNode(
     Index=0
     Array=RigVMDispatchNode(Index=1, Array=@n3.Value) {template="DISPATCH_RigVMDispatch_ArrayGetAtIndex(in Array,in Index,out Element)"}.Element.Keys
    ) {template="DISPATCH_RigVMDispatch_ArrayGetAtIndex(in Array,in Index,out Element)"}.Element
    Value=RigVMVariableNode() {variable=Knee Rig Transform}.Value
   )
   &RigUnit_SetTransform(
    Weight=1
    bInitial=False
    Value=&RigVMFunction_MathTransformMakeAbsolute(Parent=RigVMVariableNode() {variable=Ankle Rig Transform}.Value, Local=RigVMVariableNode() {variable=Ankle Local Transform}.Value) {template="Make Absolute::Execute(in Local,in Parent,out Global)"}.Global
    Item=RigVMVariableNode() {variable=Ankle Connector}.Value
    ExecutePin=@RigVMFunction_Sequence_1.C
   )
   n4 = RigVMVariableNode() {variable=Ik Fk Bones}
   > &RigUnit_SetTransform(
    Weight=1
    bInitial=False
    Item=RigVMDispatchNode(
     Index=0
     Array=RigVMDispatchNode(Index=2, Array=@n4.Value) {template="DISPATCH_RigVMDispatch_ArrayGetAtIndex(in Array,in Index,out Element)"}.Element.Keys
    ) {template="DISPATCH_RigVMDispatch_ArrayGetAtIndex(in Array,in Index,out Element)"}.Element
    Value=RigVMVariableNode() {variable=Ankle Rig Transform}.Value
   )
   n5 = RigVMVariableNode() {variable=Toe Fk Control}
   &RigUnit_SetTransform(
    Space=LocalSpace
    Weight=1
    bInitial=False
    Value=&RigUnit_GetTransform(Space=LocalSpace, bInitial=true, Item=&RigUnit_HierarchyGetParent(bDefaultParent=True, Child=@n5.Value).Parent).Transform
    Item=&RigUnit_HierarchyGetParent(bDefaultParent=True, Child=RigVMVariableNode() {variable=Toe Fk Control}.Value).Parent
    ExecutePin=@RigVMFunction_Sequence_1.E
   )
   n6 = RigVMVariableNode() {variable=Ball Pivot Control}
   > &RigUnit_SetTransform(
    Space=LocalSpace
    Weight=1
    bInitial=False
    Value=&RigUnit_GetTransform(Space=LocalSpace, bInitial=true, Item=&RigUnit_HierarchyGetParent(bDefaultParent=True, Child=@n6.Value).Parent).Transform
    Item=&RigUnit_HierarchyGetParent(bDefaultParent=True, Child=RigVMVariableNode() {variable=Ball Pivot Control}.Value).Parent
   )
   n7 = RigVMVariableNode() {variable=Ball Fk Control}
   > &RigUnit_SetTransform(
    Space=LocalSpace
    Weight=1
    bInitial=False
    Value=&RigUnit_GetTransform(Space=LocalSpace, bInitial=true, Item=&RigUnit_HierarchyGetParent(bDefaultParent=True, Child=@n7.Value).Parent).Transform
    Item=&RigUnit_HierarchyGetParent(bDefaultParent=True, Child=RigVMVariableNode() {variable=Ball Fk Control}.Value).Parent
   )
   &RigUnit_SetTransform(
    Weight=1
    bInitial=False
    Value=&RigVMFunction_MathTransformMakeAbsolute(Local=RigVMVariableNode() {variable=Ball Local Tranform}.Value, Parent=RigVMVariableNode() {variable=Ball Rig Transform}.Value) {template="Make Absolute::Execute(in Local,in Parent,out Global)"}.Global
    Item=RigVMVariableNode() {variable=Ball Connector}.Value
    ExecutePin=@RigVMFunction_Sequence_1.D
   )
   n8 = RigVMVariableNode() {variable=Ik Fk Bones}
   > &RigUnit_SetTransform(
    Weight=1
    bInitial=False
    Item=RigVMDispatchNode(
     Index=0
     Array=RigVMDispatchNode(Index=3, Array=@n8.Value) {template="DISPATCH_RigVMDispatch_ArrayGetAtIndex(in Array,in Index,out Element)"}.Element.Keys
    ) {template="DISPATCH_RigVMDispatch_ArrayGetAtIndex(in Array,in Index,out Element)"}.Element
    Value=RigVMVariableNode() {variable=Ball Rig Transform}.Value
   )
   &RigUnit_SetTransform(
    Weight=1
    bInitial=False
    Value=&RigVMFunction_MathTransformMakeAbsolute(Parent=RigVMVariableNode() {variable=Toe Rig Transform}.Value, Local=RigVMVariableNode() {variable=Toe Local Transform}.Value) {template="Make Absolute::Execute(in Local,in Parent,out Global)"}.Global
    Item=RigVMVariableNode() {variable=Toe Connector}.Value
    ExecutePin=@RigVMFunction_Sequence_1.F
   )
   n9 = RigVMVariableNode() {variable=Ik Fk Bones}
   > &RigUnit_SetTransform(
    Weight=1
    bInitial=False
    Item=RigVMDispatchNode(
     Index=0
     Array=RigVMDispatchNode(Index=4, Array=@n9.Value) {template="DISPATCH_RigVMDispatch_ArrayGetAtIndex(in Array,in Index,out Element)"}.Element.Keys
    ) {template="DISPATCH_RigVMDispatch_ArrayGetAtIndex(in Array,in Index,out Element)"}.Element
    Value=RigVMVariableNode() {variable=Toe Rig Transform}.Value
   )
   RigVMVariableNode(
    Value=&RigUnit_ResolveConnector(Connector=(Type=Connector,Name="IK Rear Connector"), SkipSocket=False).Result
    ExecuteContext=@RigVMFunction_Sequence_1.G
   ) {variable=IK Front Connector}
   > &RigUnit_ParentConstraint(
    AdvancedSettings=(InterpolationType=Average,RotationOrderForFilter=XZY)
    Parents=((Item=(Type=Connector,Name="Ball"),Weight=1))
    Weight=1
    Child=RigVMVariableNode() {variable=IK Front Connector}.Value
   )
   RigVMFunctionReturnNode(ExecuteContext=@RigVMFunction_Sequence.B)
  ]}
 New Function_ContainedGraph {nodes [
   n1 = RigVMFunctionEntryNode()
   RigVMFunction_Sequence = &RigVMFunction_Sequence(ExecuteContext=@n1.ExecuteContext)
   RigVMFunction_Sequence_1 = RigVMAggregateNode(ExecuteContext=@RigVMFunction_Sequence.A)
   Get Main Leg PV = RigVMFunctionReferenceNode(Initial=true)
   n2 = &RigUnit_AimBoneMath(
    DebugSettings=(bEnabled=False,Scale=10,WorldOffset=(Rotation=(X=0,Y=0,Z=0,W=1),Translation=(X=0,Y=0,Z=0),Scale3D=(X=1,Y=1,Z=1)))
    Primary=(Weight=1,Axis=(X=1,Y=0,Z=0),Target=(X=1,Y=0,Z=0),Kind=Location,Space=(Type=Bone,Name="None"))
    Secondary=(Weight=1,Axis=(X=0,Y=0,Z=1),Target=(X=1,Y=0,Z=0),Kind=Direction,Space=(Type=Bone,Name="None"))
    Weight=1
    Secondary.Target=@"Get Main Leg PV".Pole Vector
    InputTransform=&RigUnit_GetTransform(bInitial=true, Item=RigVMVariableNode() {variable=Hip Socket Connector}.Value).Transform
    Primary.Target=&RigUnit_GetTransform(bInitial=true, Item=RigVMVariableNode() {variable=Knee Connector}.Value).Transform.Translation
   )
   n3 = RigVMVariableNode() {variable=Hip Socket Connector}
   RigVMVariableNode(
    Value=&RigVMFunction_MathTransformMakeRelative(Parent=@n2.Result, Global=&RigUnit_GetTransform(bInitial=true, Item=@n3.Value).Transform).Local
    ExecuteContext=@RigVMFunction_Sequence_1.A
   ) {variable=Hip Socket Local Transform}
   Get Main Leg PV_4 = RigVMFunctionReferenceNode(Initial=true)
   n4 = &RigUnit_AimBoneMath(
    DebugSettings=(bEnabled=False,Scale=10,WorldOffset=(Rotation=(X=0,Y=0,Z=0,W=1),Translation=(X=0,Y=0,Z=0),Scale3D=(X=1,Y=1,Z=1)))
    InputTransform=(Rotation=(X=0,Y=0,Z=0,W=1),Translation=(X=0,Y=0,Z=0),Scale3D=(X=1,Y=1,Z=1))
    Primary=(Weight=1,Axis=(X=1,Y=0,Z=0),Target=(X=1,Y=0,Z=0),Kind=Location,Space=(Type=Bone,Name="None"))
    Secondary=(Weight=1,Axis=(X=0,Y=0,Z=1),Target=(X=1,Y=0,Z=0),Kind=Direction,Space=(Type=Bone,Name="None"))
    Weight=1
    Secondary.Target=@"Get Main Leg PV_4".Pole Vector
    InputTransform.Translation=&RigUnit_GetTransform(bInitial=true, Item=RigVMVariableNode() {variable=Toe Connector}.Value).Transform.Translation
    Primary.Target=&RigUnit_GetTransform(bInitial=true, Item=RigVMVariableNode() {variable=Toe Tip Connector}.Value).Transform.Translation
   )
   RigVMVariableNode(Value=@n4.Result, ExecuteContext=@RigVMFunction_Sequence_1.E) {variable=Toe Rig Transform}
   n5 = RigVMVariableNode() {variable=Toe Connector}
   > RigVMVariableNode(Value=&RigVMFunction_MathTransformMakeRelative(
     Global=&RigUnit_GetTransform(bInitial=true, Item=@n5.Value).Transform
     Parent=RigVMVariableNode() {variable=Toe Rig Transform}.Value
    ).Local) {variable=Toe Local Transform}
   Get Main Leg PV_1 = RigVMFunctionReferenceNode(Initial=true)
   n6 = &RigUnit_AimBoneMath(
    DebugSettings=(bEnabled=False,Scale=10,WorldOffset=(Rotation=(X=0,Y=0,Z=0,W=1),Translation=(X=0,Y=0,Z=0),Scale3D=(X=1,Y=1,Z=1)))
    Primary=(Weight=1,Axis=(X=1,Y=0,Z=0),Target=(X=1,Y=0,Z=0),Kind=Location,Space=(Type=Bone,Name="None"))
    Secondary=(Weight=1,Axis=(X=0,Y=0,Z=1),Target=(X=1,Y=0,Z=0),Kind=Direction,Space=(Type=Bone,Name="None"))
    Weight=1
    Secondary.Target=@"Get Main Leg PV_1".Pole Vector
    InputTransform=&RigUnit_GetTransform(bInitial=true, Item=RigVMVariableNode() {variable=Knee Connector}.Value).Transform
    Primary.Target=&RigUnit_GetTransform(bInitial=true, Item=RigVMVariableNode() {variable=Ankle Connector}.Value).Transform.Translation
   )
   n7 = RigVMVariableNode() {variable=Knee Connector}
   RigVMVariableNode(
    Value=&RigVMFunction_MathTransformMakeRelative(Parent=@n6.Result, Global=&RigUnit_GetTransform(bInitial=true, Item=@n7.Value).Transform).Local
    ExecuteContext=@RigVMFunction_Sequence_1.B
   ) {variable=Knee Local Transform}
   Get Main Leg PV_2 = RigVMFunctionReferenceNode(Initial=true)
   n8 = &RigUnit_AimBoneMath(
    DebugSettings=(bEnabled=False,Scale=10,WorldOffset=(Rotation=(X=0,Y=0,Z=0,W=1),Translation=(X=0,Y=0,Z=0),Scale3D=(X=1,Y=1,Z=1)))
    Primary=(Weight=1,Axis=(X=1,Y=0,Z=0),Target=(X=1,Y=0,Z=0),Kind=Location,Space=(Type=Bone,Name="None"))
    Secondary=(Weight=1,Axis=(X=0,Y=0,Z=1),Target=(X=1,Y=0,Z=0),Kind=Direction,Space=(Type=Bone,Name="None"))
    Weight=1
    Secondary.Target=@"Get Main Leg PV_2".Pole Vector
    Primary.Target=&RigUnit_GetTransform(bInitial=true, Item=RigVMVariableNode() {variable=Ball Connector}.Value).Transform.Translation
    InputTransform=&RigUnit_GetTransform(bInitial=true, Item=RigVMVariableNode() {variable=Ankle Connector}.Value).Transform
   )
   n9 = RigVMVariableNode() {variable=Ankle Connector}
   RigVMVariableNode(
    Value=&RigVMFunction_MathTransformMakeRelative(Parent=@n8.Result, Global=&RigUnit_GetTransform(bInitial=true, Item=@n9.Value).Transform).Local
    ExecuteContext=@RigVMFunction_Sequence_1.C
   ) {variable=Ankle Local Transform}
   n10 = RigVMVariableNode() {variable=Ball Connector}
   n11 = RigVMVariableNode() {variable=Foot Bone Length}
   > RigVMVariableNode(Value=&RigVMFunction_MathTransformMakeRelative(
     Global=&RigUnit_GetTransform(bInitial=true, Item=@n10.Value).Transform
     Parent=&RigVMFunction_MathTransformMakeAbsolute(Local=(Rotation=(X=0,Y=0,Z=0,W=1),Translation=(X=0,Y=0,Z=0),Scale3D=(X=1,Y=1,Z=1)), Parent=@n8.Result, Local.Translation.X=@n11.Value) {template="Make Absolute::Execute(in Local,in Parent,out Global)"}.Global
    ).Local) {variable=Ball Local Tranform}
   Get Main Leg PV_3 = RigVMFunctionReferenceNode(Initial=true)
   n12 = &RigUnit_AimBoneMath(
    DebugSettings=(bEnabled=False,Scale=10,WorldOffset=(Rotation=(X=0,Y=0,Z=0,W=1),Translation=(X=0,Y=0,Z=0),Scale3D=(X=1,Y=1,Z=1)))
    Primary=(Weight=1,Axis=(X=1,Y=0,Z=0),Target=(X=1,Y=0,Z=0),Kind=Location,Space=(Type=Bone,Name="None"))
    Secondary=(Weight=1,Axis=(X=0,Y=0,Z=1),Target=(X=1,Y=0,Z=0),Kind=Direction,Space=(Type=Bone,Name="None"))
    Weight=1
    Secondary.Target=@"Get Main Leg PV_3".Pole Vector
    InputTransform=&RigUnit_GetTransform(bInitial=true, Item=RigVMVariableNode() {variable=Ball Connector}.Value).Transform
    Primary.Target=&RigUnit_GetTransform(bInitial=true, Item=RigVMVariableNode() {variable=Toe Connector}.Value).Transform.Translation
   )
   RigVMVariableNode(Value=@n12.Result, ExecuteContext=@RigVMFunction_Sequence_1.D) {variable=Ball Rig Transform}
   n13 = RigVMVariableNode() {variable=Ball Connector}
   > RigVMVariableNode(Value=&RigVMFunction_MathTransformMakeRelative(
     Global=&RigUnit_GetTransform(bInitial=true, Item=@n13.Value).Transform
     Parent=RigVMVariableNode() {variable=Ball Rig Transform}.Value
    ).Local) {variable=Ball Local Tranform}
   RigVMFunctionReturnNode(ExecuteContext=@RigVMFunction_Sequence.B)
  ]}
 New Function_ContainedGraph {nodes [
   n1 = RigVMFunctionEntryNode()
   RigVMFunction_Sequence = &RigVMFunction_Sequence(ExecuteContext=@n1.ExecuteContext)
   RigVMFunction_Sequence_1 = RigVMAggregateNode(ExecuteContext=@RigVMFunction_Sequence.A)
   Get Lower Leg Ik PV = RigVMFunctionReferenceNode(Initial=true)
   n2 = &RigUnit_AimBoneMath(
    DebugSettings=(bEnabled=False,Scale=10,WorldOffset=(Rotation=(X=0,Y=0,Z=0,W=1),Translation=(X=0,Y=0,Z=0),Scale3D=(X=1,Y=1,Z=1)))
    Primary=(Weight=1,Axis=(X=1,Y=0,Z=0),Target=(X=1,Y=0,Z=0),Kind=Location,Space=(Type=Bone,Name="None"))
    Secondary=(Weight=1,Axis=(X=0,Y=0,Z=1),Target=(X=1,Y=0,Z=0),Kind=Direction,Space=(Type=Bone,Name="None"))
    Weight=1
    Secondary.Target=@"Get Lower Leg Ik PV".Pole Vector
    InputTransform=&RigUnit_GetTransform(bInitial=true, Item=RigVMVariableNode() {variable=Ball Connector}.Value).Transform
    Primary.Target=RigVMVariableNode() {variable=Input Ankle Transform}.Value.Translation
   )
   RigVMVariableNode(Value=@n2.Result, ExecuteContext=@RigVMFunction_Sequence_1.A) {variable=Ankle_Rig_Transform}
   > RigVMVariableNode(Value=&RigVMFunction_MathTransformMakeAbsolute(
     Local=(Rotation=(X=0,Y=0,Z=0,W=1),Translation=(X=0,Y=0,Z=0),Scale3D=(X=1,Y=1,Z=1))
     Local.Translation.X=RigVMVariableNode() {variable=Foot Bone Length}.Value
     Parent=RigVMVariableNode() {variable=Ankle_Rig_Transform}.Value
    ) {template="Make Absolute::Execute(in Local,in Parent,out Global)"}.Global) {variable=Ankle_Tip_Rig_Transform}
   RigVMFunction_MathVectorDistance = &RigVMFunction_MathVectorDistance(
    A=RigVMVariableNode() {variable=Input Hip Socket Transform}.Value.Translation
    B=RigVMVariableNode() {variable=Ankle_Tip_Rig_Transform}.Value.Translation
   )
   RigVMVariableNode(Value=@RigVMFunction_MathVectorDistance.Result, ExecuteContext=@RigVMFunction_Sequence_1.B) {variable=Ik Distance}
   LawOfCosine = &RigVMFunction_MathDoubleLawOfCosine(
    C=RigVMVariableNode() {variable=Ik Distance}.Value
    A=RigVMVariableNode() {variable=Hip Socket Bone Length}.Value
    B=RigVMVariableNode() {variable=Knee Bone Length}.Value
   ) {template="LawOfCosine::Execute(in A,in B,in C,out AlphaAngle,out BetaAngle,out GammaAngle,out bValid)"}
   Degrees = &RigVMFunction_MathDoubleDeg(Value=@LawOfCosine.BetaAngle) {template="Degrees::Execute(in Value,out Result)"}
   RigVMVariableNode(Value=@Degrees.Result, ExecuteContext=@RigVMFunction_Sequence_1.C) {variable=First Angle}
   > RigVMVariableNode(Value=&RigVMFunction_MathDoubleAdd(B=180, A=&RigVMFunction_MathDoubleDeg(Value=@LawOfCosine.GammaAngle) {template="Degrees::Execute(in Value,out Result)"}.Result).Result) {variable=Second Angle}
   Get Upper Leg Ik PV = RigVMFunctionReferenceNode(Initial=true)
   n3 = &RigUnit_AimBoneMath(
    DebugSettings=(bEnabled=False,Scale=10,WorldOffset=(Rotation=(X=0,Y=0,Z=0,W=1),Translation=(X=0,Y=0,Z=0),Scale3D=(X=1,Y=1,Z=1)))
    Primary=(Weight=1,Axis=(X=1,Y=0,Z=0),Target=(X=1,Y=0,Z=0),Kind=Location,Space=(Type=Bone,Name="None"))
    Secondary=(Weight=1,Axis=(X=0,Y=0,Z=1),Target=(X=1,Y=0,Z=0),Kind=Direction,Space=(Type=Bone,Name="None"))
    Weight=1
    Secondary.Target=@"Get Upper Leg Ik PV".Vector
    InputTransform=RigVMVariableNode() {variable=Input Hip Socket Transform}.Value
    Primary.Target=RigVMVariableNode() {variable=Ankle_Tip_Rig_Transform}.Value.Translation
   )
   > RigVMVariableNode(Value=@n3.Result) {variable=Ik Aim Transform}
   n4 = RigVMVariableNode() {variable=First Angle}
   RigVMVariableNode(
    Value=&RigVMFunction_MathTransformMakeAbsolute(
     Local=(Rotation=(X=0,Y=0,Z=0,W=1),Translation=(X=0,Y=0,Z=0),Scale3D=(X=1,Y=1,Z=1))
     Local.Rotation=&RigVMFunction_MathQuaternionFromEuler(Euler=(X=0,Y=0,Z=0), Euler.Z=@n4.Value).Result
     Parent=RigVMVariableNode() {variable=Ik Aim Transform}.Value
    ) {template="Make Absolute::Execute(in Local,in Parent,out Global)"}.Global
    ExecuteContext=@RigVMFunction_Sequence_1.D
   ) {variable=Hip Socket Rig Transform}
   n5 = RigVMVariableNode() {variable=Second Angle}
   > RigVMVariableNode(Value=&RigVMFunction_MathTransformMakeAbsolute(
     Local=(Rotation=(X=0,Y=0,Z=0,W=1),Translation=(X=0,Y=0,Z=0),Scale3D=(X=1,Y=1,Z=1))
     Local.Rotation=&RigVMFunction_MathQuaternionFromEuler(Euler=(X=0,Y=0,Z=0), Euler.Z=@n5.Value).Result
     Local.Translation.X=RigVMVariableNode() {variable=Hip Socket Bone Length}.Value
     Parent=RigVMVariableNode() {variable=Hip Socket Rig Transform}.Value
    ) {template="Make Absolute::Execute(in Local,in Parent,out Global)"}.Global) {variable=Knee Rig Transform}
   RigVMFunctionReturnNode(
    ExecuteContext=@RigVMFunction_Sequence.B
    Hip Socket Transform=RigVMVariableNode() {variable=Hip Socket Rig Transform}.Value
    Knee Transform=RigVMVariableNode() {variable=Knee Rig Transform}.Value
    Ankle Transform=RigVMVariableNode() {variable=Ankle_Rig_Transform}.Value
   )
   n6 = &RigUnit_HierarchyAddBone(
    ExecutePin=()
    Name=NewBone
    Parent=(Type=Bone,Name="None")
    Space=LocalSpace
    Transform=RigVMVariableNode() {variable=Hip Socket Rig Transform}.Value
   )
   > &RigUnit_HierarchyAddBone(Name=NewBone, Space=LocalSpace, Transform=(Rotation=(X=0,Y=0,Z=0,W=1),Translation=(X=25,Y=0,Z=0),Scale3D=(X=1,Y=1,Z=1)), Parent=@n6.Item)
   Get Main Leg PV = RigVMFunctionReferenceNode(Initial=true)
   &RigVMFunction_MathDoubleNegate(Value=@LawOfCosine.BetaAngle) {template="Negate::Execute(in Value,out Result)"}
   &RigVMFunction_MathDoubleNegate(Value=@LawOfCosine.GammaAngle) {template="Negate::Execute(in Value,out Result)"}
   RigVMVariableNode() {variable=Foot Bone Length}
   RigVMVariableNode() {variable=Knee Rig Transform}
   RigVMVariableNode() {variable=Knee Bone Length}
  ]}
 New Function_ContainedGraph {nodes [
   n1 = RigVMFunctionEntryNode()
   RigVMFunction_Sequence = &RigVMFunction_Sequence(ExecuteContext=@n1.ExecuteContext)
   RigVMFunction_Sequence_1 = RigVMAggregateNode(ExecuteContext=@RigVMFunction_Sequence.A)
   RigVMFunction_MathVectorDistance = &RigVMFunction_MathVectorDistance(
    A=&RigUnit_GetTransform(bInitial=true, Item=RigVMVariableNode() {variable=Hip Socket Connector}.Value).Transform.Translation
    B=&RigUnit_GetTransform(bInitial=true, Item=RigVMVariableNode() {variable=Knee Connector}.Value).Transform.Translation
   )
   RigVMVariableNode(Value=@RigVMFunction_MathVectorDistance.Result, ExecuteContext=@RigVMFunction_Sequence_1.A) {variable=Hip Socket Bone Length}
   n2 = RigVMVariableNode() {variable=Knee Connector}
   n3 = RigVMVariableNode() {variable=Ankle Connector}
   RigVMVariableNode(
    Value=&RigVMFunction_MathVectorDistance(
     A=&RigUnit_GetTransform(bInitial=true, Item=@n2.Value).Transform.Translation
     B=&RigUnit_GetTransform(bInitial=true, Item=@n3.Value).Transform.Translation
    ).Result
    ExecuteContext=@RigVMFunction_Sequence_1.B
   ) {variable=Knee Bone Length}
   n4 = RigVMVariableNode() {variable=Ankle Connector}
   n5 = RigVMVariableNode() {variable=Ball Connector}
   RigVMVariableNode(
    Value=&RigVMFunction_MathVectorDistance(
     A=&RigUnit_GetTransform(bInitial=true, Item=@n4.Value).Transform.Translation
     B=&RigUnit_GetTransform(bInitial=true, Item=@n5.Value).Transform.Translation
    ).Result
    ExecuteContext=@RigVMFunction_Sequence_1.C
   ) {variable=Foot Bone Length}
   n6 = RigVMVariableNode() {variable=Ball Connector}
   n7 = RigVMVariableNode() {variable=Toe Connector}
   RigVMVariableNode(
    Value=&RigVMFunction_MathVectorDistance(
     A=&RigUnit_GetTransform(bInitial=true, Item=@n6.Value).Transform.Translation
     B=&RigUnit_GetTransform(bInitial=true, Item=@n7.Value).Transform.Translation
    ).Result
    ExecuteContext=@RigVMFunction_Sequence_1.D
   ) {variable=Ball Bone Length}
   n8 = RigVMVariableNode() {variable=Toe Connector}
   n9 = RigVMVariableNode() {variable=Toe Tip Connector}
   RigVMVariableNode(
    Value=&RigVMFunction_MathVectorDistance(
     A=&RigUnit_GetTransform(bInitial=true, Item=@n8.Value).Transform.Translation
     B=&RigUnit_GetTransform(bInitial=true, Item=@n9.Value).Transform.Translation
    ).Result
    ExecuteContext=@RigVMFunction_Sequence_1.E
   ) {variable=Toe Bone Length}
   RigVMFunctionReturnNode(ExecuteContext=@RigVMFunction_Sequence.B)
  ]}
 New Function_ContainedGraph {nodes [
   n1 = RigVMFunctionEntryNode()
   Subtract = &RigVMFunction_MathVectorSub(A=@n1.Mid Transform.Translation, B=@n1.Root Transform.Translation) {template="Subtract::Execute(in A,in B,out Result)"}
   RigVMFunction_MathVectorCross = &RigVMFunction_MathVectorCross(
    B=@Subtract.Result
    A=&RigVMFunction_MathVectorSub(B=@n1.Mid Transform.Translation, A=@n1.Tip Transform.Translation) {template="Subtract::Execute(in A,in B,out Result)"}.Result
   ) {template="Cross::Execute(in A,in B,out Result)"}
   RigVMFunction_MathVectorSetLength = &RigVMFunction_MathVectorSetLength(Value=@RigVMFunction_MathVectorCross.Result, Length=RigVMVariableNode() {variable=Vector Length}.Value) {template="SetLength::Execute(in Value,in Length,out Result)"}
   Add = &RigVMFunction_MathVectorAdd(A=@RigVMFunction_MathVectorSetLength.Result, B=RigVMVariableNode() {variable=Root Transform}.Value.Translation) {template="Add::Execute(in A,in B,out Result)"}
   > RigVMFunctionReturnNode(Position=@Add.Result, Vector=@RigVMFunction_MathVectorSetLength.Result)
  ]}
 New Function_ContainedGraph {nodes [
   RigVMFunctionEntryNode()
   n1 = &RigUnit_GetTransform(bInitial=true, Item=RigVMVariableNode() {variable=Ankle Connector}.Value)
   Subtract = &RigVMFunction_MathVectorSub(
    B=&RigUnit_GetTransform(bInitial=true, Item=RigVMVariableNode() {variable=Knee Connector}.Value).Transform.Translation
    A=@n1.Transform.Translation
   ) {template="Subtract::Execute(in A,in B,out Result)"}
   n2 = RigVMVariableNode() {variable=Ball Connector}
   RigVMFunction_MathVectorCross = &RigVMFunction_MathVectorCross(
    A=@Subtract.Result
    B=&RigVMFunction_MathVectorSub(B=@n1.Transform.Translation, A=&RigUnit_GetTransform(bInitial=true, Item=@n2.Value).Transform.Translation) {template="Subtract::Execute(in A,in B,out Result)"}.Result
   ) {template="Cross::Execute(in A,in B,out Result)"}
   RigVMFunction_MathVectorSetLength = &RigVMFunction_MathVectorSetLength(Value=@RigVMFunction_MathVectorCross.Result, Length=RigVMVariableNode() {variable=Vector Length}.Value) {template="SetLength::Execute(in Value,in Length,out Result)"}
   Add = &RigVMFunction_MathVectorAdd(
    B=&RigUnit_GetTransform(bInitial=true, Item=RigVMVariableNode() {variable=Knee Connector}.Value).Transform.Translation
    A=@RigVMFunction_MathVectorSetLength.Result
   ) {template="Add::Execute(in A,in B,out Result)"}
   RigVMFunction_MathQuaternionFromEuler = &RigVMFunction_MathQuaternionFromEuler(Euler=(X=0,Y=0,Z=0), Euler.X=RigVMVariableNode() {variable=Rotate}.Value)
   Get Main Leg PV = RigVMFunctionReferenceNode(Initial=true)
   n3 = &RigUnit_AimBoneMath(
    DebugSettings=(bEnabled=False,Scale=10,WorldOffset=(Rotation=(X=0,Y=0,Z=0,W=1),Translation=(X=0,Y=0,Z=0),Scale3D=(X=1,Y=1,Z=1)))
    Primary=(Weight=1,Axis=(X=1,Y=0,Z=0),Target=(X=1,Y=0,Z=0),Kind=Location,Space=(Type=Bone,Name="None"))
    Secondary=(Weight=1,Axis=(X=0,Y=0,Z=1),Target=(X=1,Y=0,Z=0),Kind=Direction,Space=(Type=Bone,Name="None"))
    Weight=1
    Secondary.Target=@"Get Main Leg PV".Pole Vector
    Primary.Target=&RigUnit_GetTransform(bInitial=true, Item=RigVMVariableNode() {variable=Knee Connector}.Value).Transform.Translation
    InputTransform=&RigUnit_GetTransform(bInitial=true, Item=RigVMVariableNode() {variable=Hip Socket Connector}.Value).Transform
   )
   n4 = &RigVMFunction_MathTransformMakeAbsolute(
    Local=&RigVMFunction_MathTransformMakeRelative(
     Global=(Rotation=(X=0,Y=0,Z=0,W=1),Translation=(X=0,Y=0,Z=0),Scale3D=(X=1,Y=1,Z=1))
     Global.Translation=@Add.Result
     Parent=&RigVMFunction_MathTransformMakeAbsolute(
      Local=(Rotation=(X=0,Y=0,Z=0,W=1),Translation=(X=0,Y=0,Z=0),Scale3D=(X=1,Y=1,Z=1))
      Local.Rotation=@RigVMFunction_MathQuaternionFromEuler.Result
      Parent=@n3.Result
     ) {template="Make Absolute::Execute(in Local,in Parent,out Global)"}.Global
    ).Local
    Parent=RigVMVariableNode() {variable=Hip Socket Transform}.Value
   ) {template="Make Absolute::Execute(in Local,in Parent,out Global)"}
   > RigVMFunctionReturnNode(
    Position=@n4.Global.Translation
    Vector=&RigVMFunction_MathVectorSub(A=@n4.Global.Translation, B=RigVMVariableNode() {variable=Knee Transform}.Value.Translation) {template="Subtract::Execute(in A,in B,out Result)"}.Result
   )
  ]}
 New Function_ContainedGraph {nodes [
   RigVMFunctionEntryNode()
   n1 = &RigUnit_GetTransform(bInitial=true, Item=RigVMVariableNode() {variable=Knee Connector}.Value)
   Subtract = &RigVMFunction_MathVectorSub(
    B=&RigUnit_GetTransform(bInitial=true, Item=RigVMVariableNode() {variable=Hip Socket Connector}.Value).Transform.Translation
    A=@n1.Transform.Translation
   ) {template="Subtract::Execute(in A,in B,out Result)"}
   n2 = RigVMVariableNode() {variable=Ankle Connector}
   RigVMFunction_MathVectorCross = &RigVMFunction_MathVectorCross(
    B=@Subtract.Result
    A=&RigVMFunction_MathVectorSub(B=@n1.Transform.Translation, A=&RigUnit_GetTransform(bInitial=true, Item=@n2.Value).Transform.Translation) {template="Subtract::Execute(in A,in B,out Result)"}.Result
   ) {template="Cross::Execute(in A,in B,out Result)"}
   RigVMFunction_MathVectorSetLength = &RigVMFunction_MathVectorSetLength(Value=@RigVMFunction_MathVectorCross.Result, Length=RigVMVariableNode() {variable=Vector Length}.Value) {template="SetLength::Execute(in Value,in Length,out Result)"}
   Add = &RigVMFunction_MathVectorAdd(
    B=&RigUnit_GetTransform(bInitial=true, Item=RigVMVariableNode() {variable=Hip Socket Connector}.Value).Transform.Translation
    A=@RigVMFunction_MathVectorSetLength.Result
   ) {template="Add::Execute(in A,in B,out Result)"}
   RigVMFunction_MathQuaternionFromEuler = &RigVMFunction_MathQuaternionFromEuler(Euler=(X=0,Y=0,Z=0), Euler.X=RigVMVariableNode() {variable=Rotate}.Value)
   Get Main Leg PV = RigVMFunctionReferenceNode(Initial=true)
   n3 = &RigUnit_AimBoneMath(
    DebugSettings=(bEnabled=False,Scale=10,WorldOffset=(Rotation=(X=0,Y=0,Z=0,W=1),Translation=(X=0,Y=0,Z=0),Scale3D=(X=1,Y=1,Z=1)))
    Primary=(Weight=1,Axis=(X=1,Y=0,Z=0),Target=(X=1,Y=0,Z=0),Kind=Location,Space=(Type=Bone,Name="None"))
    Secondary=(Weight=1,Axis=(X=0,Y=0,Z=1),Target=(X=1,Y=0,Z=0),Kind=Direction,Space=(Type=Bone,Name="None"))
    Weight=1
    Secondary.Target=@"Get Main Leg PV".Pole Vector
    Primary.Target=&RigUnit_GetTransform(bInitial=true, Item=RigVMVariableNode() {variable=Knee Connector}.Value).Transform.Translation
    InputTransform=&RigUnit_GetTransform(bInitial=true, Item=RigVMVariableNode() {variable=Hip Socket Connector}.Value).Transform
   )
   n4 = &RigVMFunction_MathTransformMakeAbsolute(
    Local=&RigVMFunction_MathTransformMakeRelative(
     Global=(Rotation=(X=0,Y=0,Z=0,W=1),Translation=(X=0,Y=0,Z=0),Scale3D=(X=1,Y=1,Z=1))
     Global.Translation=@Add.Result
     Parent=&RigVMFunction_MathTransformMakeAbsolute(
      Local=(Rotation=(X=0,Y=0,Z=0,W=1),Translation=(X=0,Y=0,Z=0),Scale3D=(X=1,Y=1,Z=1))
      Local.Rotation=@RigVMFunction_MathQuaternionFromEuler.Result
      Parent=@n3.Result
     ) {template="Make Absolute::Execute(in Local,in Parent,out Global)"}.Global
    ).Local
    Parent=RigVMVariableNode() {variable=Hip Socket Transform}.Value
   ) {template="Make Absolute::Execute(in Local,in Parent,out Global)"}
   > RigVMFunctionReturnNode(
    Position=@n4.Global.Translation
    Vector=&RigVMFunction_MathVectorSub(A=@n4.Global.Translation, B=RigVMVariableNode() {variable=Hip Socket Transform}.Value.Translation) {template="Subtract::Execute(in A,in B,out Result)"}.Result
   )
  ]}
 New Function_ContainedGraph {nodes [
   n1 = RigVMFunctionEntryNode()
   RigVMFunction_Sequence = &RigVMFunction_Sequence(ExecuteContext=@n1.ExecuteContext)
   RigVMFunction_Sequence_1 = RigVMAggregateNode(ExecuteContext=@RigVMFunction_Sequence.A)
   Create Name = RigVMFunctionReferenceNode(Tokens=("Hip Socket Null"), ExecuteContext=@RigVMFunction_Sequence_1.A)
   n2 = RigVMVariableNode() {variable=Hip Socket Connector}
   n3 = RigVMVariableNode() {variable=Right Side}
   Multiply = &RigVMFunction_MathDoubleMul(
    A=1
    B=RigVMDispatchNode(
     True=1
     False=RigVMDispatchNode(False=1, True=-1, Condition=@n3.Value) {template="DISPATCH_RigVMDispatch_If(in Condition,in True,in False,out Result)"}.Result
     Condition=RigVMVariableNode() {variable=Left Side}.Value
    ) {template="DISPATCH_RigVMDispatch_If(in Condition,in True,in False,out Result)"}.Result
   )
   > n4 = &RigUnit_HierarchyAddNull(
    Space=GlobalSpace
    Transform=(Rotation=(X=0,Y=0,Z=0,W=1),Translation=(X=0,Y=0,Z=0),Scale3D=(X=1,Y=1,Z=1))
    Name=@"Create Name".Name
    Transform.Translation=&RigUnit_GetTransform(bInitial=true, Item=@n2.Value).Transform.Translation
    Parent=&RigUnit_HierarchyGetParent(bDefaultParent=True, Child=@n2.Value).Parent
    Transform.Scale3D.X=@Multiply.Result
   )
   > RigVMVariableNode(Value=@n4.Item) {variable=Hip Socket Null}
   Create Name_1 = RigVMFunctionReferenceNode(Tokens=("Hip Socket Ctrl"), ExecuteContext=@RigVMFunction_Sequence_1.B)
   > Set Side Color = RigVMFunctionReferenceNode()
   n5 = &RigVMFunction_MathDoubleMul(B=0.35, A=RigVMVariableNode() {variable=Control Size}.Value)
   > SpawnControl = &RigUnit_HierarchyAddControlTransform(
    InitialValue=(Rotation=(X=0,Y=0,Z=0,W=1),Translation=(X=0,Y=0,Z=0),Scale3D=(X=1,Y=1,Z=1))
    OffsetSpace=LocalSpace
    OffsetTransform=(Rotation=(X=0,Y=0,Z=0,W=1),Translation=(X=0,Y=0,Z=0),Scale3D=(X=1,Y=1,Z=1))
    Settings="(InitialSpace=LocalSpace,bUsePreferredRotationOrder=False,PreferredRotationOrder=YZX,Limits=(LimitTranslationX=(bMinimum=False,bMaximum=False),LimitTranslationY=(bMinimum=False,bMaximum=False),LimitTranslationZ=(bMinimum=False,bMaximum=False),LimitPitch=(b ...(truncated)"
    Settings.Shape.Transform.Translation.X=RigVMDispatchNode(False=7, True=7, Condition=RigVMVariableNode() {variable=Right Side}.Value) {template="DISPATCH_RigVMDispatch_If(in Condition,in True,in False,out Result)"}.Result
    Settings.Shape.Transform.Scale3D.X=@n5.Result
    Settings.Shape.Transform.Scale3D.Y=@n5.Result
    Settings.Shape.Transform.Scale3D.Z=@n5.Result
    Settings.Shape.Color=@"Set Side Color".Color
    Parent=RigVMVariableNode() {variable=Hip Socket Null}.Value
    Name=@"Create Name_1".Name
   ) {template="SpawnControl::Execute(in InitialValue,in Settings,in OffsetTransform,in OffsetSpace,in Parent,in Name,out Item)"}
   > RigVMVariableNode(Value=@SpawnControl.Item) {variable=Hip Socket Control}
   Create Name_2 = RigVMFunctionReferenceNode(Tokens=("Hip Socket Ref"), ExecuteContext=@RigVMFunction_Sequence_1.C)
   > n6 = &RigUnit_HierarchyAddNull(
    Space=LocalSpace
    Transform=(Rotation=(X=0,Y=0,Z=0,W=1),Translation=(X=0,Y=0,Z=0),Scale3D=(X=1,Y=1,Z=1))
    Parent=RigVMVariableNode() {variable=Hip Socket Null}.Value
    Name=@"Create Name_2".Name
   )
   > RigVMVariableNode(Value=@n6.Item) {variable=Hip Socket Ref}
   RigVMFunctionReturnNode(ExecuteContext=@RigVMFunction_Sequence.B)
  ]}
 New Function_ContainedGraph {nodes [
   n1 = RigVMFunctionEntryNode()
   RigVMFunction_Sequence = &RigVMFunction_Sequence(ExecuteContext=@n1.ExecuteContext)
   RigVMFunctionReturnNode(ExecuteContext=@RigVMFunction_Sequence.B)
   RigVMVariableNode() {variable=Hip Socket Null}
  ]}
 New Function_ContainedGraph {nodes [
   n1 = RigVMFunctionEntryNode()
   RigVMFunction_Sequence = &RigVMFunction_Sequence(ExecuteContext=@n1.ExecuteContext)
   RigVMFunction_Sequence_1 = RigVMAggregateNode(ExecuteContext=@RigVMFunction_Sequence.A)
   DISPATCH_RigDispatch_SetModuleMetadata = RigVMDispatchNode(
    Name=Ik Control Null
    NameSpace=Self
    Value=&RigUnit_HierarchyGetParent(bDefaultParent=True, Child=RigVMVariableNode() {variable=Ik Control}.Value).Parent
    ExecuteContext=@RigVMFunction_Sequence_1.A
   ) {template="DISPATCH_RigDispatch_SetModuleMetadata(in Name,in NameSpace,in Value,out Success)"}
   RigVMDispatchNode(Name=Ik Control, NameSpace=Self, Value=RigVMVariableNode() {variable=Ik Control}.Value, ExecuteContext=@RigVMFunction_Sequence_1.B) {template="DISPATCH_RigDispatch_SetModuleMetadata(in Name,in NameSpace,in Value,out Success)"}
   RigVMDispatchNode(Name=Ik Point, NameSpace=Self, Value=(Type=Bone,Name="None"), ExecuteContext=@RigVMFunction_Sequence_1.C) {template="DISPATCH_RigDispatch_SetModuleMetadata(in Name,in NameSpace,in Value,out Success)"}
   RigVMFunctionReturnNode(ExecuteContext=@RigVMFunction_Sequence.B)
  ]}
 New Function_ContainedGraph {nodes [
   n1 = RigVMFunctionEntryNode()
   n2 = &RigVMFunction_MathDoubleLawOfCosine(A=@n1.Length A, B=@n1.Length B, C=@n1.Length C) {template="LawOfCosine::Execute(in A,in B,in C,out AlphaAngle,out BetaAngle,out GammaAngle,out bValid)"}
   n3 = &RigVMFunction_MathDoubleDeg(Value=RigVMDispatchNode(
     True=&RigVMFunction_MathDoubleNegate(Value=@n2.AlphaAngle) {template="Negate::Execute(in Value,out Result)"}.Result
     Condition=RigVMVariableNode() {variable=Negate A}.Value
     False=@n2.AlphaAngle
    ) {template="DISPATCH_RigVMDispatch_If(in Condition,in True,in False,out Result)"}.Result) {template="Degrees::Execute(in Value,out Result)"}
   n4 = &RigVMFunction_MathDoubleDeg(Value=RigVMDispatchNode(
     True=&RigVMFunction_MathDoubleNegate(Value=@n2.BetaAngle) {template="Negate::Execute(in Value,out Result)"}.Result
     Condition=RigVMVariableNode() {variable=Negate B}.Value
     False=@n2.BetaAngle
    ) {template="DISPATCH_RigVMDispatch_If(in Condition,in True,in False,out Result)"}.Result) {template="Degrees::Execute(in Value,out Result)"}
   n5 = &RigVMFunction_MathDoubleDeg(Value=RigVMDispatchNode(
     True=&RigVMFunction_MathDoubleNegate(Value=@n2.GammaAngle) {template="Negate::Execute(in Value,out Result)"}.Result
     Condition=RigVMVariableNode() {variable=Negate C}.Value
     False=@n2.GammaAngle
    ) {template="DISPATCH_RigVMDispatch_If(in Condition,in True,in False,out Result)"}.Result) {template="Degrees::Execute(in Value,out Result)"}
   > RigVMFunctionReturnNode(
    Angle A=RigVMDispatchNode(
     True=&RigVMFunction_MathDoubleAdd(B=180, A=@n3.Result).Result
     Condition=RigVMVariableNode() {variable=A 180}.Value
     False=@n3.Result
    ) {template="DISPATCH_RigVMDispatch_If(in Condition,in True,in False,out Result)"}.Result
    Angle B=RigVMDispatchNode(
     True=&RigVMFunction_MathDoubleAdd(B=180, A=@n4.Result).Result
     Condition=RigVMVariableNode() {variable="B 180 "}.Value
     False=@n4.Result
    ) {template="DISPATCH_RigVMDispatch_If(in Condition,in True,in False,out Result)"}.Result
    Angle C=RigVMDispatchNode(
     True=&RigVMFunction_MathDoubleAdd(B=180, A=@n5.Result).Result
     Condition=RigVMVariableNode() {variable=C 180}.Value
     False=@n5.Result
    ) {template="DISPATCH_RigVMDispatch_If(in Condition,in True,in False,out Result)"}.Result
   )
  ]}
 New Function_ContainedGraph {nodes [
   n1 = RigVMFunctionEntryNode()
   RigVMFunction_Sequence = &RigVMFunction_Sequence(ExecuteContext=@n1.ExecuteContext)
   RigVMFunction_Sequence_1 = RigVMAggregateNode(ExecuteContext=@RigVMFunction_Sequence.A)
   RigVMFunction_MathVectorDistance = &RigVMFunction_MathVectorDistance(
    A=RigVMVariableNode() {variable=Hip Socket Rig Transform}.Value.Translation
    B=RigVMVariableNode() {variable=Ankle Rig Transform}.Value.Translation
   )
   Get Ik Angles = RigVMFunctionReferenceNode(
    "B 180 "=false
    C 180=true
    Negate B=false
    Negate C=false
    Length C=@RigVMFunction_MathVectorDistance.Result
    Length A=RigVMVariableNode() {variable=Hip Socket Bone Length}.Value
    Length B=RigVMVariableNode() {variable=Knee Bone Length}.Value
   )
   RigVMVariableNode(Value=@"Get Ik Angles".Angle B, ExecuteContext=@RigVMFunction_Sequence_1.B) {variable=Hip Socket Angle}
   > RigVMVariableNode(Value=RigVMDispatchNode(
     True=0
     Condition=RigVMDispatchNode(B=180, A=@"Get Ik Angles".Angle C) {template="DISPATCH_RigVMDispatch_CoreEquals(in A,in B,out Result)"}.Result
     False=@"Get Ik Angles".Angle C
    ) {template="DISPATCH_RigVMDispatch_If(in Condition,in True,in False,out Result)"}.Result) {variable=Knee Angle}
   > &RigVMFunction_DebugLineStripNoSpace(
    Color=(R=1,G=0,B=0,A=1)
    DebugDrawSettings=(DepthPriority=SDPG_Foreground,Lifetime=-1)
    Points=((X=0,Y=0,Z=0),(X=0,Y=0,Z=0),(X=0,Y=0,Z=0))
    Thickness=1
    WorldOffset=(Rotation=(X=0,Y=0,Z=0,W=1),Translation=(X=0,Y=0,Z=0),Scale3D=(X=1,Y=1,Z=1))
    bEnabled=false
    Points.1=RigVMVariableNode() {variable=Hip Socket Rig Transform}.Value.Translation
    Points.2=RigVMVariableNode() {variable=Ankle Rig Transform}.Value.Translation
    Points.0=RigVMVariableNode() {variable=Knee Rig Transform}.Value.Translation
   )
   n2 = &RigUnit_AimBoneMath(
    DebugSettings=(bEnabled=False,Scale=10,WorldOffset=(Rotation=(X=0,Y=0,Z=0,W=1),Translation=(X=0,Y=0,Z=0),Scale3D=(X=1,Y=1,Z=1)))
    InputTransform=(Rotation=(X=0,Y=0,Z=0,W=1),Translation=(X=0,Y=0,Z=0),Scale3D=(X=1,Y=1,Z=1))
    Primary=(Weight=1,Axis=(X=1,Y=0,Z=0),Target=(X=1,Y=0,Z=0),Kind=Location,Space=(Type=Bone,Name="None"))
    Secondary=(Weight=1,Axis=(X=0,Y=0,Z=1),Target=(X=1,Y=0,Z=0),Kind=Direction,Space=(Type=Bone,Name="None"))
    Weight=1
    InputTransform.Translation=RigVMVariableNode() {variable=Hip Socket Rig Transform}.Value.Translation
    Primary.Target=RigVMVariableNode() {variable=Ankle Rig Transform}.Value.Translation
    Secondary.Target=RigVMVariableNode() {variable=Knee Ik PV}.Value
   )
   GetAnimationChannelFromItem = /Script/ControlRig.RigUnit_GetFloatAnimationChannelFromItem(bInitial=False, Item=RigVMVariableNode() {variable=Knee Twist}.Value) {is=struct RigVMUnitNode, template="GetAnimationChannelFromItem::Execute(out Value,in Item,in bInitial)"}
   RigVMFunction_MathQuaternionFromEuler = &RigVMFunction_MathQuaternionFromEuler(Euler=(X=0,Y=0,Z=0), Euler.X=@GetAnimationChannelFromItem.Value)
   n3 = RigVMVariableNode() {variable=Hip Socket Angle}
   RigVMVariableNode(
    Value=&RigVMFunction_MathTransformMakeAbsolute(
     Local=(Rotation=(X=0,Y=0,Z=0,W=1),Translation=(X=0,Y=0,Z=0),Scale3D=(X=1,Y=1,Z=1))
     Parent=&RigVMFunction_MathTransformMakeAbsolute(
      Local=(Rotation=(X=0,Y=0,Z=0,W=1),Translation=(X=0,Y=0,Z=0),Scale3D=(X=1,Y=1,Z=1))
      Parent=@n2.Result
      Local.Rotation=@RigVMFunction_MathQuaternionFromEuler.Result
     ) {template="Make Absolute::Execute(in Local,in Parent,out Global)"}.Global
     Local.Rotation=&RigVMFunction_MathQuaternionFromEuler(Euler=(X=0,Y=0,Z=0), Euler.Z=@n3.Value).Result
    ) {template="Make Absolute::Execute(in Local,in Parent,out Global)"}.Global
    ExecuteContext=@RigVMFunction_Sequence_1.D
   ) {variable=Temp Hip Socket Transform}
   n4 = RigVMVariableNode() {variable=Hip Socket Bone Length}
   n5 = RigVMVariableNode() {variable=Temp Hip Socket Transform}
   n6 = RigVMVariableNode() {variable=Knee Angle}
   > RigVMVariableNode(Value=&RigVMFunction_MathTransformMakeAbsolute(
     Local=(Rotation=(X=0,Y=0,Z=0,W=1),Translation=(X=0,Y=0,Z=0),Scale3D=(X=1,Y=1,Z=1))
     Parent=&RigVMFunction_MathTransformMakeAbsolute(Local=(Rotation=(X=0,Y=0,Z=0,W=1),Translation=(X=0,Y=0,Z=0),Scale3D=(X=1,Y=1,Z=1)), Local.Translation.X=@n4.Value, Parent=@n5.Value) {template="Make Absolute::Execute(in Local,in Parent,out Global)"}.Global
     Local.Rotation=&RigVMFunction_MathQuaternionFromEuler(Euler=(X=0,Y=0,Z=0), Euler.Z=@n6.Value).Result
    ) {template="Make Absolute::Execute(in Local,in Parent,out Global)"}.Global) {variable=Temp Knee Transform}
   > RigVMVariableNode(Value=&RigVMFunction_MathTransformMakeAbsolute(
     Local=(Rotation=(X=0,Y=0,Z=0,W=1),Translation=(X=0,Y=0,Z=0),Scale3D=(X=1,Y=1,Z=1))
     Local.Translation.X=RigVMVariableNode() {variable=Knee Bone Length}.Value
     Parent=RigVMVariableNode() {variable=Temp Knee Transform}.Value
    ) {template="Make Absolute::Execute(in Local,in Parent,out Global)"}.Global) {variable=Temp Ankle Transform}
   n7 = RigVMVariableNode() {variable=Temp Hip Socket Transform}
   > RigVMFunction_DebugLineStripNoSpace = &RigVMFunction_DebugLineStripNoSpace(
    Color=(R=0,G=0,B=1,A=1)
    DebugDrawSettings=(DepthPriority=SDPG_Foreground,Lifetime=-1)
    Points=((X=0,Y=0,Z=0),(X=0,Y=0,Z=0),(X=0,Y=0,Z=0),(X=0,Y=0,Z=0))
    Thickness=0.5
    WorldOffset=(Rotation=(X=0,Y=0,Z=0,W=1),Translation=(X=0,Y=0,Z=0),Scale3D=(X=1,Y=1,Z=1))
    bEnabled=RigVMVariableNode() {variable=DebugOnOff}.Value
    Points.0=@n7.Value.Translation
    Points.3=@n7.Value.Translation
    Points.1=RigVMVariableNode() {variable=Temp Knee Transform}.Value.Translation
    Points.2=RigVMVariableNode() {variable=Temp Ankle Transform}.Value.Translation
   )
   n8 = &RigUnit_AimBoneMath(
    DebugSettings=(bEnabled=False,Scale=10,WorldOffset=(Rotation=(X=0,Y=0,Z=0,W=1),Translation=(X=0,Y=0,Z=0),Scale3D=(X=1,Y=1,Z=1)))
    InputTransform=(Rotation=(X=0,Y=0,Z=0,W=1),Translation=(X=0,Y=0,Z=0),Scale3D=(X=1,Y=1,Z=1))
    Primary=(Weight=1,Axis=(X=1,Y=0,Z=0),Target=(X=1,Y=0,Z=0),Kind=Location,Space=(Type=Bone,Name="None"))
    Secondary=(Weight=1,Axis=(X=0,Y=0,Z=1),Target=(X=1,Y=0,Z=0),Kind=Direction,Space=(Type=Bone,Name="None"))
    Weight=1
    InputTransform.Translation=RigVMVariableNode() {variable=Temp Hip Socket Transform}.Value.Translation
    Primary.Target=RigVMVariableNode() {variable=Temp Knee Transform}.Value.Translation
    Secondary.Target=RigVMVariableNode() {variable=Knee Ik PV}.Value
   )
   RigVMVariableNode(Value=@n8.Result, ExecuteContext=@RigVMFunction_Sequence_1.E) {variable=Hip Socket Rig Transform}
   n9 = &RigUnit_AimBoneMath(
    DebugSettings=(bEnabled=False,Scale=10,WorldOffset=(Rotation=(X=0,Y=0,Z=0,W=1),Translation=(X=0,Y=0,Z=0),Scale3D=(X=1,Y=1,Z=1)))
    InputTransform=(Rotation=(X=0,Y=0,Z=0,W=1),Translation=(X=0,Y=0,Z=0),Scale3D=(X=1,Y=1,Z=1))
    Primary=(Weight=1,Axis=(X=1,Y=0,Z=0),Target=(X=1,Y=0,Z=0),Kind=Location,Space=(Type=Bone,Name="None"))
    Secondary=(Weight=1,Axis=(X=0,Y=0,Z=1),Target=(X=1,Y=0,Z=0),Kind=Direction,Space=(Type=Bone,Name="None"))
    Weight=1
    InputTransform.Translation=RigVMVariableNode() {variable=Temp Knee Transform}.Value.Translation
    Primary.Target=RigVMVariableNode() {variable=Temp Ankle Transform}.Value.Translation
    Secondary.Target=RigVMVariableNode() {variable=Knee Ik PV}.Value
   )
   > RigVMVariableNode(Value=@n9.Result) {variable=Knee Rig Transform}
   Get Main Leg PV_2 = RigVMFunctionReferenceNode(Initial=False)
   n10 = RigVMVariableNode() {variable=Hip Socket Rig Transform}
   n11 = RigVMVariableNode() {variable=Ankle Rig Transform}
   n12 = &RigUnit_AimBoneMath(
    DebugSettings=(bEnabled=False,Scale=10,WorldOffset=(Rotation=(X=0,Y=0,Z=0,W=1),Translation=(X=0,Y=0,Z=0),Scale3D=(X=1,Y=1,Z=1)))
    InputTransform=(Rotation=(X=0,Y=0,Z=0,W=1),Translation=(X=0,Y=0,Z=0),Scale3D=(X=1,Y=1,Z=1))
    Primary=(Weight=1,Axis=(X=0,Y=0,Z=1),Target=(X=1,Y=0,Z=0),Kind=Location,Space=(Type=Bone,Name="None"))
    Secondary=(Weight=1,Axis=(X=1,Y=0,Z=0),Target=(X=1,Y=0,Z=0),Kind=Direction,Space=(Type=Bone,Name="None"))
    Weight=1
    Secondary.Target=@"Get Main Leg PV_2".Pole Vector
    Primary.Target=@n10.Value.Translation
    InputTransform.Translation=@n11.Value.Translation
   )
   Multiply = &RigVMFunction_MathDoubleMul(B=0.5, A=&RigVMFunction_MathVectorDistance(A=@n10.Value.Translation, B=@n11.Value.Translation).Result)
   &RigUnit_SetTransform(
    Weight=1
    bInitial=False
    Item=&RigUnit_HierarchyGetParent(bDefaultParent=True, Child=RigVMVariableNode() {variable=Knee Rotate Control Null}.Value).Parent
    Value=&RigVMFunction_MathTransformMakeAbsolute(Local=(Rotation=(X=0,Y=0,Z=0,W=1),Translation=(X=0,Y=0,Z=0),Scale3D=(X=1,Y=1,Z=1)), Parent=@n12.Result, Local.Translation.Z=@Multiply.Result) {template="Make Absolute::Execute(in Local,in Parent,out Global)"}.Global
    ExecutePin=@RigVMFunction_Sequence_1.A
   )
   n13 = &RigUnit_GetTransform(bInitial=False, Item=RigVMVariableNode() {variable=Knee Rotate Control Null}.Value)
   n14 = &RigUnit_AimBoneMath(
    DebugSettings=(bEnabled=False,Scale=10,WorldOffset=(Rotation=(X=0,Y=0,Z=0,W=1),Translation=(X=0,Y=0,Z=0),Scale3D=(X=1,Y=1,Z=1)))
    InputTransform=(Rotation=(X=0,Y=0,Z=0,W=1),Translation=(X=0,Y=0,Z=0),Scale3D=(X=1,Y=1,Z=1))
    Primary=(Weight=1,Axis=(X=0,Y=0,Z=1),Target=(X=1,Y=0,Z=0),Kind=Location,Space=(Type=Bone,Name="None"))
    Secondary=(Weight=1,Axis=(X=0,Y=1,Z=0),Target=(X=1,Y=0,Z=0),Kind=Location,Space=(Type=Bone,Name="None"))
    Weight=1
    InputTransform.Translation=@n13.Transform.Translation
    Secondary.Target=&RigUnit_GetTransform(bInitial=False, Item=RigVMVariableNode() {variable=Knee PV}.Value).Transform.Translation
    Primary.Target=&RigVMFunction_MathTransformMakeAbsolute(Local=(Rotation=(X=0,Y=0,Z=0,W=1),Translation=(X=0,Y=0,Z=10),Scale3D=(X=1,Y=1,Z=1)), Parent=@n13.Transform) {template="Make Absolute::Execute(in Local,in Parent,out Global)"}.Global.Translation
   )
   &RigUnit_SetTransform(
    Weight=1
    bInitial=False
    Item=RigVMVariableNode() {variable=Knee Rotate Control Null}.Value
    ExecutePin=@RigVMFunction_Sequence_1.C
    Value=@n14.Result
   )
   n15 = RigVMVariableNode() {variable=Knee Rotate Control}
   n16 = RigVMVariableNode() {variable=Knee Rotate Up Vector}
   > RigVMVariableNode(Value=&RigVMFunction_MathVectorSub(
     B=&RigUnit_GetTransform(bInitial=False, Item=@n15.Value).Transform.Translation
     A=&RigUnit_GetTransform(bInitial=False, Item=@n16.Value).Transform.Translation
    ) {template="Subtract::Execute(in A,in B,out Result)"}.Result) {variable=Knee Ik PV}
   RigVMFunctionReturnNode(ExecuteContext=@RigVMFunction_Sequence.B)
   n17 = RigVMRerouteNode(Value=())
   Get Main Leg PV_1 = RigVMFunctionReferenceNode(Initial=False)
   Dot__ = /Script/RigVM.RigVMFunction_MathVectorDot(B=@"Get Main Leg PV_1".Pole Vector, A=RigVMVariableNode() {variable=Knee Ik PV}.Value) {is=struct RigVMUnitNode, template="Dot,|::Execute(in A,in B,out Result)"}
   Greater = &RigVMFunction_MathDoubleGreater(B=0, A=@Dot__.Result) {template="Greater::Execute(in A,in B,out Result)"}
   n18 = RigVMVariableNode() {variable=Temp Hip Socket Transform}
   n19 = RigVMVariableNode() {variable=Temp Knee Transform}
   n20 = RigVMVariableNode() {variable=Temp Ankle Transform}
   n21 = &RigVMFunction_MathVectorSetLength(
    Length=10
    Value=&RigVMFunction_MathVectorCross(
     B=&RigVMFunction_MathVectorSub(B=@n18.Value.Translation, A=@n19.Value.Translation) {template="Subtract::Execute(in A,in B,out Result)"}.Result
     A=&RigVMFunction_MathVectorSub(B=@n18.Value.Translation, A=@n20.Value.Translation) {template="Subtract::Execute(in A,in B,out Result)"}.Result
    ) {template="Cross::Execute(in A,in B,out Result)"}.Result
   ) {template="SetLength::Execute(in Value,in Length,out Result)"}
   Negate = /Script/RigVM.RigVMFunction_MathVectorNegate(Value=@n21.Result) {is=struct RigVMUnitNode, template="Negate::Execute(in Value,out Result)"}
   RigVMVariableNode(
    Value=RigVMDispatchNode(Condition=@Greater.Result, False=@Negate.Result, True=@n21.Result) {template="DISPATCH_RigVMDispatch_If(in Condition,in True,in False,out Result)"}.Result
    ExecuteContext=@n17.Value
   ) {variable=Bone Aim Up Vector}
   n22 = RigVMVariableNode() {variable=Hip Socket Rig Transform}
   Add = &RigVMFunction_MathVectorAdd(A=@n22.Value.Translation, B=RigVMVariableNode() {variable=Bone Aim Up Vector}.Value) {template="Add::Execute(in A,in B,out Result)"}
   > RigVMFunction_DebugLineNoSpace = &RigVMFunction_DebugLineNoSpace(
    Color=(R=0,G=0,B=1,A=1)
    DebugDrawSettings=(DepthPriority=SDPG_Foreground,Lifetime=-1)
    Thickness=0.5
    WorldOffset=(Rotation=(X=0,Y=0,Z=0,W=1),Translation=(X=0,Y=0,Z=0),Scale3D=(X=1,Y=1,Z=1))
    bEnabled=false
    B=@Add.Result
    A=@n22.Value.Translation
   )
   RigVMVariableNode() {variable=Knee Twist Offset}
  ]}
 New Function_ContainedGraph {nodes [
   RigVMFunctionEntryNode()
   > RigVMFunctionReturnNode()
  ]}
 New Function_ContainedGraph {nodes [
   n1 = RigVMFunctionEntryNode()
   RigVMFunction_Sequence = &RigVMFunction_Sequence(ExecuteContext=@n1.ExecuteContext)
   RigVMFunction_Sequence_1 = RigVMAggregateNode(ExecuteContext=@RigVMFunction_Sequence.A)
   Create Name = RigVMFunctionReferenceNode(Tokens=("Bk Ball Null"), ExecuteContext=@RigVMFunction_Sequence_1.A)
   n2 = &RigUnit_AimBoneMath(
    DebugSettings=(bEnabled=False,Scale=10,WorldOffset=(Rotation=(X=0,Y=0,Z=0,W=1),Translation=(X=0,Y=0,Z=0),Scale3D=(X=1,Y=1,Z=1)))
    Primary=(Weight=1,Axis=(X=0,Y=1,Z=0),Target=(X=1,Y=0,Z=0),Kind=Location,Space=(Type=Bone,Name="None"))
    Secondary=(Weight=1,Axis=(X=1,Y=0,Z=0),Target=(X=10,Y=0,Z=0),Kind=Direction,Space=(Type=Bone,Name="None"))
    Weight=1
    InputTransform=&RigUnit_GetTransform(bInitial=true, Item=RigVMVariableNode() {variable=Ball Connector}.Value).Transform
    Primary.Target=&RigUnit_GetTransform(bInitial=true, Item=RigVMVariableNode() {variable=Toe Connector}.Value).Transform.Translation
   )
   > n3 = &RigUnit_HierarchyAddNull(Space=GlobalSpace, Name=@"Create Name".Name, Parent=RigVMVariableNode() {variable=Heel Roll Control}.Value, Transform=@n2.Result)
   > Create Name_1 = RigVMFunctionReferenceNode(Tokens=("Bk Ball"))
   n4 = &RigVMFunction_MathDoubleMul(B=0.7, A=RigVMVariableNode() {variable=Control Size}.Value)
   > SpawnControl = &RigUnit_HierarchyAddControlTransform(
    InitialValue=(Rotation=(X=0,Y=0,Z=0,W=1),Translation=(X=0,Y=0,Z=0),Scale3D=(X=1,Y=1,Z=1))
    OffsetSpace=LocalSpace
    OffsetTransform=(Rotation=(X=0,Y=0,Z=0,W=1),Translation=(X=0,Y=0,Z=0),Scale3D=(X=1,Y=1,Z=1))
    Settings="(InitialSpace=LocalSpace,bUsePreferredRotationOrder=False,PreferredRotationOrder=YZX,Limits=(LimitTranslationX=(bMinimum=True,bMaximum=True),LimitTranslationY=(bMinimum=True,bMaximum=True),LimitTranslationZ=(bMinimum=True,bMaximum=True),LimitPitch=(bMinimu ...(truncated)"
    Name=@"Create Name_1".Name
    Settings.Shape.Transform.Scale3D.X=@n4.Result
    Settings.Shape.Transform.Scale3D.Y=@n4.Result
    Settings.Shape.Transform.Scale3D.Z=@n4.Result
    Parent=@n3.Item
   ) {template="SpawnControl::Execute(in InitialValue,in Settings,in OffsetTransform,in OffsetSpace,in Parent,in Name,out Item)"}
   > RigVMVariableNode(Value=@SpawnControl.Item) {variable=Ball Fk Control}
   Create Name_2 = RigVMFunctionReferenceNode(Tokens=("Bk Toe Null"), ExecuteContext=@RigVMFunction_Sequence_1.B)
   n5 = &RigUnit_AimBoneMath(
    DebugSettings=(bEnabled=False,Scale=10,WorldOffset=(Rotation=(X=0,Y=0,Z=0,W=1),Translation=(X=0,Y=0,Z=0),Scale3D=(X=1,Y=1,Z=1)))
    Primary=(Weight=1,Axis=(X=0,Y=1,Z=0),Target=(X=1,Y=0,Z=0),Kind=Location,Space=(Type=Bone,Name="None"))
    Secondary=(Weight=1,Axis=(X=1,Y=0,Z=0),Target=(X=10,Y=0,Z=0),Kind=Direction,Space=(Type=Bone,Name="None"))
    Weight=1
    InputTransform=&RigUnit_GetTransform(bInitial=true, Item=RigVMVariableNode() {variable=Toe Connector}.Value).Transform
    Primary.Target=&RigUnit_GetTransform(bInitial=true, Item=RigVMVariableNode() {variable=Toe Tip Connector}.Value).Transform.Translation
   )
   > n6 = &RigUnit_HierarchyAddNull(Space=GlobalSpace, Name=@"Create Name_2".Name, Parent=RigVMVariableNode() {variable=Ball Fk Control}.Value, Transform=@n5.Result)
   > Create Name_3 = RigVMFunctionReferenceNode(Tokens=("Bk Toe"))
   n7 = &RigVMFunction_MathDoubleMul(B=0.7, A=RigVMVariableNode() {variable=Control Size}.Value)
   > n8 = &RigUnit_HierarchyAddControlTransform(
    InitialValue=(Rotation=(X=0,Y=0,Z=0,W=1),Translation=(X=0,Y=0,Z=0),Scale3D=(X=1,Y=1,Z=1))
    OffsetSpace=LocalSpace
    OffsetTransform=(Rotation=(X=0,Y=0,Z=0,W=1),Translation=(X=0,Y=0,Z=0),Scale3D=(X=1,Y=1,Z=1))
    Settings="(InitialSpace=LocalSpace,bUsePreferredRotationOrder=False,PreferredRotationOrder=YZX,Limits=(LimitTranslationX=(bMinimum=True,bMaximum=True),LimitTranslationY=(bMinimum=True,bMaximum=True),LimitTranslationZ=(bMinimum=True,bMaximum=True),LimitPitch=(bMinimu ...(truncated)"
    Name=@"Create Name_3".Name
    Settings.Shape.Transform.Scale3D.X=@n7.Result
    Settings.Shape.Transform.Scale3D.Y=@n7.Result
    Settings.Shape.Transform.Scale3D.Z=@n7.Result
    Parent=@n6.Item
   ) {template="SpawnControl::Execute(in InitialValue,in Settings,in OffsetTransform,in OffsetSpace,in Parent,in Name,out Item)"}
   > RigVMVariableNode(Value=@n8.Item) {variable=Toe Fk Control}
   > Create Name_8 = RigVMFunctionReferenceNode(Tokens=("Bk Toe Tip Point"))
   > n9 = &RigUnit_HierarchyAddNull(
    Space=GlobalSpace
    Name=@"Create Name_8".Name
    Transform=&RigUnit_GetTransform(bInitial=true, Item=RigVMVariableNode() {variable=Toe Tip Connector}.Value).Transform
    Parent=RigVMVariableNode() {variable=Toe Fk Control}.Value
   )
   > RigVMVariableNode(Value=@n9.Item) {variable=Toe Tip Point}
   Create Name_4 = RigVMFunctionReferenceNode(Tokens=("Bk Ball Pivot Null"), ExecuteContext=@RigVMFunction_Sequence_1.C)
   n10 = &RigUnit_AimBoneMath(
    DebugSettings=(bEnabled=False,Scale=10,WorldOffset=(Rotation=(X=0,Y=0,Z=0,W=1),Translation=(X=0,Y=0,Z=0),Scale3D=(X=1,Y=1,Z=1)))
    Primary=(Weight=1,Axis=(X=0,Y=1,Z=0),Target=(X=1,Y=0,Z=0),Kind=Location,Space=(Type=Bone,Name="None"))
    Secondary=(Weight=1,Axis=(X=1,Y=0,Z=0),Target=(X=10,Y=0,Z=0),Kind=Direction,Space=(Type=Bone,Name="None"))
    Weight=1
    InputTransform=&RigUnit_GetTransform(bInitial=true, Item=RigVMVariableNode() {variable=Ball Connector}.Value).Transform
    Primary.Target=&RigUnit_GetTransform(bInitial=true, Item=RigVMVariableNode() {variable=Toe Connector}.Value).Transform.Translation
   )
   > n11 = &RigUnit_HierarchyAddNull(
    Space=GlobalSpace
    Transform=(Rotation=(X=0,Y=0,Z=0,W=1),Translation=(X=0,Y=0,Z=0),Scale3D=(X=1,Y=1,Z=1))
    Name=@"Create Name_4".Name
    Parent=RigVMVariableNode() {variable=Ball Fk Control}.Value
    Transform.Scale3D=@n10.Result.Scale3D
    Transform.Rotation=@n10.Result.Rotation
    Transform.Translation=&RigUnit_GetTransform(bInitial=true, Item=RigVMVariableNode() {variable=Toe Connector}.Value).Transform.Translation
   )
   > Create Name_5 = RigVMFunctionReferenceNode(Tokens=("Bk Ball Pivot"))
   n12 = &RigVMFunction_MathDoubleMul(B=1, A=RigVMVariableNode() {variable=Control Size}.Value)
   > n13 = &RigUnit_HierarchyAddControlTransform(
    InitialValue=(Rotation=(X=0,Y=0,Z=0,W=1),Translation=(X=0,Y=0,Z=0),Scale3D=(X=1,Y=1,Z=1))
    OffsetSpace=LocalSpace
    OffsetTransform=(Rotation=(X=0,Y=0,Z=0,W=1),Translation=(X=0,Y=0,Z=0),Scale3D=(X=1,Y=1,Z=1))
    Settings="(InitialSpace=LocalSpace,bUsePreferredRotationOrder=False,PreferredRotationOrder=YZX,Limits=(LimitTranslationX=(bMinimum=True,bMaximum=True),LimitTranslationY=(bMinimum=True,bMaximum=True),LimitTranslationZ=(bMinimum=True,bMaximum=True),LimitPitch=(bMinimu ...(truncated)"
    Name=@"Create Name_5".Name
    Settings.Shape.Transform.Scale3D.X=@n12.Result
    Settings.Shape.Transform.Scale3D.Y=@n12.Result
    Settings.Shape.Transform.Scale3D.Z=@n12.Result
    Parent=@n11.Item
   ) {template="SpawnControl::Execute(in InitialValue,in Settings,in OffsetTransform,in OffsetSpace,in Parent,in Name,out Item)"}
   > RigVMVariableNode(Value=@n13.Item) {variable=Ball Pivot Control}
   > Create Name_7 = RigVMFunctionReferenceNode(Tokens=("Bk IkPoint"))
   > n14 = &RigUnit_HierarchyAddNull(
    Space=GlobalSpace
    Name=@"Create Name_7".Name
    Transform=&RigUnit_GetTransform(bInitial=true, Item=RigVMVariableNode() {variable=Ik Control}.Value).Transform
    Parent=RigVMVariableNode() {variable=Ball Pivot Control}.Value
   )
   > RigVMVariableNode(Value=@n14.Item) {variable=Ik Point}
   RigVMFunctionReturnNode(ExecuteContext=@RigVMFunction_Sequence.B)
   RigVMVariableNode() {variable=Ik Control}
  ]}
 New Function_ContainedGraph {nodes [
   n1 = RigVMFunctionEntryNode()
   RigVMFunction_Sequence = &RigVMFunction_Sequence(ExecuteContext=@n1.ExecuteContext)
   n2 = &RigVMFunction_Sequence(ExecuteContext=@RigVMFunction_Sequence.A)
   n3 = &RigUnit_AimBoneMath(
    DebugSettings=(bEnabled=False,Scale=10,WorldOffset=(Rotation=(X=0,Y=0,Z=0,W=1),Translation=(X=0,Y=0,Z=0),Scale3D=(X=1,Y=1,Z=1)))
    InputTransform=(Rotation=(X=0,Y=0,Z=0,W=1),Translation=(X=0,Y=0,Z=0),Scale3D=(X=1,Y=1,Z=1))
    Primary=(Weight=1,Axis=(X=1,Y=0,Z=0),Target=(X=1,Y=0,Z=0),Kind=Location,Space=(Type=Bone,Name="None"))
    Secondary=(Weight=1,Axis=(X=0,Y=0,Z=1),Target=(X=1,Y=0,Z=0),Kind=Direction,Space=(Type=Bone,Name="None"))
    Weight=1
    Primary.Target=&RigUnit_GetTransform(bInitial=False, Item=RigVMVariableNode() {variable=Toe Fk Control}.Value).Transform.Translation
    InputTransform.Translation=RigVMVariableNode() {variable=Ball Rig Transform}.Value.Translation
    Secondary.Space=RigVMVariableNode() {variable=Ball Pivot Control}.Value
   )
   RigVMVariableNode(Value=@n3.Result, ExecuteContext=@n2.A) {variable=Ball Rig Transform}
   n4 = &RigUnit_AimBoneMath(
    DebugSettings=(bEnabled=False,Scale=10,WorldOffset=(Rotation=(X=0,Y=0,Z=0,W=1),Translation=(X=0,Y=0,Z=0),Scale3D=(X=1,Y=1,Z=1)))
    InputTransform=(Rotation=(X=0,Y=0,Z=0,W=1),Translation=(X=0,Y=0,Z=0),Scale3D=(X=1,Y=1,Z=1))
    Primary=(Weight=1,Axis=(X=1,Y=0,Z=0),Target=(X=1,Y=0,Z=0),Kind=Location,Space=(Type=Bone,Name="None"))
    Secondary=(Weight=1,Axis=(X=0,Y=0,Z=1),Target=(X=1,Y=0,Z=0),Kind=Direction,Space=(Type=Bone,Name="None"))
    Weight=1
    Primary.Target=&RigUnit_GetTransform(bInitial=False, Item=RigVMVariableNode() {variable=Toe Tip Point}.Value).Transform.Translation
    InputTransform.Translation=&RigVMFunction_MathTransformMakeAbsolute(
     Local=(Rotation=(X=0,Y=0,Z=0,W=1),Translation=(X=0,Y=0,Z=0),Scale3D=(X=1,Y=1,Z=1))
     Parent=RigVMVariableNode() {variable=Ball Rig Transform}.Value
     Local.Translation.X=RigVMVariableNode() {variable=Ball Bone Length}.Value
    ) {template="Make Absolute::Execute(in Local,in Parent,out Global)"}.Global.Translation
    Secondary.Space=RigVMVariableNode() {variable=Toe Fk Control}.Value
   )
   RigVMVariableNode(Value=@n4.Result, ExecuteContext=@n2.B) {variable=Toe Rig Transform}
   RigVMFunctionReturnNode(ExecuteContext=@RigVMFunction_Sequence.B)
  ]}
 New Function_ContainedGraph {nodes [
   n1 = RigVMFunctionEntryNode()
   RigVMFunction_Sequence = &RigVMFunction_Sequence(ExecuteContext=@n1.ExecuteContext)
   n2 = &RigVMFunction_Sequence(ExecuteContext=@RigVMFunction_Sequence.A)
   Create Name = RigVMFunctionReferenceNode(Tokens=("Bk Toe Roll Null"), ExecuteContext=@n2.A)
   > n3 = &RigUnit_HierarchyAddNull(
    Space=LocalSpace
    Parent=RigVMVariableNode() {variable=Ik Control}.Value
    Transform=RigVMVariableNode() {variable=Toe Roll Initial Transform}.Value
    Name=@"Create Name".Name
   )
   > Create Name_1 = RigVMFunctionReferenceNode(Tokens=("Bk Toe Roll"))
   > Set Side Color = RigVMFunctionReferenceNode()
   n4 = &RigVMFunction_MathDoubleMul(B=0.3, A=RigVMVariableNode() {variable=Control Size}.Value)
   > SpawnControl = &RigUnit_HierarchyAddControlTransform(
    InitialValue=(Rotation=(X=0,Y=0,Z=0,W=1),Translation=(X=0,Y=0,Z=0),Scale3D=(X=1,Y=1,Z=1))
    OffsetSpace=LocalSpace
    OffsetTransform=(Rotation=(X=0,Y=0,Z=0,W=1),Translation=(X=0,Y=0,Z=0),Scale3D=(X=1,Y=1,Z=1))
    Settings="(InitialSpace=LocalSpace,bUsePreferredRotationOrder=False,PreferredRotationOrder=YZX,Limits=(LimitTranslationX=(bMinimum=False,bMaximum=False),LimitTranslationY=(bMinimum=False,bMaximum=False),LimitTranslationZ=(bMinimum=False,bMaximum=False),LimitPitch=(b ...(truncated)"
    Name=@"Create Name_1".Name
    Settings.Shape.Transform.Scale3D.X=@n4.Result
    Settings.Shape.Transform.Scale3D.Y=@n4.Result
    Settings.Shape.Transform.Scale3D.Z=@n4.Result
    Settings.Shape.Color=@"Set Side Color".Color
    Parent=@n3.Item
   ) {template="SpawnControl::Execute(in InitialValue,in Settings,in OffsetTransform,in OffsetSpace,in Parent,in Name,out Item)"}
   > RigVMVariableNode(Value=@SpawnControl.Item) {variable=Toe Roll Control}
   Create Name_3 = RigVMFunctionReferenceNode(Tokens=("Bk Heel Roll Null"), ExecuteContext=@n2.B)
   > n5 = &RigUnit_HierarchyAddNull(
    Space=LocalSpace
    Parent=RigVMVariableNode() {variable=Toe Roll Control}.Value
    Transform=RigVMVariableNode() {variable=Heel Roll Initial Transform}.Value
    Name=@"Create Name_3".Name
   )
   > Create Name_2 = RigVMFunctionReferenceNode(Tokens=("Bk Heel Roll"))
   > Set Side Color_1 = RigVMFunctionReferenceNode()
   n6 = &RigVMFunction_MathDoubleMul(B=0.3, A=RigVMVariableNode() {variable=Control Size}.Value)
   > n7 = &RigUnit_HierarchyAddControlTransform(
    InitialValue=(Rotation=(X=0,Y=0,Z=0,W=1),Translation=(X=0,Y=0,Z=0),Scale3D=(X=1,Y=1,Z=1))
    OffsetSpace=LocalSpace
    OffsetTransform=(Rotation=(X=0,Y=0,Z=0,W=1),Translation=(X=0,Y=0,Z=0),Scale3D=(X=1,Y=1,Z=1))
    Settings="(InitialSpace=LocalSpace,bUsePreferredRotationOrder=False,PreferredRotationOrder=YZX,Limits=(LimitTranslationX=(bMinimum=False,bMaximum=False),LimitTranslationY=(bMinimum=False,bMaximum=False),LimitTranslationZ=(bMinimum=False,bMaximum=False),LimitPitch=(b ...(truncated)"
    Name=@"Create Name_2".Name
    Settings.Shape.Transform.Scale3D.X=@n6.Result
    Settings.Shape.Transform.Scale3D.Y=@n6.Result
    Settings.Shape.Transform.Scale3D.Z=@n6.Result
    Settings.Shape.Color=@"Set Side Color_1".Color
    Parent=@n5.Item
   ) {template="SpawnControl::Execute(in InitialValue,in Settings,in OffsetTransform,in OffsetSpace,in Parent,in Name,out Item)"}
   > RigVMVariableNode(Value=@n7.Item) {variable=Heel Roll Control}
   RigVMFunctionReturnNode(ExecuteContext=@RigVMFunction_Sequence.B)
  ]}
 New Function_ContainedGraph {nodes [
   n1 = RigVMFunctionEntryNode()
   RigVMFunction_Sequence = &RigVMFunction_Sequence(ExecuteContext=@n1.ExecuteContext)
   RigVMFunctionReturnNode(ExecuteContext=@RigVMFunction_Sequence.B, Color=RigVMVariableNode() {variable=Output Color}.Value)
   RigVMFunction_ControlFlowBranch = &RigVMFunction_ControlFlowBranch(Condition=RigVMVariableNode() {variable=Left Side}.Value, ExecuteContext=@RigVMFunction_Sequence.A)
   n2 = &RigVMFunction_ControlFlowBranch(ExecuteContext=@RigVMFunction_ControlFlowBranch.False, Condition=RigVMVariableNode() {variable=Right Side}.Value)
   n3 = &RigVMFunction_ControlFlowBranch(ExecuteContext=@n2.False, Condition=RigVMVariableNode() {variable=Custom Side}.Value)
   RigVMVariableNode(Value=RigVMVariableNode() {variable=Custom Side Color}.Value, ExecuteContext=@n3.True) {variable=Output Color}
   RigVMVariableNode(Value=RigVMVariableNode() {variable=Right Side Color}.Value, ExecuteContext=@n2.True) {variable=Output Color}
   RigVMVariableNode(Value=RigVMVariableNode() {variable=Left Side Color}.Value, ExecuteContext=@RigVMFunction_ControlFlowBranch.True) {variable=Output Color}
  ]}
 New Function_ContainedGraph {nodes [
   n1 = RigVMFunctionEntryNode()
   RigVMFunction_Sequence = &RigVMFunction_Sequence(ExecuteContext=@n1.ExecuteContext)
   RigVMFunction_Sequence_1 = RigVMAggregateNode(ExecuteContext=@RigVMFunction_Sequence.A)
   /Script/ControlRig.RigUnit_AddParent(
    DisplayLabel=None
    Parent=RigVMVariableNode() {variable=Ik Control}.Value
    Child=RigVMVariableNode() {variable=Main PV Control}.Value
    ExecutePin=@RigVMFunction_Sequence_1.A
   ) {is=struct RigVMUnitNode}
   Create Name = RigVMFunctionReferenceNode(Tokens=("Hip Space"), ExecuteContext=@RigVMFunction_Sequence_1.B)
   n2 = &RigUnit_HierarchyGetParent(bDefaultParent=True, Child=RigVMVariableNode() {variable=Hip Socket Connector}.Value)
   > n3 = &RigUnit_HierarchyAddNull(
    Space=GlobalSpace
    Transform=(Rotation=(X=0,Y=0,Z=0,W=1),Translation=(X=0,Y=0,Z=0),Scale3D=(X=1,Y=1,Z=1))
    Name=@"Create Name".Name
    Transform.Translation=&RigUnit_GetTransform(bInitial=False, Item=@n2.Parent).Transform.Translation
    Parent=@n2.Parent
   )
   > &RigUnit_AddParent(DisplayLabel=None, Child=RigVMVariableNode() {variable=Ik Control}.Value, Parent=@n3.Item)
   RigVMFunctionReturnNode(ExecuteContext=@RigVMFunction_Sequence.B)
  ]}
 New Function_ContainedGraph {nodes [
   n1 = RigVMFunctionEntryNode()
   RigVMFunction_Sequence = &RigVMFunction_Sequence(ExecuteContext=@n1.ExecuteContext)
   RigVMFunction_Sequence_1 = RigVMAggregateNode(ExecuteContext=@RigVMFunction_Sequence.A)
   Create Name = RigVMFunctionReferenceNode(Tokens=("Hock Rotate Aim Null"), ExecuteContext=@RigVMFunction_Sequence_1.A)
   n2 = &RigUnit_AimBoneMath(
    DebugSettings=(bEnabled=False,Scale=10,WorldOffset=(Rotation=(X=0,Y=0,Z=0,W=1),Translation=(X=0,Y=0,Z=0),Scale3D=(X=1,Y=1,Z=1)))
    InputTransform=(Rotation=(X=0,Y=0,Z=0,W=1),Translation=(X=0,Y=0,Z=0),Scale3D=(X=1,Y=1,Z=1))
    Primary=(Weight=1,Axis=(X=0,Y=0,Z=1),Target=(X=1,Y=0,Z=0),Kind=Location,Space=(Type=Bone,Name="None"))
    Secondary=(Weight=1,Axis=(X=1,Y=0,Z=0),Target=(X=1,Y=0,Z=0),Kind=Direction,Space=(Type=Bone,Name="None"))
    Weight=1
    Primary.Target=&RigUnit_GetTransform(bInitial=true, Item=RigVMVariableNode() {variable=Ankle Connector}.Value).Transform.Translation
    InputTransform.Translation=&RigUnit_GetTransform(bInitial=true, Item=RigVMVariableNode() {variable=Ball Connector}.Value).Transform.Translation
   )
   > n3 = &RigUnit_HierarchyAddNull(Space=GlobalSpace, Transform=@n2.Result, Parent=RigVMVariableNode() {variable=Ik Control}.Value, Name=@"Create Name".Name)
   > RigVMVariableNode(Value=@n3.Item) {variable=Hock Rotate Aim Null}
   Create Name_1 = RigVMFunctionReferenceNode(Tokens=("Hock Rotate Null"), ExecuteContext=@RigVMFunction_Sequence_1.B)
   > n4 = &RigUnit_HierarchyAddNull(
    Space=LocalSpace
    Transform=(Rotation=(X=0,Y=0,Z=0,W=1),Translation=(X=0,Y=0,Z=0),Scale3D=(X=1,Y=1,Z=1))
    Parent=RigVMVariableNode() {variable=Hock Rotate Aim Null}.Value
    Name=@"Create Name_1".Name
   )
   > RigVMVariableNode(Value=@n4.Item) {variable=Hock Rotate Offset}
   Create Name_3 = RigVMFunctionReferenceNode(Tokens=("Ankle Rotate Aim Point"), ExecuteContext=@RigVMFunction_Sequence_1.D)
   > n5 = &RigUnit_HierarchyAddNull(
    Space=GlobalSpace
    Transform=&RigUnit_GetTransform(bInitial=true, Item=RigVMVariableNode() {variable=Ankle Connector}.Value).Transform
    Parent=RigVMVariableNode() {variable=Hock Rotate Control}.Value
    Name=@"Create Name_3".Name
   )
   > RigVMVariableNode(Value=@n5.Item) {variable=Ankle Rotate Aim Point}
   Create Name_2 = RigVMFunctionReferenceNode(Tokens=("Hock Rotate"), ExecuteContext=@RigVMFunction_Sequence_1.C)
   n6 = RigVMVariableNode() {variable=Hock Control Transform}
   n7 = RigVMVariableNode() {variable=Control Size}
   > SpawnControl = &RigUnit_HierarchyAddControlTransform(
    InitialValue=(Rotation=(X=0,Y=0,Z=0,W=1),Translation=(X=0,Y=0,Z=0),Scale3D=(X=1,Y=1,Z=1))
    OffsetSpace=LocalSpace
    OffsetTransform=(Rotation=(X=0,Y=0,Z=0,W=1),Translation=(X=0,Y=0,Z=0),Scale3D=(X=1,Y=1,Z=1))
    Settings="(InitialSpace=LocalSpace,bUsePreferredRotationOrder=False,PreferredRotationOrder=YZX,Limits=(LimitTranslationX=(bMinimum=True,bMaximum=True),LimitTranslationY=(bMinimum=True,bMaximum=True),LimitTranslationZ=(bMinimum=True,bMaximum=True),LimitPitch=(bMinimu ...(truncated)"
    Settings.Shape.Transform.Scale3D.Z=&RigVMFunction_MathDoubleMul(A=@n6.Value.Scale3D.Z, B=@n7.Value).Result
    Settings.Shape.Transform.Scale3D.X=&RigVMFunction_MathDoubleMul(B=@n6.Value.Scale3D.X, A=@n7.Value).Result
    Settings.Shape.Transform.Scale3D.Y=&RigVMFunction_MathDoubleMul(A=@n6.Value.Scale3D.Y, B=@n7.Value).Result
    Parent=RigVMVariableNode() {variable=Hock Rotate Offset}.Value
    Settings.Shape.Transform.Rotation=@n6.Value.Rotation
    Settings.Shape.Transform.Translation=@n6.Value.Translation
    Name=@"Create Name_2".Name
   ) {template="SpawnControl::Execute(in InitialValue,in Settings,in OffsetTransform,in OffsetSpace,in Parent,in Name,out Item)"}
   > RigVMVariableNode(Value=@SpawnControl.Item) {variable=Hock Rotate Control}
   RigVMFunctionReturnNode(ExecuteContext=@RigVMFunction_Sequence.B)
  ]}
 New Function_ContainedGraph {nodes [
   RigVMFunctionEntryNode()
   > RigVMFunctionReturnNode()
  ]}
 New Function_ContainedGraph {nodes [
   RigVMFunctionEntryNode()
   n1 = RigVMVariableNode() {variable=Initial}
   Get Main Leg PV_3 = RigVMFunctionReferenceNode(Initial=@n1.Value)
   n2 = &RigUnit_AimBoneMath(
    DebugSettings=(bEnabled=False,Scale=10,WorldOffset=(Rotation=(X=0,Y=0,Z=0,W=1),Translation=(X=0,Y=0,Z=0),Scale3D=(X=1,Y=1,Z=1)))
    InputTransform=(Rotation=(X=0,Y=0,Z=0,W=1),Translation=(X=0,Y=0,Z=0),Scale3D=(X=1,Y=1,Z=1))
    Primary=(Weight=1,Axis=(X=1,Y=0,Z=0),Target=(X=1,Y=0,Z=0),Kind=Location,Space=(Type=Bone,Name="None"))
    Secondary=(Weight=1,Axis=(X=0,Y=0,Z=1),Target=(X=1,Y=0,Z=0),Kind=Direction,Space=(Type=Bone,Name="None"))
    Weight=1
    Secondary.Target=@"Get Main Leg PV_3".Pole Vector
    InputTransform.Translation=&RigUnit_GetTransform(bInitial=RigVMVariableNode() {variable=Initial}.Value, Item=RigVMVariableNode() {variable=Hip Socket Control}.Value).Transform.Translation
    Primary.Target=&RigUnit_GetTransform(bInitial=@n1.Value, Item=RigVMVariableNode() {variable=Ik Control}.Value).Transform.Translation
   )
   > RigVMFunctionReturnNode(Aim=@n2.Result)
  ]}
 New Function_ContainedGraph {nodes [
   n1 = RigVMFunctionEntryNode()
   > RigVMFunctionReturnNode(Relative=&RigVMFunction_MathTransformMakeRelative(Parent=@n1.Parent, Global=&RigUnit_GetTransform(bInitial=@n1.Initial, Item=@n1.RigNode).Transform).Local)
  ]}
 New Function_ContainedGraph {nodes [
   RigVMFunctionEntryNode()
   n1 = &RigUnit_GetTransform(bInitial=False, Item=RigVMVariableNode() {variable=Hip Socket Control}.Value)
   n2 = &RigUnit_GetTransform(
    bInitial=False
    Item=&RigUnit_HierarchyGetParent(bDefaultParent=True, Child=RigVMVariableNode() {variable=Hip Socket Control}.Value).Parent
   )
   Add_5 = RigVMAggregateNode(
    B=0
    C=&RigUnit_GetFloatAnimationChannelFromItem(bInitial=False, Item=RigVMVariableNode() {variable=Knee Offset}.Value) {template="GetAnimationChannelFromItem::Execute(out Value,in Item,in bInitial)"}.Value
    A=&RigVMFunction_MathQuaternionToEuler(RotationOrder=ZYX, Value=&RigVMFunction_MathTransformMakeRelative(Global=@n1.Transform, Parent=@n2.Transform).Local.Rotation).Result.X
   )
   > RigVMFunctionReturnNode(Rotation=&RigVMFunction_MathQuaternionFromEuler(Euler=(X=0,Y=0,Z=1), Euler.Z=@Add_5.Result).Result)
   RigVMFunction_MathQuaternionToEuler = &RigVMFunction_MathQuaternionToEuler(
    RotationOrder=ZYX
    Value=&RigUnit_GetTransform(Space=LocalSpace, bInitial=False, Item=RigVMVariableNode() {variable=Hip Socket Control}.Value).Transform.Rotation
   )
  ]}
 New Function_ContainedGraph {nodes [
   n1 = RigVMFunctionEntryNode()
   RigVMFunction_Sequence = &RigVMFunction_Sequence(ExecuteContext=@n1.ExecuteContext)
   RigVMFunction_Sequence_1 = RigVMAggregateNode(ExecuteContext=@RigVMFunction_Sequence.A)
   RigVMFunctionReturnNode(ExecuteContext=@RigVMFunction_Sequence.B)
  ]}
 New Function_ContainedGraph {nodes [
   n1 = RigVMFunctionEntryNode()
   RigVMFunction_Sequence = &RigVMFunction_Sequence(ExecuteContext=@n1.ExecuteContext)
   RigVMFunction_Sequence_1 = RigVMAggregateNode(ExecuteContext=@RigVMFunction_Sequence.A)
   Create Name = RigVMFunctionReferenceNode(Tokens=("Knee Rotate Control Aim"), ExecuteContext=@RigVMFunction_Sequence_1.B)
   > n2 = &RigUnit_HierarchyAddNull(
    Space=GlobalSpace
    Name=@"Create Name".Name
    Parent=&RigUnit_HierarchyGetParent(bDefaultParent=True, Child=RigVMVariableNode() {variable=Hip Socket Connector}.Value).Parent
    Transform=RigVMVariableNode() {variable=Control Null Transform}.Value
   )
   > Create Name_5 = RigVMFunctionReferenceNode(Tokens=("Knee Rotate Control Null"))
   > n3 = &RigUnit_HierarchyAddNull(
    Space=LocalSpace
    Transform=(Rotation=(X=0,Y=0,Z=0,W=1),Translation=(X=0,Y=0,Z=0),Scale3D=(X=1,Y=1,Z=1))
    Name=@"Create Name_5".Name
    Parent=@n2.Item
   )
   > RigVMVariableNode(Value=@n3.Item) {variable=Knee Rotate Control Null}
   > Create Name_1 = RigVMFunctionReferenceNode(Tokens=("Knee Rotate"))
   n4 = &RigVMFunction_MathDoubleMul(B=0, A=RigVMVariableNode() {variable=Control Size}.Value)
   > SpawnControl = &RigUnit_HierarchyAddControlTransform(
    InitialValue=(Rotation=(X=0,Y=0,Z=0,W=1),Translation=(X=0,Y=0,Z=0),Scale3D=(X=1,Y=1,Z=1))
    OffsetSpace=LocalSpace
    OffsetTransform=(Rotation=(X=0,Y=0,Z=0,W=1),Translation=(X=0,Y=0,Z=0),Scale3D=(X=1,Y=1,Z=1))
    Settings="(InitialSpace=LocalSpace,bUsePreferredRotationOrder=False,PreferredRotationOrder=YZX,Limits=(LimitTranslationX=(bMinimum=True,bMaximum=True),LimitTranslationY=(bMinimum=True,bMaximum=True),LimitTranslationZ=(bMinimum=True,bMaximum=True),LimitPitch=(bMinimu ...(truncated)"
    Name=@"Create Name_1".Name
    Settings.Shape.Transform.Scale3D.X=@n4.Result
    Settings.Shape.Transform.Scale3D.Y=@n4.Result
    Settings.Shape.Transform.Scale3D.Z=@n4.Result
    Parent=@n3.Item
   ) {template="SpawnControl::Execute(in InitialValue,in Settings,in OffsetTransform,in OffsetSpace,in Parent,in Name,out Item)"}
   > RigVMVariableNode(Value=@SpawnControl.Item) {variable=Knee Rotate Control}
   Get Main Leg PV = RigVMFunctionReferenceNode(Initial=true)
   n5 = &RigUnit_GetTransform(bInitial=true, Item=RigVMVariableNode() {variable=Hip Socket Connector}.Value)
   n6 = &RigUnit_GetTransform(bInitial=true, Item=RigVMVariableNode() {variable=Ankle Connector}.Value)
   n7 = &RigUnit_AimBoneMath(
    DebugSettings=(bEnabled=False,Scale=10,WorldOffset=(Rotation=(X=0,Y=0,Z=0,W=1),Translation=(X=0,Y=0,Z=0),Scale3D=(X=1,Y=1,Z=1)))
    InputTransform=(Rotation=(X=0,Y=0,Z=0,W=1),Translation=(X=0,Y=0,Z=0),Scale3D=(X=1,Y=1,Z=1))
    Primary=(Weight=1,Axis=(X=0,Y=0,Z=1),Target=(X=1,Y=0,Z=0),Kind=Location,Space=(Type=Bone,Name="None"))
    Secondary=(Weight=1,Axis=(X=1,Y=0,Z=0),Target=(X=1,Y=0,Z=0),Kind=Direction,Space=(Type=Bone,Name="None"))
    Weight=1
    Secondary.Target=@"Get Main Leg PV".Pole Vector
    Primary.Target=@n5.Transform.Translation
    InputTransform.Translation=@n6.Transform.Translation
   )
   RigVMFunction_MathVectorDistance = &RigVMFunction_MathVectorDistance(A=@n5.Transform.Translation, B=@n6.Transform.Translation)
   Multiply = &RigVMFunction_MathDoubleMul(B=0.5, A=@RigVMFunction_MathVectorDistance.Result)
   RigVMVariableNode(
    Value=&RigVMFunction_MathTransformMakeAbsolute(Local=(Rotation=(X=0,Y=0,Z=0,W=1),Translation=(X=0,Y=0,Z=0),Scale3D=(X=1,Y=1,Z=1)), Parent=@n7.Result, Local.Translation.Z=@Multiply.Result) {template="Make Absolute::Execute(in Local,in Parent,out Global)"}.Global
    ExecuteContext=@RigVMFunction_Sequence_1.A
   ) {variable=Control Null Transform}
   Create Name_2 = RigVMFunctionReferenceNode(Tokens=("Knee Rotate Up Vector"), ExecuteContext=@RigVMFunction_Sequence_1.C)
   > n8 = &RigUnit_HierarchyAddNull(
    Space=LocalSpace
    Transform=(Rotation=(X=0,Y=0,Z=0,W=1),Translation=(X=10,Y=0,Z=0),Scale3D=(X=1,Y=1,Z=1))
    Name=@"Create Name_2".Name
    Parent=RigVMVariableNode() {variable=Knee Rotate Control}.Value
   )
   > RigVMVariableNode(Value=@n8.Item) {variable=Knee Rotate Up Vector}
   Create Name_3 = RigVMFunctionReferenceNode(Tokens=("Knee PV Null"), ExecuteContext=@RigVMFunction_Sequence_1.D)
   > n9 = &RigUnit_HierarchyAddNull(
    Space=LocalSpace
    Transform=(Rotation=(X=0,Y=0,Z=0,W=1),Translation=(X=0,Y=20,Z=0),Scale3D=(X=1,Y=1,Z=1))
    Name=@"Create Name_3".Name
    Parent=&RigUnit_HierarchyGetParent(bDefaultParent=True, Child=RigVMVariableNode() {variable=Knee Rotate Control Null}.Value).Parent
   )
   > Create Name_4 = RigVMFunctionReferenceNode(Tokens=("Knee PV"))
   > Set Side Color = RigVMFunctionReferenceNode()
   > n10 = &RigUnit_HierarchyAddControlTransform(
    InitialValue=(Rotation=(X=0,Y=0,Z=0,W=1),Translation=(X=0,Y=0,Z=0),Scale3D=(X=1,Y=1,Z=1))
    OffsetSpace=LocalSpace
    OffsetTransform=(Rotation=(X=0,Y=0,Z=0,W=1),Translation=(X=0,Y=0,Z=0),Scale3D=(X=1,Y=1,Z=1))
    Settings="(InitialSpace=LocalSpace,bUsePreferredRotationOrder=False,PreferredRotationOrder=YZX,Limits=(LimitTranslationX=(bMinimum=False,bMaximum=False),LimitTranslationY=(bMinimum=False,bMaximum=False),LimitTranslationZ=(bMinimum=False,bMaximum=False),LimitPitch=(b ...(truncated)"
    Name=@"Create Name_4".Name
    Settings.Shape.Color=@"Set Side Color".Color
    Parent=@n9.Item
   ) {template="SpawnControl::Execute(in InitialValue,in Settings,in OffsetTransform,in OffsetSpace,in Parent,in Name,out Item)"}
   > RigVMVariableNode(Value=@n10.Item) {variable=Knee PV}
   RigVMFunctionReturnNode(ExecuteContext=@RigVMFunction_Sequence.B)
  ]}
 New Function_ContainedGraph {nodes [
   n1 = RigVMFunctionEntryNode()
   RigVMFunction_Sequence = &RigVMFunction_Sequence(ExecuteContext=@n1.ExecuteContext)
   n2 = &RigVMFunction_Sequence(ExecuteContext=@RigVMFunction_Sequence.A)
   n3 = RigVMVariableNode() {variable=Ankle Connector}
   n4 = &RigUnit_GetTransform(bInitial=true, Item=RigVMVariableNode() {variable=Hock Rotate Control}.Value)
   n5 = RigVMVariableNode() {variable=Rig Up Vector}
   RigVMVariableNode(
    Value=&RigVMFunction_MathTransformMakeRelative(
     Parent=&RigUnit_GetTransform(bInitial=true, Item=@n3.Value).Transform
     Global=&RigVMFunction_MathTransformMakeAbsolute(Local=(Rotation=(X=0,Y=0,Z=0,W=1),Translation=(X=0,Y=0,Z=0),Scale3D=(X=1,Y=1,Z=1)), Parent=@n4.Transform, Local.Translation.X=@n5.Value.X) {template="Make Absolute::Execute(in Local,in Parent,out Global)"}.Global
    ).Local
    ExecuteContext=@n2.A
   ) {variable=PV Local Transform}
   n6 = RigVMVariableNode() {variable=Hip Socket Connector}
   n7 = RigVMVariableNode() {variable=Ankle Connector}
   n8 = &RigVMFunction_MathVectorSub(
    B=&RigUnit_GetTransform(bInitial=False, Item=RigVMVariableNode() {variable=Ball Connector}.Value).Transform.Translation
    A=&RigVMFunction_MathTransformMakeAbsolute(
     Parent=&RigUnit_GetTransform(bInitial=False, Item=@n7.Value).Transform
     Local=RigVMVariableNode() {variable=PV Local Transform}.Value
    ) {template="Make Absolute::Execute(in Local,in Parent,out Global)"}.Global.Translation
   ) {template="Subtract::Execute(in A,in B,out Result)"}
   n9 = &RigVMFunction_MathVectorMul(B=(X=-1,Y=-1,Z=-1), A=@n8.Result) {template="Multiply::Execute(in A,in B,out Result)"}
   n10 = RigVMVariableNode() {variable=Right Side}
   Set Transform_1 = /Script/ControlRig.RigUnit_SetTranslation(
    Space=GlobalSpace
    Weight=1
    bInitial=False
    bPropagateToChildren=True
    Value=&RigVMFunction_MathVectorAdd(
     B=&RigUnit_GetTransform(bInitial=False, Item=@n6.Value).Transform.Translation
     A=RigVMDispatchNode(True=@n9.Result, Condition=@n10.Value, False=@n8.Result) {template="DISPATCH_RigVMDispatch_If(in Condition,in True,in False,out Result)"}.Result
    ) {template="Add::Execute(in A,in B,out Result)"}.Result
    Item=RigVMVariableNode() {variable=Main PV Control}.Value
    ExecutePin=@n2.B
   ) {is=struct RigVMUnitNode, template="Set Transform::Execute(in Item,in Space,in bInitial,in Value,in Weight,in bPropagateToChildren)"}
   RigVMFunctionReturnNode(ExecuteContext=@RigVMFunction_Sequence.B)
  ]}
 New Function_ContainedGraph {nodes [
   n1 = RigVMFunctionEntryNode()
   RigVMFunction_Sequence = &RigVMFunction_Sequence(ExecuteContext=@n1.ExecuteContext)
   RigVMFunction_Sequence_1 = RigVMAggregateNode(ExecuteContext=@RigVMFunction_Sequence.A)
   Get Main Leg PV_3 = RigVMFunctionReferenceNode(Initial=False)
   n2 = RigVMVariableNode() {variable=Ankle Backwards Transform}
   n3 = RigVMVariableNode() {variable=Hip Socket Backwards Transform}
   n4 = &RigUnit_AimBoneMath(
    DebugSettings=(bEnabled=False,Scale=10,WorldOffset=(Rotation=(X=0,Y=0,Z=0,W=1),Translation=(X=0,Y=0,Z=0),Scale3D=(X=1,Y=1,Z=1)))
    InputTransform=(Rotation=(X=0,Y=0,Z=0,W=1),Translation=(X=0,Y=0,Z=0),Scale3D=(X=1,Y=1,Z=1))
    Primary=(Weight=1,Axis=(X=0,Y=0,Z=-1),Target=(X=1,Y=0,Z=0),Kind=Location,Space=(Type=Bone,Name="None"))
    Secondary=(Weight=1,Axis=(X=1,Y=0,Z=0),Target=(X=1,Y=0,Z=0),Kind=Direction,Space=(Type=Bone,Name="None"))
    Weight=1
    Secondary.Target=@"Get Main Leg PV_3".Pole Vector
    Primary.Target=@n2.Value.Translation
    InputTransform.Translation=@n3.Value.Translation
   )
   n5 = &RigVMFunction_MathDoubleMul(B=-0.5, A=&RigVMFunction_MathVectorDistance(B=@n2.Value.Translation, A=@n3.Value.Translation).Result)
   &RigUnit_SetTransform(
    Weight=1
    bInitial=False
    Item=&RigUnit_HierarchyGetParent(bDefaultParent=True, Child=RigVMVariableNode() {variable=Knee Rotate Control Null}.Value).Parent
    Value=&RigVMFunction_MathTransformMakeAbsolute(
     Local=(Rotation=(X=0,Y=0,Z=0,W=1),Translation=(X=0,Y=0,Z=0),Scale3D=(X=1,Y=1,Z=1))
     Parent=&RigVMFunction_MathTransformMakeAbsolute(Local=(Rotation=(X=0,Y=0,Z=0,W=1),Translation=(X=0,Y=0,Z=0),Scale3D=(X=1,Y=1,Z=1)), Parent=@n4.Result, Local.Translation.Z=@n5.Result) {template="Make Absolute::Execute(in Local,in Parent,out Global)"}.Global
    ) {template="Make Absolute::Execute(in Local,in Parent,out Global)"}.Global
    ExecutePin=@RigVMFunction_Sequence_1.A
   )
   Get Main Leg PV_4 = RigVMFunctionReferenceNode(Initial=False)
   n6 = RigVMVariableNode() {variable=Hip Socket Input Transform}
   n7 = RigVMVariableNode() {variable=Ankle Input Transform}
   n8 = &RigUnit_AimBoneMath(
    DebugSettings=(bEnabled=False,Scale=10,WorldOffset=(Rotation=(X=0,Y=0,Z=0,W=1),Translation=(X=0,Y=0,Z=0),Scale3D=(X=1,Y=1,Z=1)))
    InputTransform=(Rotation=(X=0,Y=0,Z=0,W=1),Translation=(X=0,Y=0,Z=0),Scale3D=(X=1,Y=1,Z=1))
    Primary=(Weight=1,Axis=(X=0,Y=0,Z=-1),Target=(X=1,Y=0,Z=0),Kind=Location,Space=(Type=Bone,Name="None"))
    Secondary=(Weight=1,Axis=(X=1,Y=0,Z=0),Target=(X=1,Y=0,Z=0),Kind=Direction,Space=(Type=Bone,Name="None"))
    Weight=1
    Secondary.Target=@"Get Main Leg PV_4".Pole Vector
    InputTransform.Translation=@n6.Value.Translation
    Primary.Target=@n7.Value.Translation
   )
   n9 = &RigVMFunction_MathDoubleMul(B=-0.5, A=&RigVMFunction_MathVectorDistance(A=@n6.Value.Translation, B=@n7.Value.Translation).Result)
   &RigUnit_SetTransform(
    Weight=1
    bInitial=False
    Value=&RigVMFunction_MathTransformMakeAbsolute(
     Local=(Rotation=(X=0,Y=0,Z=0,W=1),Translation=(X=0,Y=20,Z=0),Scale3D=(X=1,Y=1,Z=1))
     Parent=&RigVMFunction_MathTransformMakeAbsolute(Local=(Rotation=(X=0,Y=0,Z=0,W=1),Translation=(X=0,Y=0,Z=0),Scale3D=(X=1,Y=1,Z=1)), Parent=@n8.Result, Local.Translation.Z=@n9.Result) {template="Make Absolute::Execute(in Local,in Parent,out Global)"}.Global
    ) {template="Make Absolute::Execute(in Local,in Parent,out Global)"}.Global
    Item=RigVMVariableNode() {variable=Knee PV}.Value
    ExecutePin=@RigVMFunction_Sequence_1.B
   )
   n10 = &RigUnit_HierarchyGetParent(bDefaultParent=True, Child=RigVMVariableNode() {variable=Knee Rotate Control Null}.Value)
   &RigUnit_SetTransform(
    Space=LocalSpace
    Weight=1
    bInitial=False
    Value=&RigUnit_GetTransform(Space=LocalSpace, bInitial=true, Item=@n10.Parent).Transform
    Item=@n10.Parent
    ExecutePin=@RigVMFunction_Sequence_1.C
   )
   RigVMFunctionReturnNode(ExecuteContext=@RigVMFunction_Sequence.B)
   n11 = RigVMRerouteNode(Value=())
   RigVMFunction_MathVectorDistance = &RigVMFunction_MathVectorDistance(
    A=RigVMVariableNode() {variable=Hip Socket Rig Transform}.Value.Translation
    B=RigVMVariableNode() {variable=Ankle Rig Transform}.Value.Translation
   )
   Get Ik Angles = RigVMFunctionReferenceNode(
    "B 180 "=false
    C 180=true
    Negate B=false
    Negate C=false
    Length C=@RigVMFunction_MathVectorDistance.Result
    Length A=RigVMVariableNode() {variable=Hip Socket Bone Length}.Value
    Length B=RigVMVariableNode() {variable=Knee Bone Length}.Value
   )
   RigVMVariableNode(Value=@"Get Ik Angles".Angle B, ExecuteContext=@n11.Value) {variable=Hip Socket Angle}
   > RigVMVariableNode(Value=RigVMDispatchNode(
     True=0
     Condition=RigVMDispatchNode(B=180, A=@"Get Ik Angles".Angle C) {template="DISPATCH_RigVMDispatch_CoreEquals(in A,in B,out Result)"}.Result
     False=@"Get Ik Angles".Angle C
    ) {template="DISPATCH_RigVMDispatch_If(in Condition,in True,in False,out Result)"}.Result) {variable=Knee Angle}
   n12 = RigVMRerouteNode(Value=())
   n13 = RigVMVariableNode() {variable=Knee Connector}
   n14 = RigVMVariableNode() {variable=Knee Rotate Control}
   RigVMVariableNode(
    Value=&RigVMFunction_MathTransformMakeRelative(
     Parent=&RigUnit_GetTransform(bInitial=true, Item=@n13.Value).Transform
     Global=&RigUnit_GetTransform(bInitial=true, Item=@n14.Value).Transform
    ).Local
    ExecuteContext=@n12.Value
   ) {variable=Knee Rotate Local Transform}
   n15 = RigVMRerouteNode(Value=())
   n16 = &RigUnit_AimBoneMath(
    DebugSettings=(bEnabled=False,Scale=10,WorldOffset=(Rotation=(X=0,Y=0,Z=0,W=1),Translation=(X=0,Y=0,Z=0),Scale3D=(X=1,Y=1,Z=1)))
    InputTransform=(Rotation=(X=0,Y=0,Z=0,W=1),Translation=(X=0,Y=0,Z=0),Scale3D=(X=1,Y=1,Z=1))
    Primary=(Weight=1,Axis=(X=1,Y=0,Z=0),Target=(X=1,Y=0,Z=0),Kind=Location,Space=(Type=Bone,Name="None"))
    Secondary=(Weight=1,Axis=(X=0,Y=0,Z=1),Target=(X=1,Y=0,Z=0),Kind=Direction,Space=(Type=Bone,Name="None"))
    Weight=1
   )
   RigVMVariableNode(Value=@n16.Result, ExecuteContext=@n15.Value) {variable=Hip Socket Rig Transform}
   n17 = &RigUnit_AimBoneMath(
    DebugSettings=(bEnabled=False,Scale=10,WorldOffset=(Rotation=(X=0,Y=0,Z=0,W=1),Translation=(X=0,Y=0,Z=0),Scale3D=(X=1,Y=1,Z=1)))
    InputTransform=(Rotation=(X=0,Y=0,Z=0,W=1),Translation=(X=0,Y=0,Z=0),Scale3D=(X=1,Y=1,Z=1))
    Primary=(Weight=1,Axis=(X=1,Y=0,Z=0),Target=(X=1,Y=0,Z=0),Kind=Location,Space=(Type=Bone,Name="None"))
    Secondary=(Weight=1,Axis=(X=0,Y=0,Z=1),Target=(X=1,Y=0,Z=0),Kind=Direction,Space=(Type=Bone,Name="None"))
    Weight=1
   )
   > RigVMVariableNode(Value=@n17.Result) {variable=Knee Rig Transform}
   RigVMFunction_Sequence_2 = RigVMAggregateNode(ExecuteContext=())
   &RigUnit_SetTransform(
    Weight=1
    bInitial=False
    Value=RigVMVariableNode() {variable=Knee Control Null Transform}.Value
    Item=RigVMVariableNode() {variable=Knee Rotate Control Null}.Value
    ExecutePin=@RigVMFunction_Sequence_2.E
   )
   > &RigUnit_SetTransform(
    Weight=1
    bInitial=False
    Value=RigVMVariableNode() {variable=Knee Control Transform}.Value
    Item=RigVMVariableNode() {variable=Knee Rotate Control}.Value
   )
   n18 = &RigVMFunction_MathTransformMakeAbsolute(
    Parent=&RigUnit_GetTransform(bInitial=false, Item=RigVMVariableNode() {variable=Knee Connector}.Value).Transform
    Local=RigVMVariableNode() {variable=Knee Rotate Local Transform}.Value
   ) {template="Make Absolute::Execute(in Local,in Parent,out Global)"}
   RigVMVariableNode(
    Value=&RigVMFunction_MathVectorSub(
     A=&RigVMFunction_MathTransformMakeAbsolute(Local=(Rotation=(X=0,Y=0,Z=0,W=1),Translation=(X=10,Y=0,Z=0),Scale3D=(X=1,Y=1,Z=1)), Parent=@n18.Global) {template="Make Absolute::Execute(in Local,in Parent,out Global)"}.Global.Translation
     B=@n18.Global.Translation
    ) {template="Subtract::Execute(in A,in B,out Result)"}.Result
    ExecuteContext=@RigVMFunction_Sequence_2.C
   ) {variable=Knee Ik PV}
   Get Main Leg PV_2 = RigVMFunctionReferenceNode(Initial=False)
   n19 = &RigUnit_GetTransform(bInitial=False, Item=RigVMVariableNode() {variable=Hip Socket Connector}.Value)
   n20 = &RigUnit_GetTransform(bInitial=False, Item=RigVMVariableNode() {variable=Ankle Connector}.Value)
   n21 = &RigUnit_AimBoneMath(
    DebugSettings=(bEnabled=False,Scale=10,WorldOffset=(Rotation=(X=0,Y=0,Z=0,W=1),Translation=(X=0,Y=0,Z=0),Scale3D=(X=1,Y=1,Z=1)))
    InputTransform=(Rotation=(X=0,Y=0,Z=0,W=1),Translation=(X=0,Y=0,Z=0),Scale3D=(X=1,Y=1,Z=1))
    Primary=(Weight=1,Axis=(X=0,Y=0,Z=-1),Target=(X=1,Y=0,Z=0),Kind=Location,Space=(Type=Bone,Name="None"))
    Secondary=(Weight=1,Axis=(X=1,Y=0,Z=0),Target=(X=1,Y=0,Z=0),Kind=Direction,Space=(Type=Bone,Name="None"))
    Weight=1
    Secondary.Target=@"Get Main Leg PV_2".Pole Vector
    InputTransform.Translation=@n19.Transform.Translation
    Primary.Target=@n20.Transform.Translation
   )
   Multiply = &RigVMFunction_MathDoubleMul(B=-0.5, A=&RigVMFunction_MathVectorDistance(A=@n19.Transform.Translation, B=@n20.Transform.Translation).Result)
   RigVMVariableNode(
    Value=&RigVMFunction_MathTransformMakeAbsolute(Local=(Rotation=(X=0,Y=0,Z=0,W=1),Translation=(X=0,Y=0,Z=0),Scale3D=(X=1,Y=1,Z=1)), Parent=@n21.Result, Local.Translation.Z=@Multiply.Result) {template="Make Absolute::Execute(in Local,in Parent,out Global)"}.Global
    ExecuteContext=@RigVMFunction_Sequence_2.B
   ) {variable=Knee Control Null Transform}
   n22 = &RigUnit_GetTransform(bInitial=False, Item=RigVMVariableNode() {variable=Hip Socket Connector}.Value)
   n23 = &RigUnit_GetTransform(bInitial=False, Item=RigVMVariableNode() {variable=Ankle Connector}.Value)
   n24 = &RigUnit_AimBoneMath(
    DebugSettings=(bEnabled=False,Scale=10,WorldOffset=(Rotation=(X=0,Y=0,Z=0,W=1),Translation=(X=0,Y=0,Z=0),Scale3D=(X=1,Y=1,Z=1)))
    InputTransform=(Rotation=(X=0,Y=0,Z=0,W=1),Translation=(X=0,Y=0,Z=0),Scale3D=(X=1,Y=1,Z=1))
    Primary=(Weight=1,Axis=(X=0,Y=0,Z=-1),Target=(X=1,Y=0,Z=0),Kind=Location,Space=(Type=Bone,Name="None"))
    Secondary=(Weight=1,Axis=(X=1,Y=0,Z=0),Target=(X=1,Y=0,Z=0),Kind=Direction,Space=(Type=Bone,Name="None"))
    Weight=1
    InputTransform.Translation=@n22.Transform.Translation
    Primary.Target=@n23.Transform.Translation
    Secondary.Target=RigVMVariableNode() {variable=Knee Ik PV}.Value
   )
   n25 = &RigVMFunction_MathVectorDistance(A=@n22.Transform.Translation, B=@n23.Transform.Translation)
   RigVMVariableNode(
    Value=&RigVMFunction_MathTransformMakeAbsolute(
     Local=(Rotation=(X=0,Y=0,Z=0,W=1),Translation=(X=0,Y=0,Z=0),Scale3D=(X=1,Y=1,Z=1))
     Parent=@n24.Result
     Local.Translation.Z=&RigVMFunction_MathDoubleMul(B=0.5, A=@n25.Result).Result
    ) {template="Make Absolute::Execute(in Local,in Parent,out Global)"}.Global
    ExecuteContext=@RigVMFunction_Sequence_2.D
   ) {variable=Knee Control Transform}
   Get Main Leg PV_1 = RigVMFunctionReferenceNode(Initial=False)
   &RigUnit_HierarchyGetParent(bDefaultParent=True, Child=RigVMVariableNode() {variable=Knee PV}.Value)
   n26 = &RigUnit_AimBoneMath(
    DebugSettings=(bEnabled=False,Scale=10,WorldOffset=(Rotation=(X=0,Y=0,Z=0,W=1),Translation=(X=0,Y=0,Z=0),Scale3D=(X=1,Y=1,Z=1)))
    InputTransform=(Rotation=(X=0,Y=0,Z=0,W=1),Translation=(X=0,Y=0,Z=0),Scale3D=(X=1,Y=1,Z=1))
    Primary=(Weight=1,Axis=(X=1,Y=0,Z=0),Target=(X=1,Y=0,Z=0),Kind=Location,Space=(Type=Bone,Name="None"))
    Secondary=(Weight=1,Axis=(X=0,Y=0,Z=1),Target=(X=1,Y=0,Z=0),Kind=Direction,Space=(Type=Bone,Name="None"))
    Weight=1
    InputTransform.Translation=RigVMVariableNode() {variable=Hip Socket Rig Transform}.Value.Translation
    Secondary.Target=RigVMVariableNode() {variable=Knee Ik PV}.Value
    Primary.Target=RigVMVariableNode() {variable=Ankle Rig Transform}.Value.Translation
   )
   GetAnimationChannelFromItem = &RigUnit_GetFloatAnimationChannelFromItem(bInitial=False, Item=RigVMVariableNode() {variable=Knee Twist}.Value) {template="GetAnimationChannelFromItem::Execute(out Value,in Item,in bInitial)"}
   RigVMFunction_MathQuaternionFromEuler = &RigVMFunction_MathQuaternionFromEuler(Euler=(X=0,Y=0,Z=0), Euler.X=@GetAnimationChannelFromItem.Value)
   &RigVMFunction_MathTransformMakeAbsolute(
    Local=(Rotation=(X=0,Y=0,Z=0,W=1),Translation=(X=0,Y=0,Z=0),Scale3D=(X=1,Y=1,Z=1))
    Parent=&RigVMFunction_MathTransformMakeAbsolute(
     Local=(Rotation=(X=0,Y=0,Z=0,W=1),Translation=(X=0,Y=0,Z=0),Scale3D=(X=1,Y=1,Z=1))
     Parent=@n26.Result
     Local.Rotation=@RigVMFunction_MathQuaternionFromEuler.Result
    ) {template="Make Absolute::Execute(in Local,in Parent,out Global)"}.Global
    Local.Rotation=&RigVMFunction_MathQuaternionFromEuler(Euler=(X=0,Y=0,Z=0), Euler.Z=RigVMVariableNode() {variable=Hip Socket Angle}.Value).Result
   ) {template="Make Absolute::Execute(in Local,in Parent,out Global)"}
   n27 = &RigVMFunction_MathTransformMakeAbsolute(
    Local=(Rotation=(X=0,Y=0,Z=0,W=1),Translation=(X=0,Y=0,Z=0),Scale3D=(X=1,Y=1,Z=1))
    Parent=(Rotation=(X=0,Y=0,Z=0,W=1),Translation=(X=0,Y=0,Z=0),Scale3D=(X=1,Y=1,Z=1))
    Local.Translation.X=RigVMVariableNode() {variable=Hip Socket Bone Length}.Value
   ) {template="Make Absolute::Execute(in Local,in Parent,out Global)"}
   &RigVMFunction_MathTransformMakeAbsolute(
    Local=(Rotation=(X=0,Y=0,Z=0,W=1),Translation=(X=0,Y=0,Z=0),Scale3D=(X=1,Y=1,Z=1))
    Parent=@n27.Global
    Local.Rotation=&RigVMFunction_MathQuaternionFromEuler(Euler=(X=0,Y=0,Z=0)).Result
   ) {template="Make Absolute::Execute(in Local,in Parent,out Global)"}
   &RigVMFunction_MathTransformMakeAbsolute(
    Local=(Rotation=(X=0,Y=0,Z=0,W=1),Translation=(X=0,Y=0,Z=0),Scale3D=(X=1,Y=1,Z=1))
    Parent=(Rotation=(X=0,Y=0,Z=0,W=1),Translation=(X=0,Y=0,Z=0),Scale3D=(X=1,Y=1,Z=1))
    Local.Translation.X=RigVMVariableNode() {variable=Knee Bone Length}.Value
   ) {template="Make Absolute::Execute(in Local,in Parent,out Global)"}
   RigVMRerouteNode(Value=())
   RigVMRerouteNode(Value=())
   RigVMFunction_DebugLineStripNoSpace = &RigVMFunction_DebugLineStripNoSpace(
    Color=(R=0,G=0,B=1,A=1)
    DebugDrawSettings=(DepthPriority=SDPG_Foreground,Lifetime=-1)
    ExecutePin=()
    Points=((X=0,Y=0,Z=0),(X=0,Y=0,Z=0),(X=0,Y=0,Z=0),(X=0,Y=0,Z=0))
    Thickness=0.5
    WorldOffset=(Rotation=(X=0,Y=0,Z=0,W=1),Translation=(X=0,Y=0,Z=0),Scale3D=(X=1,Y=1,Z=1))
    bEnabled=RigVMVariableNode() {variable=DebugOnOff}.Value
   )
   &RigVMFunction_DebugLineStripNoSpace(
    Color=(R=1,G=0,B=0,A=1)
    DebugDrawSettings=(DepthPriority=SDPG_Foreground,Lifetime=-1)
    ExecutePin=()
    Points=((X=0,Y=0,Z=0),(X=0,Y=0,Z=0),(X=0,Y=0,Z=0))
    Thickness=1
    WorldOffset=(Rotation=(X=0,Y=0,Z=0,W=1),Translation=(X=0,Y=0,Z=0),Scale3D=(X=1,Y=1,Z=1))
    bEnabled=false
    Points.1=RigVMVariableNode() {variable=Hip Socket Rig Transform}.Value.Translation
    Points.2=RigVMVariableNode() {variable=Ankle Rig Transform}.Value.Translation
    Points.0=RigVMVariableNode() {variable=Knee Rig Transform}.Value.Translation
   )
   &RigVMFunction_MathVectorSub(
    B=&RigUnit_GetTransform(bInitial=False, Item=RigVMVariableNode() {variable=Knee Rotate Control}.Value).Transform.Translation
    A=&RigUnit_GetTransform(bInitial=False, Item=RigVMVariableNode() {variable=Knee Rotate Up Vector}.Value).Transform.Translation
   ) {template="Subtract::Execute(in A,in B,out Result)"}
   RigVMVariableNode(ExecuteContext=(), Value=(Rotation=(X=0,Y=0,Z=0,W=1),Translation=(X=0,Y=0,Z=0),Scale3D=(X=1,Y=1,Z=1))) {variable=Knee Control Null Transform}
   RigVMVariableNode() {variable=Knee Twist Offset}
  ]}
 New Function_ContainedGraph {nodes [
   n1 = RigVMFunctionEntryNode()
   RigVMFunction_Sequence = &RigVMFunction_Sequence(ExecuteContext=@n1.ExecuteContext)
   RigVMFunctionReturnNode(ExecuteContext=@RigVMFunction_Sequence.B)
   Set Transform_4 = &RigUnit_SetTranslation(
    Space=GlobalSpace
    Weight=1
    bInitial=False
    bPropagateToChildren=True
    Value=&RigUnit_GetTransform(bInitial=False, Item=RigVMVariableNode() {variable=Hip Socket Connector}.Value).Transform.Translation
    Item=RigVMVariableNode() {variable=Hip Socket Control}.Value
    ExecutePin=@RigVMFunction_Sequence.A
   ) {template="Set Transform::Execute(in Item,in Space,in bInitial,in Value,in Weight,in bPropagateToChildren)"}
   RigVMFunction_Sequence_1 = RigVMAggregateNode(ExecuteContext=())
  ]}
 New Function_ContainedGraph {nodes [
   n1 = RigVMFunctionEntryNode()
   RigVMFunction_Sequence = &RigVMFunction_Sequence(ExecuteContext=@n1.ExecuteContext)
   RigVMFunction_Sequence_1 = RigVMAggregateNode(ExecuteContext=@RigVMFunction_Sequence.A)
   n2 = &RigUnit_GetTransform(
    bInitial=False
    Item=&RigUnit_HierarchyGetParent(bDefaultParent=True, Child=RigVMVariableNode() {variable=Ik Control}.Value).Parent
   )
   RigVMVariableNode(
    Value=&RigVMFunction_MathVectorSub(
     A=&RigVMFunction_MathTransformMakeAbsolute(Local=(Rotation=(X=0,Y=0,Z=0,W=1),Translation=(X=0,Y=0,Z=10),Scale3D=(X=1,Y=1,Z=1)), Parent=@n2.Transform) {template="Make Absolute::Execute(in Local,in Parent,out Global)"}.Global.Translation
     B=@n2.Transform.Translation
    ) {template="Subtract::Execute(in A,in B,out Result)"}.Result
    ExecuteContext=@RigVMFunction_Sequence_1.A
   ) {variable=Ik Aim Vector}
   n3 = &RigUnit_GetTransform(bInitial=False, Item=RigVMVariableNode() {variable=Ball Connector}.Value)
   Subtract = &RigVMFunction_MathVectorSub(
    A=&RigVMFunction_MathTransformMakeAbsolute(Local=(Rotation=(X=0,Y=0,Z=0,W=1),Translation=(X=0,Y=0,Z=10),Scale3D=(X=1,Y=1,Z=1)), Parent=@n3.Transform) {template="Make Absolute::Execute(in Local,in Parent,out Global)"}.Global.Translation
    B=@n3.Transform.Translation
   ) {template="Subtract::Execute(in A,in B,out Result)"}
   RigVMVariableNode(Value=@Subtract.Result, ExecuteContext=@RigVMFunction_Sequence_1.B) {variable=Ik Up Vector}
   n4 = &RigUnit_GetTransform(bInitial=False, Item=RigVMVariableNode() {variable=Ball Connector}.Value)
   Add = &RigVMFunction_MathVectorAdd(A=@n4.Transform.Translation, B=RigVMVariableNode() {variable=Ik Aim Vector}.Value) {template="Add::Execute(in A,in B,out Result)"}
   n5 = &RigUnit_AimBoneMath(
    DebugSettings=(bEnabled=False,Scale=10,WorldOffset=(Rotation=(X=0,Y=0,Z=0,W=1),Translation=(X=0,Y=0,Z=0),Scale3D=(X=1,Y=1,Z=1)))
    InputTransform=(Rotation=(X=0,Y=0,Z=0,W=1),Translation=(X=0,Y=0,Z=0),Scale3D=(X=1,Y=1,Z=1))
    Primary=(Weight=1,Axis=(X=0,Y=0,Z=1),Target=(X=1,Y=0,Z=0),Kind=Location,Space=(Type=Bone,Name="None"))
    Secondary=(Weight=1,Axis=(X=1,Y=0,Z=0),Target=(X=1,Y=0,Z=0),Kind=Direction,Space=(Type=Bone,Name="None"))
    Weight=1
    Primary.Target=@Add.Result
    InputTransform.Translation=@n4.Transform.Translation
    Secondary.Target=RigVMVariableNode() {variable=Ik Up Vector}.Value
   )
   RigVMVariableNode(Value=@n5.Result, ExecuteContext=@RigVMFunction_Sequence_1.C) {variable=Ik Transform}
   &RigUnit_SetTransform(
    Weight=1
    bInitial=False
    Item=RigVMVariableNode() {variable=Ik Control}.Value
    Value=RigVMVariableNode() {variable=Ik Transform}.Value
    ExecutePin=@RigVMFunction_Sequence_1.D
   )
   RigVMFunctionReturnNode(ExecuteContext=@RigVMFunction_Sequence.B)
  ]}
 New Function_ContainedGraph {nodes [
   n1 = RigVMFunctionEntryNode()
   RigVMFunction_Sequence = &RigVMFunction_Sequence(ExecuteContext=@n1.ExecuteContext)
   n2 = &RigVMFunction_Sequence(ExecuteContext=@RigVMFunction_Sequence.A)
   n3 = &RigUnit_AimBoneMath(
    DebugSettings=(bEnabled=False,Scale=10,WorldOffset=(Rotation=(X=0,Y=0,Z=0,W=1),Translation=(X=0,Y=0,Z=0),Scale3D=(X=1,Y=1,Z=1)))
    InputTransform=(Rotation=(X=0,Y=0,Z=0,W=1),Translation=(X=0,Y=0,Z=0),Scale3D=(X=1,Y=1,Z=1))
    Primary=(Weight=1,Axis=(X=0,Y=1,Z=0),Target=(X=1,Y=0,Z=0),Kind=Location,Space=(Type=Bone,Name="None"))
    Secondary=(Weight=1,Axis=(X=1,Y=0,Z=0),Target=(X=1,Y=0,Z=0),Kind=Direction,Space=(Type=Bone,Name="None"))
    Weight=1
    InputTransform.Translation=&RigUnit_GetTransform(bInitial=False, Item=RigVMVariableNode() {variable=Ball Connector}.Value).Transform.Translation
    Primary.Target=&RigUnit_GetTransform(bInitial=False, Item=RigVMVariableNode() {variable=Toe Connector}.Value).Transform.Translation
    Secondary.Space=RigVMVariableNode() {variable=Ik Control}.Value
   )
   &RigUnit_SetTransform(Weight=1, bInitial=False, Item=RigVMVariableNode() {variable=Ball Fk Control}.Value, ExecutePin=@n2.A, Value=@n3.Result)
   RigVMFunctionReturnNode(ExecuteContext=@RigVMFunction_Sequence.B)
  ]}
 New Function_ContainedGraph {nodes [
   n1 = RigVMFunctionEntryNode()
   RigVMFunction_Sequence = &RigVMFunction_Sequence(ExecuteContext=@n1.ExecuteContext)
   n2 = &RigVMFunction_Sequence(ExecuteContext=@RigVMFunction_Sequence.A)
   n3 = &RigUnit_AimBoneMath(
    DebugSettings=(bEnabled=False,Scale=10,WorldOffset=(Rotation=(X=0,Y=0,Z=0,W=1),Translation=(X=0,Y=0,Z=0),Scale3D=(X=1,Y=1,Z=1)))
    InputTransform=(Rotation=(X=0,Y=0,Z=0,W=1),Translation=(X=0,Y=0,Z=0),Scale3D=(X=1,Y=1,Z=1))
    Primary=(Weight=1,Axis=(X=0,Y=1,Z=0),Target=(X=1,Y=0,Z=0),Kind=Location,Space=(Type=Bone,Name="None"))
    Secondary=(Weight=1,Axis=(X=1,Y=0,Z=0),Target=(X=1,Y=0,Z=0),Kind=Direction,Space=(Type=Bone,Name="None"))
    Weight=1
    InputTransform.Translation=&RigUnit_GetTransform(bInitial=False, Item=RigVMVariableNode() {variable=Toe Connector}.Value).Transform.Translation
    Primary.Target=&RigUnit_GetTransform(bInitial=False, Item=RigVMVariableNode() {variable=Toe Tip Connector}.Value).Transform.Translation
    Secondary.Space=RigVMVariableNode() {variable=Ik Control}.Value
   )
   &RigUnit_SetTransform(Weight=1, bInitial=False, Item=RigVMVariableNode() {variable=Toe Fk Control}.Value, ExecutePin=@n2.A, Value=@n3.Result)
   RigVMFunctionReturnNode(ExecuteContext=@RigVMFunction_Sequence.B)
  ]}
 New Function_ContainedGraph {nodes [
   n1 = RigVMFunctionEntryNode()
   RigVMFunction_Sequence = &RigVMFunction_Sequence(ExecuteContext=@n1.ExecuteContext)
   n2 = &RigVMFunction_Sequence(ExecuteContext=@RigVMFunction_Sequence.A)
   RigVMFunction_MathBoolToInteger = &RigVMFunction_MathBoolToInteger(Value=@n1.Active)
   Switch = RigVMDispatchNode(ExecuteContext=@n2.A, Index=@RigVMFunction_MathBoolToInteger.Result) {template="DISPATCH_RigVMDispatch_SwitchInt32(in Index)"}
   n3 = &RigUnit_AimBoneMath(
    DebugSettings=(bEnabled=False,Scale=10,WorldOffset=(Rotation=(X=0,Y=0,Z=0,W=1),Translation=(X=0,Y=0,Z=0),Scale3D=(X=1,Y=1,Z=1)))
    InputTransform=(Rotation=(X=0,Y=0,Z=0,W=1),Translation=(X=0,Y=0,Z=0),Scale3D=(X=1,Y=1,Z=1))
    Primary=(Weight=1,Axis=(X=0,Y=0,Z=1),Target=(X=1,Y=0,Z=0),Kind=Location,Space=(Type=Bone,Name="None"))
    Secondary=(Weight=1,Axis=(X=1,Y=0,Z=0),Target=(X=1,Y=0,Z=0),Kind=Direction,Space=(Type=Bone,Name="None"))
    Weight=1
    Primary.Target=&RigUnit_GetTransform(bInitial=False, Item=RigVMVariableNode() {variable=Ankle Connector}.Value).Transform.Translation
    InputTransform.Translation=&RigUnit_GetTransform(bInitial=False, Item=RigVMVariableNode() {variable=Ball Connector}.Value).Transform.Translation
    Secondary.Space=RigVMVariableNode() {variable=Ik Control}.Value
   )
   &RigUnit_SetTransform(Weight=1, bInitial=False, Value=@n3.Result, Item=RigVMVariableNode() {variable=Hock Rotate Aim Null}.Value, ExecutePin=@Switch.Cases.1)
   RigVMFunctionReturnNode(ExecuteContext=@RigVMFunction_Sequence.B)
   &RigUnit_GetTransform(bInitial=False, Item=RigVMVariableNode() {variable=Ik Point}.Value)
   RigVMVariableNode() {variable=Ankle Rig Transform}
  ]}
 New Function_ContainedGraph {nodes [
   n1 = RigVMFunctionEntryNode()
   RigVMFunction_Sequence = &RigVMFunction_Sequence(ExecuteContext=@n1.ExecuteContext)
   RigVMFunction_Sequence_1 = RigVMAggregateNode(ExecuteContext=@RigVMFunction_Sequence.A)
   Get Hip To Ik Control Aim = RigVMFunctionReferenceNode(Initial=true)
   n2 = RigVMVariableNode() {variable=Knee Connector}
   RigVMVariableNode(
    Value=&RigVMFunction_MathTransformMakeRelative(Parent=@"Get Hip To Ik Control Aim".Aim, Global=&RigUnit_GetTransform(bInitial=true, Item=@n2.Value).Transform).Local
    ExecuteContext=@RigVMFunction_Sequence_1.A
   ) {variable=Initial Knee Local Offset}
   RigVMFunctionReturnNode(ExecuteContext=@RigVMFunction_Sequence.B)
   n3 = &RigUnit_GetTransform(bInitial=false, Item=RigVMVariableNode() {variable=Hip Socket Connector}.Value)
   Subtract = &RigVMFunction_MathVectorSub(
    B=@n3.Transform.Translation
    A=&RigUnit_GetTransform(bInitial=false, Item=RigVMVariableNode() {variable=Knee Connector}.Value).Transform.Translation
   ) {template="Subtract::Execute(in A,in B,out Result)"}
   Get Hip To Ik Control Aim_1 = RigVMFunctionReferenceNode(Initial=false)
   n4 = RigVMVariableNode() {variable=Initial Knee Local Offset}
   RigVMFunction_MathVectorAngle = /Script/RigVM.RigVMFunction_MathVectorAngle(
    A=@Subtract.Result
    B=&RigVMFunction_MathVectorSub(
     B=@n3.Transform.Translation
     A=&RigVMFunction_MathTransformMakeAbsolute(Parent=@"Get Hip To Ik Control Aim_1".Aim, Local=@n4.Value) {template="Make Absolute::Execute(in Local,in Parent,out Global)"}.Global.Translation
    ) {template="Subtract::Execute(in A,in B,out Result)"}.Result
   ) {is=struct RigVMUnitNode, template="AngleBetween::Execute(in A,in B,out Result)"}
  ]}
 New Function_ContainedGraph {nodes [
   n1 = RigVMFunctionEntryNode()
   RigVMFunction_Sequence = &RigVMFunction_Sequence(ExecuteContext=@n1.ExecuteContext)
   n2 = &RigVMFunction_Sequence(ExecuteContext=@RigVMFunction_Sequence.A)
   Get Main Leg PV = RigVMFunctionReferenceNode(Initial=False)
   n3 = &RigUnit_AimBoneMath(
    DebugSettings=(bEnabled=False,Scale=10,WorldOffset=(Rotation=(X=0,Y=0,Z=0,W=1),Translation=(X=0,Y=0,Z=0),Scale3D=(X=1,Y=1,Z=1)))
    InputTransform=(Rotation=(X=0,Y=0,Z=0,W=1),Translation=(X=0,Y=0,Z=0),Scale3D=(X=1,Y=1,Z=1))
    Primary=(Weight=1,Axis=(X=0,Y=0,Z=1),Target=(X=1,Y=0,Z=0),Kind=Location,Space=(Type=Bone,Name="None"))
    Secondary=(Weight=1,Axis=(X=1,Y=0,Z=0),Target=(X=1,Y=0,Z=0),Kind=Direction,Space=(Type=Bone,Name="None"))
    Weight=1
    Secondary.Target=@"Get Main Leg PV".Pole Vector
    Primary.Target=&RigUnit_GetTransform(bInitial=False, Item=RigVMVariableNode() {variable=Ankle Connector}.Value).Transform.Translation
    InputTransform.Translation=&RigUnit_GetTransform(bInitial=False, Item=RigVMVariableNode() {variable=Ball Connector}.Value).Transform.Translation
   )
   &RigUnit_SetTransform(Weight=1, bInitial=False, Value=@n3.Result, Item=RigVMVariableNode() {variable=Hock Rotate Control}.Value, ExecutePin=@n2.A)
   RigVMFunctionReturnNode(ExecuteContext=@RigVMFunction_Sequence.B)
  ]}
 New Function_ContainedGraph {nodes [
   n1 = RigVMFunctionEntryNode()
   RigVMFunction_Sequence = &RigVMFunction_Sequence(ExecuteContext=@n1.ExecuteContext)
   RigVMFunction_Sequence_2 = RigVMAggregateNode(ExecuteContext=@RigVMFunction_Sequence.A)
   Get Main Leg PV_7 = RigVMFunctionReferenceNode(Initial=False)
   n2 = &RigUnit_AimBoneMath(
    DebugSettings=(bEnabled=False,Scale=10,WorldOffset=(Rotation=(X=0,Y=0,Z=0,W=1),Translation=(X=0,Y=0,Z=0),Scale3D=(X=1,Y=1,Z=1)))
    InputTransform=(Rotation=(X=0,Y=0,Z=0,W=1),Translation=(X=0,Y=0,Z=0),Scale3D=(X=1,Y=1,Z=1))
    Primary=(Weight=1,Axis=(X=0,Y=0,Z=1),Target=(X=1,Y=0,Z=0),Kind=Location,Space=(Type=Bone,Name="None"))
    Secondary=(Weight=1,Axis=(X=1,Y=0,Z=0),Target=(X=1,Y=0,Z=0),Kind=Direction,Space=(Type=Bone,Name="None"))
    Weight=1
    Secondary.Target=@"Get Main Leg PV_7".Pole Vector
    Primary.Target=RigVMVariableNode() {variable=Ankle Backwards Transform}.Value.Translation
    InputTransform.Translation=RigVMVariableNode() {variable=Ball Backwards Transform}.Value.Translation
   )
   &RigUnit_SetTransform(
    Weight=1
    bInitial=False
    Value=@n2.Result
    Item=&RigUnit_HierarchyGetParent(bDefaultParent=True, Child=RigVMVariableNode() {variable=Hock Rotate Control}.Value).Parent
    ExecutePin=@RigVMFunction_Sequence_2.A
   )
   Get Main Leg PV_4 = RigVMFunctionReferenceNode(Initial=False)
   n3 = &RigUnit_AimBoneMath(
    DebugSettings=(bEnabled=False,Scale=10,WorldOffset=(Rotation=(X=0,Y=0,Z=0,W=1),Translation=(X=0,Y=0,Z=0),Scale3D=(X=1,Y=1,Z=1)))
    InputTransform=(Rotation=(X=0,Y=0,Z=0,W=1),Translation=(X=0,Y=0,Z=0),Scale3D=(X=1,Y=1,Z=1))
    Primary=(Weight=1,Axis=(X=0,Y=0,Z=1),Target=(X=1,Y=0,Z=0),Kind=Location,Space=(Type=Bone,Name="None"))
    Secondary=(Weight=1,Axis=(X=1,Y=0,Z=0),Target=(X=1,Y=0,Z=0),Kind=Direction,Space=(Type=Bone,Name="None"))
    Weight=1
    Secondary.Target=@"Get Main Leg PV_4".Pole Vector
    Primary.Target=RigVMVariableNode() {variable=Ankle Input Transform}.Value.Translation
    InputTransform.Translation=RigVMVariableNode() {variable=Ball Input Transform}.Value.Translation
   )
   &RigUnit_SetTransform(
    Weight=1
    bInitial=False
    Value=@n3.Result
    Item=RigVMVariableNode() {variable=Hock Rotate Control}.Value
    ExecutePin=@RigVMFunction_Sequence_2.B
   )
   RigVMFunctionReturnNode(ExecuteContext=@RigVMFunction_Sequence.B)
  ]}
 New Function_ContainedGraph {nodes [
   n1 = RigVMFunctionEntryNode()
   RigVMFunction_Sequence = &RigVMFunction_Sequence(ExecuteContext=@n1.ExecuteContext)
   RigVMFunction_Sequence_1 = RigVMAggregateNode(ExecuteContext=@RigVMFunction_Sequence.A)
   RigVMVariableNode(
    Value=&RigUnit_GetTransform(bInitial=False, Item=RigVMVariableNode() {variable=Hip Socket Connector}.Value).Transform
    ExecuteContext=@RigVMFunction_Sequence_1.A
   ) {variable=Hip Socket Input Transform}
   > RigVMVariableNode(Value=&RigUnit_GetTransform(bInitial=False, Item=RigVMVariableNode() {variable=Knee Connector}.Value).Transform) {variable=Knee Input Transform}
   > RigVMVariableNode(Value=&RigUnit_GetTransform(bInitial=False, Item=RigVMVariableNode() {variable=Ankle Connector}.Value).Transform) {variable=Ankle Input Transform}
   > RigVMVariableNode(Value=&RigUnit_GetTransform(bInitial=False, Item=RigVMVariableNode() {variable=Ball Connector}.Value).Transform) {variable=Ball Input Transform}
   > RigVMVariableNode(Value=&RigUnit_GetTransform(bInitial=False, Item=RigVMVariableNode() {variable=Toe Connector}.Value).Transform) {variable=Toe Input Transform}
   > RigVMVariableNode(Value=&RigUnit_GetTransform(bInitial=False, Item=RigVMVariableNode() {variable=Toe Tip Connector}.Value).Transform) {variable=Toe Tip Input Transform}
   RigVMFunctionReturnNode(ExecuteContext=@RigVMFunction_Sequence.B)
   n2 = RigVMRerouteNode(Value=())
   RigVMFunction_DebugLineStripNoSpace = &RigVMFunction_DebugLineStripNoSpace(
    Color=(R=1,G=0,B=0,A=1)
    DebugDrawSettings=(DepthPriority=SDPG_Foreground,Lifetime=-1)
    Points=((X=0,Y=0,Z=0),(X=0,Y=0,Z=0),(X=0,Y=0,Z=0),(X=0,Y=0,Z=0),(X=0,Y=0,Z=0),(X=0,Y=0,Z=0))
    Thickness=0.1
    WorldOffset=(Rotation=(X=0,Y=0,Z=0,W=1),Translation=(X=0,Y=0,Z=0),Scale3D=(X=1,Y=1,Z=1))
    bEnabled=false
    Points.0=&RigUnit_GetTransform(bInitial=False, Item=RigVMVariableNode() {variable=Hip Socket Connector}.Value).Transform.Translation
    Points.1=&RigUnit_GetTransform(bInitial=False, Item=RigVMVariableNode() {variable=Knee Connector}.Value).Transform.Translation
    Points.2=&RigUnit_GetTransform(bInitial=False, Item=RigVMVariableNode() {variable=Ankle Connector}.Value).Transform.Translation
    Points.3=&RigUnit_GetTransform(bInitial=False, Item=RigVMVariableNode() {variable=Ball Connector}.Value).Transform.Translation
    Points.4=&RigUnit_GetTransform(bInitial=False, Item=RigVMVariableNode() {variable=Toe Connector}.Value).Transform.Translation
    Points.5=&RigUnit_GetTransform(bInitial=False, Item=RigVMVariableNode() {variable=Toe Tip Connector}.Value).Transform.Translation
    ExecutePin=@n2.Value
   )
   n3 = RigVMRerouteNode(Value=())
   RigVMFunction_DebugTransformArrayMutableNoSpace = /Script/RigVM.RigVMFunction_DebugTransformArrayMutableNoSpace(
    Color=(R=0,G=0,B=1,A=1)
    DebugDrawSettings=(DepthPriority=SDPG_Foreground,Lifetime=-1)
    Mode=Point
    ParentIndices=()
    Scale=0.3
    Thickness=0
    Transforms="((Rotation=(X=0,Y=0,Z=0,W=1),Translation=(X=0,Y=0,Z=0),Scale3D=(X=1,Y=1,Z=1)),(Rotation=(X=0,Y=0,Z=0,W=1),Translation=(X=0,Y=0,Z=0),Scale3D=(X=1,Y=1,Z=1)),(Rotation=(X=0,Y=0,Z=0,W=1),Translation=(X=0,Y=0,Z=0),Scale3D=(X=1,Y=1,Z=1)),(Rotation=(X=0,Y=0,Z=0,W ...(truncated)"
    WorldOffset=(Rotation=(X=0,Y=0,Z=0,W=1),Translation=(X=0,Y=0,Z=0),Scale3D=(X=1,Y=1,Z=1))
    bEnabled=false
    Transforms.4=&RigUnit_GetTransform(bInitial=False, Item=RigVMVariableNode() {variable=Toe Connector}.Value).Transform
    Transforms.5=&RigUnit_GetTransform(bInitial=False, Item=RigVMVariableNode() {variable=Toe Tip Connector}.Value).Transform
    Transforms.0=&RigUnit_GetTransform(bInitial=False, Item=RigVMVariableNode() {variable=Hip Socket Connector}.Value).Transform
    Transforms.1=&RigUnit_GetTransform(bInitial=False, Item=RigVMVariableNode() {variable=Knee Connector}.Value).Transform
    Transforms.2=&RigUnit_GetTransform(bInitial=False, Item=RigVMVariableNode() {variable=Ankle Connector}.Value).Transform
    Transforms.3=&RigUnit_GetTransform(bInitial=False, Item=RigVMVariableNode() {variable=Ball Connector}.Value).Transform
    ExecutePin=@n3.Value
   ) {is=struct RigVMUnitNode}
  ]}
 New Function_ContainedGraph {nodes [
   n1 = RigVMFunctionEntryNode()
   RigVMFunction_Sequence = &RigVMFunction_Sequence(ExecuteContext=@n1.ExecuteContext)
   RigVMFunction_Sequence_1 = RigVMAggregateNode(ExecuteContext=@RigVMFunction_Sequence.A)
   RigVMVariableNode(
    Value=((Type=None,Name="None"),(Type=None,Name="None"),(Type=None,Name="None"),(Type=None,Name="None"),(Type=None,Name="None"),(Type=None,Name="None"))
    Value.0=RigVMVariableNode() {variable=Hip Socket Connector}.Value
    Value.1=RigVMVariableNode() {variable=Knee Connector}.Value
    Value.2=RigVMVariableNode() {variable=Ankle Connector}.Value
    Value.3=RigVMVariableNode() {variable=Ball Connector}.Value
    Value.4=RigVMVariableNode() {variable=Toe Connector}.Value
    Value.5=RigVMVariableNode() {variable=Toe Tip Connector}.Value
    ExecuteContext=@RigVMFunction_Sequence_1.A
   ) {variable=Connectors}
   n2 = &RigUnit_HierarchyGetParent(bDefaultParent=True, Child=RigVMVariableNode() {variable=Hip Socket Connector}.Value)
   RigVMVariableNode(Value=@n2.Parent, ExecuteContext=@RigVMFunction_Sequence_1.B) {variable=Fk Parent}
   > RigVMVariableNode(Value=@n2.Parent) {variable=Ik Parent}
   > RigVMVariableNode(Value=@n2.Parent) {variable=Blend Parent}
   For_Each = RigVMDispatchNode(Array=RigVMVariableNode() {variable=Connectors}.Value, ExecuteContext=@RigVMFunction_Sequence_1.C) {template="DISPATCH_RigVMDispatch_ArrayIterator(in Array,out Element,out Index,out Count,out Ratio)"}
   Greater = /Script/RigVM.RigVMFunction_MathIntGreater(B=0, A=@For_Each.Index) {is=struct RigVMUnitNode, template="Greater::Execute(in A,in B,out Result)"}
   RigVMFunction_ControlFlowBranch = &RigVMFunction_ControlFlowBranch(ExecuteContext=@For_Each.ExecuteContext, Condition=@Greater.Result)
   RigVMVariableNode(ExecuteContext=@RigVMFunction_ControlFlowBranch.True, Value=@For_Each.Index) {variable=Index}
   > RigVMFunction_Sequence_2 = RigVMAggregateNode()
   Set_At = RigVMDispatchNode(
    Element=(Keys=((Type=None,Name="None"),(Type=None,Name="None"),(Type=None,Name="None")))
    Element.Keys.0=RigVMVariableNode() {variable=Fk Bone}.Value
    Element.Keys.1=RigVMVariableNode() {variable=Ik Bone}.Value
    Element.Keys.2=RigVMVariableNode() {variable=Blend Bone}.Value
    Index=/Script/RigVM.RigVMFunction_MathIntSub(B=1, A=RigVMVariableNode() {variable=Index}.Value) {is=struct RigVMUnitNode, template="Subtract::Execute(in A,in B,out Result)"}.Result
    Array=RigVMVariableNode() {variable=Ik Fk Bones}.Value
    ExecuteContext=@RigVMFunction_Sequence_2.F
   ) {template="DISPATCH_RigVMDispatch_ArraySetAtIndex(io Array,in Index,in Element)"}
   Get Main Leg PV = RigVMFunctionReferenceNode(Initial=true)
   Subtract = &RigVMFunction_MathIntSub(B=1, A=RigVMVariableNode() {variable=Index}.Value) {template="Subtract::Execute(in A,in B,out Result)"}
   n3 = RigVMVariableNode() {variable=Connectors}
   n4 = RigVMVariableNode() {variable=Connectors}
   n5 = RigVMVariableNode() {variable=Index}
   n6 = &RigUnit_AimBoneMath(
    DebugSettings=(bEnabled=False,Scale=10,WorldOffset=(Rotation=(X=0,Y=0,Z=0,W=1),Translation=(X=0,Y=0,Z=0),Scale3D=(X=1,Y=1,Z=1)))
    InputTransform=(Rotation=(X=0,Y=0,Z=0,W=1),Translation=(X=0,Y=0,Z=0),Scale3D=(X=1,Y=1,Z=1))
    Primary=(Weight=1,Axis=(X=1,Y=0,Z=0),Target=(X=1,Y=0,Z=0),Kind=Location,Space=(Type=Bone,Name="None"))
    Secondary=(Weight=1,Axis=(X=0,Y=0,Z=1),Target=(X=1,Y=0,Z=0),Kind=Direction,Space=(Type=Bone,Name="None"))
    Weight=1
    Secondary.Target=@"Get Main Leg PV".Pole Vector
    InputTransform.Translation=&RigUnit_GetTransform(
     bInitial=true
     Item=RigVMDispatchNode(Index=@Subtract.Result, Array=@n3.Value) {template="DISPATCH_RigVMDispatch_ArrayGetAtIndex(in Array,in Index,out Element)"}.Element
    ).Transform.Translation
    Primary.Target=&RigUnit_GetTransform(
     bInitial=true
     Item=RigVMDispatchNode(Array=@n4.Value, Index=@n5.Value) {template="DISPATCH_RigVMDispatch_ArrayGetAtIndex(in Array,in Index,out Element)"}.Element
    ).Transform.Translation
   )
   RigVMVariableNode(Value=@n6.Result, ExecuteContext=@RigVMFunction_Sequence_2.A) {variable=Bone Transform}
   n7 = RigVMVariableNode() {variable=Index}
   To_String = RigVMDispatchNode(Value=RigVMDispatchNode(
     Index=&RigVMFunction_MathIntSub(B=1, A=@n7.Value) {template="Subtract::Execute(in A,in B,out Result)"}.Result
     Array=RigVMVariableNode() {variable=Connectors}.Value
    ) {template="DISPATCH_RigVMDispatch_ArrayGetAtIndex(in Array,in Index,out Element)"}.Element.Name) {template="DISPATCH_RigDispatch_ToString(in Value,out Result)"}
   RigVMVariableNode(Value=@To_String.Result, ExecuteContext=@RigVMFunction_Sequence_2.B) {variable=Base Name}
   RigVMFunction_StringJoin = &RigVMFunction_StringJoin(Values=("","_Rig_Fk"), Values.0=RigVMVariableNode() {variable=Base Name}.Value)
   From_String = RigVMDispatchNode(String=@RigVMFunction_StringJoin.Result) {template="DISPATCH_RigDispatch_FromString(in String,out Result)"}
   n8 = &RigUnit_HierarchyAddBone(
    Space=GlobalSpace
    Name=@From_String.Result
    Transform=RigVMVariableNode() {variable=Bone Transform}.Value
    Parent=RigVMVariableNode() {variable=Fk Parent}.Value
    ExecutePin=@RigVMFunction_Sequence_2.C
   )
   > RigVMVariableNode(Value=@n8.Item) {variable=Fk Bone}
   > RigVMVariableNode(Value=@n8.Item) {variable=Fk Parent}
   From_String_1 = RigVMDispatchNode(String=&RigVMFunction_StringJoin(Values=("","_Rig_Ik"), Values.0=RigVMVariableNode() {variable=Base Name}.Value).Result) {template="DISPATCH_RigDispatch_FromString(in String,out Result)"}
   n9 = &RigUnit_HierarchyAddBone(
    Space=GlobalSpace
    Name=@From_String_1.Result
    Transform=RigVMVariableNode() {variable=Bone Transform}.Value
    Parent=RigVMVariableNode() {variable=Ik Parent}.Value
    ExecutePin=@RigVMFunction_Sequence_2.D
   )
   > RigVMVariableNode(Value=@n9.Item) {variable=Ik Bone}
   > RigVMVariableNode(Value=@n9.Item) {variable=Ik Parent}
   From_String_2 = RigVMDispatchNode(String=&RigVMFunction_StringJoin(Values=("","_Rig_Blend"), Values.0=RigVMVariableNode() {variable=Base Name}.Value).Result) {template="DISPATCH_RigDispatch_FromString(in String,out Result)"}
   n10 = &RigUnit_HierarchyAddBone(
    Space=GlobalSpace
    Name=@From_String_2.Result
    Transform=RigVMVariableNode() {variable=Bone Transform}.Value
    Parent=RigVMVariableNode() {variable=Blend Parent}.Value
    ExecutePin=@RigVMFunction_Sequence_2.E
   )
   > RigVMVariableNode(Value=@n10.Item) {variable=Blend Bone}
   > RigVMVariableNode(Value=@n10.Item) {variable=Blend Parent}
   RigVMFunctionReturnNode(ExecuteContext=@RigVMFunction_Sequence.B)
  ]}
 New Function_ContainedGraph {nodes [
   n1 = RigVMFunctionEntryNode()
   RigVMFunction_Sequence = &RigVMFunction_Sequence(ExecuteContext=@n1.ExecuteContext)
   n2 = &RigVMFunction_Sequence(ExecuteContext=@RigVMFunction_Sequence.A)
   For_Each = RigVMDispatchNode(Array=RigVMVariableNode() {variable=Ik Fk Bones}.Value, ExecuteContext=@n2.A) {template="DISPATCH_RigVMDispatch_ArrayIterator(in Array,out Element,out Index,out Count,out Ratio)"}
   > RigVMVariableNode(Value=@For_Each.Element) {variable=Bones}
   > RigVMFunction_Sequence_2 = RigVMAggregateNode()
   RigVMVariableNode(
    Value=RigVMDispatchNode(Index=0, Array=RigVMVariableNode() {variable=Bones}.Value.Keys) {template="DISPATCH_RigVMDispatch_ArrayGetAtIndex(in Array,in Index,out Element)"}.Element
    ExecuteContext=@RigVMFunction_Sequence_2.A
   ) {variable=Fk Bone}
   RigVMVariableNode(
    Value=RigVMDispatchNode(Index=1, Array=RigVMVariableNode() {variable=Bones}.Value.Keys) {template="DISPATCH_RigVMDispatch_ArrayGetAtIndex(in Array,in Index,out Element)"}.Element
    ExecuteContext=@RigVMFunction_Sequence_2.B
   ) {variable=Ik Bone}
   RigVMVariableNode(
    Value=RigVMDispatchNode(Index=2, Array=RigVMVariableNode() {variable=Bones}.Value.Keys) {template="DISPATCH_RigVMDispatch_ArrayGetAtIndex(in Array,in Index,out Element)"}.Element
    ExecuteContext=@RigVMFunction_Sequence_2.C
   ) {variable=Blend Bone}
   Interpolate = /Script/RigVM.RigVMFunction_MathTransformLerp(
    A=&RigUnit_GetTransform(Space=LocalSpace, bInitial=False, Item=RigVMVariableNode() {variable=Fk Bone}.Value).Transform
    B=&RigUnit_GetTransform(Space=LocalSpace, bInitial=False, Item=RigVMVariableNode() {variable=Ik Bone}.Value).Transform
    T=RigVMVariableNode() {variable=Ik Fk Blend}.Value
   ) {is=struct RigVMUnitNode, template="Interpolate::Execute(in A,in B,in T,out Result)"}
   &RigUnit_SetTransform(
    Space=LocalSpace
    Weight=1
    bInitial=False
    Item=RigVMVariableNode() {variable=Blend Bone}.Value
    Value=@Interpolate.Result
    ExecutePin=@RigVMFunction_Sequence_2.D
   )
   RigVMFunctionReturnNode(ExecuteContext=@RigVMFunction_Sequence.B)
  ]}
 New Function_ContainedGraph {nodes [
   n1 = RigVMFunctionEntryNode()
   RigVMFunction_Sequence = &RigVMFunction_Sequence(ExecuteContext=@n1.ExecuteContext)
   RigVMFunction_Sequence_1 = RigVMAggregateNode(ExecuteContext=@RigVMFunction_Sequence.A)
   &RigUnit_SetTransform(
    Weight=1
    bInitial=False
    Value=&RigVMFunction_MathTransformMakeAbsolute(Parent=RigVMVariableNode() {variable=Hip Socket Rig Transform}.Value, Local=RigVMVariableNode() {variable=Hip Socket Local Transform}.Value) {template="Make Absolute::Execute(in Local,in Parent,out Global)"}.Global
    Item=RigVMVariableNode() {variable=Hip Socket Connector}.Value
    ExecutePin=@RigVMFunction_Sequence_1.A
   )
   &RigUnit_SetTransform(
    Weight=1
    bInitial=False
    Value=&RigVMFunction_MathTransformMakeAbsolute(Parent=RigVMVariableNode() {variable=Knee Rig Transform}.Value, Local=RigVMVariableNode() {variable=Knee Local Transform}.Value) {template="Make Absolute::Execute(in Local,in Parent,out Global)"}.Global
    Item=RigVMVariableNode() {variable=Knee Connector}.Value
    ExecutePin=@RigVMFunction_Sequence_1.B
   )
   &RigUnit_SetTransform(
    Weight=1
    bInitial=False
    Value=&RigVMFunction_MathTransformMakeAbsolute(Parent=RigVMVariableNode() {variable=Ankle Rig Transform}.Value, Local=RigVMVariableNode() {variable=Ankle Local Transform}.Value) {template="Make Absolute::Execute(in Local,in Parent,out Global)"}.Global
    Item=RigVMVariableNode() {variable=Ankle Connector}.Value
    ExecutePin=@RigVMFunction_Sequence_1.C
   )
   &RigUnit_SetTransform(
    Weight=1
    bInitial=False
    Value=&RigVMFunction_MathTransformMakeAbsolute(Parent=RigVMVariableNode() {variable=Ball Rig Transform}.Value, Local=RigVMVariableNode() {variable=Ball Local Tranform}.Value) {template="Make Absolute::Execute(in Local,in Parent,out Global)"}.Global
    Item=RigVMVariableNode() {variable=Ball Connector}.Value
    ExecutePin=@RigVMFunction_Sequence_1.D
   )
   &RigUnit_SetTransform(
    Weight=1
    bInitial=False
    Value=&RigVMFunction_MathTransformMakeAbsolute(Parent=RigVMVariableNode() {variable=Toe Rig Transform}.Value, Local=RigVMVariableNode() {variable=Toe Local Transform}.Value) {template="Make Absolute::Execute(in Local,in Parent,out Global)"}.Global
    Item=RigVMVariableNode() {variable=Toe Connector}.Value
    ExecutePin=@RigVMFunction_Sequence_1.E
   )
   RigVMFunctionReturnNode(ExecuteContext=@RigVMFunction_Sequence.B)
   n2 = RigVMVariableNode() {variable=Ik Fk Bones}
   &RigUnit_GetTransform(
    bInitial=False
    Item=RigVMDispatchNode(
     Index=2
     Array=RigVMDispatchNode(Index=0, Array=@n2.Value) {template="DISPATCH_RigVMDispatch_ArrayGetAtIndex(in Array,in Index,out Element)"}.Element.Keys
    ) {template="DISPATCH_RigVMDispatch_ArrayGetAtIndex(in Array,in Index,out Element)"}.Element
   )
   n3 = RigVMVariableNode() {variable=Ik Fk Bones}
   &RigUnit_GetTransform(
    bInitial=False
    Item=RigVMDispatchNode(
     Index=2
     Array=RigVMDispatchNode(Index=1, Array=@n3.Value) {template="DISPATCH_RigVMDispatch_ArrayGetAtIndex(in Array,in Index,out Element)"}.Element.Keys
    ) {template="DISPATCH_RigVMDispatch_ArrayGetAtIndex(in Array,in Index,out Element)"}.Element
   )
   n4 = RigVMVariableNode() {variable=Ik Fk Bones}
   &RigUnit_GetTransform(
    bInitial=False
    Item=RigVMDispatchNode(
     Index=2
     Array=RigVMDispatchNode(Index=2, Array=@n4.Value) {template="DISPATCH_RigVMDispatch_ArrayGetAtIndex(in Array,in Index,out Element)"}.Element.Keys
    ) {template="DISPATCH_RigVMDispatch_ArrayGetAtIndex(in Array,in Index,out Element)"}.Element
   )
   n5 = RigVMVariableNode() {variable=Ik Fk Bones}
   &RigUnit_GetTransform(
    bInitial=False
    Item=RigVMDispatchNode(
     Index=2
     Array=RigVMDispatchNode(Index=3, Array=@n5.Value) {template="DISPATCH_RigVMDispatch_ArrayGetAtIndex(in Array,in Index,out Element)"}.Element.Keys
    ) {template="DISPATCH_RigVMDispatch_ArrayGetAtIndex(in Array,in Index,out Element)"}.Element
   )
   n6 = RigVMVariableNode() {variable=Ik Fk Bones}
   &RigUnit_GetTransform(
    bInitial=False
    Item=RigVMDispatchNode(
     Index=2
     Array=RigVMDispatchNode(Index=4, Array=@n6.Value) {template="DISPATCH_RigVMDispatch_ArrayGetAtIndex(in Array,in Index,out Element)"}.Element.Keys
    ) {template="DISPATCH_RigVMDispatch_ArrayGetAtIndex(in Array,in Index,out Element)"}.Element
   )
  ]}
 New Function_ContainedGraph {nodes [
   n1 = RigVMFunctionEntryNode()
   RigVMFunction_Sequence = &RigVMFunction_Sequence(ExecuteContext=@n1.ExecuteContext)
   n2 = &RigVMFunction_Sequence(ExecuteContext=@RigVMFunction_Sequence.A)
   Set Side Color = RigVMFunctionReferenceNode(ExecuteContext=@n2.A)
   > RigVMFunction_DebugLineNoSpace = &RigVMFunction_DebugLineNoSpace(
    DebugDrawSettings=(DepthPriority=SDPG_Foreground,Lifetime=-1)
    Thickness=0
    WorldOffset=(Rotation=(X=0,Y=0,Z=0,W=1),Translation=(X=0,Y=0,Z=0),Scale3D=(X=1,Y=1,Z=1))
    bEnabled=true
    B=&RigUnit_GetTransform(bInitial=False, Item=RigVMVariableNode() {variable=Knee PV}.Value).Transform.Translation
    A=&RigUnit_GetTransform(bInitial=False, Item=RigVMVariableNode() {variable=Knee Connector}.Value).Transform.Translation
    Color=@"Set Side Color".Color
   )
   RigVMFunctionReturnNode(ExecuteContext=@RigVMFunction_Sequence.B)
  ]}
 New Function_ContainedGraph {nodes [
   n1 = RigVMFunctionEntryNode()
   RigVMFunction_Sequence = &RigVMFunction_Sequence(ExecuteContext=@n1.ExecuteContext)
   RigVMFunction_Sequence_4 = RigVMAggregateNode(ExecuteContext=@RigVMFunction_Sequence.A)
   RigVMFunction_Sequence_3 = RigVMAggregateNode(ExecuteContext=@RigVMFunction_Sequence_4.C)
   Get Relative Transform_4 = RigVMFunctionReferenceNode(
    Initial=true
    RigNode=RigVMVariableNode() {variable=Hip Socket Connector}.Value
    Parent=RigVMVariableNode() {variable=Initial Hip To Ik Control Aim}.Value
   )
   RigVMVariableNode(
    Value=&RigVMFunction_MathTransformMakeAbsolute(Local=@"Get Relative Transform_4".Relative, Parent=RigVMVariableNode() {variable=Hip To Ik Control Aim}.Value) {template="Make Absolute::Execute(in Local,in Parent,out Global)"}.Global
    ExecuteContext=@RigVMFunction_Sequence_3.B
   ) {variable=Hip Socket Backwards Transform}
   Get Relative Transform_8 = RigVMFunctionReferenceNode(
    Initial=true
    RigNode=RigVMVariableNode() {variable=Toe Connector}.Value
    Parent=RigVMVariableNode() {variable=Initial Hip To Ik Control Aim}.Value
   )
   RigVMVariableNode(
    Value=&RigVMFunction_MathTransformMakeAbsolute(Local=@"Get Relative Transform_8".Relative, Parent=RigVMVariableNode() {variable=Hip To Ik Control Aim}.Value) {template="Make Absolute::Execute(in Local,in Parent,out Global)"}.Global
    ExecuteContext=@RigVMFunction_Sequence_3.F
   ) {variable=Toe Backwards Transform}
   Get Relative Transform_5 = RigVMFunctionReferenceNode(
    Initial=true
    RigNode=RigVMVariableNode() {variable=Knee Connector}.Value
    Parent=RigVMVariableNode() {variable=Initial Hip To Ik Control Aim}.Value
   )
   RigVMVariableNode(
    Value=&RigVMFunction_MathTransformMakeAbsolute(Local=@"Get Relative Transform_5".Relative, Parent=RigVMVariableNode() {variable=Hip To Ik Control Aim}.Value) {template="Make Absolute::Execute(in Local,in Parent,out Global)"}.Global
    ExecuteContext=@RigVMFunction_Sequence_3.C
   ) {variable=Knee Backwards Transform}
   Get Relative Transform_9 = RigVMFunctionReferenceNode(
    Initial=true
    RigNode=RigVMVariableNode() {variable=Toe Tip Connector}.Value
    Parent=RigVMVariableNode() {variable=Initial Hip To Ik Control Aim}.Value
   )
   RigVMVariableNode(
    Value=&RigVMFunction_MathTransformMakeAbsolute(Local=@"Get Relative Transform_9".Relative, Parent=RigVMVariableNode() {variable=Hip To Ik Control Aim}.Value) {template="Make Absolute::Execute(in Local,in Parent,out Global)"}.Global
    ExecuteContext=@RigVMFunction_Sequence_3.G
   ) {variable=Toe Tip Backwards Transform}
   Get Relative Transform_6 = RigVMFunctionReferenceNode(
    Initial=true
    RigNode=RigVMVariableNode() {variable=Ankle Connector}.Value
    Parent=RigVMVariableNode() {variable=Initial Hip To Ik Control Aim}.Value
   )
   RigVMVariableNode(
    Value=&RigVMFunction_MathTransformMakeAbsolute(Local=@"Get Relative Transform_6".Relative, Parent=RigVMVariableNode() {variable=Hip To Ik Control Aim}.Value) {template="Make Absolute::Execute(in Local,in Parent,out Global)"}.Global
    ExecuteContext=@RigVMFunction_Sequence_3.D
   ) {variable=Ankle Backwards Transform}
   Get Relative Transform_7 = RigVMFunctionReferenceNode(
    Initial=true
    Parent=RigVMVariableNode() {variable=Initial Hip To Ik Control Aim}.Value
    RigNode=RigVMVariableNode() {variable=Ball Connector}.Value
   )
   RigVMVariableNode(
    Value=&RigVMFunction_MathTransformMakeAbsolute(Local=@"Get Relative Transform_7".Relative, Parent=RigVMVariableNode() {variable=Hip To Ik Control Aim}.Value) {template="Make Absolute::Execute(in Local,in Parent,out Global)"}.Global
    ExecuteContext=@RigVMFunction_Sequence_3.E
   ) {variable=Ball Backwards Transform}
   Get Hip To Ik Control Aim_1 = RigVMFunctionReferenceNode()
   RigVMVariableNode(Value=@"Get Hip To Ik Control Aim_1".Aim, ExecuteContext=@RigVMFunction_Sequence_4.B) {variable=Hip To Ik Control Aim}
   Get Hip To Ik Control Aim_2 = RigVMFunctionReferenceNode(Initial=true)
   RigVMVariableNode(Value=@"Get Hip To Ik Control Aim_2".Aim, ExecuteContext=@RigVMFunction_Sequence_4.A) {variable=Initial Hip To Ik Control Aim}
   &RigVMFunction_DebugLineNoSpace(
    Color=(R=0.0069,G=0.6198,B=0,A=1)
    DebugDrawSettings=(DepthPriority=SDPG_Foreground,Lifetime=-1)
    Thickness=0.2
    WorldOffset=(Rotation=(X=0,Y=0,Z=0,W=1),Translation=(X=0,Y=0,Z=0),Scale3D=(X=1,Y=1,Z=1))
    bEnabled=True
    A=RigVMVariableNode() {variable=Hip Socket Backwards Transform}.Value.Translation
    B=RigVMVariableNode() {variable=Knee Backwards Transform}.Value.Translation
    ExecutePin=@RigVMFunction_Sequence_4.D
   )
   RigVMFunctionReturnNode(ExecuteContext=@RigVMFunction_Sequence.B)
  ]}
 New Function_ContainedGraph {nodes [
   n1 = RigVMFunctionEntryNode()
   RigVMFunction_Sequence = &RigVMFunction_Sequence(ExecuteContext=@n1.ExecuteContext)
   RigVMFunction_Sequence_3 = RigVMAggregateNode(ExecuteContext=@RigVMFunction_Sequence.A)
   Get Main Leg PV_1 = RigVMFunctionReferenceNode(Initial=true)
   n2 = &RigUnit_AimBoneMath(
    DebugSettings=(bEnabled=False,Scale=10,WorldOffset=(Rotation=(X=0,Y=0,Z=0,W=1),Translation=(X=0,Y=0,Z=0),Scale3D=(X=1,Y=1,Z=1)))
    InputTransform=(Rotation=(X=0,Y=0,Z=0,W=1),Translation=(X=0,Y=0,Z=0),Scale3D=(X=1,Y=1,Z=1))
    Primary=(Weight=1,Axis=(X=1,Y=0,Z=0),Target=(X=1,Y=0,Z=0),Kind=Location,Space=(Type=Bone,Name="None"))
    Secondary=(Weight=1,Axis=(X=0,Y=0,Z=1),Target=(X=1,Y=0,Z=0),Kind=Direction,Space=(Type=Bone,Name="None"))
    Weight=1
    Secondary.Target=@"Get Main Leg PV_1".Pole Vector
    Primary.Target=&RigUnit_GetTransform(bInitial=true, Item=RigVMVariableNode() {variable=Ball Connector}.Value).Transform.Translation
    InputTransform.Translation=&RigUnit_GetTransform(bInitial=true, Item=RigVMVariableNode() {variable=Ankle Connector}.Value).Transform.Translation
   )
   RigVMVariableNode(Value=@n2.Result, ExecuteContext=@RigVMFunction_Sequence_3.D) {variable=Initial Ankle Rig Transform}
   Get Relative Transform = RigVMFunctionReferenceNode(
    Initial=true
    Parent=RigVMVariableNode() {variable=Initial Ankle Rig Transform}.Value
    RigNode=RigVMVariableNode() {variable=Ball Connector}.Value
   )
   RigVMVariableNode(
    Value=&RigVMFunction_MathTransformMakeAbsolute(Local=@"Get Relative Transform".Relative, Parent=RigVMVariableNode() {variable=Ankle Backwards Transform}.Value) {template="Make Absolute::Execute(in Local,in Parent,out Global)"}.Global
    ExecuteContext=@RigVMFunction_Sequence_3.E
   ) {variable=Ball Backwards Transform}
   n3 = RigVMVariableNode() {variable=Ball Connector}
   Get Ik Angles = RigVMFunctionReferenceNode(
    "B 180 "=false
    C 180=true
    Negate A=false
    Negate B=true
    Negate C=true
    Length C=&RigVMFunction_MathVectorDistance(
     A=&RigUnit_GetTransform(bInitial=False, Item=@n3.Value).Transform.Translation
     B=RigVMVariableNode() {variable=Knee Backwards Transform}.Value.Translation
    ).Result
    Length A=RigVMVariableNode() {variable=Knee Bone Length}.Value
    Length B=RigVMVariableNode() {variable=Foot Bone Length}.Value
   )
   RigVMVariableNode(Value=@"Get Ik Angles".Angle B, ExecuteContext=@RigVMFunction_Sequence_3.A) {variable=Knee Rig Angle}
   > RigVMVariableNode(Value=RigVMDispatchNode(
     True=0
     Condition=RigVMDispatchNode(B=180, A=@"Get Ik Angles".Angle C) {template="DISPATCH_RigVMDispatch_CoreEquals(in A,in B,out Result)"}.Result
     False=@"Get Ik Angles".Angle C
    ) {template="DISPATCH_RigVMDispatch_If(in Condition,in True,in False,out Result)"}.Result) {variable=Foot Rig Angle}
   Get Relative Transform_1 = RigVMFunctionReferenceNode(
    Initial=true
    Parent=RigVMVariableNode() {variable=Initial Ankle Rig Transform}.Value
    RigNode=RigVMVariableNode() {variable=Toe Connector}.Value
   )
   RigVMVariableNode(
    Value=&RigVMFunction_MathTransformMakeAbsolute(Local=@"Get Relative Transform_1".Relative, Parent=RigVMVariableNode() {variable=Ankle Rig Transform}.Value) {template="Make Absolute::Execute(in Local,in Parent,out Global)"}.Global
    ExecuteContext=@RigVMFunction_Sequence_3.F
   ) {variable=Toe Backwards Transform}
   Get Main Leg PV = RigVMFunctionReferenceNode(Initial=False)
   n4 = &RigUnit_AimBoneMath(
    DebugSettings=(bEnabled=False,Scale=10,WorldOffset=(Rotation=(X=0,Y=0,Z=0,W=1),Translation=(X=0,Y=0,Z=0),Scale3D=(X=1,Y=1,Z=1)))
    InputTransform=(Rotation=(X=0,Y=0,Z=0,W=1),Translation=(X=0,Y=0,Z=0),Scale3D=(X=1,Y=1,Z=1))
    Primary=(Weight=1,Axis=(X=1,Y=0,Z=0),Target=(X=1,Y=0,Z=0),Kind=Location,Space=(Type=Bone,Name="None"))
    Secondary=(Weight=1,Axis=(X=0,Y=0,Z=1),Target=(X=1,Y=0,Z=0),Kind=Direction,Space=(Type=Bone,Name="None"))
    Weight=1
    Secondary.Target=@"Get Main Leg PV".Pole Vector
    Primary.Target=&RigUnit_GetTransform(bInitial=False, Item=RigVMVariableNode() {variable=Ball Connector}.Value).Transform.Translation
    InputTransform.Translation=RigVMVariableNode() {variable=Knee Backwards Transform}.Value.Translation
   )
   n5 = RigVMVariableNode() {variable=Knee Rig Angle}
   RigVMVariableNode(
    Value=&RigVMFunction_MathTransformMakeAbsolute(
     Local=(Rotation=(X=0,Y=0,Z=0,W=1),Translation=(X=0,Y=0,Z=0),Scale3D=(X=1,Y=1,Z=1))
     Parent=@n4.Result
     Local.Rotation=&RigVMFunction_MathQuaternionFromEuler(Euler=(X=0,Y=0,Z=0), Euler.Z=@n5.Value).Result
    ) {template="Make Absolute::Execute(in Local,in Parent,out Global)"}.Global
    ExecuteContext=@RigVMFunction_Sequence_3.B
   ) {variable=Knee Backwards Transform}
   n6 = RigVMVariableNode() {variable=Foot Rig Angle}
   n7 = RigVMVariableNode() {variable=Knee Backwards Transform}
   n8 = RigVMVariableNode() {variable=Knee Bone Length}
   RigVMVariableNode(
    Value=&RigVMFunction_MathTransformMakeAbsolute(
     Local=(Rotation=(X=0,Y=0,Z=0,W=1),Translation=(X=0,Y=0,Z=0),Scale3D=(X=1,Y=1,Z=1))
     Local.Rotation=&RigVMFunction_MathQuaternionFromEuler(Euler=(X=0,Y=0,Z=0), Euler.Z=@n6.Value).Result
     Parent=&RigVMFunction_MathTransformMakeAbsolute(Local=(Rotation=(X=0,Y=0,Z=0,W=1),Translation=(X=0,Y=0,Z=0),Scale3D=(X=1,Y=1,Z=1)), Parent=@n7.Value, Local.Translation.X=@n8.Value) {template="Make Absolute::Execute(in Local,in Parent,out Global)"}.Global
    ) {template="Make Absolute::Execute(in Local,in Parent,out Global)"}.Global
    ExecuteContext=@RigVMFunction_Sequence_3.C
   ) {variable=Ankle Backwards Transform}
   RigVMFunction_DebugLineStripNoSpace = &RigVMFunction_DebugLineStripNoSpace(
    Color=(R=1,G=1,B=0.0064,A=1)
    DebugDrawSettings=(DepthPriority=SDPG_Foreground,Lifetime=-1)
    Points=((X=0,Y=0,Z=0),(X=0,Y=0,Z=0),(X=0,Y=0,Z=0))
    Thickness=0
    WorldOffset=(Rotation=(X=0,Y=0,Z=0,W=1),Translation=(X=0,Y=0,Z=0),Scale3D=(X=1,Y=1,Z=1))
    bEnabled=false
    Points.0=RigVMVariableNode() {variable=Knee Backwards Transform}.Value.Translation
    Points.1=RigVMVariableNode() {variable=Ankle Backwards Transform}.Value.Translation
    Points.2=RigVMVariableNode() {variable=Ball Backwards Transform}.Value.Translation
    ExecutePin=@RigVMFunction_Sequence_3.H
   )
   > RigVMFunction_DebugTransformArrayMutableNoSpace = &RigVMFunction_DebugTransformArrayMutableNoSpace(
    Color=(R=1,G=1,B=1,A=1)
    DebugDrawSettings=(DepthPriority=SDPG_Foreground,Lifetime=-1)
    Mode=Point
    ParentIndices=()
    Scale=5
    Thickness=0
    Transforms=((Rotation=(X=0,Y=0,Z=0,W=1),Translation=(X=0,Y=0,Z=0),Scale3D=(X=1,Y=1,Z=1)),(Rotation=(X=0,Y=0,Z=0,W=1),Translation=(X=0,Y=0,Z=0),Scale3D=(X=1,Y=1,Z=1)),(Rotation=(X=0,Y=0,Z=0,W=1),Translation=(X=0,Y=0,Z=0),Scale3D=(X=1,Y=1,Z=1)))
    WorldOffset=(Rotation=(X=0,Y=0,Z=0,W=1),Translation=(X=0,Y=0,Z=0),Scale3D=(X=1,Y=1,Z=1))
    bEnabled=false
    Transforms.0=RigVMVariableNode() {variable=Knee Backwards Transform}.Value
    Transforms.1=RigVMVariableNode() {variable=Ankle Backwards Transform}.Value
    Transforms.2=RigVMVariableNode() {variable=Ball Backwards Transform}.Value
   )
   Get Relative Transform_2 = RigVMFunctionReferenceNode(
    Initial=true
    RigNode=RigVMVariableNode() {variable=Toe Tip Connector}.Value
    Parent=RigVMVariableNode() {variable=Initial Ankle Rig Transform}.Value
   )
   RigVMVariableNode(
    Value=&RigVMFunction_MathTransformMakeAbsolute(Local=@"Get Relative Transform_2".Relative, Parent=RigVMVariableNode() {variable=Ankle Rig Transform}.Value) {template="Make Absolute::Execute(in Local,in Parent,out Global)"}.Global
    ExecuteContext=@RigVMFunction_Sequence_3.G
   ) {variable=Toe Tip Backwards Transform}
   RigVMFunctionReturnNode(ExecuteContext=@RigVMFunction_Sequence.B)
   Multiply = &RigVMFunction_MathDoubleMul(
    B=RigVMDispatchNode(False=0, True=-1, Condition=RigVMVariableNode() {variable=Right Side}.Value) {template="DISPATCH_RigVMDispatch_If(in Condition,in True,in False,out Result)"}.Result
    A=@n5.Value
   )
   RigVMVariableNode() {variable=Ik Control}
   RigVMVariableNode() {variable=Ik Control}
  ]}
 New Function_ContainedGraph {nodes [
   RigVMFunctionEntryNode()
   > RigVMFunctionReturnNode()
  ]}
 New Function_ContainedGraph {nodes [
   n1 = RigVMFunctionEntryNode()
   RigVMFunction_Sequence = &RigVMFunction_Sequence(ExecuteContext=@n1.ExecuteContext)
   RigVMFunctionReturnNode(ExecuteContext=@RigVMFunction_Sequence.B)
   GetAnimationChannelFromItem = /Script/ControlRig.RigUnit_GetBoolAnimationChannelFromItem(bInitial=False, Item=RigVMVariableNode() {variable=Sec Constrol Visility}.Value) {is=struct RigVMUnitNode, template="GetAnimationChannelFromItem::Execute(out Value,in Item,in bInitial)"}
   Set_Module_Metadata = RigVMDispatchNode(Name=None, NameSpace=Self, Value=@GetAnimationChannelFromItem.Value, ExecuteContext=@RigVMFunction_Sequence.A) {template="DISPATCH_RigDispatch_SetModuleMetadata(in Name,in NameSpace,in Value,out Success)"}
   Make_Array = RigVMDispatchNode(
    Values=((Type=None,Name="None"),(Type=None,Name="None"))
    Values.0=RigVMVariableNode() {variable=Ball Fk Control}.Value
    Values.1=RigVMVariableNode() {variable=Toe Fk Control}.Value
   ) {template="DISPATCH_RigVMDispatch_ArrayMake(in Values,out Array)"}
   > For_Each = RigVMDispatchNode(Array=@Make_Array.Array) {template="DISPATCH_RigVMDispatch_ArrayIterator(in Array,out Element,out Index,out Count,out Ratio)"}
   > SetControlVisibility = /Script/ControlRig.RigUnit_SetControlVisibility(Item=@For_Each.Element, bVisible=@GetAnimationChannelFromItem.Value) {is=struct RigVMUnitNode, template="SetControlVisibility::Execute(in Item,in Pattern,in bVisible)"}
  ]}
 New Function_ContainedGraph {nodes [
   n1 = RigVMFunctionEntryNode()
   RigVMFunction_Sequence = &RigVMFunction_Sequence(ExecuteContext=@n1.ExecuteContext)
   RigVMFunctionReturnNode(ExecuteContext=@RigVMFunction_Sequence.B)
   n2 = /Script/ControlRig.RigUnit_HierarchyAddAnimationChannelBool(
    InitialValue=false
    LimitsEnabled=()
    MaximumValue=True
    MinimumValue=false
    Name=Sec Control Visibility
    Parent=RigVMVariableNode() {variable=Ik Control}.Value
    ExecutePin=@RigVMFunction_Sequence.A
   ) {
    is=struct RigVMUnitNode
    template="SpawnAnimationChannel::Execute(in InitialValue,in MinimumValue,in MaximumValue,in LimitsEnabled,in Parent,in Name,out Item)"
   }
   Make_Array = RigVMDispatchNode(
    Values=((Type=None,Name="None"),(Type=None,Name="None"))
    Values.0=RigVMVariableNode() {variable=Ball Fk Control}.Value
    Values.1=RigVMVariableNode() {variable=Toe Fk Control}.Value
   ) {template="DISPATCH_RigVMDispatch_ArrayMake(in Values,out Array)"}
   > /Script/ControlRig.RigUnit_SetChannelHosts(Hosts=@Make_Array.Array, Channel=@n2.Item) {is=struct RigVMUnitNode}
   > RigVMVariableNode(Value=@n2.Item) {variable=Sec Constrol Visility}
  ]}
 RigVMFunctionLibrary {nodes [
   Ankle Offset Ik System = RigVMCollapseNode(ExecuteContext=(), Initial=False)
   Backward Ankle Offset Ik System = RigVMCollapseNode(ExecuteContext=())
   Backward Knee Offset Ik System = RigVMCollapseNode(ExecuteContext=())
   Backward Main Leg Ik System = RigVMCollapseNode(ExecuteContext=())
   Clean Up Rig Bone Transforms = RigVMCollapseNode(ExecuteContext=())
   Create Ankle Offset Ik Control = RigVMCollapseNode(ExecuteContext=())
   Create Hip Socket Control = RigVMCollapseNode(ExecuteContext=())
   Create Ik Control = RigVMCollapseNode(ExecuteContext=())
   Create Knee Twist And PV Control = RigVMCollapseNode(ExecuteContext=())
   Create Local Transforms = RigVMCollapseNode(ExecuteContext=())
   Create Lower Leg Ik System = RigVMCollapseNode(ExecuteContext=(), Input Hip Socket Transform=(Rotation=(X=0,Y=0,Z=0,W=1),Translation=(X=0,Y=0,Z=0),Scale3D=(X=1,Y=1,Z=1)))
   Create Main Leg Ik System = RigVMCollapseNode(ExecuteContext=())
   Create Name = RigVMCollapseNode(ExecuteContext=(), Tokens=())
   Create Rig Bones = RigVMCollapseNode(ExecuteContext=())
   Create Rig Joints = RigVMCollapseNode(ExecuteContext=())
   Create Rig Transforms = RigVMCollapseNode(ExecuteContext=())
   Create Rig Up Vector = RigVMCollapseNode(ExecuteContext=())
   Create Roll Controls = RigVMCollapseNode(ExecuteContext=())
   Create Sec Control Visibility Switch = RigVMCollapseNode(ExecuteContext=())
   Create Space Switches = RigVMCollapseNode(ExecuteContext=())
   Create Toe Controls = RigVMCollapseNode(ExecuteContext=())
   Create Upper Leg Ik System = RigVMCollapseNode(
    ExecuteContext=()
    Input Ankle Transform=(Rotation=(X=0,Y=0,Z=0,W=1),Translation=(X=0,Y=0,Z=0),Scale3D=(X=1,Y=1,Z=1))
    Input Hip Socket Transform=(Rotation=(X=0,Y=0,Z=0,W=1),Translation=(X=0,Y=0,Z=0),Scale3D=(X=1,Y=1,Z=1))
    Input Knee Transform=(Rotation=(X=0,Y=0,Z=0,W=1),Translation=(X=0,Y=0,Z=0),Scale3D=(X=1,Y=1,Z=1))
   )
   Debug = RigVMCollapseNode(ExecuteContext=())
   Debug Anim Sequence Bone Positions = RigVMCollapseNode(ExecuteContext=())
   Debug Ik Line = RigVMCollapseNode(ExecuteContext=())
   Debug Lower Leg Ik Line = RigVMCollapseNode(ExecuteContext=())
   Debug Lower Leg Ik Pv Line = RigVMCollapseNode(ExecuteContext=())
   Debug Main Leg Ik Line = RigVMCollapseNode(ExecuteContext=())
   Debug Upper Leg Ik PV Line = RigVMCollapseNode(ExecuteContext=())
   Find Up Vector = RigVMCollapseNode(ExecuteContext=(), First Bone=(Type=None,Name="None"), Length=50, Second Bone=(Type=None,Name="None"), Tip Bone=(Type=None,Name="None"))
   Find Up Vector From Transforms = RigVMCollapseNode(
    Mid Transform=(Rotation=(X=0,Y=0,Z=0,W=1),Translation=(X=0,Y=0,Z=0),Scale3D=(X=1,Y=1,Z=1))
    Root Transform=(Rotation=(X=0,Y=0,Z=0,W=1),Translation=(X=0,Y=0,Z=0),Scale3D=(X=1,Y=1,Z=1))
    Tip Transform=(Rotation=(X=0,Y=0,Z=0,W=1),Translation=(X=0,Y=0,Z=0),Scale3D=(X=1,Y=1,Z=1))
    Vector Length=20
   )
   Get Bone Lengths = RigVMCollapseNode(ExecuteContext=())
   Get Connectors = RigVMCollapseNode(ExecuteContext=())
   Get Distance Between = RigVMCollapseNode(ExecuteContext=(), First Object=(Type=None,Name="None"), Second Object=(Type=None,Name="None"))
   Get Hip Socket Control Rotation = RigVMCollapseNode()
   Get Hip To Ik Control Aim = RigVMCollapseNode()
   Get Ik Angles = RigVMCollapseNode()
   Get Ik Control Transform = RigVMCollapseNode(ExecuteContext=(), Initial=False)
   Get Ik Distance = RigVMCollapseNode()
   Get Ik Vector = RigVMCollapseNode()
   Get Lower Ik Pole Vector = RigVMCollapseNode(
    Hip Socket Transform=(Rotation=(X=0,Y=0,Z=0,W=1),Translation=(X=0,Y=0,Z=0),Scale3D=(X=1,Y=1,Z=1))
    Knee Transform=(Rotation=(X=0,Y=0,Z=0,W=1),Translation=(X=0,Y=0,Z=0),Scale3D=(X=1,Y=1,Z=1))
    Vector Length=20
   )
   Get Lower Leg Ik PV = RigVMCollapseNode()
   Get Main Leg PV = RigVMCollapseNode(Initial=False)
   Get PV = RigVMCollapseNode(ExecuteContext=(), Root Object=(Type=None,Name="None"), Tip Object=(Type=None,Name="None"))
   Get Relative Transform = RigVMCollapseNode(Parent=(Rotation=(X=0,Y=0,Z=0,W=1),Translation=(X=0,Y=0,Z=0),Scale3D=(X=1,Y=1,Z=1)), RigNode=(Type=None,Name="None"))
   Get Upper Ik Pole Vector = RigVMCollapseNode(Hip Socket Transform=(Rotation=(X=0,Y=0,Z=0,W=1),Translation=(X=0,Y=0,Z=0),Scale3D=(X=1,Y=1,Z=1)), Vector Length=20)
   Get Upper Leg Ik PV = RigVMCollapseNode(Initial=False)
   Knee Offset Ik System = RigVMCollapseNode(ExecuteContext=(), Initial=False)
   Main Leg Ik System = RigVMCollapseNode(ExecuteContext=(), Initial=False)
   Position Ball Fk Control = RigVMCollapseNode(ExecuteContext=())
   Position Hip Socket Control = RigVMCollapseNode(ExecuteContext=())
   Position Hock Control = RigVMCollapseNode(ExecuteContext=())
   Position Hock Rotate Control = RigVMCollapseNode(Active=true, ExecuteContext=())
   Position Hock Rotate Control Null = RigVMCollapseNode(ExecuteContext=())
   Position Ik Control = RigVMCollapseNode(ExecuteContext=())
   Position Knee Rotation Control = RigVMCollapseNode(ExecuteContext=())
   Position Main PV Control = RigVMCollapseNode(ExecuteContext=())
   Position Toe Fk Control = RigVMCollapseNode(ExecuteContext=())
   Set Blend Bone Transforms = RigVMCollapseNode(ExecuteContext=())
   Set Bone Transforms = RigVMCollapseNode(ExecuteContext=())
   Set Control Lines = RigVMCollapseNode(ExecuteContext=())
   Set Ik Bone Transforms = RigVMCollapseNode(ExecuteContext=())
   Set Metadata = RigVMCollapseNode(ExecuteContext=())
   Set Sec Control Visibility = RigVMCollapseNode(ExecuteContext=())
   Set Side Color = RigVMCollapseNode(ExecuteContext=())
   Set Toe Fk Controls = RigVMCollapseNode(ExecuteContext=())
   Snap Hip Socket Bone to Null = RigVMCollapseNode(ExecuteContext=())
   Solve Hip Socket Control Rotation = RigVMCollapseNode(ExecuteContext=())
   Twist Ankle = RigVMCollapseNode(ExecuteContext=())
   Twist Knee = RigVMCollapseNode(ExecuteContext=())
  ]}
 RigVMFunction_Sequence_1_1_ContainedGraph {nodes [
   n1 = RigVMFunctionEntryNode()
   n2 = &RigVMFunction_Sequence(ExecuteContext=@n1.ExecuteContext)
   n3 = &RigVMFunction_Sequence(ExecuteContext=@n2.B)
   n4 = &RigVMFunction_Sequence(ExecuteContext=@n3.B)
   n5 = &RigVMFunction_Sequence(ExecuteContext=@n4.B)
   RigVMFunctionReturnNode(A=@n2.A, B=@n3.A, C=@n4.A, D=@n5.A, E=@n5.B)
  ]}
 RigVMFunction_Sequence_1_ContainedGraph {nodes [
   n1 = RigVMFunctionEntryNode()
   RigVMFunction_Sequence = &RigVMFunction_Sequence(ExecuteContext=@n1.ExecuteContext)
   n2 = &RigVMFunction_Sequence(ExecuteContext=@RigVMFunction_Sequence.B)
   n3 = &RigVMFunction_Sequence(ExecuteContext=@n2.B)
   n4 = &RigVMFunction_Sequence(ExecuteContext=@n3.B)
   RigVMFunctionReturnNode(A=@RigVMFunction_Sequence.A, B=@n2.A, C=@n3.A, D=@n4.A, E=@n4.B)
  ]}
 RigVMFunction_Sequence_1_ContainedGraph {nodes [
   n1 = RigVMFunctionEntryNode()
   RigVMFunction_Sequence = &RigVMFunction_Sequence(ExecuteContext=@n1.ExecuteContext)
   n2 = &RigVMFunction_Sequence(ExecuteContext=@RigVMFunction_Sequence.B)
   n3 = &RigVMFunction_Sequence(ExecuteContext=@n2.B)
   n4 = &RigVMFunction_Sequence(ExecuteContext=@n3.B)
   n5 = &RigVMFunction_Sequence(ExecuteContext=@n4.B)
   n6 = &RigVMFunction_Sequence(ExecuteContext=@n5.B)
   RigVMFunctionReturnNode(A=@RigVMFunction_Sequence.A, B=@n2.A, C=@n3.A, D=@n4.A, E=@n5.A, F=@n6.A, G=@n6.B)
  ]}
 RigVMFunction_Sequence_1_ContainedGraph {nodes [
   n1 = RigVMFunctionEntryNode()
   RigVMFunction_Sequence = &RigVMFunction_Sequence(ExecuteContext=@n1.ExecuteContext)
   n2 = &RigVMFunction_Sequence(ExecuteContext=@RigVMFunction_Sequence.B)
   n3 = &RigVMFunction_Sequence(ExecuteContext=@n2.B)
   n4 = &RigVMFunction_Sequence(ExecuteContext=@n3.B)
   n5 = &RigVMFunction_Sequence(ExecuteContext=@n4.B)
   RigVMFunctionReturnNode(A=@RigVMFunction_Sequence.A, B=@n2.A, C=@n3.A, D=@n4.A, E=@n5.A, F=@n5.B)
  ]}
 RigVMFunction_Sequence_2_ContainedGraph {nodes [
   n1 = RigVMFunctionEntryNode()
   n2 = &RigVMFunction_Sequence(ExecuteContext=@n1.ExecuteContext)
   n3 = &RigVMFunction_Sequence(ExecuteContext=@n2.B)
   RigVMFunctionReturnNode(A=@n2.A, B=@n3.A, C=@n3.B)
  ]}
 RigVMFunction_Sequence_2_ContainedGraph {nodes [
   n1 = RigVMFunctionEntryNode()
   n2 = &RigVMFunction_Sequence(ExecuteContext=@n1.ExecuteContext)
   n3 = &RigVMFunction_Sequence(ExecuteContext=@n2.B)
   n4 = &RigVMFunction_Sequence(ExecuteContext=@n3.B)
   RigVMFunctionReturnNode(A=@n2.A, B=@n3.A, C=@n4.A, D=@n4.B)
  ]}
 RigVMFunction_Sequence_2_ContainedGraph {nodes [
   n1 = RigVMFunctionEntryNode()
   n2 = &RigVMFunction_Sequence(ExecuteContext=@n1.ExecuteContext)
   n3 = &RigVMFunction_Sequence(ExecuteContext=@n2.B)
   RigVMFunctionReturnNode(A=@n2.A, B=@n3.A, C=@n3.B)
  ]}
 RigVMFunction_Sequence_2_ContainedGraph {nodes [
   n1 = RigVMFunctionEntryNode()
   n2 = &RigVMFunction_Sequence(ExecuteContext=@n1.ExecuteContext)
   n3 = &RigVMFunction_Sequence(ExecuteContext=@n2.B)
   n4 = &RigVMFunction_Sequence(ExecuteContext=@n3.B)
   n5 = &RigVMFunction_Sequence(ExecuteContext=@n4.B)
   n6 = &RigVMFunction_Sequence(ExecuteContext=@n5.B)
   n7 = &RigVMFunction_Sequence(ExecuteContext=@n6.B)
   n8 = &RigVMFunction_Sequence(ExecuteContext=@n7.B)
   n9 = &RigVMFunction_Sequence(ExecuteContext=@n8.B)
   n10 = &RigVMFunction_Sequence(ExecuteContext=@n9.B)
   RigVMFunctionReturnNode(A=@n2.A, B=@n3.A, C=@n4.A, D=@n5.A, E=@n6.A, F=@n7.A, G=@n8.A, H=@n9.A, I=@n10.A, J=@n10.B)
  ]}
 RigVMFunction_Sequence_2_ContainedGraph {nodes [
   n1 = RigVMFunctionEntryNode()
   n2 = &RigVMFunction_Sequence(ExecuteContext=@n1.ExecuteContext)
   n3 = &RigVMFunction_Sequence(ExecuteContext=@n2.B)
   n4 = &RigVMFunction_Sequence(ExecuteContext=@n3.B)
   n5 = &RigVMFunction_Sequence(ExecuteContext=@n4.B)
   n6 = &RigVMFunction_Sequence(ExecuteContext=@n5.B)
   n7 = &RigVMFunction_Sequence(ExecuteContext=@n6.B)
   RigVMFunctionReturnNode(A=@n2.A, B=@n3.A, C=@n4.A, D=@n5.A, E=@n6.A, F=@n7.A, G=@n7.B)
  ]}
 RigVMFunction_Sequence_2_ContainedGraph {nodes [
   n1 = RigVMFunctionEntryNode()
   n2 = &RigVMFunction_Sequence(ExecuteContext=@n1.ExecuteContext)
   n3 = &RigVMFunction_Sequence(ExecuteContext=@n2.B)
   n4 = &RigVMFunction_Sequence(ExecuteContext=@n3.B)
   n5 = &RigVMFunction_Sequence(ExecuteContext=@n4.B)
   RigVMFunctionReturnNode(A=@n2.A, B=@n3.A, C=@n4.A, D=@n5.A, E=@n5.B)
  ]}
 RigVMFunction_Sequence_2_ContainedGraph {nodes [
   n1 = RigVMFunctionEntryNode()
   n2 = &RigVMFunction_Sequence(ExecuteContext=@n1.ExecuteContext)
   n3 = &RigVMFunction_Sequence(ExecuteContext=@n2.B)
   n4 = &RigVMFunction_Sequence(ExecuteContext=@n3.B)
   n5 = &RigVMFunction_Sequence(ExecuteContext=@n4.B)
   RigVMFunctionReturnNode(A=@n2.A, B=@n3.A, C=@n4.A, D=@n5.A, E=@n5.B)
  ]}
 RigVMFunction_Sequence_2_ContainedGraph {nodes [
   n1 = RigVMFunctionEntryNode()
   n2 = &RigVMFunction_Sequence(ExecuteContext=@n1.ExecuteContext)
   n3 = &RigVMFunction_Sequence(ExecuteContext=@n2.B)
   n4 = &RigVMFunction_Sequence(ExecuteContext=@n3.B)
   n5 = &RigVMFunction_Sequence(ExecuteContext=@n4.B)
   RigVMFunctionReturnNode(A=@n2.A, B=@n3.A, C=@n4.A, D=@n5.A, E=@n5.B)
  ]}
 RigVMFunction_Sequence_2_ContainedGraph {nodes [
   n1 = RigVMFunctionEntryNode()
   n2 = &RigVMFunction_Sequence(ExecuteContext=@n1.ExecuteContext)
   n3 = &RigVMFunction_Sequence(ExecuteContext=@n2.B)
   RigVMFunctionReturnNode(A=@n2.A, B=@n3.A, C=@n3.B)
  ]}
 RigVMFunction_Sequence_2_ContainedGraph {nodes [
   n1 = RigVMFunctionEntryNode()
   n2 = &RigVMFunction_Sequence(ExecuteContext=@n1.ExecuteContext)
   n3 = &RigVMFunction_Sequence(ExecuteContext=@n2.B)
   RigVMFunctionReturnNode(A=@n2.A, B=@n3.A, C=@n3.B)
  ]}
 RigVMFunction_Sequence_2_ContainedGraph {nodes [
   n1 = RigVMFunctionEntryNode()
   n2 = &RigVMFunction_Sequence(ExecuteContext=@n1.ExecuteContext)
   n3 = &RigVMFunction_Sequence(ExecuteContext=@n2.B)
   n4 = &RigVMFunction_Sequence(ExecuteContext=@n3.B)
   n5 = &RigVMFunction_Sequence(ExecuteContext=@n4.B)
   RigVMFunctionReturnNode(A=@n2.A, B=@n3.A, C=@n4.A, D=@n5.A, E=@n5.B)
  ]}
 RigVMFunction_Sequence_2_ContainedGraph {nodes [
   n1 = RigVMFunctionEntryNode()
   n2 = &RigVMFunction_Sequence(ExecuteContext=@n1.ExecuteContext)
   n3 = &RigVMFunction_Sequence(ExecuteContext=@n2.B)
   RigVMFunctionReturnNode(A=@n2.A, B=@n3.A, C=@n3.B)
  ]}
 RigVMFunction_Sequence_2_ContainedGraph {nodes [
   n1 = RigVMFunctionEntryNode()
   n2 = &RigVMFunction_Sequence(ExecuteContext=@n1.ExecuteContext)
   n3 = &RigVMFunction_Sequence(ExecuteContext=@n2.B)
   RigVMFunctionReturnNode(A=@n2.A, B=@n3.A, C=@n3.B)
  ]}
 RigVMFunction_Sequence_2_ContainedGraph {nodes [
   n1 = RigVMFunctionEntryNode()
   n2 = &RigVMFunction_Sequence(ExecuteContext=@n1.ExecuteContext)
   n3 = &RigVMFunction_Sequence(ExecuteContext=@n2.B)
   n4 = &RigVMFunction_Sequence(ExecuteContext=@n3.B)
   RigVMFunctionReturnNode(A=@n2.A, B=@n3.A, C=@n4.A, D=@n4.B)
  ]}
 RigVMFunction_Sequence_2_ContainedGraph {nodes [
   n1 = RigVMFunctionEntryNode()
   n2 = &RigVMFunction_Sequence(ExecuteContext=@n1.ExecuteContext)
   n3 = &RigVMFunction_Sequence(ExecuteContext=@n2.B)
   n4 = &RigVMFunction_Sequence(ExecuteContext=@n3.B)
   n5 = &RigVMFunction_Sequence(ExecuteContext=@n4.B)
   n6 = &RigVMFunction_Sequence(ExecuteContext=@n5.B)
   RigVMFunctionReturnNode(A=@n2.A, B=@n3.A, C=@n4.A, D=@n5.A, E=@n6.A, F=@n6.B)
  ]}
 RigVMFunction_Sequence_2_ContainedGraph {nodes [
   n1 = RigVMFunctionEntryNode()
   n2 = &RigVMFunction_Sequence(ExecuteContext=@n1.ExecuteContext)
   n3 = &RigVMFunction_Sequence(ExecuteContext=@n2.B)
   n4 = &RigVMFunction_Sequence(ExecuteContext=@n3.B)
   RigVMFunctionReturnNode(A=@n2.A, B=@n3.A, C=@n4.A, D=@n4.B)
  ]}
 RigVMFunction_Sequence_2_ContainedGraph {nodes [
   n1 = RigVMFunctionEntryNode()
   n2 = &RigVMFunction_Sequence(ExecuteContext=@n1.ExecuteContext)
   n3 = &RigVMFunction_Sequence(ExecuteContext=@n2.B)
   n4 = &RigVMFunction_Sequence(ExecuteContext=@n3.B)
   n5 = &RigVMFunction_Sequence(ExecuteContext=@n4.B)
   RigVMFunctionReturnNode(A=@n2.A, B=@n3.A, C=@n4.A, D=@n5.A, E=@n5.B)
  ]}
 RigVMFunction_Sequence_2_ContainedGraph {nodes [
   n1 = RigVMFunctionEntryNode()
   n2 = &RigVMFunction_Sequence(ExecuteContext=@n1.ExecuteContext)
   n3 = &RigVMFunction_Sequence(ExecuteContext=@n2.B)
   n4 = &RigVMFunction_Sequence(ExecuteContext=@n3.B)
   n5 = &RigVMFunction_Sequence(ExecuteContext=@n4.B)
   n6 = &RigVMFunction_Sequence(ExecuteContext=@n5.B)
   n7 = &RigVMFunction_Sequence(ExecuteContext=@n6.B)
   RigVMFunctionReturnNode(A=@n2.A, B=@n3.A, C=@n4.A, D=@n5.A, E=@n6.A, F=@n7.A, G=@n7.B)
  ]}
 RigVMFunction_Sequence_2_ContainedGraph {nodes [
   n1 = RigVMFunctionEntryNode()
   n2 = &RigVMFunction_Sequence(ExecuteContext=@n1.ExecuteContext)
   n3 = &RigVMFunction_Sequence(ExecuteContext=@n2.B)
   n4 = &RigVMFunction_Sequence(ExecuteContext=@n3.B)
   RigVMFunctionReturnNode(A=@n2.A, B=@n3.A, C=@n4.A, D=@n4.B)
  ]}
 RigVMFunction_Sequence_2_ContainedGraph {nodes [
   n1 = RigVMFunctionEntryNode()
   n2 = &RigVMFunction_Sequence(ExecuteContext=@n1.ExecuteContext)
   n3 = &RigVMFunction_Sequence(ExecuteContext=@n2.B)
   n4 = &RigVMFunction_Sequence(ExecuteContext=@n3.B)
   RigVMFunctionReturnNode(A=@n2.A, B=@n3.A, C=@n4.A, D=@n4.B)
  ]}
 RigVMFunction_Sequence_2_ContainedGraph {nodes [
   n1 = RigVMFunctionEntryNode()
   n2 = &RigVMFunction_Sequence(ExecuteContext=@n1.ExecuteContext)
   n3 = &RigVMFunction_Sequence(ExecuteContext=@n2.B)
   RigVMFunctionReturnNode(A=@n2.A, B=@n3.A, C=@n3.B)
  ]}
 RigVMFunction_Sequence_2_ContainedGraph {nodes [
   n1 = RigVMFunctionEntryNode()
   n2 = &RigVMFunction_Sequence(ExecuteContext=@n1.ExecuteContext)
   n3 = &RigVMFunction_Sequence(ExecuteContext=@n2.B)
   RigVMFunctionReturnNode(A=@n2.A, B=@n3.A, C=@n3.B)
  ]}
 RigVMFunction_Sequence_2_ContainedGraph {nodes [
   n1 = RigVMFunctionEntryNode()
   n2 = &RigVMFunction_Sequence(ExecuteContext=@n1.ExecuteContext)
   n3 = &RigVMFunction_Sequence(ExecuteContext=@n2.B)
   n4 = &RigVMFunction_Sequence(ExecuteContext=@n3.B)
   n5 = &RigVMFunction_Sequence(ExecuteContext=@n4.B)
   RigVMFunctionReturnNode(A=@n2.A, B=@n3.A, C=@n4.A, D=@n5.A, E=@n5.B)
  ]}
 RigVMFunction_Sequence_3_ContainedGraph {nodes [
   n1 = RigVMFunctionEntryNode()
   n2 = &RigVMFunction_Sequence(ExecuteContext=@n1.ExecuteContext)
   n3 = &RigVMFunction_Sequence(ExecuteContext=@n2.B)
   RigVMFunctionReturnNode(A=@n2.A, B=@n3.A, C=@n3.B)
  ]}
 RigVMFunction_Sequence_3_ContainedGraph {nodes [
   n1 = RigVMFunctionEntryNode()
   n2 = &RigVMFunction_Sequence(ExecuteContext=@n1.ExecuteContext)
   n3 = &RigVMFunction_Sequence(ExecuteContext=@n2.B)
   RigVMFunctionReturnNode(B=@n3.A, C=@n3.B, A=@n2.A)
  ]}
 RigVMFunction_Sequence_3_ContainedGraph {nodes [
   n1 = RigVMFunctionEntryNode()
   n2 = &RigVMFunction_Sequence(ExecuteContext=@n1.ExecuteContext)
   n3 = &RigVMFunction_Sequence(ExecuteContext=@n2.B)
   n4 = &RigVMFunction_Sequence(ExecuteContext=@n3.B)
   n5 = &RigVMFunction_Sequence(ExecuteContext=@n4.B)
   n6 = &RigVMFunction_Sequence(ExecuteContext=@n5.B)
   RigVMFunctionReturnNode(B=@n3.A, A=@n2.A, C=@n4.A, D=@n5.A, E=@n6.A, F=@n6.B)
  ]}
 RigVMFunction_Sequence_3_ContainedGraph {nodes [
   n1 = RigVMFunctionEntryNode()
   n2 = &RigVMFunction_Sequence(ExecuteContext=@n1.ExecuteContext)
   n3 = &RigVMFunction_Sequence(ExecuteContext=@n2.B)
   n4 = &RigVMFunction_Sequence(ExecuteContext=@n3.B)
   n5 = &RigVMFunction_Sequence(ExecuteContext=@n4.B)
   RigVMFunctionReturnNode(B=@n3.A, A=@n2.A, C=@n4.A, D=@n5.A, E=@n5.B)
  ]}
 RigVMFunction_Sequence_4_ContainedGraph {nodes [
   n1 = RigVMFunctionEntryNode()
   n2 = &RigVMFunction_Sequence(ExecuteContext=@n1.ExecuteContext)
   n3 = &RigVMFunction_Sequence(ExecuteContext=@n2.B)
   n4 = &RigVMFunction_Sequence(ExecuteContext=@n3.B)
   n5 = &RigVMFunction_Sequence(ExecuteContext=@n4.B)
   n6 = &RigVMFunction_Sequence(ExecuteContext=@n5.B)
   n7 = &RigVMFunction_Sequence(ExecuteContext=@n6.B)
   n8 = &RigVMFunction_Sequence(ExecuteContext=@n7.B)
   RigVMFunctionReturnNode(B=@n3.A, C=@n4.A, A=@n2.A, D=@n5.A, E=@n6.A, F=@n7.A, G=@n8.A, H=@n8.B)
  ]}
 RigVMFunction_Sequence_4_ContainedGraph {nodes [
   n1 = RigVMFunctionEntryNode()
   n2 = &RigVMFunction_Sequence(ExecuteContext=@n1.ExecuteContext)
   n3 = &RigVMFunction_Sequence(ExecuteContext=@n2.B)
   n4 = &RigVMFunction_Sequence(ExecuteContext=@n3.B)
   n5 = &RigVMFunction_Sequence(ExecuteContext=@n4.B)
   n6 = &RigVMFunction_Sequence(ExecuteContext=@n5.B)
   n7 = &RigVMFunction_Sequence(ExecuteContext=@n6.B)
   RigVMFunctionReturnNode(B=@n3.A, C=@n4.A, A=@n2.A, D=@n5.A, E=@n6.A, F=@n7.A, G=@n7.B)
  ]}
 RigVMFunction_Sequence_4_ContainedGraph {nodes [
   n1 = RigVMFunctionEntryNode()
   n2 = &RigVMFunction_Sequence(ExecuteContext=@n1.ExecuteContext)
   n3 = &RigVMFunction_Sequence(ExecuteContext=@n2.B)
   n4 = &RigVMFunction_Sequence(ExecuteContext=@n3.B)
   n5 = &RigVMFunction_Sequence(ExecuteContext=@n4.B)
   n6 = &RigVMFunction_Sequence(ExecuteContext=@n5.B)
   n7 = &RigVMFunction_Sequence(ExecuteContext=@n6.B)
   RigVMFunctionReturnNode(B=@n3.A, C=@n4.A, A=@n2.A, D=@n5.A, E=@n6.A, F=@n7.A, G=@n7.B)
  ]}
 RigVMFunction_Sequence_4_ContainedGraph {nodes [
   n1 = RigVMFunctionEntryNode()
   n2 = &RigVMFunction_Sequence(ExecuteContext=@n1.ExecuteContext)
   n3 = &RigVMFunction_Sequence(ExecuteContext=@n2.B)
   n4 = &RigVMFunction_Sequence(ExecuteContext=@n3.B)
   n5 = &RigVMFunction_Sequence(ExecuteContext=@n4.B)
   n6 = &RigVMFunction_Sequence(ExecuteContext=@n5.B)
   n7 = &RigVMFunction_Sequence(ExecuteContext=@n6.B)
   n8 = &RigVMFunction_Sequence(ExecuteContext=@n7.B)
   RigVMFunctionReturnNode(B=@n3.A, C=@n4.A, A=@n2.A, D=@n5.A, E=@n6.A, F=@n7.A, G=@n8.A, H=@n8.B)
  ]}
 RigVMFunction_Sequence_5_ContainedGraph {nodes [
   n1 = RigVMFunctionEntryNode()
   n2 = &RigVMFunction_Sequence(ExecuteContext=@n1.ExecuteContext)
   n3 = &RigVMFunction_Sequence(ExecuteContext=@n2.B)
   n4 = &RigVMFunction_Sequence(ExecuteContext=@n3.B)
   RigVMFunctionReturnNode(B=@n3.A, C=@n4.A, D=@n4.B, A=@n2.A)
  ]}
 RigVMFunction_Sequence_5_ContainedGraph {nodes [
   n1 = RigVMFunctionEntryNode()
   n2 = &RigVMFunction_Sequence(ExecuteContext=@n1.ExecuteContext)
   n3 = &RigVMFunction_Sequence(ExecuteContext=@n2.B)
   n4 = &RigVMFunction_Sequence(ExecuteContext=@n3.B)
   RigVMFunctionReturnNode(B=@n3.A, C=@n4.A, D=@n4.B, A=@n2.A)
  ]}
 RigVMModel {nodes [
   RigUnit_BeginExecution = &RigUnit_BeginExecution()
   RigVMFunction_Sequence = RigVMAggregateNode(ExecuteContext=@RigUnit_BeginExecution.ExecutePin)
   Main Leg Ik System = RigVMFunctionReferenceNode(Initial=False, ExecuteContext=@RigVMFunction_Sequence.A)
   > Knee Offset Ik System = RigVMFunctionReferenceNode(Initial=False)
   > Ankle Offset Ik System = RigVMFunctionReferenceNode(Initial=False)
   Twist Knee = RigVMFunctionReferenceNode(ExecuteContext=@RigVMFunction_Sequence.B)
   Set Toe Fk Controls = RigVMFunctionReferenceNode(ExecuteContext=@RigVMFunction_Sequence.C)
   Set Ik Bone Transforms = RigVMFunctionReferenceNode(ExecuteContext=@RigVMFunction_Sequence.D)
   > Set Bone Transforms = RigVMFunctionReferenceNode()
   Set Control Lines = RigVMFunctionReferenceNode(ExecuteContext=@RigVMFunction_Sequence.E)
   > Set Sec Control Visibility = RigVMFunctionReferenceNode()
   Set Blend Bone Transforms = RigVMFunctionReferenceNode(ExecuteContext=())
  ]}
 RigVMModel Rig {nodes [
   n1 = /Script/ControlRig.RigUnit_InverseExecution() {is=struct RigVMUnitNode}
   RigVMFunction_Sequence = RigVMAggregateNode(ExecuteContext=@n1.ExecutePin)
   Debug Anim Sequence Bone Positions = RigVMFunctionReferenceNode(ExecuteContext=@RigVMFunction_Sequence.A)
   Position Hip Socket Control = RigVMFunctionReferenceNode(ExecuteContext=@RigVMFunction_Sequence.B)
   > Position Ik Control = RigVMFunctionReferenceNode()
   Backward Main Leg Ik System = RigVMFunctionReferenceNode(ExecuteContext=@RigVMFunction_Sequence.C)
   > Backward Knee Offset Ik System = RigVMFunctionReferenceNode()
   Position Hock Rotate Control Null = RigVMFunctionReferenceNode(ExecuteContext=@RigVMFunction_Sequence.D)
   Position Knee Rotation Control = RigVMFunctionReferenceNode(ExecuteContext=@RigVMFunction_Sequence.E)
   Position Ball Fk Control = RigVMFunctionReferenceNode(ExecuteContext=@RigVMFunction_Sequence.F)
   > Position Toe Fk Control = RigVMFunctionReferenceNode()
   Position Main PV Control = RigVMFunctionReferenceNode(ExecuteContext=())
  ]}
]
hierarchy [
 Hip Socket {type=Connector, loc=<0 0 0>, rot=<0 0 0 1>}
 Root {type=Socket, loc=<0 0 0>, rot=<0 0 0 1>}
 Knee {type=Connector, loc=<0 0 0>, rot=<0 0 0 1>}
 Ankle {type=Connector, loc=<0 0 0>, rot=<0 0 0 1>}
 Ball {type=Connector, loc=<0 0 0>, rot=<0 0 0 1>}
 Toe {type=Connector, loc=<0 0 0>, rot=<0 0 0 1>}
 Toe Tip {type=Connector, loc=<0 0 0>, rot=<0 0 0 1>}
 IK Rear Connector {type=Connector, loc=<0 0 0>, rot=<0 0 0 1>}
]
variables [
 Ankle Backwards Transform {type=FTransform, default=(Rotation=(X=0,Y=0,Z=0,W=1),Translation=(X=0,Y=0,Z=0),Scale3D=(X=1,Y=1,Z=1))}
 Ankle Connector {type=FRigElementKey, default=(Type=Connector,Name="")}
 Ankle Input Transform {type=FTransform, default=(Rotation=(X=0,Y=0,Z=0,W=1),Translation=(X=0,Y=0,Z=0),Scale3D=(X=1,Y=1,Z=1))}
 Ankle Local Transform {type=FTransform, default=(Rotation=(X=0,Y=0,Z=0,W=1),Translation=(X=0,Y=0,Z=0),Scale3D=(X=1,Y=1,Z=1))}
 Ankle Rig Transform {type=FTransform, default=(Rotation=(X=0,Y=0,Z=0,W=1),Translation=(X=0,Y=0,Z=0),Scale3D=(X=1,Y=1,Z=1))}
 Ankle Rotate Aim Point {type=FRigElementKey, default=(Type=Bone,Name="")}
 Ankle Twist {type=FRigElementKey, default=(Type=Bone,Name="")}
 Ankle Twist Offset {type=double, default=0}
 Ball Backwards Transform {type=FTransform, default=(Rotation=(X=0,Y=0,Z=0,W=1),Translation=(X=0,Y=0,Z=0),Scale3D=(X=1,Y=1,Z=1))}
 Ball Bone Length {type=double, default=0}
 Ball Connector {type=FRigElementKey, default=(Type=Connector,Name="")}
 Ball Fk Control {type=FRigElementKey, default=(Type=Bone,Name="")}
 Ball Input Transform {type=FTransform, default=(Rotation=(X=0,Y=0,Z=0,W=1),Translation=(X=0,Y=0,Z=0),Scale3D=(X=1,Y=1,Z=1))}
 Ball Local Tranform {type=FTransform, default=(Rotation=(X=0,Y=0,Z=0,W=1),Translation=(X=0,Y=0,Z=0),Scale3D=(X=1,Y=1,Z=1))}
 Ball Pivot Control {type=FRigElementKey, default=(Type=Bone,Name="")}
 Ball Point Ref {type=FRigElementKey, default=(Type=Bone,Name="")}
 Ball Rig Transform {type=FTransform, default=(Rotation=(X=0,Y=0,Z=0,W=1),Translation=(X=0,Y=0,Z=0),Scale3D=(X=1,Y=1,Z=1))}
 Connectors {type="TArray<FRigElementKey>", default=(())}
 Control Size {type=double, default=1}
 Custom Side {type=bool, default=False}
 Custom Side Color {type=FLinearColor, default=(R=1,G=0,B=1,A=1)}
 Custom Side Name {type=FString, default=ZZ}
 Debug Lower Ik PV Points {type="TArray<FVector>"}
 Debug Lower Leg Ik Aim Points {type="TArray<FVector>"}
 Debug Lower Leg Ik Points {type="TArray<FVector>"}
 Debug Main Leg Ik Points {type="TArray<FVector>"}
 Debug Upper Ik PV Points {type="TArray<FVector>"}
 Debug Upper Leg Ik Aim Points {type="TArray<FVector>"}
 Debug Upper Leg Ik Foot Points {type="TArray<FVector>"}
 Debug Upper Leg Ik Points {type="TArray<FVector>"}
 DebugOnOff {type=bool, default=False}
 First Bone Tip Local Offset {type=FTransform, default=(Rotation=(X=0,Y=0,Z=0,W=1),Translation=(X=0,Y=0,Z=0),Scale3D=(X=1,Y=1,Z=1))}
 Foot Bone Length {type=double, default=0}
 Heel Roll Control {type=FRigElementKey, default=(Type=Bone,Name="")}
 Heel Roll Initial Transform {type=FTransform, default=(Rotation=(X=0,Y=0,Z=0,W=1),Translation=(X=0,Y=-9,Z=0),Scale3D=(X=1,Y=1,Z=1))}
 Hip Socket Backwards Transform {type=FTransform, default=(Rotation=(X=0,Y=0,Z=0,W=1),Translation=(X=0,Y=0,Z=0),Scale3D=(X=1,Y=1,Z=1))}
 Hip Socket Bone Length {type=double, default=0}
 Hip Socket Connector {type=FRigElementKey, default=(Type=Connector,Name="")}
 Hip Socket Control {type=FRigElementKey, default=(Type=Bone,Name="")}
 Hip Socket Input Transform {type=FTransform, default=(Rotation=(X=0,Y=0,Z=0,W=1),Translation=(X=0,Y=0,Z=0),Scale3D=(X=1,Y=1,Z=1))}
 Hip Socket Local Transform {type=FTransform, default=(Rotation=(X=0,Y=0,Z=0,W=1),Translation=(X=0,Y=0,Z=0),Scale3D=(X=1,Y=1,Z=1))}
 Hip Socket Null {type=FRigElementKey, default=(Type=Bone,Name="")}
 Hip Socket Ref {type=FRigElementKey, default=(Type=Bone,Name="")}
 Hip Socket Rig Transform {type=FTransform, default=(Rotation=(X=0,Y=0,Z=0,W=1),Translation=(X=0,Y=0,Z=0),Scale3D=(X=1,Y=1,Z=1))}
 Hock Control Transform {type=FTransform, default=(Rotation=(X=0.5,Y=0.5,Z=0.5,W=0.5),Translation=(X=0,Y=-2,Z=12),Scale3D=(X=0.7,Y=0.7,Z=0.7))}
 Hock Offset {type=FRigElementKey, default=(Type=Bone,Name="")}
 Hock Rotate Aim Null {type=FRigElementKey, default=(Type=Bone,Name="")}
 Hock Rotate Control {type=FRigElementKey, default=(Type=Bone,Name="")}
 Hock Rotate Offset {type=FRigElementKey, default=(Type=Bone,Name="")}
 IK Front Connector {type=FRigElementKey, default=(Type=Bone,Name="IK_Fot_Rear_L")}
 Ik Control {type=FRigElementKey, default=(Type=Bone,Name="")}
 Ik Control Transform {type=FTransform, default=(Rotation=(X=0,Y=0,Z=1,W=0),Translation=(X=0.3,Y=2,Z=-2.1),Scale3D=(X=0.5,Y=0.5,Z=1))}
 Ik Fk Blend {type=double, default=0}
 Ik Fk Bones {
  type="TArray<FRigElementKeyCollection>"
  default=((Keys=((),(),())),(Keys=((),(),())),(Keys=((),(),())),(Keys=((),(),())),(Keys=((),(),())),(Keys=((),(),())))
 }
 Ik Max Distance {type=double, default=100}
 Ik Min Distance {type=double, default=30}
 Ik Parent {type=FRigElementKey, default=(Type=Bone,Name="")}
 Ik Point {type=FRigElementKey, default=(Type=Bone,Name="")}
 Knee Backwards Transform {type=FTransform, default=(Rotation=(X=0,Y=0,Z=0,W=1),Translation=(X=0,Y=0,Z=0),Scale3D=(X=1,Y=1,Z=1))}
 Knee Bone Length {type=double, default=0}
 Knee Connector {type=FRigElementKey, default=(Type=Connector,Name="")}
 Knee Input Transform {type=FTransform, default=(Rotation=(X=0,Y=0,Z=0,W=1),Translation=(X=0,Y=0,Z=0),Scale3D=(X=1,Y=1,Z=1))}
 Knee Local Transform {type=FTransform, default=(Rotation=(X=0,Y=0,Z=0,W=1),Translation=(X=0,Y=0,Z=0),Scale3D=(X=1,Y=1,Z=1))}
 Knee Offset {type=FRigElementKey, default=(Type=Bone,Name="")}
 Knee PV {type=FRigElementKey, default=(Type=Bone,Name="")}
 Knee Rig Transform {type=FTransform, default=(Rotation=(X=0,Y=0,Z=0,W=1),Translation=(X=0,Y=0,Z=0),Scale3D=(X=1,Y=1,Z=1))}
 Knee Rotate Control {type=FRigElementKey, default=(Type=Bone,Name="")}
 Knee Rotate Control Null {type=FRigElementKey, default=(Type=Bone,Name="")}
 Knee Rotate Up Vector {type=FRigElementKey, default=(Type=Bone,Name="")}
 Knee Twist {type=FRigElementKey, default=(Type=Bone,Name="")}
 Knee Twist Offset {type=double, default=0}
 Left Side {type=bool, default=True}
 Left Side Color {type=FLinearColor, default=(R=0,G=0,B=1,A=1)}
 Lower Ik PV Local Transform {type=FTransform, default=(Rotation=(X=0,Y=0,Z=0,W=1),Translation=(X=0,Y=0,Z=0),Scale3D=(X=1,Y=1,Z=1))}
 Main Hip Socket Local Transform {type=FTransform, default=(Rotation=(X=0,Y=0,Z=0,W=1),Translation=(X=0,Y=0,Z=0),Scale3D=(X=1,Y=1,Z=1))}
 Main PV Control {type=FRigElementKey, default=(Type=Bone,Name="")}
 PV Space Switch {type=FRigElementKey, default=(Type=None,Name="")}
 Rig Up Vector {type=FVector, default=(X=30,Y=0,Z=0)}
 Right Side {type=bool, default=False}
 Right Side Color {type=FLinearColor, default=(R=1,G=0,B=0,A=1)}
 Sec Constrol Visility {type=FRigElementKey, default=(Type=None,Name="")}
 Side Color {type=FLinearColor, default=(R=0,G=0,B=0,A=0)}
 Toe Backwards Transform {type=FTransform, default=(Rotation=(X=0,Y=0,Z=0,W=1),Translation=(X=0,Y=0,Z=0),Scale3D=(X=1,Y=1,Z=1))}
 Toe Bone Length {type=double, default=0}
 Toe Connector {type=FRigElementKey, default=(Type=Connector,Name="")}
 Toe Fk Control {type=FRigElementKey, default=(Type=Bone,Name="")}
 Toe Input Transform {type=FTransform, default=(Rotation=(X=0,Y=0,Z=0,W=1),Translation=(X=0,Y=0,Z=0),Scale3D=(X=1,Y=1,Z=1))}
 Toe Local Transform {type=FTransform, default=(Rotation=(X=0,Y=0,Z=0,W=1),Translation=(X=0,Y=0,Z=0),Scale3D=(X=1,Y=1,Z=1))}
 Toe Rig Transform {type=FTransform, default=(Rotation=(X=0,Y=0,Z=0,W=1),Translation=(X=0,Y=0,Z=0),Scale3D=(X=1,Y=1,Z=1))}
 Toe Roll Control {type=FRigElementKey, default=(Type=Bone,Name="")}
 Toe Roll Initial Transform {type=FTransform, default=(Rotation=(X=0,Y=0,Z=0,W=1),Translation=(X=0,Y=7,Z=-3),Scale3D=(X=1,Y=1,Z=1))}
 Toe Tip Backwards Transform {type=FTransform, default=(Rotation=(X=0,Y=0,Z=0,W=1),Translation=(X=0,Y=0,Z=0),Scale3D=(X=1,Y=1,Z=1))}
 Toe Tip Connector {type=FRigElementKey, default=(Type=Connector,Name="")}
 Toe Tip Input Transform {type=FTransform, default=(Rotation=(X=0,Y=0,Z=0,W=1),Translation=(X=0,Y=0,Z=0),Scale3D=(X=1,Y=1,Z=1))}
 Toe Tip Local Transform {type=FTransform, default=(Rotation=(X=0,Y=0,Z=0,W=1),Translation=(X=0,Y=0,Z=0),Scale3D=(X=1,Y=1,Z=1))}
 Toe Tip Point {type=FRigElementKey, default=(Type=Bone,Name="")}
 Toe Tip Rig Transform {type=FTransform, default=(Rotation=(X=0,Y=0,Z=0,W=1),Translation=(X=0,Y=0,Z=0),Scale3D=(X=1,Y=1,Z=1))}
 Upper Bone Auto Angle {type=double, default=0}
 Upper Ik PV Local Transform {type=FTransform, default=(Rotation=(X=0,Y=0,Z=0,W=1),Translation=(X=0,Y=0,Z=0),Scale3D=(X=1,Y=1,Z=1))}
 Upper Leg Angle Max Distance {type=double, default=54}
 Upper Leg Angle Min Distance {type=double, default=15}
 Upper Leg Max Angle {type=double, default=30}
 Upper Leg Min Angle {type=double, default=0}
]
