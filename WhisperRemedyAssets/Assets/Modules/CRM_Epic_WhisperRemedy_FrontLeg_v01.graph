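# Modular Control Rig module for a front leg with a scapula, driven from connectors for scapula tip through toe tip.
# Construction resolves connectors, spawns IK, scapula, up-vector, roll, toe and twist controls, then caches local transforms and sets space switches.
# The scapula aims at a point following a percentage of IK control movement.
# Two chained law-of-cosines solves give shoulder, elbow and wrist angles for the leg IK.
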ControlRigBlueprint /WhisperRemedyAssets/Assets/Modules/CRM_Epic_WhisperRemedy_FrontLeg_v01
graphs [
 ConstructionGraph {nodes [
   RigUnit_PrepareForExecution = &RigUnit_PrepareForExecution()
   RigVMFunction_Sequence = RigVMAggregateNode(ExecuteContext=@RigUnit_PrepareForExecution.ExecutePin)
   Get Connectors = RigVMFunctionReferenceNode(ExecuteContext=@RigVMFunction_Sequence.A)
   > Get Lengths = RigVMFunctionReferenceNode()
   Create Scapula Tip Control = RigVMFunctionReferenceNode(ExecuteContext=@RigVMFunction_Sequence.B)
   > Create Rig Up Vector = RigVMFunctionReferenceNode()
   Create Ik Control = RigVMFunctionReferenceNode(ExecuteContext=@RigVMFunction_Sequence.C)
   > Create Roll Controls = RigVMFunctionReferenceNode()
   > Create Toe Controls = RigVMFunctionReferenceNode()
   > Create Elbow Twist Control = RigVMFunctionReferenceNode()
   Create Rig Transforms = RigVMFunctionReferenceNode(ExecuteContext=@RigVMFunction_Sequence.D)
   > Create Local Transforms = RigVMFunctionReferenceNode()
   > Create Backward Local Transforms = RigVMFunctionReferenceNode()
   Set Metadata = RigVMFunctionReferenceNode(ExecuteContext=@RigVMFunction_Sequence.E)
   > Create Space Switches = RigVMFunctionReferenceNode()
   > Create Sec Control Visibility Switch = RigVMFunctionReferenceNode()
  ]}
 New Function_ContainedGraph {nodes [
   n1 = RigVMFunctionEntryNode()
   RigVMFunction_Sequence = &RigVMFunction_Sequence(ExecuteContext=@n1.ExecuteContext)
   RigVMFunction_Sequence_1 = RigVMAggregateNode(ExecuteContext=@RigVMFunction_Sequence.A)
   RigVMVariableNode(
    Value=/Script/ControlRig.RigUnit_ResolveConnector(Connector=(Type=Connector,Name="Scapula Tip"), SkipSocket=False) {is=struct RigVMUnitNode}.Result
    ExecuteContext=@RigVMFunction_Sequence_1.A
   ) {variable=Scapula Tip Connector}
   > RigVMVariableNode(Value=&RigUnit_ResolveConnector(Connector=(Type=Connector,Name="Shoulder"), SkipSocket=False).Result) {variable=Shoulder Connector}
   RigVMVariableNode(
    Value=&RigUnit_ResolveConnector(Connector=(Type=Connector,Name="Elbow"), SkipSocket=False).Result
    ExecuteContext=@RigVMFunction_Sequence_1.B
   ) {variable=Elbow Connector}
   > RigVMVariableNode(Value=&RigUnit_ResolveConnector(Connector=(Type=Connector,Name="Wrist"), SkipSocket=False).Result) {variable=Wrist Connector}
   > RigVMVariableNode(Value=&RigUnit_ResolveConnector(Connector=(Type=Connector,Name="Ball"), SkipSocket=False).Result) {variable=Ball Connector}
   RigVMVariableNode(
    Value=&RigUnit_ResolveConnector(Connector=(Type=Connector,Name="Toe"), SkipSocket=False).Result
    ExecuteContext=@RigVMFunction_Sequence_1.C
   ) {variable=Toe Connector}
   > RigVMVariableNode(Value=&RigUnit_ResolveConnector(Connector=(Type=Connector,Name="Toe Tip"), SkipSocket=False).Result) {variable=Toe Tip Connector}
   RigVMFunctionReturnNode(ExecuteContext=@RigVMFunction_Sequence.B)
  ]}
 New Function_ContainedGraph {nodes [
   Create Name_3 = RigVMFunctionReferenceNode(ExecuteContext=(), Tokens=("Fr Ik Point"))
   > n1 = &RigUnit_HierarchyAddNull(
    Space=LocalSpace
    Transform=(Rotation=(X=0,Y=0,Z=0,W=1),Translation=(X=0,Y=0,Z=0),Scale3D=(X=1,Y=1,Z=1))
    Name=@"Create Name_3".Name
    Parent=RigVMVariableNode() {variable=Ik Control}.Value
   )
   > RigVMVariableNode(Value=@n1.Item) {variable=Ik Point}
   n2 = RigVMFunctionEntryNode()
   RigVMFunction_Sequence = &RigVMFunction_Sequence(ExecuteContext=@n2.ExecuteContext)
   n3 = &RigVMFunction_Sequence(ExecuteContext=@RigVMFunction_Sequence.A)
   Create Name = RigVMFunctionReferenceNode(Tokens=("Fr Ik Null"), ExecuteContext=@n3.A)
   > n4 = &RigUnit_HierarchyAddNull(
    Space=GlobalSpace
    Transform=(Rotation=(X=0,Y=0,Z=0,W=1),Translation=(X=0,Y=0,Z=0),Scale3D=(X=1,Y=1,Z=1))
    Name=@"Create Name".Name
    Parent=RigVMDispatchNode(Default=(Type=Bone,Name="None"), Name=Offset Control, NameSpace=Root) {template="DISPATCH_RigDispatch_GetModuleMetadata(in Name,in NameSpace,in Default,out Value,out Found)"}.Value
    Transform.Translation=&RigUnit_GetTransform(bInitial=true, Item=RigVMVariableNode() {variable=Ball Connector}.Value).Transform.Translation
   )
   > RigVMVariableNode(Value=@n4.Item) {variable=Null}
   > Set Side Color = RigVMFunctionReferenceNode()
   > Create Name_1 = RigVMFunctionReferenceNode(Tokens=("Fr Ik"))
   n5 = RigVMVariableNode() {variable=Control Scale}
   n6 = RigVMVariableNode() {variable=IK Control Transform}
   Multiply = &RigVMFunction_MathDoubleMul(B=@n5.Value, A=@n6.Value.Scale3D.X)
   > SpawnControl = /Script/ControlRig.RigUnit_HierarchyAddControlTransform(
    InitialValue=(Rotation=(X=0,Y=0,Z=0,W=1),Translation=(X=0,Y=0,Z=0),Scale3D=(X=1,Y=1,Z=1))
    OffsetSpace=LocalSpace
    OffsetTransform=(Rotation=(X=0,Y=0,Z=0,W=1),Translation=(X=0,Y=0,Z=0),Scale3D=(X=1,Y=1,Z=1))
    Settings="(InitialSpace=LocalSpace,bUsePreferredRotationOrder=False,PreferredRotationOrder=YZX,Limits=(LimitTranslationX=(bMinimum=False,bMaximum=False),LimitTranslationY=(bMinimum=False,bMaximum=False),LimitTranslationZ=(bMinimum=False,bMaximum=False),LimitPitch=(b ...(truncated)"
    Name=@"Create Name_1".Name
    Settings.Shape.Transform.Scale3D.X=@Multiply.Result
    Settings.Shape.Transform.Scale3D.Y=&RigVMFunction_MathDoubleMul(B=@n5.Value, A=@n6.Value.Scale3D.Y).Result
    Settings.Shape.Transform.Scale3D.Z=&RigVMFunction_MathDoubleMul(B=@n5.Value, A=@n6.Value.Scale3D.Z).Result
    Settings.Shape.Transform.Rotation=@n6.Value.Rotation
    Settings.Shape.Transform.Translation=@n6.Value.Translation
    Parent=RigVMVariableNode() {variable=Null}.Value
    Settings.Shape.Color=@"Set Side Color".Color
   ) {
    is=struct RigVMUnitNode
    template="SpawnControl::Execute(in InitialValue,in Settings,in OffsetTransform,in OffsetSpace,in Parent,in Name,out Item)"
   }
   > RigVMVariableNode(Value=@SpawnControl.Item) {variable=Ik Control}
   > Create Name_2 = RigVMFunctionReferenceNode(Tokens=("Scapula Tip Aim Point"))
   > n7 = &RigUnit_HierarchyAddNull(
    Space=LocalSpace
    Transform=(Rotation=(X=0,Y=0,Z=0,W=1),Translation=(X=0,Y=0,Z=0),Scale3D=(X=1,Y=1,Z=1))
    Name=@"Create Name_2".Name
    Parent=RigVMVariableNode() {variable=Null}.Value
   )
   > RigVMVariableNode(Value=@n7.Item) {variable=Scapula Tip Aim Point}
   n8 = /Script/ControlRig.RigUnit_HierarchyAddAnimationChannelFloat(
    InitialValue=0
    LimitsEnabled=(Enabled=(bMinimum=False,bMaximum=False))
    MaximumValue=45
    MinimumValue=-45
    Name=Shoulder Rotate X Offset
    Parent=RigVMVariableNode() {variable=Ik Control}.Value
    ExecutePin=@n3.B
   ) {
    is=struct RigVMUnitNode
    template="SpawnAnimationChannel::Execute(in InitialValue,in MinimumValue,in MaximumValue,in LimitsEnabled,in Parent,in Name,out Item)"
   }
   > RigVMVariableNode(Value=@n8.Item) {variable=Shoulder Rotate Offset}
   > SpawnAnimationChannel = &RigUnit_HierarchyAddAnimationChannelFloat(
    InitialValue=0
    LimitsEnabled=(Enabled=(bMinimum=False,bMaximum=False))
    MaximumValue=45
    MinimumValue=-45
    Name=Ball Rotate
    Parent=RigVMVariableNode() {variable=Ik Control}.Value
   ) {template="SpawnAnimationChannel::Execute(in InitialValue,in MinimumValue,in MaximumValue,in LimitsEnabled,in Parent,in Name,out Item)"}
   > RigVMVariableNode(Value=@SpawnAnimationChannel.Item) {variable=Ball Rotate Channel}
   > Create Name_6 = RigVMFunctionReferenceNode(Tokens=("Ball Rotate Aim Null"))
   Get Main Pole Vector = RigVMFunctionReferenceNode(Initial=true)
   n9 = /Script/ControlRig.RigUnit_AimBoneMath(
    DebugSettings=(bEnabled=False,Scale=10,WorldOffset=(Rotation=(X=0,Y=0,Z=0,W=1),Translation=(X=0,Y=0,Z=0),Scale3D=(X=1,Y=1,Z=1)))
    InputTransform=(Rotation=(X=0,Y=0,Z=0,W=1),Translation=(X=0,Y=0,Z=0),Scale3D=(X=1,Y=1,Z=1))
    Primary=(Weight=1,Axis=(X=1,Y=0,Z=0),Target=(X=1,Y=0,Z=0),Kind=Location,Space=(Type=Bone,Name="None"))
    Secondary=(Weight=1,Axis=(X=0,Y=0,Z=1),Target=(X=1,Y=0,Z=0),Kind=Direction,Space=(Type=Bone,Name="None"))
    Weight=1
    Secondary.Target=@"Get Main Pole Vector".Vector
    InputTransform.Translation=&RigUnit_GetTransform(bInitial=true, Item=RigVMVariableNode() {variable=Ik Control}.Value).Transform.Translation
    Primary.Target=&RigUnit_GetTransform(bInitial=true, Item=RigVMVariableNode() {variable=Elbow Connector}.Value).Transform.Translation
   ) {is=struct RigVMUnitNode}
   > n10 = &RigUnit_HierarchyAddNull(Space=GlobalSpace, Transform=@n9.Result, Name=@"Create Name_6".Name, Parent=RigVMVariableNode() {variable=Ik Control}.Value)
   > RigVMVariableNode(Value=@n10.Item) {variable=Ball Rotate Aim Null}
   > Create Name_5 = RigVMFunctionReferenceNode(Tokens=("Ball Rotate Null"))
   Get Main Pole Vector_1 = RigVMFunctionReferenceNode(Initial=true)
   n11 = &RigUnit_AimBoneMath(
    DebugSettings=(bEnabled=False,Scale=10,WorldOffset=(Rotation=(X=0,Y=0,Z=0,W=1),Translation=(X=0,Y=0,Z=0),Scale3D=(X=1,Y=1,Z=1)))
    InputTransform=(Rotation=(X=0,Y=0,Z=0,W=1),Translation=(X=0,Y=0,Z=0),Scale3D=(X=1,Y=1,Z=1))
    Primary=(Weight=1,Axis=(X=0,Y=0,Z=1),Target=(X=1,Y=0,Z=0),Kind=Location,Space=(Type=Bone,Name="None"))
    Secondary=(Weight=1,Axis=(X=1,Y=0,Z=0),Target=(X=1,Y=0,Z=0),Kind=Direction,Space=(Type=Bone,Name="None"))
    Weight=1
    Secondary.Target=@"Get Main Pole Vector_1".Vector
    InputTransform.Translation=&RigUnit_GetTransform(bInitial=true, Item=RigVMVariableNode() {variable=Ik Control}.Value).Transform.Translation
    Primary.Target=&RigUnit_GetTransform(bInitial=true, Item=RigVMVariableNode() {variable=Wrist Connector}.Value).Transform.Translation
   )
   > n12 = &RigUnit_HierarchyAddNull(Space=GlobalSpace, Transform=@n11.Result, Name=@"Create Name_5".Name, Parent=RigVMVariableNode() {variable=Ball Rotate Aim Null}.Value)
   > Create Name_4 = RigVMFunctionReferenceNode(Tokens=("Ball Rotate"))
   n13 = &RigVMFunction_MathDoubleMul(B=0.75, A=RigVMVariableNode() {variable=Control Scale}.Value)
   > n14 = &RigUnit_HierarchyAddControlTransform(
    InitialValue=(Rotation=(X=0,Y=0,Z=0,W=1),Translation=(X=0,Y=0,Z=0),Scale3D=(X=1,Y=1,Z=1))
    OffsetSpace=LocalSpace
    OffsetTransform=(Rotation=(X=0,Y=0,Z=0,W=1),Translation=(X=0,Y=0,Z=0),Scale3D=(X=1,Y=1,Z=1))
    Settings="(InitialSpace=LocalSpace,bUsePreferredRotationOrder=False,PreferredRotationOrder=YZX,Limits=(LimitTranslationX=(bMinimum=True,bMaximum=True),LimitTranslationY=(bMinimum=True,bMaximum=True),LimitTranslationZ=(bMinimum=True,bMaximum=True),LimitPitch=(bMinimu ...(truncated)"
    Name=@"Create Name_4".Name
    Settings.Shape.Transform.Scale3D.X=@n13.Result
    Settings.Shape.Transform.Scale3D.Y=@n13.Result
    Settings.Shape.Transform.Scale3D.Z=@n13.Result
    Parent=@n12.Item
   ) {template="SpawnControl::Execute(in InitialValue,in Settings,in OffsetTransform,in OffsetSpace,in Parent,in Name,out Item)"}
   > RigVMVariableNode(Value=@n14.Item) {variable=Ball Rotate Control}
   > Create Name_7 = RigVMFunctionReferenceNode(Tokens=("Wrist Point"))
   n15 = &RigUnit_GetTransform(bInitial=true, Item=RigVMVariableNode() {variable=Ball Rotate Control}.Value)
   > n16 = &RigUnit_HierarchyAddNull(
    Space=GlobalSpace
    Transform=(Rotation=(X=0,Y=0,Z=0,W=1),Translation=(X=0,Y=0,Z=0),Scale3D=(X=1,Y=1,Z=1))
    Name=@"Create Name_7".Name
    Transform.Translation=&RigUnit_GetTransform(bInitial=true, Item=RigVMVariableNode() {variable=Wrist Connector}.Value).Transform.Translation
    Transform.Rotation=@n15.Transform.Rotation
    Transform.Scale3D=@n15.Transform.Scale3D
    Parent=RigVMVariableNode() {variable=Ball Rotate Control}.Value
   )
   > RigVMVariableNode(Value=@n16.Item) {variable=Wrist Ik Point}
   RigVMFunctionReturnNode(ExecuteContext=@RigVMFunction_Sequence.B)
   RigVMVariableNode() {variable=Heel Roll Control}
  ]}
 New Function_ContainedGraph {nodes [
   n1 = RigVMFunctionEntryNode()
   > n2 = RigVMVariableNode(Value=@n1.Tokens) {variable=Name Parts}
   RigVMFunction_Sequence = &RigVMFunction_Sequence(ExecuteContext=@n2.ExecuteContext)
   RigVMFunction_StringJoin = /Script/RigVM.RigVMFunction_StringJoin(Separator=" ", Values=RigVMVariableNode() {variable=Name Parts}.Value) {is=struct RigVMUnitNode}
   From_String = RigVMDispatchNode(String=@RigVMFunction_StringJoin.Result) {template="DISPATCH_RigDispatch_FromString(in String,out Result)"}
   RigVMFunctionReturnNode(Name=@From_String.Result, ExecuteContext=@RigVMFunction_Sequence.B)
   RigVMFunction_ControlFlowBranch = &RigVMFunction_ControlFlowBranch(Condition=RigVMVariableNode() {variable=Left Side}.Value, ExecuteContext=@RigVMFunction_Sequence.A)
   n3 = &RigVMFunction_ControlFlowBranch(ExecuteContext=@RigVMFunction_ControlFlowBranch.False, Condition=RigVMVariableNode() {variable=Right Side}.Value)
   n4 = &RigVMFunction_ControlFlowBranch(ExecuteContext=@n3.Completed, Condition=RigVMVariableNode() {variable=Custom Side}.Value)
   n5 = RigVMDispatchNode(
    Index=0
    ExecuteContext=@n4.True
    Element=RigVMVariableNode() {variable=Custom Side Name}.Value
    Array=RigVMVariableNode() {variable=Name Parts}.Value
   ) {template="DISPATCH_RigVMDispatch_ArrayInsert(io Array,in Index,in Element)"}
   > RigVMVariableNode(Value=@n5.Array) {variable=Name Parts}
   n6 = RigVMDispatchNode(Element=R, Index=0, Array=RigVMVariableNode() {variable=Name Parts}.Value, ExecuteContext=@n3.True) {template="DISPATCH_RigVMDispatch_ArrayInsert(io Array,in Index,in Element)"}
   > RigVMVariableNode(Value=@n6.Array) {variable=Name Parts}
   n7 = RigVMDispatchNode(Element=L, Index=0, Array=RigVMVariableNode() {variable=Name Parts}.Value, ExecuteContext=@RigVMFunction_ControlFlowBranch.True) {template="DISPATCH_RigVMDispatch_ArrayInsert(io Array,in Index,in Element)"}
   > RigVMVariableNode(Value=@n7.Array) {variable=Name Parts}
  ]}
 New Function_ContainedGraph {nodes [
   n1 = RigVMFunctionEntryNode()
   RigVMFunction_Sequence = &RigVMFunction_Sequence(ExecuteContext=@n1.ExecuteContext)
   RigVMFunctionReturnNode(ExecuteContext=@RigVMFunction_Sequence.B)
   Create Name = RigVMFunctionReferenceNode(Tokens=("Fr Rig Up Vector Null"), ExecuteContext=@RigVMFunction_Sequence.A)
   n2 = RigVMDispatchNode(
    True=1
    False=RigVMDispatchNode(False=1, True=-1, Condition=RigVMVariableNode() {variable=Right Side}.Value) {template="DISPATCH_RigVMDispatch_If(in Condition,in True,in False,out Result)"}.Result
    Condition=RigVMVariableNode() {variable=Left Side}.Value
   ) {template="DISPATCH_RigVMDispatch_If(in Condition,in True,in False,out Result)"}
   Add = /Script/RigVM.RigVMFunction_MathVectorAdd(
    A=&RigUnit_GetTransform(bInitial=true, Item=RigVMVariableNode() {variable=Scapula Tip Control}.Value).Transform.Translation
    B=/Script/RigVM.RigVMFunction_MathVectorMul(B=(X=1,Y=1,Z=1), B.X=@n2.Result, B.Y=@n2.Result, B.Z=@n2.Result, A=RigVMVariableNode() {variable=Rig Up Vector}.Value) {is=struct RigVMUnitNode, template="Multiply::Execute(in A,in B,out Result)"}.Result
   ) {is=struct RigVMUnitNode, template="Add::Execute(in A,in B,out Result)"}
   > n3 = &RigUnit_HierarchyAddNull(
    Space=GlobalSpace
    Transform=(Rotation=(X=0,Y=0,Z=0,W=1),Translation=(X=0,Y=0,Z=0),Scale3D=(X=1,Y=1,Z=1))
    Transform.Translation=@Add.Result
    Name=@"Create Name".Name
    Parent=RigVMVariableNode() {variable=Scapula Tip Control}.Value
   )
   > Create Name_1 = RigVMFunctionReferenceNode(Tokens=("Fr Rig Up Vector"))
   n4 = &RigVMFunction_MathDoubleMul(B=0.333, A=RigVMVariableNode() {variable=Control Scale}.Value)
   > SpawnControl = &RigUnit_HierarchyAddControlTransform(
    InitialValue=(Rotation=(X=0,Y=0,Z=0,W=1),Translation=(X=0,Y=0,Z=0),Scale3D=(X=1,Y=1,Z=1))
    OffsetSpace=LocalSpace
    OffsetTransform=(Rotation=(X=0,Y=0,Z=0,W=1),Translation=(X=0,Y=0,Z=0),Scale3D=(X=1,Y=1,Z=1))
    Settings="(InitialSpace=LocalSpace,bUsePreferredRotationOrder=False,PreferredRotationOrder=YZX,Limits=(LimitTranslationX=(bMinimum=False,bMaximum=False),LimitTranslationY=(bMinimum=False,bMaximum=False),LimitTranslationZ=(bMinimum=False,bMaximum=False),LimitPitch=(b ...(truncated)"
    Name=@"Create Name_1".Name
    Settings.Shape.Transform.Scale3D.X=@n4.Result
    Settings.Shape.Transform.Scale3D.Y=@n4.Result
    Settings.Shape.Transform.Scale3D.Z=@n4.Result
    Parent=@n3.Item
   ) {template="SpawnControl::Execute(in InitialValue,in Settings,in OffsetTransform,in OffsetSpace,in Parent,in Name,out Item)"}
   > RigVMVariableNode(Value=@SpawnControl.Item) {variable=Main PV Control}
  ]}
 New Function_ContainedGraph {nodes [
   n1 = RigVMFunctionEntryNode()
   RigVMFunction_Sequence = &RigVMFunction_Sequence(ExecuteContext=@n1.ExecuteContext)
   RigVMFunctionReturnNode(ExecuteContext=@RigVMFunction_Sequence.B)
   Create Name = RigVMFunctionReferenceNode(Tokens=("Scapula Tip Control Null"), ExecuteContext=@RigVMFunction_Sequence.A)
   > n2 = &RigUnit_HierarchyAddNull(
    Space=GlobalSpace
    Transform=(Rotation=(X=0,Y=0,Z=0,W=1),Translation=(X=0,Y=0,Z=0),Scale3D=(X=1,Y=1,Z=1))
    Name=@"Create Name".Name
    Transform.Translation=&RigUnit_GetTransform(bInitial=true, Item=RigVMVariableNode() {variable=Scapula Tip Connector}.Value).Transform.Translation
    Parent=/Script/ControlRig.RigUnit_HierarchyGetParent(bDefaultParent=True, Child=RigVMVariableNode() {variable=Shoulder Connector}.Value) {is=struct RigVMUnitNode}.Parent
   )
   > RigVMVariableNode(Value=@n2.Item) {variable=Null}
   > Create Name_1 = RigVMFunctionReferenceNode(Tokens=("Scapula Tip Control"))
   > Set Side Color = RigVMFunctionReferenceNode()
   n3 = &RigVMFunction_MathDoubleMul(B=5, A=RigVMVariableNode() {variable=Control Scale}.Value)
   SpawnControl = &RigUnit_HierarchyAddControlTransform(
    ExecutePin=()
    InitialValue=(Rotation=(X=0,Y=0,Z=0,W=1),Translation=(X=0,Y=0,Z=0),Scale3D=(X=1,Y=1,Z=1))
    OffsetSpace=LocalSpace
    OffsetTransform=(Rotation=(X=0,Y=0,Z=0,W=1),Translation=(X=0,Y=0,Z=0),Scale3D=(X=1,Y=1,Z=1))
    Settings="(InitialSpace=LocalSpace,bUsePreferredRotationOrder=False,PreferredRotationOrder=YZX,Limits=(LimitTranslationX=(bMinimum=False,bMaximum=False),LimitTranslationY=(bMinimum=False,bMaximum=False),LimitTranslationZ=(bMinimum=False,bMaximum=False),LimitPitch=(b ...(truncated)"
    Name=@"Create Name_1".Name
    Settings.Shape.Transform.Scale3D.X=@n3.Result
    Settings.Shape.Transform.Scale3D.Y=@n3.Result
    Settings.Shape.Transform.Scale3D.Z=@n3.Result
    Settings.Shape.Color=@"Set Side Color".Color
    Parent=RigVMVariableNode() {variable=Null}.Value
   ) {template="SpawnControl::Execute(in InitialValue,in Settings,in OffsetTransform,in OffsetSpace,in Parent,in Name,out Item)"}
   > RigVMVariableNode(Value=@SpawnControl.Item) {variable=Scapula Tip Control}
  ]}
 New Function_ContainedGraph {nodes [
   RigVMFunctionEntryNode()
   > RigVMFunction_Sequence = RigVMAggregateNode()
   RigVMDispatchNode(ExecuteContext=@RigVMFunction_Sequence.A, Array=RigVMVariableNode() {variable=Debug Line Points}.Value) {template="DISPATCH_RigVMDispatch_ArrayReset(io Array)"}
   > RigVMFunction_Sequence_1 = RigVMAggregateNode()
   RigVMFunction_Sequence_2 = RigVMAggregateNode(ExecuteContext=@RigVMFunction_Sequence_1.A)
   n1 = &RigVMFunction_MathTransformMakeRelative(
    Global=&RigUnit_GetTransform(bInitial=False, Item=RigVMVariableNode() {variable=Ik Control}.Value).Transform
    Parent=RigVMVariableNode() {variable=Current Ik Control Default Transform}.Value
   )
   n2 = /Script/RigVM.RigVMFunction_MathVectorFromDouble(Value=RigVMVariableNode() {variable=Scapula Tip Aim Percent}.Value) {is=struct RigVMUnitNode, template="Cast::Execute(in Value,out Result)"}
   &RigUnit_SetTransform(
    Weight=1
    bInitial=False
    Value=/Script/RigVM.RigVMFunction_MathTransformMakeAbsolute(
     Local=(Rotation=(X=0,Y=0,Z=0,W=1),Translation=(X=0,Y=0,Z=0),Scale3D=(X=1,Y=1,Z=1))
     Local.Translation=&RigVMFunction_MathVectorMul(A=@n1.Local.Translation, B=@n2.Result) {template="Multiply::Execute(in A,in B,out Result)"}.Result
     Parent=RigVMVariableNode() {variable=Current Ik Control Default Transform}.Value
    ) {is=struct RigVMUnitNode, template="Make Absolute::Execute(in Local,in Parent,out Global)"}.Global
    Item=RigVMVariableNode() {variable=Scapula Tip Aim Point}.Value
    ExecutePin=@RigVMFunction_Sequence_2.C
   )
   n3 = &RigUnit_HierarchyGetParent(bDefaultParent=True, Child=RigVMVariableNode() {variable=Scapula Tip Control}.Value)
   RigVMVariableNode(
    Value=&RigVMFunction_MathTransformMakeAbsolute(
     Parent=&RigUnit_GetTransform(bInitial=False, Item=@n3.Parent).Transform
     Local=RigVMVariableNode() {variable=Default Ik Control Local Transform}.Value
    ) {template="Make Absolute::Execute(in Local,in Parent,out Global)"}.Global
    ExecuteContext=@RigVMFunction_Sequence_2.B
   ) {variable=Current Ik Control Default Transform}
   n4 = &RigUnit_HierarchyGetParent(bDefaultParent=True, Child=RigVMVariableNode() {variable=Scapula Tip Control}.Value)
   n5 = RigVMVariableNode() {variable=Ik Control}
   RigVMVariableNode(
    Value=&RigVMFunction_MathTransformMakeRelative(
     Parent=&RigUnit_GetTransform(bInitial=true, Item=@n4.Parent).Transform
     Global=&RigUnit_GetTransform(bInitial=true, Item=@n5.Value).Transform
    ).Local
    ExecuteContext=@RigVMFunction_Sequence_2.A
   ) {variable=Default Ik Control Local Transform}
   Get Scapula Tip Aim Transform = RigVMFunctionReferenceNode()
   Get Scapula Tip Aim Transform_1 = RigVMFunctionReferenceNode(Initial=true)
   Get Scapula Tip Local Transform = RigVMFunctionReferenceNode(Input Transform=@"Get Scapula Tip Aim Transform_1".Aim Transform)
   n6 = &RigVMFunction_MathTransformMakeAbsolute(Parent=@"Get Scapula Tip Aim Transform".Aim Transform, Local=@"Get Scapula Tip Local Transform".Local Transform) {template="Make Absolute::Execute(in Local,in Parent,out Global)"}
   RigVMVariableNode(Value=@n6.Global, ExecuteContext=@RigVMFunction_Sequence_1.B) {variable=Scapula Rig Transform}
   > Add = RigVMDispatchNode(Element=@n6.Global.Translation, Array=RigVMVariableNode() {variable=Debug Line Points}.Value) {template="DISPATCH_RigVMDispatch_ArrayAdd(io Array,in Element,out Index)"}
   Get Scapula Tip Aim Transform_2 = RigVMFunctionReferenceNode()
   Get Scapula Tip Aim Transform_3 = RigVMFunctionReferenceNode(Initial=true)
   Get Shoulder Local Transform_1 = RigVMFunctionReferenceNode(Input Transform=@"Get Scapula Tip Aim Transform_3".Aim Transform)
   n7 = &RigVMFunction_MathTransformMakeAbsolute(Parent=@"Get Scapula Tip Aim Transform_2".Aim Transform, Local=@"Get Shoulder Local Transform_1".Local Trasnform) {template="Make Absolute::Execute(in Local,in Parent,out Global)"}
   RigVMVariableNode(Value=@n7.Global, ExecuteContext=@RigVMFunction_Sequence_1.C) {variable=Shoulder Rig Transform}
   > RigVMDispatchNode(Element=@n7.Global.Translation, Array=RigVMVariableNode() {variable=Debug Line Points}.Value) {template="DISPATCH_RigVMDispatch_ArrayAdd(io Array,in Element,out Index)"}
   Get Scapula Tip Aim Transform_5 = RigVMFunctionReferenceNode(Initial=true)
   Get Elbow Local Transform = RigVMFunctionReferenceNode(Input Transform=@"Get Scapula Tip Aim Transform_5".Aim Transform)
   Get Scapula Tip Aim Transform_4 = RigVMFunctionReferenceNode()
   n8 = &RigVMFunction_MathTransformMakeAbsolute(Local=@"Get Elbow Local Transform".Local Transform, Parent=@"Get Scapula Tip Aim Transform_4".Aim Transform) {template="Make Absolute::Execute(in Local,in Parent,out Global)"}
   RigVMVariableNode(Value=@n8.Global, ExecuteContext=@RigVMFunction_Sequence_1.D) {variable=Elbow Rig Transform}
   > RigVMDispatchNode(Element=@n8.Global.Translation, Array=RigVMVariableNode() {variable=Debug Line Points}.Value) {template="DISPATCH_RigVMDispatch_ArrayAdd(io Array,in Element,out Index)"}
   Get Scapula Tip Aim Transform_6 = RigVMFunctionReferenceNode()
   Get Scapula Tip Aim Transform_7 = RigVMFunctionReferenceNode(Initial=true)
   Get Wrist Local Transform = RigVMFunctionReferenceNode(Input Transform=@"Get Scapula Tip Aim Transform_7".Aim Transform)
   n9 = &RigVMFunction_MathTransformMakeAbsolute(Parent=@"Get Scapula Tip Aim Transform_6".Aim Transform, Local=@"Get Wrist Local Transform".Local Transform) {template="Make Absolute::Execute(in Local,in Parent,out Global)"}
   RigVMVariableNode(Value=@n9.Global, ExecuteContext=@RigVMFunction_Sequence_1.E) {variable=Wrist Rig Transform}
   > RigVMDispatchNode(Element=@n9.Global.Translation, Array=RigVMVariableNode() {variable=Debug Line Points}.Value) {template="DISPATCH_RigVMDispatch_ArrayAdd(io Array,in Element,out Index)"}
   Get Scapula Tip Aim Transform_9 = RigVMFunctionReferenceNode(Initial=true)
   Get Ball Local Transform = RigVMFunctionReferenceNode(Input Transform=@"Get Scapula Tip Aim Transform_9".Aim Transform)
   Get Scapula Tip Aim Transform_8 = RigVMFunctionReferenceNode()
   n10 = &RigVMFunction_MathTransformMakeAbsolute(Local=@"Get Ball Local Transform".Local Transform, Parent=@"Get Scapula Tip Aim Transform_8".Aim Transform) {template="Make Absolute::Execute(in Local,in Parent,out Global)"}
   RigVMVariableNode(Value=@n10.Global, ExecuteContext=@RigVMFunction_Sequence_1.F) {variable=Ball Rig Transform}
   > RigVMDispatchNode(Element=@n10.Global.Translation, Array=RigVMVariableNode() {variable=Debug Line Points}.Value) {template="DISPATCH_RigVMDispatch_ArrayAdd(io Array,in Element,out Index)"}
   RigVMFunction_DebugLineStripNoSpace = /Script/RigVM.RigVMFunction_DebugLineStripNoSpace(
    Color=(R=1,G=1,B=0,A=1)
    DebugDrawSettings=(DepthPriority=SDPG_Foreground,Lifetime=-1)
    Thickness=0.4
    WorldOffset=(Rotation=(X=0,Y=0,Z=0,W=1),Translation=(X=0,Y=0,Z=0),Scale3D=(X=1,Y=1,Z=1))
    ExecutePin=@RigVMFunction_Sequence.B
    Points=RigVMVariableNode() {variable=Debug Line Points}.Value
    bEnabled=RigVMVariableNode() {variable=Debug}.Value
   ) {is=struct RigVMUnitNode}
   RigVMFunctionReturnNode(ExecuteContext=@RigVMFunction_Sequence.C)
   n11 = RigVMRerouteNode(Value=())
   n12 = &RigUnit_HierarchyGetParent(bDefaultParent=True, Child=RigVMVariableNode() {variable=Ik Control}.Value)
   n13 = RigVMVariableNode() {variable=Ik Point}
   RigVMFunction_MathVectorFromDouble = &RigVMFunction_MathVectorFromDouble(Value=RigVMVariableNode() {variable=Scapula Tip Aim Percent}.Value) {template="Cast::Execute(in Value,out Result)"}
   Multiply = &RigVMFunction_MathVectorMul(
    A=&RigVMFunction_MathTransformMakeRelative(
     Parent=&RigUnit_GetTransform(bInitial=False, Item=@n12.Parent).Transform
     Global=&RigUnit_GetTransform(bInitial=False, Item=@n13.Value).Transform
    ).Local.Translation
    B=@RigVMFunction_MathVectorFromDouble.Result
   ) {template="Multiply::Execute(in A,in B,out Result)"}
   Set Transform = /Script/ControlRig.RigUnit_SetTranslation(
    Space=LocalSpace
    Weight=1
    bInitial=False
    bPropagateToChildren=True
    Value=@Multiply.Result
    ExecutePin=@n11.Value
    Item=RigVMVariableNode() {variable=Scapula Tip Aim Point}.Value
   ) {is=struct RigVMUnitNode, template="Set Transform::Execute(in Item,in Space,in bInitial,in Value,in Weight,in bPropagateToChildren)"}
  ]}
 New Function_ContainedGraph {nodes [
   n1 = RigVMFunctionEntryNode()
   RigVMFunction_Sequence = &RigVMFunction_Sequence(ExecuteContext=@n1.ExecuteContext)
   RigVMFunction_Sequence_1 = RigVMAggregateNode(ExecuteContext=@RigVMFunction_Sequence.A)
   Get Main Pole Vector = RigVMFunctionReferenceNode(Initial=true)
   n2 = &RigUnit_AimBoneMath(
    DebugSettings=(bEnabled=False,Scale=10,WorldOffset=(Rotation=(X=0,Y=0,Z=0,W=1),Translation=(X=0,Y=0,Z=0),Scale3D=(X=1,Y=1,Z=1)))
    InputTransform=(Rotation=(X=0,Y=0,Z=0,W=1),Translation=(X=0,Y=0,Z=0),Scale3D=(X=1,Y=1,Z=1))
    Primary=(Weight=1,Axis=(X=1,Y=0,Z=0),Target=(X=1,Y=0,Z=0),Kind=Location,Space=(Type=Bone,Name="None"))
    Secondary=(Weight=1,Axis=(X=0,Y=0,Z=1),Target=(X=1,Y=0,Z=0),Kind=Direction,Space=(Type=Bone,Name="None"))
    Weight=1
    Secondary.Target=@"Get Main Pole Vector".Vector
    InputTransform.Translation=&RigUnit_GetTransform(bInitial=true, Item=RigVMVariableNode() {variable=Scapula Tip Control}.Value).Transform.Translation
    Primary.Target=&RigUnit_GetTransform(bInitial=true, Item=RigVMVariableNode() {variable=Shoulder Connector}.Value).Transform.Translation
   )
   n3 = &RigVMFunction_MathTransformMakeRelative(
    Parent=@n2.Result
    Global=&RigUnit_GetTransform(bInitial=true, Item=RigVMVariableNode() {variable=Scapula Tip Connector}.Value).Transform
   )
   RigVMVariableNode(Value=@n3.Local, ExecuteContext=@RigVMFunction_Sequence_1.A) {variable=Scapula Tip Local Transform}
   n4 = RigVMVariableNode() {variable=Shoulder Connector}
   RigVMVariableNode(
    Value=&RigVMFunction_MathTransformMakeRelative(
     Global=&RigUnit_GetTransform(bInitial=true, Item=@n4.Value).Transform
     Parent=RigVMVariableNode() {variable=Shoulder Rig Transform}.Value
    ).Local
    ExecuteContext=@RigVMFunction_Sequence_1.B
   ) {variable=Shoulder Local Transform}
   n5 = RigVMVariableNode() {variable=Elbow Connector}
   RigVMVariableNode(
    Value=&RigVMFunction_MathTransformMakeRelative(
     Global=&RigUnit_GetTransform(bInitial=true, Item=@n5.Value).Transform
     Parent=RigVMVariableNode() {variable=Elbow Rig Transform}.Value
    ).Local
    ExecuteContext=@RigVMFunction_Sequence_1.C
   ) {variable=Elbow Local Transform}
   n6 = RigVMVariableNode() {variable=Wrist Connector}
   RigVMVariableNode(
    Value=&RigVMFunction_MathTransformMakeRelative(
     Global=&RigUnit_GetTransform(bInitial=true, Item=@n6.Value).Transform
     Parent=RigVMVariableNode() {variable=Wrist Rig Transform}.Value
    ).Local
    ExecuteContext=@RigVMFunction_Sequence_1.D
   ) {variable=Wrist Local Transform}
   n7 = RigVMVariableNode() {variable=Ball Connector}
   RigVMVariableNode(
    Value=&RigVMFunction_MathTransformMakeRelative(
     Global=&RigUnit_GetTransform(bInitial=true, Item=@n7.Value).Transform
     Parent=RigVMVariableNode() {variable=Ball Rig Transform}.Value
    ).Local
    ExecuteContext=@RigVMFunction_Sequence_1.E
   ) {variable=Ball Local Transform}
   n8 = RigVMVariableNode() {variable=Toe Connector}
   RigVMVariableNode(
    Value=&RigVMFunction_MathTransformMakeRelative(
     Global=&RigUnit_GetTransform(bInitial=true, Item=@n8.Value).Transform
     Parent=RigVMVariableNode() {variable=Toe Rig Transform}.Value
    ).Local
    ExecuteContext=@RigVMFunction_Sequence_1.F
   ) {variable=Toe Local Transform}
   RigVMFunctionReturnNode(ExecuteContext=@RigVMFunction_Sequence.B)
   &RigVMFunction_MathTransformMakeAbsolute(Parent=@n3.Local, Local=RigVMVariableNode() {variable=Scapula Tip Offset Transform}.Value) {template="Make Absolute::Execute(in Local,in Parent,out Global)"}
   RigVMVariableNode() {variable=Scapula Rig Transform}
  ]}
 New Function_ContainedGraph {nodes [
   RigVMFunctionEntryNode()
   Get Main Pole Vector = RigVMFunctionReferenceNode(Initial=RigVMVariableNode() {variable=Initial}.Value)
   n1 = &RigUnit_AimBoneMath(
    DebugSettings=(bEnabled=False,Scale=10,WorldOffset=(Rotation=(X=0,Y=0,Z=0,W=1),Translation=(X=0,Y=0,Z=0),Scale3D=(X=1,Y=1,Z=1)))
    InputTransform=(Rotation=(X=0,Y=0,Z=0,W=1),Translation=(X=0,Y=0,Z=0),Scale3D=(X=1,Y=1,Z=1))
    Primary=(Weight=1,Axis=(X=1,Y=0,Z=0),Target=(X=1,Y=0,Z=0),Kind=Location,Space=(Type=Bone,Name="None"))
    Secondary=(Weight=1,Axis=(X=0,Y=0,Z=1),Target=(X=1,Y=0,Z=0),Kind=Direction,Space=(Type=Bone,Name="None"))
    Weight=1
    Secondary.Target=@"Get Main Pole Vector".Vector
    InputTransform.Translation=&RigUnit_GetTransform(Item=RigVMVariableNode() {variable=Scapula Tip Control}.Value, bInitial=RigVMVariableNode() {variable=Initial}.Value).Transform.Translation
    Primary.Target=&RigUnit_GetTransform(bInitial=RigVMVariableNode() {variable=Initial}.Value, Item=RigVMVariableNode() {variable=Scapula Tip Aim Point}.Value).Transform.Translation
   )
   GetAnimationChannelFromItem = /Script/ControlRig.RigUnit_GetFloatAnimationChannelFromItem(Item=RigVMVariableNode() {variable=Shoulder Rotate Offset}.Value, bInitial=RigVMVariableNode() {variable=Initial}.Value) {is=struct RigVMUnitNode, template="GetAnimationChannelFromItem::Execute(out Value,in Item,in bInitial)"}
   n2 = RigVMVariableNode() {variable=Scapula Tip Control}
   n3 = RigVMVariableNode() {variable=Initial}
   n4 = &RigUnit_HierarchyGetParent(bDefaultParent=True, Child=RigVMVariableNode() {variable=Scapula Tip Control}.Value)
   n5 = RigVMVariableNode() {variable=Initial}
   RigVMFunction_MathQuaternionToEuler = &RigVMFunction_MathQuaternionToEuler(
    RotationOrder=ZYX
    Value=&RigVMFunction_MathTransformMakeRelative(
     Global=&RigUnit_GetTransform(Item=@n2.Value, bInitial=@n3.Value).Transform
     Parent=&RigUnit_GetTransform(Item=@n4.Parent, bInitial=@n5.Value).Transform
    ).Local.Rotation
   )
   Add = &RigVMFunction_MathDoubleAdd(B=@GetAnimationChannelFromItem.Value, A=@RigVMFunction_MathQuaternionToEuler.Result.X)
   RigVMFunction_MathQuaternionFromEuler = &RigVMFunction_MathQuaternionFromEuler(Euler=(X=0,Y=0,Z=0), Euler.Z=@Add.Result)
   > RigVMFunctionReturnNode(Aim Transform=&RigVMFunction_MathTransformMakeAbsolute(
     Local=(Rotation=(X=0,Y=0,Z=0,W=1),Translation=(X=0,Y=0,Z=0),Scale3D=(X=1,Y=1,Z=1))
     Parent=@n1.Result
     Local.Rotation=@RigVMFunction_MathQuaternionFromEuler.Result
    ) {template="Make Absolute::Execute(in Local,in Parent,out Global)"}.Global)
  ]}
 New Function_ContainedGraph {nodes [
   n1 = RigVMFunctionEntryNode()
   n2 = /Script/RigVM.RigVMFunction_MathDoubleLawOfCosine(A=@n1.Length A, B=@n1.Length B, C=@n1.Length C) {is=struct RigVMUnitNode, template="LawOfCosine::Execute(in A,in B,in C,out AlphaAngle,out BetaAngle,out GammaAngle,out bValid)"}
   n3 = /Script/RigVM.RigVMFunction_MathDoubleDeg(Value=RigVMDispatchNode(
     True=/Script/RigVM.RigVMFunction_MathDoubleNegate(Value=@n2.AlphaAngle) {is=struct RigVMUnitNode, template="Negate::Execute(in Value,out Result)"}.Result
     Condition=RigVMVariableNode() {variable=Negate A}.Value
     False=@n2.AlphaAngle
    ) {template="DISPATCH_RigVMDispatch_If(in Condition,in True,in False,out Result)"}.Result) {is=struct RigVMUnitNode, template="Degrees::Execute(in Value,out Result)"}
   n4 = &RigVMFunction_MathDoubleDeg(Value=RigVMDispatchNode(
     True=&RigVMFunction_MathDoubleNegate(Value=@n2.BetaAngle) {template="Negate::Execute(in Value,out Result)"}.Result
     Condition=RigVMVariableNode() {variable=Negate B}.Value
     False=@n2.BetaAngle
    ) {template="DISPATCH_RigVMDispatch_If(in Condition,in True,in False,out Result)"}.Result) {template="Degrees::Execute(in Value,out Result)"}
   n5 = &RigVMFunction_MathDoubleDeg(Value=RigVMDispatchNode(
     True=&RigVMFunction_MathDoubleNegate(Value=@n2.GammaAngle) {template="Negate::Execute(in Value,out Result)"}.Result
     Condition=RigVMVariableNode() {variable=Negate C}.Value
     False=@n2.GammaAngle
    ) {template="DISPATCH_RigVMDispatch_If(in Condition,in True,in False,out Result)"}.Result) {template="Degrees::Execute(in Value,out Result)"}
   > RigVMFunctionReturnNode(
    Angle A=RigVMDispatchNode(
     True=&RigVMFunction_MathDoubleAdd(B=180, A=@n3.Result).Result
     Condition=RigVMVariableNode() {variable=A 180}.Value
     False=@n3.Result
    ) {template="DISPATCH_RigVMDispatch_If(in Condition,in True,in False,out Result)"}.Result
    Angle B=RigVMDispatchNode(
     True=&RigVMFunction_MathDoubleAdd(B=180, A=@n4.Result).Result
     Condition=RigVMVariableNode() {variable=B 180}.Value
     False=@n4.Result
    ) {template="DISPATCH_RigVMDispatch_If(in Condition,in True,in False,out Result)"}.Result
    Angle C=RigVMDispatchNode(
     True=&RigVMFunction_MathDoubleAdd(B=180, A=@n5.Result).Result
     Condition=RigVMVariableNode() {variable=C 180}.Value
     False=@n5.Result
    ) {template="DISPATCH_RigVMDispatch_If(in Condition,in True,in False,out Result)"}.Result
   )
  ]}
 New Function_ContainedGraph {nodes [
   RigVMFunctionEntryNode()
   > RigVMFunction_Sequence = RigVMAggregateNode()
   RigVMFunction_Sequence_1 = RigVMAggregateNode(ExecuteContext=@RigVMFunction_Sequence.A)
   n1 = RigVMVariableNode() {variable=Ik Point}
   Get Ik Angles = RigVMFunctionReferenceNode(
    C 180=true
    Negate A=false
    Negate B=true
    Negate C=true
    Length C=&RigVMFunction_MathVectorDistance(
     B=&RigUnit_GetTransform(bInitial=False, Item=@n1.Value).Transform.Translation
     A=RigVMVariableNode() {variable=Shoulder Rig Transform}.Value.Translation
    ).Result
    Length A=RigVMVariableNode() {variable=Shoulder Length}.Value
    Length B=RigVMVariableNode() {variable=Elbow Ball Aim Length}.Value
   )
   RigVMVariableNode(Value=@"Get Ik Angles".Angle B, ExecuteContext=@RigVMFunction_Sequence_1.B) {variable=Shoulder Angle}
   > RigVMVariableNode(Value=RigVMDispatchNode(
     True=0
     Condition=RigVMDispatchNode(B=180, A=@"Get Ik Angles".Angle C) {template="DISPATCH_RigVMDispatch_CoreEquals(in A,in B,out Result)"}.Result
     False=@"Get Ik Angles".Angle C
    ) {template="DISPATCH_RigVMDispatch_If(in Condition,in True,in False,out Result)"}.Result) {variable=Lower Leg Angle}
   Get Ik Angles_1 = RigVMFunctionReferenceNode(
    C 180=true
    Negate B=true
    Negate C=true
    Length C=RigVMVariableNode() {variable=Elbow Ball Aim Length}.Value
    Length A=RigVMVariableNode() {variable=Elbow Length}.Value
    Length B=RigVMVariableNode() {variable=Wrist Length}.Value
   )
   > RigVMVariableNode(Value=@"Get Ik Angles_1".Angle B) {variable=Elbow Angle}
   > RigVMVariableNode(Value=@"Get Ik Angles_1".Angle C) {variable=Wrist Angle}
   RigVMVariableNode(
    Value=&RigVMFunction_MathVectorDistance(
     A=RigVMVariableNode() {variable=Elbow Rig Transform}.Value.Translation
     B=RigVMVariableNode() {variable=Ball Rig Transform}.Value.Translation
    ).Result
    ExecuteContext=@RigVMFunction_Sequence_1.A
   ) {variable=Elbow Ball Aim Length}
   > RigVMFunction_DebugLineNoSpace = /Script/RigVM.RigVMFunction_DebugLineNoSpace(
    Color=(R=1,G=0,B=0,A=1)
    DebugDrawSettings=(DepthPriority=SDPG_Foreground,Lifetime=-1)
    Thickness=0.1
    WorldOffset=(Rotation=(X=0,Y=0,Z=0,W=1),Translation=(X=0,Y=0,Z=0),Scale3D=(X=1,Y=1,Z=1))
    bEnabled=false
    A=&RigUnit_GetTransform(bInitial=False, Item=RigVMVariableNode() {variable=Scapula Tip Control}.Value).Transform.Translation
    B=&RigUnit_GetTransform(bInitial=False, Item=RigVMVariableNode() {variable=Main PV Control}.Value).Transform.Translation
   ) {is=struct RigVMUnitNode}
   RigVMFunction_Sequence_2 = RigVMAggregateNode(ExecuteContext=@RigVMFunction_Sequence.B)
   n2 = RigVMVariableNode() {variable=Wrist Angle}
   n3 = RigVMVariableNode() {variable=Elbow Rig Transform}
   n4 = RigVMVariableNode() {variable=Elbow Length}
   RigVMVariableNode(
    Value=&RigVMFunction_MathTransformMakeAbsolute(
     Local=(Rotation=(X=0,Y=0,Z=0,W=1),Translation=(X=0,Y=0,Z=0),Scale3D=(X=1,Y=1,Z=1))
     Local.Rotation=&RigVMFunction_MathQuaternionFromEuler(Euler=(X=0,Y=0,Z=0), Euler.Z=@n2.Value).Result
     Parent=&RigVMFunction_MathTransformMakeAbsolute(Local=(Rotation=(X=0,Y=0,Z=0,W=1),Translation=(X=0,Y=0,Z=0),Scale3D=(X=1,Y=1,Z=1)), Parent=@n3.Value, Local.Translation.X=@n4.Value) {template="Make Absolute::Execute(in Local,in Parent,out Global)"}.Global
    ) {template="Make Absolute::Execute(in Local,in Parent,out Global)"}.Global
    ExecuteContext=@RigVMFunction_Sequence_2.D
   ) {variable=Wrist Rig Transform}
   RigVMVariableNode(
    Value=&RigVMFunction_MathTransformMakeAbsolute(
     Local=(Rotation=(X=0,Y=0,Z=0,W=1),Translation=(X=0,Y=0,Z=0),Scale3D=(X=1,Y=1,Z=1))
     Parent=RigVMVariableNode() {variable=Wrist Rig Transform}.Value
     Local.Translation.X=RigVMVariableNode() {variable=Wrist Length}.Value
    ) {template="Make Absolute::Execute(in Local,in Parent,out Global)"}.Global
    ExecuteContext=@RigVMFunction_Sequence_2.E
   ) {variable=Ball Rig Transform}
   n5 = RigVMVariableNode() {variable=Shoulder Length}
   n6 = RigVMVariableNode() {variable=Shoulder Rig Transform}
   n7 = RigVMVariableNode() {variable=Lower Leg Angle}
   RigVMVariableNode(
    Value=&RigVMFunction_MathTransformMakeAbsolute(
     Local=(Rotation=(X=0,Y=0,Z=0,W=1),Translation=(X=0,Y=0,Z=0),Scale3D=(X=1,Y=1,Z=1))
     Parent=&RigVMFunction_MathTransformMakeAbsolute(Local=(Rotation=(X=0,Y=0,Z=0,W=1),Translation=(X=0,Y=0,Z=0),Scale3D=(X=1,Y=1,Z=1)), Local.Translation.X=@n5.Value, Parent=@n6.Value) {template="Make Absolute::Execute(in Local,in Parent,out Global)"}.Global
     Local.Rotation=&RigVMFunction_MathQuaternionFromEuler(Euler=(X=0,Y=0,Z=0), Euler.Z=@n7.Value).Result
    ) {template="Make Absolute::Execute(in Local,in Parent,out Global)"}.Global
    ExecuteContext=@RigVMFunction_Sequence_2.B
   ) {variable=Elbow Ik Aim Transform}
   Get Main Pole Vector = RigVMFunctionReferenceNode(Initial=False)
   n8 = &RigUnit_AimBoneMath(
    DebugSettings=(bEnabled=False,Scale=10,WorldOffset=(Rotation=(X=0,Y=0,Z=0,W=1),Translation=(X=0,Y=0,Z=0),Scale3D=(X=1,Y=1,Z=1)))
    InputTransform=(Rotation=(X=0,Y=0,Z=0,W=1),Translation=(X=0,Y=0,Z=0),Scale3D=(X=1,Y=1,Z=1))
    Primary=(Weight=1,Axis=(X=1,Y=0,Z=0),Target=(X=1,Y=0,Z=0),Kind=Location,Space=(Type=Bone,Name="None"))
    Secondary=(Weight=1,Axis=(X=0,Y=0,Z=1),Target=(X=1,Y=0,Z=0),Kind=Direction,Space=(Type=Bone,Name="None"))
    Weight=1
    Secondary.Target=@"Get Main Pole Vector".Vector
    Primary.Target=&RigUnit_GetTransform(bInitial=False, Item=RigVMVariableNode() {variable=Ik Point}.Value).Transform.Translation
    InputTransform.Translation=RigVMVariableNode() {variable=Shoulder Rig Transform}.Value.Translation
   )
   RigVMVariableNode(Value=@n8.Result, ExecuteContext=@RigVMFunction_Sequence_2.A) {variable=Shoulder Ik Aim Transform}
   RigVMFunction_MathQuaternionFromEuler = &RigVMFunction_MathQuaternionFromEuler(Euler=(X=0,Y=0,Z=0), Euler.Z=RigVMVariableNode() {variable=Shoulder Angle}.Value)
   > RigVMVariableNode(Value=&RigVMFunction_MathTransformMakeAbsolute(
     Local=(Rotation=(X=0,Y=0,Z=0,W=1),Translation=(X=0,Y=0,Z=0),Scale3D=(X=1,Y=1,Z=1))
     Local.Rotation=@RigVMFunction_MathQuaternionFromEuler.Result
     Parent=RigVMVariableNode() {variable=Shoulder Ik Aim Transform}.Value
    ) {template="Make Absolute::Execute(in Local,in Parent,out Global)"}.Global) {variable=Shoulder Rig Transform}
   n9 = RigVMVariableNode() {variable=Elbow Angle}
   RigVMVariableNode(
    Value=&RigVMFunction_MathTransformMakeAbsolute(
     Local=(Rotation=(X=0,Y=0,Z=0,W=1),Translation=(X=0,Y=0,Z=0),Scale3D=(X=1,Y=1,Z=1))
     Local.Rotation=&RigVMFunction_MathQuaternionFromEuler(Euler=(X=0,Y=0,Z=0), Euler.Z=@n9.Value).Result
     Parent=RigVMVariableNode() {variable=Elbow Ik Aim Transform}.Value
    ) {template="Make Absolute::Execute(in Local,in Parent,out Global)"}.Global
    ExecuteContext=@RigVMFunction_Sequence_2.C
   ) {variable=Elbow Rig Transform}
   RigVMFunctionReturnNode(ExecuteContext=@RigVMFunction_Sequence.C)
  ]}
 New Function_ContainedGraph {nodes [
   n1 = RigVMFunctionEntryNode()
   RigVMFunction_Sequence = &RigVMFunction_Sequence(ExecuteContext=@n1.ExecuteContext)
   RigVMFunction_Sequence_1 = RigVMAggregateNode(ExecuteContext=@RigVMFunction_Sequence.A)
   RigVMFunction_MathVectorDistance = &RigVMFunction_MathVectorDistance(
    A=&RigUnit_GetTransform(bInitial=true, Item=RigVMVariableNode() {variable=Scapula Tip Connector}.Value).Transform.Translation
    B=&RigUnit_GetTransform(bInitial=true, Item=RigVMVariableNode() {variable=Shoulder Connector}.Value).Transform.Translation
   )
   RigVMVariableNode(Value=@RigVMFunction_MathVectorDistance.Result, ExecuteContext=@RigVMFunction_Sequence_1.A) {variable=Scapula Length}
   n2 = RigVMVariableNode() {variable=Shoulder Connector}
   n3 = RigVMVariableNode() {variable=Elbow Connector}
   RigVMVariableNode(
    Value=&RigVMFunction_MathVectorDistance(
     A=&RigUnit_GetTransform(bInitial=true, Item=@n2.Value).Transform.Translation
     B=&RigUnit_GetTransform(bInitial=true, Item=@n3.Value).Transform.Translation
    ).Result
    ExecuteContext=@RigVMFunction_Sequence_1.B
   ) {variable=Shoulder Length}
   n4 = RigVMVariableNode() {variable=Elbow Connector}
   n5 = RigVMVariableNode() {variable=Wrist Connector}
   RigVMVariableNode(
    Value=&RigVMFunction_MathVectorDistance(
     A=&RigUnit_GetTransform(bInitial=true, Item=@n4.Value).Transform.Translation
     B=&RigUnit_GetTransform(bInitial=true, Item=@n5.Value).Transform.Translation
    ).Result
    ExecuteContext=@RigVMFunction_Sequence_1.C
   ) {variable=Elbow Length}
   n6 = RigVMVariableNode() {variable=Wrist Connector}
   n7 = RigVMVariableNode() {variable=Ball Connector}
   RigVMVariableNode(
    Value=&RigVMFunction_MathVectorDistance(
     A=&RigUnit_GetTransform(bInitial=true, Item=@n6.Value).Transform.Translation
     B=&RigUnit_GetTransform(bInitial=true, Item=@n7.Value).Transform.Translation
    ).Result
    ExecuteContext=@RigVMFunction_Sequence_1.D
   ) {variable=Wrist Length}
   n8 = RigVMVariableNode() {variable=Ball Connector}
   n9 = RigVMVariableNode() {variable=Toe Connector}
   RigVMVariableNode(
    Value=&RigVMFunction_MathVectorDistance(
     A=&RigUnit_GetTransform(bInitial=true, Item=@n8.Value).Transform.Translation
     B=&RigUnit_GetTransform(bInitial=true, Item=@n9.Value).Transform.Translation
    ).Result
    ExecuteContext=@RigVMFunction_Sequence_1.E
   ) {variable=Ball Length}
   n10 = RigVMVariableNode() {variable=Toe Connector}
   n11 = RigVMVariableNode() {variable=Toe Tip Connector}
   RigVMVariableNode(
    Value=&RigVMFunction_MathVectorDistance(
     A=&RigUnit_GetTransform(bInitial=true, Item=@n10.Value).Transform.Translation
     B=&RigUnit_GetTransform(bInitial=true, Item=@n11.Value).Transform.Translation
    ).Result
    ExecuteContext=@RigVMFunction_Sequence_1.F
   ) {variable=Toe Length}
   RigVMFunctionReturnNode(ExecuteContext=@RigVMFunction_Sequence.B)
  ]}
 New Function_ContainedGraph {nodes [
   n1 = RigVMFunctionEntryNode()
   RigVMFunction_Sequence = &RigVMFunction_Sequence(ExecuteContext=@n1.ExecuteContext)
   RigVMFunctionReturnNode(ExecuteContext=@RigVMFunction_Sequence.B)
  ]}
 New Function_ContainedGraph {nodes [
   RigVMFunctionEntryNode()
   > RigVMFunctionReturnNode()
  ]}
 New Function_ContainedGraph {nodes [
   RigVMFunctionEntryNode()
   > RigVMFunctionReturnNode()
  ]}
 New Function_ContainedGraph {nodes [
   RigVMFunctionEntryNode()
   > RigVMFunctionReturnNode()
  ]}
 New Function_ContainedGraph {nodes [
   RigVMFunctionEntryNode()
   > RigVMFunctionReturnNode()
  ]}
 New Function_ContainedGraph {nodes [
   n1 = RigVMFunctionEntryNode()
   RigVMFunction_Sequence = &RigVMFunction_Sequence(ExecuteContext=@n1.ExecuteContext)
   RigVMFunction_Sequence_1 = RigVMAggregateNode(ExecuteContext=@RigVMFunction_Sequence.A)
   Get Main Pole Vector_1 = RigVMFunctionReferenceNode(Initial=true)
   n2 = &RigUnit_AimBoneMath(
    DebugSettings=(bEnabled=False,Scale=10,WorldOffset=(Rotation=(X=0,Y=0,Z=0,W=1),Translation=(X=0,Y=0,Z=0),Scale3D=(X=1,Y=1,Z=1)))
    Primary=(Weight=1,Axis=(X=1,Y=0,Z=0),Target=(X=1,Y=0,Z=0),Kind=Location,Space=(Type=Bone,Name="None"))
    Secondary=(Weight=1,Axis=(X=0,Y=0,Z=1),Target=(X=1,Y=0,Z=0),Kind=Direction,Space=(Type=Bone,Name="None"))
    Weight=1
    Secondary.Target=@"Get Main Pole Vector_1".Vector
    InputTransform=&RigUnit_GetTransform(bInitial=true, Item=RigVMVariableNode() {variable=Scapula Tip Connector}.Value).Transform
    Primary.Target=&RigUnit_GetTransform(bInitial=true, Item=RigVMVariableNode() {variable=Shoulder Connector}.Value).Transform.Translation
   )
   RigVMVariableNode(Value=@n2.Result, ExecuteContext=@RigVMFunction_Sequence_1.A) {variable=Scapula Rig Transform}
   Get Main Pole Vector = RigVMFunctionReferenceNode(Initial=true)
   n3 = &RigUnit_AimBoneMath(
    DebugSettings=(bEnabled=False,Scale=10,WorldOffset=(Rotation=(X=0,Y=0,Z=0,W=1),Translation=(X=0,Y=0,Z=0),Scale3D=(X=1,Y=1,Z=1)))
    Primary=(Weight=1,Axis=(X=1,Y=0,Z=0),Target=(X=1,Y=0,Z=0),Kind=Location,Space=(Type=Bone,Name="None"))
    Secondary=(Weight=1,Axis=(X=0,Y=0,Z=1),Target=(X=1,Y=0,Z=0),Kind=Direction,Space=(Type=Bone,Name="None"))
    Weight=1
    Secondary.Target=@"Get Main Pole Vector".Vector
    InputTransform=&RigUnit_GetTransform(bInitial=true, Item=RigVMVariableNode() {variable=Shoulder Connector}.Value).Transform
    Primary.Target=&RigUnit_GetTransform(bInitial=true, Item=RigVMVariableNode() {variable=Elbow Connector}.Value).Transform.Translation
   )
   RigVMVariableNode(Value=@n3.Result, ExecuteContext=@RigVMFunction_Sequence_1.B) {variable=Shoulder Rig Transform}
   Get Main Pole Vector_2 = RigVMFunctionReferenceNode(Initial=true)
   n4 = &RigUnit_AimBoneMath(
    DebugSettings=(bEnabled=False,Scale=10,WorldOffset=(Rotation=(X=0,Y=0,Z=0,W=1),Translation=(X=0,Y=0,Z=0),Scale3D=(X=1,Y=1,Z=1)))
    Primary=(Weight=1,Axis=(X=1,Y=0,Z=0),Target=(X=1,Y=0,Z=0),Kind=Location,Space=(Type=Bone,Name="None"))
    Secondary=(Weight=1,Axis=(X=0,Y=0,Z=1),Target=(X=1,Y=0,Z=0),Kind=Direction,Space=(Type=Bone,Name="None"))
    Weight=1
    Secondary.Target=@"Get Main Pole Vector_2".Vector
    InputTransform=&RigUnit_GetTransform(bInitial=true, Item=RigVMVariableNode() {variable=Elbow Connector}.Value).Transform
    Primary.Target=&RigUnit_GetTransform(bInitial=true, Item=RigVMVariableNode() {variable=Wrist Connector}.Value).Transform.Translation
   )
   RigVMVariableNode(Value=@n4.Result, ExecuteContext=@RigVMFunction_Sequence_1.C) {variable=Elbow Rig Transform}
   Get Main Pole Vector_3 = RigVMFunctionReferenceNode(Initial=true)
   n5 = &RigUnit_AimBoneMath(
    DebugSettings=(bEnabled=False,Scale=10,WorldOffset=(Rotation=(X=0,Y=0,Z=0,W=1),Translation=(X=0,Y=0,Z=0),Scale3D=(X=1,Y=1,Z=1)))
    Primary=(Weight=1,Axis=(X=1,Y=0,Z=0),Target=(X=1,Y=0,Z=0),Kind=Location,Space=(Type=Bone,Name="None"))
    Secondary=(Weight=1,Axis=(X=0,Y=0,Z=1),Target=(X=1,Y=0,Z=0),Kind=Direction,Space=(Type=Bone,Name="None"))
    Weight=1
    Secondary.Target=@"Get Main Pole Vector_3".Vector
    InputTransform=&RigUnit_GetTransform(bInitial=true, Item=RigVMVariableNode() {variable=Wrist Connector}.Value).Transform
    Primary.Target=&RigUnit_GetTransform(bInitial=true, Item=RigVMVariableNode() {variable=Ball Connector}.Value).Transform.Translation
   )
   RigVMVariableNode(Value=@n5.Result, ExecuteContext=@RigVMFunction_Sequence_1.D) {variable=Wrist Rig Transform}
   Get Main Pole Vector_4 = RigVMFunctionReferenceNode(Initial=true)
   n6 = &RigUnit_AimBoneMath(
    DebugSettings=(bEnabled=False,Scale=10,WorldOffset=(Rotation=(X=0,Y=0,Z=0,W=1),Translation=(X=0,Y=0,Z=0),Scale3D=(X=1,Y=1,Z=1)))
    Primary=(Weight=1,Axis=(X=1,Y=0,Z=0),Target=(X=1,Y=0,Z=0),Kind=Location,Space=(Type=Bone,Name="None"))
    Secondary=(Weight=1,Axis=(X=0,Y=0,Z=1),Target=(X=1,Y=0,Z=0),Kind=Direction,Space=(Type=Bone,Name="None"))
    Weight=1
    Secondary.Target=@"Get Main Pole Vector_4".Vector
    InputTransform=&RigUnit_GetTransform(bInitial=true, Item=RigVMVariableNode() {variable=Ball Connector}.Value).Transform
    Primary.Target=&RigUnit_GetTransform(bInitial=true, Item=RigVMVariableNode() {variable=Toe Connector}.Value).Transform.Translation
   )
   RigVMVariableNode(Value=@n6.Result, ExecuteContext=@RigVMFunction_Sequence_1.E) {variable=Ball Rig Transform}
   Get Main Pole Vector_5 = RigVMFunctionReferenceNode(Initial=true)
   n7 = &RigUnit_AimBoneMath(
    DebugSettings=(bEnabled=False,Scale=10,WorldOffset=(Rotation=(X=0,Y=0,Z=0,W=1),Translation=(X=0,Y=0,Z=0),Scale3D=(X=1,Y=1,Z=1)))
    Primary=(Weight=1,Axis=(X=1,Y=0,Z=0),Target=(X=1,Y=0,Z=0),Kind=Location,Space=(Type=Bone,Name="None"))
    Secondary=(Weight=1,Axis=(X=0,Y=0,Z=1),Target=(X=1,Y=0,Z=0),Kind=Direction,Space=(Type=Bone,Name="None"))
    Weight=1
    Secondary.Target=@"Get Main Pole Vector_5".Vector
    InputTransform=&RigUnit_GetTransform(bInitial=true, Item=RigVMVariableNode() {variable=Toe Connector}.Value).Transform
    Primary.Target=&RigUnit_GetTransform(bInitial=true, Item=RigVMVariableNode() {variable=Toe Tip Connector}.Value).Transform.Translation
   )
   RigVMVariableNode(Value=@n7.Result, ExecuteContext=@RigVMFunction_Sequence_1.F) {variable=Toe Rig Transform}
   RigVMFunctionReturnNode(ExecuteContext=@RigVMFunction_Sequence.B)
  ]}
 New Function_ContainedGraph {nodes [
   n1 = RigVMFunctionEntryNode()
   n2 = RigVMDispatchNode(
    True=1
    False=RigVMDispatchNode(False=1, True=-1, Condition=RigVMVariableNode() {variable=Right Side}.Value) {template="DISPATCH_RigVMDispatch_If(in Condition,in True,in False,out Result)"}.Result
    Condition=RigVMVariableNode() {variable=Left Side}.Value
   ) {template="DISPATCH_RigVMDispatch_If(in Condition,in True,in False,out Result)"}
   Subtract = /Script/RigVM.RigVMFunction_MathVectorSub(
    B=&RigUnit_GetTransform(bInitial=@n1.Initial, Item=RigVMVariableNode() {variable=Scapula Tip Control}.Value).Transform.Translation
    A=&RigUnit_GetTransform(bInitial=@n1.Initial, Item=RigVMVariableNode() {variable=Main PV Control}.Value).Transform.Translation
   ) {is=struct RigVMUnitNode, template="Subtract::Execute(in A,in B,out Result)"}
   Multiply = &RigVMFunction_MathVectorMul(B=(X=1,Y=1,Z=1), B.X=@n2.Result, B.Y=@n2.Result, B.Z=@n2.Result, A=@Subtract.Result) {template="Multiply::Execute(in A,in B,out Result)"}
   > RigVMFunctionReturnNode(Vector=@Multiply.Result)
  ]}
 New Function_ContainedGraph {nodes [
   n1 = RigVMFunctionEntryNode()
   n2 = RigVMVariableNode() {variable=Shoulder Connector}
   > RigVMFunctionReturnNode(Local Trasnform=&RigVMFunction_MathTransformMakeRelative(Parent=@n1.Input Transform, Global=&RigUnit_GetTransform(bInitial=true, Item=@n2.Value).Transform).Local)
  ]}
 New Function_ContainedGraph {nodes [
   n1 = RigVMFunctionEntryNode()
   n2 = RigVMVariableNode() {variable=Scapula Tip Connector}
   > RigVMFunctionReturnNode(Local Transform=&RigVMFunction_MathTransformMakeRelative(Parent=@n1.Input Transform, Global=&RigUnit_GetTransform(bInitial=true, Item=@n2.Value).Transform).Local)
  ]}
 New Function_ContainedGraph {nodes [
   n1 = RigVMFunctionEntryNode()
   n2 = RigVMVariableNode() {variable=Elbow Connector}
   > RigVMFunctionReturnNode(Local Transform=&RigVMFunction_MathTransformMakeRelative(Parent=@n1.Input Transform, Global=&RigUnit_GetTransform(bInitial=true, Item=@n2.Value).Transform).Local)
  ]}
 New Function_ContainedGraph {nodes [
   n1 = RigVMFunctionEntryNode()
   n2 = RigVMVariableNode() {variable=Wrist Connector}
   > RigVMFunctionReturnNode(Local Transform=&RigVMFunction_MathTransformMakeRelative(Parent=@n1.Input Transform, Global=&RigUnit_GetTransform(bInitial=true, Item=@n2.Value).Transform).Local)
  ]}
 New Function_ContainedGraph {nodes [
   n1 = RigVMFunctionEntryNode()
   n2 = RigVMVariableNode() {variable=Ball Connector}
   > RigVMFunctionReturnNode(Local Transform=&RigVMFunction_MathTransformMakeRelative(Parent=@n1.Input Transform, Global=&RigUnit_GetTransform(bInitial=true, Item=@n2.Value).Transform).Local)
  ]}
 New Function_ContainedGraph {nodes [
   RigVMFunctionEntryNode()
   > RigVMFunction_Sequence = RigVMAggregateNode()
   RigVMFunction_Sequence_1 = RigVMAggregateNode(ExecuteContext=@RigVMFunction_Sequence.A)
   Get Main Pole Vector_3 = RigVMFunctionReferenceNode(Initial=False)
   n1 = &RigUnit_AimBoneMath(
    DebugSettings=(bEnabled=False,Scale=10,WorldOffset=(Rotation=(X=0,Y=0,Z=0,W=1),Translation=(X=0,Y=0,Z=0),Scale3D=(X=1,Y=1,Z=1)))
    InputTransform=(Rotation=(X=0,Y=0,Z=0,W=1),Translation=(X=0,Y=0,Z=0),Scale3D=(X=1,Y=1,Z=1))
    Primary=(Weight=1,Axis=(X=1,Y=0,Z=0),Target=(X=1,Y=0,Z=0),Kind=Location,Space=(Type=Bone,Name="None"))
    Secondary=(Weight=1,Axis=(X=0,Y=0,Z=1),Target=(X=1,Y=0,Z=0),Kind=Direction,Space=(Type=Bone,Name="None"))
    Weight=1
    Secondary.Target=@"Get Main Pole Vector_3".Vector
    Primary.Target=RigVMVariableNode() {variable=Ball Rig Transform}.Value.Translation
    InputTransform.Translation=RigVMVariableNode() {variable=Wrist Point Transform}.Value.Translation
   )
   RigVMVariableNode(Value=@n1.Result, ExecuteContext=@RigVMFunction_Sequence_1.F) {variable=Wrist Rig Transform}
   RigVMFunction_DebugLineNoSpace = &RigVMFunction_DebugLineNoSpace(
    Color=(R=1,G=0,B=0,A=1)
    DebugDrawSettings=(DepthPriority=SDPG_Foreground,Lifetime=-1)
    Thickness=1
    WorldOffset=(Rotation=(X=0,Y=0,Z=0,W=1),Translation=(X=0,Y=0,Z=0),Scale3D=(X=1,Y=1,Z=1))
    A=RigVMVariableNode() {variable=Wrist Rig Transform}.Value.Translation
    B=RigVMVariableNode() {variable=Ball Rig Transform}.Value.Translation
    bEnabled=RigVMVariableNode() {variable=Debug}.Value
    ExecutePin=@RigVMFunction_Sequence_1.H
   )
   > RigVMFunction_DebugLineStripNoSpace = &RigVMFunction_DebugLineStripNoSpace(
    Color=(R=1,G=0,B=0,A=1)
    DebugDrawSettings=(DepthPriority=SDPG_Foreground,Lifetime=-1)
    Points=((X=0,Y=0,Z=0),(X=0,Y=0,Z=0),(X=0,Y=0,Z=0),(X=0,Y=0,Z=0))
    Thickness=0
    WorldOffset=(Rotation=(X=0,Y=0,Z=0,W=1),Translation=(X=0,Y=0,Z=0),Scale3D=(X=1,Y=1,Z=1))
    Points.1=RigVMVariableNode() {variable=Elbow Rig Transform}.Value.Translation
    Points.2=RigVMVariableNode() {variable=Wrist Rig Transform}.Value.Translation
    Points.0=RigVMVariableNode() {variable=Shoulder Rig Transform}.Value.Translation
    Points.3=RigVMVariableNode() {variable=Shoulder Rig Transform}.Value.Translation
    bEnabled=RigVMVariableNode() {variable=Debug}.Value
   )
   RigVMFunction_MathVectorDistance = &RigVMFunction_MathVectorDistance(
    B=RigVMVariableNode() {variable=Wrist Point Transform}.Value.Translation
    A=RigVMVariableNode() {variable=Shoulder Rig Transform}.Value.Translation
   )
   Get Ik Angles = RigVMFunctionReferenceNode(
    B 180=false
    C 180=true
    Negate B=true
    Negate C=true
    Length C=@RigVMFunction_MathVectorDistance.Result
    Length B=RigVMVariableNode() {variable=Elbow Length}.Value
    Length A=RigVMVariableNode() {variable=Shoulder Length}.Value
   )
   RigVMVariableNode(Value=@"Get Ik Angles".Angle B, ExecuteContext=@RigVMFunction_Sequence_1.B) {variable=Shoulder Angle}
   > RigVMVariableNode(Value=RigVMDispatchNode(
     True=0
     Condition=RigVMDispatchNode(B=180, A=@"Get Ik Angles".Angle C) {template="DISPATCH_RigVMDispatch_CoreEquals(in A,in B,out Result)"}.Result
     False=@"Get Ik Angles".Angle C
    ) {template="DISPATCH_RigVMDispatch_If(in Condition,in True,in False,out Result)"}.Result) {variable=Elbow Angle}
   Get Main Pole Vector_1 = RigVMFunctionReferenceNode(Initial=False)
   n2 = &RigUnit_AimBoneMath(
    DebugSettings=(bEnabled=False,Scale=10,WorldOffset=(Rotation=(X=0,Y=0,Z=0,W=1),Translation=(X=0,Y=0,Z=0),Scale3D=(X=1,Y=1,Z=1)))
    InputTransform=(Rotation=(X=0,Y=0,Z=0,W=1),Translation=(X=0,Y=0,Z=0),Scale3D=(X=1,Y=1,Z=1))
    Primary=(Weight=1,Axis=(X=1,Y=0,Z=0),Target=(X=1,Y=0,Z=0),Kind=Location,Space=(Type=Bone,Name="None"))
    Secondary=(Weight=1,Axis=(X=0,Y=0,Z=1),Target=(X=1,Y=0,Z=0),Kind=Direction,Space=(Type=Bone,Name="None"))
    Weight=1
    Secondary.Target=@"Get Main Pole Vector_1".Vector
    Primary.Target=RigVMVariableNode() {variable=Wrist Point Transform}.Value.Translation
    InputTransform.Translation=RigVMVariableNode() {variable=Shoulder Rig Transform}.Value.Translation
   )
   RigVMVariableNode(Value=@n2.Result, ExecuteContext=@RigVMFunction_Sequence_1.C) {variable=Ik Aim Transform}
   n3 = RigVMVariableNode() {variable=Shoulder Angle}
   RigVMVariableNode(
    Value=&RigVMFunction_MathTransformMakeAbsolute(
     Local=(Rotation=(X=0,Y=0,Z=0,W=1),Translation=(X=0,Y=0,Z=0),Scale3D=(X=1,Y=1,Z=1))
     Local.Rotation=&RigVMFunction_MathQuaternionFromEuler(Euler=(X=0,Y=0,Z=0), Euler.Z=@n3.Value).Result
     Parent=RigVMVariableNode() {variable=Ik Aim Transform}.Value
    ) {template="Make Absolute::Execute(in Local,in Parent,out Global)"}.Global
    ExecuteContext=@RigVMFunction_Sequence_1.D
   ) {variable=Shoulder Rig Transform}
   n4 = RigVMVariableNode() {variable=Shoulder Rig Transform}
   n5 = RigVMVariableNode() {variable=Shoulder Length}
   n6 = RigVMVariableNode() {variable=Elbow Angle}
   RigVMVariableNode(
    Value=&RigVMFunction_MathTransformMakeAbsolute(
     Local=(Rotation=(X=0,Y=0,Z=0,W=1),Translation=(X=0,Y=0,Z=0),Scale3D=(X=1,Y=1,Z=1))
     Parent=&RigVMFunction_MathTransformMakeAbsolute(Local=(Rotation=(X=0,Y=0,Z=0,W=1),Translation=(X=0,Y=0,Z=0),Scale3D=(X=1,Y=1,Z=1)), Parent=@n4.Value, Local.Translation.X=@n5.Value) {template="Make Absolute::Execute(in Local,in Parent,out Global)"}.Global
     Local.Rotation=&RigVMFunction_MathQuaternionFromEuler(Euler=(X=0,Y=0,Z=0), Euler.Z=@n6.Value).Result
    ) {template="Make Absolute::Execute(in Local,in Parent,out Global)"}.Global
    ExecuteContext=@RigVMFunction_Sequence_1.E
   ) {variable=Elbow Rig Transform}
   > RigVMVariableNode(Value=&RigVMFunction_MathTransformMakeAbsolute(
     Local=(Rotation=(X=0,Y=0,Z=0,W=1),Translation=(X=0,Y=0,Z=0),Scale3D=(X=1,Y=1,Z=1))
     Parent=RigVMVariableNode() {variable=Elbow Rig Transform}.Value
     Local.Translation.X=RigVMVariableNode() {variable=Elbow Length}.Value
    ) {template="Make Absolute::Execute(in Local,in Parent,out Global)"}.Global) {variable=Wrist Point Transform}
   RigVMVariableNode(
    Value=&RigVMFunction_MathTransformMakeAbsolute(
     Local=(Rotation=(X=0,Y=0,Z=0,W=1),Translation=(X=0,Y=0,Z=0),Scale3D=(X=1,Y=1,Z=1))
     Local.Translation.X=RigVMVariableNode() {variable=Wrist Length}.Value
     Parent=RigVMVariableNode() {variable=Wrist Rig Transform}.Value
    ) {template="Make Absolute::Execute(in Local,in Parent,out Global)"}.Global
    ExecuteContext=@RigVMFunction_Sequence_1.G
   ) {variable=Ball Rig Transform}
   Get Main Pole Vector_2 = RigVMFunctionReferenceNode(Initial=False)
   n7 = &RigUnit_AimBoneMath(
    DebugSettings=(bEnabled=False,Scale=10,WorldOffset=(Rotation=(X=0,Y=0,Z=0,W=1),Translation=(X=0,Y=0,Z=0),Scale3D=(X=1,Y=1,Z=1)))
    InputTransform=(Rotation=(X=0,Y=0,Z=0,W=1),Translation=(X=0,Y=0,Z=0),Scale3D=(X=1,Y=1,Z=1))
    Primary=(Weight=1,Axis=(X=1,Y=0,Z=0),Target=(X=1,Y=0,Z=0),Kind=Location,Space=(Type=Bone,Name="None"))
    Secondary=(Weight=1,Axis=(X=0,Y=0,Z=1),Target=(X=1,Y=0,Z=0),Kind=Direction,Space=(Type=Bone,Name="None"))
    Weight=1
    Secondary.Target=@"Get Main Pole Vector_2".Vector
    Primary.Target=RigVMVariableNode() {variable=Elbow Rig Transform}.Value.Translation
    InputTransform.Translation=RigVMVariableNode() {variable=Ball Rig Transform}.Value.Translation
   )
   &RigUnit_SetTransform(
    Weight=1
    bInitial=False
    Value=@n7.Result
    Item=RigVMVariableNode() {variable=Ball Rotate Aim Null}.Value
    ExecutePin=@RigVMFunction_Sequence_1.A
   )
   > RigVMVariableNode(Value=&RigUnit_GetTransform(bInitial=False, Item=RigVMVariableNode() {variable=Wrist Ik Point}.Value).Transform) {variable=Wrist Point Transform}
   RigVMFunctionReturnNode(ExecuteContext=@RigVMFunction_Sequence.B)
   &RigUnit_GetTransform(bInitial=False, Item=RigVMVariableNode() {variable=Ik Control}.Value)
   Get Main Pole Vector = RigVMFunctionReferenceNode(Initial=False)
   n8 = &RigUnit_AimBoneMath(
    DebugSettings=(bEnabled=False,Scale=10,WorldOffset=(Rotation=(X=0,Y=0,Z=0,W=1),Translation=(X=0,Y=0,Z=0),Scale3D=(X=1,Y=1,Z=1)))
    InputTransform=(Rotation=(X=0,Y=0,Z=0,W=1),Translation=(X=0,Y=0,Z=0),Scale3D=(X=1,Y=1,Z=1))
    Primary=(Weight=1,Axis=(X=1,Y=0,Z=0),Target=(X=1,Y=0,Z=0),Kind=Location,Space=(Type=Bone,Name="None"))
    Secondary=(Weight=1,Axis=(X=0,Y=0,Z=1),Target=(X=1,Y=0,Z=0),Kind=Direction,Space=(Type=Bone,Name="None"))
    Weight=1
    Secondary.Target=@"Get Main Pole Vector".Vector
    InputTransform.Translation=RigVMVariableNode() {variable=Ball Rig Transform}.Value.Translation
    Primary.Target=RigVMVariableNode() {variable=Wrist Rig Transform}.Value.Translation
   )
   GetAnimationChannelFromItem = &RigUnit_GetFloatAnimationChannelFromItem(bInitial=False, Item=RigVMVariableNode() {variable=Ball Rotate Channel}.Value) {template="GetAnimationChannelFromItem::Execute(out Value,in Item,in bInitial)"}
   RigVMFunction_MathQuaternionFromEuler = &RigVMFunction_MathQuaternionFromEuler(Euler=(X=0,Y=0,Z=0), Euler.Z=@GetAnimationChannelFromItem.Value)
   &RigVMFunction_MathTransformMakeAbsolute(
    Local=(Rotation=(X=0,Y=0,Z=0,W=1),Translation=(X=0,Y=0,Z=0),Scale3D=(X=1,Y=1,Z=1))
    Parent=&RigVMFunction_MathTransformMakeAbsolute(
     Local=(Rotation=(X=0,Y=0,Z=0,W=1),Translation=(X=0,Y=0,Z=0),Scale3D=(X=1,Y=1,Z=1))
     Parent=@n8.Result
     Local.Rotation=@RigVMFunction_MathQuaternionFromEuler.Result
    ) {template="Make Absolute::Execute(in Local,in Parent,out Global)"}.Global
    Local.Translation.X=RigVMVariableNode() {variable=Wrist Length}.Value
   ) {template="Make Absolute::Execute(in Local,in Parent,out Global)"}
  ]}
 New Function_ContainedGraph {nodes [
   n1 = RigVMFunctionEntryNode()
   RigVMFunction_Sequence = &RigVMFunction_Sequence(ExecuteContext=@n1.ExecuteContext)
   &RigVMFunction_Sequence(ExecuteContext=@RigVMFunction_Sequence.A)
   RigVMFunctionReturnNode(
    Ball Transform=(Rotation=(X=0,Y=0,Z=0,W=1),Translation=(X=0,Y=0,Z=0),Scale3D=(X=1,Y=1,Z=1))
    Elbow Transform=(Rotation=(X=0,Y=0,Z=0,W=1),Translation=(X=0,Y=0,Z=0),Scale3D=(X=1,Y=1,Z=1))
    Scapula Tip Transform=(Rotation=(X=0,Y=0,Z=0,W=1),Translation=(X=0,Y=0,Z=0),Scale3D=(X=1,Y=1,Z=1))
    Shoulder Transform=(Rotation=(X=0,Y=0,Z=0,W=1),Translation=(X=0,Y=0,Z=0),Scale3D=(X=1,Y=1,Z=1))
    Wrist Transform=(Rotation=(X=0,Y=0,Z=0,W=1),Translation=(X=0,Y=0,Z=0),Scale3D=(X=1,Y=1,Z=1))
    ExecuteContext=@RigVMFunction_Sequence.B
   )
  ]}
 New Function_ContainedGraph {nodes [
   n1 = RigVMFunctionEntryNode()
   RigVMFunction_Sequence = &RigVMFunction_Sequence(ExecuteContext=@n1.ExecuteContext)
   RigVMFunction_Sequence_1 = RigVMAggregateNode(ExecuteContext=@RigVMFunction_Sequence.A)
   Create Name_2 = RigVMFunctionReferenceNode(Tokens=("Scapula Tip Control Null"), ExecuteContext=@RigVMFunction_Sequence_1.A)
   n2 = RigVMDispatchNode(False=1, True=-1, Condition=RigVMVariableNode() {variable=Right Side}.Value) {template="DISPATCH_RigVMDispatch_If(in Condition,in True,in False,out Result)"}
   n3 = RigVMVariableNode() {variable=Left Side}
   > n4 = &RigUnit_HierarchyAddNull(
    Space=GlobalSpace
    Transform=(Rotation=(X=0,Y=0,Z=0,W=1),Translation=(X=0,Y=0,Z=0),Scale3D=(X=1,Y=1,Z=1))
    Name=@"Create Name_2".Name
    Transform.Translation=&RigUnit_GetTransform(bInitial=true, Item=RigVMVariableNode() {variable=Scapula Tip Connector}.Value).Transform.Translation
    Parent=&RigUnit_HierarchyGetParent(bDefaultParent=True, Child=RigVMVariableNode() {variable=Shoulder Connector}.Value).Parent
    Transform.Scale3D.X=&RigVMFunction_MathDoubleMul(
     A=1
     B=RigVMDispatchNode(True=1, False=@n2.Result, Condition=@n3.Value) {template="DISPATCH_RigVMDispatch_If(in Condition,in True,in False,out Result)"}.Result
    ).Result
   )
   > Create Name_3 = RigVMFunctionReferenceNode(Tokens=("Scapula Tip"))
   > Set Side Color = RigVMFunctionReferenceNode()
   n5 = &RigVMFunction_MathDoubleMul(B=0.35, A=RigVMVariableNode() {variable=Control Scale}.Value)
   > n6 = &RigUnit_HierarchyAddControlTransform(
    InitialValue=(Rotation=(X=0,Y=0,Z=0,W=1),Translation=(X=0,Y=0,Z=0),Scale3D=(X=1,Y=1,Z=1))
    OffsetSpace=LocalSpace
    OffsetTransform=(Rotation=(X=0,Y=0,Z=0,W=1),Translation=(X=0,Y=0,Z=0),Scale3D=(X=1,Y=1,Z=1))
    Settings="(InitialSpace=LocalSpace,bUsePreferredRotationOrder=False,PreferredRotationOrder=YZX,Limits=(LimitTranslationX=(bMinimum=False,bMaximum=False),LimitTranslationY=(bMinimum=False,bMaximum=False),LimitTranslationZ=(bMinimum=False,bMaximum=False),LimitPitch=(b ...(truncated)"
    Name=@"Create Name_3".Name
    Settings.Shape.Transform.Scale3D.X=@n5.Result
    Settings.Shape.Transform.Scale3D.Y=@n5.Result
    Settings.Shape.Transform.Scale3D.Z=@n5.Result
    Settings.Shape.Color=@"Set Side Color".Color
    Parent=@n4.Item
   ) {template="SpawnControl::Execute(in InitialValue,in Settings,in OffsetTransform,in OffsetSpace,in Parent,in Name,out Item)"}
   > RigVMVariableNode(Value=@n6.Item) {variable=Scapula Tip Control}
   RigVMFunctionReturnNode(ExecuteContext=@RigVMFunction_Sequence.B)
   n7 = RigVMRerouteNode(Value=())
   n8 = &RigUnit_GetTransform(bInitial=true, Item=RigVMVariableNode() {variable=Shoulder Connector}.Value)
   Subtract = &RigVMFunction_MathVectorSub(
    A=@n8.Transform.Translation
    B=&RigUnit_GetTransform(bInitial=true, Item=RigVMVariableNode() {variable=Ball Connector}.Value).Transform.Translation
   ) {template="Subtract::Execute(in A,in B,out Result)"}
   Add = &RigVMFunction_MathVectorAdd(A=@n8.Transform.Translation, B=@Subtract.Result) {template="Add::Execute(in A,in B,out Result)"}
   RigVMVariableNode(Value=@Add.Result, ExecuteContext=@n7.Value) {variable=Aim Point}
   RigVMFunction_MathVectorDistance = &RigVMFunction_MathVectorDistance(
    A=&RigUnit_GetTransform(bInitial=true, Item=RigVMVariableNode() {variable=Scapula Tip Connector}.Value).Transform.Translation
    B=&RigUnit_GetTransform(bInitial=true, Item=RigVMVariableNode() {variable=Elbow Connector}.Value).Transform.Translation
   )
   > RigVMVariableNode(Value=@RigVMFunction_MathVectorDistance.Result) {variable=Scapula To Elbow Distance}
   n9 = RigVMRerouteNode(Value=())
   Get Ik Angles = RigVMFunctionReferenceNode(
    Length A=RigVMVariableNode() {variable=Scapula Length}.Value
    Length B=RigVMVariableNode() {variable=Shoulder Length}.Value
    Length C=RigVMVariableNode() {variable=Scapula To Elbow Distance}.Value
   )
   RigVMVariableNode(Value=@"Get Ik Angles".Angle B, ExecuteContext=@n9.Value) {variable=Scapula Angle}
   > Create Name = RigVMFunctionReferenceNode(Tokens=("Fr_Leg_Main_Ik_Root"))
   n10 = &RigUnit_AimBoneMath(
    DebugSettings=(bEnabled=False,Scale=10,WorldOffset=(Rotation=(X=0,Y=0,Z=0,W=1),Translation=(X=0,Y=0,Z=0),Scale3D=(X=1,Y=1,Z=1)))
    Primary=(Weight=1,Axis=(X=1,Y=0,Z=0),Target=(X=1,Y=0,Z=0),Kind=Location,Space=(Type=Bone,Name="None"))
    Secondary=(Weight=1,Axis=(X=0,Y=0,Z=1),Target=(X=1,Y=0,Z=0),Kind=Direction,Space=(Type=Bone,Name="None"))
    Weight=1
    InputTransform=&RigUnit_GetTransform(bInitial=true, Item=RigVMVariableNode() {variable=Shoulder Connector}.Value).Transform
    Primary.Target=RigVMVariableNode() {variable=Aim Point}.Value
    Secondary.Target=RigVMVariableNode() {variable=Rig Up Vector}.Value
   )
   RigVMFunction_MathQuaternionFromEuler = &RigVMFunction_MathQuaternionFromEuler(Euler=(X=0,Y=0,Z=0), Euler.Z=RigVMVariableNode() {variable=Scapula Angle}.Value)
   n11 = RigVMVariableNode() {variable=Right Side}
   Multiply = &RigVMFunction_MathDoubleMul(
    A=1
    B=RigVMDispatchNode(
     True=1
     False=RigVMDispatchNode(False=1, True=-1, Condition=@n11.Value) {template="DISPATCH_RigVMDispatch_If(in Condition,in True,in False,out Result)"}.Result
     Condition=RigVMVariableNode() {variable=Left Side}.Value
    ) {template="DISPATCH_RigVMDispatch_If(in Condition,in True,in False,out Result)"}.Result
   )
   > n12 = &RigUnit_HierarchyAddNull(
    Space=GlobalSpace
    Transform=(Rotation=(X=0,Y=0,Z=0,W=1),Translation=(X=0,Y=0,Z=0),Scale3D=(X=1,Y=1,Z=1))
    Name=@"Create Name".Name
    Parent=&RigUnit_HierarchyGetParent(bDefaultParent=True, Child=RigVMVariableNode() {variable=Shoulder Connector}.Value).Parent
    Transform.Translation=&RigVMFunction_MathTransformMakeAbsolute(
     Local=(Rotation=(X=0,Y=0,Z=0,W=1),Translation=(X=0,Y=0,Z=0),Scale3D=(X=1,Y=1,Z=1))
     Parent=&RigVMFunction_MathTransformMakeAbsolute(
      Local=(Rotation=(X=0,Y=0,Z=0,W=1),Translation=(X=0,Y=0,Z=0),Scale3D=(X=1,Y=1,Z=1))
      Parent=@n10.Result
      Local.Rotation=@RigVMFunction_MathQuaternionFromEuler.Result
     ) {template="Make Absolute::Execute(in Local,in Parent,out Global)"}.Global
     Local.Translation.X=RigVMVariableNode() {variable=Scapula Length}.Value
    ) {template="Make Absolute::Execute(in Local,in Parent,out Global)"}.Global.Translation
    Transform.Scale3D.X=@Multiply.Result
   )
   > Create Name_1 = RigVMFunctionReferenceNode(Tokens=("Scapula Tip"))
   n13 = &RigVMFunction_MathDoubleMul(B=1.5, A=RigVMVariableNode() {variable=Control Scale}.Value)
   > SpawnControl = &RigUnit_HierarchyAddControlTransform(
    InitialValue=(Rotation=(X=0,Y=0,Z=0,W=1),Translation=(X=0,Y=0,Z=0),Scale3D=(X=1,Y=1,Z=1))
    OffsetSpace=LocalSpace
    OffsetTransform=(Rotation=(X=0,Y=0,Z=0,W=1),Translation=(X=0,Y=0,Z=0),Scale3D=(X=1,Y=1,Z=1))
    Settings="(InitialSpace=LocalSpace,bUsePreferredRotationOrder=False,PreferredRotationOrder=YZX,Limits=(LimitTranslationX=(bMinimum=False,bMaximum=False),LimitTranslationY=(bMinimum=False,bMaximum=False),LimitTranslationZ=(bMinimum=False,bMaximum=False),LimitPitch=(b ...(truncated)"
    Name=@"Create Name_1".Name
    Settings.Shape.Transform.Scale3D.X=@n13.Result
    Settings.Shape.Transform.Scale3D.Y=@n13.Result
    Settings.Shape.Transform.Scale3D.Z=@n13.Result
    Parent=@n12.Item
   ) {template="SpawnControl::Execute(in InitialValue,in Settings,in OffsetTransform,in OffsetSpace,in Parent,in Name,out Item)"}
   > RigVMVariableNode(Value=@SpawnControl.Item) {variable=Scapula Tip Control}
  ]}
 New Function_ContainedGraph {nodes [
   n1 = RigVMFunctionEntryNode()
   RigVMFunction_Sequence = &RigVMFunction_Sequence(ExecuteContext=@n1.ExecuteContext)
   RigVMFunction_Sequence_1 = RigVMAggregateNode(ExecuteContext=@RigVMFunction_Sequence.A)
   Get Main Pole Vector = RigVMFunctionReferenceNode(Initial=False)
   n2 = &RigUnit_AimBoneMath(
    DebugSettings=(bEnabled=False,Scale=10,WorldOffset=(Rotation=(X=0,Y=0,Z=0,W=1),Translation=(X=0,Y=0,Z=0),Scale3D=(X=1,Y=1,Z=1)))
    Primary=(Weight=1,Axis=(X=1,Y=0,Z=0),Target=(X=1,Y=0,Z=0),Kind=Location,Space=(Type=Bone,Name="None"))
    Secondary=(Weight=1,Axis=(X=0,Y=0,Z=1),Target=(X=1,Y=0,Z=0),Kind=Direction,Space=(Type=Bone,Name="None"))
    Weight=1
    Secondary.Target=@"Get Main Pole Vector".Vector
    InputTransform=&RigUnit_GetTransform(bInitial=False, Item=RigVMVariableNode() {variable=Scapula Tip Control}.Value).Transform
    Primary.Target=RigVMVariableNode() {variable=Shoulder Rig Transform}.Value.Translation
   )
   n3 = RigVMVariableNode() {variable=Scapula Tip Local Transform}
   &RigUnit_SetTransform(
    Weight=1
    bInitial=False
    Value=&RigVMFunction_MathTransformMakeAbsolute(
     Parent=&RigVMFunction_MathTransformMakeAbsolute(Parent=@n2.Result, Local=@n3.Value) {template="Make Absolute::Execute(in Local,in Parent,out Global)"}.Global
     Local=RigVMVariableNode() {variable=Scapula Tip Offset Transform}.Value
    ) {template="Make Absolute::Execute(in Local,in Parent,out Global)"}.Global
    Item=RigVMVariableNode() {variable=Scapula Tip Connector}.Value
    ExecutePin=@RigVMFunction_Sequence_1.B
   )
   &RigUnit_SetTransform(
    Weight=1
    bInitial=False
    Value=&RigVMFunction_MathTransformMakeAbsolute(Parent=RigVMVariableNode() {variable=Shoulder Rig Transform}.Value, Local=RigVMVariableNode() {variable=Shoulder Local Transform}.Value) {template="Make Absolute::Execute(in Local,in Parent,out Global)"}.Global
    Item=RigVMVariableNode() {variable=Shoulder Connector}.Value
    ExecutePin=@RigVMFunction_Sequence_1.A
   )
   &RigUnit_SetTransform(
    Weight=1
    bInitial=False
    Value=&RigVMFunction_MathTransformMakeAbsolute(Local=RigVMVariableNode() {variable=Elbow Local Transform}.Value, Parent=RigVMVariableNode() {variable=Elbow Rig Transform}.Value) {template="Make Absolute::Execute(in Local,in Parent,out Global)"}.Global
    Item=RigVMVariableNode() {variable=Elbow Connector}.Value
    ExecutePin=@RigVMFunction_Sequence_1.C
   )
   &RigUnit_SetTransform(
    Weight=1
    bInitial=False
    Value=&RigVMFunction_MathTransformMakeAbsolute(Parent=RigVMVariableNode() {variable=Wrist Rig Transform}.Value, Local=RigVMVariableNode() {variable=Wrist Local Transform}.Value) {template="Make Absolute::Execute(in Local,in Parent,out Global)"}.Global
    Item=RigVMVariableNode() {variable=Wrist Connector}.Value
    ExecutePin=@RigVMFunction_Sequence_1.D
   )
   &RigUnit_SetTransform(
    Weight=1
    bInitial=False
    Value=&RigVMFunction_MathTransformMakeAbsolute(Local=RigVMVariableNode() {variable=Ball Local Transform}.Value, Parent=RigVMVariableNode() {variable=Ball Rig Transform}.Value) {template="Make Absolute::Execute(in Local,in Parent,out Global)"}.Global
    Item=RigVMVariableNode() {variable=Ball Connector}.Value
    ExecutePin=@RigVMFunction_Sequence_1.E
   )
   &RigUnit_SetTransform(
    Weight=1
    bInitial=False
    Value=&RigVMFunction_MathTransformMakeAbsolute(Parent=RigVMVariableNode() {variable=Toe Rig Transform}.Value, Local=RigVMVariableNode() {variable=Toe Local Transform}.Value) {template="Make Absolute::Execute(in Local,in Parent,out Global)"}.Global
    Item=RigVMVariableNode() {variable=Toe Connector}.Value
    ExecutePin=@RigVMFunction_Sequence_1.G
   )
   n4 = RigVMVariableNode() {variable=Toe Fk Control}
   &RigUnit_SetTransform(
    Space=LocalSpace
    Weight=1
    bInitial=False
    Value=&RigUnit_GetTransform(Space=LocalSpace, bInitial=true, Item=&RigUnit_HierarchyGetParent(bDefaultParent=True, Child=@n4.Value).Parent).Transform
    Item=&RigUnit_HierarchyGetParent(bDefaultParent=True, Child=RigVMVariableNode() {variable=Toe Fk Control}.Value).Parent
    ExecutePin=@RigVMFunction_Sequence_1.F
   )
   n5 = RigVMVariableNode() {variable=Ball Pivot Control}
   > &RigUnit_SetTransform(
    Space=LocalSpace
    Weight=1
    bInitial=False
    Value=&RigUnit_GetTransform(Space=LocalSpace, bInitial=true, Item=&RigUnit_HierarchyGetParent(bDefaultParent=True, Child=@n5.Value).Parent).Transform
    Item=&RigUnit_HierarchyGetParent(bDefaultParent=True, Child=RigVMVariableNode() {variable=Ball Pivot Control}.Value).Parent
   )
   n6 = RigVMVariableNode() {variable=Ball Fk Control}
   > &RigUnit_SetTransform(
    Space=LocalSpace
    Weight=1
    bInitial=False
    Value=&RigUnit_GetTransform(Space=LocalSpace, bInitial=true, Item=&RigUnit_HierarchyGetParent(bDefaultParent=True, Child=@n6.Value).Parent).Transform
    Item=&RigUnit_HierarchyGetParent(bDefaultParent=True, Child=RigVMVariableNode() {variable=Ball Fk Control}.Value).Parent
   )
   > &RigUnit_SetTransform(
    Space=LocalSpace
    Weight=1
    bInitial=False
    Value=&RigUnit_GetTransform(Space=LocalSpace, bInitial=true, Item=RigVMVariableNode() {variable=Wrist Ik Point}.Value).Transform
    Item=RigVMVariableNode() {variable=Wrist Ik Point}.Value
   )
   RigVMVariableNode(
    Value=&RigUnit_ResolveConnector(Connector=(Type=Connector,Name="IK Front Connector"), SkipSocket=False).Result
    ExecuteContext=@RigVMFunction_Sequence_1.H
   ) {variable=IK Front Connector}
   > &RigUnit_ParentConstraint(
    AdvancedSettings=(InterpolationType=Average,RotationOrderForFilter=XZY)
    Parents=((Item=(Type=Connector,Name="Ball"),Weight=1))
    Weight=1
    Child=RigVMVariableNode() {variable=IK Front Connector}.Value
   )
   RigVMFunctionReturnNode(ExecuteContext=@RigVMFunction_Sequence.B)
   RigVMVariableNode() {variable=Ball Pivot Control}
   RigVMVariableNode() {variable=Ball Pivot Control}
  ]}
 New Function_ContainedGraph {nodes [
   n1 = RigVMFunctionEntryNode()
   RigVMFunction_Sequence = &RigVMFunction_Sequence(ExecuteContext=@n1.ExecuteContext)
   RigVMFunction_Sequence_1 = RigVMAggregateNode(ExecuteContext=@RigVMFunction_Sequence.A)
   DISPATCH_RigDispatch_SetModuleMetadata = RigVMDispatchNode(
    Name=Ik Control Null
    NameSpace=Self
    Value=&RigUnit_HierarchyGetParent(bDefaultParent=True, Child=RigVMVariableNode() {variable=Ik Control}.Value).Parent
    ExecuteContext=@RigVMFunction_Sequence_1.A
   ) {template="DISPATCH_RigDispatch_SetModuleMetadata(in Name,in NameSpace,in Value,out Success)"}
   RigVMDispatchNode(Name=Ik Control, NameSpace=Self, Value=RigVMVariableNode() {variable=Ik Control}.Value, ExecuteContext=@RigVMFunction_Sequence_1.B) {template="DISPATCH_RigDispatch_SetModuleMetadata(in Name,in NameSpace,in Value,out Success)"}
   RigVMDispatchNode(Name=Ik Point, NameSpace=Self, Value=RigVMVariableNode() {variable=Ik Point}.Value, ExecuteContext=@RigVMFunction_Sequence_1.C) {template="DISPATCH_RigDispatch_SetModuleMetadata(in Name,in NameSpace,in Value,out Success)"}
   RigVMFunctionReturnNode(ExecuteContext=@RigVMFunction_Sequence.B)
  ]}
 New Function_ContainedGraph {nodes [
   Create Name_6 = RigVMFunctionReferenceNode(ExecuteContext=(), Tokens=("Ball Point Ref"))
   > n1 = &RigUnit_HierarchyAddNull(
    Space=GlobalSpace
    Transform=&RigUnit_GetTransform(bInitial=true, Item=RigVMVariableNode() {variable=Ball Pivot Control}.Value).Transform
    Parent=RigVMVariableNode() {variable=Ball Pivot Control}.Value
    Name=@"Create Name_6".Name
   )
   > RigVMVariableNode(Value=@n1.Item) {variable=Ball Point Ref}
   n2 = RigVMFunctionEntryNode()
   RigVMFunction_Sequence = &RigVMFunction_Sequence(ExecuteContext=@n2.ExecuteContext)
   RigVMFunction_Sequence_1 = RigVMAggregateNode(ExecuteContext=@RigVMFunction_Sequence.A)
   Create Name = RigVMFunctionReferenceNode(Tokens=("Fr Ball Null"), ExecuteContext=@RigVMFunction_Sequence_1.A)
   n3 = &RigUnit_AimBoneMath(
    DebugSettings=(bEnabled=False,Scale=10,WorldOffset=(Rotation=(X=0,Y=0,Z=0,W=1),Translation=(X=0,Y=0,Z=0),Scale3D=(X=1,Y=1,Z=1)))
    Primary=(Weight=1,Axis=(X=0,Y=1,Z=0),Target=(X=1,Y=0,Z=0),Kind=Location,Space=(Type=Bone,Name="None"))
    Secondary=(Weight=1,Axis=(X=1,Y=0,Z=0),Target=(X=10,Y=0,Z=0),Kind=Direction,Space=(Type=Bone,Name="None"))
    Weight=1
    InputTransform=&RigUnit_GetTransform(bInitial=true, Item=RigVMVariableNode() {variable=Ball Connector}.Value).Transform
    Primary.Target=&RigUnit_GetTransform(bInitial=true, Item=RigVMVariableNode() {variable=Toe Connector}.Value).Transform.Translation
   )
   > n4 = &RigUnit_HierarchyAddNull(Space=GlobalSpace, Name=@"Create Name".Name, Parent=RigVMVariableNode() {variable=Heel Roll Control}.Value, Transform=@n3.Result)
   > Create Name_1 = RigVMFunctionReferenceNode(Tokens=("Fr Ball"))
   n5 = &RigVMFunction_MathDoubleMul(B=0.7, A=RigVMVariableNode() {variable=Control Scale}.Value)
   > SpawnControl = &RigUnit_HierarchyAddControlTransform(
    InitialValue=(Rotation=(X=0,Y=0,Z=0,W=1),Translation=(X=0,Y=0,Z=0),Scale3D=(X=1,Y=1,Z=1))
    OffsetSpace=LocalSpace
    OffsetTransform=(Rotation=(X=0,Y=0,Z=0,W=1),Translation=(X=0,Y=0,Z=0),Scale3D=(X=1,Y=1,Z=1))
    Settings="(InitialSpace=LocalSpace,bUsePreferredRotationOrder=False,PreferredRotationOrder=YZX,Limits=(LimitTranslationX=(bMinimum=True,bMaximum=True),LimitTranslationY=(bMinimum=True,bMaximum=True),LimitTranslationZ=(bMinimum=True,bMaximum=True),LimitPitch=(bMinimu ...(truncated)"
    Name=@"Create Name_1".Name
    Settings.Shape.Transform.Scale3D.X=@n5.Result
    Settings.Shape.Transform.Scale3D.Y=@n5.Result
    Settings.Shape.Transform.Scale3D.Z=@n5.Result
    Parent=@n4.Item
   ) {template="SpawnControl::Execute(in InitialValue,in Settings,in OffsetTransform,in OffsetSpace,in Parent,in Name,out Item)"}
   > RigVMVariableNode(Value=@SpawnControl.Item) {variable=Ball Fk Control}
   Create Name_2 = RigVMFunctionReferenceNode(Tokens=("Fr Toe Null"), ExecuteContext=@RigVMFunction_Sequence_1.B)
   n6 = &RigUnit_AimBoneMath(
    DebugSettings=(bEnabled=False,Scale=10,WorldOffset=(Rotation=(X=0,Y=0,Z=0,W=1),Translation=(X=0,Y=0,Z=0),Scale3D=(X=1,Y=1,Z=1)))
    Primary=(Weight=1,Axis=(X=0,Y=1,Z=0),Target=(X=1,Y=0,Z=0),Kind=Location,Space=(Type=Bone,Name="None"))
    Secondary=(Weight=1,Axis=(X=1,Y=0,Z=0),Target=(X=10,Y=0,Z=0),Kind=Direction,Space=(Type=Bone,Name="None"))
    Weight=1
    InputTransform=&RigUnit_GetTransform(bInitial=true, Item=RigVMVariableNode() {variable=Toe Connector}.Value).Transform
    Primary.Target=&RigUnit_GetTransform(bInitial=true, Item=RigVMVariableNode() {variable=Toe Tip Connector}.Value).Transform.Translation
   )
   > n7 = &RigUnit_HierarchyAddNull(Space=GlobalSpace, Name=@"Create Name_2".Name, Parent=RigVMVariableNode() {variable=Ball Fk Control}.Value, Transform=@n6.Result)
   > Create Name_3 = RigVMFunctionReferenceNode(Tokens=("Fr Toe"))
   > n8 = &RigUnit_HierarchyAddControlTransform(
    InitialValue=(Rotation=(X=0,Y=0,Z=0,W=1),Translation=(X=0,Y=0,Z=0),Scale3D=(X=1,Y=1,Z=1))
    OffsetSpace=LocalSpace
    OffsetTransform=(Rotation=(X=0,Y=0,Z=0,W=1),Translation=(X=0,Y=0,Z=0),Scale3D=(X=1,Y=1,Z=1))
    Settings="(InitialSpace=LocalSpace,bUsePreferredRotationOrder=False,PreferredRotationOrder=YZX,Limits=(LimitTranslationX=(bMinimum=True,bMaximum=True),LimitTranslationY=(bMinimum=True,bMaximum=True),LimitTranslationZ=(bMinimum=True,bMaximum=True),LimitPitch=(bMinimu ...(truncated)"
    Name=@"Create Name_3".Name
    Parent=@n7.Item
   ) {template="SpawnControl::Execute(in InitialValue,in Settings,in OffsetTransform,in OffsetSpace,in Parent,in Name,out Item)"}
   > RigVMVariableNode(Value=@n8.Item) {variable=Toe Fk Control}
   > Create Name_8 = RigVMFunctionReferenceNode(Tokens=("Fr Toe Tip Point"))
   > n9 = &RigUnit_HierarchyAddNull(
    Space=GlobalSpace
    Name=@"Create Name_8".Name
    Transform=&RigUnit_GetTransform(bInitial=true, Item=RigVMVariableNode() {variable=Toe Tip Connector}.Value).Transform
    Parent=RigVMVariableNode() {variable=Toe Fk Control}.Value
   )
   > RigVMVariableNode(Value=@n9.Item) {variable=Toe Tip Point}
   Create Name_4 = RigVMFunctionReferenceNode(Tokens=("Fr Ball Pivot Null"), ExecuteContext=@RigVMFunction_Sequence_1.C)
   n10 = &RigUnit_AimBoneMath(
    DebugSettings=(bEnabled=False,Scale=10,WorldOffset=(Rotation=(X=0,Y=0,Z=0,W=1),Translation=(X=0,Y=0,Z=0),Scale3D=(X=1,Y=1,Z=1)))
    Primary=(Weight=1,Axis=(X=0,Y=1,Z=0),Target=(X=1,Y=0,Z=0),Kind=Location,Space=(Type=Bone,Name="None"))
    Secondary=(Weight=1,Axis=(X=1,Y=0,Z=0),Target=(X=10,Y=0,Z=0),Kind=Direction,Space=(Type=Bone,Name="None"))
    Weight=1
    InputTransform=&RigUnit_GetTransform(bInitial=true, Item=RigVMVariableNode() {variable=Ball Connector}.Value).Transform
    Primary.Target=&RigUnit_GetTransform(bInitial=true, Item=RigVMVariableNode() {variable=Toe Connector}.Value).Transform.Translation
   )
   > n11 = &RigUnit_HierarchyAddNull(
    Space=GlobalSpace
    Transform=(Rotation=(X=0,Y=0,Z=0,W=1),Translation=(X=0,Y=0,Z=0),Scale3D=(X=1,Y=1,Z=1))
    Name=@"Create Name_4".Name
    Parent=RigVMVariableNode() {variable=Ball Fk Control}.Value
    Transform.Scale3D=@n10.Result.Scale3D
    Transform.Rotation=@n10.Result.Rotation
    Transform.Translation=&RigUnit_GetTransform(bInitial=true, Item=RigVMVariableNode() {variable=Toe Connector}.Value).Transform.Translation
   )
   > Create Name_5 = RigVMFunctionReferenceNode(Tokens=("Fr Ball Pivot"))
   n12 = &RigVMFunction_MathDoubleMul(B=1, A=RigVMVariableNode() {variable=Control Scale}.Value)
   > n13 = &RigUnit_HierarchyAddControlTransform(
    InitialValue=(Rotation=(X=0,Y=0,Z=0,W=1),Translation=(X=0,Y=0,Z=0),Scale3D=(X=1,Y=1,Z=1))
    OffsetSpace=LocalSpace
    OffsetTransform=(Rotation=(X=0,Y=0,Z=0,W=1),Translation=(X=0,Y=0,Z=0),Scale3D=(X=1,Y=1,Z=1))
    Settings="(InitialSpace=LocalSpace,bUsePreferredRotationOrder=False,PreferredRotationOrder=YZX,Limits=(LimitTranslationX=(bMinimum=True,bMaximum=True),LimitTranslationY=(bMinimum=True,bMaximum=True),LimitTranslationZ=(bMinimum=True,bMaximum=True),LimitPitch=(bMinimu ...(truncated)"
    Name=@"Create Name_5".Name
    Settings.Shape.Transform.Scale3D.X=@n12.Result
    Settings.Shape.Transform.Scale3D.Y=@n12.Result
    Settings.Shape.Transform.Scale3D.Z=@n12.Result
    Parent=@n11.Item
   ) {template="SpawnControl::Execute(in InitialValue,in Settings,in OffsetTransform,in OffsetSpace,in Parent,in Name,out Item)"}
   > RigVMVariableNode(Value=@n13.Item) {variable=Ball Pivot Control}
   > Create Name_7 = RigVMFunctionReferenceNode(Tokens=("Fr Ik Point"))
   > n14 = &RigUnit_HierarchyAddNull(
    Space=GlobalSpace
    Name=@"Create Name_7".Name
    Transform=&RigUnit_GetTransform(bInitial=true, Item=RigVMVariableNode() {variable=Ik Control}.Value).Transform
    Parent=RigVMVariableNode() {variable=Ball Pivot Control}.Value
   )
   > RigVMVariableNode(Value=@n14.Item) {variable=Ik Point}
   RigVMFunctionReturnNode(ExecuteContext=@RigVMFunction_Sequence.B)
   &RigVMFunction_MathDoubleMul(B=0.7, A=RigVMVariableNode() {variable=Control Scale}.Value)
   RigVMVariableNode() {variable=Ik Control}
  ]}
 New Function_ContainedGraph {nodes [
   n1 = RigVMFunctionEntryNode()
   RigVMFunction_Sequence = &RigVMFunction_Sequence(ExecuteContext=@n1.ExecuteContext)
   n2 = &RigVMFunction_Sequence(ExecuteContext=@RigVMFunction_Sequence.A)
   n3 = &RigUnit_AimBoneMath(
    DebugSettings=(bEnabled=False,Scale=10,WorldOffset=(Rotation=(X=0,Y=0,Z=0,W=1),Translation=(X=0,Y=0,Z=0),Scale3D=(X=1,Y=1,Z=1)))
    InputTransform=(Rotation=(X=0,Y=0,Z=0,W=1),Translation=(X=0,Y=0,Z=0),Scale3D=(X=1,Y=1,Z=1))
    Primary=(Weight=1,Axis=(X=1,Y=0,Z=0),Target=(X=1,Y=0,Z=0),Kind=Location,Space=(Type=Bone,Name="None"))
    Secondary=(Weight=1,Axis=(X=0,Y=0,Z=1),Target=(X=1,Y=0,Z=0),Kind=Direction,Space=(Type=Bone,Name="None"))
    Weight=1
    Primary.Target=&RigUnit_GetTransform(bInitial=False, Item=RigVMVariableNode() {variable=Toe Fk Control}.Value).Transform.Translation
    InputTransform.Translation=RigVMVariableNode() {variable=Ball Rig Transform}.Value.Translation
    Secondary.Space=RigVMVariableNode() {variable=Ball Pivot Control}.Value
   )
   RigVMVariableNode(Value=@n3.Result, ExecuteContext=@n2.A) {variable=Ball Rig Transform}
   n4 = &RigUnit_AimBoneMath(
    DebugSettings=(bEnabled=False,Scale=10,WorldOffset=(Rotation=(X=0,Y=0,Z=0,W=1),Translation=(X=0,Y=0,Z=0),Scale3D=(X=1,Y=1,Z=1)))
    InputTransform=(Rotation=(X=0,Y=0,Z=0,W=1),Translation=(X=0,Y=0,Z=0),Scale3D=(X=1,Y=1,Z=1))
    Primary=(Weight=1,Axis=(X=1,Y=0,Z=0),Target=(X=1,Y=0,Z=0),Kind=Location,Space=(Type=Bone,Name="None"))
    Secondary=(Weight=1,Axis=(X=0,Y=0,Z=1),Target=(X=1,Y=0,Z=0),Kind=Direction,Space=(Type=Bone,Name="None"))
    Weight=1
    Primary.Target=&RigUnit_GetTransform(bInitial=False, Item=RigVMVariableNode() {variable=Toe Tip Point}.Value).Transform.Translation
    InputTransform.Translation=&RigVMFunction_MathTransformMakeAbsolute(
     Local=(Rotation=(X=0,Y=0,Z=0,W=1),Translation=(X=0,Y=0,Z=0),Scale3D=(X=1,Y=1,Z=1))
     Parent=RigVMVariableNode() {variable=Ball Rig Transform}.Value
     Local.Translation.X=RigVMVariableNode() {variable=Ball Length}.Value
    ) {template="Make Absolute::Execute(in Local,in Parent,out Global)"}.Global.Translation
    Secondary.Space=RigVMVariableNode() {variable=Toe Fk Control}.Value
   )
   RigVMVariableNode(Value=@n4.Result, ExecuteContext=@n2.B) {variable=Toe Rig Transform}
   RigVMFunctionReturnNode(ExecuteContext=@RigVMFunction_Sequence.B)
  ]}
 New Function_ContainedGraph {nodes [
   n1 = RigVMFunctionEntryNode()
   RigVMFunction_Sequence = &RigVMFunction_Sequence(ExecuteContext=@n1.ExecuteContext)
   n2 = &RigVMFunction_Sequence(ExecuteContext=@RigVMFunction_Sequence.A)
   n3 = RigVMVariableNode() {variable=Scapula Tip Connector}
   n4 = RigVMVariableNode() {variable=Scapula Tip Control}
   RigVMVariableNode(
    Value=&RigVMFunction_MathTransformMakeRelative(
     Parent=&RigUnit_GetTransform(bInitial=true, Item=@n3.Value).Transform
     Global=&RigUnit_GetTransform(bInitial=true, Item=@n4.Value).Transform
    ).Local
    ExecuteContext=@n2.A
   ) {variable=Scapula Tip Control Local Transform}
   RigVMFunctionReturnNode(ExecuteContext=@RigVMFunction_Sequence.B)
  ]}
 New Function_ContainedGraph {nodes [
   n1 = RigVMFunctionEntryNode()
   RigVMFunction_Sequence = &RigVMFunction_Sequence(ExecuteContext=@n1.ExecuteContext)
   Set Side Color_1 = RigVMFunctionReferenceNode(ExecuteContext=@RigVMFunction_Sequence.A)
   > n2 = RigVMVariableNode(Value=@"Set Side Color_1".Color) {variable=Color}
   n3 = &RigVMFunction_Sequence(ExecuteContext=@n2.ExecuteContext)
   Create Name = RigVMFunctionReferenceNode(Tokens=("Fr Toe Roll Null"), ExecuteContext=@n3.A)
   > n4 = &RigUnit_HierarchyAddNull(
    Space=LocalSpace
    Parent=RigVMVariableNode() {variable=Ik Control}.Value
    Transform=RigVMVariableNode() {variable=Toe Roll Initial Transform}.Value
    Name=@"Create Name".Name
   )
   > Create Name_1 = RigVMFunctionReferenceNode(Tokens=("Fr Toe Roll"))
   n5 = &RigVMFunction_MathDoubleMul(B=0.3, A=RigVMVariableNode() {variable=Control Scale}.Value)
   > SpawnControl = &RigUnit_HierarchyAddControlTransform(
    InitialValue=(Rotation=(X=0,Y=0,Z=0,W=1),Translation=(X=0,Y=0,Z=0),Scale3D=(X=1,Y=1,Z=1))
    OffsetSpace=LocalSpace
    OffsetTransform=(Rotation=(X=0,Y=0,Z=0,W=1),Translation=(X=0,Y=0,Z=0),Scale3D=(X=1,Y=1,Z=1))
    Settings="(InitialSpace=LocalSpace,bUsePreferredRotationOrder=False,PreferredRotationOrder=YZX,Limits=(LimitTranslationX=(bMinimum=False,bMaximum=False),LimitTranslationY=(bMinimum=False,bMaximum=False),LimitTranslationZ=(bMinimum=False,bMaximum=False),LimitPitch=(b ...(truncated)"
    Name=@"Create Name_1".Name
    Settings.Shape.Transform.Scale3D.X=@n5.Result
    Settings.Shape.Transform.Scale3D.Y=@n5.Result
    Settings.Shape.Transform.Scale3D.Z=@n5.Result
    Settings.Shape.Color=RigVMVariableNode() {variable=Color}.Value
    Parent=@n4.Item
   ) {template="SpawnControl::Execute(in InitialValue,in Settings,in OffsetTransform,in OffsetSpace,in Parent,in Name,out Item)"}
   > RigVMVariableNode(Value=@SpawnControl.Item) {variable=Toe Roll Control}
   Create Name_2 = RigVMFunctionReferenceNode(Tokens=("Fr Heel Roll Null"), ExecuteContext=@n3.B)
   > n6 = &RigUnit_HierarchyAddNull(
    Space=LocalSpace
    Parent=RigVMVariableNode() {variable=Toe Roll Control}.Value
    Transform=RigVMVariableNode() {variable=Heel Roll Initial Transform}.Value
    Name=@"Create Name_2".Name
   )
   > Create Name_3 = RigVMFunctionReferenceNode(Tokens=("Fr Heel Roll"))
   n7 = &RigVMFunction_MathDoubleMul(B=0.3, A=RigVMVariableNode() {variable=Control Scale}.Value)
   > n8 = &RigUnit_HierarchyAddControlTransform(
    InitialValue=(Rotation=(X=0,Y=0,Z=0,W=1),Translation=(X=0,Y=0,Z=0),Scale3D=(X=1,Y=1,Z=1))
    OffsetSpace=LocalSpace
    OffsetTransform=(Rotation=(X=0,Y=0,Z=0,W=1),Translation=(X=0,Y=0,Z=0),Scale3D=(X=1,Y=1,Z=1))
    Settings="(InitialSpace=LocalSpace,bUsePreferredRotationOrder=False,PreferredRotationOrder=YZX,Limits=(LimitTranslationX=(bMinimum=False,bMaximum=False),LimitTranslationY=(bMinimum=False,bMaximum=False),LimitTranslationZ=(bMinimum=False,bMaximum=False),LimitPitch=(b ...(truncated)"
    Name=@"Create Name_3".Name
    Settings.Shape.Transform.Scale3D.X=@n7.Result
    Settings.Shape.Transform.Scale3D.Y=@n7.Result
    Settings.Shape.Transform.Scale3D.Z=@n7.Result
    Settings.Shape.Color=RigVMVariableNode() {variable=Color}.Value
    Parent=@n6.Item
   ) {template="SpawnControl::Execute(in InitialValue,in Settings,in OffsetTransform,in OffsetSpace,in Parent,in Name,out Item)"}
   > RigVMVariableNode(Value=@n8.Item) {variable=Heel Roll Control}
   RigVMFunctionReturnNode(ExecuteContext=@RigVMFunction_Sequence.B)
  ]}
 New Function_ContainedGraph {nodes [
   n1 = RigVMFunctionEntryNode()
   RigVMFunction_Sequence = &RigVMFunction_Sequence(ExecuteContext=@n1.ExecuteContext)
   n2 = &RigVMFunction_Sequence(ExecuteContext=@RigVMFunction_Sequence.A)
   RigVMFunction_ControlFlowBranch = &RigVMFunction_ControlFlowBranch(Condition=RigVMVariableNode() {variable=Left Side}.Value, ExecuteContext=@n2.A)
   n3 = &RigVMFunction_ControlFlowBranch(ExecuteContext=@RigVMFunction_ControlFlowBranch.False, Condition=RigVMVariableNode() {variable=Right Side}.Value)
   n4 = &RigVMFunction_ControlFlowBranch(ExecuteContext=@n3.False, Condition=RigVMVariableNode() {variable=Custom Side}.Value)
   RigVMVariableNode(Value=RigVMVariableNode() {variable=Custom Side Color}.Value, ExecuteContext=@n4.True) {variable=Output Color}
   RigVMVariableNode(Value=RigVMVariableNode() {variable=Right Side Color}.Value, ExecuteContext=@n3.True) {variable=Output Color}
   RigVMVariableNode(Value=RigVMVariableNode() {variable=Left Side Color}.Value, ExecuteContext=@RigVMFunction_ControlFlowBranch.True) {variable=Output Color}
   RigVMFunctionReturnNode(ExecuteContext=@RigVMFunction_Sequence.B, Color=RigVMVariableNode() {variable=Output Color}.Value)
  ]}
 New Function_ContainedGraph {nodes [
   n1 = RigVMFunctionEntryNode()
   RigVMFunction_Sequence = &RigVMFunction_Sequence(ExecuteContext=@n1.ExecuteContext)
   n2 = &RigVMFunction_Sequence(ExecuteContext=@RigVMFunction_Sequence.A)
   /Script/ControlRig.RigUnit_AddParent(
    DisplayLabel=None
    Child=RigVMVariableNode() {variable=Main PV Control}.Value
    Parent=RigVMVariableNode() {variable=Ik Control}.Value
    ExecutePin=@n2.A
   ) {is=struct RigVMUnitNode}
   Create Name_2 = RigVMFunctionReferenceNode(Tokens=("Chest Space"), ExecuteContext=@n2.B)
   n3 = &RigUnit_HierarchyGetParent(bDefaultParent=True, Child=RigVMVariableNode() {variable=Shoulder Connector}.Value)
   > n4 = &RigUnit_HierarchyAddNull(
    Space=GlobalSpace
    Transform=(Rotation=(X=0,Y=0,Z=0,W=1),Translation=(X=0,Y=0,Z=0),Scale3D=(X=1,Y=1,Z=1))
    Name=@"Create Name_2".Name
    Transform.Translation=&RigUnit_GetTransform(bInitial=False, Item=@n3.Parent).Transform.Translation
    Parent=@n3.Parent
   )
   > &RigUnit_AddParent(DisplayLabel=None, Child=RigVMVariableNode() {variable=Ik Control}.Value, Parent=@n4.Item)
   RigVMFunctionReturnNode(ExecuteContext=@RigVMFunction_Sequence.B)
  ]}
 New Function_ContainedGraph {nodes [
   n1 = RigVMFunctionEntryNode()
   RigVMFunction_Sequence = &RigVMFunction_Sequence(ExecuteContext=@n1.ExecuteContext)
   RigVMFunction_Sequence_1 = RigVMAggregateNode(ExecuteContext=@RigVMFunction_Sequence.A)
   n2 = &RigUnit_AimBoneMath(
    DebugSettings=(bEnabled=False,Scale=10,WorldOffset=(Rotation=(X=0,Y=0,Z=0,W=1),Translation=(X=0,Y=0,Z=0),Scale3D=(X=1,Y=1,Z=1)))
    InputTransform=(Rotation=(X=0,Y=0,Z=0,W=1),Translation=(X=0,Y=0,Z=0),Scale3D=(X=1,Y=1,Z=1))
    Primary=(Weight=1,Axis=(X=1,Y=0,Z=0),Target=(X=1,Y=0,Z=0),Kind=Location,Space=(Type=Bone,Name="None"))
    Secondary=(Weight=1,Axis=(X=0,Y=0,Z=1),Target=(X=1,Y=0,Z=0),Kind=Direction,Space=(Type=Bone,Name="None"))
    Weight=1
    Secondary.Target=RigVMVariableNode() {variable=Backwards Main Pole Vector}.Value
    Primary.Target=RigVMVariableNode() {variable=Current Default Shoulder Position}.Value.Translation
    InputTransform.Translation=RigVMVariableNode() {variable=Scapula Tip Input Transform}.Value.Translation
   )
   n3 = &RigUnit_AimBoneMath(
    DebugSettings=(bEnabled=False,Scale=10,WorldOffset=(Rotation=(X=0,Y=0,Z=0,W=1),Translation=(X=0,Y=0,Z=0),Scale3D=(X=1,Y=1,Z=1)))
    InputTransform=(Rotation=(X=0,Y=0,Z=0,W=1),Translation=(X=0,Y=0,Z=0),Scale3D=(X=1,Y=1,Z=1))
    Primary=(Weight=1,Axis=(X=1,Y=0,Z=0),Target=(X=1,Y=0,Z=0),Kind=Location,Space=(Type=Bone,Name="None"))
    Secondary=(Weight=1,Axis=(X=0,Y=0,Z=1),Target=(X=1,Y=0,Z=0),Kind=Direction,Space=(Type=Bone,Name="None"))
    Weight=1
    Primary.Target=RigVMVariableNode() {variable=Current Shoulder Position}.Value.Translation
    InputTransform.Translation=RigVMVariableNode() {variable=Scapula Tip Input Transform}.Value.Translation
    Secondary.Target=RigVMVariableNode() {variable=Backwards Main Pole Vector}.Value
   )
   RigVMVariableNode(
    Value=&RigVMFunction_MathTransformMakeRelative(Parent=@n2.Result, Global=@n3.Result).Local.Rotation
    ExecuteContext=@RigVMFunction_Sequence_1.D
   ) {variable=Scapula Local Rotation}
   n4 = &RigUnit_GetTransform(bInitial=false, Item=RigVMVariableNode() {variable=Scapula Tip Connector}.Value)
   > &RigVMFunction_DebugLineNoSpace(
    Color=(R=0.0836,G=0,B=1,A=1)
    DebugDrawSettings=(DepthPriority=SDPG_Foreground,Lifetime=-1)
    Thickness=0.25
    WorldOffset=(Rotation=(X=0,Y=0,Z=0,W=1),Translation=(X=0,Y=0,Z=0),Scale3D=(X=1,Y=1,Z=1))
    bEnabled=true
    B=@n4.Transform.Translation
    A=RigVMVariableNode() {variable=Current Default Shoulder Position}.Value.Translation
   )
   > &RigVMFunction_DebugLineNoSpace(
    Color=(R=1,G=0.0188,B=0.4874,A=1)
    DebugDrawSettings=(DepthPriority=SDPG_Foreground,Lifetime=-1)
    Thickness=0.5
    WorldOffset=(Rotation=(X=0,Y=0,Z=0,W=1),Translation=(X=0,Y=0,Z=0),Scale3D=(X=1,Y=1,Z=1))
    bEnabled=true
    A=@n4.Transform.Translation
    B=RigVMVariableNode() {variable=Current Shoulder Position}.Value.Translation
   )
   Get Main Pole Vector = RigVMFunctionReferenceNode(Initial=true)
   n5 = &RigUnit_AimBoneMath(
    DebugSettings=(bEnabled=False,Scale=10,WorldOffset=(Rotation=(X=0,Y=0,Z=0,W=1),Translation=(X=0,Y=0,Z=0),Scale3D=(X=1,Y=1,Z=1)))
    InputTransform=(Rotation=(X=0,Y=0,Z=0,W=1),Translation=(X=0,Y=0,Z=0),Scale3D=(X=1,Y=1,Z=1))
    Primary=(Weight=1,Axis=(X=1,Y=0,Z=0),Target=(X=1,Y=0,Z=0),Kind=Location,Space=(Type=Bone,Name="None"))
    Secondary=(Weight=1,Axis=(X=0,Y=0,Z=1),Target=(X=1,Y=0,Z=0),Kind=Direction,Space=(Type=Bone,Name="None"))
    Weight=1
    Secondary.Target=@"Get Main Pole Vector".Vector
    InputTransform.Translation=&RigUnit_GetTransform(bInitial=true, Item=RigVMVariableNode() {variable=Scapula Tip Connector}.Value).Transform.Translation
    Primary.Target=&RigUnit_GetTransform(bInitial=true, Item=RigVMVariableNode() {variable=Scapula Tip Aim Point}.Value).Transform.Translation
   )
   RigVMVariableNode(Value=@n5.Result, ExecuteContext=@RigVMFunction_Sequence_1.B) {variable=Default Ik Aim Transform}
   n6 = RigVMVariableNode() {variable=Shoulder Connector}
   > RigVMVariableNode(Value=&RigVMFunction_MathTransformMakeRelative(
     Global=&RigUnit_GetTransform(bInitial=true, Item=@n6.Value).Transform
     Parent=RigVMVariableNode() {variable=Default Ik Aim Transform}.Value
    ).Local) {variable=Shoulder Default Local Transform}
   n7 = &RigUnit_AimBoneMath(
    DebugSettings=(bEnabled=False,Scale=10,WorldOffset=(Rotation=(X=0,Y=0,Z=0,W=1),Translation=(X=0,Y=0,Z=0),Scale3D=(X=1,Y=1,Z=1)))
    InputTransform=(Rotation=(X=0,Y=0,Z=0,W=1),Translation=(X=0,Y=0,Z=0),Scale3D=(X=1,Y=1,Z=1))
    Primary=(Weight=1,Axis=(X=1,Y=0,Z=0),Target=(X=1,Y=0,Z=0),Kind=Location,Space=(Type=Bone,Name="None"))
    Secondary=(Weight=1,Axis=(X=0,Y=0,Z=1),Target=(X=1,Y=0,Z=0),Kind=Direction,Space=(Type=Bone,Name="None"))
    Weight=1
    Primary.Target=RigVMVariableNode() {variable=Scapula Aim Point Backwards Transform}.Value.Translation
    InputTransform.Translation=RigVMVariableNode() {variable=Scapula Tip Input Transform}.Value.Translation
    Secondary.Target=RigVMVariableNode() {variable=Backwards Main Pole Vector}.Value
   )
   RigVMVariableNode(Value=@n7.Result, ExecuteContext=@RigVMFunction_Sequence_1.C) {variable=Current Ik Aim Transform}
   > RigVMVariableNode(Value=&RigVMFunction_MathTransformMakeAbsolute(
     Parent=RigVMVariableNode() {variable=Current Ik Aim Transform}.Value
     Local=RigVMVariableNode() {variable=Shoulder Default Local Transform}.Value
    ) {template="Make Absolute::Execute(in Local,in Parent,out Global)"}.Global) {variable=Current Default Shoulder Position}
   > RigVMVariableNode(Value=&RigUnit_GetTransform(bInitial=false, Item=RigVMVariableNode() {variable=Shoulder Connector}.Value).Transform) {variable=Current Shoulder Position}
   Set Transform_2 = &RigUnit_SetTranslation(
    Space=GlobalSpace
    Weight=1
    bInitial=False
    bPropagateToChildren=True
    Value=&RigUnit_GetTransform(bInitial=False, Item=RigVMVariableNode() {variable=Scapula Tip Connector}.Value).Transform.Translation
    Item=RigVMVariableNode() {variable=Scapula Tip Control}.Value
    ExecutePin=@RigVMFunction_Sequence_1.A
   ) {template="Set Transform::Execute(in Item,in Space,in bInitial,in Value,in Weight,in bPropagateToChildren)"}
   Get Main Pole Vector_1 = RigVMFunctionReferenceNode(Initial=false)
   > RigVMVariableNode(Value=@"Get Main Pole Vector_1".Vector) {variable=Backwards Main Pole Vector}
   RigVMFunction_MathQuaternionToEuler = &RigVMFunction_MathQuaternionToEuler(RotationOrder=ZYX, Value=RigVMVariableNode() {variable=Scapula Local Rotation}.Value)
   Set Transform_1 = &RigUnit_SetRotation(
    Weight=1
    bInitial=False
    Value=&RigVMFunction_MathQuaternionFromEuler(Euler=(X=0,Y=0,Z=0), Euler.X=@RigVMFunction_MathQuaternionToEuler.Result.Z).Result
    Item=RigVMVariableNode() {variable=Scapula Tip Control}.Value
    ExecutePin=@RigVMFunction_Sequence_1.E
   )
   RigVMFunctionReturnNode(ExecuteContext=@RigVMFunction_Sequence.B)
   n8 = RigVMRerouteNode(Value=())
   Set Transform_5 = &RigUnit_SetTranslation(
    Space=GlobalSpace
    Weight=1
    bInitial=False
    bPropagateToChildren=True
    Value=&RigUnit_GetTransform(bInitial=False, Item=RigVMVariableNode() {variable=Scapula Tip Connector}.Value).Transform.Translation
    ExecutePin=@n8.Value
    Item=RigVMVariableNode() {variable=Scapula Tip Control}.Value
   ) {template="Set Transform::Execute(in Item,in Space,in bInitial,in Value,in Weight,in bPropagateToChildren)"}
   n9 = RigVMRerouteNode(Value=())
   n10 = &RigUnit_GetTransform(bInitial=false, Item=RigVMVariableNode() {variable=Scapula Tip Connector}.Value)
   Subtract = &RigVMFunction_MathVectorSub(B=@n10.Transform.Translation, A=RigVMVariableNode() {variable=Current Default Shoulder Position}.Value.Translation) {template="Subtract::Execute(in A,in B,out Result)"}
   RigVMFunction_MathVectorAngle = /Script/RigVM.RigVMFunction_MathVectorAngle(
    A=@Subtract.Result
    B=&RigVMFunction_MathVectorSub(B=@n10.Transform.Translation, A=RigVMVariableNode() {variable=Current Shoulder Position}.Value.Translation) {template="Subtract::Execute(in A,in B,out Result)"}.Result
   ) {is=struct RigVMUnitNode, template="AngleBetween::Execute(in A,in B,out Result)"}
   Degrees = &RigVMFunction_MathDoubleDeg(Value=@RigVMFunction_MathVectorAngle.Result) {template="Degrees::Execute(in Value,out Result)"}
   RigVMVariableNode(Value=@Degrees.Result, ExecuteContext=@n9.Value) {variable=Angle Between Result}
   n11 = RigVMRerouteNode(Value=())
   n12 = RigVMVariableNode() {variable=Shoulder Connector}
   n13 = RigVMVariableNode() {variable=Scapula Tip Connector}
   n14 = RigVMVariableNode() {variable=Scapula Tip Connector}
   Dot__ = /Script/RigVM.RigVMFunction_MathVectorDot(
    A=&RigVMFunction_MathVectorSub(
     A=&RigUnit_GetTransform(bInitial=False, Item=@n12.Value).Transform.Translation
     B=&RigUnit_GetTransform(bInitial=false, Item=@n13.Value).Transform.Translation
    ) {template="Subtract::Execute(in A,in B,out Result)"}.Result
    B=&RigVMFunction_MathVectorSub(
     B=&RigUnit_GetTransform(bInitial=false, Item=@n14.Value).Transform.Translation
     A=RigVMVariableNode() {variable=Current Default Shoulder Position}.Value.Translation
    ) {template="Subtract::Execute(in A,in B,out Result)"}.Result
   ) {is=struct RigVMUnitNode, template="Dot,|::Execute(in A,in B,out Result)"}
   Greater = /Script/RigVM.RigVMFunction_MathDoubleGreater(B=0, A=@Dot__.Result) {is=struct RigVMUnitNode, template="Greater::Execute(in A,in B,out Result)"}
   RigVMVariableNode(
    Value=RigVMDispatchNode(False=1, True=-1, Condition=@Greater.Result) {template="DISPATCH_RigVMDispatch_If(in Condition,in True,in False,out Result)"}.Result
    ExecuteContext=@n11.Value
   ) {variable=Rotation Direction}
   /Script/RigVM.RigVMFunction_DebugTransformMutableNoSpace(
    Color=(R=1,G=0,B=0,A=1)
    DebugDrawSettings=(DepthPriority=SDPG_Foreground,Lifetime=-1)
    ExecutePin=()
    Mode=Axes
    Scale=10
    Thickness=0.5
    WorldOffset=(Rotation=(X=0,Y=0,Z=0,W=1),Translation=(X=0,Y=0,Z=0),Scale3D=(X=1,Y=1,Z=1))
    bEnabled=false
    Transform=&RigUnit_GetTransform(bInitial=False, Item=RigVMVariableNode() {variable=Scapula Tip Control}.Value).Transform
   ) {is=struct RigVMUnitNode}
   n15 = RigVMVariableNode() {variable=Current Ik Aim Transform}
   > RigVMFunction_DebugLineNoSpace = &RigVMFunction_DebugLineNoSpace(
    Color=(R=0,G=1,B=0,A=1)
    DebugDrawSettings=(DepthPriority=SDPG_Foreground,Lifetime=-1)
    Thickness=0.5
    WorldOffset=(Rotation=(X=0,Y=0,Z=0,W=1),Translation=(X=0,Y=0,Z=0),Scale3D=(X=1,Y=1,Z=1))
    bEnabled=false
    A=&RigVMFunction_MathTransformMakeAbsolute(Parent=@n15.Value, Local=RigVMVariableNode() {variable=Shoulder Default Local Transform}.Value) {template="Make Absolute::Execute(in Local,in Parent,out Global)"}.Global.Translation
    B=@n15.Value.Translation
   )
   &RigVMFunction_MathTransformMakeAbsolute(
    Parent=(Rotation=(X=0,Y=0,Z=0,W=1),Translation=(X=0,Y=0,Z=0),Scale3D=(X=1,Y=1,Z=1))
    Local=RigVMVariableNode() {variable=Scapula Tip Control Local Transform}.Value
   ) {template="Make Absolute::Execute(in Local,in Parent,out Global)"}
   RigVMFunction_MathQuaternionFromEuler = &RigVMFunction_MathQuaternionFromEuler(Euler=(X=0,Y=0,Z=0), Euler.Z=RigVMVariableNode() {variable=Angle Between Result}.Value)
   &RigVMFunction_MathTransformMakeAbsolute(
    Local=(Rotation=(X=0,Y=0,Z=0,W=1),Translation=(X=0,Y=0,Z=0),Scale3D=(X=1,Y=1,Z=1))
    Local.Rotation=@RigVMFunction_MathQuaternionFromEuler.Result
    Parent=RigVMVariableNode() {variable=Current Ik Aim Transform}.Value
   ) {template="Make Absolute::Execute(in Local,in Parent,out Global)"}
   n16 = RigVMVariableNode() {variable=Ik Control}
   &RigVMFunction_MathTransformMakeRelative(
    Global=(Rotation=(X=0,Y=0,Z=0,W=1),Translation=(X=0,Y=0,Z=0),Scale3D=(X=1,Y=1,Z=1))
    Parent=&RigUnit_GetTransform(bInitial=False, Item=&RigUnit_HierarchyGetParent(bDefaultParent=True, Child=@n16.Value).Parent).Transform
   )
   &RigVMFunction_MathDoubleMul(A=RigVMVariableNode() {variable=Angle Between Result}.Value, B=RigVMVariableNode() {variable=Rotation Direction}.Value)
   RigVMFunction_MathVectorCross = /Script/RigVM.RigVMFunction_MathVectorCross(A=(X=0,Y=0,Z=0), B=(X=0,Y=0,Z=0)) {is=struct RigVMUnitNode, template="Cross::Execute(in A,in B,out Result)"}
  ]}
 New Function_ContainedGraph {nodes [
   n1 = RigVMFunctionEntryNode()
   RigVMFunction_Sequence = &RigVMFunction_Sequence(ExecuteContext=@n1.ExecuteContext)
   RigVMFunction_Sequence_1 = RigVMAggregateNode(ExecuteContext=@RigVMFunction_Sequence.A)
   &RigUnit_SetTransform(
    Weight=1
    bInitial=False
    Item=RigVMVariableNode() {variable=Ik Control}.Value
    Value=RigVMVariableNode() {variable=Ik Transform}.Value
    ExecutePin=@RigVMFunction_Sequence_1.D
   )
   n2 = &RigUnit_GetTransform(
    bInitial=False
    Item=&RigUnit_HierarchyGetParent(bDefaultParent=True, Child=RigVMVariableNode() {variable=Ik Control}.Value).Parent
   )
   RigVMVariableNode(
    Value=&RigVMFunction_MathVectorSub(
     A=&RigVMFunction_MathTransformMakeAbsolute(Local=(Rotation=(X=0,Y=0,Z=0,W=1),Translation=(X=0,Y=0,Z=10),Scale3D=(X=1,Y=1,Z=1)), Parent=@n2.Transform) {template="Make Absolute::Execute(in Local,in Parent,out Global)"}.Global.Translation
     B=@n2.Transform.Translation
    ) {template="Subtract::Execute(in A,in B,out Result)"}.Result
    ExecuteContext=@RigVMFunction_Sequence_1.A
   ) {variable=Ik Aim Vector}
   n3 = &RigUnit_GetTransform(bInitial=False, Item=RigVMVariableNode() {variable=Ball Connector}.Value)
   Subtract = &RigVMFunction_MathVectorSub(
    A=&RigVMFunction_MathTransformMakeAbsolute(Local=(Rotation=(X=0,Y=0,Z=0,W=1),Translation=(X=0,Y=0,Z=10),Scale3D=(X=1,Y=1,Z=1)), Parent=@n3.Transform) {template="Make Absolute::Execute(in Local,in Parent,out Global)"}.Global.Translation
    B=@n3.Transform.Translation
   ) {template="Subtract::Execute(in A,in B,out Result)"}
   RigVMVariableNode(Value=@Subtract.Result, ExecuteContext=@RigVMFunction_Sequence_1.B) {variable=Ik Up Vector}
   n4 = &RigUnit_GetTransform(bInitial=False, Item=RigVMVariableNode() {variable=Ball Connector}.Value)
   Add = &RigVMFunction_MathVectorAdd(A=@n4.Transform.Translation, B=RigVMVariableNode() {variable=Ik Aim Vector}.Value) {template="Add::Execute(in A,in B,out Result)"}
   n5 = &RigUnit_AimBoneMath(
    DebugSettings=(bEnabled=False,Scale=10,WorldOffset=(Rotation=(X=0,Y=0,Z=0,W=1),Translation=(X=0,Y=0,Z=0),Scale3D=(X=1,Y=1,Z=1)))
    InputTransform=(Rotation=(X=0,Y=0,Z=0,W=1),Translation=(X=0,Y=0,Z=0),Scale3D=(X=1,Y=1,Z=1))
    Primary=(Weight=1,Axis=(X=0,Y=0,Z=1),Target=(X=1,Y=0,Z=0),Kind=Location,Space=(Type=Bone,Name="None"))
    Secondary=(Weight=1,Axis=(X=1,Y=0,Z=0),Target=(X=1,Y=0,Z=0),Kind=Direction,Space=(Type=Bone,Name="None"))
    Weight=1
    Primary.Target=@Add.Result
    InputTransform.Translation=@n4.Transform.Translation
    Secondary.Target=RigVMVariableNode() {variable=Ik Up Vector}.Value
   )
   RigVMVariableNode(Value=@n5.Result, ExecuteContext=@RigVMFunction_Sequence_1.C) {variable=Ik Transform}
   n6 = &RigUnit_HierarchyGetParent(bDefaultParent=True, Child=RigVMVariableNode() {variable=Ik Control}.Value)
   RigVMFunction_MathVectorFromDouble = &RigVMFunction_MathVectorFromDouble(Value=RigVMVariableNode() {variable=Scapula Tip Aim Percent}.Value) {template="Cast::Execute(in Value,out Result)"}
   Multiply = &RigVMFunction_MathVectorMul(
    A=&RigUnit_GetTransform(Space=LocalSpace, bInitial=False, Item=RigVMVariableNode() {variable=Ik Control}.Value).Transform.Translation
    B=@RigVMFunction_MathVectorFromDouble.Result
   ) {template="Multiply::Execute(in A,in B,out Result)"}
   RigVMVariableNode(
    Value=&RigVMFunction_MathTransformMakeAbsolute(
     Local=(Rotation=(X=0,Y=0,Z=0,W=1),Translation=(X=0,Y=0,Z=0),Scale3D=(X=1,Y=1,Z=1))
     Parent=&RigUnit_GetTransform(bInitial=False, Item=@n6.Parent).Transform
     Local.Translation=@Multiply.Result
    ) {template="Make Absolute::Execute(in Local,in Parent,out Global)"}.Global
    ExecuteContext=@RigVMFunction_Sequence_1.E
   ) {variable=Scapula Aim Point Backwards Transform}
   RigVMFunctionReturnNode(ExecuteContext=@RigVMFunction_Sequence.B)
  ]}
 New Function_ContainedGraph {nodes [
   n1 = RigVMFunctionEntryNode()
   RigVMFunction_Sequence = &RigVMFunction_Sequence(ExecuteContext=@n1.ExecuteContext)
   n2 = &RigVMFunction_Sequence(ExecuteContext=@RigVMFunction_Sequence.A)
   n3 = &RigUnit_AimBoneMath(
    DebugSettings=(bEnabled=False,Scale=10,WorldOffset=(Rotation=(X=0,Y=0,Z=0,W=1),Translation=(X=0,Y=0,Z=0),Scale3D=(X=1,Y=1,Z=1)))
    InputTransform=(Rotation=(X=0,Y=0,Z=0,W=1),Translation=(X=0,Y=0,Z=0),Scale3D=(X=1,Y=1,Z=1))
    Primary=(Weight=1,Axis=(X=0,Y=1,Z=0),Target=(X=1,Y=0,Z=0),Kind=Location,Space=(Type=Bone,Name="None"))
    Secondary=(Weight=1,Axis=(X=1,Y=0,Z=0),Target=(X=1,Y=0,Z=0),Kind=Direction,Space=(Type=Bone,Name="None"))
    Weight=1
    InputTransform.Translation=&RigUnit_GetTransform(bInitial=False, Item=RigVMVariableNode() {variable=Ball Connector}.Value).Transform.Translation
    Primary.Target=&RigUnit_GetTransform(bInitial=False, Item=RigVMVariableNode() {variable=Toe Connector}.Value).Transform.Translation
    Secondary.Space=RigVMVariableNode() {variable=Ik Control}.Value
   )
   &RigUnit_SetTransform(Weight=1, bInitial=False, Value=@n3.Result, Item=RigVMVariableNode() {variable=Ball Fk Control}.Value, ExecutePin=@n2.A)
   RigVMFunctionReturnNode(ExecuteContext=@RigVMFunction_Sequence.B)
  ]}
 New Function_ContainedGraph {nodes [
   n1 = RigVMFunctionEntryNode()
   RigVMFunction_Sequence = &RigVMFunction_Sequence(ExecuteContext=@n1.ExecuteContext)
   n2 = &RigVMFunction_Sequence(ExecuteContext=@RigVMFunction_Sequence.A)
   n3 = &RigUnit_AimBoneMath(
    DebugSettings=(bEnabled=False,Scale=10,WorldOffset=(Rotation=(X=0,Y=0,Z=0,W=1),Translation=(X=0,Y=0,Z=0),Scale3D=(X=1,Y=1,Z=1)))
    InputTransform=(Rotation=(X=0,Y=0,Z=0,W=1),Translation=(X=0,Y=0,Z=0),Scale3D=(X=1,Y=1,Z=1))
    Primary=(Weight=1,Axis=(X=0,Y=1,Z=0),Target=(X=1,Y=0,Z=0),Kind=Location,Space=(Type=Bone,Name="None"))
    Secondary=(Weight=1,Axis=(X=1,Y=0,Z=0),Target=(X=1,Y=0,Z=0),Kind=Direction,Space=(Type=Bone,Name="None"))
    Weight=1
    Primary.Target=&RigUnit_GetTransform(bInitial=False, Item=RigVMVariableNode() {variable=Toe Tip Connector}.Value).Transform.Translation
    InputTransform.Translation=&RigUnit_GetTransform(bInitial=False, Item=RigVMVariableNode() {variable=Toe Connector}.Value).Transform.Translation
    Secondary.Space=RigVMVariableNode() {variable=Ik Control}.Value
   )
   &RigUnit_SetTransform(Weight=1, bInitial=False, Value=@n3.Result, Item=RigVMVariableNode() {variable=Toe Fk Control}.Value, ExecutePin=@n2.A)
   RigVMFunctionReturnNode(ExecuteContext=@RigVMFunction_Sequence.B)
  ]}
 New Function_ContainedGraph {nodes [
   n1 = RigVMFunctionEntryNode()
   RigVMFunction_Sequence = &RigVMFunction_Sequence(ExecuteContext=@n1.ExecuteContext)
   RigVMFunction_Sequence_1 = RigVMAggregateNode(ExecuteContext=@RigVMFunction_Sequence.A)
   Add = &RigVMFunction_MathVectorAdd(
    B=&RigUnit_GetTransform(bInitial=False, Item=RigVMVariableNode() {variable=Scapula Tip Connector}.Value).Transform.Translation
    A=RigVMVariableNode() {variable=Vector}.Value
   ) {template="Add::Execute(in A,in B,out Result)"}
   Set Transform = &RigUnit_SetTranslation(
    Space=GlobalSpace
    Weight=1
    bInitial=False
    bPropagateToChildren=True
    Value=@Add.Result
    Item=RigVMVariableNode() {variable=Main PV Control}.Value
    ExecutePin=@RigVMFunction_Sequence_1.D
   ) {template="Set Transform::Execute(in Item,in Space,in bInitial,in Value,in Weight,in bPropagateToChildren)"}
   n2 = RigVMVariableNode() {variable=Shoulder Connector}
   n3 = RigVMVariableNode() {variable=Main PV Control}
   RigVMVariableNode(
    Value=&RigVMFunction_MathTransformMakeRelative(
     Parent=&RigUnit_GetTransform(bInitial=true, Item=@n2.Value).Transform
     Global=&RigUnit_GetTransform(bInitial=true, Item=@n3.Value).Transform
    ).Local
    ExecuteContext=@RigVMFunction_Sequence_1.A
   ) {variable=PV Initial Local Transform}
   n4 = RigVMVariableNode() {variable=Shoulder Connector}
   n5 = &RigVMFunction_MathVectorSub(
    B=&RigUnit_GetTransform(bInitial=False, Item=RigVMVariableNode() {variable=Scapula Tip Connector}.Value).Transform.Translation
    A=&RigVMFunction_MathTransformMakeAbsolute(
     Parent=&RigUnit_GetTransform(bInitial=False, Item=@n4.Value).Transform
     Local=RigVMVariableNode() {variable=PV Initial Local Transform}.Value
    ) {template="Make Absolute::Execute(in Local,in Parent,out Global)"}.Global.Translation
   ) {template="Subtract::Execute(in A,in B,out Result)"}
   RigVMVariableNode(
    Value=RigVMDispatchNode(
     True=&RigVMFunction_MathVectorMul(B=(X=-1,Y=-1,Z=-1), A=@n5.Result) {template="Multiply::Execute(in A,in B,out Result)"}.Result
     Condition=RigVMVariableNode() {variable=Right Side}.Value
     False=@n5.Result
    ) {template="DISPATCH_RigVMDispatch_If(in Condition,in True,in False,out Result)"}.Result
    ExecuteContext=@RigVMFunction_Sequence_1.B
   ) {variable=Vector}
   RigVMFunctionReturnNode(ExecuteContext=@RigVMFunction_Sequence.B)
   n6 = RigVMRerouteNode(Value=())
   RigVMFunction_MathVectorLength = /Script/RigVM.RigVMFunction_MathVectorLength(Value=RigVMVariableNode() {variable=Rig Up Vector}.Value) {is=struct RigVMUnitNode, template="Length::Execute(in Value,out Result)"}
   RigVMFunction_MathVectorSetLength = /Script/RigVM.RigVMFunction_MathVectorSetLength(Length=@RigVMFunction_MathVectorLength.Result, Value=RigVMVariableNode() {variable=Vector}.Value) {is=struct RigVMUnitNode, template="SetLength::Execute(in Value,in Length,out Result)"}
   RigVMVariableNode(ExecuteContext=@n6.Value, Value=@RigVMFunction_MathVectorSetLength.Result) {variable=Vector}
   n7 = RigVMRerouteNode(Value=())
   n8 = &RigUnit_GetTransform(bInitial=False, Item=RigVMVariableNode() {variable=Shoulder Connector}.Value)
   Subtract = &RigVMFunction_MathVectorSub(
    B=&RigVMFunction_MathTransformMakeAbsolute(Local=(Rotation=(X=0,Y=0,Z=0,W=1),Translation=(X=0,Y=0,Z=1),Scale3D=(X=1,Y=1,Z=1)), Parent=@n8.Transform) {template="Make Absolute::Execute(in Local,in Parent,out Global)"}.Global.Translation
    A=@n8.Transform.Translation
   ) {template="Subtract::Execute(in A,in B,out Result)"}
   Multiply = &RigVMFunction_MathVectorMul(B=(X=-1,Y=-1,Z=-1), A=@Subtract.Result) {template="Multiply::Execute(in A,in B,out Result)"}
   RigVMVariableNode(
    Value=RigVMDispatchNode(True=@Multiply.Result, Condition=RigVMVariableNode() {variable=Right Side}.Value, False=@Subtract.Result) {template="DISPATCH_RigVMDispatch_If(in Condition,in True,in False,out Result)"}.Result
    ExecuteContext=@n7.Value
   ) {variable=Vector}
  ]}
 New Function_ContainedGraph {nodes [
   n1 = RigVMFunctionEntryNode()
   RigVMFunction_Sequence = &RigVMFunction_Sequence(ExecuteContext=@n1.ExecuteContext)
   RigVMFunction_Sequence_1 = RigVMAggregateNode(ExecuteContext=@RigVMFunction_Sequence.A)
   Get Main Pole Vector = RigVMFunctionReferenceNode(Initial=False)
   n2 = &RigUnit_GetTransform(bInitial=true, Item=RigVMVariableNode() {variable=Shoulder Connector}.Value)
   n3 = &RigUnit_GetTransform(bInitial=true, Item=RigVMVariableNode() {variable=Wrist Connector}.Value)
   n4 = &RigUnit_AimBoneMath(
    DebugSettings=(bEnabled=False,Scale=10,WorldOffset=(Rotation=(X=0,Y=0,Z=0,W=1),Translation=(X=0,Y=0,Z=0),Scale3D=(X=1,Y=1,Z=1)))
    InputTransform=(Rotation=(X=0,Y=0,Z=0,W=1),Translation=(X=0,Y=0,Z=0),Scale3D=(X=1,Y=1,Z=1))
    Primary=(Weight=1,Axis=(X=0,Y=0,Z=1),Target=(X=1,Y=0,Z=0),Kind=Location,Space=(Type=Bone,Name="None"))
    Secondary=(Weight=1,Axis=(X=1,Y=0,Z=0),Target=(X=1,Y=0,Z=0),Kind=Direction,Space=(Type=Bone,Name="None"))
    Weight=1
    Secondary.Target=@"Get Main Pole Vector".Vector
    InputTransform.Translation=@n2.Transform.Translation
    Primary.Target=@n3.Transform.Translation
   )
   RigVMFunction_MathVectorDistance = &RigVMFunction_MathVectorDistance(A=@n2.Transform.Translation, B=@n3.Transform.Translation)
   Multiply = &RigVMFunction_MathDoubleMul(B=0.5, A=@RigVMFunction_MathVectorDistance.Result)
   RigVMVariableNode(
    Value=&RigVMFunction_MathTransformMakeAbsolute(Local=(Rotation=(X=0,Y=0,Z=0,W=1),Translation=(X=0,Y=0,Z=0),Scale3D=(X=1,Y=1,Z=1)), Parent=@n4.Result, Local.Translation.Z=@Multiply.Result) {template="Make Absolute::Execute(in Local,in Parent,out Global)"}.Global
    ExecuteContext=@RigVMFunction_Sequence_1.A
   ) {variable=Control Null Transform}
   Create Name = RigVMFunctionReferenceNode(Tokens=("Elbow Rotate Control Aim"), ExecuteContext=@RigVMFunction_Sequence_1.B)
   > n5 = &RigUnit_HierarchyAddNull(
    Space=GlobalSpace
    Name=@"Create Name".Name
    Parent=&RigUnit_HierarchyGetParent(bDefaultParent=True, Child=RigVMVariableNode() {variable=Shoulder Connector}.Value).Parent
    Transform=RigVMVariableNode() {variable=Control Null Transform}.Value
   )
   > Create Name_3 = RigVMFunctionReferenceNode(Tokens=("Elbow Rotate Control Null"))
   > n6 = &RigUnit_HierarchyAddNull(
    Space=LocalSpace
    Transform=(Rotation=(X=0,Y=0,Z=0,W=1),Translation=(X=0,Y=0,Z=0),Scale3D=(X=1,Y=1,Z=1))
    Name=@"Create Name_3".Name
    Parent=@n5.Item
   )
   > RigVMVariableNode(Value=@n6.Item) {variable=Elbow Rotate Control Null}
   > Create Name_1 = RigVMFunctionReferenceNode(Tokens=("Elbow Rotate"))
   n7 = &RigVMFunction_MathDoubleMul(B=0, A=RigVMVariableNode() {variable=Control Scale}.Value)
   > SpawnControl = &RigUnit_HierarchyAddControlTransform(
    InitialValue=(Rotation=(X=0,Y=0,Z=0,W=1),Translation=(X=0,Y=0,Z=0),Scale3D=(X=1,Y=1,Z=1))
    OffsetSpace=LocalSpace
    OffsetTransform=(Rotation=(X=0,Y=0,Z=0,W=1),Translation=(X=0,Y=0,Z=0),Scale3D=(X=1,Y=1,Z=1))
    Settings="(InitialSpace=LocalSpace,bUsePreferredRotationOrder=False,PreferredRotationOrder=YZX,Limits=(LimitTranslationX=(bMinimum=True,bMaximum=True),LimitTranslationY=(bMinimum=True,bMaximum=True),LimitTranslationZ=(bMinimum=True,bMaximum=True),LimitPitch=(bMinimu ...(truncated)"
    Name=@"Create Name_1".Name
    Settings.Shape.Transform.Scale3D.X=@n7.Result
    Settings.Shape.Transform.Scale3D.Y=@n7.Result
    Settings.Shape.Transform.Scale3D.Z=@n7.Result
    Parent=@n6.Item
   ) {template="SpawnControl::Execute(in InitialValue,in Settings,in OffsetTransform,in OffsetSpace,in Parent,in Name,out Item)"}
   > RigVMVariableNode(Value=@SpawnControl.Item) {variable=Elbow Rotate Control}
   Create Name_2 = RigVMFunctionReferenceNode(Tokens=("Elbow Rotate Up Vector"), ExecuteContext=@RigVMFunction_Sequence_1.C)
   > n8 = &RigUnit_HierarchyAddNull(
    Space=LocalSpace
    Transform=(Rotation=(X=0,Y=0,Z=0,W=1),Translation=(X=10,Y=0,Z=0),Scale3D=(X=1,Y=1,Z=1))
    Name=@"Create Name_2".Name
    Parent=RigVMVariableNode() {variable=Elbow Rotate Control}.Value
   )
   > RigVMVariableNode(Value=@n8.Item) {variable=Elbow Rotate Up Vector}
   Create Name_4 = RigVMFunctionReferenceNode(Tokens=("Elbow PV Null"), ExecuteContext=@RigVMFunction_Sequence_1.D)
   > n9 = &RigUnit_HierarchyAddNull(
    Space=LocalSpace
    Transform=(Rotation=(X=1,Y=0,Z=0,W=0),Translation=(X=0,Y=20,Z=0),Scale3D=(X=1,Y=1,Z=1))
    Name=@"Create Name_4".Name
    Parent=&RigUnit_HierarchyGetParent(bDefaultParent=True, Child=RigVMVariableNode() {variable=Elbow Rotate Control Null}.Value).Parent
   )
   > Create Name_5 = RigVMFunctionReferenceNode(Tokens=("Elbow PV"))
   > Set Side Color = RigVMFunctionReferenceNode()
   > n10 = &RigUnit_HierarchyAddControlTransform(
    InitialValue=(Rotation=(X=0,Y=0,Z=0,W=1),Translation=(X=0,Y=0,Z=0),Scale3D=(X=1,Y=1,Z=1))
    OffsetSpace=LocalSpace
    OffsetTransform=(Rotation=(X=0,Y=0,Z=0,W=1),Translation=(X=0,Y=0,Z=0),Scale3D=(X=1,Y=1,Z=1))
    Settings="(InitialSpace=LocalSpace,bUsePreferredRotationOrder=False,PreferredRotationOrder=YZX,Limits=(LimitTranslationX=(bMinimum=False,bMaximum=False),LimitTranslationY=(bMinimum=False,bMaximum=False),LimitTranslationZ=(bMinimum=False,bMaximum=False),LimitPitch=(b ...(truncated)"
    Name=@"Create Name_5".Name
    Settings.Shape.Color=@"Set Side Color".Color
    Parent=@n9.Item
   ) {template="SpawnControl::Execute(in InitialValue,in Settings,in OffsetTransform,in OffsetSpace,in Parent,in Name,out Item)"}
   > RigVMVariableNode(Value=@n10.Item) {variable=Elbow PV}
   RigVMFunctionReturnNode(ExecuteContext=@RigVMFunction_Sequence.B)
  ]}
 New Function_ContainedGraph {nodes [
   n1 = RigVMFunctionEntryNode()
   RigVMFunction_Sequence = &RigVMFunction_Sequence(ExecuteContext=@n1.ExecuteContext)
   RigVMFunction_Sequence_1 = RigVMAggregateNode(ExecuteContext=@RigVMFunction_Sequence.A)
   Get Main Pole Vector = RigVMFunctionReferenceNode(Initial=False)
   n2 = RigVMVariableNode() {variable=Shoulder Rig Transform}
   n3 = RigVMVariableNode() {variable=Wrist Rig Transform}
   n4 = &RigUnit_AimBoneMath(
    DebugSettings=(bEnabled=False,Scale=10,WorldOffset=(Rotation=(X=0,Y=0,Z=0,W=1),Translation=(X=0,Y=0,Z=0),Scale3D=(X=1,Y=1,Z=1)))
    InputTransform=(Rotation=(X=0,Y=0,Z=0,W=1),Translation=(X=0,Y=0,Z=0),Scale3D=(X=1,Y=1,Z=1))
    Primary=(Weight=1,Axis=(X=0,Y=0,Z=1),Target=(X=1,Y=0,Z=0),Kind=Location,Space=(Type=Bone,Name="None"))
    Secondary=(Weight=1,Axis=(X=1,Y=0,Z=0),Target=(X=1,Y=0,Z=0),Kind=Direction,Space=(Type=Bone,Name="None"))
    Weight=1
    Secondary.Target=@"Get Main Pole Vector".Vector
    InputTransform.Translation=@n2.Value.Translation
    Primary.Target=@n3.Value.Translation
   )
   Multiply = &RigVMFunction_MathDoubleMul(B=0.5, A=&RigVMFunction_MathVectorDistance(A=@n2.Value.Translation, B=@n3.Value.Translation).Result)
   &RigUnit_SetTransform(
    Weight=1
    bInitial=False
    Item=&RigUnit_HierarchyGetParent(bDefaultParent=True, Child=RigVMVariableNode() {variable=Elbow Rotate Control Null}.Value).Parent
    Value=&RigVMFunction_MathTransformMakeAbsolute(Local=(Rotation=(X=0,Y=0,Z=0,W=1),Translation=(X=0,Y=0,Z=0),Scale3D=(X=1,Y=1,Z=1)), Parent=@n4.Result, Local.Translation.Z=@Multiply.Result) {template="Make Absolute::Execute(in Local,in Parent,out Global)"}.Global
    ExecutePin=@RigVMFunction_Sequence_1.A
   )
   RigVMFunction_MathVectorDistance = &RigVMFunction_MathVectorDistance(
    A=RigVMVariableNode() {variable=Shoulder Rig Transform}.Value.Translation
    B=RigVMVariableNode() {variable=Wrist Rig Transform}.Value.Translation
   )
   Get Ik Angles = RigVMFunctionReferenceNode(
    C 180=true
    Length C=@RigVMFunction_MathVectorDistance.Result
    Length A=RigVMVariableNode() {variable=Shoulder Length}.Value
    Length B=RigVMVariableNode() {variable=Elbow Length}.Value
   )
   RigVMVariableNode(Value=@"Get Ik Angles".Angle B, ExecuteContext=@RigVMFunction_Sequence_1.B) {variable=Shoulder Angle}
   > RigVMVariableNode(Value=RigVMDispatchNode(
     True=0
     Condition=RigVMDispatchNode(B=180, A=@"Get Ik Angles".Angle C) {template="DISPATCH_RigVMDispatch_CoreEquals(in A,in B,out Result)"}.Result
     False=@"Get Ik Angles".Angle C
    ) {template="DISPATCH_RigVMDispatch_If(in Condition,in True,in False,out Result)"}.Result) {variable=Elbow Angle}
   n5 = &RigUnit_AimBoneMath(
    DebugSettings=(bEnabled=False,Scale=10,WorldOffset=(Rotation=(X=0,Y=0,Z=0,W=1),Translation=(X=0,Y=0,Z=0),Scale3D=(X=1,Y=1,Z=1)))
    InputTransform=(Rotation=(X=0,Y=0,Z=0,W=1),Translation=(X=0,Y=0,Z=0),Scale3D=(X=1,Y=1,Z=1))
    Primary=(Weight=1,Axis=(X=1,Y=0,Z=0),Target=(X=1,Y=0,Z=0),Kind=Location,Space=(Type=Bone,Name="None"))
    Secondary=(Weight=1,Axis=(X=0,Y=0,Z=-1),Target=(X=1,Y=0,Z=0),Kind=Direction,Space=(Type=Bone,Name="None"))
    Weight=1
    InputTransform.Translation=RigVMVariableNode() {variable=Shoulder Rig Transform}.Value.Translation
    Primary.Target=RigVMVariableNode() {variable=Wrist Rig Transform}.Value.Translation
    Secondary.Target=RigVMVariableNode() {variable=Elbow Ik PV}.Value
   )
   n6 = RigVMVariableNode() {variable=Shoulder Angle}
   RigVMVariableNode(
    Value=&RigVMFunction_MathTransformMakeAbsolute(
     Local=(Rotation=(X=0,Y=0,Z=0,W=1),Translation=(X=0,Y=0,Z=0),Scale3D=(X=1,Y=1,Z=1))
     Parent=@n5.Result
     Local.Rotation=&RigVMFunction_MathQuaternionFromEuler(Euler=(X=0,Y=0,Z=0), Euler.Z=@n6.Value).Result
    ) {template="Make Absolute::Execute(in Local,in Parent,out Global)"}.Global
    ExecuteContext=@RigVMFunction_Sequence_1.D
   ) {variable=Temp Shoulder Transform}
   n7 = RigVMVariableNode() {variable=Temp Shoulder Transform}
   n8 = RigVMVariableNode() {variable=Shoulder Length}
   n9 = RigVMVariableNode() {variable=Elbow Angle}
   > RigVMVariableNode(Value=&RigVMFunction_MathTransformMakeAbsolute(
     Local=(Rotation=(X=0,Y=0,Z=0,W=1),Translation=(X=0,Y=0,Z=0),Scale3D=(X=1,Y=1,Z=1))
     Parent=&RigVMFunction_MathTransformMakeAbsolute(Local=(Rotation=(X=0,Y=0,Z=0,W=1),Translation=(X=0,Y=0,Z=0),Scale3D=(X=1,Y=1,Z=1)), Parent=@n7.Value, Local.Translation.X=@n8.Value) {template="Make Absolute::Execute(in Local,in Parent,out Global)"}.Global
     Local.Rotation=&RigVMFunction_MathQuaternionFromEuler(Euler=(X=0,Y=0,Z=0), Euler.Z=@n9.Value).Result
    ) {template="Make Absolute::Execute(in Local,in Parent,out Global)"}.Global) {variable=Temp Elbow Transform}
   > RigVMVariableNode(Value=&RigVMFunction_MathTransformMakeAbsolute(
     Local=(Rotation=(X=0,Y=0,Z=0,W=1),Translation=(X=0,Y=0,Z=0),Scale3D=(X=1,Y=1,Z=1))
     Parent=RigVMVariableNode() {variable=Temp Elbow Transform}.Value
     Local.Translation.X=RigVMVariableNode() {variable=Elbow Length}.Value
    ) {template="Make Absolute::Execute(in Local,in Parent,out Global)"}.Global) {variable=Temp Wrist Transform}
   n10 = &RigUnit_AimBoneMath(
    DebugSettings=(bEnabled=False,Scale=10,WorldOffset=(Rotation=(X=0,Y=0,Z=0,W=1),Translation=(X=0,Y=0,Z=0),Scale3D=(X=1,Y=1,Z=1)))
    InputTransform=(Rotation=(X=0,Y=0,Z=0,W=1),Translation=(X=0,Y=0,Z=0),Scale3D=(X=1,Y=1,Z=1))
    Primary=(Weight=1,Axis=(X=1,Y=0,Z=0),Target=(X=1,Y=0,Z=0),Kind=Location,Space=(Type=Bone,Name="None"))
    Secondary=(Weight=1,Axis=(X=0,Y=0,Z=1),Target=(X=1,Y=0,Z=0),Kind=Direction,Space=(Type=Bone,Name="None"))
    Weight=1
    InputTransform.Translation=RigVMVariableNode() {variable=Temp Shoulder Transform}.Value.Translation
    Primary.Target=RigVMVariableNode() {variable=Temp Elbow Transform}.Value.Translation
    Secondary.Target=RigVMVariableNode() {variable=Elbow Ik PV}.Value
   )
   RigVMVariableNode(Value=@n10.Result, ExecuteContext=@RigVMFunction_Sequence_1.E) {variable=Shoulder Rig Transform}
   n11 = &RigUnit_AimBoneMath(
    DebugSettings=(bEnabled=False,Scale=10,WorldOffset=(Rotation=(X=0,Y=0,Z=0,W=1),Translation=(X=0,Y=0,Z=0),Scale3D=(X=1,Y=1,Z=1)))
    InputTransform=(Rotation=(X=0,Y=0,Z=0,W=1),Translation=(X=0,Y=0,Z=0),Scale3D=(X=1,Y=1,Z=1))
    Primary=(Weight=1,Axis=(X=1,Y=0,Z=0),Target=(X=1,Y=0,Z=0),Kind=Location,Space=(Type=Bone,Name="None"))
    Secondary=(Weight=1,Axis=(X=0,Y=0,Z=1),Target=(X=1,Y=0,Z=0),Kind=Direction,Space=(Type=Bone,Name="None"))
    Weight=1
    InputTransform.Translation=RigVMVariableNode() {variable=Temp Elbow Transform}.Value.Translation
    Primary.Target=RigVMVariableNode() {variable=Temp Wrist Transform}.Value.Translation
    Secondary.Target=RigVMVariableNode() {variable=Elbow Ik PV}.Value
   )
   > RigVMVariableNode(Value=@n11.Result) {variable=Elbow Rig Transform}
   n12 = &RigUnit_GetTransform(bInitial=False, Item=RigVMVariableNode() {variable=Elbow Rotate Control Null}.Value)
   n13 = &RigUnit_AimBoneMath(
    DebugSettings=(bEnabled=False,Scale=10,WorldOffset=(Rotation=(X=0,Y=0,Z=0,W=1),Translation=(X=0,Y=0,Z=0),Scale3D=(X=1,Y=1,Z=1)))
    InputTransform=(Rotation=(X=0,Y=0,Z=0,W=1),Translation=(X=0,Y=0,Z=0),Scale3D=(X=1,Y=1,Z=1))
    Primary=(Weight=1,Axis=(X=0,Y=0,Z=1),Target=(X=1,Y=0,Z=0),Kind=Location,Space=(Type=Bone,Name="None"))
    Secondary=(Weight=1,Axis=(X=0,Y=1,Z=0),Target=(X=1,Y=0,Z=0),Kind=Location,Space=(Type=Bone,Name="None"))
    Weight=1
    InputTransform.Translation=@n12.Transform.Translation
    Secondary.Target=&RigUnit_GetTransform(bInitial=False, Item=RigVMVariableNode() {variable=Elbow PV}.Value).Transform.Translation
    Primary.Target=&RigVMFunction_MathTransformMakeAbsolute(Local=(Rotation=(X=0,Y=0,Z=0,W=1),Translation=(X=0,Y=0,Z=10),Scale3D=(X=1,Y=1,Z=1)), Parent=@n12.Transform) {template="Make Absolute::Execute(in Local,in Parent,out Global)"}.Global.Translation
   )
   &RigUnit_SetTransform(
    Weight=1
    bInitial=False
    Item=RigVMVariableNode() {variable=Elbow Rotate Control Null}.Value
    ExecutePin=@RigVMFunction_Sequence_1.C
    Value=@n13.Result
   )
   n14 = RigVMVariableNode() {variable=Elbow Rotate Control}
   n15 = RigVMVariableNode() {variable=Elbow Rotate Up Vector}
   > RigVMVariableNode(Value=&RigVMFunction_MathVectorSub(
     B=&RigUnit_GetTransform(bInitial=False, Item=@n14.Value).Transform.Translation
     A=&RigUnit_GetTransform(bInitial=False, Item=@n15.Value).Transform.Translation
    ) {template="Subtract::Execute(in A,in B,out Result)"}.Result) {variable=Elbow Ik PV}
   RigVMFunctionReturnNode(ExecuteContext=@RigVMFunction_Sequence.B)
  ]}
 New Function_ContainedGraph {nodes [
   n1 = RigVMFunctionEntryNode()
   RigVMFunction_Sequence = &RigVMFunction_Sequence(ExecuteContext=@n1.ExecuteContext)
   RigVMFunction_Sequence_2 = RigVMAggregateNode(ExecuteContext=@RigVMFunction_Sequence.A)
   n2 = RigVMVariableNode() {variable=Shoulder Backwards Transform}
   n3 = RigVMVariableNode() {variable=Wrist Backwards Transform}
   n4 = &RigUnit_AimBoneMath(
    DebugSettings=(bEnabled=False,Scale=10,WorldOffset=(Rotation=(X=0,Y=0,Z=0,W=1),Translation=(X=0,Y=0,Z=0),Scale3D=(X=1,Y=1,Z=1)))
    InputTransform=(Rotation=(X=0,Y=0,Z=0,W=1),Translation=(X=0,Y=0,Z=0),Scale3D=(X=1,Y=1,Z=1))
    Primary=(Weight=1,Axis=(X=0,Y=0,Z=1),Target=(X=1,Y=0,Z=0),Kind=Location,Space=(Type=Bone,Name="None"))
    Secondary=(Weight=1,Axis=(X=1,Y=0,Z=0),Target=(X=1,Y=0,Z=0),Kind=Direction,Space=(Type=Bone,Name="None"))
    Weight=1
    Secondary.Target=RigVMVariableNode() {variable=Backwards Main Pole Vector}.Value
    InputTransform.Translation=@n2.Value.Translation
    Primary.Target=@n3.Value.Translation
   )
   n5 = &RigVMFunction_MathVectorDistance(A=@n2.Value.Translation, B=@n3.Value.Translation)
   &RigUnit_SetTransform(
    Weight=1
    bInitial=False
    Item=&RigUnit_HierarchyGetParent(bDefaultParent=True, Child=RigVMVariableNode() {variable=Elbow Rotate Control Null}.Value).Parent
    ExecutePin=@RigVMFunction_Sequence_2.A
    Value=&RigVMFunction_MathTransformMakeAbsolute(
     Local=(Rotation=(X=0,Y=0,Z=0,W=1),Translation=(X=0,Y=0,Z=0),Scale3D=(X=1,Y=1,Z=1))
     Parent=@n4.Result
     Local.Translation.Z=&RigVMFunction_MathDoubleMul(B=0.5, A=@n5.Result).Result
    ) {template="Make Absolute::Execute(in Local,in Parent,out Global)"}.Global
   )
   n6 = RigVMVariableNode() {variable=Elbow Rotate Control Null}
   &RigUnit_SetTransform(
    Space=LocalSpace
    Weight=1
    bInitial=False
    Value=&RigUnit_GetTransform(Space=LocalSpace, bInitial=true, Item=@n6.Value).Transform
    Item=@n6.Value
    ExecutePin=@RigVMFunction_Sequence_2.C
   )
   n7 = RigVMVariableNode() {variable=Elbow Rotate Control}
   n8 = &RigVMFunction_MathVectorSub(
    B=&RigUnit_GetTransform(bInitial=False, Item=RigVMVariableNode() {variable=Elbow Rotate Control}.Value).Transform.Translation
    A=RigVMVariableNode() {variable=Elbow Input Transform}.Value.Translation
   ) {template="Subtract::Execute(in A,in B,out Result)"}
   Set Transform_5 = &RigUnit_SetTranslation(
    Space=GlobalSpace
    Weight=1
    bInitial=False
    bPropagateToChildren=True
    Value=&RigVMFunction_MathVectorAdd(
     B=&RigUnit_GetTransform(bInitial=False, Item=@n7.Value).Transform.Translation
     A=&RigVMFunction_MathVectorSetLength(Length=20, Value=@n8.Result) {template="SetLength::Execute(in Value,in Length,out Result)"}.Result
    ) {template="Add::Execute(in A,in B,out Result)"}.Result
    Item=RigVMVariableNode() {variable=Elbow PV}.Value
    ExecutePin=@RigVMFunction_Sequence_2.B
   ) {template="Set Transform::Execute(in Item,in Space,in bInitial,in Value,in Weight,in bPropagateToChildren)"}
   RigVMFunctionReturnNode(ExecuteContext=@RigVMFunction_Sequence.B)
   n9 = RigVMRerouteNode(Value=())
   n10 = RigVMVariableNode() {variable=Shoulder Input Transform}
   n11 = RigVMVariableNode() {variable=Wrist Input Transform}
   n12 = &RigUnit_AimBoneMath(
    DebugSettings=(bEnabled=False,Scale=10,WorldOffset=(Rotation=(X=0,Y=0,Z=0,W=1),Translation=(X=0,Y=0,Z=0),Scale3D=(X=1,Y=1,Z=1)))
    InputTransform=(Rotation=(X=0,Y=0,Z=0,W=1),Translation=(X=0,Y=0,Z=0),Scale3D=(X=1,Y=1,Z=1))
    Primary=(Weight=1,Axis=(X=0,Y=0,Z=1),Target=(X=1,Y=0,Z=0),Kind=Location,Space=(Type=Bone,Name="None"))
    Secondary=(Weight=1,Axis=(X=1,Y=0,Z=0),Target=(X=1,Y=0,Z=0),Kind=Direction,Space=(Type=Bone,Name="None"))
    Weight=1
    Secondary.Target=RigVMVariableNode() {variable=Backwards Main Pole Vector}.Value
    InputTransform.Translation=@n10.Value.Translation
    Primary.Target=@n11.Value.Translation
   )
   n13 = &RigVMFunction_MathDoubleMul(B=0.5, A=&RigVMFunction_MathVectorDistance(A=@n10.Value.Translation, B=@n11.Value.Translation).Result)
   &RigUnit_SetTransform(
    Weight=1
    bInitial=False
    ExecutePin=@n9.Value
    Item=RigVMVariableNode() {variable=Elbow PV}.Value
    Value=&RigVMFunction_MathTransformMakeAbsolute(
     Local=(Rotation=(X=0,Y=0,Z=0,W=1),Translation=(X=0,Y=20,Z=0),Scale3D=(X=1,Y=1,Z=1))
     Parent=&RigVMFunction_MathTransformMakeAbsolute(Local=(Rotation=(X=0,Y=0,Z=0,W=1),Translation=(X=0,Y=0,Z=0),Scale3D=(X=1,Y=1,Z=1)), Parent=@n12.Result, Local.Translation.Z=@n13.Result) {template="Make Absolute::Execute(in Local,in Parent,out Global)"}.Global
    ) {template="Make Absolute::Execute(in Local,in Parent,out Global)"}.Global
   )
   n14 = RigVMRerouteNode(Value=())
   n15 = RigVMDispatchNode(
    True=1
    False=RigVMDispatchNode(False=1, True=1, Condition=RigVMVariableNode() {variable=Right Side}.Value) {template="DISPATCH_RigVMDispatch_If(in Condition,in True,in False,out Result)"}.Result
    Condition=RigVMVariableNode() {variable=Left Side}.Value
   ) {template="DISPATCH_RigVMDispatch_If(in Condition,in True,in False,out Result)"}
   n16 = &RigUnit_GetTransform(bInitial=False, Item=RigVMVariableNode() {variable=Shoulder Connector}.Value)
   Subtract = &RigVMFunction_MathVectorSub(
    B=@n16.Transform.Translation
    A=&RigUnit_GetTransform(bInitial=False, Item=RigVMVariableNode() {variable=Elbow Connector}.Value).Transform.Translation
   ) {template="Subtract::Execute(in A,in B,out Result)"}
   n17 = RigVMVariableNode() {variable=Wrist Connector}
   RigVMFunction_MathVectorCross = &RigVMFunction_MathVectorCross(
    B=@Subtract.Result
    A=&RigVMFunction_MathVectorSub(B=@n16.Transform.Translation, A=&RigUnit_GetTransform(bInitial=False, Item=@n17.Value).Transform.Translation) {template="Subtract::Execute(in A,in B,out Result)"}.Result
   ) {template="Cross::Execute(in A,in B,out Result)"}
   RigVMFunction_MathVectorSetLength = &RigVMFunction_MathVectorSetLength(Length=-10, Value=@RigVMFunction_MathVectorCross.Result) {template="SetLength::Execute(in Value,in Length,out Result)"}
   n18 = &RigVMFunction_MathVectorMul(B=(X=1,Y=1,Z=1), B.X=@n15.Result, B.Y=@n15.Result, B.Z=@n15.Result, A=@RigVMFunction_MathVectorSetLength.Result) {template="Multiply::Execute(in A,in B,out Result)"}
   RigVMVariableNode(Value=@n18.Result, ExecuteContext=@n14.Value) {variable=Elbow Ik PV}
   Add = &RigVMFunction_MathVectorAdd(B=@n18.Result, A=@n16.Transform.Translation) {template="Add::Execute(in A,in B,out Result)"}
   > RigVMFunction_DebugLineNoSpace = &RigVMFunction_DebugLineNoSpace(
    Color=(R=1,G=0,B=0,A=1)
    DebugDrawSettings=(DepthPriority=SDPG_Foreground,Lifetime=-1)
    Thickness=0
    WorldOffset=(Rotation=(X=0,Y=0,Z=0,W=1),Translation=(X=0,Y=0,Z=0),Scale3D=(X=1,Y=1,Z=1))
    bEnabled=True
    B=@Add.Result
    A=@n16.Transform.Translation
   )
   RigVMFunction_Sequence_1 = RigVMAggregateNode(ExecuteContext=())
   Get Main Pole Vector = RigVMFunctionReferenceNode(Initial=False)
   n19 = &RigUnit_GetTransform(bInitial=False, Item=RigVMVariableNode() {variable=Shoulder Connector}.Value)
   n20 = &RigUnit_GetTransform(bInitial=False, Item=RigVMVariableNode() {variable=Wrist Connector}.Value)
   n21 = &RigUnit_AimBoneMath(
    DebugSettings=(bEnabled=False,Scale=10,WorldOffset=(Rotation=(X=0,Y=0,Z=0,W=1),Translation=(X=0,Y=0,Z=0),Scale3D=(X=1,Y=1,Z=1)))
    InputTransform=(Rotation=(X=0,Y=0,Z=0,W=1),Translation=(X=0,Y=0,Z=0),Scale3D=(X=1,Y=1,Z=1))
    Primary=(Weight=1,Axis=(X=0,Y=0,Z=1),Target=(X=1,Y=0,Z=0),Kind=Location,Space=(Type=Bone,Name="None"))
    Secondary=(Weight=1,Axis=(X=1,Y=0,Z=0),Target=(X=1,Y=0,Z=0),Kind=Direction,Space=(Type=Bone,Name="None"))
    Weight=1
    Secondary.Target=@"Get Main Pole Vector".Vector
    InputTransform.Translation=@n19.Transform.Translation
    Primary.Target=@n20.Transform.Translation
   )
   Multiply = &RigVMFunction_MathDoubleMul(B=0.5, A=&RigVMFunction_MathVectorDistance(A=@n19.Transform.Translation, B=@n20.Transform.Translation).Result)
   RigVMVariableNode(
    Value=&RigVMFunction_MathTransformMakeAbsolute(Local=(Rotation=(X=0,Y=0,Z=0,W=1),Translation=(X=0,Y=0,Z=0),Scale3D=(X=1,Y=1,Z=1)), Parent=@n21.Result, Local.Translation.Z=@Multiply.Result) {template="Make Absolute::Execute(in Local,in Parent,out Global)"}.Global
    ExecuteContext=@RigVMFunction_Sequence_1.B
   ) {variable=Elbow Control Null Transform}
   &RigUnit_SetTransform(
    Weight=1
    bInitial=False
    Item=RigVMVariableNode() {variable=Elbow Rotate Control Null}.Value
    Value=RigVMVariableNode() {variable=Elbow Control Null Transform}.Value
    ExecutePin=@RigVMFunction_Sequence_1.E
   )
   > &RigUnit_SetTransform(
    Weight=1
    bInitial=False
    Item=RigVMVariableNode() {variable=Elbow Rotate Control}.Value
    Value=RigVMVariableNode() {variable=Elbow Control Transform}.Value
   )
   n22 = RigVMVariableNode() {variable=Elbow Connector}
   n23 = RigVMVariableNode() {variable=Elbow Rotate Control}
   RigVMVariableNode(
    Value=&RigVMFunction_MathTransformMakeRelative(
     Parent=&RigUnit_GetTransform(bInitial=true, Item=@n22.Value).Transform
     Global=&RigUnit_GetTransform(bInitial=true, Item=@n23.Value).Transform
    ).Local
    ExecuteContext=@RigVMFunction_Sequence_1.A
   ) {variable=Elbow Rotate Local Transform}
   n24 = &RigVMFunction_MathTransformMakeAbsolute(
    Parent=&RigUnit_GetTransform(bInitial=false, Item=RigVMVariableNode() {variable=Elbow Connector}.Value).Transform
    Local=RigVMVariableNode() {variable=Elbow Rotate Local Transform}.Value
   ) {template="Make Absolute::Execute(in Local,in Parent,out Global)"}
   RigVMVariableNode(
    Value=&RigVMFunction_MathVectorSub(
     A=&RigVMFunction_MathTransformMakeAbsolute(Local=(Rotation=(X=0,Y=0,Z=0,W=1),Translation=(X=10,Y=0,Z=0),Scale3D=(X=1,Y=1,Z=1)), Parent=@n24.Global) {template="Make Absolute::Execute(in Local,in Parent,out Global)"}.Global.Translation
     B=@n24.Global.Translation
    ) {template="Subtract::Execute(in A,in B,out Result)"}.Result
    ExecuteContext=@RigVMFunction_Sequence_1.C
   ) {variable=Elbow Ik PV}
   n25 = &RigUnit_GetTransform(bInitial=False, Item=RigVMVariableNode() {variable=Shoulder Connector}.Value)
   n26 = &RigUnit_GetTransform(bInitial=False, Item=RigVMVariableNode() {variable=Wrist Connector}.Value)
   n27 = &RigUnit_AimBoneMath(
    DebugSettings=(bEnabled=False,Scale=10,WorldOffset=(Rotation=(X=0,Y=0,Z=0,W=1),Translation=(X=0,Y=0,Z=0),Scale3D=(X=1,Y=1,Z=1)))
    InputTransform=(Rotation=(X=0,Y=0,Z=0,W=1),Translation=(X=0,Y=0,Z=0),Scale3D=(X=1,Y=1,Z=1))
    Primary=(Weight=1,Axis=(X=0,Y=0,Z=1),Target=(X=1,Y=0,Z=0),Kind=Location,Space=(Type=Bone,Name="None"))
    Secondary=(Weight=1,Axis=(X=1,Y=0,Z=0),Target=(X=1,Y=0,Z=0),Kind=Direction,Space=(Type=Bone,Name="None"))
    Weight=1
    InputTransform.Translation=@n25.Transform.Translation
    Primary.Target=@n26.Transform.Translation
    Secondary.Target=RigVMVariableNode() {variable=Elbow Ik PV}.Value
   )
   n28 = &RigVMFunction_MathVectorDistance(A=@n25.Transform.Translation, B=@n26.Transform.Translation)
   RigVMVariableNode(
    Value=&RigVMFunction_MathTransformMakeAbsolute(
     Local=(Rotation=(X=0,Y=0,Z=0,W=1),Translation=(X=0,Y=0,Z=0),Scale3D=(X=1,Y=1,Z=1))
     Parent=@n27.Result
     Local.Translation.Z=&RigVMFunction_MathDoubleMul(B=0.5, A=@n28.Result).Result
    ) {template="Make Absolute::Execute(in Local,in Parent,out Global)"}.Global
    ExecuteContext=@RigVMFunction_Sequence_1.D
   ) {variable=Elbow Control Transform}
   &RigUnit_HierarchyGetParent(bDefaultParent=True, Child=@n6.Value)
  ]}
 New Function_ContainedGraph {nodes [
   n1 = RigVMFunctionEntryNode()
   RigVMFunction_Sequence = &RigVMFunction_Sequence(ExecuteContext=@n1.ExecuteContext)
   &RigVMFunction_Sequence(ExecuteContext=@RigVMFunction_Sequence.A)
   RigVMFunctionReturnNode(ExecuteContext=@RigVMFunction_Sequence.B)
  ]}
 New Function_ContainedGraph {nodes [
   n1 = RigVMFunctionEntryNode()
   RigVMFunction_Sequence = &RigVMFunction_Sequence(ExecuteContext=@n1.ExecuteContext)
   RigVMFunction_Sequence_1 = RigVMAggregateNode(ExecuteContext=@RigVMFunction_Sequence.A)
   Get Main Pole Vector_2 = RigVMFunctionReferenceNode(Initial=False)
   n2 = &RigUnit_AimBoneMath(
    DebugSettings=(bEnabled=False,Scale=10,WorldOffset=(Rotation=(X=0,Y=0,Z=0,W=1),Translation=(X=0,Y=0,Z=0),Scale3D=(X=1,Y=1,Z=1)))
    InputTransform=(Rotation=(X=0,Y=0,Z=0,W=1),Translation=(X=0,Y=0,Z=0),Scale3D=(X=1,Y=1,Z=1))
    Primary=(Weight=1,Axis=(X=1,Y=0,Z=0),Target=(X=1,Y=0,Z=0),Kind=Location,Space=(Type=Bone,Name="None"))
    Secondary=(Weight=1,Axis=(X=0,Y=0,Z=1),Target=(X=1,Y=0,Z=0),Kind=Direction,Space=(Type=Bone,Name="None"))
    Weight=1
    Secondary.Target=@"Get Main Pole Vector_2".Vector
    InputTransform.Translation=&RigUnit_GetTransform(bInitial=False, Item=RigVMVariableNode() {variable=Ball Connector}.Value).Transform.Translation
    Primary.Target=&RigUnit_GetTransform(bInitial=False, Item=RigVMVariableNode() {variable=Elbow Connector}.Value).Transform.Translation
   )
   &RigUnit_SetTransform(
    Weight=1
    bInitial=False
    Value=@n2.Result
    Item=RigVMVariableNode() {variable=Ball Rotate Aim Null}.Value
    ExecutePin=@RigVMFunction_Sequence_1.A
   )
   Get Main Pole Vector_1 = RigVMFunctionReferenceNode(Initial=False)
   n3 = &RigUnit_AimBoneMath(
    DebugSettings=(bEnabled=False,Scale=10,WorldOffset=(Rotation=(X=0,Y=0,Z=0,W=1),Translation=(X=0,Y=0,Z=0),Scale3D=(X=1,Y=1,Z=1)))
    InputTransform=(Rotation=(X=0,Y=0,Z=0,W=1),Translation=(X=0,Y=0,Z=0),Scale3D=(X=1,Y=1,Z=1))
    Primary=(Weight=1,Axis=(X=0,Y=0,Z=1),Target=(X=1,Y=0,Z=0),Kind=Location,Space=(Type=Bone,Name="None"))
    Secondary=(Weight=1,Axis=(X=1,Y=0,Z=0),Target=(X=1,Y=0,Z=0),Kind=Direction,Space=(Type=Bone,Name="None"))
    Weight=1
    Secondary.Target=@"Get Main Pole Vector_1".Vector
    InputTransform.Translation=&RigUnit_GetTransform(bInitial=False, Item=RigVMVariableNode() {variable=Ball Connector}.Value).Transform.Translation
    Primary.Target=&RigUnit_GetTransform(bInitial=False, Item=RigVMVariableNode() {variable=Wrist Connector}.Value).Transform.Translation
   )
   &RigUnit_SetTransform(
    Weight=1
    bInitial=False
    Value=@n3.Result
    Item=RigVMVariableNode() {variable=Ball Rotate Control}.Value
    ExecutePin=@RigVMFunction_Sequence_1.B
   )
   RigVMFunctionReturnNode(ExecuteContext=@RigVMFunction_Sequence.B)
  ]}
 New Function_ContainedGraph {nodes [
   n1 = RigVMFunctionEntryNode()
   RigVMFunction_Sequence = &RigVMFunction_Sequence(ExecuteContext=@n1.ExecuteContext)
   n2 = &RigVMFunction_Sequence(ExecuteContext=@RigVMFunction_Sequence.A)
   RigVMFunction_DebugLineStripNoSpace = &RigVMFunction_DebugLineStripNoSpace(
    Color=(R=1,G=0,B=0,A=1)
    DebugDrawSettings=(DepthPriority=SDPG_Foreground,Lifetime=-1)
    Points=((X=0,Y=0,Z=0),(X=0,Y=0,Z=0),(X=0,Y=0,Z=0),(X=0,Y=0,Z=0),(X=0,Y=0,Z=0),(X=0,Y=0,Z=0),(X=0,Y=0,Z=0))
    Thickness=0
    WorldOffset=(Rotation=(X=0,Y=0,Z=0,W=1),Translation=(X=0,Y=0,Z=0),Scale3D=(X=1,Y=1,Z=1))
    bEnabled=True
    Points.0=&RigUnit_GetTransform(bInitial=false, Item=RigVMVariableNode() {variable=Scapula Tip Connector}.Value).Transform.Translation
    Points.1=&RigUnit_GetTransform(bInitial=false, Item=RigVMVariableNode() {variable=Shoulder Connector}.Value).Transform.Translation
    Points.2=&RigUnit_GetTransform(bInitial=false, Item=RigVMVariableNode() {variable=Elbow Connector}.Value).Transform.Translation
    Points.3=&RigUnit_GetTransform(bInitial=false, Item=RigVMVariableNode() {variable=Wrist Connector}.Value).Transform.Translation
    Points.4=&RigUnit_GetTransform(bInitial=false, Item=RigVMVariableNode() {variable=Ball Connector}.Value).Transform.Translation
    Points.5=&RigUnit_GetTransform(bInitial=False, Item=RigVMVariableNode() {variable=Toe Connector}.Value).Transform.Translation
    Points.6=&RigUnit_GetTransform(bInitial=false, Item=RigVMVariableNode() {variable=Toe Tip Connector}.Value).Transform.Translation
    ExecutePin=@n2.B
   )
   RigVMVariableNode(
    Value=&RigUnit_GetTransform(bInitial=false, Item=RigVMVariableNode() {variable=Scapula Tip Connector}.Value).Transform
    ExecuteContext=@n2.A
   ) {variable=Scapula Tip Input Transform}
   > RigVMVariableNode(Value=&RigUnit_GetTransform(bInitial=false, Item=RigVMVariableNode() {variable=Shoulder Connector}.Value).Transform) {variable=Shoulder Input Transform}
   > RigVMVariableNode(Value=&RigUnit_GetTransform(bInitial=false, Item=RigVMVariableNode() {variable=Elbow Connector}.Value).Transform) {variable=Elbow Input Transform}
   > RigVMVariableNode(Value=&RigUnit_GetTransform(bInitial=false, Item=RigVMVariableNode() {variable=Wrist Connector}.Value).Transform) {variable=Wrist Input Transform}
   > RigVMVariableNode(Value=&RigUnit_GetTransform(bInitial=false, Item=RigVMVariableNode() {variable=Ball Connector}.Value).Transform) {variable=Ball Input Transform}
   > RigVMVariableNode(Value=&RigUnit_GetTransform(bInitial=False, Item=RigVMVariableNode() {variable=Toe Connector}.Value).Transform) {variable=Toe Input Transform}
   > RigVMVariableNode(Value=&RigUnit_GetTransform(bInitial=false, Item=RigVMVariableNode() {variable=Toe Tip Connector}.Value).Transform) {variable=Toe Tip Input Transform}
   RigVMFunctionReturnNode(ExecuteContext=@RigVMFunction_Sequence.B)
   Get Main Pole Vector_1 = RigVMFunctionReferenceNode(Initial=false)
   RigVMVariableNode(ExecuteContext=(), Value=@"Get Main Pole Vector_1".Vector) {variable=Backwards Main Pole Vector}
  ]}
 New Function_ContainedGraph {nodes [
   n1 = RigVMFunctionEntryNode()
   RigVMFunction_Sequence = &RigVMFunction_Sequence(ExecuteContext=@n1.ExecuteContext)
   n2 = &RigVMFunction_Sequence(ExecuteContext=@RigVMFunction_Sequence.A)
   Set Side Color_0 = RigVMFunctionReferenceNode(ExecuteContext=@n2.A)
   > RigVMFunction_DebugLineNoSpace = &RigVMFunction_DebugLineNoSpace(
    DebugDrawSettings=(DepthPriority=SDPG_Foreground,Lifetime=-1)
    Thickness=0
    WorldOffset=(Rotation=(X=0,Y=0,Z=0,W=1),Translation=(X=0,Y=0,Z=0),Scale3D=(X=1,Y=1,Z=1))
    bEnabled=true
    B=&RigUnit_GetTransform(bInitial=False, Item=RigVMVariableNode() {variable=Elbow PV}.Value).Transform.Translation
    A=&RigUnit_GetTransform(bInitial=False, Item=RigVMVariableNode() {variable=Elbow Connector}.Value).Transform.Translation
    Color=@"Set Side Color_0".Color
   )
   RigVMFunctionReturnNode(ExecuteContext=@RigVMFunction_Sequence.B)
  ]}
 New Function_ContainedGraph {nodes [
   n1 = RigVMFunctionEntryNode()
   RigVMFunction_Sequence = &RigVMFunction_Sequence(ExecuteContext=@n1.ExecuteContext)
   RigVMFunction_Sequence_1 = RigVMAggregateNode(ExecuteContext=@RigVMFunction_Sequence.A)
   n2 = &RigUnit_HierarchyGetParent(bDefaultParent=True, Child=RigVMVariableNode() {variable=Ik Control}.Value)
   n3 = RigVMVariableNode() {variable=Ik Point}
   RigVMFunction_MathVectorFromDouble = &RigVMFunction_MathVectorFromDouble(Value=RigVMVariableNode() {variable=Scapula Tip Aim Percent}.Value) {template="Cast::Execute(in Value,out Result)"}
   Multiply = &RigVMFunction_MathVectorMul(
    A=&RigVMFunction_MathTransformMakeRelative(
     Parent=&RigUnit_GetTransform(bInitial=False, Item=@n2.Parent).Transform
     Global=&RigUnit_GetTransform(bInitial=False, Item=@n3.Value).Transform
    ).Local.Translation
    B=@RigVMFunction_MathVectorFromDouble.Result
   ) {template="Multiply::Execute(in A,in B,out Result)"}
   Set Transform = &RigUnit_SetTranslation(
    Space=LocalSpace
    Weight=1
    bInitial=False
    bPropagateToChildren=True
    Value=@Multiply.Result
    Item=RigVMVariableNode() {variable=Scapula Tip Aim Point}.Value
    ExecutePin=@RigVMFunction_Sequence_1.A
   ) {template="Set Transform::Execute(in Item,in Space,in bInitial,in Value,in Weight,in bPropagateToChildren)"}
   Get Scapula Tip Local Transform = RigVMFunctionReferenceNode(Input Transform=RigVMVariableNode() {variable=Initial Aim}.Value)
   RigVMVariableNode(
    Value=&RigVMFunction_MathTransformMakeAbsolute(Local=@"Get Scapula Tip Local Transform".Local Transform, Parent=RigVMVariableNode() {variable=Current Aim}.Value) {template="Make Absolute::Execute(in Local,in Parent,out Global)"}.Global
    ExecuteContext=@RigVMFunction_Sequence_1.C
   ) {variable=Scapula Tip Backwards Transform}
   Get Main Pole Vector = RigVMFunctionReferenceNode(Initial=False)
   n4 = &RigUnit_AimBoneMath(
    DebugSettings=(bEnabled=False,Scale=10,WorldOffset=(Rotation=(X=0,Y=0,Z=0,W=1),Translation=(X=0,Y=0,Z=0),Scale3D=(X=1,Y=1,Z=1)))
    InputTransform=(Rotation=(X=0,Y=0,Z=0,W=1),Translation=(X=0,Y=0,Z=0),Scale3D=(X=1,Y=1,Z=1))
    Primary=(Weight=1,Axis=(X=1,Y=0,Z=0),Target=(X=1,Y=0,Z=0),Kind=Location,Space=(Type=Bone,Name="None"))
    Secondary=(Weight=1,Axis=(X=0,Y=0,Z=1),Target=(X=1,Y=0,Z=0),Kind=Direction,Space=(Type=Bone,Name="None"))
    Weight=1
    Secondary.Target=@"Get Main Pole Vector".Vector
    InputTransform.Translation=RigVMVariableNode() {variable=Scapula Tip Input Transform}.Value.Translation
    Primary.Target=RigVMVariableNode() {variable=Ball Input Transform}.Value.Translation
   )
   RigVMVariableNode(Value=@n4.Result, ExecuteContext=@RigVMFunction_Sequence_1.B) {variable=Scapula Tip Aim Transform}
   Get Scapula Tip Aim Transform = RigVMFunctionReferenceNode(Initial=true)
   > RigVMVariableNode(Value=@"Get Scapula Tip Aim Transform".Aim Transform) {variable=Initial Aim}
   Get Scapula Tip Aim Transform_1 = RigVMFunctionReferenceNode()
   > RigVMVariableNode(Value=@"Get Scapula Tip Aim Transform_1".Aim Transform) {variable=Current Aim}
   Get Shoulder Local Transform_1 = RigVMFunctionReferenceNode(Input Transform=RigVMVariableNode() {variable=Initial Aim}.Value)
   RigVMVariableNode(
    Value=&RigVMFunction_MathTransformMakeAbsolute(Local=@"Get Shoulder Local Transform_1".Local Trasnform, Parent=RigVMVariableNode() {variable=Current Aim}.Value) {template="Make Absolute::Execute(in Local,in Parent,out Global)"}.Global
    ExecuteContext=@RigVMFunction_Sequence_1.D
   ) {variable=Shoulder Backwards Transform}
   Get Elbow Local Transform = RigVMFunctionReferenceNode(Input Transform=RigVMVariableNode() {variable=Initial Aim}.Value)
   RigVMVariableNode(
    Value=&RigVMFunction_MathTransformMakeAbsolute(Local=@"Get Elbow Local Transform".Local Transform, Parent=RigVMVariableNode() {variable=Current Aim}.Value) {template="Make Absolute::Execute(in Local,in Parent,out Global)"}.Global
    ExecuteContext=@RigVMFunction_Sequence_1.E
   ) {variable=Elbow Backwards Transform}
   Get Wrist Local Transform = RigVMFunctionReferenceNode(Input Transform=RigVMVariableNode() {variable=Initial Aim}.Value)
   RigVMVariableNode(
    Value=&RigVMFunction_MathTransformMakeAbsolute(Local=@"Get Wrist Local Transform".Local Transform, Parent=RigVMVariableNode() {variable=Current Aim}.Value) {template="Make Absolute::Execute(in Local,in Parent,out Global)"}.Global
    ExecuteContext=@RigVMFunction_Sequence_1.F
   ) {variable=Wrist Backwards Transform}
   Get Ball Local Transform = RigVMFunctionReferenceNode(Input Transform=RigVMVariableNode() {variable=Initial Aim}.Value)
   RigVMVariableNode(
    Value=&RigVMFunction_MathTransformMakeAbsolute(Local=@"Get Ball Local Transform".Local Transform, Parent=RigVMVariableNode() {variable=Current Aim}.Value) {template="Make Absolute::Execute(in Local,in Parent,out Global)"}.Global
    ExecuteContext=@RigVMFunction_Sequence_1.G
   ) {variable=Ball Backwards Transform}
   RigVMFunction_DebugLineStripNoSpace = &RigVMFunction_DebugLineStripNoSpace(
    Color=(R=1,G=1,B=0.0493,A=1)
    DebugDrawSettings=(DepthPriority=SDPG_Foreground,Lifetime=-1)
    Points=((X=0,Y=0,Z=0),(X=0,Y=0,Z=0),(X=0,Y=0,Z=0),(X=0,Y=0,Z=0),(X=0,Y=0,Z=0))
    Thickness=0.2
    WorldOffset=(Rotation=(X=0,Y=0,Z=0,W=1),Translation=(X=0,Y=0,Z=0),Scale3D=(X=1,Y=1,Z=1))
    bEnabled=false
    Points.0=RigVMVariableNode() {variable=Scapula Tip Backwards Transform}.Value.Translation
    Points.1=RigVMVariableNode() {variable=Shoulder Backwards Transform}.Value.Translation
    Points.2=RigVMVariableNode() {variable=Elbow Backwards Transform}.Value.Translation
    Points.3=RigVMVariableNode() {variable=Wrist Backwards Transform}.Value.Translation
    Points.4=RigVMVariableNode() {variable=Ball Backwards Transform}.Value.Translation
    ExecutePin=@RigVMFunction_Sequence_1.H
   )
   RigVMFunctionReturnNode(ExecuteContext=@RigVMFunction_Sequence.B)
   Backwards Get Scapula Tip Aim Transform_2 = RigVMFunctionReferenceNode(Initial=true)
   Backwards Get Scapula Tip Aim Transform_3 = RigVMFunctionReferenceNode(Initial=False)
   &RigUnit_GetTransform(bInitial=true, Item=RigVMVariableNode() {variable=Scapula Tip Control}.Value)
   &RigUnit_GetTransform(bInitial=true, Item=RigVMVariableNode() {variable=Scapula Tip Aim Point}.Value)
   RigVMVariableNode() {variable=Toe Backwards Transform}
   RigVMVariableNode() {variable=Toe Tip Backwards Transform}
  ]}
 New Function_ContainedGraph {nodes [
   n1 = RigVMFunctionEntryNode()
   RigVMFunction_Sequence = &RigVMFunction_Sequence(ExecuteContext=@n1.ExecuteContext)
   n2 = &RigVMFunction_Sequence(ExecuteContext=@RigVMFunction_Sequence.A)
   RigVMFunction_Sequence_2 = RigVMAggregateNode(ExecuteContext=@n2.A)
   Get Ik Angles = RigVMFunctionReferenceNode(
    C 180=true
    Negate A=false
    Negate B=true
    Negate C=true
    Length C=&RigVMFunction_MathVectorDistance(
     B=RigVMVariableNode() {variable=Ball Input Transform}.Value.Translation
     A=RigVMVariableNode() {variable=Shoulder Input Transform}.Value.Translation
    ).Result
    Length B=RigVMVariableNode() {variable=Elbow Ball Aim Length}.Value
    Length A=RigVMVariableNode() {variable=Shoulder Length}.Value
   )
   RigVMVariableNode(Value=@"Get Ik Angles".Angle B, ExecuteContext=@RigVMFunction_Sequence_2.B) {variable=Shoulder Angle}
   > RigVMVariableNode(Value=RigVMDispatchNode(
     True=0
     Condition=RigVMDispatchNode(B=180, A=@"Get Ik Angles".Angle C) {template="DISPATCH_RigVMDispatch_CoreEquals(in A,in B,out Result)"}.Result
     False=@"Get Ik Angles".Angle C
    ) {template="DISPATCH_RigVMDispatch_If(in Condition,in True,in False,out Result)"}.Result) {variable=Lower Leg Angle}
   Get Ik Angles_1 = RigVMFunctionReferenceNode(
    C 180=true
    Negate B=true
    Negate C=true
    Length A=RigVMVariableNode() {variable=Elbow Length}.Value
    Length B=RigVMVariableNode() {variable=Wrist Length}.Value
   )
   > RigVMVariableNode(Value=@"Get Ik Angles_1".Angle B) {variable=Elbow Angle}
   > RigVMVariableNode(Value=@"Get Ik Angles_1".Angle C) {variable=Wrist Angle}
   RigVMVariableNode(
    Value=&RigVMFunction_MathVectorDistance(
     A=RigVMVariableNode() {variable=Elbow Input Transform}.Value.Translation
     B=RigVMVariableNode() {variable=Ball Input Transform}.Value.Translation
    ).Result
    ExecuteContext=@RigVMFunction_Sequence_2.A
   ) {variable=Elbow Ball Aim Length}
   RigVMFunction_Sequence_3 = RigVMAggregateNode(ExecuteContext=@n2.B)
   RigVMFunction_DebugLineStripNoSpace = &RigVMFunction_DebugLineStripNoSpace(
    Color=(R=1,G=1,B=0.0493,A=1)
    DebugDrawSettings=(DepthPriority=SDPG_Foreground,Lifetime=-1)
    Points=((X=0,Y=0,Z=0),(X=0,Y=0,Z=0),(X=0,Y=0,Z=0),(X=0,Y=0,Z=0),(X=0,Y=0,Z=0))
    Thickness=0.2
    WorldOffset=(Rotation=(X=0,Y=0,Z=0,W=1),Translation=(X=0,Y=0,Z=0),Scale3D=(X=1,Y=1,Z=1))
    bEnabled=false
    Points.0=RigVMVariableNode() {variable=Scapula Tip Backwards Transform}.Value.Translation
    Points.1=RigVMVariableNode() {variable=Shoulder Backwards Transform}.Value.Translation
    Points.2=RigVMVariableNode() {variable=Elbow Backwards Transform}.Value.Translation
    Points.3=RigVMVariableNode() {variable=Wrist Backwards Transform}.Value.Translation
    Points.4=RigVMVariableNode() {variable=Ball Backwards Transform}.Value.Translation
    ExecutePin=@RigVMFunction_Sequence_3.F
   )
   n3 = RigVMVariableNode() {variable=Wrist Angle}
   n4 = RigVMVariableNode() {variable=Elbow Backwards Transform}
   n5 = RigVMVariableNode() {variable=Elbow Length}
   RigVMVariableNode(
    Value=&RigVMFunction_MathTransformMakeAbsolute(
     Local=(Rotation=(X=0,Y=0,Z=0,W=1),Translation=(X=0,Y=0,Z=0),Scale3D=(X=1,Y=1,Z=1))
     Local.Rotation=&RigVMFunction_MathQuaternionFromEuler(Euler=(X=0,Y=0,Z=0), Euler.Z=@n3.Value).Result
     Parent=&RigVMFunction_MathTransformMakeAbsolute(Local=(Rotation=(X=0,Y=0,Z=0,W=1),Translation=(X=0,Y=0,Z=0),Scale3D=(X=1,Y=1,Z=1)), Parent=@n4.Value, Local.Translation.X=@n5.Value) {template="Make Absolute::Execute(in Local,in Parent,out Global)"}.Global
    ) {template="Make Absolute::Execute(in Local,in Parent,out Global)"}.Global
    ExecuteContext=@RigVMFunction_Sequence_3.D
   ) {variable=Wrist Backwards Transform}
   RigVMVariableNode(
    Value=&RigVMFunction_MathTransformMakeAbsolute(
     Local=(Rotation=(X=0,Y=0,Z=0,W=1),Translation=(X=0,Y=0,Z=0),Scale3D=(X=1,Y=1,Z=1))
     Local.Translation.X=RigVMVariableNode() {variable=Wrist Length}.Value
     Parent=RigVMVariableNode() {variable=Wrist Backwards Transform}.Value
    ) {template="Make Absolute::Execute(in Local,in Parent,out Global)"}.Global
    ExecuteContext=@RigVMFunction_Sequence_3.E
   ) {variable=Ball Backwards Transform}
   n6 = RigVMVariableNode() {variable=Shoulder Length}
   n7 = RigVMVariableNode() {variable=Shoulder Backwards Transform}
   n8 = RigVMVariableNode() {variable=Lower Leg Angle}
   RigVMVariableNode(
    Value=&RigVMFunction_MathTransformMakeAbsolute(
     Local=(Rotation=(X=0,Y=0,Z=0,W=1),Translation=(X=0,Y=0,Z=0),Scale3D=(X=1,Y=1,Z=1))
     Parent=&RigVMFunction_MathTransformMakeAbsolute(Local=(Rotation=(X=0,Y=0,Z=0,W=1),Translation=(X=0,Y=0,Z=0),Scale3D=(X=1,Y=1,Z=1)), Local.Translation.X=@n6.Value, Parent=@n7.Value) {template="Make Absolute::Execute(in Local,in Parent,out Global)"}.Global
     Local.Rotation=&RigVMFunction_MathQuaternionFromEuler(Euler=(X=0,Y=0,Z=0), Euler.Z=@n8.Value).Result
    ) {template="Make Absolute::Execute(in Local,in Parent,out Global)"}.Global
    ExecuteContext=@RigVMFunction_Sequence_3.B
   ) {variable=Elbow Ik Aim Transform}
   n9 = &RigUnit_AimBoneMath(
    DebugSettings=(bEnabled=False,Scale=10,WorldOffset=(Rotation=(X=0,Y=0,Z=0,W=1),Translation=(X=0,Y=0,Z=0),Scale3D=(X=1,Y=1,Z=1)))
    InputTransform=(Rotation=(X=0,Y=0,Z=0,W=1),Translation=(X=0,Y=0,Z=0),Scale3D=(X=1,Y=1,Z=1))
    Primary=(Weight=1,Axis=(X=1,Y=0,Z=0),Target=(X=1,Y=0,Z=0),Kind=Location,Space=(Type=Bone,Name="None"))
    Secondary=(Weight=1,Axis=(X=0,Y=0,Z=1),Target=(X=1,Y=0,Z=0),Kind=Direction,Space=(Type=Bone,Name="None"))
    Weight=1
    Secondary.Target=RigVMVariableNode() {variable=Backwards Main Pole Vector}.Value
    Primary.Target=RigVMVariableNode() {variable=Ball Input Transform}.Value.Translation
    InputTransform.Translation=RigVMVariableNode() {variable=Shoulder Input Transform}.Value.Translation
   )
   RigVMVariableNode(Value=@n9.Result, ExecuteContext=@RigVMFunction_Sequence_3.A) {variable=Shoulder Ik Aim Transform}
   RigVMFunction_MathQuaternionFromEuler = &RigVMFunction_MathQuaternionFromEuler(Euler=(X=0,Y=0,Z=0), Euler.Z=RigVMVariableNode() {variable=Shoulder Angle}.Value)
   > RigVMVariableNode(Value=&RigVMFunction_MathTransformMakeAbsolute(
     Local=(Rotation=(X=0,Y=0,Z=0,W=1),Translation=(X=0,Y=0,Z=0),Scale3D=(X=1,Y=1,Z=1))
     Local.Rotation=@RigVMFunction_MathQuaternionFromEuler.Result
     Parent=RigVMVariableNode() {variable=Shoulder Ik Aim Transform}.Value
    ) {template="Make Absolute::Execute(in Local,in Parent,out Global)"}.Global) {variable=Shoulder Backwards Transform}
   > RigVMFunction_DebugTransformMutableNoSpace = &RigVMFunction_DebugTransformMutableNoSpace(
    Color=(R=1,G=0,B=0,A=1)
    DebugDrawSettings=(DepthPriority=SDPG_Foreground,Lifetime=-1)
    Mode=Axes
    Scale=5
    Thickness=0.5
    WorldOffset=(Rotation=(X=0,Y=0,Z=0,W=1),Translation=(X=0,Y=0,Z=0),Scale3D=(X=1,Y=1,Z=1))
    bEnabled=false
    Transform=RigVMVariableNode() {variable=Shoulder Backwards Transform}.Value
   )
   n10 = RigVMVariableNode() {variable=Elbow Angle}
   RigVMVariableNode(
    Value=&RigVMFunction_MathTransformMakeAbsolute(
     Local=(Rotation=(X=0,Y=0,Z=0,W=1),Translation=(X=0,Y=0,Z=0),Scale3D=(X=1,Y=1,Z=1))
     Local.Rotation=&RigVMFunction_MathQuaternionFromEuler(Euler=(X=0,Y=0,Z=0), Euler.Z=@n10.Value).Result
     Parent=RigVMVariableNode() {variable=Elbow Ik Aim Transform}.Value
    ) {template="Make Absolute::Execute(in Local,in Parent,out Global)"}.Global
    ExecuteContext=@RigVMFunction_Sequence_3.C
   ) {variable=Elbow Backwards Transform}
   RigVMFunctionReturnNode(ExecuteContext=@RigVMFunction_Sequence.B)
   RigVMVariableNode() {variable=Ball Backwards Transform}
   RigVMVariableNode() {variable=Shoulder Backwards Transform}
   RigVMVariableNode() {variable=Elbow Rig Transform}
   RigVMVariableNode() {variable=Ball Rig Transform}
   RigVMVariableNode() {variable=Ball Backwards Transform}
   RigVMVariableNode() {variable=Shoulder Backwards Transform}
  ]}
 New Function_ContainedGraph {nodes [
   RigVMFunctionEntryNode()
   Get Main Pole Vector = RigVMFunctionReferenceNode(Initial=RigVMVariableNode() {variable=Initial}.Value)
   n1 = &RigUnit_AimBoneMath(
    DebugSettings=(bEnabled=False,Scale=10,WorldOffset=(Rotation=(X=0,Y=0,Z=0,W=1),Translation=(X=0,Y=0,Z=0),Scale3D=(X=1,Y=1,Z=1)))
    InputTransform=(Rotation=(X=0,Y=0,Z=0,W=1),Translation=(X=0,Y=0,Z=0),Scale3D=(X=1,Y=1,Z=1))
    Primary=(Weight=1,Axis=(X=1,Y=0,Z=0),Target=(X=1,Y=0,Z=0),Kind=Location,Space=(Type=Bone,Name="None"))
    Secondary=(Weight=1,Axis=(X=0,Y=0,Z=1),Target=(X=1,Y=0,Z=0),Kind=Direction,Space=(Type=Bone,Name="None"))
    Weight=1
    Secondary.Target=@"Get Main Pole Vector".Vector
    InputTransform.Translation=&RigUnit_GetTransform(Item=RigVMVariableNode() {variable=Scapula Tip Control}.Value, bInitial=RigVMVariableNode() {variable=Initial}.Value).Transform.Translation
    Primary.Target=&RigUnit_GetTransform(Item=RigVMVariableNode() {variable=Ball Connector}.Value, bInitial=RigVMVariableNode() {variable=Initial}.Value).Transform.Translation
   )
   GetAnimationChannelFromItem = &RigUnit_GetFloatAnimationChannelFromItem(Item=RigVMVariableNode() {variable=Shoulder Rotate Offset}.Value, bInitial=RigVMVariableNode() {variable=Initial}.Value) {template="GetAnimationChannelFromItem::Execute(out Value,in Item,in bInitial)"}
   n2 = RigVMVariableNode() {variable=Scapula Tip Control}
   n3 = RigVMVariableNode() {variable=Initial}
   n4 = &RigUnit_HierarchyGetParent(bDefaultParent=True, Child=RigVMVariableNode() {variable=Scapula Tip Control}.Value)
   n5 = RigVMVariableNode() {variable=Initial}
   RigVMFunction_MathQuaternionToEuler = &RigVMFunction_MathQuaternionToEuler(
    RotationOrder=ZYX
    Value=&RigVMFunction_MathTransformMakeRelative(
     Global=&RigUnit_GetTransform(Space=LocalSpace, Item=@n2.Value, bInitial=@n3.Value).Transform
     Parent=&RigUnit_GetTransform(Space=LocalSpace, Item=@n4.Parent, bInitial=@n5.Value).Transform
    ).Local.Rotation
   )
   Add = &RigVMFunction_MathDoubleAdd(B=@GetAnimationChannelFromItem.Value, A=@RigVMFunction_MathQuaternionToEuler.Result.X)
   RigVMFunction_MathQuaternionFromEuler = &RigVMFunction_MathQuaternionFromEuler(Euler=(X=0,Y=0,Z=0), Euler.Z=@Add.Result)
   > RigVMFunctionReturnNode(Aim Transform=&RigVMFunction_MathTransformMakeAbsolute(
     Local=(Rotation=(X=0,Y=0,Z=0,W=1),Translation=(X=0,Y=0,Z=0),Scale3D=(X=1,Y=1,Z=1))
     Parent=@n1.Result
     Local.Rotation=@RigVMFunction_MathQuaternionFromEuler.Result
    ) {template="Make Absolute::Execute(in Local,in Parent,out Global)"}.Global)
   RigVMVariableNode() {variable=Scapula Tip Aim Point}
  ]}
 New Function_ContainedGraph {nodes [
   n1 = RigVMFunctionEntryNode()
   RigVMFunction_Sequence = &RigVMFunction_Sequence(ExecuteContext=@n1.ExecuteContext)
   RigVMFunctionReturnNode(ExecuteContext=@RigVMFunction_Sequence.B)
   n2 = /Script/ControlRig.RigUnit_HierarchyAddAnimationChannelBool(
    InitialValue=false
    LimitsEnabled=()
    MaximumValue=True
    MinimumValue=false
    Name=Sec Control Visibility
    Parent=RigVMVariableNode() {variable=Ik Control}.Value
    ExecutePin=@RigVMFunction_Sequence.A
   ) {
    is=struct RigVMUnitNode
    template="SpawnAnimationChannel::Execute(in InitialValue,in MinimumValue,in MaximumValue,in LimitsEnabled,in Parent,in Name,out Item)"
   }
   Make_Array = RigVMDispatchNode(
    Values=((Type=None,Name="None"),(Type=None,Name="None"))
    Values.0=RigVMVariableNode() {variable=Ball Fk Control}.Value
    Values.1=RigVMVariableNode() {variable=Toe Fk Control}.Value
   ) {template="DISPATCH_RigVMDispatch_ArrayMake(in Values,out Array)"}
   > /Script/ControlRig.RigUnit_SetChannelHosts(Hosts=@Make_Array.Array, Channel=@n2.Item) {is=struct RigVMUnitNode}
   > RigVMVariableNode(Value=@n2.Item) {variable=Sec Control Visibility}
  ]}
 New Function_ContainedGraph {nodes [
   n1 = RigVMFunctionEntryNode()
   RigVMFunction_Sequence = &RigVMFunction_Sequence(ExecuteContext=@n1.ExecuteContext)
   RigVMFunctionReturnNode(ExecuteContext=@RigVMFunction_Sequence.B)
   GetAnimationChannelFromItem = /Script/ControlRig.RigUnit_GetBoolAnimationChannelFromItem(bInitial=False, Item=RigVMVariableNode() {variable=Sec Control Visibility}.Value) {is=struct RigVMUnitNode, template="GetAnimationChannelFromItem::Execute(out Value,in Item,in bInitial)"}
   Set_Module_Metadata = RigVMDispatchNode(Name=None, NameSpace=Self, Value=@GetAnimationChannelFromItem.Value, ExecuteContext=@RigVMFunction_Sequence.A) {template="DISPATCH_RigDispatch_SetModuleMetadata(in Name,in NameSpace,in Value,out Success)"}
   Make_Array = RigVMDispatchNode(
    Values=((Type=None,Name="None"),(Type=None,Name="None"))
    Values.0=RigVMVariableNode() {variable=Ball Fk Control}.Value
    Values.1=RigVMVariableNode() {variable=Toe Fk Control}.Value
   ) {template="DISPATCH_RigVMDispatch_ArrayMake(in Values,out Array)"}
   > For_Each = RigVMDispatchNode(Array=@Make_Array.Array) {template="DISPATCH_RigVMDispatch_ArrayIterator(in Array,out Element,out Index,out Count,out Ratio)"}
   > SetControlVisibility = /Script/ControlRig.RigUnit_SetControlVisibility(Item=@For_Each.Element, bVisible=@GetAnimationChannelFromItem.Value) {is=struct RigVMUnitNode, template="SetControlVisibility::Execute(in Item,in Pattern,in bVisible)"}
  ]}
 RigVMFunctionLibrary {nodes [
   Backwards Get Scapula Tip Aim Transform = RigVMCollapseNode(Initial=False)
   Backwards Scapula Tip to Ik Control System = RigVMCollapseNode(ExecuteContext=())
   Backwards Shoulder To Ball Ik System = RigVMCollapseNode(ExecuteContext=())
   Create Backward Local Transforms = RigVMCollapseNode(ExecuteContext=())
   Create Ball Rotate Control = RigVMCollapseNode(ExecuteContext=())
   Create Elbow Twist Control = RigVMCollapseNode(ExecuteContext=())
   Create Ik Control = RigVMCollapseNode(ExecuteContext=())
   Create Local Transforms = RigVMCollapseNode(ExecuteContext=())
   Create Name = RigVMCollapseNode(ExecuteContext=(), Tokens=())
   Create Rig Transforms = RigVMCollapseNode(ExecuteContext=())
   Create Rig Up Vector = RigVMCollapseNode(ExecuteContext=())
   Create Roll Controls = RigVMCollapseNode(ExecuteContext=())
   Create Scapula Tip Control = RigVMCollapseNode(ExecuteContext=())
   Create Scapula Tip Control_OOP = RigVMCollapseNode(ExecuteContext=())
   Create Sec Control Visibility Switch = RigVMCollapseNode(ExecuteContext=())
   Create Space Switches = RigVMCollapseNode(ExecuteContext=())
   Create Toe Controls = RigVMCollapseNode(ExecuteContext=())
   Debug Anim Sequence Bone Positions = RigVMCollapseNode(ExecuteContext=())
   Elbow Twist = RigVMCollapseNode(ExecuteContext=())
   Elbow to Ball Ik System = RigVMCollapseNode(ExecuteContext=())
   Get Ball Local Transform = RigVMCollapseNode(Input Transform=(Rotation=(X=0,Y=0,Z=0,W=1),Translation=(X=0,Y=0,Z=0),Scale3D=(X=1,Y=1,Z=1)))
   Get Connectors = RigVMCollapseNode(ExecuteContext=())
   Get Elbow Local Transform = RigVMCollapseNode(Input Transform=(Rotation=(X=0,Y=0,Z=0,W=1),Translation=(X=0,Y=0,Z=0),Scale3D=(X=1,Y=1,Z=1)))
   Get Ik Angles = RigVMCollapseNode()
   Get Ik Position = RigVMCollapseNode(ExecuteContext=())
   Get Lengths = RigVMCollapseNode(ExecuteContext=())
   Get Main Pole Vector = RigVMCollapseNode(Initial=False)
   Get Scapula Tip Aim Transform = RigVMCollapseNode()
   Get Scapula Tip Local Transform = RigVMCollapseNode(Input Transform=(Rotation=(X=0,Y=0,Z=0,W=1),Translation=(X=0,Y=0,Z=0),Scale3D=(X=1,Y=1,Z=1)))
   Get Shoulder Local Transform = RigVMCollapseNode(Input Transform=(Rotation=(X=0,Y=0,Z=0,W=1),Translation=(X=0,Y=0,Z=0),Scale3D=(X=1,Y=1,Z=1)))
   Get Wrist Local Transform = RigVMCollapseNode(Input Transform=(Rotation=(X=0,Y=0,Z=0,W=1),Translation=(X=0,Y=0,Z=0),Scale3D=(X=1,Y=1,Z=1)))
   Position Ball Fk Control = RigVMCollapseNode(ExecuteContext=())
   Position Ball Pivot Control = RigVMCollapseNode(ExecuteContext=())
   Position Ball Rotate Control = RigVMCollapseNode(ExecuteContext=())
   Position Elbow Rotation Control = RigVMCollapseNode(ExecuteContext=())
   Position Ik Control = RigVMCollapseNode(ExecuteContext=())
   Position Main Ik PV Control = RigVMCollapseNode(ExecuteContext=())
   Position Scapula Tip Control = RigVMCollapseNode(ExecuteContext=())
   Position Toe Fk Control = RigVMCollapseNode(ExecuteContext=())
   Scapula Tip to Elbow Ik System = RigVMCollapseNode(
    ExecuteContext=()
    Input Ball Transform=(Rotation=(X=0,Y=0,Z=0,W=1),Translation=(X=0,Y=0,Z=0),Scale3D=(X=1,Y=1,Z=1))
    Input Elbow Transform=(Rotation=(X=0,Y=0,Z=0,W=1),Translation=(X=0,Y=0,Z=0),Scale3D=(X=1,Y=1,Z=1))
    Input Scapula Tip Transform=(Rotation=(X=0,Y=0,Z=0,W=1),Translation=(X=0,Y=0,Z=0),Scale3D=(X=1,Y=1,Z=1))
    Input Shoulder Transform=(Rotation=(X=0,Y=0,Z=0,W=1),Translation=(X=0,Y=0,Z=0),Scale3D=(X=1,Y=1,Z=1))
    Input Wrist Transform=(Rotation=(X=0,Y=0,Z=0,W=1),Translation=(X=0,Y=0,Z=0),Scale3D=(X=1,Y=1,Z=1))
   )
   Scapula Tip to Ik Control System = RigVMCollapseNode(ExecuteContext=())
   Scapula Tip to Shoulder System = RigVMCollapseNode(ExecuteContext=())
   Set Bone Transforms = RigVMCollapseNode(ExecuteContext=())
   Set Control Lines = RigVMCollapseNode(ExecuteContext=())
   Set Fk Toes = RigVMCollapseNode(ExecuteContext=())
   Set Metadata = RigVMCollapseNode(ExecuteContext=())
   Set Sec Control Visibility = RigVMCollapseNode(ExecuteContext=())
   Set Side Color = RigVMCollapseNode(ExecuteContext=())
   Shoulder to Ball Ik System = RigVMCollapseNode(ExecuteContext=())
   Shoulder to Wrist Ik System = RigVMCollapseNode(ExecuteContext=())
   Wrist Offset ik System = RigVMCollapseNode(ExecuteContext=())
  ]}
 RigVMFunction_Sequence_1_ContainedGraph {nodes [
   n1 = RigVMFunctionEntryNode()
   RigVMFunction_Sequence = &RigVMFunction_Sequence(ExecuteContext=@n1.ExecuteContext)
   n2 = &RigVMFunction_Sequence(ExecuteContext=@RigVMFunction_Sequence.B)
   n3 = &RigVMFunction_Sequence(ExecuteContext=@n2.B)
   n4 = &RigVMFunction_Sequence(ExecuteContext=@n3.B)
   n5 = &RigVMFunction_Sequence(ExecuteContext=@n4.B)
   RigVMFunctionReturnNode(A=@RigVMFunction_Sequence.A, B=@n2.A, C=@n3.A, D=@n4.A, E=@n5.A, F=@n5.B)
  ]}
 RigVMFunction_Sequence_1_ContainedGraph {nodes [
   n1 = RigVMFunctionEntryNode()
   RigVMFunction_Sequence = &RigVMFunction_Sequence(ExecuteContext=@n1.ExecuteContext)
   n2 = &RigVMFunction_Sequence(ExecuteContext=@RigVMFunction_Sequence.B)
   n3 = &RigVMFunction_Sequence(ExecuteContext=@n2.B)
   n4 = &RigVMFunction_Sequence(ExecuteContext=@n3.B)
   RigVMFunctionReturnNode(A=@RigVMFunction_Sequence.A, B=@n2.A, C=@n3.A, D=@n4.A, E=@n4.B)
  ]}
 RigVMFunction_Sequence_1_ContainedGraph {nodes [
   n1 = RigVMFunctionEntryNode()
   RigVMFunction_Sequence = &RigVMFunction_Sequence(ExecuteContext=@n1.ExecuteContext)
   n2 = &RigVMFunction_Sequence(ExecuteContext=@RigVMFunction_Sequence.B)
   n3 = &RigVMFunction_Sequence(ExecuteContext=@n2.B)
   n4 = &RigVMFunction_Sequence(ExecuteContext=@n3.B)
   n5 = &RigVMFunction_Sequence(ExecuteContext=@n4.B)
   RigVMFunctionReturnNode(A=@RigVMFunction_Sequence.A, B=@n2.A, C=@n3.A, D=@n4.A, E=@n5.A, F=@n5.B)
  ]}
 RigVMFunction_Sequence_2_ContainedGraph {nodes [
   n1 = RigVMFunctionEntryNode()
   n2 = &RigVMFunction_Sequence(ExecuteContext=@n1.ExecuteContext)
   n3 = &RigVMFunction_Sequence(ExecuteContext=@n2.B)
   RigVMFunctionReturnNode(A=@n2.A, B=@n3.A, C=@n3.B)
  ]}
 RigVMFunction_Sequence_2_ContainedGraph {nodes [
   n1 = RigVMFunctionEntryNode()
   n2 = &RigVMFunction_Sequence(ExecuteContext=@n1.ExecuteContext)
   n3 = &RigVMFunction_Sequence(ExecuteContext=@n2.B)
   n4 = &RigVMFunction_Sequence(ExecuteContext=@n3.B)
   n5 = &RigVMFunction_Sequence(ExecuteContext=@n4.B)
   n6 = &RigVMFunction_Sequence(ExecuteContext=@n5.B)
   RigVMFunctionReturnNode(A=@n2.A, B=@n3.A, C=@n4.A, D=@n5.A, E=@n6.A, F=@n6.B)
  ]}
 RigVMFunction_Sequence_2_ContainedGraph {nodes [
   n1 = RigVMFunctionEntryNode()
   RigVMFunction_Sequence = &RigVMFunction_Sequence(ExecuteContext=@n1.ExecuteContext)
   n2 = &RigVMFunction_Sequence(ExecuteContext=@RigVMFunction_Sequence.B)
   RigVMFunctionReturnNode(A=@RigVMFunction_Sequence.A, B=@n2.A, C=@n2.B)
  ]}
 RigVMFunction_Sequence_2_ContainedGraph {nodes [
   n1 = RigVMFunctionEntryNode()
   n2 = &RigVMFunction_Sequence(ExecuteContext=@n1.ExecuteContext)
   n3 = &RigVMFunction_Sequence(ExecuteContext=@n2.B)
   n4 = &RigVMFunction_Sequence(ExecuteContext=@n3.B)
   n5 = &RigVMFunction_Sequence(ExecuteContext=@n4.B)
   n6 = &RigVMFunction_Sequence(ExecuteContext=@n5.B)
   RigVMFunctionReturnNode(A=@n2.A, B=@n3.A, C=@n4.A, D=@n5.A, E=@n6.A, F=@n6.B)
  ]}
 RigVMFunction_Sequence_2_ContainedGraph {nodes [
   n1 = RigVMFunctionEntryNode()
   RigVMFunction_Sequence = &RigVMFunction_Sequence(ExecuteContext=@n1.ExecuteContext)
   n2 = &RigVMFunction_Sequence(ExecuteContext=@RigVMFunction_Sequence.B)
   RigVMFunctionReturnNode(A=@RigVMFunction_Sequence.A, B=@n2.A, C=@n2.B)
  ]}
 RigVMFunction_Sequence_2_ContainedGraph {nodes [
   n1 = RigVMFunctionEntryNode()
   n2 = &RigVMFunction_Sequence(ExecuteContext=@n1.ExecuteContext)
   n3 = &RigVMFunction_Sequence(ExecuteContext=@n2.B)
   RigVMFunctionReturnNode(A=@n2.A, B=@n3.A, C=@n3.B)
  ]}
 RigVMFunction_Sequence_2_ContainedGraph {nodes [
   n1 = RigVMFunctionEntryNode()
   n2 = &RigVMFunction_Sequence(ExecuteContext=@n1.ExecuteContext)
   n3 = &RigVMFunction_Sequence(ExecuteContext=@n2.B)
   n4 = &RigVMFunction_Sequence(ExecuteContext=@n3.B)
   n5 = &RigVMFunction_Sequence(ExecuteContext=@n4.B)
   n6 = &RigVMFunction_Sequence(ExecuteContext=@n5.B)
   RigVMFunctionReturnNode(A=@n2.A, B=@n3.A, C=@n4.A, D=@n5.A, E=@n6.A, F=@n6.B)
  ]}
 RigVMFunction_Sequence_2_ContainedGraph {nodes [
   n1 = RigVMFunctionEntryNode()
   n2 = &RigVMFunction_Sequence(ExecuteContext=@n1.ExecuteContext)
   n3 = &RigVMFunction_Sequence(ExecuteContext=@n2.B)
   n4 = &RigVMFunction_Sequence(ExecuteContext=@n3.B)
   n5 = &RigVMFunction_Sequence(ExecuteContext=@n4.B)
   n6 = &RigVMFunction_Sequence(ExecuteContext=@n5.B)
   RigVMFunctionReturnNode(A=@n2.A, B=@n3.A, C=@n4.A, D=@n5.A, E=@n6.A, F=@n6.B)
  ]}
 RigVMFunction_Sequence_2_ContainedGraph {nodes [
   n1 = RigVMFunctionEntryNode()
   RigVMFunction_Sequence = &RigVMFunction_Sequence(ExecuteContext=@n1.ExecuteContext)
   n2 = &RigVMFunction_Sequence(ExecuteContext=@RigVMFunction_Sequence.B)
   RigVMFunctionReturnNode(A=@RigVMFunction_Sequence.A, B=@n2.A, C=@n2.B)
  ]}
 RigVMFunction_Sequence_2_ContainedGraph {nodes [
   n1 = RigVMFunctionEntryNode()
   n2 = &RigVMFunction_Sequence(ExecuteContext=@n1.ExecuteContext)
   n3 = &RigVMFunction_Sequence(ExecuteContext=@n2.B)
   n4 = &RigVMFunction_Sequence(ExecuteContext=@n3.B)
   n5 = &RigVMFunction_Sequence(ExecuteContext=@n4.B)
   n6 = &RigVMFunction_Sequence(ExecuteContext=@n5.B)
   n7 = &RigVMFunction_Sequence(ExecuteContext=@n6.B)
   n8 = &RigVMFunction_Sequence(ExecuteContext=@n7.B)
   RigVMFunctionReturnNode(A=@n2.A, B=@n3.A, C=@n4.A, D=@n5.A, E=@n6.A, F=@n7.A, G=@n8.A, H=@n8.B)
  ]}
 RigVMFunction_Sequence_2_ContainedGraph {nodes [
   n1 = RigVMFunctionEntryNode()
   n2 = &RigVMFunction_Sequence(ExecuteContext=@n1.ExecuteContext)
   n3 = &RigVMFunction_Sequence(ExecuteContext=@n2.B)
   RigVMFunctionReturnNode(A=@n2.A, B=@n3.A, C=@n3.B)
  ]}
 RigVMFunction_Sequence_2_ContainedGraph {nodes [
   n1 = RigVMFunctionEntryNode()
   n2 = &RigVMFunction_Sequence(ExecuteContext=@n1.ExecuteContext)
   n3 = &RigVMFunction_Sequence(ExecuteContext=@n2.B)
   n4 = &RigVMFunction_Sequence(ExecuteContext=@n3.B)
   n5 = &RigVMFunction_Sequence(ExecuteContext=@n4.B)
   n6 = &RigVMFunction_Sequence(ExecuteContext=@n5.B)
   n7 = &RigVMFunction_Sequence(ExecuteContext=@n6.B)
   n8 = &RigVMFunction_Sequence(ExecuteContext=@n7.B)
   RigVMFunctionReturnNode(A=@n2.A, B=@n3.A, C=@n4.A, D=@n5.A, E=@n6.A, F=@n7.A, G=@n8.A, H=@n8.B)
  ]}
 RigVMFunction_Sequence_2_ContainedGraph {nodes [
   n1 = RigVMFunctionEntryNode()
   n2 = &RigVMFunction_Sequence(ExecuteContext=@n1.ExecuteContext)
   n3 = &RigVMFunction_Sequence(ExecuteContext=@n2.B)
   RigVMFunctionReturnNode(A=@n2.A, B=@n3.A, C=@n3.B)
  ]}
 RigVMFunction_Sequence_2_ContainedGraph {nodes [
   n1 = RigVMFunctionEntryNode()
   n2 = &RigVMFunction_Sequence(ExecuteContext=@n1.ExecuteContext)
   n3 = &RigVMFunction_Sequence(ExecuteContext=@n2.B)
   RigVMFunctionReturnNode(A=@n2.A, B=@n3.A, C=@n3.B)
  ]}
 RigVMFunction_Sequence_2_ContainedGraph {nodes [
   n1 = RigVMFunctionEntryNode()
   n2 = &RigVMFunction_Sequence(ExecuteContext=@n1.ExecuteContext)
   n3 = &RigVMFunction_Sequence(ExecuteContext=@n2.B)
   n4 = &RigVMFunction_Sequence(ExecuteContext=@n3.B)
   n5 = &RigVMFunction_Sequence(ExecuteContext=@n4.B)
   n6 = &RigVMFunction_Sequence(ExecuteContext=@n5.B)
   RigVMFunctionReturnNode(A=@n2.A, B=@n3.A, C=@n4.A, D=@n5.A, E=@n6.A, F=@n6.B)
  ]}
 RigVMFunction_Sequence_2_ContainedGraph {nodes [
   n1 = RigVMFunctionEntryNode()
   n2 = &RigVMFunction_Sequence(ExecuteContext=@n1.ExecuteContext)
   n3 = &RigVMFunction_Sequence(ExecuteContext=@n2.B)
   n4 = &RigVMFunction_Sequence(ExecuteContext=@n3.B)
   n5 = &RigVMFunction_Sequence(ExecuteContext=@n4.B)
   RigVMFunctionReturnNode(A=@n2.A, B=@n3.A, C=@n4.A, D=@n5.A, E=@n5.B)
  ]}
 RigVMFunction_Sequence_2_ContainedGraph {nodes [
   n1 = RigVMFunctionEntryNode()
   n2 = &RigVMFunction_Sequence(ExecuteContext=@n1.ExecuteContext)
   n3 = &RigVMFunction_Sequence(ExecuteContext=@n2.B)
   n4 = &RigVMFunction_Sequence(ExecuteContext=@n3.B)
   RigVMFunctionReturnNode(A=@n2.A, B=@n3.A, C=@n4.A, D=@n4.B)
  ]}
 RigVMFunction_Sequence_2_ContainedGraph {nodes [
   n1 = RigVMFunctionEntryNode()
   n2 = &RigVMFunction_Sequence(ExecuteContext=@n1.ExecuteContext)
   n3 = &RigVMFunction_Sequence(ExecuteContext=@n2.B)
   n4 = &RigVMFunction_Sequence(ExecuteContext=@n3.B)
   RigVMFunctionReturnNode(A=@n2.A, B=@n3.A, C=@n4.A, D=@n4.B)
  ]}
 RigVMFunction_Sequence_2_ContainedGraph {nodes [
   n1 = RigVMFunctionEntryNode()
   n2 = &RigVMFunction_Sequence(ExecuteContext=@n1.ExecuteContext)
   n3 = &RigVMFunction_Sequence(ExecuteContext=@n2.B)
   n4 = &RigVMFunction_Sequence(ExecuteContext=@n3.B)
   n5 = &RigVMFunction_Sequence(ExecuteContext=@n4.B)
   RigVMFunctionReturnNode(A=@n2.A, B=@n3.A, C=@n4.A, D=@n5.A, E=@n5.B)
  ]}
 RigVMFunction_Sequence_2_ContainedGraph {nodes [
   n1 = RigVMFunctionEntryNode()
   n2 = &RigVMFunction_Sequence(ExecuteContext=@n1.ExecuteContext)
   n3 = &RigVMFunction_Sequence(ExecuteContext=@n2.B)
   n4 = &RigVMFunction_Sequence(ExecuteContext=@n3.B)
   n5 = &RigVMFunction_Sequence(ExecuteContext=@n4.B)
   RigVMFunctionReturnNode(A=@n2.A, B=@n3.A, C=@n4.A, D=@n5.A, E=@n5.B)
  ]}
 RigVMFunction_Sequence_2_ContainedGraph {nodes [
   n1 = RigVMFunctionEntryNode()
   n2 = &RigVMFunction_Sequence(ExecuteContext=@n1.ExecuteContext)
   n3 = &RigVMFunction_Sequence(ExecuteContext=@n2.B)
   RigVMFunctionReturnNode(A=@n2.A, B=@n3.A, C=@n3.B)
  ]}
 RigVMFunction_Sequence_2_ContainedGraph {nodes [
   n1 = RigVMFunctionEntryNode()
   n2 = &RigVMFunction_Sequence(ExecuteContext=@n1.ExecuteContext)
   n3 = &RigVMFunction_Sequence(ExecuteContext=@n2.B)
   n4 = &RigVMFunction_Sequence(ExecuteContext=@n3.B)
   n5 = &RigVMFunction_Sequence(ExecuteContext=@n4.B)
   n6 = &RigVMFunction_Sequence(ExecuteContext=@n5.B)
   n7 = &RigVMFunction_Sequence(ExecuteContext=@n6.B)
   n8 = &RigVMFunction_Sequence(ExecuteContext=@n7.B)
   RigVMFunctionReturnNode(A=@n2.A, B=@n3.A, C=@n4.A, D=@n5.A, E=@n6.A, F=@n7.A, G=@n8.A, H=@n8.B)
  ]}
 RigVMFunction_Sequence_2_ContainedGraph {nodes [
   n1 = RigVMFunctionEntryNode()
   n2 = &RigVMFunction_Sequence(ExecuteContext=@n1.ExecuteContext)
   n3 = &RigVMFunction_Sequence(ExecuteContext=@n2.B)
   RigVMFunctionReturnNode(A=@n2.A, B=@n3.A, C=@n3.B)
  ]}
 RigVMFunction_Sequence_3_ContainedGraph {nodes [
   n1 = RigVMFunctionEntryNode()
   n2 = &RigVMFunction_Sequence(ExecuteContext=@n1.ExecuteContext)
   n3 = &RigVMFunction_Sequence(ExecuteContext=@n2.B)
   RigVMFunctionReturnNode(B=@n3.A, C=@n3.B, A=@n2.A)
  ]}
 RigVMFunction_Sequence_3_ContainedGraph {nodes [
   n1 = RigVMFunctionEntryNode()
   n2 = &RigVMFunction_Sequence(ExecuteContext=@n1.ExecuteContext)
   n3 = &RigVMFunction_Sequence(ExecuteContext=@n2.B)
   n4 = &RigVMFunction_Sequence(ExecuteContext=@n3.B)
   n5 = &RigVMFunction_Sequence(ExecuteContext=@n4.B)
   n6 = &RigVMFunction_Sequence(ExecuteContext=@n5.B)
   RigVMFunctionReturnNode(B=@n3.A, A=@n2.A, C=@n4.A, D=@n5.A, E=@n6.A, F=@n6.B)
  ]}
 RigVMFunction_Sequence_3_ContainedGraph {nodes [
   n1 = RigVMFunctionEntryNode()
   n2 = &RigVMFunction_Sequence(ExecuteContext=@n1.ExecuteContext)
   n3 = &RigVMFunction_Sequence(ExecuteContext=@n2.B)
   RigVMFunctionReturnNode(B=@n3.A, C=@n3.B, A=@n2.A)
  ]}
 RigVMFunction_Sequence_3_ContainedGraph {nodes [
   n1 = RigVMFunctionEntryNode()
   n2 = &RigVMFunction_Sequence(ExecuteContext=@n1.ExecuteContext)
   n3 = &RigVMFunction_Sequence(ExecuteContext=@n2.B)
   n4 = &RigVMFunction_Sequence(ExecuteContext=@n3.B)
   n5 = &RigVMFunction_Sequence(ExecuteContext=@n4.B)
   n6 = &RigVMFunction_Sequence(ExecuteContext=@n5.B)
   RigVMFunctionReturnNode(B=@n3.A, A=@n2.A, C=@n4.A, D=@n5.A, E=@n6.A, F=@n6.B)
  ]}
 RigVMModel {nodes [
   RigUnit_BeginExecution = &RigUnit_BeginExecution()
   RigVMFunction_Sequence = RigVMAggregateNode(ExecuteContext=@RigUnit_BeginExecution.ExecutePin)
   Scapula Tip to Ik Control System = RigVMFunctionReferenceNode(ExecuteContext=@RigVMFunction_Sequence.A)
   Shoulder to Ball Ik System = RigVMFunctionReferenceNode(ExecuteContext=@RigVMFunction_Sequence.B)
   > Wrist Offset ik System = RigVMFunctionReferenceNode()
   Elbow Twist = RigVMFunctionReferenceNode(ExecuteContext=@RigVMFunction_Sequence.C)
   Set Fk Toes = RigVMFunctionReferenceNode(ExecuteContext=@RigVMFunction_Sequence.D)
   Set Bone Transforms = RigVMFunctionReferenceNode(ExecuteContext=@RigVMFunction_Sequence.E)
   Set Control Lines = RigVMFunctionReferenceNode(ExecuteContext=@RigVMFunction_Sequence.F)
   > Set Sec Control Visibility = RigVMFunctionReferenceNode()
  ]}
 RigVMModel Rig {nodes [
   n1 = /Script/ControlRig.RigUnit_InverseExecution() {is=struct RigVMUnitNode}
   RigVMFunction_Sequence = RigVMAggregateNode(ExecuteContext=@n1.ExecutePin)
   Debug Anim Sequence Bone Positions = RigVMFunctionReferenceNode(ExecuteContext=@RigVMFunction_Sequence.A)
   Position Ik Control = RigVMFunctionReferenceNode(ExecuteContext=@RigVMFunction_Sequence.B)
   > Position Scapula Tip Control = RigVMFunctionReferenceNode()
   Backwards Scapula Tip to Ik Control System = RigVMFunctionReferenceNode(ExecuteContext=@RigVMFunction_Sequence.C)
   > Backwards Shoulder To Ball Ik System = RigVMFunctionReferenceNode()
   Position Elbow Rotation Control = RigVMFunctionReferenceNode(ExecuteContext=@RigVMFunction_Sequence.D)
   Position Ball Rotate Control = RigVMFunctionReferenceNode(ExecuteContext=@RigVMFunction_Sequence.E)
   Position Ball Fk Control = RigVMFunctionReferenceNode(ExecuteContext=@RigVMFunction_Sequence.F)
   > Position Toe Fk Control = RigVMFunctionReferenceNode()
   Position Main Ik PV Control = RigVMFunctionReferenceNode(ExecuteContext=())
   RigVMRerouteNode(Value=RigVMRerouteNode(Value=()).Value)
   n2 = &RigUnit_HierarchyGetParent(bDefaultParent=True, Child=RigVMVariableNode() {variable=Shoulder Connector}.Value)
   n3 = RigVMVariableNode() {variable=Scapula Tip Connector}
   n4 = RigVMDispatchNode(
    True=1
    False=RigVMDispatchNode(False=1, True=-1, Condition=RigVMVariableNode() {variable=Right Side}.Value) {template="DISPATCH_RigVMDispatch_If(in Condition,in True,in False,out Result)"}.Result
    Condition=RigVMVariableNode() {variable=Left Side}.Value
   ) {template="DISPATCH_RigVMDispatch_If(in Condition,in True,in False,out Result)"}
   Add = &RigVMFunction_MathVectorAdd(
    A=&RigVMFunction_MathTransformMakeRelative(
     Parent=&RigUnit_GetTransform(bInitial=False, Item=@n2.Parent).Transform
     Global=&RigUnit_GetTransform(bInitial=False, Item=@n3.Value).Transform
    ).Local.Translation
    B=&RigVMFunction_MathVectorMul(B=(X=1,Y=1,Z=1), B.X=@n4.Result, B.Y=@n4.Result, B.Z=@n4.Result, A=RigVMVariableNode() {variable=Rig Up Vector}.Value) {template="Multiply::Execute(in A,in B,out Result)"}.Result
   ) {template="Add::Execute(in A,in B,out Result)"}
   n5 = RigVMVariableNode() {variable=Scapula Tip Connector}
   Set Transform_3 = &RigUnit_SetTranslation(
    ExecutePin=()
    Space=GlobalSpace
    Weight=1
    bInitial=False
    bPropagateToChildren=True
    Value=&RigVMFunction_MathTransformMakeAbsolute(
     Local=(Rotation=(X=0,Y=0,Z=0,W=1),Translation=(X=0,Y=0,Z=0),Scale3D=(X=1,Y=1,Z=1))
     Local.Translation=@Add.Result
     Parent=&RigUnit_GetTransform(bInitial=False, Item=@n5.Value).Transform
    ) {template="Make Absolute::Execute(in Local,in Parent,out Global)"}.Global.Translation
    Item=RigVMVariableNode() {variable=Main PV Control}.Value
   ) {template="Set Transform::Execute(in Item,in Space,in bInitial,in Value,in Weight,in bPropagateToChildren)"}
   RigVMVariableNode() {variable=Scapula Tip Control}
  ]}
]
hierarchy [
 Shoulder {type=Connector, loc=<0 0 0>, rot=<0 0 0 1>}
 Root {type=Socket, loc=<0 0 0>, rot=<0 0 0 1>}
 Scapula Tip {type=Connector, loc=<0 0 0>, rot=<0 0 0 1>}
 Elbow {type=Connector, loc=<0 0 0>, rot=<0 0 0 1>}
 Wrist {type=Connector, loc=<0 0 0>, rot=<0 0 0 1>}
 Ball {type=Connector, loc=<0 0 0>, rot=<0 0 0 1>}
 Toe {type=Connector, loc=<0 0 0>, rot=<0 0 0 1>}
 Toe Tip {type=Connector, loc=<0 0 0>, rot=<0 0 0 1>}
 IK Front Connector {type=Connector, loc=<0 0 0>, rot=<0 0 0 1>}
]
variables [
 Back Leg {type=bool, default=False}
 Backwards Main Pole Vector {type=FVector, default=(X=0,Y=0,Z=0)}
 Ball Backwards Transform {type=FTransform, default=(Rotation=(X=0,Y=0,Z=0,W=1),Translation=(X=0,Y=0,Z=0),Scale3D=(X=1,Y=1,Z=1))}
 Ball Connector {type=FRigElementKey, default=(Type=Connector,Name="")}
 Ball Fk Control {type=FRigElementKey, default=(Type=Bone,Name="")}
 Ball Input Transform {type=FTransform, default=(Rotation=(X=0,Y=0,Z=0,W=1),Translation=(X=0,Y=0,Z=0),Scale3D=(X=1,Y=1,Z=1))}
 Ball Length {type=double, default=0}
 Ball Local Transform {type=FTransform, default=(Rotation=(X=0,Y=0,Z=0,W=1),Translation=(X=0,Y=0,Z=0),Scale3D=(X=1,Y=1,Z=1))}
 Ball Pivot Control {type=FRigElementKey, default=(Type=Bone,Name="")}
 Ball Point Ref {type=FRigElementKey, default=(Type=Bone,Name="")}
 Ball Rig Transform {type=FTransform, default=(Rotation=(X=0,Y=0,Z=0,W=1),Translation=(X=0,Y=0,Z=0),Scale3D=(X=1,Y=1,Z=1))}
 Ball Rotate Aim Null {type=FRigElementKey, default=(Type=Bone,Name="")}
 Ball Rotate Channel {type=FRigElementKey, default=(Type=Bone,Name="")}
 Ball Rotate Control {type=FRigElementKey, default=(Type=Bone,Name="")}
 Control Scale {type=double, default=1}
 Custom Side {type=bool, default=False}
 Custom Side Color {type=FLinearColor, default=(R=1,G=0,B=1,A=1)}
 Custom Side Name {type=FString, default=ZZ}
 Debug {type=bool, default=False}
 Elbow Backwards Transform {type=FTransform, default=(Rotation=(X=0,Y=0,Z=0,W=1),Translation=(X=0,Y=0,Z=0),Scale3D=(X=1,Y=1,Z=1))}
 Elbow Connector {type=FRigElementKey, default=(Type=Connector,Name="")}
 Elbow Input Transform {type=FTransform, default=(Rotation=(X=0,Y=0,Z=0,W=1),Translation=(X=0,Y=0,Z=0),Scale3D=(X=1,Y=1,Z=1))}
 Elbow Length {type=float, default=0}
 Elbow Local Transform {type=FTransform, default=(Rotation=(X=0,Y=0,Z=0,W=1),Translation=(X=0,Y=0,Z=0),Scale3D=(X=1,Y=1,Z=1))}
 Elbow PV {type=FRigElementKey, default=(Type=Bone,Name="")}
 Elbow Rig Transform {type=FTransform, default=(Rotation=(X=0,Y=0,Z=0,W=1),Translation=(X=0,Y=0,Z=0),Scale3D=(X=1,Y=1,Z=1))}
 Elbow Rotate Control {type=FRigElementKey, default=(Type=Bone,Name="")}
 Elbow Rotate Control Null {type=FRigElementKey, default=(Type=Bone,Name="")}
 Elbow Rotate Up Vector {type=FRigElementKey, default=(Type=Bone,Name="")}
 Front Leg {type=bool, default=True}
 Heel Roll Control {type=FRigElementKey, default=(Type=Bone,Name="")}
 Heel Roll Initial Transform {type=FTransform, default=(Rotation=(X=0,Y=0,Z=0,W=1),Translation=(X=0,Y=-9,Z=0),Scale3D=(X=1,Y=1,Z=1))}
 IK Control Transform {type=FTransform, default=(Rotation=(X=0,Y=0,Z=1,W=0),Translation=(X=0.7,Y=2,Z=-2.7),Scale3D=(X=0.5,Y=0.5,Z=0.5))}
 IK Front Connector {type=FRigElementKey, default=(Type=None,Name="")}
 Ik Control {type=FRigElementKey, default=(Type=None,Name="")}
 Ik Point {type=FRigElementKey, default=(Type=Bone,Name="")}
 Left Side {type=bool, default=True}
 Left Side Color {type=FLinearColor, default=(R=0,G=0,B=1,A=1)}
 Main PV Control {type=FRigElementKey, default=(Type=Bone,Name="")}
 PV Spaces {type="TArray<FRigElementKey>", default=(())}
 Rig Up Vector {type=FVector, default=(X=30,Y=0,Z=0)}
 Right Side {type=bool, default=False}
 Right Side Color {type=FLinearColor, default=(R=1,G=0,B=0,A=1)}
 Scapula Aim Point Backwards Transform {type=FTransform, default=(Rotation=(X=0,Y=0,Z=0,W=1),Translation=(X=0,Y=0,Z=0),Scale3D=(X=1,Y=1,Z=1))}
 Scapula Length {type=float, default=0}
 Scapula Rig Transform {type=FTransform, default=(Rotation=(X=0,Y=0,Z=0,W=1),Translation=(X=0,Y=0,Z=0),Scale3D=(X=1,Y=1,Z=1))}
 Scapula Tip Aim Percent {type=double, default=0.5}
 Scapula Tip Aim Point {type=FRigElementKey, default=(Type=Bone,Name="")}
 Scapula Tip Backwards Transform {type=FTransform, default=(Rotation=(X=0,Y=0,Z=0,W=1),Translation=(X=0,Y=0,Z=0),Scale3D=(X=1,Y=1,Z=1))}
 Scapula Tip Connector {type=FRigElementKey, default=(Type=Connector,Name="")}
 Scapula Tip Control {type=FRigElementKey, default=(Type=Bone,Name="")}
 Scapula Tip Control Local Transform {type=FTransform, default=(Rotation=(X=0,Y=0,Z=0,W=1),Translation=(X=0,Y=0,Z=0),Scale3D=(X=1,Y=1,Z=1))}
 Scapula Tip Input Transform {type=FTransform, default=(Rotation=(X=0,Y=0,Z=0,W=1),Translation=(X=0,Y=0,Z=0),Scale3D=(X=1,Y=1,Z=1))}
 Scapula Tip Local Transform {type=FTransform, default=(Rotation=(X=0,Y=0,Z=0,W=1),Translation=(X=0,Y=0,Z=0),Scale3D=(X=1,Y=1,Z=1))}
 Scapula Tip Offset Transform {type=FTransform, default=(Rotation=(X=0,Y=0,Z=0,W=1),Translation=(X=0,Y=0,Z=0),Scale3D=(X=1,Y=1,Z=1))}
 Sec Control Visibility {type=FRigElementKey, default=(Type=None,Name="")}
 Shoulder Backwards Transform {type=FTransform, default=(Rotation=(X=0,Y=0,Z=0,W=1),Translation=(X=0,Y=0,Z=0),Scale3D=(X=1,Y=1,Z=1))}
 Shoulder Connector {type=FRigElementKey, default=(Type=Connector,Name="")}
 Shoulder Input Transform {type=FTransform, default=(Rotation=(X=0,Y=0,Z=0,W=1),Translation=(X=0,Y=0,Z=0),Scale3D=(X=1,Y=1,Z=1))}
 Shoulder Length {type=float, default=0}
 Shoulder Local Transform {type=FTransform, default=(Rotation=(X=0,Y=0,Z=0,W=1),Translation=(X=0,Y=0,Z=0),Scale3D=(X=1,Y=1,Z=1))}
 Shoulder Rig Transform {type=FTransform, default=(Rotation=(X=0,Y=0,Z=0,W=1),Translation=(X=0,Y=0,Z=0),Scale3D=(X=1,Y=1,Z=1))}
 Shoulder Rotate Offset {type=FRigElementKey, default=(Type=Bone,Name="")}
 Toe Backwards Transform {type=FTransform, default=(Rotation=(X=0,Y=0,Z=0,W=1),Translation=(X=0,Y=0,Z=0),Scale3D=(X=1,Y=1,Z=1))}
 Toe Connector {type=FRigElementKey, default=(Type=Connector,Name="")}
 Toe Fk Control {type=FRigElementKey, default=(Type=Bone,Name="")}
 Toe Input Transform {type=FTransform, default=(Rotation=(X=0,Y=0,Z=0,W=1),Translation=(X=0,Y=0,Z=0),Scale3D=(X=1,Y=1,Z=1))}
 Toe Length {type=double, default=0}
 Toe Local Transform {type=FTransform, default=(Rotation=(X=0,Y=0,Z=0,W=1),Translation=(X=0,Y=0,Z=0),Scale3D=(X=1,Y=1,Z=1))}
 Toe Rig Transform {type=FTransform, default=(Rotation=(X=0,Y=0,Z=0,W=1),Translation=(X=0,Y=0,Z=0),Scale3D=(X=1,Y=1,Z=1))}
 Toe Roll Control {type=FRigElementKey, default=(Type=Bone,Name="")}
 Toe Roll Initial Transform {type=FTransform, default=(Rotation=(X=0,Y=0,Z=0,W=1),Translation=(X=0,Y=7,Z=-3),Scale3D=(X=1,Y=1,Z=1))}
 Toe Tip Backwards Transform {type=FTransform, default=(Rotation=(X=0,Y=0,Z=0,W=1),Translation=(X=0,Y=0,Z=0),Scale3D=(X=1,Y=1,Z=1))}
 Toe Tip Connector {type=FRigElementKey, default=(Type=Connector,Name="")}
 Toe Tip Input Transform {type=FTransform, default=(Rotation=(X=0,Y=0,Z=0,W=1),Translation=(X=0,Y=0,Z=0),Scale3D=(X=1,Y=1,Z=1))}
 Toe Tip Point {type=FRigElementKey, default=(Type=Bone,Name="")}
 Toe Tip Rig Transform {type=FTransform, default=(Rotation=(X=0,Y=0,Z=0,W=1),Translation=(X=0,Y=0,Z=0),Scale3D=(X=1,Y=1,Z=1))}
 Wrist Backwards Transform {type=FTransform, default=(Rotation=(X=0,Y=0,Z=0,W=1),Translation=(X=0,Y=0,Z=0),Scale3D=(X=1,Y=1,Z=1))}
 Wrist Connector {type=FRigElementKey, default=(Type=Connector,Name="")}
 Wrist Ik Point {type=FRigElementKey, default=(Type=Bone,Name="")}
 Wrist Input Transform {type=FTransform, default=(Rotation=(X=0,Y=0,Z=0,W=1),Translation=(X=0,Y=0,Z=0),Scale3D=(X=1,Y=1,Z=1))}
 Wrist Length {type=float, default=0}
 Wrist Local Transform {type=FTransform, default=(Rotation=(X=0,Y=0,Z=0,W=1),Translation=(X=0,Y=0,Z=0),Scale3D=(X=1,Y=1,Z=1))}
 Wrist Rig Transform {type=FTransform, default=(Rotation=(X=0,Y=0,Z=0,W=1),Translation=(X=0,Y=0,Z=0),Scale3D=(X=1,Y=1,Z=1))}
]
