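# Modular Control Rig tail module that builds an FK rig for the bone chain between the TailBase and TailTip connectors.
# Construction generates aim-oriented rig joints, tapered alternating-color FK controls, gimbal and offset controls, and space switches.
# Offset controls drive rig joints, then tail bones, through stored local-transform metadata; reverse functions snap controls back.
# A Tail Rotate proxy adds interactive rotation deltas across controls selected by position and spread channels, with optional debug-line feedback.
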ControlRigBlueprint /WhisperRemedyAssets/Assets/Modules/CRM_Epic_WhisperRemedy_Tail_v01
graphs [
 ConstructionGraph {nodes [
   RigUnit_PrepareForExecution = &RigUnit_PrepareForExecution()
   RigVMFunction_Sequence = RigVMAggregateNode(ExecuteContext=@RigUnit_PrepareForExecution.ExecutePin)
   find_tail_chain = RigVMFunctionReferenceNode(ExecuteContext=@RigVMFunction_Sequence.B)
   create_chain_fk_joints = RigVMFunctionReferenceNode(ExecuteContext=@RigVMFunction_Sequence.C)
   > get_fk_joint_local_transforms = RigVMFunctionReferenceNode()
   create gimal control = RigVMFunctionReferenceNode(ExecuteContext=@RigVMFunction_Sequence.D)
   > create_chain_fk_controls = RigVMFunctionReferenceNode()
   > create_fk_offset_controls = RigVMFunctionReferenceNode()
   > create_proxy_controls = RigVMFunctionReferenceNode()
   > get_backwards_local_transforms = RigVMFunctionReferenceNode()
   create_joint_nulls = RigVMFunctionReferenceNode(ExecuteContext=@RigVMFunction_Sequence.E)
   > create Space Switches = RigVMFunctionReferenceNode()
  ]}
 Multiply_1_ContainedGraph {nodes [
   n1 = RigVMFunctionEntryNode()
   Multiply = &RigVMFunction_MathDoubleMul(A=@n1.A, B=@n1.B)
   RigVMFunctionReturnNode(Result=&RigVMFunction_MathDoubleMul(B=@n1.C, A=@Multiply.Result).Result)
  ]}
 Multiply_1_ContainedGraph {nodes [
   n1 = RigVMFunctionEntryNode()
   Multiply = &RigVMFunction_MathDoubleMul(A=@n1.A, B=@n1.B)
   RigVMFunctionReturnNode(Result=&RigVMFunction_MathDoubleMul(B=@n1.C, A=@Multiply.Result).Result)
  ]}
 Multiply_3_ContainedGraph {nodes [
   n1 = RigVMFunctionEntryNode()
   RigVMFunctionReturnNode(Result=&RigVMFunction_MathDoubleMul(B=@n1.C, A=&RigVMFunction_MathDoubleMul(A=@n1.A, B=@n1.B).Result).Result)
  ]}
 Multiply_3_ContainedGraph {nodes [
   n1 = RigVMFunctionEntryNode()
   RigVMFunctionReturnNode(Result=&RigVMFunction_MathDoubleMul(B=@n1.C, A=&RigVMFunction_MathDoubleMul(A=@n1.A, B=@n1.B).Result).Result)
  ]}
 Multiply_3_ContainedGraph {nodes [
   n1 = RigVMFunctionEntryNode()
   RigVMFunctionReturnNode(Result=&RigVMFunction_MathDoubleMul(B=@n1.C, A=&RigVMFunction_MathDoubleMul(A=@n1.A, B=@n1.B).Result).Result)
  ]}
 New Function_ContainedGraph {nodes [
   n1 = RigVMFunctionEntryNode()
   n2 = &RigVMFunction_Sequence(ExecuteContext=@n1.ExecuteContext)
   RigVMFunction_Sequence = RigVMAggregateNode(ExecuteContext=@n2.A)
   RigVMDispatchNode(Array=RigVMVariableNode() {variable=tail_chain_joints}.Value, ExecuteContext=@RigVMFunction_Sequence.A) {template="DISPATCH_RigVMDispatch_ArrayReset(io Array)"}
   > RigVMVariableNode() {variable=tail_base_found}
   > RigVMVariableNode(Value=/Script/ControlRig.RigUnit_ResolveConnector(Connector=(Type=Connector,Name="HipRoot"), SkipSocket=False) {is=struct RigVMUnitNode}.Result) {variable=connector_hip_root}
   > RigVMVariableNode(Value=&RigUnit_ResolveConnector(Connector=(Type=Connector,Name="TailBase"), SkipSocket=False).Result) {variable=connector_tail_base}
   > RigVMVariableNode(Value=&RigUnit_ResolveConnector(Connector=(Type=Connector,Name="TailTip"), SkipSocket=False).Result) {variable=connector_tail_tip}
   For_Each_1 = RigVMDispatchNode(
    Array=/Script/ControlRig.RigUnit_HierarchyGetParentsItemArray(bDefaultParent=True, bIncludeChild=true, bReverse=true, Child=RigVMVariableNode() {variable=connector_tail_tip}.Value) {is=struct RigVMUnitNode}.Parents
    ExecuteContext=@RigVMFunction_Sequence.B
   ) {template="DISPATCH_RigVMDispatch_ArrayIterator(in Array,out Element,out Index,out Count,out Ratio)"}
   > n3 = RigVMVariableNode(Value=@For_Each_1.Element) {variable=current_item}
   RigVMFunction_ControlFlowBranch = &RigVMFunction_ControlFlowBranch(Condition=RigVMVariableNode() {variable=tail_base_found}.Value, ExecuteContext=@n3.ExecuteContext)
   RigVMDispatchNode(
    Array=RigVMVariableNode() {variable=tail_chain_joints}.Value
    Element=RigVMVariableNode() {variable=current_item}.Value
    ExecuteContext=@RigVMFunction_ControlFlowBranch.True
   ) {template="DISPATCH_RigVMDispatch_ArrayAdd(io Array,in Element,out Index)"}
   n4 = &RigVMFunction_ControlFlowBranch(
    Condition=RigVMDispatchNode(B=RigVMVariableNode() {variable=connector_tail_base}.Value, A=RigVMVariableNode() {variable=current_item}.Value) {template="DISPATCH_RigVMDispatch_CoreEquals(in A,in B,out Result)"}.Result
    ExecuteContext=@RigVMFunction_ControlFlowBranch.False
   )
   RigVMVariableNode(Value=true, ExecuteContext=@n4.True) {variable=tail_base_found}
   > Add = RigVMDispatchNode(Array=RigVMVariableNode() {variable=tail_chain_joints}.Value, Element=RigVMVariableNode() {variable=current_item}.Value) {template="DISPATCH_RigVMDispatch_ArrayAdd(io Array,in Element,out Index)"}
   RigVMFunctionReturnNode(ExecuteContext=@n2.B)
  ]}
 New Function_ContainedGraph {nodes [
   RigVMFunctionEntryNode()
   > RigVMFunction_Sequence_1 = RigVMAggregateNode()
   RigVMFunctionReturnNode(ExecuteContext=@RigVMFunction_Sequence_1.C)
   RigVMDispatchNode(Array=RigVMVariableNode() {variable=fk_controls}.Value, ExecuteContext=@RigVMFunction_Sequence_1.B) {template="DISPATCH_RigVMDispatch_ArrayReset(io Array)"}
   > For_Each = RigVMDispatchNode(Array=RigVMVariableNode() {variable=fk_joints}.Value) {template="DISPATCH_RigVMDispatch_ArrayIterator(in Array,out Element,out Index,out Count,out Ratio)"}
   > RigVMVariableNode(Value=@For_Each.Element) {variable=current_item}
   > RigVMVariableNode(Value=@For_Each.Index) {variable=index}
   > RigVMVariableNode(Value=@For_Each.Count) {variable=Count}
   Add = /Script/RigVM.RigVMFunction_MathIntAdd(B=1, A=RigVMVariableNode() {variable=index}.Value) {is=struct RigVMUnitNode, template="Add::Execute(in A,in B,out Result)"}
   RigVMFunction_StringPadInteger = /Script/RigVM.RigVMFunction_StringPadInteger(Digits=2, Value=@Add.Result) {is=struct RigVMUnitNode}
   RigVMFunction_StringJoin = /Script/RigVM.RigVMFunction_StringJoin(Separator=" ", Values=("",""), Values.1=@RigVMFunction_StringPadInteger.Result, Values.0=RigVMVariableNode() {variable=basename}.Value) {is=struct RigVMUnitNode}
   > RigVMVariableNode(Value=@RigVMFunction_StringJoin.Result) {variable=token}
   n1 = &RigUnit_GetTransform(bInitial=true, Item=RigVMVariableNode() {variable=current_item}.Value)
   n2 = /Script/ControlRig.RigUnit_AimBoneMath(
    DebugSettings=(bEnabled=False,Scale=10,WorldOffset=(Rotation=(X=0,Y=0,Z=0,W=1),Translation=(X=0,Y=0,Z=0),Scale3D=(X=1,Y=1,Z=1)))
    InputTransform=(Rotation=(X=0,Y=0,Z=0,W=1),Translation=(X=0,Y=0,Z=0),Scale3D=(X=1,Y=1,Z=1))
    Primary=(Weight=1,Axis=(X=0,Y=1,Z=0),Target=(X=1,Y=0,Z=0),Kind=Location,Space=(Type=Bone,Name="None"))
    Secondary=(Weight=1,Axis=(X=0,Y=0,Z=1),Target=(X=1,Y=0,Z=0),Kind=Location,Space=(Type=Bone,Name="None"))
    Weight=1
    Primary.Target=/Script/RigVM.RigVMFunction_MathTransformMakeAbsolute(Local=(Rotation=(X=0,Y=0,Z=0,W=1),Translation=(X=1,Y=0,Z=0),Scale3D=(X=1,Y=1,Z=1)), Parent=@n1.Transform) {is=struct RigVMUnitNode, template="Make Absolute::Execute(in Local,in Parent,out Global)"}.Global.Translation
    Secondary.Target=&RigVMFunction_MathTransformMakeAbsolute(Local=(Rotation=(X=0,Y=0,Z=0,W=1),Translation=(X=0,Y=0,Z=1),Scale3D=(X=1,Y=1,Z=1)), Parent=@n1.Transform) {template="Make Absolute::Execute(in Local,in Parent,out Global)"}.Global.Translation
    InputTransform.Translation=@n1.Transform.Translation
   ) {is=struct RigVMUnitNode}
   From_String = RigVMDispatchNode(String=&RigVMFunction_StringJoin(Separator=" ", Values=("","null"), Values.0=RigVMVariableNode() {variable=token}.Value).Result) {template="DISPATCH_RigDispatch_FromString(in String,out Result)"}
   > n3 = &RigUnit_HierarchyAddNull(Space=GlobalSpace, Transform=@n2.Result, Name=@From_String.Result, Parent=RigVMVariableNode() {variable=current_parent_item}.Value)
   > RigVMFunction_Sequence = RigVMAggregateNode()
   RigVMFunction_MathIntToDouble = /Script/RigVM.RigVMFunction_MathIntToFloat(Value=&RigVMFunction_MathIntAdd(B=0, A=RigVMVariableNode() {variable=Index}.Value) {template="Add::Execute(in A,in B,out Result)"}.Result) {is=struct RigVMUnitNode, template="Cast::Execute(in Value,out Result)"}
   Multiply_2 = RigVMAggregateNode(A=-1, C=@RigVMFunction_MathIntToDouble.Result, B=RigVMVariableNode() {variable=scale_increment_factor}.Value)
   RigVMVariableNode(Value=&RigVMFunction_MathDoubleAdd(A=1, B=@Multiply_2.Result).Result, ExecuteContext=@RigVMFunction_Sequence.B) {variable=output_item_scale}
   From_String_1 = RigVMDispatchNode(String=RigVMVariableNode() {variable=token}.Value) {template="DISPATCH_RigDispatch_FromString(in String,out Result)"}
   Modulo = /Script/RigVM.RigVMFunction_MathIntMod(B=2, A=RigVMVariableNode() {variable=Index}.Value) {is=struct RigVMUnitNode, template="Modulo::Execute(in A,in B,out Result)"}
   Not_Equals = RigVMDispatchNode(B=0, A=@Modulo.Result) {template="DISPATCH_RigVMDispatch_CoreNotEquals(in A,in B,out Result)"}
   Multiply = RigVMAggregateNode(B=2, C=RigVMVariableNode() {variable=output_item_scale}.Value, A=RigVMVariableNode() {variable=FK Control Scale}.Value)
   SpawnControl = /Script/ControlRig.RigUnit_HierarchyAddControlTransform(
    InitialValue=(Rotation=(X=0,Y=0,Z=0,W=1),Translation=(X=0,Y=0,Z=0),Scale3D=(X=1,Y=1,Z=1))
    OffsetSpace=LocalSpace
    OffsetTransform=(Rotation=(X=0,Y=0,Z=0,W=1),Translation=(X=0,Y=0,Z=0),Scale3D=(X=1,Y=1,Z=1))
    Settings="(InitialSpace=LocalSpace,bUsePreferredRotationOrder=true,PreferredRotationOrder=YZX,Limits=(LimitTranslationX=(bMinimum=False,bMaximum=False),LimitTranslationY=(bMinimum=False,bMaximum=False),LimitTranslationZ=(bMinimum=False,bMaximum=False),LimitPitch=(bM ...(truncated)"
    Name=@From_String_1.Result
    Settings.Shape.Color=RigVMDispatchNode(
     Condition=@Not_Equals.Result
     True=RigVMVariableNode() {variable=even_control_color}.Value
     False=RigVMVariableNode() {variable=odd_control_color}.Value
    ) {template="DISPATCH_RigVMDispatch_If(in Condition,in True,in False,out Result)"}.Result
    Settings.Shape.Transform.Scale3D.X=@Multiply.Result
    Settings.Shape.Transform.Scale3D.Y=@Multiply.Result
    Parent=@n3.Item
    ExecutePin=@RigVMFunction_Sequence.C
   ) {
    is=struct RigVMUnitNode
    template="SpawnControl::Execute(in InitialValue,in Settings,in OffsetTransform,in OffsetSpace,in Parent,in Name,out Item)"
   }
   > RigVMVariableNode(Value=@SpawnControl.Item) {variable=current_parent_item}
   > RigVMDispatchNode(Array=RigVMVariableNode() {variable=fk_controls}.Value, Element=@SpawnControl.Item) {template="DISPATCH_RigVMDispatch_ArrayAdd(io Array,in Element,out Index)"}
   Divide = &RigVMFunction_MathDoubleDiv(A=1, B=RigVMVariableNode() {variable=Control Taper Scale}.Value)
   n4 = &RigVMFunction_MathDoubleMul(A=-1, B=@Divide.Result)
   RigVMFunction_MathIntToDouble_1 = &RigVMFunction_MathIntToFloat(Value=RigVMDispatchNode(Array=RigVMVariableNode() {variable=fk_joints}.Value) {template="DISPATCH_RigVMDispatch_ArrayGetNum(in Array,out Num)"}.Num) {template="Cast::Execute(in Value,out Result)"}
   RigVMVariableNode(
    Value=&RigVMFunction_MathDoubleDiv(A=&RigVMFunction_MathDoubleAdd(A=1, B=@n4.Result).Result, B=@RigVMFunction_MathIntToDouble_1.Result).Result
    ExecuteContext=@RigVMFunction_Sequence.A
   ) {variable=scale_increment_factor}
   RigVMVariableNode(Value=RigVMVariableNode() {variable=tail gimbal control}.Value, ExecuteContext=@RigVMFunction_Sequence_1.A) {variable=current_parent_item}
   Num = RigVMDispatchNode(Array=RigVMVariableNode() {variable=fk_joints}.Value) {template="DISPATCH_RigVMDispatch_ArrayGetNum(in Array,out Num)"}
   calculate_tapered_scale = RigVMFunctionReferenceNode(
    ExecuteContext=()
    item_count=@Num.Num
    taper_scale=RigVMVariableNode() {variable=Control Taper Scale}.Value
    item_index=RigVMVariableNode() {variable=Index}.Value
   )
  ]}
 New Function_ContainedGraph {nodes [
   n1 = RigVMFunctionEntryNode()
   To_String = RigVMDispatchNode(Value=@n1.input_node.Name) {template="DISPATCH_RigDispatch_ToString(in Value,out Result)"}
   RigVMFunction_StringJoin = &RigVMFunction_StringJoin(Separator=_, Values=("",""), Values.0=@To_String.Result, Values.1=RigVMVariableNode() {variable=suffex}.Value)
   From_String = RigVMDispatchNode(String=@RigVMFunction_StringJoin.Result) {template="DISPATCH_RigDispatch_FromString(in String,out Result)"}
   > RigVMFunctionReturnNode(output_name=@From_String.Result)
  ]}
 New Function_ContainedGraph {nodes [
   n1 = RigVMFunctionEntryNode()
   RigVMFunction_Sequence = &RigVMFunction_Sequence(ExecuteContext=@n1.ExecuteContext)
   n2 = &RigVMFunction_Sequence(ExecuteContext=@RigVMFunction_Sequence.A)
   For_Each = RigVMDispatchNode(Array=RigVMVariableNode() {variable=fk_offset_controls}.Value, ExecuteContext=@n2.A) {template="DISPATCH_RigVMDispatch_ArrayIterator(in Array,out Element,out Index,out Count,out Ratio)"}
   > RigVMVariableNode(Value=@For_Each.Element) {variable=current_control}
   > RigVMVariableNode(Value=@For_Each.Index) {variable=index}
   > RigVMVariableNode(Value=RigVMDispatchNode(Index=RigVMVariableNode() {variable=index}.Value, Array=RigVMVariableNode() {variable=fk_joints}.Value) {template="DISPATCH_RigVMDispatch_ArrayGetAtIndex(in Array,in Index,out Element)"}.Element) {variable=current_joint}
   n3 = RigVMVariableNode() {variable=current_control}
   Get_Metadata = RigVMDispatchNode(
    Default=(Rotation=(X=0,Y=0,Z=0,W=1),Translation=(X=0,Y=0,Z=0),Scale3D=(X=1,Y=1,Z=1))
    Name=BoneLocalTransform
    NameSpace=Self
    Item=@n3.Value
   ) {template="DISPATCH_RigDispatch_GetMetadata(in Item,in Name,in NameSpace,in Default,out Value,out Found)"}
   > &RigUnit_SetTransform(
    Weight=1
    bInitial=False
    Value=&RigVMFunction_MathTransformMakeAbsolute(Parent=&RigUnit_GetTransform(bInitial=False, Item=@n3.Value).Transform, Local=@Get_Metadata.Value) {template="Make Absolute::Execute(in Local,in Parent,out Global)"}.Global
    Item=RigVMVariableNode() {variable=current_joint}.Value
   )
   For_Each_1 = RigVMDispatchNode(Array=RigVMVariableNode() {variable=fk_joints}.Value, ExecuteContext=@n2.B) {template="DISPATCH_RigVMDispatch_ArrayIterator(in Array,out Element,out Index,out Count,out Ratio)"}
   > RigVMVariableNode(Value=@For_Each_1.Index) {variable=index}
   Get_Metadata_1 = RigVMDispatchNode(
    Default=(Rotation=(X=0,Y=0,Z=0,W=1),Translation=(X=0,Y=0,Z=0),Scale3D=(X=1,Y=1,Z=1))
    Name=BoneLocalTransform
    NameSpace=Self
    Item=@For_Each_1.Element
   ) {template="DISPATCH_RigDispatch_GetMetadata(in Item,in Name,in NameSpace,in Default,out Value,out Found)"}
   > &RigUnit_SetTransform(
    Weight=1
    bInitial=False
    Item=RigVMDispatchNode(Array=RigVMVariableNode() {variable=tail_chain_joints}.Value, Index=RigVMVariableNode() {variable=index}.Value) {template="DISPATCH_RigVMDispatch_ArrayGetAtIndex(in Array,in Index,out Element)"}.Element
    Value=&RigVMFunction_MathTransformMakeAbsolute(Parent=&RigUnit_GetTransform(bInitial=False, Item=@For_Each_1.Element).Transform, Local=@Get_Metadata_1.Value) {template="Make Absolute::Execute(in Local,in Parent,out Global)"}.Global
   )
   RigVMFunctionReturnNode(ExecuteContext=@RigVMFunction_Sequence.B)
   RigVMVariableNode() {variable=fk_controls}
  ]}
 New Function_ContainedGraph {nodes [
   n1 = RigVMFunctionEntryNode()
   RigVMFunction_Sequence = &RigVMFunction_Sequence(ExecuteContext=@n1.ExecuteContext)
   RigVMFunctionReturnNode(ExecuteContext=@RigVMFunction_Sequence.B)
   RigVMFunction_ControlFlowBranch = &RigVMFunction_ControlFlowBranch(Condition=RigVMVariableNode() {variable=is_interacting}.Value, ExecuteContext=@RigVMFunction_Sequence.A)
   RigVMVariableNode(Value=(X=0,Y=0,Z=0), ExecuteContext=@RigVMFunction_ControlFlowBranch.False) {variable=previous_rotation}
   > &RigUnit_SetTransform(
    Space=LocalSpace
    Value=(Rotation=(X=0,Y=0,Z=0,W=1),Translation=(X=0,Y=0,Z=0),Scale3D=(X=1,Y=1,Z=1))
    Weight=1
    bInitial=False
    Item=RigVMVariableNode() {variable=tail_rotate_proxyControl}.Value
   )
   n2 = &RigVMFunction_ControlFlowBranch(
    Condition=RigVMDispatchNode(Array=RigVMVariableNode() {variable=interacting_items}.Value, Element=RigVMVariableNode() {variable=tail_rotate_proxyControl}.Value) {template="DISPATCH_RigVMDispatch_ArrayFind(in Array,in Element,out Index,out Success)"}.Success
    ExecuteContext=@RigVMFunction_ControlFlowBranch.True
   )
   RigVMFunction_Sequence_1 = RigVMAggregateNode(ExecuteContext=@n2.True)
   RigVMVariableNode(Value=RigVMVariableNode() {variable=proxy_euler_rotation}.Value, ExecuteContext=@RigVMFunction_Sequence_1.D) {variable=previous_rotation}
   RigVMVariableNode(ExecuteContext=@RigVMFunction_Sequence_1.C) {variable=count}
   > For_Each = RigVMDispatchNode(Array=RigVMVariableNode() {variable=fk_controls}.Value) {template="DISPATCH_RigVMDispatch_ArrayIterator(in Array,out Element,out Index,out Count,out Ratio)"}
   > n3 = RigVMVariableNode(Value=@For_Each.Element) {variable=fk_control}
   GetAnimationChannelFromItem = /Script/ControlRig.RigUnit_GetIntAnimationChannelFromItem(bInitial=False, Item=RigVMVariableNode() {variable=tail_rotate_position_proxyChannel}.Value) {is=struct RigVMUnitNode, template="GetAnimationChannelFromItem::Execute(out Value,in Item,in bInitial)"}
   Subtract = /Script/RigVM.RigVMFunction_MathIntSub(B=1, A=@GetAnimationChannelFromItem.Value) {is=struct RigVMUnitNode, template="Subtract::Execute(in A,in B,out Result)"}
   GreaterEqual = /Script/RigVM.RigVMFunction_MathIntGreaterEqual(A=@For_Each.Index, B=@Subtract.Result) {is=struct RigVMUnitNode, template="GreaterEqual::Execute(in A,in B,out Result)"}
   n4 = &RigVMFunction_ControlFlowBranch(Condition=@GreaterEqual.Result, ExecuteContext=@n3.ExecuteContext)
   Less = /Script/RigVM.RigVMFunction_MathIntLess(
    B=&RigUnit_GetIntAnimationChannelFromItem(bInitial=False, Item=RigVMVariableNode() {variable=tail_rotate_spread_proxyChannel}.Value) {template="GetAnimationChannelFromItem::Execute(out Value,in Item,in bInitial)"}.Value
    A=RigVMVariableNode() {variable=count}.Value
   ) {is=struct RigVMUnitNode, template="Less::Execute(in A,in B,out Result)"}
   n5 = &RigVMFunction_ControlFlowBranch(Condition=@Less.Result, ExecuteContext=@n4.True)
   RigVMFunction_MathQuaternionToEuler = &RigVMFunction_MathQuaternionToEuler(
    RotationOrder=ZYX
    Value=&RigUnit_GetTransform(Space=LocalSpace, bInitial=False, Item=RigVMVariableNode() {variable=fk_control}.Value).Transform.Rotation
   )
   Add = /Script/RigVM.RigVMFunction_MathVectorAdd(A=@RigVMFunction_MathQuaternionToEuler.Result, B=RigVMVariableNode() {variable=proxy_euler_rotation_delta}.Value) {is=struct RigVMUnitNode, template="Add::Execute(in A,in B,out Result)"}
   RigVMFunction_MathQuaternionFromEuler = &RigVMFunction_MathQuaternionFromEuler(Euler=@Add.Result)
   Set Transform = &RigUnit_SetRotation(
    Weight=1
    bInitial=False
    ExecutePin=@n5.True
    Value=@RigVMFunction_MathQuaternionFromEuler.Result
    Item=RigVMVariableNode() {variable=fk_control}.Value
   )
   > RigVMVariableNode(Value=&RigVMFunction_MathIntAdd(B=1, A=RigVMVariableNode() {variable=count}.Value) {template="Add::Execute(in A,in B,out Result)"}.Result) {variable=count}
   n6 = RigVMVariableNode() {variable=tail_rotate_proxyControl}
   RigVMVariableNode(
    Value=&RigVMFunction_MathQuaternionToEuler(RotationOrder=ZYX, Value=&RigUnit_GetTransform(Space=LocalSpace, bInitial=False, Item=@n6.Value).Transform.Rotation).Result
    ExecuteContext=@RigVMFunction_Sequence_1.A
   ) {variable=proxy_euler_rotation}
   RigVMVariableNode(
    Value=/Script/RigVM.RigVMFunction_MathVectorSub(B=RigVMVariableNode() {variable=previous_rotation}.Value, A=RigVMVariableNode() {variable=proxy_euler_rotation}.Value) {is=struct RigVMUnitNode, template="Subtract::Execute(in A,in B,out Result)"}.Result
    ExecuteContext=@RigVMFunction_Sequence_1.B
   ) {variable=proxy_euler_rotation_delta}
  ]}
 New Function_ContainedGraph {nodes [
   RigVMFunctionEntryNode()
   > RigVMFunction_Sequence = RigVMAggregateNode()
   RigVMDispatchNode(ExecuteContext=@RigVMFunction_Sequence.B, Array=RigVMVariableNode() {variable=fk_joints}.Value) {template="DISPATCH_RigVMDispatch_ArrayReset(io Array)"}
   > For_Each = RigVMDispatchNode(Array=RigVMVariableNode() {variable=tail_chain_joints}.Value) {template="DISPATCH_RigVMDispatch_ArrayIterator(in Array,out Element,out Index,out Count,out Ratio)"}
   > RigVMVariableNode(Value=@For_Each.Element) {variable=current_item}
   > n1 = RigVMVariableNode(Value=@For_Each.Index) {variable=index}
   Greater = /Script/RigVM.RigVMFunction_MathIntGreater(B=0, A=RigVMVariableNode() {variable=index}.Value) {is=struct RigVMUnitNode, template="Greater::Execute(in A,in B,out Result)"}
   RigVMFunction_ControlFlowBranch = &RigVMFunction_ControlFlowBranch(Condition=@Greater.Result, ExecuteContext=@n1.ExecuteContext)
   n2 = &RigVMFunction_Sequence(ExecuteContext=@RigVMFunction_ControlFlowBranch.True)
   Subtract = &RigVMFunction_MathIntSub(B=1, A=RigVMVariableNode() {variable=index}.Value) {template="Subtract::Execute(in A,in B,out Result)"}
   RigVMVariableNode(
    Value=RigVMDispatchNode(Index=@Subtract.Result, Array=RigVMVariableNode() {variable=tail_chain_joints}.Value) {template="DISPATCH_RigVMDispatch_ArrayGetAtIndex(in Array,in Index,out Element)"}.Element
    ExecuteContext=@n2.A
   ) {variable=previous_item}
   n3 = &RigUnit_AimBoneMath(
    DebugSettings=(bEnabled=False,Scale=10,WorldOffset=(Rotation=(X=0,Y=0,Z=0,W=1),Translation=(X=0,Y=0,Z=0),Scale3D=(X=1,Y=1,Z=1)))
    Primary=(Weight=1,Axis=(X=1,Y=0,Z=0),Target=(X=1,Y=0,Z=0),Kind=Location,Space=(Type=Bone,Name="None"))
    Secondary=(Weight=1,Axis=(X=0,Y=0,Z=1),Target=(X=0,Y=0,Z=1),Kind=Direction,Space=(Type=Bone,Name="None"))
    Weight=1
    InputTransform=&RigUnit_GetTransform(bInitial=False, Item=RigVMVariableNode() {variable=previous_item}.Value).Transform
    Primary.Target=&RigUnit_GetTransform(bInitial=False, Item=RigVMVariableNode() {variable=current_item}.Value).Transform.Translation
   )
   RigVMFunction_StringPadInteger = &RigVMFunction_StringPadInteger(Digits=2, Value=RigVMVariableNode() {variable=index}.Value)
   RigVMFunction_StringJoin = &RigVMFunction_StringJoin(Separator=" ", Values=("fk","","rigJnt"), Values.1=@RigVMFunction_StringPadInteger.Result)
   From_String = RigVMDispatchNode(String=@RigVMFunction_StringJoin.Result) {template="DISPATCH_RigDispatch_FromString(in String,out Result)"}
   n4 = /Script/ControlRig.RigUnit_HierarchyAddBone(
    Space=GlobalSpace
    Transform=@n3.Result
    Name=@From_String.Result
    Parent=RigVMVariableNode() {variable=current_parent_item}.Value
    ExecutePin=@n2.B
   ) {is=struct RigVMUnitNode}
   > RigVMVariableNode(Value=@n4.Item) {variable=current_parent_item}
   > RigVMDispatchNode(Array=RigVMVariableNode() {variable=fk_joints}.Value, Element=@n4.Item) {template="DISPATCH_RigVMDispatch_ArrayAdd(io Array,in Element,out Index)"}
   Add = &RigVMFunction_MathIntAdd(B=1, A=RigVMVariableNode() {variable=index}.Value) {template="Add::Execute(in A,in B,out Result)"}
   From_String_1 = RigVMDispatchNode(String=&RigVMFunction_StringJoin(Separator=" ", Values=("Fk","","RigJnt"), Values.1=&RigVMFunction_StringPadInteger(Digits=2, Value=@Add.Result).Result).Result) {template="DISPATCH_RigDispatch_FromString(in String,out Result)"}
   n5 = &RigUnit_HierarchyAddBone(
    Space=GlobalSpace
    ExecutePin=@For_Each.Completed
    Name=@From_String_1.Result
    Transform=&RigUnit_GetTransform(bInitial=False, Item=RigVMVariableNode() {variable=current_item}.Value).Transform
    Parent=RigVMVariableNode() {variable=current_parent_item}.Value
   )
   > RigVMDispatchNode(Array=RigVMVariableNode() {variable=fk_joints}.Value, Element=@n5.Item) {template="DISPATCH_RigVMDispatch_ArrayAdd(io Array,in Element,out Index)"}
   > Set Transform = &RigUnit_SetRotation(Value=(X=0,Y=0,Z=0,W=1), Weight=1, bInitial=true, Item=@n5.Item)
   RigVMFunctionReturnNode(ExecuteContext=@RigVMFunction_Sequence.C)
   RigVMVariableNode(Value=RigVMVariableNode() {variable=connector_hip_root}.Value, ExecuteContext=@RigVMFunction_Sequence.A) {variable=current_parent_item}
  ]}
 New Function_ContainedGraph {nodes [
   n1 = RigVMFunctionEntryNode()
   RigVMFunction_Sequence = &RigVMFunction_Sequence(ExecuteContext=@n1.ExecuteContext)
   RigVMFunctionReturnNode(ExecuteContext=@RigVMFunction_Sequence.B)
   RigVMDispatchNode(Array=RigVMVariableNode() {variable=fk_joint_local_transforms}.Value, ExecuteContext=@RigVMFunction_Sequence.A) {template="DISPATCH_RigVMDispatch_ArrayReset(io Array)"}
   > For_Each = RigVMDispatchNode(Array=RigVMVariableNode() {variable=tail_chain_joints}.Value) {template="DISPATCH_RigVMDispatch_ArrayIterator(in Array,out Element,out Index,out Count,out Ratio)"}
   > RigVMVariableNode(Value=@For_Each.Element) {variable=tail_joint}
   > RigVMVariableNode(Value=RigVMDispatchNode(Array=RigVMVariableNode() {variable=fk_joints}.Value, Index=@For_Each.Index) {template="DISPATCH_RigVMDispatch_ArrayGetAtIndex(in Array,in Index,out Element)"}.Element) {variable=rig_joint}
   n2 = RigVMVariableNode() {variable=rig_joint}
   n3 = RigVMVariableNode() {variable=tail_joint}
   > Set_Metadata = RigVMDispatchNode(
    Name=BoneLocalTransform
    NameSpace=Self
    Value=&RigVMFunction_MathTransformMakeRelative(
     Parent=&RigUnit_GetTransform(bInitial=False, Item=@n2.Value).Transform
     Global=&RigUnit_GetTransform(bInitial=False, Item=@n3.Value).Transform
    ).Local
    Item=RigVMVariableNode() {variable=rig_joint}.Value
   ) {template="DISPATCH_RigDispatch_SetMetadata(in Item,in Name,in NameSpace,in Value,out Success)"}
  ]}
 New Function_ContainedGraph {nodes [
   n1 = RigVMFunctionEntryNode()
   RigVMFunction_Sequence = &RigVMFunction_Sequence(ExecuteContext=@n1.ExecuteContext)
   RigVMFunction_Sequence_1 = RigVMAggregateNode(ExecuteContext=@RigVMFunction_Sequence.A)
   n2 = &RigUnit_HierarchyAddNull(
    Name=ProxyNull
    Parent=(Type=Bone,Name="None")
    Space=LocalSpace
    Transform=(Rotation=(X=0,Y=0,Z=1,W=0),Translation=(X=0,Y=0,Z=0),Scale3D=(X=1,Y=1,Z=1))
    Transform.Translation=&RigUnit_GetTransform(bInitial=False, Item=RigVMVariableNode() {variable=connector_tail_base}.Value).Transform.Translation
    ExecutePin=@RigVMFunction_Sequence_1.B
   )
   > n3 = &RigUnit_HierarchyAddControlTransform(
    InitialValue=(Rotation=(X=0,Y=0,Z=0,W=1),Translation=(X=0,Y=0,Z=0),Scale3D=(X=1,Y=1,Z=1))
    Name=Tail Rotate
    OffsetSpace=GlobalSpace
    Settings="(InitialSpace=LocalSpace,bUsePreferredRotationOrder=true,PreferredRotationOrder=YZX,Limits=(LimitTranslationX=(bMinimum=False,bMaximum=False),LimitTranslationY=(bMinimum=False,bMaximum=False),LimitTranslationZ=(bMinimum=False,bMaximum=False),LimitPitch=(bM ...(truncated)"
    OffsetTransform=&RigUnit_GetTransform(bInitial=False, Item=@n2.Item).Transform
    Settings.Proxy.DrivenControls=RigVMVariableNode() {variable=fk_controls}.Value
    Parent=RigVMVariableNode() {variable=tail_rotate_proxyControlParent}.Value
    Settings.Shape.Transform=RigVMVariableNode() {variable=Proxy shape position}.Value
   ) {template="SpawnControl::Execute(in InitialValue,in Settings,in OffsetTransform,in OffsetSpace,in Parent,in Name,out Item)"}
   > RigVMVariableNode(Value=@n3.Item) {variable=tail_rotate_proxyControl}
   Num = RigVMDispatchNode(Array=RigVMVariableNode() {variable=fk_joints}.Value) {template="DISPATCH_RigVMDispatch_ArrayGetNum(in Array,out Num)"}
   n4 = /Script/ControlRig.RigUnit_HierarchyAddAnimationChannelInteger(
    InitialValue=1
    LimitsEnabled=(Enabled=(bMinimum=True,bMaximum=true))
    MinimumValue=1
    Name=position
    MaximumValue=@Num.Num
    Parent=RigVMVariableNode() {variable=tail_rotate_proxyControl}.Value
    ExecutePin=@RigVMFunction_Sequence_1.C
   ) {is=struct RigVMUnitNode}
   > RigVMVariableNode(Value=@n4.Item) {variable=tail_rotate_position_proxyChannel}
   n5 = RigVMDispatchNode(Array=RigVMVariableNode() {variable=fk_joints}.Value) {template="DISPATCH_RigVMDispatch_ArrayGetNum(in Array,out Num)"}
   > n6 = &RigUnit_HierarchyAddAnimationChannelInteger(
    LimitsEnabled=(Enabled=(bMinimum=True,bMaximum=true))
    MinimumValue=1
    Name=spread
    InitialValue=@n5.Num
    MaximumValue=@n5.Num
    Parent=RigVMVariableNode() {variable=tail_rotate_proxyControl}.Value
   )
   > RigVMVariableNode(Value=@n6.Item) {variable=tail_rotate_spread_proxyChannel}
   SpawnAnimationChannel = /Script/ControlRig.RigUnit_HierarchyAddAnimationChannelBool(
    InitialValue=false
    LimitsEnabled=()
    MaximumValue=True
    MinimumValue=False
    Name=Feedback On Off
    Parent=RigVMVariableNode() {variable=tail_rotate_proxyControl}.Value
    ExecutePin=@RigVMFunction_Sequence_1.D
   ) {
    is=struct RigVMUnitNode
    template="SpawnAnimationChannel::Execute(in InitialValue,in MinimumValue,in MaximumValue,in LimitsEnabled,in Parent,in Name,out Item)"
   }
   > RigVMVariableNode(Value=@SpawnAnimationChannel.Item) {variable=tail_rotate_feedback_proxyChannel}
   > n7 = &RigUnit_HierarchyAddAnimationChannelInteger(
    InitialValue=5
    LimitsEnabled=(Enabled=(bMinimum=True,bMaximum=true))
    MaximumValue=20
    MinimumValue=1
    Name=Feedback Thickness
    Parent=RigVMVariableNode() {variable=tail_rotate_proxyControl}.Value
   )
   > RigVMVariableNode(Value=@n7.Item) {variable=tail_rotate_feedbackThickness_proxyChannel}
   n8 = RigVMDispatchNode(Index=0, Array=RigVMVariableNode() {variable=fk_controls}.Value) {template="DISPATCH_RigVMDispatch_ArrayGetAtIndex(in Array,in Index,out Element)"}
   n9 = &RigUnit_HierarchyAddNull(
    Name=Tail Rotate Null
    Space=GlobalSpace
    Transform=&RigUnit_GetTransform(bInitial=False, Item=@n8.Element).Transform
    Parent=/Script/ControlRig.RigUnit_HierarchyGetParent(bDefaultParent=True, Child=@n8.Element) {is=struct RigVMUnitNode}.Parent
    ExecutePin=@RigVMFunction_Sequence_1.A
   )
   > RigVMVariableNode(Value=@n9.Item) {variable=tail_rotate_proxyControlParent}
   RigVMFunctionReturnNode(ExecuteContext=@RigVMFunction_Sequence.B)
  ]}
 New Function_ContainedGraph {nodes [
   n1 = RigVMFunctionEntryNode()
   RigVMFunction_Sequence = &RigVMFunction_Sequence(ExecuteContext=@n1.ExecuteContext)
   RigVMFunctionReturnNode(ExecuteContext=@RigVMFunction_Sequence.B)
   n2 = &RigVMFunction_MathIntSub(
    B=1
    A=&RigUnit_GetIntAnimationChannelFromItem(bInitial=False, Item=RigVMVariableNode() {variable=tail_rotate_position_proxyChannel}.Value) {template="GetAnimationChannelFromItem::Execute(out Value,in Item,in bInitial)"}.Value
   ) {template="Subtract::Execute(in A,in B,out Result)"}
   n3 = RigVMVariableNode() {variable=fk_controls}
   &RigUnit_ParentConstraint(
    AdvancedSettings=(InterpolationType=Average,RotationOrderForFilter=XZY)
    Parents=((Item=(Type=Bone,Name="None"),Weight=1))
    Weight=1
    bMaintainOffset=false
    Parents.0.Item=&RigUnit_HierarchyGetParent(
     bDefaultParent=True
     Child=RigVMDispatchNode(Index=@n2.Result, Array=@n3.Value) {template="DISPATCH_RigVMDispatch_ArrayGetAtIndex(in Array,in Index,out Element)"}.Element
    ).Parent
    Child=RigVMVariableNode() {variable=tail_rotate_proxyControlParent}.Value
    ExecutePin=@RigVMFunction_Sequence.A
   )
   GetAnimationChannelFromItem = &RigUnit_GetIntAnimationChannelFromItem(bInitial=False, Item=RigVMVariableNode() {variable=tail_rotate_position_proxyChannel}.Value) {template="GetAnimationChannelFromItem::Execute(out Value,in Item,in bInitial)"}
   Subtract = &RigVMFunction_MathIntSub(B=1, A=@GetAnimationChannelFromItem.Value) {template="Subtract::Execute(in A,in B,out Result)"}
   n4 = RigVMVariableNode(ExecuteContext=(), Value=@Subtract.Result) {variable=new_proxy_position}
   Not_Equals = RigVMDispatchNode(B=RigVMVariableNode() {variable=current_proxy_position}.Value, A=RigVMVariableNode() {variable=new_proxy_position}.Value) {template="DISPATCH_RigVMDispatch_CoreNotEquals(in A,in B,out Result)"}
   RigVMFunction_ControlFlowBranch = &RigVMFunction_ControlFlowBranch(Condition=@Not_Equals.Result, ExecuteContext=@n4.ExecuteContext)
   n5 = RigVMVariableNode() {variable=fk_controls}
   n6 = RigVMVariableNode() {variable=new_proxy_position}
   /Script/ControlRig.RigUnit_SwitchParent(
    Mode=ParentItem
    bMaintainGlobal=false
    Parent=&RigUnit_HierarchyGetParent(
     bDefaultParent=True
     Child=RigVMDispatchNode(Array=@n5.Value, Index=@n6.Value) {template="DISPATCH_RigVMDispatch_ArrayGetAtIndex(in Array,in Index,out Element)"}.Element
    ).Parent
    Child=RigVMVariableNode() {variable=tail_rotate_proxyControlParent}.Value
    ExecutePin=@RigVMFunction_ControlFlowBranch.True
   ) {is=struct RigVMUnitNode}
   > RigVMVariableNode(Value=RigVMVariableNode() {variable=new_proxy_position}.Value) {variable=current_proxy_position}
  ]}
 New Function_ContainedGraph {nodes [
   RigVMFunctionEntryNode()
   n1 = /Script/ControlRig.RigUnit_IsInteracting() {is=struct RigVMUnitNode}
   > RigVMVariableNode(Value=@n1.bIsRotating) {variable=is_interacting}
   > RigVMVariableNode(Value=@n1.Items) {variable=interacting_items}
   > RigVMFunctionReturnNode()
  ]}
 New Function_ContainedGraph {nodes [
   n1 = RigVMFunctionEntryNode()
   n2 = &RigVMFunction_Sequence(ExecuteContext=@n1.ExecuteContext)
   RigVMFunction_Sequence = RigVMAggregateNode(ExecuteContext=@n2.A)
   RigVMDispatchNode(Array=RigVMVariableNode() {variable=curve_transforms}.Value, ExecuteContext=@RigVMFunction_Sequence.A) {template="DISPATCH_RigVMDispatch_ArrayReset(io Array)"}
   > For_Each = RigVMDispatchNode(Array=RigVMVariableNode() {variable=fk_joints}.Value) {template="DISPATCH_RigVMDispatch_ArrayIterator(in Array,out Element,out Index,out Count,out Ratio)"}
   > RigVMVariableNode(Value=@For_Each.Element) {variable=item}
   > n3 = RigVMVariableNode(Value=@For_Each.Index) {variable=index}
   GetAnimationChannelFromItem = &RigUnit_GetIntAnimationChannelFromItem(bInitial=False, Item=RigVMVariableNode() {variable=tail_rotate_position_proxyChannel}.Value) {template="GetAnimationChannelFromItem::Execute(out Value,in Item,in bInitial)"}
   Subtract = &RigVMFunction_MathIntSub(B=1, A=@GetAnimationChannelFromItem.Value) {template="Subtract::Execute(in A,in B,out Result)"}
   GreaterEqual = &RigVMFunction_MathIntGreaterEqual(B=@Subtract.Result, A=RigVMVariableNode() {variable=index}.Value) {template="GreaterEqual::Execute(in A,in B,out Result)"}
   RigVMFunction_ControlFlowBranch = &RigVMFunction_ControlFlowBranch(Condition=@GreaterEqual.Result, ExecuteContext=@n3.ExecuteContext)
   LessEqual = /Script/RigVM.RigVMFunction_MathIntLessEqual(
    B=&RigUnit_GetIntAnimationChannelFromItem(bInitial=False, Item=RigVMVariableNode() {variable=tail_rotate_spread_proxyChannel}.Value) {template="GetAnimationChannelFromItem::Execute(out Value,in Item,in bInitial)"}.Value
    A=RigVMVariableNode() {variable=count}.Value
   ) {is=struct RigVMUnitNode, template="LessEqual::Execute(in A,in B,out Result)"}
   n4 = &RigVMFunction_ControlFlowBranch(Condition=@LessEqual.Result, ExecuteContext=@RigVMFunction_ControlFlowBranch.True)
   Add = RigVMDispatchNode(
    Element=&RigUnit_GetTransform(bInitial=False, Item=RigVMVariableNode() {variable=item}.Value).Transform.Translation
    ExecuteContext=@n4.True
    Array=RigVMVariableNode() {variable=curve_transforms}.Value
   ) {template="DISPATCH_RigVMDispatch_ArrayAdd(io Array,in Element,out Index)"}
   > RigVMVariableNode(Value=&RigVMFunction_MathIntAdd(B=1, A=RigVMVariableNode() {variable=count}.Value) {template="Add::Execute(in A,in B,out Result)"}.Result) {variable=count}
   RigVMFunction_MathIntToFloat = &RigVMFunction_MathIntToFloat(Value=&RigUnit_GetIntAnimationChannelFromItem(bInitial=False, Item=RigVMVariableNode() {variable=tail_rotate_feedbackThickness_proxyChannel}.Value) {template="GetAnimationChannelFromItem::Execute(out Value,in Item,in bInitial)"}.Value) {template="Cast::Execute(in Value,out Result)"}
   RigVMFunction_DebugLineStripNoSpace = /Script/RigVM.RigVMFunction_DebugLineStripNoSpace(
    Color=(R=1,G=0,B=0,A=1)
    DebugDrawSettings=(DepthPriority=SDPG_Foreground,Lifetime=-1)
    WorldOffset=(Rotation=(X=0,Y=0,Z=0,W=1),Translation=(X=0,Y=0,Z=0),Scale3D=(X=1,Y=1,Z=1))
    bEnabled=/Script/ControlRig.RigUnit_GetBoolAnimationChannelFromItem(bInitial=False, Item=RigVMVariableNode() {variable=tail_rotate_feedback_proxyChannel}.Value) {is=struct RigVMUnitNode, template="GetAnimationChannelFromItem::Execute(out Value,in Item,in bInitial)"}.Value
    Thickness=@RigVMFunction_MathIntToFloat.Result
    Points=RigVMVariableNode() {variable=curve_transforms}.Value
    ExecutePin=@RigVMFunction_Sequence.B
   ) {is=struct RigVMUnitNode}
   RigVMFunctionReturnNode(ExecuteContext=@n2.B)
  ]}
 New Function_ContainedGraph {nodes [
   RigVMFunctionEntryNode()
   > RigVMFunction_Sequence = RigVMAggregateNode()
   RigVMFunctionReturnNode(ExecuteContext=@RigVMFunction_Sequence.C)
   RigVMDispatchNode(Array=RigVMVariableNode() {variable=backwards_local_transforms}.Value, ExecuteContext=@RigVMFunction_Sequence.A) {template="DISPATCH_RigVMDispatch_ArrayReset(io Array)"}
   > For_Each = RigVMDispatchNode(Array=RigVMVariableNode() {variable=tail_chain_joints}.Value) {template="DISPATCH_RigVMDispatch_ArrayIterator(in Array,out Element,out Index,out Count,out Ratio)"}
   > RigVMVariableNode(Value=@For_Each.Element) {variable=joint}
   > n1 = RigVMVariableNode(Value=@For_Each.Index) {variable=index}
   n2 = &RigVMFunction_Sequence(ExecuteContext=@n1.ExecuteContext)
   RigVMVariableNode(
    Value=RigVMDispatchNode(Array=RigVMVariableNode() {variable=fk_controls}.Value, Index=RigVMVariableNode() {variable=index}.Value) {template="DISPATCH_RigVMDispatch_ArrayGetAtIndex(in Array,in Index,out Element)"}.Element
    ExecuteContext=@n2.A
   ) {variable=control}
   n3 = RigVMVariableNode() {variable=joint}
   n4 = RigVMVariableNode() {variable=control}
   Add = RigVMDispatchNode(
    Element=&RigVMFunction_MathTransformMakeRelative(
     Parent=&RigUnit_GetTransform(bInitial=False, Item=@n3.Value).Transform
     Global=&RigUnit_GetTransform(bInitial=False, Item=@n4.Value).Transform
    ).Local
    Array=RigVMVariableNode() {variable=backwards_local_transforms}.Value
    ExecuteContext=@n2.B
   ) {template="DISPATCH_RigVMDispatch_ArrayAdd(io Array,in Element,out Index)"}
   n5 = RigVMVariableNode() {variable=connector_tail_base}
   n6 = RigVMVariableNode() {variable=tail gimbal control}
   RigVMVariableNode(
    Value=&RigVMFunction_MathTransformMakeRelative(
     Parent=&RigUnit_GetTransform(bInitial=False, Item=@n5.Value).Transform
     Global=&RigUnit_GetTransform(bInitial=False, Item=@n6.Value).Transform
    ).Local
    ExecuteContext=@RigVMFunction_Sequence.B
   ) {variable=tail gimbal transform offset}
  ]}
 New Function_ContainedGraph {nodes [
   RigVMFunctionEntryNode()
   > RigVMFunction_Sequence = RigVMAggregateNode()
   For_Each = RigVMDispatchNode(Array=RigVMVariableNode() {variable=tail joint nulls}.Value, ExecuteContext=@RigVMFunction_Sequence.B) {template="DISPATCH_RigVMDispatch_ArrayIterator(in Array,out Element,out Index,out Count,out Ratio)"}
   > &RigUnit_SetTransform(
    Weight=1
    bInitial=False
    Value=&RigUnit_GetTransform(bInitial=False, Item=@For_Each.Element).Transform
    Item=RigVMDispatchNode(Array=RigVMVariableNode() {variable=fk_controls}.Value, Index=@For_Each.Index) {template="DISPATCH_RigVMDispatch_ArrayGetAtIndex(in Array,in Index,out Element)"}.Element
   )
   n1 = RigVMVariableNode() {variable=connector_tail_base}
   &RigUnit_SetTransform(
    Weight=1
    bInitial=False
    Value=&RigVMFunction_MathTransformMakeAbsolute(
     Parent=&RigUnit_GetTransform(bInitial=False, Item=@n1.Value).Transform
     Local=RigVMVariableNode() {variable=tail gimbal transform offset}.Value
    ) {template="Make Absolute::Execute(in Local,in Parent,out Global)"}.Global
    Item=RigVMVariableNode() {variable=tail gimbal control}.Value
    ExecutePin=@RigVMFunction_Sequence.A
   )
   RigVMFunctionReturnNode(ExecuteContext=@RigVMFunction_Sequence.C)
  ]}
 New Function_ContainedGraph {nodes [
   n1 = RigVMFunctionEntryNode()
   RigVMFunction_Sequence = &RigVMFunction_Sequence(ExecuteContext=@n1.ExecuteContext)
   RigVMFunction_Sequence_1 = RigVMAggregateNode(ExecuteContext=@RigVMFunction_Sequence.A)
   RigVMDispatchNode(Array=RigVMVariableNode() {variable=fk_offset_controls}.Value, ExecuteContext=@RigVMFunction_Sequence_1.A) {template="DISPATCH_RigVMDispatch_ArrayReset(io Array)"}
   > For_Each = RigVMDispatchNode(Array=RigVMVariableNode() {variable=fk_controls}.Value) {template="DISPATCH_RigVMDispatch_ArrayIterator(in Array,out Element,out Index,out Count,out Ratio)"}
   > RigVMVariableNode(Value=@For_Each.Index) {variable=Index}
   > RigVMFunction_Sequence_2 = RigVMAggregateNode()
   RigVMFunction_StringPadInteger = &RigVMFunction_StringPadInteger(Digits=2, Value=&RigVMFunction_MathIntAdd(B=1, A=@For_Each.Index) {template="Add::Execute(in A,in B,out Result)"}.Result)
   RigVMFunction_StringJoin = &RigVMFunction_StringJoin(Separator=" ", Values=("fk","","Offset"), Values.1=@RigVMFunction_StringPadInteger.Result)
   From_String = RigVMDispatchNode(String=@RigVMFunction_StringJoin.Result) {template="DISPATCH_RigDispatch_FromString(in String,out Result)"}
   Multiply = RigVMAggregateNode(C=1.3, B=RigVMVariableNode() {variable=output_scale_factor_lcl}.Value, A=RigVMVariableNode() {variable=FK Control Scale}.Value)
   SpawnControl = &RigUnit_HierarchyAddControlTransform(
    InitialValue=(Rotation=(X=0,Y=0,Z=0,W=1),Translation=(X=0,Y=0,Z=0),Scale3D=(X=1,Y=1,Z=1))
    OffsetSpace=LocalSpace
    OffsetTransform=(Rotation=(X=0,Y=0,Z=0,W=1),Translation=(X=0,Y=0,Z=0),Scale3D=(X=1,Y=1,Z=1))
    Settings="(InitialSpace=LocalSpace,bUsePreferredRotationOrder=False,PreferredRotationOrder=YZX,Limits=(LimitTranslationX=(bMinimum=False,bMaximum=False),LimitTranslationY=(bMinimum=False,bMaximum=False),LimitTranslationZ=(bMinimum=False,bMaximum=False),LimitPitch=(b ...(truncated)"
    Parent=@For_Each.Element
    Name=@From_String.Result
    Settings.Shape.Transform.Scale3D.X=@Multiply.Result
    Settings.Shape.Transform.Scale3D.Y=@Multiply.Result
    ExecutePin=@RigVMFunction_Sequence_2.C
   ) {template="SpawnControl::Execute(in InitialValue,in Settings,in OffsetTransform,in OffsetSpace,in Parent,in Name,out Item)"}
   > Add = RigVMDispatchNode(Array=RigVMVariableNode() {variable=fk_offset_controls}.Value, Element=@SpawnControl.Item) {template="DISPATCH_RigVMDispatch_ArrayAdd(io Array,in Element,out Index)"}
   RigVMFunction_MathIntToDouble = &RigVMFunction_MathIntToFloat(Value=&RigVMFunction_MathIntAdd(B=0, A=RigVMVariableNode() {variable=Index}.Value) {template="Add::Execute(in A,in B,out Result)"}.Result) {template="Cast::Execute(in Value,out Result)"}
   Multiply_2 = RigVMAggregateNode(A=-1, C=@RigVMFunction_MathIntToDouble.Result, B=RigVMVariableNode() {variable=scale_increment_multiplier_lcl}.Value)
   RigVMVariableNode(Value=&RigVMFunction_MathDoubleAdd(A=1, B=@Multiply_2.Result).Result, ExecuteContext=@RigVMFunction_Sequence_2.B) {variable=output_scale_factor_lcl}
   Divide = &RigVMFunction_MathDoubleDiv(A=1, B=RigVMVariableNode() {variable=Control Taper Scale}.Value)
   n2 = &RigVMFunction_MathDoubleMul(A=-1, B=@Divide.Result)
   RigVMFunction_MathIntToDouble_1 = &RigVMFunction_MathIntToFloat(Value=RigVMDispatchNode(Array=RigVMVariableNode() {variable=fk_joints}.Value) {template="DISPATCH_RigVMDispatch_ArrayGetNum(in Array,out Num)"}.Num) {template="Cast::Execute(in Value,out Result)"}
   RigVMVariableNode(
    Value=&RigVMFunction_MathDoubleDiv(A=&RigVMFunction_MathDoubleAdd(A=1, B=@n2.Result).Result, B=@RigVMFunction_MathIntToDouble_1.Result).Result
    ExecuteContext=@RigVMFunction_Sequence_2.A
   ) {variable=scale_increment_multiplier_lcl}
   For_Each_1 = RigVMDispatchNode(Array=RigVMVariableNode() {variable=fk_joints}.Value, ExecuteContext=@RigVMFunction_Sequence_1.C) {template="DISPATCH_RigVMDispatch_ArrayIterator(in Array,out Element,out Index,out Count,out Ratio)"}
   n3 = RigVMDispatchNode(Array=RigVMVariableNode() {variable=fk_offset_controls}.Value, Index=@For_Each_1.Index) {template="DISPATCH_RigVMDispatch_ArrayGetAtIndex(in Array,in Index,out Element)"}
   > Set_Metadata = RigVMDispatchNode(
    Name=BoneLocalTransform
    NameSpace=Self
    Value=&RigVMFunction_MathTransformMakeRelative(
     Parent=&RigUnit_GetTransform(bInitial=true, Item=@n3.Element).Transform
     Global=&RigUnit_GetTransform(bInitial=true, Item=@For_Each_1.Element).Transform
    ).Local
    Item=@n3.Element
   ) {template="DISPATCH_RigDispatch_SetMetadata(in Item,in Name,in NameSpace,in Value,out Success)"}
   n4 = &RigUnit_HierarchyAddAnimationChannelBool(
    InitialValue=false
    LimitsEnabled=()
    MaximumValue=True
    MinimumValue=false
    Name=fk offset controls Visibilty
    Parent=RigVMVariableNode() {variable=tail gimbal control}.Value
    ExecutePin=@RigVMFunction_Sequence_1.B
   ) {template="SpawnAnimationChannel::Execute(in InitialValue,in MinimumValue,in MaximumValue,in LimitsEnabled,in Parent,in Name,out Item)"}
   > /Script/ControlRig.RigUnit_SetChannelHosts(Channel=@n4.Item, Hosts=RigVMVariableNode() {variable=fk_offset_controls}.Value) {is=struct RigVMUnitNode}
   > RigVMVariableNode(Value=@n4.Item) {variable=fk_offset_controls_visibility}
   RigVMFunctionReturnNode(ExecuteContext=@RigVMFunction_Sequence.B)
   RigVMVariableNode() {variable=fk_controls}
  ]}
 New Function_ContainedGraph {nodes [
   RigVMFunctionEntryNode()
   > RigVMFunction_Sequence = RigVMAggregateNode()
   RigVMFunctionReturnNode(ExecuteContext=@RigVMFunction_Sequence.C)
   RigVMDispatchNode(Array=RigVMVariableNode() {variable=tail space nulls}.Value, ExecuteContext=@RigVMFunction_Sequence.B) {template="DISPATCH_RigVMDispatch_ArrayReset(io Array)"}
   > For_Each_1 = RigVMDispatchNode(Array=RigVMVariableNode() {variable=tail spaces}.Value) {template="DISPATCH_RigVMDispatch_ArrayIterator(in Array,out Element,out Index,out Count,out Ratio)"}
   > n1 = &RigUnit_HierarchyAddNull(
    Space=GlobalSpace
    Transform=&RigUnit_GetTransform(bInitial=False, Item=RigVMVariableNode() {variable=postion Parent Null}.Value).Transform
    Parent=@For_Each_1.Element
    Name=/Script/RigVM.RigVMFunction_NameConcat(B=" Orient Space", A=/Script/ControlRig.RigUnit_GetItemShortName(Item=@For_Each_1.Element) {is=struct RigVMUnitNode}.ShortName) {is=struct RigVMUnitNode, template="Concat::Execute(in A,in B,out Result)"}.Result
   )
   > RigVMDispatchNode(Array=RigVMVariableNode() {variable=tail space nulls}.Value, Element=@n1.Item) {template="DISPATCH_RigVMDispatch_ArrayAdd(io Array,in Element,out Index)"}
   > /Script/ControlRig.RigUnit_AddParent(DisplayLabel=None, Child=RigVMVariableNode() {variable=tail gimbal control}.Value, Parent=@n1.Item) {is=struct RigVMUnitNode}
   n2 = &RigUnit_HierarchyAddNull(
    Name=Parent Null
    Space=GlobalSpace
    Transform=(Rotation=(X=0,Y=0,Z=0,W=1),Translation=(X=0,Y=0,Z=0),Scale3D=(X=1,Y=1,Z=1))
    Transform.Translation=&RigUnit_GetTransform(bInitial=false, Item=RigVMVariableNode() {variable=tail gimbal control}.Value).Transform.Translation
    Parent=RigVMVariableNode() {variable=connector_hip_root}.Value
    ExecutePin=@RigVMFunction_Sequence.A
   )
   > RigVMVariableNode(Value=@n2.Item) {variable=postion Parent Null}
   RigVMVariableNode() {variable=postion Parent Null}
  ]}
 New Function_ContainedGraph {nodes [
   n1 = RigVMFunctionEntryNode()
   RigVMFunction_Sequence = &RigVMFunction_Sequence(ExecuteContext=@n1.ExecuteContext)
   RigVMFunctionReturnNode(ExecuteContext=@RigVMFunction_Sequence.B)
   For_Each_4 = RigVMDispatchNode(Array=RigVMVariableNode() {variable=tail spaces}.Value, ExecuteContext=@RigVMFunction_Sequence.A) {template="DISPATCH_RigVMDispatch_ArrayIterator(in Array,out Element,out Index,out Count,out Ratio)"}
   > &RigUnit_SetTransform(
    Value=(Rotation=(X=0,Y=0,Z=0,W=1),Translation=(X=0,Y=0,Z=0),Scale3D=(X=1,Y=1,Z=1))
    Weight=1
    bInitial=False
    Item=RigVMDispatchNode(Array=RigVMVariableNode() {variable=tail space nulls}.Value, Index=@For_Each_4.Index) {template="DISPATCH_RigVMDispatch_ArrayGetAtIndex(in Array,in Index,out Element)"}.Element
    Value.Translation=&RigUnit_GetTransform(bInitial=False, Item=RigVMVariableNode() {variable=postion Parent Null}.Value).Transform.Translation
    Value.Rotation=&RigUnit_GetTransform(bInitial=False, Item=@For_Each_4.Element).Transform.Rotation
   )
   Set Transform_1 = &RigUnit_SetRotation(ExecutePin=(), Item=(Type=Null,Name="None"), Space=GlobalSpace, Value=(X=0,Y=0,Z=0,W=1), Weight=1, bInitial=False)
  ]}
 New Function_ContainedGraph {nodes [
   n1 = RigVMFunctionEntryNode()
   RigVMFunction_Sequence = &RigVMFunction_Sequence(ExecuteContext=@n1.ExecuteContext)
   RigVMDispatchNode(ExecuteContext=@RigVMFunction_Sequence.A, Array=RigVMVariableNode() {variable=tail joint nulls}.Value) {template="DISPATCH_RigVMDispatch_ArrayReset(io Array)"}
   > For_Each = RigVMDispatchNode(Array=RigVMVariableNode() {variable=tail_chain_joints}.Value) {template="DISPATCH_RigVMDispatch_ArrayIterator(in Array,out Element,out Index,out Count,out Ratio)"}
   n2 = RigVMVariableNode() {variable=fk_controls}
   > n3 = &RigUnit_HierarchyAddNull(
    Name=TailJntNull
    Space=GlobalSpace
    Parent=@For_Each.Element
    Transform=&RigUnit_GetTransform(
     bInitial=False
     Item=RigVMDispatchNode(Index=@For_Each.Index, Array=@n2.Value) {template="DISPATCH_RigVMDispatch_ArrayGetAtIndex(in Array,in Index,out Element)"}.Element
    ).Transform
   )
   > Add = RigVMDispatchNode(Element=@n3.Item, Array=RigVMVariableNode() {variable=tail joint nulls}.Value) {template="DISPATCH_RigVMDispatch_ArrayAdd(io Array,in Element,out Index)"}
   RigVMFunctionReturnNode(ExecuteContext=@RigVMFunction_Sequence.B)
  ]}
 New Function_ContainedGraph {nodes [
   RigVMFunctionEntryNode()
   > RigVMFunction_Sequence = RigVMAggregateNode()
   RigVMFunctionReturnNode(ExecuteContext=@RigVMFunction_Sequence.C)
   n1 = &RigUnit_HierarchyAddNull(
    Name=TailGimbalNull
    Space=GlobalSpace
    Transform=(Rotation=(X=0,Y=0,Z=0,W=1),Translation=(X=0,Y=0,Z=0),Scale3D=(X=1,Y=1,Z=1))
    Transform.Translation=&RigUnit_GetTransform(bInitial=false, Item=RigVMVariableNode() {variable=connector_tail_base}.Value).Transform.Translation
    Parent=RigVMVariableNode() {variable=current_parent_item}.Value
    ExecutePin=@RigVMFunction_Sequence.B
   )
   > RigVMVariableNode(Value=@n1.Item) {variable=tail gimbal null}
   > n2 = &RigUnit_HierarchyAddControlTransform(
    InitialValue=(Rotation=(X=0,Y=0,Z=0,W=1),Translation=(X=0,Y=0,Z=0),Scale3D=(X=1,Y=1,Z=1))
    Name=TailGimbal
    OffsetSpace=LocalSpace
    OffsetTransform=(Rotation=(X=0,Y=0,Z=0,W=1),Translation=(X=0,Y=0,Z=0),Scale3D=(X=1,Y=1,Z=1))
    Settings="(InitialSpace=LocalSpace,bUsePreferredRotationOrder=true,PreferredRotationOrder=YZX,Limits=(LimitTranslationX=(bMinimum=False,bMaximum=False),LimitTranslationY=(bMinimum=False,bMaximum=False),LimitTranslationZ=(bMinimum=False,bMaximum=False),LimitPitch=(bM ...(truncated)"
    Parent=@n1.Item
    Settings.Shape.Transform=RigVMVariableNode() {variable=Gimbal_Control_Position}.Value
   ) {template="SpawnControl::Execute(in InitialValue,in Settings,in OffsetTransform,in OffsetSpace,in Parent,in Name,out Item)"}
   > RigVMVariableNode(Value=@n2.Item) {variable=tail gimbal control}
   RigVMVariableNode(Value=RigVMVariableNode() {variable=connector_hip_root}.Value, ExecuteContext=@RigVMFunction_Sequence.A) {variable=current_parent_item}
  ]}
 New Function_ContainedGraph {nodes [
   n1 = RigVMFunctionEntryNode()
   > RigVMVariableNode(Value=@n1.taper_scale) {variable=Entry_taper_scale}
   > RigVMVariableNode(Value=@n1.item_count) {variable=Entry_item_count}
   > RigVMVariableNode(Value=@n1.item_index) {variable=Entry_item_index}
   > n2 = RigVMVariableNode(Value=1) {variable=Output_scale_factor}
   Greater = &RigVMFunction_MathIntGreater(B=0, A=RigVMVariableNode() {variable=Entry_item_count}.Value) {template="Greater::Execute(in A,in B,out Result)"}
   RigVMFunction_ControlFlowBranch = &RigVMFunction_ControlFlowBranch(Condition=@Greater.Result, ExecuteContext=@n2.ExecuteContext)
   RigVMFunctionReturnNode(ExecuteContext=@RigVMFunction_ControlFlowBranch.Completed, scale_factor=RigVMVariableNode() {variable=Output_scale_factor}.Value)
   RigVMFunction_Sequence = RigVMAggregateNode(ExecuteContext=@RigVMFunction_ControlFlowBranch.True)
   Divide = &RigVMFunction_MathDoubleDiv(A=1, B=RigVMVariableNode() {variable=Entry_taper_scale}.Value)
   Multiply = &RigVMFunction_MathDoubleMul(A=-1, B=@Divide.Result)
   Add = &RigVMFunction_MathDoubleAdd(A=1, B=@Multiply.Result)
   RigVMFunction_MathIntToDouble_1 = &RigVMFunction_MathIntToFloat(Value=RigVMVariableNode() {variable=Entry_item_count}.Value) {template="Cast::Execute(in Value,out Result)"}
   RigVMVariableNode(
    Value=&RigVMFunction_MathDoubleDiv(A=@Add.Result, B=@RigVMFunction_MathIntToDouble_1.Result).Result
    ExecuteContext=@RigVMFunction_Sequence.A
   ) {variable=scale_increment_mulitplier}
   RigVMFunction_MathIntToDouble = &RigVMFunction_MathIntToFloat(Value=&RigVMFunction_MathIntAdd(B=1, A=RigVMVariableNode() {variable=Entry_item_index}.Value) {template="Add::Execute(in A,in B,out Result)"}.Result) {template="Cast::Execute(in Value,out Result)"}
   Multiply_1 = RigVMAggregateNode(A=-1, C=@RigVMFunction_MathIntToDouble.Result, B=RigVMVariableNode() {variable=scale_increment_mulitplier}.Value)
   RigVMVariableNode(Value=&RigVMFunction_MathDoubleAdd(A=1, B=@Multiply_1.Result).Result, ExecuteContext=@RigVMFunction_Sequence.B) {variable=Output_scale_factor}
  ]}
 New Function_ContainedGraph {nodes [
   n1 = RigVMFunctionEntryNode()
   RigVMFunction_Sequence = &RigVMFunction_Sequence(ExecuteContext=@n1.ExecuteContext)
   RigVMFunctionReturnNode(ExecuteContext=@RigVMFunction_Sequence.B)
   GetAnimationChannelFromItem = &RigUnit_GetBoolAnimationChannelFromItem(bInitial=False, Item=RigVMVariableNode() {variable=fk_offset_controls_visibility}.Value) {template="GetAnimationChannelFromItem::Execute(out Value,in Item,in bInitial)"}
   Set_Module_Metadata = RigVMDispatchNode(Name=None, NameSpace=Self, Value=@GetAnimationChannelFromItem.Value, ExecuteContext=@RigVMFunction_Sequence.A) {template="DISPATCH_RigDispatch_SetModuleMetadata(in Name,in NameSpace,in Value,out Success)"}
   > For_Each = RigVMDispatchNode(Array=RigVMVariableNode() {variable=fk_offset_controls}.Value) {template="DISPATCH_RigVMDispatch_ArrayIterator(in Array,out Element,out Index,out Count,out Ratio)"}
   > SetControlVisibility = /Script/ControlRig.RigUnit_SetControlVisibility(Item=@For_Each.Element, bVisible=@GetAnimationChannelFromItem.Value) {is=struct RigVMUnitNode, template="SetControlVisibility::Execute(in Item,in Pattern,in bVisible)"}
  ]}
 RigVMFunctionLibrary {nodes [
   Connect Orient Space Switches = RigVMCollapseNode(ExecuteContext=())
   calculate_tapered_scale = RigVMCollapseNode(ExecuteContext=())
   check_if_interacting_with_controls = RigVMCollapseNode(ExecuteContext=())
   create Space Switches = RigVMCollapseNode(ExecuteContext=())
   create gimal control = RigVMCollapseNode(ExecuteContext=())
   create_chain_fk_controls = RigVMCollapseNode(ExecuteContext=())
   create_chain_fk_joints = RigVMCollapseNode(ExecuteContext=())
   create_fk_offset_controls = RigVMCollapseNode(ExecuteContext=())
   create_joint_nulls = RigVMCollapseNode(ExecuteContext=())
   create_name = RigVMCollapseNode(ExecuteContext=(), input_node=(Type=None,Name="None"))
   create_proxy_controls = RigVMCollapseNode(ExecuteContext=())
   find_tail_chain = RigVMCollapseNode(ExecuteContext=())
   get_backwards_local_transforms = RigVMCollapseNode(ExecuteContext=())
   get_fk_joint_local_transforms = RigVMCollapseNode(ExecuteContext=())
   proxy_feedback = RigVMCollapseNode(ExecuteContext=())
   set_controls_from_nulls = RigVMCollapseNode(ExecuteContext=())
   set_joints_from_controls = RigVMCollapseNode(ExecuteContext=())
   set_offset_fk_control_visibility = RigVMCollapseNode(ExecuteContext=())
   set_proxy_position = RigVMCollapseNode(ExecuteContext=())
   set_proxy_rotation = RigVMCollapseNode(ExecuteContext=())
  ]}
 RigVMFunction_Sequence_1_1_ContainedGraph {nodes [
   n1 = RigVMFunctionEntryNode()
   n2 = &RigVMFunction_Sequence(ExecuteContext=@n1.ExecuteContext)
   n3 = &RigVMFunction_Sequence(ExecuteContext=@n2.B)
   n4 = &RigVMFunction_Sequence(ExecuteContext=@n3.B)
   RigVMFunctionReturnNode(A=@n2.A, B=@n3.A, C=@n4.A, D=@n4.B)
  ]}
 RigVMFunction_Sequence_1_1_ContainedGraph {nodes [
   n1 = RigVMFunctionEntryNode()
   n2 = &RigVMFunction_Sequence(ExecuteContext=@n1.ExecuteContext)
   n3 = &RigVMFunction_Sequence(ExecuteContext=@n2.B)
   n4 = &RigVMFunction_Sequence(ExecuteContext=@n3.B)
   n5 = &RigVMFunction_Sequence(ExecuteContext=@n4.B)
   RigVMFunctionReturnNode(A=@n2.A, B=@n3.A, C=@n4.A, D=@n5.A, E=@n5.B)
  ]}
 RigVMFunction_Sequence_1_ContainedGraph {nodes [
   n1 = RigVMFunctionEntryNode()
   RigVMFunction_Sequence = &RigVMFunction_Sequence(ExecuteContext=@n1.ExecuteContext)
   n2 = &RigVMFunction_Sequence(ExecuteContext=@RigVMFunction_Sequence.B)
   n3 = &RigVMFunction_Sequence(ExecuteContext=@n2.B)
   n4 = &RigVMFunction_Sequence(ExecuteContext=@n3.B)
   RigVMFunctionReturnNode(A=@RigVMFunction_Sequence.A, B=@n2.A, C=@n3.A, D=@n4.A, E=@n4.B)
  ]}
 RigVMFunction_Sequence_1_ContainedGraph {nodes [
   n1 = RigVMFunctionEntryNode()
   RigVMFunction_Sequence = &RigVMFunction_Sequence(ExecuteContext=@n1.ExecuteContext)
   n2 = &RigVMFunction_Sequence(ExecuteContext=@RigVMFunction_Sequence.B)
   n3 = &RigVMFunction_Sequence(ExecuteContext=@n2.B)
   n4 = &RigVMFunction_Sequence(ExecuteContext=@n3.B)
   RigVMFunctionReturnNode(A=@RigVMFunction_Sequence.A, B=@n2.A, C=@n3.A, D=@n4.A, E=@n4.B)
  ]}
 RigVMFunction_Sequence_1_ContainedGraph {nodes [
   n1 = RigVMFunctionEntryNode()
   RigVMFunction_Sequence = &RigVMFunction_Sequence(ExecuteContext=@n1.ExecuteContext)
   n2 = &RigVMFunction_Sequence(ExecuteContext=@RigVMFunction_Sequence.B)
   RigVMFunctionReturnNode(A=@RigVMFunction_Sequence.A, B=@n2.A, C=@n2.B)
  ]}
 RigVMFunction_Sequence_1_ContainedGraph {nodes [
   n1 = RigVMFunctionEntryNode()
   RigVMFunction_Sequence = &RigVMFunction_Sequence(ExecuteContext=@n1.ExecuteContext)
   n2 = &RigVMFunction_Sequence(ExecuteContext=@RigVMFunction_Sequence.B)
   RigVMFunctionReturnNode(A=@RigVMFunction_Sequence.A, B=@n2.A, C=@n2.B)
  ]}
 RigVMFunction_Sequence_1_ContainedGraph {nodes [
   n1 = RigVMFunctionEntryNode()
   RigVMFunction_Sequence = &RigVMFunction_Sequence(ExecuteContext=@n1.ExecuteContext)
   n2 = &RigVMFunction_Sequence(ExecuteContext=@RigVMFunction_Sequence.B)
   RigVMFunctionReturnNode(A=@RigVMFunction_Sequence.A, B=@n2.A, C=@n2.B)
  ]}
 RigVMFunction_Sequence_1_ContainedGraph {nodes [
   n1 = RigVMFunctionEntryNode()
   RigVMFunction_Sequence = &RigVMFunction_Sequence(ExecuteContext=@n1.ExecuteContext)
   n2 = &RigVMFunction_Sequence(ExecuteContext=@RigVMFunction_Sequence.B)
   RigVMFunctionReturnNode(A=@RigVMFunction_Sequence.A, B=@n2.A, C=@n2.B)
  ]}
 RigVMFunction_Sequence_1_ContainedGraph {nodes [
   n1 = RigVMFunctionEntryNode()
   RigVMFunction_Sequence = &RigVMFunction_Sequence(ExecuteContext=@n1.ExecuteContext)
   n2 = &RigVMFunction_Sequence(ExecuteContext=@RigVMFunction_Sequence.B)
   RigVMFunctionReturnNode(A=@RigVMFunction_Sequence.A, B=@n2.A, C=@n2.B)
  ]}
 RigVMFunction_Sequence_1_ContainedGraph {nodes [
   n1 = RigVMFunctionEntryNode()
   RigVMFunction_Sequence = &RigVMFunction_Sequence(ExecuteContext=@n1.ExecuteContext)
   n2 = &RigVMFunction_Sequence(ExecuteContext=@RigVMFunction_Sequence.B)
   RigVMFunctionReturnNode(A=@RigVMFunction_Sequence.A, B=@n2.A, C=@n2.B)
  ]}
 RigVMFunction_Sequence_2_ContainedGraph {nodes [
   n1 = RigVMFunctionEntryNode()
   n2 = &RigVMFunction_Sequence(ExecuteContext=@n1.ExecuteContext)
   n3 = &RigVMFunction_Sequence(ExecuteContext=@n2.B)
   RigVMFunctionReturnNode(A=@n2.A, B=@n3.A, C=@n3.B)
  ]}
 RigVMFunction_Sequence_2_ContainedGraph {nodes [
   n1 = RigVMFunctionEntryNode()
   RigVMFunction_Sequence = &RigVMFunction_Sequence(ExecuteContext=@n1.ExecuteContext)
   n2 = &RigVMFunction_Sequence(ExecuteContext=@RigVMFunction_Sequence.B)
   RigVMFunctionReturnNode(A=@RigVMFunction_Sequence.A, B=@n2.A, C=@n2.B)
  ]}
 RigVMFunction_Sequence_2_ContainedGraph {nodes [
   n1 = RigVMFunctionEntryNode()
   RigVMFunction_Sequence = &RigVMFunction_Sequence(ExecuteContext=@n1.ExecuteContext)
   n2 = &RigVMFunction_Sequence(ExecuteContext=@RigVMFunction_Sequence.B)
   RigVMFunctionReturnNode(A=@RigVMFunction_Sequence.A, B=@n2.A, C=@n2.B)
  ]}
 RigVMFunction_Sequence_2_ContainedGraph {nodes [
   n1 = RigVMFunctionEntryNode()
   n2 = &RigVMFunction_Sequence(ExecuteContext=@n1.ExecuteContext)
   n3 = &RigVMFunction_Sequence(ExecuteContext=@n2.B)
   RigVMFunctionReturnNode(A=@n2.A, B=@n3.A, C=@n3.B)
  ]}
 RigVMFunction_Sequence_3_ContainedGraph {nodes [
   n1 = RigVMFunctionEntryNode()
   RigVMFunction_Sequence = &RigVMFunction_Sequence(ExecuteContext=@n1.ExecuteContext)
   n2 = &RigVMFunction_Sequence(ExecuteContext=@RigVMFunction_Sequence.B)
   RigVMFunctionReturnNode(A=@RigVMFunction_Sequence.A, B=@n2.A, C=@n2.B)
  ]}
 RigVMFunction_Sequence_3_ContainedGraph {nodes [
   n1 = RigVMFunctionEntryNode()
   n2 = &RigVMFunction_Sequence(ExecuteContext=@n1.ExecuteContext)
   n3 = &RigVMFunction_Sequence(ExecuteContext=@n2.B)
   RigVMFunctionReturnNode(B=@n3.A, C=@n3.B, A=@n2.A)
  ]}
 RigVMModel {nodes [
   RigUnit_BeginExecution = &RigUnit_BeginExecution()
   RigVMFunction_Sequence = RigVMAggregateNode(ExecuteContext=@RigUnit_BeginExecution.ExecutePin)
   set_proxy_position = RigVMFunctionReferenceNode(ExecuteContext=@RigVMFunction_Sequence.A)
   check_if_interacting_with_controls = RigVMFunctionReferenceNode(ExecuteContext=@RigVMFunction_Sequence.B)
   > set_proxy_rotation = RigVMFunctionReferenceNode()
   set_joints_from_controls = RigVMFunctionReferenceNode(ExecuteContext=@RigVMFunction_Sequence.C)
   proxy_feedback = RigVMFunctionReferenceNode(ExecuteContext=@RigVMFunction_Sequence.D)
   Connect Orient Space Switches = RigVMFunctionReferenceNode(ExecuteContext=@RigVMFunction_Sequence.E)
   > set_offset_fk_control_visibility = RigVMFunctionReferenceNode()
  ]}
 RigVMModel Backward Solve Graph {nodes [
   n1 = /Script/ControlRig.RigUnit_InverseExecution() {is=struct RigVMUnitNode}
   set_controls_from_nulls = RigVMFunctionReferenceNode(ExecuteContext=@n1.ExecutePin)
  ]}
 RigVMModel Connector Event Graph {nodes [/Script/ControlRig.RigUnit_ConnectorExecution() {is=struct RigVMUnitNode}]}
]
hierarchy [
 HipRoot {type=Connector, loc=<0 0 0>, rot=<0 0 0 1>}
 Root {type=Socket, loc=<0 0 0>, rot=<0 0 0 1>}
 TailTip {type=Connector, loc=<0 0 0>, rot=<0 0 0 1>}
 TailBase {type=Connector, loc=<0 0 0>, rot=<0 0 0 1>}
]
variables [
 Control Taper Scale {type=double, default=1}
 FK Control Scale {type=double, default=1}
 Gimbal_Control_Position {type=FTransform, default=(Rotation=(X=0,Y=0,Z=0,W=1),Translation=(X=0,Y=-3,Z=15),Scale3D=(X=0.3,Y=0.3,Z=0.3))}
 Index {type=int32, default=0}
 Proxy Control Offset {type=FTransform, default=(Rotation=(X=0,Y=0,Z=0,W=1),Translation=(X=0,Y=0,Z=0),Scale3D=(X=1,Y=1,Z=1))}
 Proxy shape position {type=FTransform, default=(Rotation=(X=0,Y=0,Z=0,W=1),Translation=(X=0,Y=0,Z=18),Scale3D=(X=0.3,Y=0.3,Z=0.3))}
 backwards_local_transforms {type="TArray<FTransform>"}
 basename {type=FString, default=tail}
 connector_hip_root {type=FRigElementKey, default=(Type=Connector,Name="")}
 connector_tail_base {type=FRigElementKey, default=(Type=Connector,Name="")}
 connector_tail_tip {type=FRigElementKey, default=(Type=Connector,Name="")}
 current_item {type=FRigElementKey, default=(Type=None,Name="")}
 current_proxy_position {type=int32, default=0}
 fk_controls {type="TArray<FRigElementKey>"}
 fk_joint_local_transforms {type="TArray<FTransform>"}
 fk_joints {type="TArray<FRigElementKey>"}
 fk_offset_controls {type="TArray<FRigElementKey>"}
 fk_offset_controls_visibility {type=FRigElementKey, default=(Type=None,Name="")}
 interacting_items {type="TArray<FRigElementKey>"}
 is_interacting {type=bool, default=False}
 odd_control_color {type=FLinearColor, default=(R=0,G=0.2344,B=0,A=1)}
 output_item_scale {type=double, default=0}
 postion Parent Null {type=FRigElementKey, default=(Type=None,Name="")}
 previous_rotation {type=FVector, default=(X=0,Y=0,Z=0)}
 scale_increment_factor {type=double, default=0}
 tail gimbal control {type=FRigElementKey, default=(Type=Bone,Name="")}
 tail gimbal null {type=FRigElementKey, default=(Type=Bone,Name="")}
 tail gimbal transform offset {type=FTransform, default=(Rotation=(X=0,Y=0,Z=0,W=1),Translation=(X=0,Y=0,Z=0),Scale3D=(X=1,Y=1,Z=1))}
 tail joint nulls {type="TArray<FRigElementKey>"}
 tail space nulls {type="TArray<FRigElementKey>"}
 tail spaces {type="TArray<FRigElementKey>"}
 tail_base_found {type=bool, default=False}
 tail_chain_joints {type="TArray<FRigElementKey>"}
 tail_rotate_feedbackThickness_proxyChannel {type=FRigElementKey, default=(Type=Bone,Name="")}
 tail_rotate_feedback_proxyChannel {type=FRigElementKey, default=(Type=Bone,Name="")}
 tail_rotate_position_proxyChannel {type=FRigElementKey, default=(Type=Bone,Name="")}
 tail_rotate_proxyControl {type=FRigElementKey, default=(Type=Bone,Name="")}
 tail_rotate_proxyControlParent {type=FRigElementKey, default=(Type=Bone,Name="")}
 tail_rotate_spread_proxyChannel {type=FRigElementKey, default=(Type=Bone,Name="")}
]
