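# Control Rig module that builds and drives a neck and head rig from Root, NeckTip, Head and parent connectors.
# Construction creates three helper joints, Neck 1, Neck 2, Neck IK and Head controls, bone offsets, and orientation-space nulls.
# Forward solve rotates space nulls, position-blends the Neck IK's parent null between head and Neck 1, then drives bones through the helper joints.
# Backward solve places controls from the connected skeleton.
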ControlRigBlueprint /WhisperRemedyAssets/Assets/Modules/CRM_WhisperRemedy_Neck_01
graphs [
 ConstructionGraph {nodes [
   RigUnit_PrepareForExecution = &RigUnit_PrepareForExecution()
   RigVMFunction_Sequence = RigVMAggregateNode(ExecuteContext=@RigUnit_PrepareForExecution.ExecutePin)
   Find Joints = RigVMFunctionReferenceNode(ExecuteContext=@RigVMFunction_Sequence.A)
   Build Joints = RigVMFunctionReferenceNode(ExecuteContext=@RigVMFunction_Sequence.B)
   > Create Controls = RigVMFunctionReferenceNode()
   Get Local Transforms = RigVMFunctionReferenceNode(ExecuteContext=@RigVMFunction_Sequence.C)
   > Create Space Switches = RigVMFunctionReferenceNode()
   Build Nulls = RigVMFunctionReferenceNode(ExecuteContext=())
  ]}
 New Function_ContainedGraph {nodes [
   RigVMFunctionEntryNode()
   > RigVMFunction_Sequence = RigVMAggregateNode()
   RigVMFunctionReturnNode(ExecuteContext=@RigVMFunction_Sequence.C)
   RigVMVariableNode(
    Value=/Script/ControlRig.RigUnit_ResolveConnector(Connector=(Type=Connector,Name="Root"), SkipSocket=False) {is=struct RigVMUnitNode}.Result
    ExecuteContext=@RigVMFunction_Sequence.A
   ) {variable=Connector Root}
   > RigVMVariableNode(Value=&RigUnit_ResolveConnector(Connector=(Type=Connector,Name="NeckTip"), SkipSocket=False).Result) {variable=Connector Neck Tip}
   > RigVMVariableNode(Value=&RigUnit_ResolveConnector(Connector=(Type=Connector,Name="Head"), SkipSocket=False).Result) {variable=Connector Head}
   > RigVMVariableNode(Value=&RigUnit_ResolveConnector(Connector=(Type=Connector,Name="parent"), SkipSocket=False).Result) {variable=Connector Parent}
   RigVMVariableNode(
    Value=/Script/ControlRig.RigUnit_HierarchyGetChainItemArray(
     bIncludeEnd=True
     bIncludeStart=True
     bReverse=False
     End=RigVMVariableNode() {variable=Connector Head}.Value
     Start=RigVMVariableNode() {variable=Connector Root}.Value
    ) {is=struct RigVMUnitNode}.Chain
    ExecuteContext=@RigVMFunction_Sequence.B
   ) {variable=Neck Bones}
  ]}
 New Function_ContainedGraph {nodes [
   /Script/ControlRig.RigUnit_AddParent(Child=(Type=Bone,Name="Neck Build 02"), DisplayLabel=None, ExecutePin=(), Parent=(Type=Bone,Name="Neck Build 01")) {is=struct RigVMUnitNode}
   > &RigUnit_AddParent(Child=(Type=Bone,Name="Head Build"), DisplayLabel=None, Parent=(Type=Bone,Name="Neck Build 02"))
   RigVMFunctionEntryNode()
   > RigVMFunction_Sequence = RigVMAggregateNode()
   n1 = /Script/ControlRig.RigUnit_AimBoneMath(
    DebugSettings=(bEnabled=False,Scale=10,WorldOffset=(Rotation=(X=0,Y=0,Z=0,W=1),Translation=(X=0,Y=0,Z=0),Scale3D=(X=1,Y=1,Z=1)))
    Primary=(Weight=1,Axis=(X=1,Y=0,Z=0),Target=(X=1,Y=0,Z=0),Kind=Location,Space=(Type=Bone,Name="None"))
    Secondary=(Weight=1,Axis=(X=1,Y=0,Z=0),Target=(X=1,Y=0,Z=0),Kind=Direction,Space=(Type=Bone,Name="None"))
    Weight=1
    Primary.Target=&RigUnit_GetTransform(bInitial=False, Item=&RigUnit_ResolveConnector(Connector=(Type=Connector,Name="Head"), SkipSocket=False).Result).Transform.Translation
    InputTransform=&RigUnit_GetTransform(bInitial=False, Item=&RigUnit_ResolveConnector(Connector=(Type=Connector,Name="NeckTip"), SkipSocket=False).Result).Transform
   ) {is=struct RigVMUnitNode}
   n2 = /Script/ControlRig.RigUnit_HierarchyAddBone(
    Name=Neck Build 02
    Space=GlobalSpace
    Transform=@n1.Result
    Parent=RigVMDispatchNode(Index=-1, Array=RigVMVariableNode() {variable=Virtual Bones}.Value) {template="DISPATCH_RigVMDispatch_ArrayGetAtIndex(in Array,in Index,out Element)"}.Element
    ExecutePin=@RigVMFunction_Sequence.B
   ) {is=struct RigVMUnitNode}
   > RigVMVariableNode(Value=@n2.Item) {variable=Neck 02 Build Joint}
   > RigVMDispatchNode(Array=RigVMVariableNode() {variable=Virtual Bones}.Value, Element=@n2.Item) {template="DISPATCH_RigVMDispatch_ArrayAdd(io Array,in Element,out Index)"}
   n3 = &RigUnit_AimBoneMath(
    DebugSettings=(bEnabled=False,Scale=10,WorldOffset=(Rotation=(X=0,Y=0,Z=0,W=1),Translation=(X=0,Y=0,Z=0),Scale3D=(X=1,Y=1,Z=1)))
    Primary=(Weight=1,Axis=(X=1,Y=0,Z=0),Target=(X=1,Y=0,Z=0),Kind=Location,Space=(Type=Bone,Name="None"))
    Secondary=(Weight=1,Axis=(X=1,Y=0,Z=0),Target=(X=1,Y=0,Z=0),Kind=Direction,Space=(Type=Bone,Name="None"))
    Weight=1
    Primary.Target=&RigUnit_GetTransform(bInitial=False, Item=&RigUnit_ResolveConnector(Connector=(Type=Connector,Name="Head"), SkipSocket=False).Result).Transform.Translation
    InputTransform=&RigUnit_GetTransform(bInitial=False, Item=&RigUnit_ResolveConnector(Connector=(Type=Connector,Name="Root"), SkipSocket=False).Result).Transform
   )
   n4 = &RigUnit_HierarchyAddBone(Name=Neck Build 01, Parent=(Type=Bone,Name="None"), Space=LocalSpace, Transform=@n3.Result, ExecutePin=@RigVMFunction_Sequence.A)
   > RigVMVariableNode(Value=@n4.Item) {variable=Neck 01 Build Joint}
   > RigVMDispatchNode(Array=RigVMVariableNode() {variable=Virtual Bones}.Value, Element=@n4.Item) {template="DISPATCH_RigVMDispatch_ArrayAdd(io Array,in Element,out Index)"}
   n5 = &RigUnit_HierarchyAddBone(
    Name=Head Build
    Space=GlobalSpace
    Parent=RigVMDispatchNode(Index=-1, Array=RigVMVariableNode() {variable=Virtual Bones}.Value) {template="DISPATCH_RigVMDispatch_ArrayGetAtIndex(in Array,in Index,out Element)"}.Element
    Transform=&RigUnit_GetTransform(bInitial=False, Item=&RigUnit_ResolveConnector(Connector=(Type=Connector,Name="Head"), SkipSocket=False).Result).Transform
    ExecutePin=@RigVMFunction_Sequence.C
   )
   > RigVMVariableNode(Value=@n5.Item) {variable=Head Build Joint}
   > RigVMDispatchNode(Array=RigVMVariableNode() {variable=Virtual Bones}.Value, Element=@n5.Item) {template="DISPATCH_RigVMDispatch_ArrayAdd(io Array,in Element,out Index)"}
   RigVMFunctionReturnNode(ExecuteContext=@RigVMFunction_Sequence.E)
   RigVMRerouteNode(Value=@RigVMFunction_Sequence.D)
   RigVMVariableNode() {variable=Neck 02 Build Joint}
  ]}
 New Function_ContainedGraph {nodes [
   n1 = RigVMFunctionEntryNode()
   RigVMFunction_Sequence = &RigVMFunction_Sequence(ExecuteContext=@n1.ExecuteContext)
   RigVMFunctionReturnNode(ExecuteContext=@RigVMFunction_Sequence.B)
   n2 = &RigUnit_HierarchyAddNull(
    Name=Neck Space Null
    Space=GlobalSpace
    Transform=&RigUnit_GetTransform(bInitial=False, Item=RigVMVariableNode() {variable=Neck 01 Build Joint}.Value).Transform
    Parent=RigVMVariableNode() {variable=Connector Parent}.Value
    ExecutePin=@RigVMFunction_Sequence.A
   )
   n3 = &RigVMFunction_MathDoubleMul(B=3, A=RigVMVariableNode() {variable=Control Scale}.Value)
   > n4 = /Script/ControlRig.RigUnit_HierarchyAddControlTransform(
    InitialValue=(Rotation=(X=0,Y=0,Z=0,W=1),Translation=(X=0,Y=0,Z=0),Scale3D=(X=1,Y=1,Z=1))
    Name=Neck 1
    OffsetSpace=GlobalSpace
    Settings="(InitialSpace=LocalSpace,bUsePreferredRotationOrder=False,PreferredRotationOrder=YZX,Limits=(LimitTranslationX=(bMinimum=False,bMaximum=False),LimitTranslationY=(bMinimum=False,bMaximum=False),LimitTranslationZ=(bMinimum=False,bMaximum=False),LimitPitch=(b ...(truncated)"
    OffsetTransform=&RigUnit_GetTransform(bInitial=False, Item=RigVMVariableNode() {variable=Neck 01 Build Joint}.Value).Transform
    Settings.Shape.Transform.Scale3D.X=@n3.Result
    Settings.Shape.Transform.Scale3D.Y=@n3.Result
    Settings.Shape.Transform.Scale3D.Z=@n3.Result
    Parent=@n2.Item
   ) {
    is=struct RigVMUnitNode
    template="SpawnControl::Execute(in InitialValue,in Settings,in OffsetTransform,in OffsetSpace,in Parent,in Name,out Item)"
   }
   > RigVMVariableNode(Value=@n4.Item) {variable=Neck 1 Cntrl}
   > RigVMDispatchNode(Array=RigVMVariableNode() {variable=Neck Controls}.Value, Element=@n4.Item) {template="DISPATCH_RigVMDispatch_ArrayAdd(io Array,in Element,out Index)"}
   n5 = &RigVMFunction_MathDoubleMul(B=3, A=RigVMVariableNode() {variable=Control Scale}.Value)
   > n6 = &RigUnit_HierarchyAddControlTransform(
    InitialValue=(Rotation=(X=0,Y=0,Z=0,W=1),Translation=(X=0,Y=0,Z=0),Scale3D=(X=1,Y=1,Z=1))
    Name=Neck 2
    OffsetSpace=GlobalSpace
    Settings="(InitialSpace=LocalSpace,bUsePreferredRotationOrder=False,PreferredRotationOrder=YZX,Limits=(LimitTranslationX=(bMinimum=False,bMaximum=False),LimitTranslationY=(bMinimum=False,bMaximum=False),LimitTranslationZ=(bMinimum=False,bMaximum=False),LimitPitch=(b ...(truncated)"
    OffsetTransform=&RigUnit_GetTransform(bInitial=False, Item=RigVMVariableNode() {variable=Neck 02 Build Joint}.Value).Transform
    Settings.Shape.Transform.Scale3D.X=@n5.Result
    Settings.Shape.Transform.Scale3D.Y=@n5.Result
    Settings.Shape.Transform.Scale3D.Z=@n5.Result
    Parent=@n4.Item
   ) {template="SpawnControl::Execute(in InitialValue,in Settings,in OffsetTransform,in OffsetSpace,in Parent,in Name,out Item)"}
   > RigVMVariableNode(Value=@n6.Item) {variable=Neck 2 Cntrl}
   > RigVMDispatchNode(Array=RigVMVariableNode() {variable=Neck Controls}.Value, Element=@n6.Item) {template="DISPATCH_RigVMDispatch_ArrayAdd(io Array,in Element,out Index)"}
   > n7 = &RigUnit_HierarchyAddNull(Name=Neck 2 Null, Space=LocalSpace, Transform=(Rotation=(X=0,Y=0,Z=0,W=1),Translation=(X=0,Y=0,Z=0),Scale3D=(X=1,Y=1,Z=1)), Parent=@n6.Item)
   > RigVMVariableNode(Value=@n7.Item) {variable=Neck 2 Null}
   n8 = &RigVMFunction_MathDoubleMul(B=0.4, A=RigVMVariableNode() {variable=Control Scale}.Value)
   > n9 = &RigUnit_HierarchyAddControlTransform(
    InitialValue=(Rotation=(X=0,Y=0,Z=0,W=1),Translation=(X=0,Y=0,Z=0),Scale3D=(X=1,Y=1,Z=1))
    Name=Neck IK 02
    OffsetSpace=GlobalSpace
    Settings="(InitialSpace=LocalSpace,bUsePreferredRotationOrder=False,PreferredRotationOrder=YZX,Limits=(LimitTranslationX=(bMinimum=False,bMaximum=False),LimitTranslationY=(bMinimum=False,bMaximum=False),LimitTranslationZ=(bMinimum=False,bMaximum=False),LimitPitch=(b ...(truncated)"
    OffsetTransform=&RigUnit_GetTransform(bInitial=False, Item=RigVMVariableNode() {variable=Neck 02 Build Joint}.Value).Transform
    Parent=@n7.Item
    Settings.Shape.Transform.Scale3D.X=@n8.Result
    Settings.Shape.Transform.Scale3D.Y=@n8.Result
    Settings.Shape.Transform.Scale3D.Z=@n8.Result
    Settings.Shape.Transform.Translation=RigVMVariableNode() {variable=Neck_IK_Control_Position}.Value
   ) {template="SpawnControl::Execute(in InitialValue,in Settings,in OffsetTransform,in OffsetSpace,in Parent,in Name,out Item)"}
   > RigVMVariableNode(Value=@n9.Item) {variable=Neck IK Cntrl}
   > n10 = &RigUnit_HierarchyAddNull(
    Name=Head Space Null
    Space=GlobalSpace
    Transform=&RigUnit_GetTransform(bInitial=False, Item=RigVMVariableNode() {variable=Head Build Joint}.Value).Transform
    Parent=RigVMVariableNode() {variable=Neck 2 Cntrl}.Value
   )
   Multiply = &RigVMFunction_MathDoubleMul(B=2.5, A=RigVMVariableNode() {variable=Control Scale}.Value)
   > n11 = &RigUnit_HierarchyAddControlTransform(
    InitialValue=(Rotation=(X=0,Y=0,Z=0,W=1),Translation=(X=0,Y=0,Z=0),Scale3D=(X=1,Y=1,Z=1))
    Name=Head Main
    OffsetSpace=GlobalSpace
    Settings="(InitialSpace=LocalSpace,bUsePreferredRotationOrder=False,PreferredRotationOrder=YZX,Limits=(LimitTranslationX=(bMinimum=False,bMaximum=False),LimitTranslationY=(bMinimum=False,bMaximum=False),LimitTranslationZ=(bMinimum=False,bMaximum=False),LimitPitch=(b ...(truncated)"
    OffsetTransform=&RigUnit_GetTransform(bInitial=False, Item=RigVMVariableNode() {variable=Head Build Joint}.Value).Transform
    Parent=@n10.Item
    Settings.Shape.Transform.Scale3D.X=@Multiply.Result
    Settings.Shape.Transform.Scale3D.Y=@Multiply.Result
    Settings.Shape.Transform.Scale3D.Z=@Multiply.Result
    Settings.Shape.Transform.Translation=RigVMVariableNode() {variable=Head_Main_Control_Position}.Value
   ) {template="SpawnControl::Execute(in InitialValue,in Settings,in OffsetTransform,in OffsetSpace,in Parent,in Name,out Item)"}
   > RigVMVariableNode(Value=@n11.Item) {variable=Head Main Cntrl}
   > RigVMDispatchNode(Array=RigVMVariableNode() {variable=Neck Controls}.Value, Element=@n11.Item) {template="DISPATCH_RigVMDispatch_ArrayAdd(io Array,in Element,out Index)"}
   RigVMRerouteNode(Value=@n9.Item)
   &RigUnit_HierarchyAddControlTransform(
    ExecutePin=()
    InitialValue=(Rotation=(X=0,Y=0,Z=0,W=1),Translation=(X=0,Y=0,Z=0),Scale3D=(X=1,Y=1,Z=1))
    Name=Neck Space
    OffsetSpace=GlobalSpace
    Parent=(Type=Connector,Name="parent")
    Settings="(InitialSpace=LocalSpace,bUsePreferredRotationOrder=False,PreferredRotationOrder=YZX,Limits=(LimitTranslationX=(bMinimum=False,bMaximum=False),LimitTranslationY=(bMinimum=False,bMaximum=False),LimitTranslationZ=(bMinimum=False,bMaximum=False),LimitPitch=(b ...(truncated)"
    OffsetTransform=&RigUnit_GetTransform(bInitial=False, Item=RigVMVariableNode() {variable=Neck 01 Build Joint}.Value).Transform
   ) {template="SpawnControl::Execute(in InitialValue,in Settings,in OffsetTransform,in OffsetSpace,in Parent,in Name,out Item)"}
  ]}
 New Function_ContainedGraph {nodes [
   n1 = RigVMFunctionEntryNode()
   RigVMFunction_Sequence = &RigVMFunction_Sequence(ExecuteContext=@n1.ExecuteContext)
   For_Each = RigVMDispatchNode(ExecuteContext=@RigVMFunction_Sequence.A, Array=RigVMVariableNode() {variable=Neck Bones}.Value) {template="DISPATCH_RigVMDispatch_ArrayIterator(in Array,out Element,out Index,out Count,out Ratio)"}
   n2 = RigVMDispatchNode(Index=@For_Each.Index, Array=RigVMVariableNode() {variable=Virtual Bones}.Value) {template="DISPATCH_RigVMDispatch_ArrayGetAtIndex(in Array,in Index,out Element)"}
   > RigVMDispatchNode(
    Element=&RigVMFunction_MathTransformMakeRelative(
     Parent=&RigUnit_GetTransform(bInitial=False, Item=@n2.Element).Transform
     Global=&RigUnit_GetTransform(bInitial=true, Item=@For_Each.Element).Transform
    ).Local
    Array=RigVMVariableNode() {variable=Neck Bones Offset Transforms}.Value
   ) {template="DISPATCH_RigVMDispatch_ArrayAdd(io Array,in Element,out Index)"}
   RigVMFunctionReturnNode(ExecuteContext=@RigVMFunction_Sequence.B)
   n3 = RigVMRerouteNode(Value=())
   n4 = RigVMVariableNode() {variable=Connector Root}
   n5 = RigVMVariableNode() {variable=Neck 01 Build Joint}
   RigVMVariableNode(
    Value=&RigVMFunction_MathTransformMakeRelative(
     Global=&RigUnit_GetTransform(bInitial=true, Item=@n4.Value).Transform
     Parent=&RigUnit_GetTransform(bInitial=False, Item=@n5.Value).Transform
    ).Local
    ExecuteContext=@n3.Value
   ) {variable=Neck Base Transform}
   n6 = RigVMRerouteNode(Value=())
   n7 = RigVMVariableNode() {variable=Connector Neck Tip}
   n8 = RigVMVariableNode() {variable=Neck 02 Build Joint}
   RigVMVariableNode(
    Value=&RigVMFunction_MathTransformMakeRelative(
     Global=&RigUnit_GetTransform(bInitial=true, Item=@n7.Value).Transform
     Parent=&RigUnit_GetTransform(bInitial=False, Item=@n8.Value).Transform
    ).Local
    ExecuteContext=@n6.Value
   ) {variable=Neck Tip Transform}
   n9 = RigVMRerouteNode(Value=())
   n10 = RigVMVariableNode() {variable=Connector Head}
   n11 = RigVMVariableNode() {variable=Head Build Joint}
   RigVMVariableNode(
    Value=&RigVMFunction_MathTransformMakeRelative(
     Global=&RigUnit_GetTransform(bInitial=true, Item=@n10.Value).Transform
     Parent=&RigUnit_GetTransform(bInitial=False, Item=@n11.Value).Transform
    ).Local
    ExecuteContext=@n9.Value
   ) {variable=Head Transform}
  ]}
 New Function_ContainedGraph {nodes [
   RigVMFunctionEntryNode()
   > RigVMFunction_Sequence = RigVMAggregateNode()
   RigVMFunctionReturnNode(ExecuteContext=@RigVMFunction_Sequence.F)
   n1 = &RigUnit_HierarchyAddNull(
    ExecutePin=()
    Name=Root Space
    Space=LocalSpace
    Transform=(Rotation=(X=0,Y=0,Z=0,W=1),Translation=(X=0,Y=0,Z=0),Scale3D=(X=1,Y=1,Z=1))
    Parent=RigVMVariableNode() {variable=Head Build Joint}.Value
   )
   > RigVMDispatchNode(Element=@n1.Item, Array=RigVMVariableNode() {variable=Head Nulls}.Value) {template="DISPATCH_RigVMDispatch_ArrayAdd(io Array,in Element,out Index)"}
   > /Script/ControlRig.RigUnit_SetDefaultParent(Parent=(Type=Control,Name="Head Space"), Child=@n1.Item) {is=struct RigVMUnitNode}
   > &RigUnit_AddParent(DisplayLabel=None, Parent=(Type=Null,Name="Root Space"), Child=RigVMVariableNode() {variable=Head Main Cntrl}.Value)
   > n2 = &RigUnit_HierarchyAddNull(
    Name=Root Head Null Space
    Parent=(Type=Bone,Name="Head Build")
    Space=LocalSpace
    Transform=(Rotation=(X=0,Y=0,Z=0,W=1),Translation=(X=0,Y=0,Z=0),Scale3D=(X=1,Y=1,Z=1))
   )
   > &RigUnit_SetDefaultParent(Parent=(Type=Connector,Name="rootSpace"), Child=@n2.Item)
   n3 = &RigUnit_HierarchyAddNull(
    ExecutePin=()
    Name=Pelvis Space
    Parent=(Type=Bone,Name="Head Build")
    Space=LocalSpace
    Transform=(Rotation=(X=0,Y=0,Z=0,W=1),Translation=(X=0,Y=0,Z=0),Scale3D=(X=1,Y=1,Z=1))
   )
   > RigVMDispatchNode(Element=@n3.Item, Array=RigVMVariableNode() {variable=Head Nulls}.Value) {template="DISPATCH_RigVMDispatch_ArrayAdd(io Array,in Element,out Index)"}
   > &RigUnit_SetDefaultParent(Parent=(Type=Control,Name="Head Space"), Child=@n3.Item)
   > &RigUnit_AddParent(Child=(Type=Control,Name="Head Main"), DisplayLabel=None, Parent=(Type=Null,Name="Pelvis Space"))
   > n4 = &RigUnit_HierarchyAddNull(
    Name=Pelvis Head Null Space
    Parent=(Type=Bone,Name="Head Build")
    Space=LocalSpace
    Transform=(Rotation=(X=0,Y=0,Z=0,W=1),Translation=(X=0,Y=0,Z=0),Scale3D=(X=1,Y=1,Z=1))
   )
   > &RigUnit_SetDefaultParent(Parent=(Type=Connector,Name="Pelvis"), Child=@n4.Item)
   n5 = &RigUnit_HierarchyAddNull(
    ExecutePin=()
    Name=Spine Tip Space
    Parent=(Type=Bone,Name="Head Build")
    Space=LocalSpace
    Transform=(Rotation=(X=0,Y=0,Z=0,W=1),Translation=(X=0,Y=0,Z=0),Scale3D=(X=1,Y=1,Z=1))
   )
   > RigVMDispatchNode(Element=@n5.Item, Array=RigVMVariableNode() {variable=Head Nulls}.Value) {template="DISPATCH_RigVMDispatch_ArrayAdd(io Array,in Element,out Index)"}
   > &RigUnit_SetDefaultParent(Parent=(Type=Control,Name="Head Space"), Child=@n5.Item)
   > &RigUnit_AddParent(Child=(Type=Control,Name="Head Main"), DisplayLabel=None, Parent=(Type=Null,Name="Spine Tip Space"))
   > n6 = &RigUnit_HierarchyAddNull(
    Name=Spine Tp Head Null Space
    Parent=(Type=Bone,Name="Head Build")
    Space=LocalSpace
    Transform=(Rotation=(X=0,Y=0,Z=0,W=1),Translation=(X=0,Y=0,Z=0),Scale3D=(X=1,Y=1,Z=1))
   )
   > &RigUnit_SetDefaultParent(Parent=(Type=Connector,Name="SpineTip"), Child=@n6.Item)
   n7 = &RigUnit_HierarchyAddNull(
    ExecutePin=()
    Name=Root Neck Space
    Parent=(Type=Bone,Name="Neck Build 01")
    Space=LocalSpace
    Transform=(Rotation=(X=0,Y=0,Z=0,W=1),Translation=(X=0,Y=0,Z=0),Scale3D=(X=1,Y=1,Z=1))
   )
   > &RigUnit_SetDefaultParent(Parent=(Type=Control,Name="Neck Space"), Child=@n7.Item)
   > &RigUnit_AddParent(Child=(Type=Control,Name="Neck 1"), DisplayLabel=None, Parent=(Type=Null,Name="Root Neck Space"))
   > n8 = &RigUnit_HierarchyAddNull(
    Name=Root Neck Null Space
    Parent=(Type=Bone,Name="Neck Build 01")
    Space=LocalSpace
    Transform=(Rotation=(X=0,Y=0,Z=0,W=1),Translation=(X=0,Y=0,Z=0),Scale3D=(X=1,Y=1,Z=1))
   )
   > &RigUnit_SetDefaultParent(Parent=(Type=Connector,Name="rootSpace"), Child=@n8.Item)
   n9 = &RigUnit_HierarchyAddNull(
    ExecutePin=()
    Name=Pelvis Neck Space
    Parent=(Type=Bone,Name="Neck Build 01")
    Space=LocalSpace
    Transform=(Rotation=(X=0,Y=0,Z=0,W=1),Translation=(X=0,Y=0,Z=0),Scale3D=(X=1,Y=1,Z=1))
   )
   > &RigUnit_SetDefaultParent(Parent=(Type=Control,Name="Neck Space"), Child=@n9.Item)
   > &RigUnit_AddParent(Child=(Type=Control,Name="Neck 1"), DisplayLabel=None, Parent=(Type=Null,Name="Pelvis Neck Space"))
   > n10 = &RigUnit_HierarchyAddNull(
    Name=Pelvis Neck Null Space
    Parent=(Type=Bone,Name="Neck Build 01")
    Space=LocalSpace
    Transform=(Rotation=(X=0,Y=0,Z=0,W=1),Translation=(X=0,Y=0,Z=0),Scale3D=(X=1,Y=1,Z=1))
   )
   > &RigUnit_SetDefaultParent(Parent=(Type=Connector,Name="Pelvis"), Child=@n10.Item)
  ]}
 New Function_ContainedGraph {nodes [
   RigVMFunctionEntryNode()
   > RigVMFunction_Sequence = RigVMAggregateNode()
   RigVMFunctionReturnNode(ExecuteContext=@RigVMFunction_Sequence.C)
   RigVMDispatchNode(Array=RigVMVariableNode() {variable=Neck Space Nulls}.Value, ExecuteContext=@RigVMFunction_Sequence.A) {template="DISPATCH_RigVMDispatch_ArrayReset(io Array)"}
   > For_Each_1 = RigVMDispatchNode(Array=RigVMVariableNode() {variable=Neck Spaces}.Value) {template="DISPATCH_RigVMDispatch_ArrayIterator(in Array,out Element,out Index,out Count,out Ratio)"}
   > n1 = &RigUnit_HierarchyAddNull(
    Space=GlobalSpace
    Transform=&RigUnit_GetTransform(bInitial=False, Item=RigVMVariableNode() {variable=Neck 01 Build Joint}.Value).Transform
    Parent=RigVMVariableNode() {variable=Connector Parent}.Value
    Name=/Script/RigVM.RigVMFunction_NameConcat(B=" Orient Neck Space", A=/Script/ControlRig.RigUnit_GetItemShortName(Item=@For_Each_1.Element) {is=struct RigVMUnitNode}.ShortName) {is=struct RigVMUnitNode, template="Concat::Execute(in A,in B,out Result)"}.Result
   )
   > RigVMDispatchNode(Array=RigVMVariableNode() {variable=Neck Space Nulls}.Value, Element=@n1.Item) {template="DISPATCH_RigVMDispatch_ArrayAdd(io Array,in Element,out Index)"}
   > &RigUnit_AddParent(DisplayLabel=None, Child=RigVMVariableNode() {variable=Neck 1 Cntrl}.Value, Parent=@n1.Item)
   RigVMDispatchNode(Array=RigVMVariableNode() {variable=Head Space Nulls}.Value, ExecuteContext=@RigVMFunction_Sequence.B) {template="DISPATCH_RigVMDispatch_ArrayReset(io Array)"}
   > For_Each_2 = RigVMDispatchNode(Array=RigVMVariableNode() {variable=Head Spaces}.Value) {template="DISPATCH_RigVMDispatch_ArrayIterator(in Array,out Element,out Index,out Count,out Ratio)"}
   > n2 = &RigUnit_HierarchyAddNull(
    Space=GlobalSpace
    Transform=&RigUnit_GetTransform(bInitial=False, Item=RigVMVariableNode() {variable=Head Build Joint}.Value).Transform
    Parent=RigVMVariableNode() {variable=Neck 2 Cntrl}.Value
    Name=&RigVMFunction_NameConcat(B=" Orient Head Space", A=&RigUnit_GetItemShortName(Item=@For_Each_2.Element).ShortName) {template="Concat::Execute(in A,in B,out Result)"}.Result
   )
   > RigVMDispatchNode(Array=RigVMVariableNode() {variable=Head Space Nulls}.Value, Element=@n2.Item) {template="DISPATCH_RigVMDispatch_ArrayAdd(io Array,in Element,out Index)"}
   > &RigUnit_AddParent(DisplayLabel=None, Child=RigVMVariableNode() {variable=Head Main Cntrl}.Value, Parent=@n2.Item)
  ]}
 RigVMFunctionLibrary {nodes [
   Build Joints = RigVMCollapseNode(ExecuteContext=())
   Build Nulls = RigVMCollapseNode(ExecuteContext=())
   Create Controls = RigVMCollapseNode(ExecuteContext=())
   Create Space Switches = RigVMCollapseNode(ExecuteContext=())
   Find Joints = RigVMCollapseNode(ExecuteContext=())
   Get Local Transforms = RigVMCollapseNode(ExecuteContext=())
  ]}
 RigVMFunction_Sequence_1_ContainedGraph {nodes [
   n1 = RigVMFunctionEntryNode()
   RigVMFunction_Sequence = &RigVMFunction_Sequence(ExecuteContext=@n1.ExecuteContext)
   n2 = &RigVMFunction_Sequence(ExecuteContext=@RigVMFunction_Sequence.B)
   n3 = &RigVMFunction_Sequence(ExecuteContext=@n2.B)
   n4 = &RigVMFunction_Sequence(ExecuteContext=@n3.B)
   RigVMFunctionReturnNode(A=@RigVMFunction_Sequence.A, B=@n2.A, C=@n3.A, D=@n4.A, E=@n4.B)
  ]}
 RigVMFunction_Sequence_1_ContainedGraph {nodes [
   n1 = RigVMFunctionEntryNode()
   RigVMFunction_Sequence = &RigVMFunction_Sequence(ExecuteContext=@n1.ExecuteContext)
   n2 = &RigVMFunction_Sequence(ExecuteContext=@RigVMFunction_Sequence.B)
   RigVMFunctionReturnNode(A=@RigVMFunction_Sequence.A, B=@n2.A, C=@n2.B)
  ]}
 RigVMFunction_Sequence_1_ContainedGraph {nodes [
   n1 = RigVMFunctionEntryNode()
   RigVMFunction_Sequence = &RigVMFunction_Sequence(ExecuteContext=@n1.ExecuteContext)
   n2 = &RigVMFunction_Sequence(ExecuteContext=@RigVMFunction_Sequence.B)
   n3 = &RigVMFunction_Sequence(ExecuteContext=@n2.B)
   RigVMFunctionReturnNode(A=@RigVMFunction_Sequence.A, B=@n2.A, C=@n3.A, D=@n3.B)
  ]}
 RigVMFunction_Sequence_1_ContainedGraph {nodes [
   n1 = RigVMFunctionEntryNode()
   RigVMFunction_Sequence = &RigVMFunction_Sequence(ExecuteContext=@n1.ExecuteContext)
   n2 = &RigVMFunction_Sequence(ExecuteContext=@RigVMFunction_Sequence.B)
   RigVMFunctionReturnNode(A=@RigVMFunction_Sequence.A, B=@n2.A, C=@n2.B)
  ]}
 RigVMFunction_Sequence_1_ContainedGraph {nodes [
   n1 = RigVMFunctionEntryNode()
   RigVMFunction_Sequence = &RigVMFunction_Sequence(ExecuteContext=@n1.ExecuteContext)
   n2 = &RigVMFunction_Sequence(ExecuteContext=@RigVMFunction_Sequence.B)
   n3 = &RigVMFunction_Sequence(ExecuteContext=@n2.B)
   n4 = &RigVMFunction_Sequence(ExecuteContext=@n3.B)
   RigVMFunctionReturnNode(A=@RigVMFunction_Sequence.A, B=@n2.A, C=@n3.A, D=@n4.A, E=@n4.B)
  ]}
 RigVMFunction_Sequence_1_ContainedGraph {nodes [
   n1 = RigVMFunctionEntryNode()
   RigVMFunction_Sequence = &RigVMFunction_Sequence(ExecuteContext=@n1.ExecuteContext)
   n2 = &RigVMFunction_Sequence(ExecuteContext=@RigVMFunction_Sequence.B)
   n3 = &RigVMFunction_Sequence(ExecuteContext=@n2.B)
   n4 = &RigVMFunction_Sequence(ExecuteContext=@n3.B)
   n5 = &RigVMFunction_Sequence(ExecuteContext=@n4.B)
   RigVMFunctionReturnNode(A=@RigVMFunction_Sequence.A, B=@n2.A, C=@n3.A, D=@n4.A, E=@n5.A, F=@n5.B)
  ]}
 RigVMFunction_Sequence_1_ContainedGraph {nodes [
   n1 = RigVMFunctionEntryNode()
   RigVMFunction_Sequence = &RigVMFunction_Sequence(ExecuteContext=@n1.ExecuteContext)
   n2 = &RigVMFunction_Sequence(ExecuteContext=@RigVMFunction_Sequence.B)
   RigVMFunctionReturnNode(A=@RigVMFunction_Sequence.A, B=@n2.A, C=@n2.B)
  ]}
 RigVMModel {nodes [
   For_Each_1 = RigVMDispatchNode(ExecuteContext=(), Array=RigVMVariableNode() {variable=Head Spaces}.Value) {template="DISPATCH_RigVMDispatch_ArrayIterator(in Array,out Element,out Index,out Count,out Ratio)"}
   > Set Transform_11 = &RigUnit_SetRotation(
    Space=GlobalSpace
    Weight=1
    bInitial=False
    Item=RigVMDispatchNode(Index=@For_Each_1.Index, Array=RigVMVariableNode() {variable=Head Space Nulls}.Value) {template="DISPATCH_RigVMDispatch_ArrayGetAtIndex(in Array,in Index,out Element)"}.Element
    Value=&RigUnit_GetTransform(bInitial=False, Item=@For_Each_1.Element).Transform.Rotation
   )
   For_Each_4 = RigVMDispatchNode(ExecuteContext=(), Array=RigVMVariableNode() {variable=Neck Spaces}.Value) {template="DISPATCH_RigVMDispatch_ArrayIterator(in Array,out Element,out Index,out Count,out Ratio)"}
   > Set Transform_13 = &RigUnit_SetRotation(
    Space=GlobalSpace
    Weight=1
    bInitial=False
    Item=RigVMDispatchNode(Index=@For_Each_4.Index, Array=RigVMVariableNode() {variable=Neck Space Nulls}.Value) {template="DISPATCH_RigVMDispatch_ArrayGetAtIndex(in Array,in Index,out Element)"}.Element
    Value=&RigUnit_GetTransform(bInitial=False, Item=@For_Each_4.Element).Transform.Rotation
   )
   For_Each_7 = RigVMDispatchNode(ExecuteContext=(), Array=RigVMVariableNode() {variable=Neck Bones Offset Transforms}.Value) {template="DISPATCH_RigVMDispatch_ArrayIterator(in Array,out Element,out Index,out Count,out Ratio)"}
   n1 = RigVMDispatchNode(Index=@For_Each_7.Index, Array=RigVMVariableNode() {variable=Virtual Bones}.Value) {template="DISPATCH_RigVMDispatch_ArrayGetAtIndex(in Array,in Index,out Element)"}
   > &RigUnit_SetTransform(
    Weight=1
    bInitial=false
    bPropagateToChildren=true
    Item=RigVMDispatchNode(Index=@For_Each_7.Index, Array=RigVMVariableNode() {variable=Neck Bones}.Value) {template="DISPATCH_RigVMDispatch_ArrayGetAtIndex(in Array,in Index,out Element)"}.Element
    Value=/Script/RigVM.RigVMFunction_MathTransformMakeAbsolute(Parent=&RigUnit_GetTransform(bInitial=False, Item=@n1.Element).Transform, Local=@For_Each_7.Element) {is=struct RigVMUnitNode, template="Make Absolute::Execute(in Local,in Parent,out Global)"}.Global
   )
   n2 = RigVMRerouteNode(Value=())
   For_Each_5 = RigVMDispatchNode(ExecuteContext=@n2.Value, Array=RigVMVariableNode() {variable=Neck Bones Offset Transforms}.Value) {template="DISPATCH_RigVMDispatch_ArrayIterator(in Array,out Element,out Index,out Count,out Ratio)"}
   n3 = RigVMDispatchNode(Index=@For_Each_5.Index, Array=RigVMVariableNode() {variable=Neck Controls}.Value) {template="DISPATCH_RigVMDispatch_ArrayGetAtIndex(in Array,in Index,out Element)"}
   > &RigUnit_SetTransform(
    Weight=1
    bInitial=false
    bPropagateToChildren=true
    Item=RigVMDispatchNode(Index=@For_Each_5.Index, Array=RigVMVariableNode() {variable=Neck Bones}.Value) {template="DISPATCH_RigVMDispatch_ArrayGetAtIndex(in Array,in Index,out Element)"}.Element
    Value=&RigVMFunction_MathTransformMakeAbsolute(Local=@For_Each_5.Element, Parent=&RigUnit_GetTransform(bInitial=False, Item=@n3.Element).Transform) {template="Make Absolute::Execute(in Local,in Parent,out Global)"}.Global
   )
   RigUnit_BeginExecution = &RigUnit_BeginExecution()
   RigVMFunction_Sequence_1 = RigVMAggregateNode(ExecuteContext=@RigUnit_BeginExecution.ExecutePin)
   For_Each = RigVMDispatchNode(Array=RigVMVariableNode() {variable=Neck Bones}.Value, ExecuteContext=@RigVMFunction_Sequence_1.E) {template="DISPATCH_RigVMDispatch_ArrayIterator(in Array,out Element,out Index,out Count,out Ratio)"}
   n4 = RigVMDispatchNode(Index=@For_Each.Index, Array=RigVMVariableNode() {variable=Virtual Bones}.Value) {template="DISPATCH_RigVMDispatch_ArrayGetAtIndex(in Array,in Index,out Element)"}
   > &RigUnit_SetTransform(
    Weight=1
    bInitial=False
    Item=@For_Each.Element
    Value=/Script/ControlRig.RigUnit_ProjectTransformToNewParent(bChildInitial=True, bNewParentInitial=False, bOldParentInitial=True, NewParent=@n4.Element, OldParent=@n4.Element, Child=@For_Each.Element) {is=struct RigVMUnitNode}.Transform
   )
   /Script/ControlRig.RigUnit_PositionConstraintLocalSpaceOffset(
    Filter=(bX=True,bY=True,bZ=True)
    Parents=((Item=(Type=Control,Name="Head Main"),Weight=0.75),(Item=(Type=Control,Name="Neck 1"),Weight=0.25))
    Weight=1
    bMaintainOffset=True
    Parents.1.Item=RigVMVariableNode() {variable=Neck 1 Cntrl}.Value
    Child=RigVMVariableNode() {variable=Neck 2 Null}.Value
    Parents.0.Item=RigVMVariableNode() {variable=Head Main Cntrl}.Value
    ExecutePin=@RigVMFunction_Sequence_1.A
   ) {is=struct RigVMUnitNode}
   For_Each_2 = RigVMDispatchNode(Array=RigVMVariableNode() {variable=Neck Spaces}.Value, ExecuteContext=@RigVMFunction_Sequence_1.B) {template="DISPATCH_RigVMDispatch_ArrayIterator(in Array,out Element,out Index,out Count,out Ratio)"}
   n5 = RigVMDispatchNode(Index=@For_Each_2.Index, Array=RigVMVariableNode() {variable=Neck Space Nulls}.Value) {template="DISPATCH_RigVMDispatch_ArrayGetAtIndex(in Array,in Index,out Element)"}
   Multiply = &RigVMFunction_MathTransformMul(
    B=&RigUnit_GetTransform(bInitial=False, Item=@For_Each_2.Element).Transform
    A=&RigVMFunction_MathTransformMakeRelative(
     Parent=&RigUnit_GetTransform(bInitial=true, Item=@For_Each_2.Element).Transform
     Global=&RigUnit_GetTransform(bInitial=true, Item=@n5.Element).Transform
    ).Local
   ) {template="Multiply::Execute(in A,in B,out Result)"}
   > Set Transform_6 = &RigUnit_SetRotation(Space=GlobalSpace, Weight=1, bInitial=False, Value=@Multiply.Result.Rotation, Item=@n5.Element)
   For_Each_3 = RigVMDispatchNode(Array=RigVMVariableNode() {variable=Head Spaces}.Value, ExecuteContext=@RigVMFunction_Sequence_1.C) {template="DISPATCH_RigVMDispatch_ArrayIterator(in Array,out Element,out Index,out Count,out Ratio)"}
   n6 = RigVMDispatchNode(Index=@For_Each_3.Index, Array=RigVMVariableNode() {variable=Head Space Nulls}.Value) {template="DISPATCH_RigVMDispatch_ArrayGetAtIndex(in Array,in Index,out Element)"}
   n7 = &RigUnit_GetTransform(bInitial=true, Item=@n6.Element)
   n8 = &RigUnit_GetTransform(bInitial=true, Item=@For_Each_3.Element)
   > Set Transform_7 = &RigUnit_SetRotation(
    Space=GlobalSpace
    Weight=1
    bInitial=False
    Value=&RigVMFunction_MathTransformMul(
     B=&RigUnit_GetTransform(bInitial=False, Item=@For_Each_3.Element).Transform
     A=&RigVMFunction_MathTransformMakeRelative(Global=@n7.Transform, Parent=@n8.Transform).Local
    ) {template="Multiply::Execute(in A,in B,out Result)"}.Result.Rotation
    Item=@n6.Element
   )
   &RigUnit_SetTransform(
    Weight=1
    bInitial=False
    bPropagateToChildren=false
    Value=&RigUnit_GetTransform(bInitial=False, Item=RigVMVariableNode() {variable=Neck 1 Cntrl}.Value).Transform
    Item=RigVMVariableNode() {variable=Neck 01 Build Joint}.Value
    ExecutePin=@RigVMFunction_Sequence_1.D
   )
   > &RigUnit_SetTransform(
    Weight=1
    bInitial=False
    bPropagateToChildren=false
    Value=&RigUnit_GetTransform(bInitial=False, Item=RigVMVariableNode() {variable=Neck IK Cntrl}.Value).Transform
    Item=RigVMVariableNode() {variable=Neck 02 Build Joint}.Value
   )
   > &RigUnit_SetTransform(
    Weight=1
    bInitial=False
    bPropagateToChildren=false
    Value=&RigUnit_GetTransform(bInitial=False, Item=RigVMVariableNode() {variable=Head Main Cntrl}.Value).Transform
    Item=RigVMVariableNode() {variable=Head Build Joint}.Value
   )
   &RigUnit_SetTransform(
    ExecutePin=()
    Weight=1
    bInitial=False
    bPropagateToChildren=false
    Value=&RigUnit_GetTransform(bInitial=False, Item=RigVMVariableNode() {variable=Neck 1 Cntrl}.Value).Transform
    Item=RigVMVariableNode() {variable=Neck 01 Build Joint}.Value
   )
   > &RigUnit_SetTransform(
    Weight=1
    bInitial=False
    bPropagateToChildren=false
    Value=&RigUnit_GetTransform(bInitial=False, Item=RigVMVariableNode() {variable=Neck IK Cntrl}.Value).Transform
    Item=RigVMVariableNode() {variable=Neck 02 Build Joint}.Value
   )
   > &RigUnit_SetTransform(
    Weight=1
    bInitial=False
    bPropagateToChildren=false
    Value=&RigUnit_GetTransform(bInitial=False, Item=RigVMVariableNode() {variable=Head Main Cntrl}.Value).Transform
    Item=RigVMVariableNode() {variable=Head Build Joint}.Value
   )
   RigVMDispatchNode(Index=-1, Array=RigVMVariableNode() {variable=Virtual Bones}.Value) {template="DISPATCH_RigVMDispatch_ArrayGetAtIndex(in Array,in Index,out Element)"}
   RigVMDispatchNode(Index=-1, Array=RigVMVariableNode() {variable=Neck Controls}.Value) {template="DISPATCH_RigVMDispatch_ArrayGetAtIndex(in Array,in Index,out Element)"}
   n9 = /Script/ControlRig.RigUnit_GetTransformItemArray(Items=(), Space=GlobalSpace, bInitial=False) {is=struct RigVMUnitNode}
   n10 = RigVMDispatchNode(Index=1, Array=@n9.Transforms) {template="DISPATCH_RigVMDispatch_ArrayGetAtIndex(in Array,in Index,out Element)"}
   Interpolate = /Script/RigVM.RigVMFunction_MathQuaternionSlerp(
    T=0.5
    A=RigVMDispatchNode(Index=0, Array=@n9.Transforms) {template="DISPATCH_RigVMDispatch_ArrayGetAtIndex(in Array,in Index,out Element)"}.Element.Rotation
    B=@n10.Element.Rotation
   ) {is=struct RigVMUnitNode, template="Interpolate::Execute(in A,in B,in T,out Result)"}
   DISPATCH_RigVMDispatch_MakeStruct = RigVMDispatchNode(
    Elements=(Rotation=(X=0,Y=0,Z=0,W=1),Translation=(X=0,Y=0,Z=0),Scale3D=(X=1,Y=1,Z=1))
    Elements.Translation=@n10.Element.Translation
    Elements.Rotation=@Interpolate.Result
   ) {template="DISPATCH_RigVMDispatch_MakeStruct(in Elements,out Struct)"}
   &RigUnit_GetTransform(bInitial=False, Item=RigVMVariableNode() {variable=Connector Parent}.Value)
   &RigUnit_GetTransform(bInitial=False, Item=RigVMVariableNode() {variable=Connector Head}.Value)
   &RigUnit_ParentConstraint(
    AdvancedSettings=(InterpolationType=Average,RotationOrderForFilter=XZY)
    Child=(Type=Bone,Name="face_attach")
    ExecutePin=()
    Parents=((Item=(Type=Bone,Name="Head"),Weight=1))
    Weight=1
   )
   &RigUnit_PositionConstraintLocalSpaceOffset(
    Child=(Type=Null,Name="Neck 2 Null")
    ExecutePin=()
    Filter=(bX=True,bY=True,bZ=True)
    Parents=((Item=(Type=Control,Name="Head Main"),Weight=0.75),(Item=(Type=Control,Name="Neck 1"),Weight=0.25))
    Weight=1
    bMaintainOffset=True
   )
   RigVMRerouteNode(Value=())
   RigVMRerouteNode(Value=())
   &RigUnit_SetTransform(
    ExecutePin=()
    Item=(Type=Null,Name="NeckTip")
    Value=(Rotation=(X=0,Y=0,Z=0,W=1),Translation=(X=0,Y=0,Z=0),Scale3D=(X=1,Y=1,Z=1))
    Weight=1
    bInitial=false
    bPropagateToChildren=false
   )
   &RigUnit_SetTransform(
    ExecutePin=()
    Weight=1
    bInitial=False
    bPropagateToChildren=false
    Value=&RigUnit_GetTransform(bInitial=False, Item=RigVMVariableNode() {variable=Head Main Cntrl}.Value).Transform
    Item=RigVMVariableNode() {variable=Head Build Joint}.Value
   )
   n11 = RigVMVariableNode() {variable=Neck 1 Cntrl}
   &RigUnit_SetTransform(
    ExecutePin=()
    Item=(Type=Connector,Name="Root")
    Weight=1
    bInitial=false
    bPropagateToChildren=false
    Value=&RigVMFunction_MathTransformMakeAbsolute(
     Parent=&RigUnit_GetTransform(bInitial=False, Item=@n11.Value).Transform
     Local=RigVMVariableNode() {variable=Neck Base Transform}.Value
    ) {template="Make Absolute::Execute(in Local,in Parent,out Global)"}.Global
   )
   n12 = RigVMVariableNode() {variable=Head Main Cntrl}
   &RigUnit_SetTransform(
    ExecutePin=()
    Item=(Type=Connector,Name="Head")
    Weight=1
    bInitial=false
    bPropagateToChildren=true
    Value=&RigVMFunction_MathTransformMakeAbsolute(
     Parent=&RigUnit_GetTransform(bInitial=False, Item=@n12.Value).Transform
     Local=RigVMVariableNode() {variable=Head Transform}.Value
    ) {template="Make Absolute::Execute(in Local,in Parent,out Global)"}.Global
   )
   n13 = RigVMVariableNode() {variable=Neck IK Cntrl}
   &RigUnit_SetTransform(
    ExecutePin=()
    Item=(Type=Connector,Name="NeckTip")
    Weight=1
    bInitial=false
    bPropagateToChildren=false
    Value=&RigVMFunction_MathTransformMakeAbsolute(
     Parent=&RigUnit_GetTransform(bInitial=False, Item=@n13.Value).Transform
     Local=RigVMVariableNode() {variable=Neck Tip Transform}.Value
    ) {template="Make Absolute::Execute(in Local,in Parent,out Global)"}.Global
   )
   &RigUnit_SetTransform(
    ExecutePin=()
    Item=(Type=Bone,Name="face_attach")
    Weight=1
    bInitial=False
    Value=&RigUnit_GetTransform(bInitial=False, Item=RigVMVariableNode() {variable=Connector Head}.Value).Transform
   )
   &RigUnit_SetTransform(
    ExecutePin=()
    Item=(Type=Bone,Name="face_attach")
    Weight=1
    bInitial=False
    Value=&RigUnit_GetTransform(bInitial=False, Item=RigVMVariableNode() {variable=Connector Head}.Value).Transform
   )
   RigVMVariableNode() {variable=Neck Controls}
  ]}
 RigVMModel Backwards {nodes [
   n1 = /Script/ControlRig.RigUnit_InverseExecution() {is=struct RigVMUnitNode}
   RigVMFunction_Sequence = RigVMAggregateNode(ExecuteContext=@n1.ExecutePin)
   &RigUnit_SetTransform(
    Weight=1
    bInitial=false
    bPropagateToChildren=false
    Value=&RigVMFunction_MathTransformMakeAbsolute(
     Parent=&RigUnit_GetTransform(Item=(Type=Connector,Name="Root"), bInitial=False).Transform
     Local=RigVMVariableNode() {variable=Neck 1 BK Local Transform}.Value
    ) {template="Make Absolute::Execute(in Local,in Parent,out Global)"}.Global
    Item=RigVMVariableNode() {variable=Neck 1 Cntrl}.Value
    ExecutePin=@RigVMFunction_Sequence.C
   )
   > &RigUnit_PositionConstraintLocalSpaceOffset(
    Filter=(bX=True,bY=True,bZ=True)
    Parents=((Item=(Type=Control,Name="Head Main"),Weight=0.75),(Item=(Type=Control,Name="Neck 1"),Weight=0.25))
    Weight=1
    bMaintainOffset=True
    Parents.0.Item=RigVMVariableNode() {variable=Head Main Cntrl}.Value
    Parents.1.Item=RigVMVariableNode() {variable=Neck 1 Cntrl}.Value
    Child=RigVMVariableNode() {variable=Neck 2 Null}.Value
   )
   > &RigUnit_SetTransform(
    Weight=1
    bInitial=false
    bPropagateToChildren=true
    Value=&RigVMFunction_MathTransformMakeAbsolute(
     Parent=&RigUnit_GetTransform(Item=(Type=Connector,Name="NeckTip"), bInitial=False).Transform
     Local=RigVMVariableNode() {variable=Neck 2 BK Local Transform}.Value
    ) {template="Make Absolute::Execute(in Local,in Parent,out Global)"}.Global
    Item=RigVMVariableNode() {variable=Neck 2 Cntrl}.Value
   )
   > &RigUnit_SetTransform(
    Weight=1
    bInitial=false
    bPropagateToChildren=true
    Value=&RigVMFunction_MathTransformMakeAbsolute(
     Parent=&RigUnit_GetTransform(Item=(Type=Connector,Name="Head"), bInitial=False).Transform
     Local=RigVMVariableNode() {variable=Head BK Local Transform}.Value
    ) {template="Make Absolute::Execute(in Local,in Parent,out Global)"}.Global
    Item=RigVMVariableNode() {variable=Head Main Cntrl}.Value
   )
   n2 = RigVMVariableNode() {variable=Connector Root}
   RigVMVariableNode(
    Value=&RigVMFunction_MathTransformMakeRelative(
     Parent=&RigUnit_GetTransform(bInitial=true, Item=@n2.Value).Transform
     Global=&RigUnit_GetTransform(Item=(Type=Control,Name="Neck 1"), bInitial=true).Transform
    ).Local
    ExecuteContext=@RigVMFunction_Sequence.B
   ) {variable=Neck 1 BK Local Transform}
   n3 = RigVMVariableNode() {variable=Connector Neck Tip}
   > RigVMVariableNode(Value=&RigVMFunction_MathTransformMakeRelative(
     Parent=&RigUnit_GetTransform(bInitial=true, Item=@n3.Value).Transform
     Global=&RigUnit_GetTransform(Item=(Type=Control,Name="Neck 2"), bInitial=true).Transform
    ).Local) {variable=Neck 2 BK Local Transform}
   n4 = RigVMVariableNode() {variable=Connector Head}
   > RigVMVariableNode(Value=&RigVMFunction_MathTransformMakeRelative(
     Parent=&RigUnit_GetTransform(bInitial=true, Item=@n4.Value).Transform
     Global=&RigUnit_GetTransform(Item=(Type=Control,Name="Head Main"), bInitial=true).Transform
    ).Local) {variable=Head BK Local Transform}
   For_Each = RigVMDispatchNode(Array=RigVMVariableNode() {variable=Neck Bones}.Value, ExecuteContext=@RigVMFunction_Sequence.A) {template="DISPATCH_RigVMDispatch_ArrayIterator(in Array,out Element,out Index,out Count,out Ratio)"}
   n5 = RigVMDispatchNode(Index=@For_Each.Index, Array=RigVMVariableNode() {variable=Virtual Bones}.Value) {template="DISPATCH_RigVMDispatch_ArrayGetAtIndex(in Array,in Index,out Element)"}
   > &RigUnit_SetTransform(
    Weight=1
    bInitial=False
    Value=&RigUnit_ProjectTransformToNewParent(
     bChildInitial=True
     bNewParentInitial=False
     bOldParentInitial=True
     Child=@n5.Element
     NewParent=@For_Each.Element
     OldParent=@For_Each.Element
    ).Transform
    Item=@n5.Element
   )
  ]}
]
hierarchy [
 Root {type=Connector, loc=<0 0 0>, rot=<0 0 0 1>}
 Root {type=Socket, loc=<0 0 0>, rot=<0 0 0 1>}
 Head {type=Connector, loc=<0 0 0>, rot=<0 0 0 1>}
 NeckTip {type=Connector, loc=<0 0 0>, rot=<0 0 0 1>}
 Parent {type=Connector, loc=<0 0 0>, rot=<0 0 0 1>}
]
variables [
 Connector Head {type=FRigElementKey, default=(Type=None,Name="")}
 Connector Neck Tip {type=FRigElementKey, default=(Type=None,Name="")}
 Connector Parent {type=FRigElementKey, default=(Type=None,Name="")}
 Connector Root {type=FRigElementKey, default=(Type=None,Name="")}
 Control Scale {type=double, default=1}
 Head BK Local Transform {type=FTransform, default=(Rotation=(X=0,Y=0,Z=0,W=1),Translation=(X=0,Y=0,Z=0),Scale3D=(X=1,Y=1,Z=1))}
 Head Build Joint {type=FRigElementKey, default=(Type=None,Name="")}
 Head Main Cntrl {type=FRigElementKey, default=(Type=Bone,Name="")}
 Head Nulls {type="TArray<FRigElementKey>", default=(())}
 Head Space Nulls {type="TArray<FRigElementKey>"}
 Head Spaces {type="TArray<FRigElementKey>"}
 Head Transform {type=FTransform, default=(Rotation=(X=0,Y=0,Z=0,W=1),Translation=(X=0,Y=0,Z=0),Scale3D=(X=1,Y=1,Z=1))}
 Head_Main_Control_Position {type=FVector, default=(X=0,Y=0,Z=15)}
 Neck 01 Build Joint {type=FRigElementKey, default=(Type=None,Name="")}
 Neck 02 Build Joint {type=FRigElementKey, default=(Type=None,Name="")}
 Neck 1 BK Local Transform {type=FTransform, default=(Rotation=(X=0,Y=0,Z=0,W=1),Translation=(X=0,Y=0,Z=0),Scale3D=(X=1,Y=1,Z=1))}
 Neck 1 Cntrl {type=FRigElementKey, default=(Type=Bone,Name="")}
 Neck 2 BK Local Transform {type=FTransform, default=(Rotation=(X=0,Y=0,Z=0,W=1),Translation=(X=0,Y=0,Z=0),Scale3D=(X=1,Y=1,Z=1))}
 Neck 2 Cntrl {type=FRigElementKey, default=(Type=Bone,Name="")}
 Neck 2 Null {type=FRigElementKey, default=(Type=None,Name="")}
 Neck Base Transform {type=FTransform, default=(Rotation=(X=0,Y=0,Z=0,W=1),Translation=(X=0,Y=0,Z=0),Scale3D=(X=1,Y=1,Z=1))}
 Neck Bones {type="TArray<FRigElementKey>"}
 Neck Bones Offset Transforms {type="TArray<FTransform>"}
 Neck Controls {type="TArray<FRigElementKey>"}
 Neck IK Cntrl {type=FRigElementKey, default=(Type=Bone,Name="")}
 Neck Space Nulls {type="TArray<FRigElementKey>"}
 Neck Spaces {type="TArray<FRigElementKey>"}
 Neck Tip Transform {type=FTransform, default=(Rotation=(X=0,Y=0,Z=0,W=1),Translation=(X=0,Y=0,Z=0),Scale3D=(X=1,Y=1,Z=1))}
 Neck_IK_Control_Position {type=FVector, default=(X=0,Y=18,Z=0)}
 Virtual Bones {type="TArray<FRigElementKey>"}
]
