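# Modular Control Rig root module that builds the root and body controls of a character rig.
# Construction resolves connectors, then spawns a nested Root Main, Root Offset, Body and Body Offset chain, sizing body shapes from Body Control Scale.
# It publishes root and offset controls as module metadata and spawns nulls to derive direction-arrow points.
# The forward solve copies Root Main's transform to the Root connector and debug-draws that arrow.
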ControlRigBlueprint /WhisperRemedyAssets/Assets/Modules/CRM_WhisperRemedy_Root_01
graphs [
 ConstructionGraph {nodes [
   RigUnit_PrepareForExecution = &RigUnit_PrepareForExecution()
   Find Joints = RigVMFunctionReferenceNode(ExecuteContext=@RigUnit_PrepareForExecution.ExecutePin)
   > Create Controls = RigVMFunctionReferenceNode()
  ]}
 New Function_ContainedGraph {nodes [
   n1 = RigVMFunctionEntryNode()
   RigVMFunction_Sequence = &RigVMFunction_Sequence(ExecuteContext=@n1.ExecuteContext)
   RigVMFunctionReturnNode(ExecuteContext=@RigVMFunction_Sequence.B)
   RigVMVariableNode(
    Value=/Script/ControlRig.RigUnit_ResolveConnector(Connector=(Type=Connector,Name="Root"), SkipSocket=False) {is=struct RigVMUnitNode}.Result
    ExecuteContext=@RigVMFunction_Sequence.A
   ) {variable=Root Connector}
   > RigVMVariableNode(Value=&RigUnit_ResolveConnector(Connector=(Type=Connector,Name="Pelvis"), SkipSocket=False).Result) {variable=Pelvis Connector}
   > RigVMVariableNode(Value=&RigUnit_ResolveConnector(Connector=(Type=Connector,Name="Body Control Position"), SkipSocket=False).Result) {variable=Body Control Position}
  ]}
 New Function_ContainedGraph {nodes [
   RigVMFunctionEntryNode()
   > RigVMFunction_Sequence = RigVMAggregateNode()
   RigVMFunctionReturnNode(ExecuteContext=@RigVMFunction_Sequence.D)
   n1 = /Script/ControlRig.RigUnit_HierarchyAddControlTransform(
    InitialValue=(Rotation=(X=0,Y=0,Z=0,W=1),Translation=(X=0,Y=0,Z=0),Scale3D=(X=1,Y=1,Z=1))
    Name=Root Main
    OffsetSpace=GlobalSpace
    Parent=(Type=None,Name="parent")
    Settings="(InitialSpace=LocalSpace,bUsePreferredRotationOrder=False,PreferredRotationOrder=YZX,Limits=(LimitTranslationX=(bMinimum=False,bMaximum=False),LimitTranslationY=(bMinimum=False,bMaximum=False),LimitTranslationZ=(bMinimum=False,bMaximum=False),LimitPitch=(b ...(truncated)"
    OffsetTransform=&RigUnit_GetTransform(bInitial=False, Item=RigVMVariableNode() {variable=Root Connector}.Value).Transform
    ExecutePin=@RigVMFunction_Sequence.A
   ) {
    is=struct RigVMUnitNode
    template="SpawnControl::Execute(in InitialValue,in Settings,in OffsetTransform,in OffsetSpace,in Parent,in Name,out Item)"
   }
   > RigVMVariableNode(Value=@n1.Item) {variable=Root Control}
   > n2 = &RigUnit_HierarchyAddControlTransform(
    InitialValue=(Rotation=(X=0,Y=0,Z=0,W=1),Translation=(X=0,Y=0,Z=0),Scale3D=(X=1,Y=1,Z=1))
    Name=Root Offset
    OffsetSpace=GlobalSpace
    OffsetTransform=(Rotation=(X=0,Y=0,Z=0,W=1),Translation=(X=0,Y=0,Z=0),Scale3D=(X=1,Y=1,Z=1))
    Settings="(InitialSpace=LocalSpace,bUsePreferredRotationOrder=False,PreferredRotationOrder=YZX,Limits=(LimitTranslationX=(bMinimum=False,bMaximum=False),LimitTranslationY=(bMinimum=False,bMaximum=False),LimitTranslationZ=(bMinimum=False,bMaximum=False),LimitPitch=(b ...(truncated)"
    Parent=@n1.Item
   ) {template="SpawnControl::Execute(in InitialValue,in Settings,in OffsetTransform,in OffsetSpace,in Parent,in Name,out Item)"}
   > RigVMVariableNode(Value=@n2.Item) {variable=Offset Control}
   n3 = RigVMVariableNode() {variable=Body Control Scale}
   Multiply = &RigVMFunction_MathDoubleMul(B=6, A=@n3.Value.X)
   > n4 = &RigUnit_HierarchyAddControlTransform(
    InitialValue=(Rotation=(X=0,Y=0,Z=0,W=1),Translation=(X=0,Y=0,Z=0),Scale3D=(X=1,Y=1,Z=1))
    Name=Body
    OffsetSpace=GlobalSpace
    OffsetTransform=(Rotation=(X=0,Y=0,Z=0,W=1),Translation=(X=0,Y=0,Z=0),Scale3D=(X=1,Y=1,Z=1))
    Settings="(InitialSpace=LocalSpace,bUsePreferredRotationOrder=False,PreferredRotationOrder=YZX,Limits=(LimitTranslationX=(bMinimum=False,bMaximum=False),LimitTranslationY=(bMinimum=False,bMaximum=False),LimitTranslationZ=(bMinimum=False,bMaximum=False),LimitPitch=(b ...(truncated)"
    OffsetTransform.Translation=&RigUnit_GetTransform(bInitial=False, Item=RigVMVariableNode() {variable=Body Control Position}.Value).Transform.Translation
    Settings.Shape.Transform.Scale3D.X=@Multiply.Result
    Settings.Shape.Transform.Scale3D.Y=&RigVMFunction_MathDoubleMul(B=6, A=@n3.Value.Y).Result
    Settings.Shape.Transform.Scale3D.Z=&RigVMFunction_MathDoubleMul(B=6, A=@n3.Value.Z).Result
    Parent=@n2.Item
   ) {template="SpawnControl::Execute(in InitialValue,in Settings,in OffsetTransform,in OffsetSpace,in Parent,in Name,out Item)"}
   > RigVMVariableNode(Value=@n4.Item) {variable=Offset Position}
   n5 = RigVMVariableNode() {variable=Body Control Scale}
   > &RigUnit_HierarchyAddControlTransform(
    InitialValue=(Rotation=(X=0,Y=0,Z=0,W=1),Translation=(X=0,Y=0,Z=0),Scale3D=(X=1,Y=1,Z=1))
    Name=Body Offset
    OffsetSpace=GlobalSpace
    OffsetTransform=(Rotation=(X=0,Y=0,Z=0,W=1),Translation=(X=0,Y=0,Z=0),Scale3D=(X=1,Y=1,Z=1))
    Settings="(InitialSpace=LocalSpace,bUsePreferredRotationOrder=False,PreferredRotationOrder=YZX,Limits=(LimitTranslationX=(bMinimum=False,bMaximum=False),LimitTranslationY=(bMinimum=False,bMaximum=False),LimitTranslationZ=(bMinimum=False,bMaximum=False),LimitPitch=(b ...(truncated)"
    OffsetTransform.Translation=&RigUnit_GetTransform(bInitial=False, Item=RigVMVariableNode() {variable=Body Control Position}.Value).Transform.Translation
    Settings.Shape.Transform.Scale3D.X=&RigVMFunction_MathDoubleMul(B=5, A=@n5.Value.X).Result
    Settings.Shape.Transform.Scale3D.Y=&RigVMFunction_MathDoubleMul(B=5, A=@n5.Value.Y).Result
    Settings.Shape.Transform.Scale3D.Z=&RigVMFunction_MathDoubleMul(B=5, A=@n5.Value.Z).Result
    Parent=@n4.Item
   ) {template="SpawnControl::Execute(in InitialValue,in Settings,in OffsetTransform,in OffsetSpace,in Parent,in Name,out Item)"}
   Set_Module_Metadata = RigVMDispatchNode(Name=Root Control, NameSpace=Self, Value=RigVMVariableNode() {variable=Root Control}.Value, ExecuteContext=@RigVMFunction_Sequence.C) {template="DISPATCH_RigDispatch_SetModuleMetadata(in Name,in NameSpace,in Value,out Success)"}
   > Set_Module_Metadata_1 = RigVMDispatchNode(Name=Offset Control, NameSpace=Self, Value=RigVMVariableNode() {variable=Offset Control}.Value) {template="DISPATCH_RigDispatch_SetModuleMetadata(in Name,in NameSpace,in Value,out Success)"}
   For_Each = RigVMDispatchNode(Array=RigVMVariableNode() {variable=Arrow Direction transform Positions}.Value, ExecuteContext=@RigVMFunction_Sequence.B) {template="DISPATCH_RigVMDispatch_ArrayIterator(in Array,out Element,out Index,out Count,out Ratio)"}
   > n6 = &RigUnit_HierarchyAddNull(Name=directionNull1, Space=LocalSpace, Transform=@For_Each.Element, Parent=RigVMVariableNode() {variable=Root Connector}.Value)
   > RigVMDispatchNode(
    Element=&RigUnit_GetTransform(bInitial=False, Item=@n6.Item).Transform.Translation
    Array=RigVMVariableNode() {variable=Arrow Vector Points}.Value
   ) {template="DISPATCH_RigVMDispatch_ArrayAdd(io Array,in Element,out Index)"}
  ]}
 RigVMFunctionLibrary {nodes [
   Create Controls = RigVMCollapseNode(ExecuteContext=())
   Find Joints = RigVMCollapseNode(ExecuteContext=())
  ]}
 RigVMFunction_Sequence_1_ContainedGraph {nodes [
   n1 = RigVMFunctionEntryNode()
   RigVMFunction_Sequence = &RigVMFunction_Sequence(ExecuteContext=@n1.ExecuteContext)
   n2 = &RigVMFunction_Sequence(ExecuteContext=@RigVMFunction_Sequence.B)
   n3 = &RigVMFunction_Sequence(ExecuteContext=@n2.B)
   RigVMFunctionReturnNode(A=@RigVMFunction_Sequence.A, B=@n2.A, C=@n3.A, D=@n3.B)
  ]}
 RigVMModel {nodes [
   RigUnit_BeginExecution = &RigUnit_BeginExecution()
   RigVMFunction_Sequence = &RigVMFunction_Sequence(ExecuteContext=@RigUnit_BeginExecution.ExecutePin)
   &RigUnit_SetTransform(
    Item=(Type=Connector,Name="Root")
    Weight=1
    bInitial=false
    Value=&RigUnit_GetTransform(bInitial=False, Item=RigVMVariableNode() {variable=Root Control}.Value).Transform
    ExecutePin=@RigVMFunction_Sequence.A
   )
   /Script/RigVM.RigVMFunction_DebugLineStripNoSpace(
    Color=(R=0.8,G=0.7,B=0.1,A=1)
    DebugDrawSettings=(DepthPriority=SDPG_Foreground,Lifetime=-1)
    Thickness=1
    bEnabled=true
    WorldOffset=&RigUnit_GetTransform(bInitial=False, Item=RigVMVariableNode() {variable=Root Connector}.Value).Transform
    Points=RigVMVariableNode() {variable=Arrow Vector Points}.Value
    ExecutePin=@RigVMFunction_Sequence.B
   ) {is=struct RigVMUnitNode}
   RigVMVariableNode() {variable=Offset Position}
  ]}
]
hierarchy [
 Root {type=Connector, loc=<0 0 0>, rot=<0 0 0 1>}
 Root {type=Socket, loc=<0 0 0>, rot=<0 0 0 1>}
 Pelvis {type=Connector, loc=<0 0 0>, rot=<0 0 0 1>}
 "CRM_Quad_Root_01:Root Main_null" {type=Null, loc=<0 0 0>, rot=<0 0 0 1>}
 Body Control Position {type=Connector, loc=<0 0 0>, rot=<0 0 0 1>}
]
variables [
 Arrow Direction transform Positions {
  type="TArray<FTransform>"
  default=((Translation=(X=0,Y=91,Z=2)),(Translation=(X=12,Y=79,Z=2)),(Translation=(X=5,Y=79,Z=2)),(Translation=(X=5,Y=58,Z=2)),(Translation=(X=-5,Y=58,Z=2)),(Translation=(X=-5,Y=79,Z=2)),(Translation=(X=-10,Y=79,Z=2)),(Translation=(X=0,Y=91,Z=2)))
 }
 Arrow Vector Points {type="TArray<FVector>"}
 Body Control Position {type=FRigElementKey, default=(Type=None,Name="")}
 Body Control Scale {type=FVector, default=(X=1,Y=1,Z=1)}
 Offset Control {type=FRigElementKey, default=(Type=Bone,Name="")}
 Offset Position {type=FRigElementKey, default=(Type=Bone,Name="")}
 Pelvis Connector {type=FRigElementKey, default=(Type=Connector,Name="")}
 Root Connector {type=FRigElementKey, default=(Type=Connector,Name="")}
 Root Control {type=FRigElementKey, default=(Type=Bone,Name="")}
 Spine Joints {type="TArray<FRigElementKey>"}
]
