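# Modular Control Rig spine module that builds Hip, Pelvis FK, Spine 01, Spine 02 and Chest controls from connector-resolved joints.
# Construction creates build joints and controls, and caches joint offsets relative to controls.
# Forward solve reapplies those offsets, blends the Spine 02 null between Spine 01 and Chest, and toggles secondary control visibility.
# Forward/back proxies move Chest or Hip during interaction; backwards solve restores controls from joints.
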ControlRigBlueprint /WhisperRemedyAssets/Assets/Modules/CRM_WhisperRemedy_Spine_01
graphs [
 New Function_ContainedGraph {nodes [
   RigVMFunctionEntryNode()
   > RigVMFunction_Sequence = RigVMAggregateNode()
   RigVMFunctionReturnNode(ExecuteContext=@RigVMFunction_Sequence.C)
   RigVMDispatchNode(Array=RigVMVariableNode() {variable=Spine Joints}.Value, ExecuteContext=@RigVMFunction_Sequence.A) {template="DISPATCH_RigVMDispatch_ArrayReset(io Array)"}
   > RigVMVariableNode(Value=/Script/ControlRig.RigUnit_ResolveConnector(Connector=(Type=Connector,Name="SpineRoot"), SkipSocket=False) {is=struct RigVMUnitNode}.Result) {variable=Connector Spine Root}
   > RigVMVariableNode(Value=&RigUnit_ResolveConnector(Connector=(Type=Connector,Name="Spine 1"), SkipSocket=False).Result) {variable=Connector Spine Base}
   > RigVMVariableNode(Value=&RigUnit_ResolveConnector(Connector=(Type=Connector,Name="Spine 2"), SkipSocket=False).Result) {variable=Spine Mid}
   > RigVMVariableNode(Value=&RigUnit_ResolveConnector(Connector=(Type=Connector,Name="Spine 3"), SkipSocket=False).Result) {variable=Connector Spine Tip}
   > RigVMVariableNode(Value=&RigUnit_ResolveConnector(Connector=(Type=Connector,Name="parent"), SkipSocket=False).Result) {variable=Connector Parent}
   n1 = /Script/ControlRig.RigUnit_HierarchyGetChainItemArray(End=(Type=Connector,Name="Spine 3"), Start=(Type=Connector,Name="Spine 1"), bIncludeEnd=True, bIncludeStart=True, bReverse=False) {is=struct RigVMUnitNode}
   RigVMVariableNode(Value=@n1.Chain, ExecuteContext=@RigVMFunction_Sequence.B) {variable=Spine Joints}
  ]}
 New Function_ContainedGraph {nodes [
   RigVMFunctionEntryNode()
   > RigVMFunction_Sequence = RigVMAggregateNode()
   RigVMFunctionReturnNode(ExecuteContext=@RigVMFunction_Sequence.D)
   RigUnit_GetTransform = &RigUnit_GetTransform(Item=(Type=Connector,Name="Spine 1"), bInitial=False)
   n1 = /Script/ControlRig.RigUnit_HierarchyAddBone(
    Name=spine 1 Build
    Parent=(Type=Bone,Name="None")
    Space=LocalSpace
    Transform=(Rotation=(X=0,Y=0,Z=0,W=1),Translation=(X=0,Y=0,Z=0),Scale3D=(X=1,Y=1,Z=1))
    Transform.Translation=@RigUnit_GetTransform.Transform.Translation
    ExecutePin=@RigVMFunction_Sequence.C
   ) {is=struct RigVMUnitNode}
   > RigVMDispatchNode(Element=@n1.Item, Array=RigVMVariableNode() {variable=Build Joints}.Value) {template="DISPATCH_RigVMDispatch_ArrayAdd(io Array,in Element,out Index)"}
   > n2 = &RigUnit_HierarchyAddBone(
    Name=Spine 2 Build
    Parent=(Type=Bone,Name="None")
    Space=LocalSpace
    Transform=(Rotation=(X=0,Y=0,Z=0,W=1),Translation=(X=0,Y=0,Z=0),Scale3D=(X=1,Y=1,Z=1))
    Transform.Translation=&RigUnit_GetTransform(Item=(Type=Connector,Name="Spine 2"), bInitial=False).Transform.Translation
   )
   > RigVMDispatchNode(Element=@n2.Item, Array=RigVMVariableNode() {variable=Build Joints}.Value) {template="DISPATCH_RigVMDispatch_ArrayAdd(io Array,in Element,out Index)"}
   > n3 = &RigUnit_HierarchyAddBone(
    Name=Spine 3 Build
    Parent=(Type=Bone,Name="None")
    Space=LocalSpace
    Transform=(Rotation=(X=0,Y=0,Z=0,W=1),Translation=(X=0,Y=0,Z=0),Scale3D=(X=1,Y=1,Z=1))
    Transform.Translation=&RigUnit_GetTransform(Item=(Type=Connector,Name="Spine 3"), bInitial=False).Transform.Translation
   )
   > RigVMDispatchNode(Element=@n3.Item, Array=RigVMVariableNode() {variable=Build Joints}.Value) {template="DISPATCH_RigVMDispatch_ArrayAdd(io Array,in Element,out Index)"}
   RigVMDispatchNode(Array=RigVMVariableNode() {variable=Build Joints}.Value, ExecuteContext=@RigVMFunction_Sequence.A) {template="DISPATCH_RigVMDispatch_ArrayReset(io Array)"}
   > n4 = &RigUnit_HierarchyAddBone(
    Name=Pelvis Build
    Parent=(Type=Bone,Name="root")
    Space=GlobalSpace
    Transform=(Rotation=(X=0,Y=0,Z=0,W=1),Translation=(X=0,Y=0,Z=0),Scale3D=(X=1,Y=1,Z=1))
    Transform.Translation=&RigUnit_GetTransform(Item=(Type=Connector,Name="SpineRoot"), bInitial=False).Transform.Translation
   )
   > RigVMVariableNode(Value=@n4.Item) {variable=Pelvis Joint}
   n5 = &RigUnit_HierarchyAddNull(
    ExecutePin=()
    Name=spineRef
    Parent=(Type=Bone,Name="None")
    Space=LocalSpace
    Transform=(Rotation=(X=0,Y=0,Z=0,W=1),Translation=(X=0,Y=0,Z=0),Scale3D=(X=1,Y=1,Z=1))
    Transform.Translation=&RigUnit_GetTransform(bInitial=False, Item=RigVMVariableNode() {variable=Connector Spine Root}.Value).Transform.Translation
   )
   > RigVMVariableNode(Value=@n5.Item) {variable=Build Parent}
   > For_Each_1 = RigVMDispatchNode(Array=RigVMVariableNode() {variable=Spine Joints}.Value) {template="DISPATCH_RigVMDispatch_ArrayIterator(in Array,out Element,out Index,out Count,out Ratio)"}
   n6 = /Script/ControlRig.RigUnit_GetTransformItemArray(Space=LocalSpace, bInitial=false, Items=RigVMVariableNode() {variable=Spine Joints}.Value) {is=struct RigVMUnitNode}
   n7 = /Script/RigVM.RigVMFunction_MathIntAdd(B=1, A=@For_Each_1.Index) {is=struct RigVMUnitNode, template="Add::Execute(in A,in B,out Result)"}
   From_String_1_1 = RigVMDispatchNode(String=/Script/RigVM.RigVMFunction_StringJoin(
     Separator=" "
     Values=("Spine Build","")
     Values.1=/Script/RigVM.RigVMFunction_StringPadInteger(Digits=2, Value=@n7.Result) {is=struct RigVMUnitNode}.Result
    ) {is=struct RigVMUnitNode}.Result) {template="DISPATCH_RigDispatch_FromString(in String,out Result)"}
   > n8 = &RigUnit_HierarchyAddBone(
    Space=LocalSpace
    Transform=(Rotation=(X=0,Y=0,Z=0,W=1),Translation=(X=0,Y=0,Z=0),Scale3D=(X=1,Y=1,Z=1))
    Transform.Translation=RigVMDispatchNode(
     Index=@For_Each_1.Index
     Array=/Script/RigVM.RigVMFunction_MathTransformArrayToSRT(Transforms=@n6.Transforms) {is=struct RigVMUnitNode}.Translations
    ) {template="DISPATCH_RigVMDispatch_ArrayGetAtIndex(in Array,in Index,out Element)"}.Element
    Name=@From_String_1_1.Result
    Parent=RigVMVariableNode() {variable=Build Parent}.Value
   )
   > RigVMVariableNode(Value=@n8.Item) {variable=Build Parent}
   > RigVMDispatchNode(Element=@n8.Item, Array=RigVMVariableNode() {variable=Build Joints}.Value) {template="DISPATCH_RigVMDispatch_ArrayAdd(io Array,in Element,out Index)"}
   RigVMDispatchNode(ExecuteContext=(), Array=RigVMVariableNode() {variable=FK Joints}.Value) {template="DISPATCH_RigVMDispatch_ArrayReset(io Array)"}
   > For_Each = RigVMDispatchNode(Array=RigVMVariableNode() {variable=Spine Joints}.Value) {template="DISPATCH_RigVMDispatch_ArrayIterator(in Array,out Element,out Index,out Count,out Ratio)"}
   > RigVMVariableNode(Value=@For_Each.Element) {variable=Current Item}
   > n9 = RigVMVariableNode(Value=@For_Each.Index) {variable=Index}
   Greater = /Script/RigVM.RigVMFunction_MathIntGreater(B=0, A=RigVMVariableNode() {variable=Index}.Value) {is=struct RigVMUnitNode, template="Greater::Execute(in A,in B,out Result)"}
   RigVMFunction_ControlFlowBranch = &RigVMFunction_ControlFlowBranch(Condition=@Greater.Result, ExecuteContext=@n9.ExecuteContext)
   n10 = &RigVMFunction_Sequence(ExecuteContext=@RigVMFunction_ControlFlowBranch.True)
   Subtract = /Script/RigVM.RigVMFunction_MathIntSub(B=1, A=RigVMVariableNode() {variable=Index}.Value) {is=struct RigVMUnitNode, template="Subtract::Execute(in A,in B,out Result)"}
   RigVMVariableNode(
    Value=RigVMDispatchNode(Index=@Subtract.Result, Array=RigVMVariableNode() {variable=Spine Joints}.Value) {template="DISPATCH_RigVMDispatch_ArrayGetAtIndex(in Array,in Index,out Element)"}.Element
    ExecuteContext=@n10.A
   ) {variable=Previous Item}
   n11 = /Script/ControlRig.RigUnit_AimBoneMath(
    DebugSettings=(bEnabled=False,Scale=10,WorldOffset=(Rotation=(X=0,Y=0,Z=0,W=1),Translation=(X=0,Y=0,Z=0),Scale3D=(X=1,Y=1,Z=1)))
    Primary=(Weight=1,Axis=(X=1,Y=0,Z=0),Target=(X=1,Y=0,Z=0),Kind=Location,Space=(Type=Bone,Name="None"))
    Secondary=(Weight=1,Axis=(X=0,Y=0,Z=1),Target=(X=1,Y=0,Z=0),Kind=Location,Space=(Type=Bone,Name="None"))
    Weight=1
    InputTransform=&RigUnit_GetTransform(bInitial=False, Item=RigVMVariableNode() {variable=Previous Item}.Value).Transform
    Primary.Target=&RigUnit_GetTransform(bInitial=False, Item=RigVMVariableNode() {variable=Current Item}.Value).Transform.Translation
   ) {is=struct RigVMUnitNode}
   RigVMFunction_StringPadInteger = &RigVMFunction_StringPadInteger(Digits=2, Value=RigVMVariableNode() {variable=Index}.Value)
   RigVMFunction_StringJoin = &RigVMFunction_StringJoin(Separator=_, Values=("fk","","rigJnt"), Values.1=@RigVMFunction_StringPadInteger.Result)
   From_String = RigVMDispatchNode(String=@RigVMFunction_StringJoin.Result) {template="DISPATCH_RigDispatch_FromString(in String,out Result)"}
   n12 = &RigUnit_HierarchyAddBone(
    Space=GlobalSpace
    Transform=@n11.Result
    Name=@From_String.Result
    ExecutePin=@n10.B
    Parent=RigVMVariableNode() {variable=Current Parent}.Value
   )
   > RigVMVariableNode(Value=@n12.Item) {variable=Current Parent}
   > Add = RigVMDispatchNode(Array=RigVMVariableNode() {variable=FK Joints}.Value, Element=@n12.Item) {template="DISPATCH_RigVMDispatch_ArrayAdd(io Array,in Element,out Index)"}
   n13 = &RigVMFunction_MathIntAdd(B=1, A=RigVMVariableNode() {variable=Index}.Value) {template="Add::Execute(in A,in B,out Result)"}
   From_String_1 = RigVMDispatchNode(String=&RigVMFunction_StringJoin(Separator=_, Values=("fk","","rigJnt"), Values.1=&RigVMFunction_StringPadInteger(Digits=2, Value=@n13.Result).Result).Result) {template="DISPATCH_RigDispatch_FromString(in String,out Result)"}
   n14 = &RigUnit_HierarchyAddBone(
    Space=GlobalSpace
    ExecutePin=@For_Each.Completed
    Name=@From_String_1.Result
    Transform=&RigUnit_GetTransform(bInitial=False, Item=RigVMVariableNode() {variable=Current Item}.Value).Transform
    Parent=RigVMVariableNode() {variable=Current Parent}.Value
   )
   > RigVMDispatchNode(Element=@n14.Item, Array=RigVMVariableNode() {variable=FK Joints}.Value) {template="DISPATCH_RigVMDispatch_ArrayAdd(io Array,in Element,out Index)"}
   > Set Transform = &RigUnit_SetRotation(Value=(X=0,Y=0,Z=0,W=1), Weight=1, bInitial=true, Item=@n14.Item)
   RigVMVariableNode(ExecuteContext=(), Value=RigVMVariableNode() {variable=Connector Spine Root}.Value) {variable=Current Parent}
  ]}
 New Function_ContainedGraph {nodes [
   n1 = RigVMFunctionEntryNode()
   n2 = &RigVMFunction_Sequence(ExecuteContext=@n1.ExecuteContext)
   RigVMFunction_Sequence = RigVMAggregateNode(ExecuteContext=@n2.A)
   n3 = RigVMVariableNode() {variable=Build Joints}
   n4 = RigVMVariableNode() {variable=Control Scale}
   n5 = &RigVMFunction_MathDoubleMul(B=3.5, A=@n4.Value)
   n6 = /Script/ControlRig.RigUnit_HierarchyAddControlTransform(
    InitialValue=(Rotation=(X=0,Y=0,Z=0,W=1),Translation=(X=0,Y=0,Z=0),Scale3D=(X=1,Y=1,Z=1))
    Name=Spine 01
    OffsetSpace=GlobalSpace
    Settings="(InitialSpace=LocalSpace,bUsePreferredRotationOrder=true,PreferredRotationOrder=YXZ,Limits=(LimitTranslationX=(bMinimum=False,bMaximum=False),LimitTranslationY=(bMinimum=False,bMaximum=False),LimitTranslationZ=(bMinimum=False,bMaximum=False),LimitPitch=(bM ...(truncated)"
    OffsetTransform=&RigUnit_GetTransform(
     bInitial=False
     Item=RigVMDispatchNode(Index=0, Array=@n3.Value) {template="DISPATCH_RigVMDispatch_ArrayGetAtIndex(in Array,in Index,out Element)"}.Element
    ).Transform
    Settings.Shape.Transform.Scale3D.X=@n5.Result
    Settings.Shape.Transform.Scale3D.Y=@n5.Result
    Settings.Shape.Transform.Scale3D.Z=&RigVMFunction_MathDoubleMul(B=3, A=@n4.Value).Result
    Parent=RigVMVariableNode() {variable=Hip Control}.Value
    ExecutePin=@RigVMFunction_Sequence.B
   ) {
    is=struct RigVMUnitNode
    template="SpawnControl::Execute(in InitialValue,in Settings,in OffsetTransform,in OffsetSpace,in Parent,in Name,out Item)"
   }
   > RigVMVariableNode(Value=@n6.Item) {variable=Spine 01 Control}
   > RigVMDispatchNode(Element=@n6.Item, Array=RigVMVariableNode() {variable=Controls}.Value) {template="DISPATCH_RigVMDispatch_ArrayAdd(io Array,in Element,out Index)"}
   n7 = RigVMVariableNode() {variable=Build Joints}
   > n8 = &RigUnit_HierarchyAddNull(
    Name=Spine 02 Null
    Space=GlobalSpace
    Transform=&RigUnit_GetTransform(
     bInitial=False
     Item=RigVMDispatchNode(Index=1, Array=@n7.Value) {template="DISPATCH_RigVMDispatch_ArrayGetAtIndex(in Array,in Index,out Element)"}.Element
    ).Transform
    Parent=RigVMVariableNode() {variable=Pelvis Control}.Value
   )
   > RigVMVariableNode(Value=@n8.Item) {variable=Spine 02 Null}
   n9 = RigVMVariableNode() {variable=Control Scale}
   n10 = &RigVMFunction_MathDoubleMul(B=3.5, A=@n9.Value)
   > n11 = &RigUnit_HierarchyAddControlTransform(
    InitialValue=(Rotation=(X=0,Y=0,Z=0,W=1),Translation=(X=0,Y=0,Z=0),Scale3D=(X=1,Y=1,Z=1))
    Name=Spine 02
    OffsetSpace=LocalSpace
    OffsetTransform=(Rotation=(X=0,Y=0,Z=0,W=1),Translation=(X=0,Y=0,Z=0),Scale3D=(X=1,Y=1,Z=1))
    Settings="(InitialSpace=LocalSpace,bUsePreferredRotationOrder=true,PreferredRotationOrder=YXZ,Limits=(LimitTranslationX=(bMinimum=False,bMaximum=False),LimitTranslationY=(bMinimum=False,bMaximum=False),LimitTranslationZ=(bMinimum=False,bMaximum=False),LimitPitch=(bM ...(truncated)"
    Settings.Shape.Transform.Scale3D.X=@n10.Result
    Settings.Shape.Transform.Scale3D.Y=@n10.Result
    Settings.Shape.Transform.Scale3D.Z=&RigVMFunction_MathDoubleMul(B=3, A=@n9.Value).Result
    Parent=RigVMVariableNode() {variable=Spine 02 Null}.Value
   ) {template="SpawnControl::Execute(in InitialValue,in Settings,in OffsetTransform,in OffsetSpace,in Parent,in Name,out Item)"}
   > RigVMVariableNode(Value=(Type=Bone,Name="None")) {variable=Spine 02}
   > RigVMDispatchNode(Element=@n11.Item, Array=RigVMVariableNode() {variable=Controls}.Value) {template="DISPATCH_RigVMDispatch_ArrayAdd(io Array,in Element,out Index)"}
   n12 = RigVMVariableNode() {variable=Build Joints}
   n13 = &RigVMFunction_MathDoubleMul(B=4.5, A=RigVMVariableNode() {variable=Control Scale}.Value)
   > n14 = &RigUnit_HierarchyAddControlTransform(
    InitialValue=(Rotation=(X=0,Y=0,Z=0,W=1),Translation=(X=0,Y=0,Z=0),Scale3D=(X=1,Y=1,Z=1))
    Name=Chest
    OffsetSpace=GlobalSpace
    Settings="(InitialSpace=LocalSpace,bUsePreferredRotationOrder=true,PreferredRotationOrder=YXZ,Limits=(LimitTranslationX=(bMinimum=False,bMaximum=False),LimitTranslationY=(bMinimum=False,bMaximum=False),LimitTranslationZ=(bMinimum=False,bMaximum=False),LimitPitch=(bM ...(truncated)"
    OffsetTransform=&RigUnit_GetTransform(
     bInitial=False
     Item=RigVMDispatchNode(Index=2, Array=@n12.Value) {template="DISPATCH_RigVMDispatch_ArrayGetAtIndex(in Array,in Index,out Element)"}.Element
    ).Transform
    Settings.Shape.Transform.Scale3D.X=@n13.Result
    Settings.Shape.Transform.Scale3D.Y=@n13.Result
    Settings.Shape.Transform.Scale3D.Z=@n13.Result
    Parent=&RigUnit_ResolveConnector(SkipSocket=False, Connector=RigVMVariableNode() {variable=Connector Parent}.Value).Result
   ) {template="SpawnControl::Execute(in InitialValue,in Settings,in OffsetTransform,in OffsetSpace,in Parent,in Name,out Item)"}
   > RigVMVariableNode(Value=@n14.Item) {variable=Chest Control}
   > RigVMDispatchNode(Element=@n14.Item, Array=RigVMVariableNode() {variable=Controls}.Value) {template="DISPATCH_RigVMDispatch_ArrayAdd(io Array,in Element,out Index)"}
   RigVMDispatchNode(Array=RigVMVariableNode() {variable=Controls}.Value, ExecuteContext=@RigVMFunction_Sequence.A) {template="DISPATCH_RigVMDispatch_ArrayReset(io Array)"}
   n15 = RigVMVariableNode() {variable=Build Joints}
   Multiply = &RigVMFunction_MathDoubleMul(B=4, A=RigVMVariableNode() {variable=Control Scale}.Value)
   > n16 = &RigUnit_HierarchyAddControlTransform(
    InitialValue=(Rotation=(X=0,Y=0,Z=0,W=1),Translation=(X=0,Y=0,Z=0),Scale3D=(X=1,Y=1,Z=1))
    Name=Hip
    OffsetSpace=GlobalSpace
    Parent=(Type=Connector,Name="parent")
    Settings="(InitialSpace=LocalSpace,bUsePreferredRotationOrder=true,PreferredRotationOrder=YXZ,Limits=(LimitTranslationX=(bMinimum=False,bMaximum=False),LimitTranslationY=(bMinimum=False,bMaximum=False),LimitTranslationZ=(bMinimum=False,bMaximum=False),LimitPitch=(bM ...(truncated)"
    OffsetTransform=&RigUnit_GetTransform(
     bInitial=false
     Item=RigVMDispatchNode(Index=0, Array=@n15.Value) {template="DISPATCH_RigVMDispatch_ArrayGetAtIndex(in Array,in Index,out Element)"}.Element
    ).Transform
    Settings.Shape.Transform.Scale3D.X=@Multiply.Result
    Settings.Shape.Transform.Scale3D.Y=@Multiply.Result
    Settings.Shape.Transform.Scale3D.Z=@Multiply.Result
   ) {template="SpawnControl::Execute(in InitialValue,in Settings,in OffsetTransform,in OffsetSpace,in Parent,in Name,out Item)"}
   > RigVMVariableNode(Value=@n16.Item) {variable=Hip Control}
   n17 = &RigVMFunction_MathDoubleMul(B=3.5, A=RigVMVariableNode() {variable=Control Scale}.Value)
   > n18 = &RigUnit_HierarchyAddControlTransform(
    InitialValue=(Rotation=(X=0,Y=0,Z=0,W=1),Translation=(X=0,Y=0,Z=0),Scale3D=(X=1,Y=1,Z=1))
    Name=Pelvis FK
    OffsetSpace=GlobalSpace
    Settings="(InitialSpace=LocalSpace,bUsePreferredRotationOrder=true,PreferredRotationOrder=YXZ,Limits=(LimitTranslationX=(bMinimum=False,bMaximum=False),LimitTranslationY=(bMinimum=False,bMaximum=False),LimitTranslationZ=(bMinimum=False,bMaximum=False),LimitPitch=(bM ...(truncated)"
    OffsetTransform=&RigUnit_GetTransform(bInitial=false, Item=RigVMVariableNode() {variable=Pelvis Joint}.Value).Transform
    Settings.Shape.Transform.Scale3D.X=@n17.Result
    Settings.Shape.Transform.Scale3D.Y=@n17.Result
    Settings.Shape.Transform.Scale3D.Z=@n17.Result
    Parent=@n16.Item
   ) {template="SpawnControl::Execute(in InitialValue,in Settings,in OffsetTransform,in OffsetSpace,in Parent,in Name,out Item)"}
   > RigVMVariableNode(Value=@n18.Item) {variable=Pelvis Control}
   n19 = RigVMVariableNode() {variable=Build Joints}
   n20 = &RigUnit_HierarchyAddControlTransform(
    InitialValue=(Rotation=(X=0,Y=0,Z=0,W=1),Translation=(X=0,Y=0,Z=0),Scale3D=(X=1,Y=1,Z=1))
    Name=Spine Fwd Proxy
    OffsetSpace=GlobalSpace
    Parent=(Type=Bone,Name="None")
    Settings="(InitialSpace=LocalSpace,bUsePreferredRotationOrder=False,PreferredRotationOrder=YZX,Limits=(LimitTranslationX=(bMinimum=False,bMaximum=False),LimitTranslationY=(bMinimum=False,bMaximum=False),LimitTranslationZ=(bMinimum=False,bMaximum=False),LimitPitch=(b ...(truncated)"
    OffsetTransform=&RigUnit_GetTransform(
     bInitial=false
     Item=RigVMDispatchNode(Index=0, Array=@n19.Value) {template="DISPATCH_RigVMDispatch_ArrayGetAtIndex(in Array,in Index,out Element)"}.Element
    ).Transform
    Settings.Proxy.DrivenControls.0=RigVMVariableNode() {variable=Chest Control}.Value
    Settings.Shape.Transform=RigVMVariableNode() {variable=Spine Fwd Proxy Position}.Value
    ExecutePin=@RigVMFunction_Sequence.C
   ) {template="SpawnControl::Execute(in InitialValue,in Settings,in OffsetTransform,in OffsetSpace,in Parent,in Name,out Item)"}
   > RigVMVariableNode(Value=@n20.Item) {variable=Spine Fwd Proxy}
   n21 = RigVMVariableNode() {variable=Build Joints}
   > n22 = &RigUnit_HierarchyAddControlTransform(
    InitialValue=(Rotation=(X=0,Y=0,Z=0,W=1),Translation=(X=0,Y=0,Z=0),Scale3D=(X=1,Y=1,Z=1))
    Name=Spine Bk Proxy
    OffsetSpace=GlobalSpace
    Parent=(Type=Connector,Name="parent")
    Settings="(InitialSpace=LocalSpace,bUsePreferredRotationOrder=False,PreferredRotationOrder=YZX,Limits=(LimitTranslationX=(bMinimum=False,bMaximum=False),LimitTranslationY=(bMinimum=False,bMaximum=False),LimitTranslationZ=(bMinimum=False,bMaximum=False),LimitPitch=(b ...(truncated)"
    OffsetTransform=&RigUnit_GetTransform(
     bInitial=false
     Item=RigVMDispatchNode(Index=2, Array=@n21.Value) {template="DISPATCH_RigVMDispatch_ArrayGetAtIndex(in Array,in Index,out Element)"}.Element
    ).Transform
    Settings.Proxy.DrivenControls.0=RigVMVariableNode() {variable=Hip Control}.Value
    Settings.Shape.Transform=RigVMVariableNode() {variable=Spine BK Proxy Position}.Value
   ) {template="SpawnControl::Execute(in InitialValue,in Settings,in OffsetTransform,in OffsetSpace,in Parent,in Name,out Item)"}
   > RigVMVariableNode(Value=@n22.Item) {variable=Spine Bk Proxy}
   n23 = /Script/ControlRig.RigUnit_HierarchyAddAnimationChannelBool(
    InitialValue=false
    LimitsEnabled=()
    MaximumValue=True
    MinimumValue=false
    Name=sec controls Visibilty
    Parent=RigVMVariableNode() {variable=Hip Control}.Value
    ExecutePin=@RigVMFunction_Sequence.D
   ) {
    is=struct RigVMUnitNode
    template="SpawnAnimationChannel::Execute(in InitialValue,in MinimumValue,in MaximumValue,in LimitsEnabled,in Parent,in Name,out Item)"
   }
   RigUnit_ItemArray = &RigUnit_ItemArray(
    Items=((Type=None,Name="CRM_Quad_Spine_01/Pelvis FK"),(Type=Control,Name="CRM_Quad_Spine_01/Pelvis FK"))
    Items.0=RigVMVariableNode() {variable=Spine 01 Control}.Value
    Items.1=RigVMVariableNode() {variable=Pelvis Control}.Value
   )
   > /Script/ControlRig.RigUnit_SetChannelHosts(Hosts=@RigUnit_ItemArray.Items, Channel=@n23.Item) {is=struct RigVMUnitNode}
   > RigVMVariableNode(Value=@n23.Item) {variable=sec_controls_visibility}
   RigVMFunctionReturnNode(ExecuteContext=@n2.B)
   &RigUnit_HierarchyAddControlTransform(
    ExecutePin=()
    InitialValue=(Rotation=(X=0,Y=0,Z=0,W=1),Translation=(X=0,Y=0,Z=0),Scale3D=(X=1,Y=1,Z=1))
    Name=Spine Fwd Proxy
    OffsetSpace=GlobalSpace
    OffsetTransform=(Rotation=(X=0,Y=0,Z=0,W=1),Translation=(X=0,Y=0,Z=0),Scale3D=(X=1,Y=1,Z=1))
    Parent=(Type=Connector,Name="parent")
    Settings="(InitialSpace=LocalSpace,bUsePreferredRotationOrder=False,PreferredRotationOrder=YZX,Limits=(LimitTranslationX=(bMinimum=False,bMaximum=False),LimitTranslationY=(bMinimum=False,bMaximum=False),LimitTranslationZ=(bMinimum=False,bMaximum=False),LimitPitch=(b ...(truncated)"
   ) {template="SpawnControl::Execute(in InitialValue,in Settings,in OffsetTransform,in OffsetSpace,in Parent,in Name,out Item)"}
   RigVMVariableNode() {variable=Connector Spine Root}
  ]}
 New Function_ContainedGraph {nodes [
   RigVMFunctionEntryNode()
   > RigVMFunction_Sequence = RigVMAggregateNode()
   RigVMFunctionReturnNode(ExecuteContext=@RigVMFunction_Sequence.C)
   RigVMDispatchNode(Array=RigVMVariableNode() {variable=Spine Local Transforms}.Value, ExecuteContext=@RigVMFunction_Sequence.A) {template="DISPATCH_RigVMDispatch_ArrayReset(io Array)"}
   > For_Each = RigVMDispatchNode(Array=RigVMVariableNode() {variable=Spine Joints}.Value) {template="DISPATCH_RigVMDispatch_ArrayIterator(in Array,out Element,out Index,out Count,out Ratio)"}
   n1 = RigVMDispatchNode(Index=@For_Each.Index, Array=RigVMVariableNode() {variable=Controls}.Value) {template="DISPATCH_RigVMDispatch_ArrayGetAtIndex(in Array,in Index,out Element)"}
   > Add = RigVMDispatchNode(
    Array=RigVMVariableNode() {variable=Spine Local Transforms}.Value
    Element=&RigVMFunction_MathTransformMakeRelative(
     Global=&RigUnit_GetTransform(bInitial=false, Item=@For_Each.Element).Transform
     Parent=&RigUnit_GetTransform(bInitial=false, Item=@n1.Element).Transform
    ).Local
   ) {template="DISPATCH_RigVMDispatch_ArrayAdd(io Array,in Element,out Index)"}
   n2 = RigVMVariableNode() {variable=Connector Spine Root}
   n3 = RigVMVariableNode() {variable=Pelvis Control}
   RigVMVariableNode(
    Value=&RigVMFunction_MathTransformMakeRelative(
     Global=&RigUnit_GetTransform(bInitial=false, Item=@n2.Value).Transform
     Parent=&RigUnit_GetTransform(bInitial=false, Item=@n3.Value).Transform
    ).Local
    ExecuteContext=@RigVMFunction_Sequence.B
   ) {variable=Pelvis Local Transforms}
   RigVMVariableNode() {variable=Build Joints}
  ]}
 New Function_ContainedGraph {nodes [
   RigVMFunctionEntryNode()
   n1 = /Script/ControlRig.RigUnit_IsInteracting() {is=struct RigVMUnitNode}
   > RigVMVariableNode(Value=@n1.bIsInteracting) {variable=is Interacting}
   > RigVMVariableNode(Value=@n1.Items) {variable=Set Interacting Items}
   > RigVMFunctionReturnNode()
  ]}
 New Function_ContainedGraph {nodes [
   n1 = RigVMFunctionEntryNode()
   RigVMFunction_Sequence = &RigVMFunction_Sequence(ExecuteContext=@n1.ExecuteContext)
   RigVMFunction_ControlFlowBranch = &RigVMFunction_ControlFlowBranch(ExecuteContext=@RigVMFunction_Sequence.A, Condition=RigVMVariableNode() {variable=is Interacting}.Value)
   &RigUnit_ParentConstraint(
    AdvancedSettings=(InterpolationType=Average,RotationOrderForFilter=XZY)
    Child=(Type=Control,Name="Spine Fwd Proxy")
    Parents=((Item=(Type=Control,Name="Hip"),Weight=1))
    Weight=1
    bMaintainOffset=false
    ExecutePin=@RigVMFunction_ControlFlowBranch.False
   )
   > &RigUnit_ParentConstraint(
    AdvancedSettings=(InterpolationType=Average,RotationOrderForFilter=XZY)
    Child=(Type=Control,Name="Spine Bk Proxy")
    Parents=((Item=(Type=Control,Name="Chest"),Weight=1))
    Weight=1
    bMaintainOffset=false
   )
   RigUnit_GetTransform = &RigUnit_GetTransform(Item=(Type=Control,Name="Chest"), bInitial=False)
   > RigVMVariableNode(Value=&RigVMFunction_MathTransformMakeRelative(
     Global=@RigUnit_GetTransform.Transform
     Parent=&RigUnit_GetTransform(Item=(Type=Control,Name="Spine Fwd Proxy"), bInitial=False).Transform
    ).Local) {variable=Chest Local Transform Offset}
   > RigVMVariableNode(Value=&RigVMFunction_MathTransformMakeRelative(
     Parent=&RigUnit_GetTransform(Item=(Type=Control,Name="Spine Fwd Proxy"), bInitial=False).Transform
     Global=&RigUnit_GetTransform(Item=(Type=Null,Name="Spine 02 Null"), bInitial=False).Transform
    ).Local) {variable=Chest Spine 02 Null Local Transform Offset}
   > RigVMVariableNode(Value=&RigVMFunction_MathTransformMakeRelative(
     Parent=&RigUnit_GetTransform(Item=(Type=Control,Name="Spine Bk Proxy"), bInitial=False).Transform
     Global=&RigUnit_GetTransform(Item=(Type=Control,Name="Hip"), bInitial=False).Transform
    ).Local) {variable=Hip Local Transform Offset}
   > RigVMVariableNode(Value=&RigVMFunction_MathTransformMakeRelative(
     Parent=&RigUnit_GetTransform(Item=(Type=Control,Name="Spine Bk Proxy"), bInitial=False).Transform
     Global=&RigUnit_GetTransform(Item=(Type=Null,Name="Spine 02 Null"), bInitial=False).Transform
    ).Local) {variable=Hip Spine 02 Null Local Transform Offset}
   n2 = &RigVMFunction_ControlFlowBranch(
    Condition=RigVMDispatchNode(Array=RigVMVariableNode() {variable=Set Interacting Items}.Value, Element=RigVMVariableNode() {variable=Spine Fwd Proxy}.Value) {template="DISPATCH_RigVMDispatch_ArrayFind(in Array,in Element,out Index,out Success)"}.Success
    ExecuteContext=@RigVMFunction_ControlFlowBranch.True
   )
   n3 = &RigVMFunction_ControlFlowBranch(
    Condition=RigVMDispatchNode(Array=RigVMVariableNode() {variable=Set Interacting Items}.Value, Element=RigVMVariableNode() {variable=Spine Bk Proxy}.Value) {template="DISPATCH_RigVMDispatch_ArrayFind(in Array,in Element,out Index,out Success)"}.Success
    ExecuteContext=@n2.False
   )
   n4 = RigVMVariableNode() {variable=Spine Bk Proxy}
   &RigUnit_SetTransform(
    Item=(Type=Control,Name="Hip")
    Weight=1
    bInitial=False
    Value=/Script/RigVM.RigVMFunction_MathTransformMakeAbsolute(
     Parent=&RigUnit_GetTransform(bInitial=False, Item=@n4.Value).Transform
     Local=RigVMVariableNode() {variable=Hip Local Transform Offset}.Value
    ) {is=struct RigVMUnitNode, template="Make Absolute::Execute(in Local,in Parent,out Global)"}.Global
    ExecutePin=@n3.True
   )
   n5 = RigVMVariableNode() {variable=Spine Fwd Proxy}
   &RigUnit_SetTransform(
    Item=(Type=Control,Name="Chest")
    Weight=1
    bInitial=False
    Value=&RigVMFunction_MathTransformMakeAbsolute(
     Parent=&RigUnit_GetTransform(bInitial=False, Item=@n5.Value).Transform
     Local=RigVMVariableNode() {variable=Chest Local Transform Offset}.Value
    ) {template="Make Absolute::Execute(in Local,in Parent,out Global)"}.Global
    ExecutePin=@n2.True
   )
   RigVMFunctionReturnNode(ExecuteContext=@RigVMFunction_Sequence.B)
   n6 = RigVMVariableNode() {variable=Spine Fwd Proxy}
   &RigUnit_SetTransform(
    ExecutePin=()
    Item=(Type=Null,Name="Spine 02 Null")
    Weight=1
    bInitial=False
    Value=&RigVMFunction_MathTransformMakeAbsolute(
     Parent=&RigUnit_GetTransform(bInitial=False, Item=@n6.Value).Transform
     Local=RigVMVariableNode() {variable=Chest Spine 02 Null Local Transform Offset}.Value
    ) {template="Make Absolute::Execute(in Local,in Parent,out Global)"}.Global
   )
   n7 = RigVMVariableNode() {variable=Spine Bk Proxy}
   &RigUnit_SetTransform(
    ExecutePin=()
    Item=(Type=Null,Name="Spine 02 Null")
    Weight=1
    bInitial=False
    Value=&RigVMFunction_MathTransformMakeAbsolute(
     Local=(Rotation=(X=0,Y=0,Z=0,W=1),Translation=(X=0,Y=0,Z=0),Scale3D=(X=1,Y=1,Z=1))
     Parent=&RigUnit_GetTransform(bInitial=False, Item=@n7.Value).Transform
    ) {template="Make Absolute::Execute(in Local,in Parent,out Global)"}.Global
   )
  ]}
 RigVMFunctionLibrary {nodes [
   Check if Interacting = RigVMCollapseNode(ExecuteContext=())
   Create Build Joints = RigVMCollapseNode(ExecuteContext=())
   Create Controls = RigVMCollapseNode(ExecuteContext=())
   Find Spine Joints = RigVMCollapseNode(ExecuteContext=())
   Get Local Transform Values = RigVMCollapseNode(ExecuteContext=())
   Set Proxy Interacting = RigVMCollapseNode(ExecuteContext=())
  ]}
 RigVMFunction_Sequence_1_ContainedGraph {nodes [
   n1 = RigVMFunctionEntryNode()
   RigVMFunction_Sequence = &RigVMFunction_Sequence(ExecuteContext=@n1.ExecuteContext)
   n2 = &RigVMFunction_Sequence(ExecuteContext=@RigVMFunction_Sequence.B)
   n3 = &RigVMFunction_Sequence(ExecuteContext=@n2.B)
   RigVMFunctionReturnNode(A=@RigVMFunction_Sequence.A, B=@n2.A, C=@n3.A, D=@n3.B)
  ]}
 RigVMFunction_Sequence_1_ContainedGraph {nodes [
   n1 = RigVMFunctionEntryNode()
   RigVMFunction_Sequence = &RigVMFunction_Sequence(ExecuteContext=@n1.ExecuteContext)
   n2 = &RigVMFunction_Sequence(ExecuteContext=@RigVMFunction_Sequence.B)
   RigVMFunctionReturnNode(A=@RigVMFunction_Sequence.A, B=@n2.A, C=@n2.B)
  ]}
 RigVMFunction_Sequence_1_ContainedGraph {nodes [
   n1 = RigVMFunctionEntryNode()
   RigVMFunction_Sequence = &RigVMFunction_Sequence(ExecuteContext=@n1.ExecuteContext)
   n2 = &RigVMFunction_Sequence(ExecuteContext=@RigVMFunction_Sequence.B)
   RigVMFunctionReturnNode(A=@RigVMFunction_Sequence.A, B=@n2.A, C=@n2.B)
  ]}
 RigVMFunction_Sequence_1_ContainedGraph {nodes [
   n1 = RigVMFunctionEntryNode()
   RigVMFunction_Sequence = &RigVMFunction_Sequence(ExecuteContext=@n1.ExecuteContext)
   n2 = &RigVMFunction_Sequence(ExecuteContext=@RigVMFunction_Sequence.B)
   n3 = &RigVMFunction_Sequence(ExecuteContext=@n2.B)
   RigVMFunctionReturnNode(A=@RigVMFunction_Sequence.A, B=@n2.A, C=@n3.A, D=@n3.B)
  ]}
 RigVMFunction_Sequence_1_ContainedGraph {nodes [
   n1 = RigVMFunctionEntryNode()
   RigVMFunction_Sequence = &RigVMFunction_Sequence(ExecuteContext=@n1.ExecuteContext)
   n2 = &RigVMFunction_Sequence(ExecuteContext=@RigVMFunction_Sequence.B)
   n3 = &RigVMFunction_Sequence(ExecuteContext=@n2.B)
   RigVMFunctionReturnNode(A=@RigVMFunction_Sequence.A, B=@n2.A, C=@n3.A, D=@n3.B)
  ]}
 RigVMFunction_Sequence_1_ContainedGraph {nodes [
   n1 = RigVMFunctionEntryNode()
   RigVMFunction_Sequence = &RigVMFunction_Sequence(ExecuteContext=@n1.ExecuteContext)
   n2 = &RigVMFunction_Sequence(ExecuteContext=@RigVMFunction_Sequence.B)
   RigVMFunctionReturnNode(A=@RigVMFunction_Sequence.A, B=@n2.A, C=@n2.B)
  ]}
 RigVMFunction_Sequence_2_ContainedGraph {nodes [
   n1 = RigVMFunctionEntryNode()
   RigVMFunction_Sequence = &RigVMFunction_Sequence(ExecuteContext=@n1.ExecuteContext)
   n2 = &RigVMFunction_Sequence(ExecuteContext=@RigVMFunction_Sequence.B)
   n3 = &RigVMFunction_Sequence(ExecuteContext=@n2.B)
   n4 = &RigVMFunction_Sequence(ExecuteContext=@n3.B)
   n5 = &RigVMFunction_Sequence(ExecuteContext=@n4.B)
   RigVMFunctionReturnNode(A=@RigVMFunction_Sequence.A, B=@n2.A, C=@n3.A, D=@n4.A, E=@n5.A, F=@n5.B)
  ]}
 RigVMModel {nodes [
   For_Each_1 = RigVMDispatchNode(ExecuteContext=(), Array=RigVMVariableNode() {variable=Controls}.Value) {template="DISPATCH_RigVMDispatch_ArrayIterator(in Array,out Element,out Index,out Count,out Ratio)"}
   > &RigUnit_SetTransform(
    Weight=1
    bInitial=False
    Item=RigVMDispatchNode(Index=@For_Each_1.Index, Array=RigVMVariableNode() {variable=Build Joints}.Value) {template="DISPATCH_RigVMDispatch_ArrayGetAtIndex(in Array,in Index,out Element)"}.Element
    Value=&RigUnit_GetTransform(bInitial=False, Item=@For_Each_1.Element).Transform
   )
   RigUnit_BeginExecution = &RigUnit_BeginExecution()
   RigVMFunction_Sequence = RigVMAggregateNode(ExecuteContext=@RigUnit_BeginExecution.ExecutePin)
   Check if Interacting = RigVMFunctionReferenceNode(ExecuteContext=@RigVMFunction_Sequence.E)
   > Set Proxy Interacting = RigVMFunctionReferenceNode()
   n1 = RigVMVariableNode() {variable=Pelvis Control}
   &RigUnit_SetTransform(
    Weight=1
    bInitial=false
    bPropagateToChildren=true
    Value=&RigVMFunction_MathTransformMakeAbsolute(
     Parent=&RigUnit_GetTransform(bInitial=False, Item=@n1.Value).Transform
     Local=RigVMVariableNode() {variable=Pelvis Local Transforms}.Value
    ) {template="Make Absolute::Execute(in Local,in Parent,out Global)"}.Global
    Item=RigVMVariableNode() {variable=Connector Spine Root}.Value
    ExecutePin=@RigVMFunction_Sequence.B
   )
   For_Each = RigVMDispatchNode(Array=RigVMVariableNode() {variable=Spine Local Transforms}.Value, ExecuteContext=@RigVMFunction_Sequence.C) {template="DISPATCH_RigVMDispatch_ArrayIterator(in Array,out Element,out Index,out Count,out Ratio)"}
   n2 = RigVMDispatchNode(Array=RigVMVariableNode() {variable=Controls}.Value, Index=@For_Each.Index) {template="DISPATCH_RigVMDispatch_ArrayGetAtIndex(in Array,in Index,out Element)"}
   > &RigUnit_SetTransform(
    Weight=1
    bInitial=False
    bPropagateToChildren=true
    Item=RigVMDispatchNode(Array=RigVMVariableNode() {variable=Spine Joints}.Value, Index=@For_Each.Index) {template="DISPATCH_RigVMDispatch_ArrayGetAtIndex(in Array,in Index,out Element)"}.Element
    Value=&RigVMFunction_MathTransformMakeAbsolute(Parent=&RigUnit_GetTransform(bInitial=false, Item=@n2.Element).Transform, Local=@For_Each.Element) {template="Make Absolute::Execute(in Local,in Parent,out Global)"}.Global
   )
   &RigUnit_ParentConstraint(
    AdvancedSettings=(InterpolationType=Average,RotationOrderForFilter=XZY)
    Parents=((Item=(Type=Control,Name="Spine 01"),Weight=0.25),(Item=(Type=Control,Name="Chest"),Weight=0.75))
    Weight=1
    Parents.0.Item=RigVMDispatchNode(Index=0, Array=RigVMVariableNode() {variable=Controls}.Value) {template="DISPATCH_RigVMDispatch_ArrayGetAtIndex(in Array,in Index,out Element)"}.Element
    Child=RigVMVariableNode() {variable=Spine 02 Null}.Value
    Parents.1.Item=RigVMVariableNode() {variable=Chest Control}.Value
    ExecutePin=@RigVMFunction_Sequence.A
   )
   GetAnimationChannelFromItem = /Script/ControlRig.RigUnit_GetBoolAnimationChannelFromItem(bInitial=False, Item=RigVMVariableNode() {variable=sec_controls_visibility}.Value) {is=struct RigVMUnitNode, template="GetAnimationChannelFromItem::Execute(out Value,in Item,in bInitial)"}
   Set_Module_Metadata = RigVMDispatchNode(Name=None, NameSpace=Self, Value=@GetAnimationChannelFromItem.Value, ExecuteContext=@RigVMFunction_Sequence.D) {template="DISPATCH_RigDispatch_SetModuleMetadata(in Name,in NameSpace,in Value,out Success)"}
   RigUnit_ItemArray = &RigUnit_ItemArray(
    Items=((Type=None,Name="CRM_Quad_Spine_01/Pelvis FK"),(Type=Control,Name="CRM_Quad_Spine_01/Pelvis FK"))
    Items.0=RigVMVariableNode() {variable=Spine 01 Control}.Value
    Items.1=RigVMVariableNode() {variable=Pelvis Control}.Value
   )
   > For_Each_2 = RigVMDispatchNode(Array=@RigUnit_ItemArray.Items) {template="DISPATCH_RigVMDispatch_ArrayIterator(in Array,out Element,out Index,out Count,out Ratio)"}
   > SetControlVisibility = /Script/ControlRig.RigUnit_SetControlVisibility(Item=@For_Each_2.Element, bVisible=@GetAnimationChannelFromItem.Value) {is=struct RigVMUnitNode, template="SetControlVisibility::Execute(in Item,in Pattern,in bVisible)"}
   n3 = &RigUnit_IsInteracting()
   RigVMFunction_ControlFlowBranch = &RigVMFunction_ControlFlowBranch(ExecuteContext=(), Condition=@n3.bIsInteracting)
   &RigUnit_SetTransform(
    Item=(Type=Control,Name="spine_fwd_proxy")
    Space=LocalSpace
    Value=(Rotation=(X=0,Y=0,Z=0,W=1),Translation=(X=0,Y=0,Z=0),Scale3D=(X=1,Y=1,Z=1))
    Weight=1
    bInitial=False
    ExecutePin=@RigVMFunction_ControlFlowBranch.False
   )
   n4 = &RigVMFunction_ControlFlowBranch(
    Condition=RigVMDispatchNode(Element=(Type=Control,Name="spine_fwd_proxy"), Array=@n3.Items) {template="DISPATCH_RigVMDispatch_ArrayFind(in Array,in Element,out Index,out Success)"}.Success
    ExecuteContext=@RigVMFunction_ControlFlowBranch.True
   )
   &RigUnit_ParentConstraint(
    AdvancedSettings=(InterpolationType=Average,RotationOrderForFilter=XZY)
    Child=(Type=Control,Name="spine_03_ctrl")
    Parents=((Item=(Type=Control,Name="spine_fwd_proxy"),Weight=1))
    Weight=1
    ExecutePin=@n4.True
   )
   &RigUnit_OffsetTransformForItem(
    ExecutePin=()
    Item=(Type=Bone,Name="None")
    OffsetTransform=(Rotation=(X=0,Y=0,Z=0,W=1),Translation=(X=0,Y=0,Z=0),Scale3D=(X=1,Y=1,Z=1))
    Weight=1
    bPropagateToChildren=True
   )
   n5 = &RigUnit_GetTransform(Item=(Type=Null,Name="spine_fwd_null"), bInitial=False)
   n6 = &RigUnit_GetTransform(Item=(Type=Control,Name="spine_03_ctrl"), bInitial=False)
   &RigUnit_SetTransform(
    ExecutePin=()
    Item=(Type=Control,Name="spine_03_ctrl")
    Weight=1
    bInitial=False
    Value=&RigVMFunction_MathTransformMakeAbsolute(
     Parent=&RigUnit_GetTransform(Item=(Type=Control,Name="spine_fwd_proxy"), bInitial=False).Transform
     Local=&RigVMFunction_MathTransformMakeRelative(Parent=@n5.Transform, Global=@n6.Transform).Local
    ) {template="Make Absolute::Execute(in Local,in Parent,out Global)"}.Global
   )
   &RigUnit_SetTransform(
    ExecutePin=()
    Weight=1
    bInitial=False
    Value=&RigUnit_GetTransform(bInitial=False, Item=RigVMVariableNode() {variable=Pelvis Control}.Value).Transform
    Item=RigVMVariableNode() {variable=Pelvis Joint}.Value
   )
  ]}
 RigVMModel Backwards {nodes [
   n1 = /Script/ControlRig.RigUnit_InverseExecution() {is=struct RigVMUnitNode}
   RigVMFunction_Sequence = RigVMAggregateNode(ExecuteContext=@n1.ExecutePin)
   n2 = RigVMVariableNode() {variable=Connector Spine Base}
   &RigUnit_SetTransform(
    Weight=1
    bInitial=false
    bPropagateToChildren=true
    Value=&RigVMFunction_MathTransformMakeAbsolute(
     Parent=&RigUnit_GetTransform(bInitial=False, Item=@n2.Value).Transform
     Local=RigVMVariableNode() {variable=Spine 1 Local Transform BK}.Value
    ) {template="Make Absolute::Execute(in Local,in Parent,out Global)"}.Global
    Item=RigVMVariableNode() {variable=Hip Control}.Value
    ExecutePin=@RigVMFunction_Sequence.C
   )
   n3 = RigVMVariableNode() {variable=Connector Spine Root}
   > &RigUnit_SetTransform(
    Weight=1
    bInitial=false
    bPropagateToChildren=false
    Value=&RigVMFunction_MathTransformMakeAbsolute(
     Parent=&RigUnit_GetTransform(bInitial=False, Item=@n3.Value).Transform
     Local=RigVMVariableNode() {variable=Pelvis Local Transforms BK}.Value
    ) {template="Make Absolute::Execute(in Local,in Parent,out Global)"}.Global
    Item=RigVMVariableNode() {variable=Pelvis Control}.Value
   )
   n4 = RigVMVariableNode() {variable=Connector Spine Tip}
   > &RigUnit_SetTransform(
    Weight=1
    bInitial=false
    bPropagateToChildren=false
    Item=RigVMDispatchNode(Index=2, Array=RigVMVariableNode() {variable=Controls}.Value) {template="DISPATCH_RigVMDispatch_ArrayGetAtIndex(in Array,in Index,out Element)"}.Element
    Value=&RigVMFunction_MathTransformMakeAbsolute(
     Parent=&RigUnit_GetTransform(bInitial=False, Item=@n4.Value).Transform
     Local=RigVMVariableNode() {variable=Chest Local Transform BK}.Value
    ) {template="Make Absolute::Execute(in Local,in Parent,out Global)"}.Global
   )
   > &RigUnit_ParentConstraint(
    AdvancedSettings=(InterpolationType=Average,RotationOrderForFilter=XZY)
    Parents=((Item=(Type=Control,Name="Spine 01"),Weight=0.25),(Item=(Type=Control,Name="Chest"),Weight=0.75))
    Weight=1
    Parents.0.Item=RigVMDispatchNode(Index=0, Array=RigVMVariableNode() {variable=Controls}.Value) {template="DISPATCH_RigVMDispatch_ArrayGetAtIndex(in Array,in Index,out Element)"}.Element
    Child=RigVMVariableNode() {variable=Spine 02 Null}.Value
    Parents.1.Item=RigVMVariableNode() {variable=Chest Control}.Value
   )
   n5 = RigVMVariableNode() {variable=Spine Mid}
   > &RigUnit_SetTransform(
    Weight=1
    bInitial=false
    bPropagateToChildren=false
    Item=RigVMDispatchNode(Index=1, Array=RigVMVariableNode() {variable=Controls}.Value) {template="DISPATCH_RigVMDispatch_ArrayGetAtIndex(in Array,in Index,out Element)"}.Element
    Value=&RigVMFunction_MathTransformMakeAbsolute(
     Parent=&RigUnit_GetTransform(bInitial=False, Item=@n5.Value).Transform
     Local=RigVMVariableNode() {variable=Spine 2 Local Transform BK}.Value
    ) {template="Make Absolute::Execute(in Local,in Parent,out Global)"}.Global
   )
   n6 = RigVMVariableNode() {variable=Connector Spine Root}
   n7 = RigVMVariableNode() {variable=Pelvis Control}
   RigVMVariableNode(
    Value=&RigVMFunction_MathTransformMakeRelative(
     Parent=&RigUnit_GetTransform(bInitial=true, Item=@n6.Value).Transform
     Global=&RigUnit_GetTransform(bInitial=true, Item=@n7.Value).Transform
    ).Local
    ExecuteContext=@RigVMFunction_Sequence.B
   ) {variable=Pelvis Local Transforms BK}
   n8 = RigVMVariableNode() {variable=Connector Spine Base}
   n9 = RigVMVariableNode() {variable=Hip Control}
   > RigVMVariableNode(Value=&RigVMFunction_MathTransformMakeRelative(
     Parent=&RigUnit_GetTransform(bInitial=true, Item=@n8.Value).Transform
     Global=&RigUnit_GetTransform(bInitial=true, Item=@n9.Value).Transform
    ).Local) {variable=Spine 1 Local Transform BK}
   n10 = RigVMVariableNode() {variable=Spine Mid}
   n11 = RigVMDispatchNode(Index=1, Array=RigVMVariableNode() {variable=Controls}.Value) {template="DISPATCH_RigVMDispatch_ArrayGetAtIndex(in Array,in Index,out Element)"}
   > RigVMVariableNode(Value=&RigVMFunction_MathTransformMakeRelative(
     Parent=&RigUnit_GetTransform(bInitial=true, Item=@n10.Value).Transform
     Global=&RigUnit_GetTransform(bInitial=true, Item=@n11.Element).Transform
    ).Local) {variable=Spine 2 Local Transform BK}
   n12 = RigVMVariableNode() {variable=Connector Spine Tip}
   n13 = RigVMDispatchNode(Index=2, Array=RigVMVariableNode() {variable=Controls}.Value) {template="DISPATCH_RigVMDispatch_ArrayGetAtIndex(in Array,in Index,out Element)"}
   > RigVMVariableNode(Value=&RigVMFunction_MathTransformMakeRelative(
     Parent=&RigUnit_GetTransform(bInitial=true, Item=@n12.Value).Transform
     Global=&RigUnit_GetTransform(bInitial=true, Item=@n13.Element).Transform
    ).Local) {variable=Chest Local Transform BK}
   RigVMRerouteNode(Value=())
  ]}
 RigVMModel Rig1_1 {nodes [
   n1 = &RigUnit_PrepareForExecution()
   RigVMFunction_Sequence = RigVMAggregateNode(ExecuteContext=@n1.ExecutePin)
   Find Spine Joints = RigVMFunctionReferenceNode(ExecuteContext=@RigVMFunction_Sequence.B)
   Create Build Joints = RigVMFunctionReferenceNode(ExecuteContext=@RigVMFunction_Sequence.C)
   > Create Controls = RigVMFunctionReferenceNode()
   Get Local Transform Values = RigVMFunctionReferenceNode(ExecuteContext=@RigVMFunction_Sequence.D)
   RigVMDispatchNode(Array=RigVMVariableNode() {variable=Spine Joints}.Value, ExecuteContext=@RigVMFunction_Sequence.A) {template="DISPATCH_RigVMDispatch_ArrayReset(io Array)"}
   > RigVMDispatchNode(Array=RigVMVariableNode() {variable=FK Joints}.Value) {template="DISPATCH_RigVMDispatch_ArrayReset(io Array)"}
   > RigVMDispatchNode(Array=RigVMVariableNode() {variable=Build Joints}.Value) {template="DISPATCH_RigVMDispatch_ArrayReset(io Array)"}
   > RigVMDispatchNode(Array=RigVMVariableNode() {variable=Controls}.Value) {template="DISPATCH_RigVMDispatch_ArrayReset(io Array)"}
   > RigVMDispatchNode(Array=RigVMVariableNode() {variable=Set Interacting Items}.Value) {template="DISPATCH_RigVMDispatch_ArrayReset(io Array)"}
  ]}
]
hierarchy [
 SpineRoot {type=Connector, loc=<0 0 0>, rot=<0 0 0 1>}
 Root {type=Socket, loc=<0 0 0>, rot=<0 0 0 1>}
 Spine 3 {type=Connector, loc=<0 0 0>, rot=<0 0 0 1>}
 Spine 1 {type=Connector, loc=<0 0 0>, rot=<0 0 0 1>}
 parent {type=Connector, loc=<0 0 0>, rot=<0 0 0 1>}
 Spine 2 {type=Connector, loc=<0 0 0>, rot=<0 0 0 1>}
]
variables [
 Build Joints {type="TArray<FRigElementKey>"}
 Build Parent {type=FRigElementKey, default=(Type=None,Name="")}
 Chest Control {type=FRigElementKey, default=(Type=Bone,Name="")}
 Chest Local Transform BK {type=FTransform, default=(Rotation=(X=0,Y=0,Z=0,W=1),Translation=(X=0,Y=0,Z=0),Scale3D=(X=1,Y=1,Z=1))}
 Chest Local Transform Offset {type=FTransform, default=(Rotation=(X=0,Y=0,Z=0,W=1),Translation=(X=0,Y=0,Z=0),Scale3D=(X=1,Y=1,Z=1))}
 Chest Spine 02 Null Local Transform Offset {type=FTransform, default=(Rotation=(X=0,Y=0,Z=0,W=1),Translation=(X=0,Y=0,Z=0),Scale3D=(X=1,Y=1,Z=1))}
 Connector Parent {type=FRigElementKey, default=(Type=Connector,Name="")}
 Connector Spine Base {type=FRigElementKey, default=(Type=Connector,Name="")}
 Connector Spine Root {type=FRigElementKey, default=(Type=Connector,Name="")}
 Connector Spine Tip {type=FRigElementKey, default=(Type=Connector,Name="")}
 Control Scale {type=double, default=1}
 Controls {type="TArray<FRigElementKey>"}
 FK Joints {type="TArray<FRigElementKey>"}
 Hip Control {type=FRigElementKey, default=(Type=Bone,Name="")}
 Hip Local Transform Offset {type=FTransform, default=(Rotation=(X=0,Y=0,Z=0,W=1),Translation=(X=0,Y=0,Z=0),Scale3D=(X=1,Y=1,Z=1))}
 Hip Spine 02 Null Local Transform Offset {type=FTransform, default=(Rotation=(X=0,Y=0,Z=0,W=1),Translation=(X=0,Y=0,Z=0),Scale3D=(X=1,Y=1,Z=1))}
 Pelvis Control {type=FRigElementKey, default=(Type=Bone,Name="")}
 Pelvis Joint {type=FRigElementKey, default=(Type=Bone,Name="")}
 Pelvis Local Transforms {type=FTransform, default=(Rotation=(X=0,Y=0,Z=0,W=1),Translation=(X=0,Y=0,Z=0),Scale3D=(X=1,Y=1,Z=1))}
 Pelvis Local Transforms BK {type=FTransform, default=(Rotation=(X=0,Y=0,Z=0,W=1),Translation=(X=0,Y=0,Z=0),Scale3D=(X=1,Y=1,Z=1))}
 Set Interacting Items {type="TArray<FRigElementKey>"}
 Spine 01 Control {type=FRigElementKey, default=(Type=None,Name="")}
 Spine 02 {type=FRigElementKey, default=(Type=Bone,Name="")}
 Spine 02 Null {type=FRigElementKey, default=(Type=Bone,Name="")}
 Spine 1 Local Transform BK {type=FTransform, default=(Rotation=(X=0,Y=0,Z=0,W=1),Translation=(X=0,Y=0,Z=0),Scale3D=(X=1,Y=1,Z=1))}
 Spine 2 Local Transform BK {type=FTransform, default=(Rotation=(X=0,Y=0,Z=0,W=1),Translation=(X=0,Y=0,Z=0),Scale3D=(X=1,Y=1,Z=1))}
 Spine BK Proxy Position {type=FTransform, default=(Rotation=(X=0,Y=0,Z=0,W=1),Translation=(X=0,Y=-7,Z=25),Scale3D=(X=0.4,Y=0.4,Z=0.4))}
 Spine Bk Proxy {type=FRigElementKey, default=(Type=Bone,Name="")}
 Spine Fwd Proxy {type=FRigElementKey, default=(Type=Bone,Name="")}
 Spine Fwd Proxy Position {type=FTransform, default=(Rotation=(X=0,Y=0,Z=0,W=1),Translation=(X=0,Y=5,Z=23),Scale3D=(X=0.4,Y=0.4,Z=0.4))}
 Spine Joints {type="TArray<FRigElementKey>"}
 Spine Local Transforms {type="TArray<FTransform>"}
 Spine Mid {type=FRigElementKey, default=(Type=None,Name="")}
 is Interacting {type=bool, default=False}
 sec_controls_visibility {type=FRigElementKey, default=(Type=None,Name="")}
]
