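# Modular Control Rig face module that builds an animator control board for driving facial pose curves.
# Construction resolves the Head connector and spawns a Face Rig control with limited 2D child controls for jaw, tongue, lips, cheeks, eyes, brows and nose.
# Forward solve maps control axes to pose curves, splitting signed values into positive/negative pairs, and derives smile-open correctives from min(smile, jaw open).
# Library functions also convert curves back into control values.
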ControlRigBlueprint /WhisperRemedyAssets/Assets/Modules/WhisperRemedy_Face
graphs [
 ConstructionGraph {nodes [
   RigUnit_PrepareForExecution = &RigUnit_PrepareForExecution()
   RigVMFunction_Sequence = RigVMAggregateNode(ExecuteContext=@RigUnit_PrepareForExecution.ExecutePin)
   RigVMVariableNode(
    Value=/Script/ControlRig.RigUnit_ResolveConnector(Connector=(Type=Connector,Name="Head"), SkipSocket=False) {is=struct RigVMUnitNode}.Result
    ExecuteContext=@RigVMFunction_Sequence.A
   ) {variable=Parent Connector}
   n1 = /Script/ControlRig.RigUnit_HierarchyAddControlTransform(
    InitialValue=(Rotation=(X=0,Y=0,Z=0,W=1),Translation=(X=0,Y=0,Z=0),Scale3D=(X=1,Y=1,Z=1))
    Name=Face Rig
    OffsetSpace=GlobalSpace
    OffsetTransform=(Rotation=(X=0,Y=0,Z=0,W=1),Translation=(X=0,Y=0,Z=-30),Scale3D=(X=1,Y=1,Z=1))
    Parent=(Type=Bone,Name="None")
    Settings="(InitialSpace=LocalSpace,bUsePreferredRotationOrder=False,PreferredRotationOrder=YZX,Limits=(LimitTranslationX=(bMinimum=False,bMaximum=False),LimitTranslationY=(bMinimum=False,bMaximum=False),LimitTranslationZ=(bMinimum=False,bMaximum=False),LimitPitch=(b ...(truncated)"
    ExecutePin=@RigVMFunction_Sequence.B
   ) {
    is=struct RigVMUnitNode
    template="SpawnControl::Execute(in InitialValue,in Settings,in OffsetTransform,in OffsetSpace,in Parent,in Name,out Item)"
   }
   > RigVMVariableNode(Value=@n1.Item.Name) {variable=Face Rig Control}
   SpawnControl = /Script/ControlRig.RigUnit_HierarchyAddControlVector2D(
    InitialValue=(X=0,Y=0)
    Name=Jaw
    OffsetSpace=LocalSpace
    OffsetTransform=(Rotation=(X=-0.7071,Y=-0.7071,Z=0,W=0),Translation=(X=0,Y=0,Z=-14),Scale3D=(X=4,Y=4,Z=4))
    Parent=(Type=Control,Name="Face Rig")
    Settings="(PrimaryAxis=X,Limits=(LimitX=(bMinimum=True,bMaximum=True),LimitY=(bMinimum=True,bMaximum=True),MinValue=(X=-1,Y=-1),MaxValue=(X=1,Y=1),bDrawLimits=True),Shape=(bVisible=True,Name=\"Circle_Solid\",Color=(R=1,G=1,B=0,A=1),Transform=(Rotation=(X=-0.5,Y=-0.5,Z ...(truncated)"
    ExecutePin=@RigVMFunction_Sequence.C
    Parent.Name=RigVMVariableNode() {variable=Face Rig Control}.Value
   ) {
    is=struct RigVMUnitNode
    template="SpawnControl::Execute(in InitialValue,in Settings,in OffsetTransform,in OffsetSpace,in Parent,in Name,out Item)"
   }
   > RigVMVariableNode(Value=@SpawnControl.Item.Name) {variable=Jaw Main Control}
   > n2 = &RigUnit_HierarchyAddControlVector2D(
    InitialValue=(X=0,Y=0)
    Name=Tongue
    OffsetSpace=LocalSpace
    OffsetTransform=(Rotation=(X=-0.7071,Y=-0.7071,Z=0,W=0),Translation=(X=8,Y=0,Z=-13),Scale3D=(X=2,Y=2,Z=2))
    Parent=(Type=Control,Name="face_attach")
    Settings="(PrimaryAxis=X,Limits=(LimitX=(bMinimum=True,bMaximum=True),LimitY=(bMinimum=True,bMaximum=True),MinValue=(X=-1,Y=-1),MaxValue=(X=1,Y=1),bDrawLimits=True),Shape=(bVisible=True,Name=\"Circle_Solid\",Color=(R=1,G=1,B=0,A=1),Transform=(Rotation=(X=-0.5,Y=-0.5,Z ...(truncated)"
    Parent.Name=RigVMVariableNode() {variable=Face Rig Control}.Value
   ) {template="SpawnControl::Execute(in InitialValue,in Settings,in OffsetTransform,in OffsetSpace,in Parent,in Name,out Item)"}
   > RigVMVariableNode(Value=@n2.Item.Name) {variable=Tongue Control}
   > n3 = &RigUnit_HierarchyAddControlVector2D(
    InitialValue=(X=0,Y=0)
    Name=Tongue Fwd
    OffsetSpace=LocalSpace
    OffsetTransform=(Rotation=(X=0.7071,Y=0,Z=0,W=0.7071),Translation=(X=11,Y=0,Z=-13),Scale3D=(X=2,Y=2,Z=2))
    Parent=(Type=Control,Name="face_attach")
    Settings="(PrimaryAxis=X,Limits=(LimitX=(bMinimum=True,bMaximum=True),LimitY=(bMinimum=True,bMaximum=True),MinValue=(X=-1,Y=0),MaxValue=(X=1,Y=0),bDrawLimits=True),Shape=(bVisible=True,Name=\"Circle_Solid\",Color=(R=1,G=1,B=0,A=1),Transform=(Rotation=(X=0,Y=0,Z=0,W=1) ...(truncated)"
    Parent.Name=RigVMVariableNode() {variable=Face Rig Control}.Value
   ) {template="SpawnControl::Execute(in InitialValue,in Settings,in OffsetTransform,in OffsetSpace,in Parent,in Name,out Item)"}
   > n4 = &RigUnit_HierarchyAddControlVector2D(
    InitialValue=(X=0,Y=0)
    Name=Tongue Twist
    OffsetSpace=LocalSpace
    OffsetTransform=(Rotation=(X=0.5,Y=-0.5,Z=0.5,W=0.5),Translation=(X=8,Y=0,Z=-10),Scale3D=(X=2,Y=2,Z=2))
    Parent=(Type=Control,Name="face_attach")
    Settings="(PrimaryAxis=X,Limits=(LimitX=(bMinimum=True,bMaximum=True),LimitY=(bMinimum=True,bMaximum=True),MinValue=(X=-1,Y=0),MaxValue=(X=1,Y=0),bDrawLimits=True),Shape=(bVisible=True,Name=\"Circle_Solid\",Color=(R=1,G=1,B=0,A=1),Transform=(Rotation=(X=0,Y=-1,Z=0,W=0 ...(truncated)"
    Parent.Name=RigVMVariableNode() {variable=Face Rig Control}.Value
   ) {template="SpawnControl::Execute(in InitialValue,in Settings,in OffsetTransform,in OffsetSpace,in Parent,in Name,out Item)"}
   > RigVMVariableNode(Value=@n4.Item.Name) {variable=Tongue Twist Control}
   > n5 = &RigUnit_HierarchyAddControlVector2D(
    InitialValue=(X=0,Y=0)
    Name=Tongue Curl Tip
    OffsetSpace=LocalSpace
    OffsetTransform=(Rotation=(X=0.7071,Y=0,Z=0,W=0.7071),Translation=(X=9,Y=0,Z=-17.5),Scale3D=(X=2,Y=2,Z=2))
    Parent=(Type=Control,Name="face_attach")
    Settings="(PrimaryAxis=X,Limits=(LimitX=(bMinimum=True,bMaximum=True),LimitY=(bMinimum=True,bMaximum=True),MinValue=(X=-1,Y=0),MaxValue=(X=1,Y=0),bDrawLimits=True),Shape=(bVisible=True,Name=\"Circle_Solid\",Color=(R=1,G=1,B=0,A=1),Transform=(Rotation=(X=0,Y=0,Z=0,W=1) ...(truncated)"
    Parent.Name=RigVMVariableNode() {variable=Face Rig Control}.Value
   ) {template="SpawnControl::Execute(in InitialValue,in Settings,in OffsetTransform,in OffsetSpace,in Parent,in Name,out Item)"}
   > RigVMVariableNode(Value=@n5.Item.Name) {variable=Tongue Curl Tip Control}
   > n6 = &RigUnit_HierarchyAddControlVector2D(
    InitialValue=(X=0,Y=0)
    Name=Tongue Curl Tip
    OffsetSpace=LocalSpace
    OffsetTransform=(Rotation=(X=0.7071,Y=0,Z=0,W=0.7071),Translation=(X=7,Y=0,Z=-17.5),Scale3D=(X=2,Y=2,Z=2))
    Parent=(Type=Control,Name="face_attach")
    Settings="(PrimaryAxis=X,Limits=(LimitX=(bMinimum=True,bMaximum=True),LimitY=(bMinimum=True,bMaximum=True),MinValue=(X=-1,Y=0),MaxValue=(X=1,Y=0),bDrawLimits=True),Shape=(bVisible=True,Name=\"Circle_Solid\",Color=(R=1,G=1,B=0,A=1),Transform=(Rotation=(X=0,Y=0,Z=0,W=1) ...(truncated)"
    Parent.Name=RigVMVariableNode() {variable=Face Rig Control}.Value
   ) {template="SpawnControl::Execute(in InitialValue,in Settings,in OffsetTransform,in OffsetSpace,in Parent,in Name,out Item)"}
   > RigVMVariableNode(Value=@n6.Item.Name) {variable=Tongue Curl Control}
   > RigVMVariableNode(Value=@n3.Item.Name) {variable=Tongue Fwd Control}
   > n7 = &RigUnit_HierarchyAddControlVector2D(
    InitialValue=(X=0,Y=0)
    Name=Reverse Jaw
    OffsetSpace=LocalSpace
    OffsetTransform=(Rotation=(X=0,Y=0,Z=0.7071,W=0.7071),Translation=(X=-8,Y=0,Z=-17),Scale3D=(X=6,Y=6,Z=6))
    Parent=(Type=Control,Name="face_attach")
    Settings="(PrimaryAxis=X,Limits=(LimitX=(bMinimum=True,bMaximum=True),LimitY=(bMinimum=True,bMaximum=True),MinValue=(X=0,Y=0),MaxValue=(X=0,Y=1),bDrawLimits=True),Shape=(bVisible=True,Name=\"Circle_Solid\",Color=(R=1,G=1,B=0,A=1),Transform=(Rotation=(X=-0.5,Y=-0.5,Z=0 ...(truncated)"
    Parent.Name=RigVMVariableNode() {variable=Face Rig Control}.Value
   ) {template="SpawnControl::Execute(in InitialValue,in Settings,in OffsetTransform,in OffsetSpace,in Parent,in Name,out Item)"}
   > RigVMVariableNode(Value=@n7.Item.Name) {variable=Reverse Jaw Control}
   > n8 = &RigUnit_HierarchyAddControlVector2D(
    InitialValue=(X=0,Y=0)
    Name=L Corner
    OffsetSpace=LocalSpace
    OffsetTransform=(Rotation=(X=0,Y=0,Z=0.7071,W=0.7071),Translation=(X=7,Y=0,Z=-4),Scale3D=(X=3,Y=3,Z=3))
    Parent=(Type=Control,Name="face_attach")
    Settings="(PrimaryAxis=X,Limits=(LimitX=(bMinimum=True,bMaximum=True),LimitY=(bMinimum=True,bMaximum=True),MinValue=(X=-1,Y=-1),MaxValue=(X=1,Y=1),bDrawLimits=True),Shape=(bVisible=True,Name=\"Circle_Solid\",Color=(R=0,G=1,B=0,A=1),Transform=(Rotation=(X=-0.5,Y=-0.5,Z ...(truncated)"
    Parent.Name=RigVMVariableNode() {variable=Face Rig Control}.Value
   ) {template="SpawnControl::Execute(in InitialValue,in Settings,in OffsetTransform,in OffsetSpace,in Parent,in Name,out Item)"}
   > RigVMVariableNode(Value=@n8.Item.Name) {variable=L Corner Control}
   > n9 = &RigUnit_HierarchyAddControlVector2D(
    InitialValue=(X=0,Y=0)
    Name=R Corner
    OffsetSpace=LocalSpace
    OffsetTransform=(Rotation=(X=0,Y=0,Z=0.7071,W=0.7071),Translation=(X=-7,Y=0,Z=-4),Scale3D=(X=3,Y=-3,Z=3))
    Parent=(Type=Control,Name="face_attach")
    Settings="(PrimaryAxis=X,Limits=(LimitX=(bMinimum=True,bMaximum=True),LimitY=(bMinimum=True,bMaximum=True),MinValue=(X=-1,Y=-1),MaxValue=(X=1,Y=1),bDrawLimits=True),Shape=(bVisible=True,Name=\"Circle_Solid\",Color=(R=1,G=0,B=0,A=1),Transform=(Rotation=(X=-0.5,Y=-0.5,Z ...(truncated)"
    Parent.Name=RigVMVariableNode() {variable=Face Rig Control}.Value
   ) {template="SpawnControl::Execute(in InitialValue,in Settings,in OffsetTransform,in OffsetSpace,in Parent,in Name,out Item)"}
   > RigVMVariableNode(Value=@n9.Item.Name) {variable=R Corner Control}
   > n10 = &RigUnit_HierarchyAddControlVector2D(
    InitialValue=(X=0,Y=0)
    Name=L Upper Lip
    OffsetSpace=LocalSpace
    OffsetTransform=(Rotation=(X=0,Y=0,Z=0.7071,W=0.7071),Translation=(X=2,Y=0,Z=0),Scale3D=(X=2,Y=-2,Z=2))
    Parent=(Type=Control,Name="face_attach")
    Settings="(PrimaryAxis=X,Limits=(LimitX=(bMinimum=True,bMaximum=True),LimitY=(bMinimum=True,bMaximum=True),MinValue=(X=0,Y=-1),MaxValue=(X=0,Y=1),bDrawLimits=True),Shape=(bVisible=True,Name=\"Circle_Solid\",Color=(R=0,G=1,B=0,A=1),Transform=(Rotation=(X=-0.5,Y=-0.5,Z= ...(truncated)"
    Parent.Name=RigVMVariableNode() {variable=Face Rig Control}.Value
   ) {template="SpawnControl::Execute(in InitialValue,in Settings,in OffsetTransform,in OffsetSpace,in Parent,in Name,out Item)"}
   > RigVMVariableNode(Value=@n10.Item.Name) {variable=L Upper Lip Control}
   > n11 = &RigUnit_HierarchyAddControlVector2D(
    InitialValue=(X=0,Y=0)
    Name=R Upper Lip
    OffsetSpace=LocalSpace
    OffsetTransform=(Rotation=(X=0,Y=0,Z=0.7071,W=0.7071),Translation=(X=-2,Y=0,Z=0),Scale3D=(X=2,Y=-2,Z=2))
    Parent=(Type=Control,Name="face_attach")
    Settings="(PrimaryAxis=X,Limits=(LimitX=(bMinimum=True,bMaximum=True),LimitY=(bMinimum=True,bMaximum=True),MinValue=(X=0,Y=-1),MaxValue=(X=0,Y=1),bDrawLimits=True),Shape=(bVisible=True,Name=\"Circle_Solid\",Color=(R=1,G=0,B=0,A=1),Transform=(Rotation=(X=-0.5,Y=-0.5,Z= ...(truncated)"
    Parent.Name=RigVMVariableNode() {variable=Face Rig Control}.Value
   ) {template="SpawnControl::Execute(in InitialValue,in Settings,in OffsetTransform,in OffsetSpace,in Parent,in Name,out Item)"}
   > RigVMVariableNode(Value=@n11.Item.Name) {variable=R Upper Lip Control}
   > n12 = &RigUnit_HierarchyAddControlVector2D(
    InitialValue=(X=0,Y=0)
    Name=L Lower LIp
    OffsetSpace=LocalSpace
    OffsetTransform=(Rotation=(X=0,Y=0,Z=0.7071,W=0.7071),Translation=(X=2,Y=0,Z=-6),Scale3D=(X=2,Y=-2,Z=2))
    Parent=(Type=Control,Name="face_attach")
    Settings="(PrimaryAxis=X,Limits=(LimitX=(bMinimum=True,bMaximum=True),LimitY=(bMinimum=True,bMaximum=True),MinValue=(X=0,Y=-1),MaxValue=(X=0,Y=1),bDrawLimits=True),Shape=(bVisible=True,Name=\"Circle_Solid\",Color=(R=0,G=1,B=0,A=1),Transform=(Rotation=(X=-0.5,Y=-0.5,Z= ...(truncated)"
    Parent.Name=RigVMVariableNode() {variable=Face Rig Control}.Value
   ) {template="SpawnControl::Execute(in InitialValue,in Settings,in OffsetTransform,in OffsetSpace,in Parent,in Name,out Item)"}
   > RigVMVariableNode(Value=@n12.Item.Name) {variable=L Lower Lip Control}
   > n13 = &RigUnit_HierarchyAddControlVector2D(
    InitialValue=(X=0,Y=0)
    Name=R Lower LIp
    OffsetSpace=LocalSpace
    OffsetTransform=(Rotation=(X=0,Y=0,Z=0.7071,W=0.7071),Translation=(X=-2,Y=0,Z=-6),Scale3D=(X=2,Y=-2,Z=2))
    Parent=(Type=Control,Name="face_attach")
    Settings="(PrimaryAxis=X,Limits=(LimitX=(bMinimum=True,bMaximum=True),LimitY=(bMinimum=True,bMaximum=True),MinValue=(X=0,Y=-1),MaxValue=(X=0,Y=1),bDrawLimits=True),Shape=(bVisible=True,Name=\"Circle_Solid\",Color=(R=1,G=0,B=0,A=1),Transform=(Rotation=(X=-0.5,Y=-0.5,Z= ...(truncated)"
    Parent.Name=RigVMVariableNode() {variable=Face Rig Control}.Value
   ) {template="SpawnControl::Execute(in InitialValue,in Settings,in OffsetTransform,in OffsetSpace,in Parent,in Name,out Item)"}
   > RigVMVariableNode(Value=@n13.Item.Name) {variable=R Lower Lip Control}
   > n14 = &RigUnit_HierarchyAddControlVector2D(
    InitialValue=(X=0,Y=0)
    Name=Upper Lip
    OffsetSpace=LocalSpace
    OffsetTransform=(Rotation=(X=0,Y=0,Z=0.7071,W=0.7071),Translation=(X=0,Y=0,Z=0),Scale3D=(X=2,Y=-2,Z=2))
    Parent=(Type=Control,Name="face_attach")
    Settings="(PrimaryAxis=X,Limits=(LimitX=(bMinimum=True,bMaximum=True),LimitY=(bMinimum=True,bMaximum=True),MinValue=(X=0,Y=-1),MaxValue=(X=0,Y=1),bDrawLimits=True),Shape=(bVisible=True,Name=\"Circle_Solid\",Color=(R=1,G=1,B=0,A=1),Transform=(Rotation=(X=-0.5,Y=-0.5,Z= ...(truncated)"
    Parent.Name=RigVMVariableNode() {variable=Face Rig Control}.Value
   ) {template="SpawnControl::Execute(in InitialValue,in Settings,in OffsetTransform,in OffsetSpace,in Parent,in Name,out Item)"}
   > RigVMVariableNode(Value=@n14.Item.Name) {variable=Upper Lip Control}
   > n15 = &RigUnit_HierarchyAddControlVector2D(
    InitialValue=(X=0,Y=0)
    Name=Lower Lip
    OffsetSpace=LocalSpace
    OffsetTransform=(Rotation=(X=0,Y=0,Z=0.7071,W=0.7071),Translation=(X=0,Y=0,Z=-6),Scale3D=(X=2,Y=-2,Z=2))
    Parent=(Type=Control,Name="face_attach")
    Settings="(PrimaryAxis=X,Limits=(LimitX=(bMinimum=True,bMaximum=True),LimitY=(bMinimum=True,bMaximum=True),MinValue=(X=0,Y=-1),MaxValue=(X=0,Y=1),bDrawLimits=True),Shape=(bVisible=True,Name=\"Circle_Solid\",Color=(R=1,G=1,B=0,A=1),Transform=(Rotation=(X=-0.5,Y=-0.5,Z= ...(truncated)"
    Parent.Name=RigVMVariableNode() {variable=Face Rig Control}.Value
   ) {template="SpawnControl::Execute(in InitialValue,in Settings,in OffsetTransform,in OffsetSpace,in Parent,in Name,out Item)"}
   > RigVMVariableNode(Value=@n15.Item.Name) {variable=Lower Lip Control}
   > n16 = &RigUnit_HierarchyAddControlVector2D(
    InitialValue=(X=0,Y=0)
    Name=L Cheek
    OffsetSpace=LocalSpace
    OffsetTransform=(Rotation=(X=0.1228,Y=0.1228,Z=0.6964,W=0.6964),Translation=(X=10,Y=0,Z=4),Scale3D=(X=4,Y=4,Z=4))
    Parent=(Type=Control,Name="face_attach")
    Settings="(PrimaryAxis=X,Limits=(LimitX=(bMinimum=True,bMaximum=True),LimitY=(bMinimum=True,bMaximum=True),MinValue=(X=0,Y=0),MaxValue=(X=0,Y=1),bDrawLimits=True),Shape=(bVisible=True,Name=\"Circle_Solid\",Color=(R=0,G=1,B=0,A=1),Transform=(Rotation=(X=-0.5,Y=-0.5,Z=0 ...(truncated)"
    Parent.Name=RigVMVariableNode() {variable=Face Rig Control}.Value
   ) {template="SpawnControl::Execute(in InitialValue,in Settings,in OffsetTransform,in OffsetSpace,in Parent,in Name,out Item)"}
   > RigVMVariableNode(Value=@n16.Item.Name) {variable=L Cheek Control}
   > n17 = &RigUnit_HierarchyAddControlVector2D(
    InitialValue=(X=0,Y=0)
    Name=R Cheek
    OffsetSpace=LocalSpace
    OffsetTransform=(Rotation=(X=-0.1228,Y=-0.1228,Z=0.6964,W=0.6964),Translation=(X=-10,Y=0,Z=4),Scale3D=(X=4,Y=4,Z=4))
    Parent=(Type=Control,Name="face_attach")
    Settings="(PrimaryAxis=X,Limits=(LimitX=(bMinimum=True,bMaximum=True),LimitY=(bMinimum=True,bMaximum=True),MinValue=(X=0,Y=0),MaxValue=(X=0,Y=1),bDrawLimits=True),Shape=(bVisible=True,Name=\"Circle_Solid\",Color=(R=1,G=0,B=0,A=1),Transform=(Rotation=(X=-0.5,Y=-0.5,Z=0 ...(truncated)"
    Parent.Name=RigVMVariableNode() {variable=Face Rig Control}.Value
   ) {template="SpawnControl::Execute(in InitialValue,in Settings,in OffsetTransform,in OffsetSpace,in Parent,in Name,out Item)"}
   > RigVMVariableNode(Value=@n17.Item.Name) {variable=R Cheek Control}
   > n18 = &RigUnit_HierarchyAddControlVector2D(
    InitialValue=(X=0,Y=0)
    Name=L Blink
    OffsetSpace=LocalSpace
    OffsetTransform=(Rotation=(X=0,Y=0,Z=0,W=1),Translation=(X=8,Y=0,Z=7),Scale3D=(X=3,Y=3,Z=3))
    Parent=(Type=Control,Name="face_attach")
    Settings="(PrimaryAxis=X,Limits=(LimitX=(bMinimum=True,bMaximum=True),LimitY=(bMinimum=True,bMaximum=True),MinValue=(X=0,Y=-1),MaxValue=(X=0,Y=1),bDrawLimits=True),Shape=(bVisible=True,Name=\"Circle_Solid\",Color=(R=0,G=1,B=0,A=1),Transform=(Rotation=(X=-0.7071,Y=0,Z= ...(truncated)"
    Parent.Name=RigVMVariableNode() {variable=Face Rig Control}.Value
   ) {template="SpawnControl::Execute(in InitialValue,in Settings,in OffsetTransform,in OffsetSpace,in Parent,in Name,out Item)"}
   > RigVMVariableNode(Value=@n18.Item.Name) {variable=L Blink Control}
   > n19 = &RigUnit_HierarchyAddControlVector2D(
    InitialValue=(X=0,Y=0)
    Name=R Blink
    OffsetSpace=LocalSpace
    OffsetTransform=(Rotation=(X=0,Y=0,Z=0,W=1),Translation=(X=-8,Y=0,Z=7),Scale3D=(X=3,Y=3,Z=3))
    Parent=(Type=Control,Name="face_attach")
    Settings="(PrimaryAxis=X,Limits=(LimitX=(bMinimum=True,bMaximum=True),LimitY=(bMinimum=True,bMaximum=True),MinValue=(X=0,Y=-1),MaxValue=(X=0,Y=1),bDrawLimits=True),Shape=(bVisible=True,Name=\"Circle_Solid\",Color=(R=1,G=0,B=0,A=1),Transform=(Rotation=(X=-0.7071,Y=0,Z= ...(truncated)"
    Parent.Name=RigVMVariableNode() {variable=Face Rig Control}.Value
   ) {template="SpawnControl::Execute(in InitialValue,in Settings,in OffsetTransform,in OffsetSpace,in Parent,in Name,out Item)"}
   > RigVMVariableNode(Value=@n19.Item.Name) {variable=R Blink Control}
   > n20 = &RigUnit_HierarchyAddControlVector2D(
    InitialValue=(X=0,Y=0)
    Name=Eye
    OffsetSpace=LocalSpace
    OffsetTransform=(Rotation=(X=0,Y=0,Z=0.7071,W=0.7071),Translation=(X=20,Y=0,Z=7),Scale3D=(X=4,Y=4,Z=4))
    Parent=(Type=Control,Name="face_attach")
    Settings="(PrimaryAxis=X,Limits=(LimitX=(bMinimum=True,bMaximum=True),LimitY=(bMinimum=True,bMaximum=True),MinValue=(X=-1,Y=-1),MaxValue=(X=1,Y=1),bDrawLimits=True),Shape=(bVisible=True,Name=\"Circle_Solid\",Color=(R=1,G=1,B=0,A=1),Transform=(Rotation=(X=-0.5,Y=-0.5,Z ...(truncated)"
    Parent.Name=RigVMVariableNode() {variable=Face Rig Control}.Value
   ) {template="SpawnControl::Execute(in InitialValue,in Settings,in OffsetTransform,in OffsetSpace,in Parent,in Name,out Item)"}
   > RigVMVariableNode(Value=@n20.Item.Name) {variable=Eye Control}
   > n21 = &RigUnit_HierarchyAddControlVector2D(
    InitialValue=(X=0,Y=0)
    Name=R Low Lid
    OffsetSpace=LocalSpace
    OffsetTransform=(Rotation=(X=0,Y=0,Z=0.7071,W=0.7071),Translation=(X=-10,Y=0,Z=9),Scale3D=(X=3,Y=3,Z=3))
    Parent=(Type=Control,Name="face_attach")
    Settings="(PrimaryAxis=X,Limits=(LimitX=(bMinimum=True,bMaximum=True),LimitY=(bMinimum=True,bMaximum=True),MinValue=(X=0,Y=0),MaxValue=(X=0,Y=1),bDrawLimits=True),Shape=(bVisible=True,Name=\"Circle_Solid\",Color=(R=1,G=0,B=0,A=1),Transform=(Rotation=(X=-0.5,Y=-0.5,Z=0 ...(truncated)"
    Parent.Name=RigVMVariableNode() {variable=Face Rig Control}.Value
   ) {template="SpawnControl::Execute(in InitialValue,in Settings,in OffsetTransform,in OffsetSpace,in Parent,in Name,out Item)"}
   > RigVMVariableNode(Value=@n21.Item.Name) {variable=R Low Lid}
   > n22 = &RigUnit_HierarchyAddControlVector2D(
    InitialValue=(X=0,Y=0)
    Name=L Low Lid
    OffsetSpace=LocalSpace
    OffsetTransform=(Rotation=(X=0,Y=0,Z=0.7071,W=0.7071),Translation=(X=10,Y=0,Z=9),Scale3D=(X=3,Y=3,Z=3))
    Parent=(Type=Control,Name="face_attach")
    Settings="(PrimaryAxis=X,Limits=(LimitX=(bMinimum=True,bMaximum=True),LimitY=(bMinimum=True,bMaximum=True),MinValue=(X=0,Y=0),MaxValue=(X=0,Y=1),bDrawLimits=True),Shape=(bVisible=True,Name=\"Circle_Solid\",Color=(R=0,G=1,B=0,A=1),Transform=(Rotation=(X=-0.5,Y=-0.5,Z=0 ...(truncated)"
    Parent.Name=RigVMVariableNode() {variable=Face Rig Control}.Value
   ) {template="SpawnControl::Execute(in InitialValue,in Settings,in OffsetTransform,in OffsetSpace,in Parent,in Name,out Item)"}
   > RigVMVariableNode(Value=@n22.Item.Name) {variable=L Low Lid}
   > n23 = &RigUnit_HierarchyAddControlVector2D(
    InitialValue=(X=0,Y=0)
    Name=L Eye
    OffsetSpace=LocalSpace
    OffsetTransform=(Rotation=(X=0,Y=0,Z=0.7071,W=0.7071),Translation=(X=4,Y=0,Z=7),Scale3D=(X=3,Y=3,Z=3))
    Parent=(Type=Control,Name="face_attach")
    Settings="(PrimaryAxis=X,Limits=(LimitX=(bMinimum=True,bMaximum=True),LimitY=(bMinimum=True,bMaximum=True),MinValue=(X=-1,Y=-1),MaxValue=(X=1,Y=1),bDrawLimits=True),Shape=(bVisible=True,Name=\"Circle_Solid\",Color=(R=1,G=1,B=0,A=1),Transform=(Rotation=(X=-0.5,Y=-0.5,Z ...(truncated)"
    Parent.Name=RigVMVariableNode() {variable=Face Rig Control}.Value
   ) {template="SpawnControl::Execute(in InitialValue,in Settings,in OffsetTransform,in OffsetSpace,in Parent,in Name,out Item)"}
   > RigVMVariableNode(Value=@n23.Item.Name) {variable=L Eye Control}
   > n24 = &RigUnit_HierarchyAddControlVector2D(
    InitialValue=(X=0,Y=0)
    Name="R  Eye"
    OffsetSpace=LocalSpace
    OffsetTransform=(Rotation=(X=0,Y=0,Z=0.7071,W=0.7071),Translation=(X=-3,Y=0,Z=7),Scale3D=(X=3,Y=3,Z=3))
    Parent=(Type=Control,Name="face_attach")
    Settings="(PrimaryAxis=X,Limits=(LimitX=(bMinimum=True,bMaximum=True),LimitY=(bMinimum=True,bMaximum=True),MinValue=(X=-1,Y=-1),MaxValue=(X=1,Y=1),bDrawLimits=True),Shape=(bVisible=True,Name=\"Circle_Solid\",Color=(R=1,G=1,B=0,A=1),Transform=(Rotation=(X=-0.5,Y=-0.5,Z ...(truncated)"
    Parent.Name=RigVMVariableNode() {variable=Face Rig Control}.Value
   ) {template="SpawnControl::Execute(in InitialValue,in Settings,in OffsetTransform,in OffsetSpace,in Parent,in Name,out Item)"}
   > RigVMVariableNode(Value=@n24.Item.Name) {variable=R Eye Control}
   > n25 = &RigUnit_HierarchyAddControlVector2D(
    InitialValue=(X=0,Y=0)
    Name=L Brow
    OffsetSpace=LocalSpace
    OffsetTransform=(Rotation=(X=0.2706,Y=0.2706,Z=0.6533,W=0.6533),Translation=(X=7,Y=0,Z=14.5),Scale3D=(X=3,Y=3,Z=3))
    Parent=(Type=Control,Name="face_attach")
    Settings="(PrimaryAxis=X,Limits=(LimitX=(bMinimum=True,bMaximum=True),LimitY=(bMinimum=True,bMaximum=True),MinValue=(X=-1,Y=-1),MaxValue=(X=1,Y=1),bDrawLimits=True),Shape=(bVisible=True,Name=\"Circle_Solid\",Color=(R=0,G=1,B=0,A=1),Transform=(Rotation=(X=-0.5,Y=-0.5,Z ...(truncated)"
    Parent.Name=RigVMVariableNode() {variable=Face Rig Control}.Value
   ) {template="SpawnControl::Execute(in InitialValue,in Settings,in OffsetTransform,in OffsetSpace,in Parent,in Name,out Item)"}
   > RigVMVariableNode(Value=@n25.Item.Name) {variable=L Brow Control}
   > n26 = &RigUnit_HierarchyAddControlVector2D(
    InitialValue=(X=0,Y=0)
    Name=R Brow
    OffsetSpace=LocalSpace
    OffsetTransform=(Rotation=(X=-0.2706,Y=-0.2706,Z=0.6533,W=0.6533),Translation=(X=-7,Y=0,Z=14.5),Scale3D=(X=3,Y=-3,Z=3))
    Parent=(Type=Control,Name="face_attach")
    Settings="(PrimaryAxis=X,Limits=(LimitX=(bMinimum=True,bMaximum=True),LimitY=(bMinimum=True,bMaximum=True),MinValue=(X=-1,Y=-1),MaxValue=(X=1,Y=1),bDrawLimits=True),Shape=(bVisible=True,Name=\"Circle_Solid\",Color=(R=1,G=0,B=0,A=1),Transform=(Rotation=(X=-0.5,Y=-0.5,Z ...(truncated)"
    Parent.Name=RigVMVariableNode() {variable=Face Rig Control}.Value
   ) {template="SpawnControl::Execute(in InitialValue,in Settings,in OffsetTransform,in OffsetSpace,in Parent,in Name,out Item)"}
   > RigVMVariableNode(Value=@n26.Item.Name) {variable=R Brow Control}
   > n27 = &RigUnit_HierarchyAddControlVector2D(
    InitialValue=(X=0,Y=0)
    Name=Squeeze
    OffsetSpace=LocalSpace
    OffsetTransform=(Rotation=(X=0.2706,Y=0.2706,Z=0.6533,W=0.6533),Translation=(X=0,Y=0,Z=15),Scale3D=(X=3,Y=3,Z=3))
    Parent=(Type=Control,Name="face_attach")
    Settings="(PrimaryAxis=X,Limits=(LimitX=(bMinimum=True,bMaximum=True),LimitY=(bMinimum=True,bMaximum=True),MinValue=(X=0,Y=0),MaxValue=(X=1,Y=1),bDrawLimits=True),Shape=(bVisible=True,Name=\"Circle_Solid\",Color=(R=1,G=1,B=0,A=1),Transform=(Rotation=(X=-0.5,Y=-0.5,Z=0 ...(truncated)"
    Parent.Name=RigVMVariableNode() {variable=Face Rig Control}.Value
   ) {template="SpawnControl::Execute(in InitialValue,in Settings,in OffsetTransform,in OffsetSpace,in Parent,in Name,out Item)"}
   > RigVMVariableNode(Value=@n27.Item.Name) {variable=Squeeze Control}
   > n28 = &RigUnit_HierarchyAddControlVector2D(
    InitialValue=(X=0,Y=0)
    Name=Nose
    OffsetSpace=LocalSpace
    OffsetTransform=(Rotation=(X=0,Y=0,Z=0.7071,W=0.7071),Translation=(X=0,Y=0,Z=12),Scale3D=(X=2,Y=2,Z=2))
    Parent=(Type=Control,Name="face_attach")
    Settings="(PrimaryAxis=X,Limits=(LimitX=(bMinimum=True,bMaximum=True),LimitY=(bMinimum=True,bMaximum=True),MinValue=(X=0,Y=0),MaxValue=(X=0,Y=1),bDrawLimits=True),Shape=(bVisible=True,Name=\"Circle_Solid\",Color=(R=1,G=0,B=0,A=1),Transform=(Rotation=(X=-0.5,Y=-0.5,Z=0 ...(truncated)"
    Parent.Name=RigVMVariableNode() {variable=Face Rig Control}.Value
   ) {template="SpawnControl::Execute(in InitialValue,in Settings,in OffsetTransform,in OffsetSpace,in Parent,in Name,out Item)"}
   > RigVMVariableNode(Value=@n28.Item.Name) {variable=Nose Control}
  ]}
 New Function_ContainedGraph {nodes [
   n1 = RigVMFunctionEntryNode()
   > n2 = RigVMVariableNode(Value=@n1.Control) {variable=BK Control}
   n3 = &RigVMFunction_Sequence(ExecuteContext=@n2.ExecuteContext)
   RigVMFunction_ControlFlowBranch = &RigVMFunction_ControlFlowBranch(Condition=@n1.Is Y, ExecuteContext=@n3.A)
   n4 = /Script/ControlRig.RigUnit_GetCurveValue(Curve=@n1.Pos Curve) {is=struct RigVMUnitNode}
   n5 = &RigVMFunction_ControlFlowBranch(
    Condition=/Script/RigVM.RigVMFunction_MathDoubleGreater(B=0, A=@n4.Value) {is=struct RigVMUnitNode, template="Greater::Execute(in A,in B,out Result)"}.Result
    ExecuteContext=@RigVMFunction_ControlFlowBranch.False
   )
   n6 = RigVMVariableNode() {variable=BK Control}
   /Script/ControlRig.RigUnit_SetControlVector2D(
    Vector=(X=0,Y=0)
    Weight=-1
    Vector.X=&RigUnit_GetCurveValue(Curve=@n1.Neg Curve).Value
    ExecutePin=@n5.False
    Control=@n6.Value
   ) {is=struct RigVMUnitNode, template="SetControlValue::Execute(in Control,in Weight,io Vector)"}
   &RigUnit_SetControlVector2D(Vector=(X=0,Y=0), Weight=1, Vector.X=@n4.Value, ExecutePin=@n5.True, Control=@n6.Value) {template="SetControlValue::Execute(in Control,in Weight,io Vector)"}
   n7 = &RigUnit_GetCurveValue(Curve=@n1.Pos Curve)
   n8 = &RigVMFunction_ControlFlowBranch(
    Condition=&RigVMFunction_MathDoubleGreater(B=0, A=@n7.Value) {template="Greater::Execute(in A,in B,out Result)"}.Result
    ExecuteContext=@RigVMFunction_ControlFlowBranch.True
   )
   n9 = RigVMVariableNode() {variable=BK Control}
   &RigUnit_SetControlVector2D(
    Vector=(X=0,Y=0)
    Weight=-1
    Vector.Y=&RigUnit_GetCurveValue(Curve=@n1.Neg Curve).Value
    ExecutePin=@n8.False
    Control=@n9.Value
   ) {template="SetControlValue::Execute(in Control,in Weight,io Vector)"}
   &RigUnit_SetControlVector2D(Vector=(X=0,Y=0), Weight=1, Vector.Y=@n7.Value, ExecutePin=@n8.True, Control=@n9.Value) {template="SetControlValue::Execute(in Control,in Weight,io Vector)"}
   RigVMFunctionReturnNode(ExecuteContext=@n3.B)
  ]}
 New Function_ContainedGraph {nodes [
   n1 = RigVMFunctionEntryNode()
   RigVMFunction_Sequence = &RigVMFunction_Sequence(ExecuteContext=@n1.ExecuteContext)
   n2 = /Script/ControlRig.RigUnit_GetControlVector2D(Control=@n1.Control.Name) {is=struct RigVMUnitNode, template="GetControlValue::Execute(in Control,out Vector,out Minimum,out Maximum)"}
   Remap = &RigVMFunction_MathDoubleRemap(SourceMaximum=1, SourceMinimum=0, TargetMaximum=1, TargetMinimum=0, Value=@n2.Vector.X)
   /Script/ControlRig.RigUnit_SetCurveValue(Value=@Remap.Result, ExecutePin=@RigVMFunction_Sequence.A, Curve=RigVMVariableNode() {variable=X_Pos_Curve}.Value.Name) {is=struct RigVMUnitNode}
   > &RigUnit_SetCurveValue(
    Value=&RigVMFunction_MathDoubleRemap(SourceMaximum=-1, SourceMinimum=0, TargetMaximum=1, TargetMinimum=0, Value=@n2.Vector.X).Result
    Curve=RigVMVariableNode() {variable=X_Neg_Curve}.Value.Name
   )
   > &RigUnit_SetCurveValue(
    Value=&RigVMFunction_MathDoubleRemap(SourceMaximum=1, SourceMinimum=0, TargetMaximum=1, TargetMinimum=0, Value=@n2.Vector.Y).Result
    Curve=RigVMVariableNode() {variable=Y_Pos_Curve}.Value.Name
   )
   > &RigUnit_SetCurveValue(
    Value=&RigVMFunction_MathDoubleRemap(SourceMaximum=-1, SourceMinimum=0, TargetMaximum=1, TargetMinimum=0, Value=@n2.Vector.Y).Result
    Curve=RigVMVariableNode() {variable=Y_Neg_Curve}.Value.Name
   )
   RigVMFunctionReturnNode(ExecuteContext=@RigVMFunction_Sequence.B)
  ]}
 New Function_ContainedGraph {nodes [
   n1 = RigVMFunctionEntryNode()
   RigVMFunction_Sequence = &RigVMFunction_Sequence(ExecuteContext=@n1.ExecuteContext)
   Subtract = &RigVMFunction_MathDoubleSub(
    A=&RigUnit_GetCurveValue(Curve=@n1.X_Pos_Curve.Name).Value
    B=&RigUnit_GetCurveValue(Curve=@n1.X_Neg_Curve.Name).Value
   )
   &RigUnit_SetControlVector2D(
    Vector=(X=0,Y=0)
    Weight=1
    Control=@n1.Control.Name
    ExecutePin=@RigVMFunction_Sequence.A
    Vector.X=@Subtract.Result
    Vector.Y=&RigVMFunction_MathDoubleSub(
     A=&RigUnit_GetCurveValue(Curve=@n1.Y_Pos_Curve.Name).Value
     B=&RigUnit_GetCurveValue(Curve=@n1.Y_Neg_Curve.Name).Value
    ).Result
   ) {template="SetControlValue::Execute(in Control,in Weight,io Vector)"}
   RigVMFunctionReturnNode(ExecuteContext=@RigVMFunction_Sequence.B)
   n2 = RigVMRerouteNode(Value=())
   n3 = &RigUnit_SetControlVector2D(
    Vector=(X=0,Y=0)
    Weight=1
    Vector.X=&RigUnit_GetCurveValue(Curve=RigVMVariableNode() {variable=X_Pos_Curve}.Value.Name).Value
    ExecutePin=@n2.Value
    Control=RigVMVariableNode() {variable=Control}.Value.Name
   ) {template="SetControlValue::Execute(in Control,in Weight,io Vector)"}
   n4 = &RigUnit_GetCurveValue(Curve=RigVMVariableNode() {variable=X_Neg_Curve}.Value.Name)
   Less = /Script/RigVM.RigVMFunction_MathDoubleLess(B=0, A=@n4.Value) {is=struct RigVMUnitNode, template="Less::Execute(in A,in B,out Result)"}
   RigVMFunction_ControlFlowBranch = &RigVMFunction_ControlFlowBranch(Condition=@Less.Result, ExecuteContext=@n3.ExecutePin)
   &RigUnit_SetControlVector2D(
    Vector=(X=0,Y=0)
    Weight=1
    Vector.X=@n4.Value
    ExecutePin=@RigVMFunction_ControlFlowBranch.True
    Control=RigVMVariableNode() {variable=Control}.Value.Name
   ) {template="SetControlValue::Execute(in Control,in Weight,io Vector)"}
   n5 = RigVMRerouteNode(Value=())
   n6 = &RigUnit_SetControlVector2D(
    Vector=(X=0,Y=0)
    Weight=1
    Vector.Y=&RigUnit_GetCurveValue(Curve=RigVMVariableNode() {variable=Y_Pos_Curve}.Value.Name).Value
    ExecutePin=@n5.Value
    Control=RigVMVariableNode() {variable=Control}.Value.Name
   ) {template="SetControlValue::Execute(in Control,in Weight,io Vector)"}
   n7 = &RigUnit_GetCurveValue(Curve=RigVMVariableNode() {variable=Y_Neg_Curve}.Value.Name)
   n8 = &RigVMFunction_ControlFlowBranch(
    Condition=&RigVMFunction_MathDoubleLess(B=0, A=@n7.Value) {template="Less::Execute(in A,in B,out Result)"}.Result
    ExecuteContext=@n6.ExecutePin
   )
   &RigUnit_SetControlVector2D(Vector=(X=0,Y=0), Weight=1, Vector.Y=@n7.Value, ExecutePin=@n8.True, Control=RigVMVariableNode() {variable=Control}.Value.Name) {template="SetControlValue::Execute(in Control,in Weight,io Vector)"}
  ]}
 RigVMFunctionLibrary {nodes [
   Backwards = RigVMCollapseNode(ExecuteContext=())
   CrvToVectorCtrlFNC = RigVMCollapseNode(
    Control=(Type=None,Name="None")
    ExecuteContext=()
    X_Neg_Curve=(Type=None,Name="None")
    X_Pos_Curve=(Type=None,Name="None")
    Y_Neg_Curve=(Type=None,Name="None")
    Y_Pos_Curve=(Type=None,Name="None")
   )
   VectorCtrlToCrvFNC = RigVMCollapseNode(
    Control=(Type=Control,Name="None")
    ExecuteContext=()
    X_Neg_Curve=(Type=Curve,Name="None")
    X_Pos_Curve=(Type=Curve,Name="None")
    Y_Neg_Curve=(Type=Curve,Name="None")
    Y_Pos_Curve=(Type=Curve,Name="None")
   )
  ]}
 RigVMFunction_Sequence_1_ContainedGraph {nodes [
   n1 = RigVMFunctionEntryNode()
   RigVMFunction_Sequence = &RigVMFunction_Sequence(ExecuteContext=@n1.ExecuteContext)
   n2 = &RigVMFunction_Sequence(ExecuteContext=@RigVMFunction_Sequence.B)
   RigVMFunctionReturnNode(A=@RigVMFunction_Sequence.A, B=@n2.A, C=@n2.B)
  ]}
 RigVMFunction_Sequence_2_ContainedGraph {nodes [
   n1 = RigVMFunctionEntryNode()
   n2 = &RigVMFunction_Sequence(ExecuteContext=@n1.ExecuteContext)
   n3 = &RigVMFunction_Sequence(ExecuteContext=@n2.B)
   n4 = &RigVMFunction_Sequence(ExecuteContext=@n3.B)
   n5 = &RigVMFunction_Sequence(ExecuteContext=@n4.B)
   RigVMFunctionReturnNode(A=@n2.A, B=@n3.A, C=@n4.A, D=@n5.A, E=@n5.B)
  ]}
 RigVMFunction_Sequence_2_ContainedGraph {nodes [
   n1 = RigVMFunctionEntryNode()
   n2 = &RigVMFunction_Sequence(ExecuteContext=@n1.ExecuteContext)
   n3 = &RigVMFunction_Sequence(ExecuteContext=@n2.B)
   n4 = &RigVMFunction_Sequence(ExecuteContext=@n3.B)
   n5 = &RigVMFunction_Sequence(ExecuteContext=@n4.B)
   RigVMFunctionReturnNode(A=@n2.A, B=@n3.A, C=@n4.A, D=@n5.A, E=@n5.B)
  ]}
 RigVMModel {nodes [
   RigUnit_BeginExecution = &RigUnit_BeginExecution()
   RigVMFunction_Sequence_1 = RigVMAggregateNode(ExecuteContext=@RigUnit_BeginExecution.ExecutePin)
   VectorCtrlToCrvFNC_11 = RigVMFunctionReferenceNode(
    Control=(Type=Control,Name="None")
    X_Neg_Curve=(Type=Curve,Name="L_brow_In_Dn_pose")
    X_Pos_Curve=(Type=Curve,Name="L_brow_In_Up_pose")
    Y_Neg_Curve=(Type=Curve,Name="L_brow_Ot_Dn_pose")
    Y_Pos_Curve=(Type=Curve,Name="L_brow_Ot_Up_pose")
    ExecuteContext=@RigVMFunction_Sequence_1.D
    Control.Name=RigVMVariableNode() {variable=L Brow Control}.Value
   )
   > VectorCtrlToCrvFNC_12 = RigVMFunctionReferenceNode(
    Control=(Type=Control,Name="None")
    X_Neg_Curve=(Type=Curve,Name="R_brow_In_Dn_pose")
    X_Pos_Curve=(Type=Curve,Name="R_brow_In_Up_pose")
    Y_Neg_Curve=(Type=Curve,Name="R_brow_Ot_Dn_pose")
    Y_Pos_Curve=(Type=Curve,Name="R_brow_Ot_Up_pose")
    Control.Name=RigVMVariableNode() {variable=R Brow Control}.Value
   )
   > VectorCtrlToCrvFNC_16 = RigVMFunctionReferenceNode(
    Control=(Type=Control,Name="None")
    X_Neg_Curve=(Type=Curve,Name="None")
    X_Pos_Curve=(Type=Curve,Name="R_brow_squeeze_in_pose")
    Y_Neg_Curve=(Type=Curve,Name="None")
    Y_Pos_Curve=(Type=Curve,Name="L_brow_squeeze_in_pose")
    Control.Name=RigVMVariableNode() {variable=Squeeze Control}.Value
   )
   > VectorCtrlToCrvFNC_18 = RigVMFunctionReferenceNode(
    Control=(Type=Control,Name="None")
    X_Neg_Curve=(Type=Curve,Name="L_eye_look_left_pose")
    X_Pos_Curve=(Type=Curve,Name="L_eye_look_right_pose")
    Y_Neg_Curve=(Type=Curve,Name="L_eye_look_down_pose")
    Y_Pos_Curve=(Type=Curve,Name="L_eye_look_up_pose")
    Control.Name=RigVMVariableNode() {variable=L Eye Control}.Value
   )
   > VectorCtrlToCrvFNC_19 = RigVMFunctionReferenceNode(
    Control=(Type=Control,Name="None")
    X_Neg_Curve=(Type=Curve,Name="R_eye_look_left_pose")
    X_Pos_Curve=(Type=Curve,Name="R_eye_look_right_pose")
    Y_Neg_Curve=(Type=Curve,Name="R_eye_look_down_pose")
    Y_Pos_Curve=(Type=Curve,Name="R_eye_look_up_pose")
    Control.Name=RigVMVariableNode() {variable=R Eye Control}.Value
   )
   > &RigUnit_SetCurveValue(
    Curve=L_lower_lid_up_pose
    Value=&RigUnit_GetControlVector2D(Control=RigVMVariableNode() {variable=L Low Lid}.Value) {template="GetControlValue::Execute(in Control,out Vector,out Minimum,out Maximum)"}.Vector.Y
   )
   > &RigUnit_SetCurveValue(
    Curve=R_lower_lid_up_pose
    Value=&RigUnit_GetControlVector2D(Control=RigVMVariableNode() {variable=R Low Lid}.Value) {template="GetControlValue::Execute(in Control,out Vector,out Minimum,out Maximum)"}.Vector.Y
   )
   > VectorCtrlToCrvFNC_14 = RigVMFunctionReferenceNode(
    Control=(Type=Control,Name="None")
    X_Neg_Curve=(Type=Curve,Name="None")
    X_Pos_Curve=(Type=Curve,Name="None")
    Y_Neg_Curve=(Type=Curve,Name="L_blink_pose")
    Y_Pos_Curve=(Type=Curve,Name="L_wide_pose")
    Control.Name=RigVMVariableNode() {variable=L Blink Control}.Value
   )
   > VectorCtrlToCrvFNC_15 = RigVMFunctionReferenceNode(
    Control=(Type=Control,Name="None")
    X_Neg_Curve=(Type=Curve,Name="None")
    X_Pos_Curve=(Type=Curve,Name="None")
    Y_Neg_Curve=(Type=Curve,Name="R_blink_pose")
    Y_Pos_Curve=(Type=Curve,Name="R_wide_pose")
    Control.Name=RigVMVariableNode() {variable=R Blink Control}.Value
   )
   Minimum = /Script/RigVM.RigVMFunction_MathDoubleMin(A=&RigUnit_GetCurveValue(Curve=L_smile_pose).Value, B=&RigUnit_GetCurveValue(Curve=jaw_open_pose).Value) {is=struct RigVMUnitNode, template="Minimum::Execute(in A,in B,out Result)"}
   &RigUnit_SetCurveValue(Curve=L_smile_open_correction_pose, Value=@Minimum.Result, ExecutePin=@RigVMFunction_Sequence_1.E)
   > &RigUnit_SetCurveValue(
    Curve=R_smile_open_correction_pose
    Value=&RigVMFunction_MathDoubleMin(A=&RigUnit_GetCurveValue(Curve=R_smile_pose).Value, B=&RigUnit_GetCurveValue(Curve=jaw_open_pose).Value) {template="Minimum::Execute(in A,in B,out Result)"}.Result
   )
   &RigUnit_SetCurveValue(
    Curve=reverse_jaw_pose
    Value=&RigUnit_GetControlVector2D(Control=RigVMVariableNode() {variable=Reverse Jaw Control}.Value) {template="GetControlValue::Execute(in Control,out Vector,out Minimum,out Maximum)"}.Vector.Y
    ExecutePin=@RigVMFunction_Sequence_1.C
   )
   > VectorCtrlToCrvFNC_7 = RigVMFunctionReferenceNode(
    Control=(Type=Control,Name="None")
    X_Neg_Curve=(Type=Curve,Name="tongue_right_pose")
    X_Pos_Curve=(Type=Curve,Name="tongue_left_pose")
    Y_Neg_Curve=(Type=Curve,Name="tongue_up_pose")
    Y_Pos_Curve=(Type=Curve,Name="tongue_down_pose")
    Control.Name=RigVMVariableNode() {variable=Tongue Control}.Value
   )
   > VectorCtrlToCrvFNC_17 = RigVMFunctionReferenceNode(
    Control=(Type=Control,Name="None")
    X_Neg_Curve=(Type=Curve,Name="tongue_back_pose")
    X_Pos_Curve=(Type=Curve,Name="tongue_forward_pose")
    Y_Neg_Curve=(Type=Curve,Name="None")
    Y_Pos_Curve=(Type=Curve,Name="None")
    Control.Name=RigVMVariableNode() {variable=Tongue Fwd Control}.Value
   )
   > VectorCtrlToCrvFNC_20 = RigVMFunctionReferenceNode(
    Control=(Type=Control,Name="None")
    X_Neg_Curve=(Type=Curve,Name="R_tongue_twist_pose")
    X_Pos_Curve=(Type=Curve,Name="L_tongue_twist_pose")
    Y_Neg_Curve=(Type=Curve,Name="None")
    Y_Pos_Curve=(Type=Curve,Name="None")
    Control.Name=RigVMVariableNode() {variable=Tongue Twist Control}.Value
   )
   > VectorCtrlToCrvFNC_21 = RigVMFunctionReferenceNode(
    Control=(Type=Control,Name="None")
    X_Neg_Curve=(Type=Curve,Name="tongue_curl_tip_down_pose")
    X_Pos_Curve=(Type=Curve,Name="tongue_curl_tip_up_pose")
    Y_Neg_Curve=(Type=Curve,Name="None")
    Y_Pos_Curve=(Type=Curve,Name="None")
    Control.Name=RigVMVariableNode() {variable=Tongue Curl Tip Control}.Value
   )
   > VectorCtrlToCrvFNC_22 = RigVMFunctionReferenceNode(
    Control=(Type=Control,Name="None")
    X_Neg_Curve=(Type=Curve,Name="tongue_curl_mid_down_pose")
    X_Pos_Curve=(Type=Curve,Name="tongue_curl_mid_up_pose")
    Y_Neg_Curve=(Type=Curve,Name="None")
    Y_Pos_Curve=(Type=Curve,Name="None")
    Control.Name=RigVMVariableNode() {variable=Tongue Curl Control}.Value
   )
   > &RigUnit_SetCurveValue(
    Curve=L_cheek_up_pose
    Value=&RigUnit_GetControlVector2D(Control=RigVMVariableNode() {variable=L Cheek Control}.Value) {template="GetControlValue::Execute(in Control,out Vector,out Minimum,out Maximum)"}.Vector.Y
   )
   > &RigUnit_SetCurveValue(
    Curve=R_cheek_up_pose
    Value=&RigUnit_GetControlVector2D(Control=RigVMVariableNode() {variable=R Cheek Control}.Value) {template="GetControlValue::Execute(in Control,out Vector,out Minimum,out Maximum)"}.Vector.Y
   )
   > &RigUnit_SetCurveValue(
    Curve=C_nose_wrinkler_pose
    Value=&RigUnit_GetControlVector2D(Control=RigVMVariableNode() {variable=Nose Control}.Value) {template="GetControlValue::Execute(in Control,out Vector,out Minimum,out Maximum)"}.Vector.Y
   )
   VectorCtrlToCrvFNC_3 = RigVMFunctionReferenceNode(
    Control=(Type=Control,Name="None")
    X_Neg_Curve=(Type=Curve,Name="None")
    X_Pos_Curve=(Type=Curve,Name="None")
    Y_Neg_Curve=(Type=Curve,Name="L_upper_lip_lower_pose")
    Y_Pos_Curve=(Type=Curve,Name="L_upper_lip_raiser_pose")
    Control.Name=RigVMVariableNode() {variable=L Upper Lip Control}.Value
    ExecuteContext=@RigVMFunction_Sequence_1.B
   )
   > VectorCtrlToCrvFNC_4 = RigVMFunctionReferenceNode(
    Control=(Type=Control,Name="None")
    X_Neg_Curve=(Type=None,Name="None")
    X_Pos_Curve=(Type=None,Name="None")
    Y_Neg_Curve=(Type=Curve,Name="R_upper_lip_lower_pose")
    Y_Pos_Curve=(Type=Curve,Name="R_upper_lip_raiser_pose")
    Control.Name=RigVMVariableNode() {variable=R Upper Lip Control}.Value
   )
   > VectorCtrlToCrvFNC_5 = RigVMFunctionReferenceNode(
    Control=(Type=Control,Name="None")
    X_Neg_Curve=(Type=None,Name="None")
    X_Pos_Curve=(Type=None,Name="None")
    Y_Neg_Curve=(Type=Curve,Name="L_lower_lip_down_pose")
    Y_Pos_Curve=(Type=Curve,Name="L_lower_lip_up_pose")
    Control.Name=RigVMVariableNode() {variable=L Lower Lip Control}.Value
   )
   > VectorCtrlToCrvFNC_6 = RigVMFunctionReferenceNode(
    Control=(Type=Control,Name="None")
    X_Neg_Curve=(Type=None,Name="None")
    X_Pos_Curve=(Type=None,Name="None")
    Y_Neg_Curve=(Type=Curve,Name="R_lower_lip_down_pose")
    Y_Pos_Curve=(Type=Curve,Name="R_lower_lip_up_pose")
    Control.Name=RigVMVariableNode() {variable=R Lower Lip Control}.Value
   )
   > VectorCtrlToCrvFNC_9 = RigVMFunctionReferenceNode(
    Control=(Type=Control,Name="None")
    X_Neg_Curve=(Type=Curve,Name="None")
    X_Pos_Curve=(Type=Curve,Name="None")
    Y_Neg_Curve=(Type=Curve,Name="upper_lip_dn_pose")
    Y_Pos_Curve=(Type=Curve,Name="upper_lip_up_pose")
    Control.Name=RigVMVariableNode() {variable=Upper Lip Control}.Value
   )
   > VectorCtrlToCrvFNC_10 = RigVMFunctionReferenceNode(
    Control=(Type=Control,Name="None")
    X_Neg_Curve=(Type=Curve,Name="None")
    X_Pos_Curve=(Type=Curve,Name="None")
    Y_Neg_Curve=(Type=Curve,Name="lower_lip_dn_pose")
    Y_Pos_Curve=(Type=Curve,Name="lower_lip_up_pose")
    Control.Name=RigVMVariableNode() {variable=Lower Lip Control}.Value
   )
   VectorCtrlToCrvFNC = RigVMFunctionReferenceNode(
    Control=(Type=Control,Name="None")
    X_Neg_Curve=(Type=Curve,Name="jaw_right_pose")
    X_Pos_Curve=(Type=Curve,Name="jaw_left_pose")
    Y_Neg_Curve=(Type=Curve,Name="jaw_up_pose")
    Y_Pos_Curve=(Type=Curve,Name="jaw_open_pose")
    Control.Name=RigVMVariableNode() {variable=Jaw Main Control}.Value
    ExecuteContext=@RigVMFunction_Sequence_1.A
   )
   > VectorCtrlToCrvFNC_1 = RigVMFunctionReferenceNode(
    Control=(Type=Control,Name="None")
    X_Neg_Curve=(Type=Curve,Name="L_lip_corner_wide_pose")
    X_Pos_Curve=(Type=Curve,Name="L_lip_corner_narrow_pose")
    Y_Neg_Curve=(Type=Curve,Name="L_frown_pose")
    Y_Pos_Curve=(Type=Curve,Name="L_smile_pose")
    Control.Name=RigVMVariableNode() {variable=L Corner Control}.Value
   )
   > VectorCtrlToCrvFNC_2 = RigVMFunctionReferenceNode(
    Control=(Type=Control,Name="None")
    X_Neg_Curve=(Type=Curve,Name="R_lip_corner_wide_pose")
    X_Pos_Curve=(Type=Curve,Name="R_lip_corner_narrow_pose")
    Y_Neg_Curve=(Type=Curve,Name="R_frown_pose")
    Y_Pos_Curve=(Type=Curve,Name="R_smile_pose")
    Control.Name=RigVMVariableNode() {variable=R Corner Control}.Value
   )
   VectorCtrlToCrvFNC_13 = RigVMFunctionReferenceNode(
    Control=(Type=Control,Name="None")
    ExecuteContext=()
    X_Neg_Curve=(Type=Curve,Name="look_left_pose")
    X_Pos_Curve=(Type=Curve,Name="look_right_pose")
    Y_Neg_Curve=(Type=Curve,Name="look_down_pose")
    Y_Pos_Curve=(Type=Curve,Name="look_up_pose")
    Control.Name=RigVMVariableNode() {variable=Eye Control}.Value
   )
   VectorCtrlToCrvFNC_8 = RigVMFunctionReferenceNode(
    Control=(Type=Control,Name="None")
    ExecuteContext=()
    X_Neg_Curve=(Type=Curve,Name="None")
    X_Pos_Curve=(Type=Curve,Name="None")
    Y_Neg_Curve=(Type=Curve,Name="None")
    Y_Pos_Curve=(Type=Curve,Name="None")
   )
  ]}
 RigVMModel Rig {nodes [
   n1 = /Script/ControlRig.RigUnit_InverseExecution() {is=struct RigVMUnitNode}
   RigVMFunction_Sequence_1 = RigVMAggregateNode(ExecuteContext=@n1.ExecutePin)
   CrvToVectorCtrlFNC = RigVMFunctionReferenceNode(
    Control=(Type=None,Name="None")
    X_Neg_Curve=(Type=Curve,Name="jaw_right_pose")
    X_Pos_Curve=(Type=Curve,Name="jaw_left_pose")
    Y_Neg_Curve=(Type=Curve,Name="jaw_up_pose")
    Y_Pos_Curve=(Type=Curve,Name="jaw_open_pose")
    ExecuteContext=@RigVMFunction_Sequence_1.A
    Control.Name=RigVMVariableNode() {variable=Jaw Main Control}.Value
   )
   > CrvToVectorCtrlFNC_1 = RigVMFunctionReferenceNode(
    Control=(Type=None,Name="None")
    X_Neg_Curve=(Type=Curve,Name="L_lip_corner_wide_pose")
    X_Pos_Curve=(Type=Curve,Name="L_lip_corner_narrow_pose")
    Y_Neg_Curve=(Type=Curve,Name="L_frown_pose")
    Y_Pos_Curve=(Type=Curve,Name="L_smile_pose")
    Control.Name=RigVMVariableNode() {variable=L Corner Control}.Value
   )
   > CrvToVectorCtrlFNC_2 = RigVMFunctionReferenceNode(
    Control=(Type=None,Name="None")
    X_Neg_Curve=(Type=Curve,Name="R_lip_corner_wide_pose")
    X_Pos_Curve=(Type=Curve,Name="R_lip_corner_narrow_pose")
    Y_Neg_Curve=(Type=Curve,Name="R_frown_pose")
    Y_Pos_Curve=(Type=Curve,Name="R_smile_pose")
    Control.Name=RigVMVariableNode() {variable=R Corner Control}.Value
   )
   CrvToVectorCtrlFNC_3 = RigVMFunctionReferenceNode(
    Control=(Type=None,Name="None")
    X_Neg_Curve=(Type=Curve,Name="None")
    X_Pos_Curve=(Type=Curve,Name="None")
    Y_Neg_Curve=(Type=Curve,Name="L_upper_lip_lower_pose")
    Y_Pos_Curve=(Type=Curve,Name="L_upper_lip_raiser_pose")
    Control.Name=RigVMVariableNode() {variable=L Upper Lip Control}.Value
    ExecuteContext=@RigVMFunction_Sequence_1.B
   )
   > CrvToVectorCtrlFNC_4 = RigVMFunctionReferenceNode(
    Control=(Type=None,Name="None")
    X_Neg_Curve=(Type=Curve,Name="None")
    X_Pos_Curve=(Type=Curve,Name="None")
    Y_Neg_Curve=(Type=Curve,Name="R_upper_lip_lower_pose")
    Y_Pos_Curve=(Type=Curve,Name="R_upper_lip_raiser_pose")
    Control.Name=RigVMVariableNode() {variable=R Upper Lip Control}.Value
   )
   > CrvToVectorCtrlFNC_5 = RigVMFunctionReferenceNode(
    Control=(Type=None,Name="None")
    X_Neg_Curve=(Type=Curve,Name="None")
    X_Pos_Curve=(Type=Curve,Name="None")
    Y_Neg_Curve=(Type=Curve,Name="L_lower_lip_down_pose")
    Y_Pos_Curve=(Type=Curve,Name="L_lower_lip_up_pose")
    Control.Name=RigVMVariableNode() {variable=L Lower Lip Control}.Value
   )
   > CrvToVectorCtrlFNC_6 = RigVMFunctionReferenceNode(
    Control=(Type=None,Name="None")
    X_Neg_Curve=(Type=Curve,Name="None")
    X_Pos_Curve=(Type=Curve,Name="None")
    Y_Neg_Curve=(Type=Curve,Name="R_lower_lip_down_pose")
    Y_Pos_Curve=(Type=Curve,Name="R_lower_lip_up_pose")
    Control.Name=RigVMVariableNode() {variable=R Lower Lip Control}.Value
   )
   > CrvToVectorCtrlFNC_7 = RigVMFunctionReferenceNode(
    Control=(Type=None,Name="None")
    X_Neg_Curve=(Type=Curve,Name="None")
    X_Pos_Curve=(Type=Curve,Name="None")
    Y_Neg_Curve=(Type=Curve,Name="upper_lip_dn_pose")
    Y_Pos_Curve=(Type=Curve,Name="upper_lip_up_pose")
    Control.Name=RigVMVariableNode() {variable=Upper Lip Control}.Value
   )
   > CrvToVectorCtrlFNC_8 = RigVMFunctionReferenceNode(
    Control=(Type=None,Name="None")
    X_Neg_Curve=(Type=Curve,Name="None")
    X_Pos_Curve=(Type=Curve,Name="None")
    Y_Neg_Curve=(Type=Curve,Name="lower_lip_dn_pose")
    Y_Pos_Curve=(Type=Curve,Name="lower_lip_up_pose")
    Control.Name=RigVMVariableNode() {variable=Lower Lip Control}.Value
   )
   CrvToVectorCtrlFNC_14 = RigVMFunctionReferenceNode(
    Control=(Type=None,Name="None")
    X_Neg_Curve=(Type=Curve,Name="L_brow_In_Dn_pose")
    X_Pos_Curve=(Type=Curve,Name="L_brow_In_Up_pose")
    Y_Neg_Curve=(Type=Curve,Name="L_brow_Ot_Dn_pose")
    Y_Pos_Curve=(Type=Curve,Name="L_brow_Ot_Up_pose")
    Control.Name=RigVMVariableNode() {variable=L Brow Control}.Value
    ExecuteContext=@RigVMFunction_Sequence_1.D
   )
   > CrvToVectorCtrlFNC_15 = RigVMFunctionReferenceNode(
    Control=(Type=None,Name="None")
    X_Neg_Curve=(Type=Curve,Name="R_brow_In_Dn_pose")
    X_Pos_Curve=(Type=Curve,Name="R_brow_In_Up_pose")
    Y_Neg_Curve=(Type=Curve,Name="R_brow_Ot_Dn_pose")
    Y_Pos_Curve=(Type=Curve,Name="R_brow_Ot_Up_pose")
    Control.Name=RigVMVariableNode() {variable=R Brow Control}.Value
   )
   > CrvToVectorCtrlFNC_16 = RigVMFunctionReferenceNode(
    Control=(Type=None,Name="None")
    X_Neg_Curve=(Type=Curve,Name="None")
    X_Pos_Curve=(Type=Curve,Name="R_brow_squeeze_in_pose")
    Y_Neg_Curve=(Type=Curve,Name="None")
    Y_Pos_Curve=(Type=Curve,Name="L_brow_squeeze_in_pose")
    Control.Name=RigVMVariableNode() {variable=Squeeze Control}.Value
   )
   > CrvToVectorCtrlFNC_17 = RigVMFunctionReferenceNode(
    Control=(Type=None,Name="None")
    X_Neg_Curve=(Type=Curve,Name="L_eye_look_left_pose")
    X_Pos_Curve=(Type=Curve,Name="L_eye_look_right_pose")
    Y_Neg_Curve=(Type=Curve,Name="L_eye_look_down_pose")
    Y_Pos_Curve=(Type=Curve,Name="L_eye_look_up_pose")
    Control.Name=RigVMVariableNode() {variable=L Eye Control}.Value
   )
   > CrvToVectorCtrlFNC_18 = RigVMFunctionReferenceNode(
    Control=(Type=None,Name="None")
    X_Neg_Curve=(Type=Curve,Name="R_eye_look_left_pose")
    X_Pos_Curve=(Type=Curve,Name="R_eye_look_right_pose")
    Y_Neg_Curve=(Type=Curve,Name="R_eye_look_down_pose")
    Y_Pos_Curve=(Type=Curve,Name="R_eye_look_up_pose")
    Control.Name=RigVMVariableNode() {variable=R Eye Control}.Value
   )
   > &RigUnit_SetControlVector2D(
    Vector=(X=0,Y=0)
    Weight=1
    Vector.Y=&RigUnit_GetCurveValue(Curve=L_lower_lid_up_pose).Value
    Control=RigVMVariableNode() {variable=L Low Lid}.Value
   ) {template="SetControlValue::Execute(in Control,in Weight,io Vector)"}
   > &RigUnit_SetControlVector2D(
    Vector=(X=0,Y=0)
    Weight=1
    Vector.Y=&RigUnit_GetCurveValue(Curve=R_lower_lid_up_pose).Value
    Control=RigVMVariableNode() {variable=R Low Lid}.Value
   ) {template="SetControlValue::Execute(in Control,in Weight,io Vector)"}
   > CrvToVectorCtrlFNC_19 = RigVMFunctionReferenceNode(
    Control=(Type=None,Name="None")
    X_Neg_Curve=(Type=Curve,Name="None")
    X_Pos_Curve=(Type=Curve,Name="None")
    Y_Neg_Curve=(Type=Curve,Name="L_blink_pose")
    Y_Pos_Curve=(Type=Curve,Name="L_wide_pose")
    Control.Name=RigVMVariableNode() {variable=L Blink Control}.Value
   )
   > CrvToVectorCtrlFNC_20 = RigVMFunctionReferenceNode(
    Control=(Type=None,Name="None")
    X_Neg_Curve=(Type=Curve,Name="None")
    X_Pos_Curve=(Type=Curve,Name="None")
    Y_Neg_Curve=(Type=Curve,Name="R_blink_pose")
    Y_Pos_Curve=(Type=Curve,Name="R_wide_pose")
    Control.Name=RigVMVariableNode() {variable=R Blink Control}.Value
   )
   &RigUnit_SetControlVector2D(
    Vector=(X=0,Y=0)
    Weight=1
    Vector.Y=&RigUnit_GetCurveValue(Curve=reverse_jaw_pose).Value
    Control=RigVMVariableNode() {variable=Reverse Jaw Control}.Value
    ExecutePin=@RigVMFunction_Sequence_1.C
   ) {template="SetControlValue::Execute(in Control,in Weight,io Vector)"}
   > CrvToVectorCtrlFNC_9 = RigVMFunctionReferenceNode(
    Control=(Type=None,Name="None")
    X_Neg_Curve=(Type=Curve,Name="tongue_right_pose")
    X_Pos_Curve=(Type=Curve,Name="tongue_left_pose")
    Y_Neg_Curve=(Type=Curve,Name="tongue_up_pose")
    Y_Pos_Curve=(Type=Curve,Name="tongue_down_pose")
    Control.Name=RigVMVariableNode() {variable=Tongue Control}.Value
   )
   > CrvToVectorCtrlFNC_10 = RigVMFunctionReferenceNode(
    Control=(Type=None,Name="None")
    X_Neg_Curve=(Type=Curve,Name="tongue_back_pose")
    X_Pos_Curve=(Type=Curve,Name="tongue_forward_pose")
    Y_Neg_Curve=(Type=Curve,Name="None")
    Y_Pos_Curve=(Type=Curve,Name="None")
    Control.Name=RigVMVariableNode() {variable=Tongue Fwd Control}.Value
   )
   > CrvToVectorCtrlFNC_11 = RigVMFunctionReferenceNode(
    Control=(Type=None,Name="None")
    X_Neg_Curve=(Type=Curve,Name="R_tongue_twist_pose")
    X_Pos_Curve=(Type=Curve,Name="L_tongue_twist_pose")
    Y_Neg_Curve=(Type=Curve,Name="None")
    Y_Pos_Curve=(Type=Curve,Name="None")
    Control.Name=RigVMVariableNode() {variable=Tongue Twist Control}.Value
   )
   > CrvToVectorCtrlFNC_12 = RigVMFunctionReferenceNode(
    Control=(Type=None,Name="None")
    X_Neg_Curve=(Type=Curve,Name="tongue_curl_tip_down_pose")
    X_Pos_Curve=(Type=Curve,Name="tongue_curl_tip_up_pose")
    Y_Neg_Curve=(Type=Curve,Name="None")
    Y_Pos_Curve=(Type=Curve,Name="None")
    Control.Name=RigVMVariableNode() {variable=Tongue Curl Tip Control}.Value
   )
   > CrvToVectorCtrlFNC_13 = RigVMFunctionReferenceNode(
    Control=(Type=None,Name="None")
    X_Neg_Curve=(Type=Curve,Name="tongue_curl_mid_down_pose")
    X_Pos_Curve=(Type=Curve,Name="tongue_curl_mid_up_pose")
    Y_Neg_Curve=(Type=Curve,Name="None")
    Y_Pos_Curve=(Type=Curve,Name="None")
    Control.Name=RigVMVariableNode() {variable=Tongue Curl Control}.Value
   )
   > &RigUnit_SetControlVector2D(
    Vector=(X=0,Y=0)
    Weight=1
    Vector.Y=&RigUnit_GetCurveValue(Curve=L_cheek_up_pose).Value
    Control=RigVMVariableNode() {variable=L Cheek Control}.Value
   ) {template="SetControlValue::Execute(in Control,in Weight,io Vector)"}
   > &RigUnit_SetControlVector2D(
    Vector=(X=0,Y=0)
    Weight=1
    Vector.Y=&RigUnit_GetCurveValue(Curve=R_cheek_up_pose).Value
    Control=RigVMVariableNode() {variable=R Cheek Control}.Value
   ) {template="SetControlValue::Execute(in Control,in Weight,io Vector)"}
   > &RigUnit_SetControlVector2D(
    Vector=(X=0,Y=0)
    Weight=1
    Vector.Y=&RigUnit_GetCurveValue(Curve=C_nose_wrinkler_pose).Value
    Control=RigVMVariableNode() {variable=Nose Control}.Value
   ) {template="SetControlValue::Execute(in Control,in Weight,io Vector)"}
   RigVMRerouteNode(Value=())
   VectorCtrlToCrvFNC_10 = RigVMFunctionReferenceNode(
    Control=(Type=Control,Name="None")
    ExecuteContext=()
    X_Neg_Curve=(Type=Curve,Name="None")
    X_Pos_Curve=(Type=Curve,Name="None")
    Y_Neg_Curve=(Type=Curve,Name="lower_lip_dn_pose")
    Y_Pos_Curve=(Type=Curve,Name="lower_lip_up_pose")
   )
  ]}
]
hierarchy [
 Head {type=Connector, loc=<0 0 0>, rot=<0 0 0 1>}
 Root {type=Socket, loc=<0 0 0>, rot=<0 0 0 1>}
 BothLegIK {type=Curve, loc=<0 0 0>, rot=<0 0 0 1>}
 look_right_pose {type=Curve, loc=<0 0 0>, rot=<0 0 0 1>}
 L_wide_pose {type=Curve, loc=<0 0 0>, rot=<0 0 0 1>}
 R_lip_corner_narrow_pose {type=Curve, loc=<0 0 0>, rot=<0 0 0 1>}
 R_upper_lip_lower_pose {type=Curve, loc=<0 0 0>, rot=<0 0 0 1>}
 L_lip_corner_narrow_pose {type=Curve, loc=<0 0 0>, rot=<0 0 0 1>}
 L_upper_lip_lower_pose {type=Curve, loc=<0 0 0>, rot=<0 0 0 1>}
 R_lower_lip_up_pose {type=Curve, loc=<0 0 0>, rot=<0 0 0 1>}
 R_brow_squeeze_in_pose {type=Curve, loc=<0 0 0>, rot=<0 0 0 1>}
 L_lip_corner_wide_pose {type=Curve, loc=<0 0 0>, rot=<0 0 0 1>}
 L_upperLidDown_pose {type=Curve, loc=<0 0 0>, rot=<0 0 0 1>}
 R_lower_lip_down_pose {type=Curve, loc=<0 0 0>, rot=<0 0 0 1>}
 R_upper_lip_raiser_pose {type=Curve, loc=<0 0 0>, rot=<0 0 0 1>}
 L_lower_lip_down_pose {type=Curve, loc=<0 0 0>, rot=<0 0 0 1>}
 L_upper_lip_raiser_pose {type=Curve, loc=<0 0 0>, rot=<0 0 0 1>}
 L_brow_squeeze_in_pose {type=Curve, loc=<0 0 0>, rot=<0 0 0 1>}
 L_lower_lip_up_pose {type=Curve, loc=<0 0 0>, rot=<0 0 0 1>}
 R_upperLidDown_pose {type=Curve, loc=<0 0 0>, rot=<0 0 0 1>}
 R_wide_pose {type=Curve, loc=<0 0 0>, rot=<0 0 0 1>}
 R_blink_pose {type=Curve, loc=<0 0 0>, rot=<0 0 0 1>}
 look_down_pose {type=Curve, loc=<0 0 0>, rot=<0 0 0 1>}
 look_up_pose {type=Curve, loc=<0 0 0>, rot=<0 0 0 1>}
 L_blink_pose {type=Curve, loc=<0 0 0>, rot=<0 0 0 1>}
 Reaction_Happy {type=Curve, loc=<0 0 0>, rot=<0 0 0 1>}
 look_left_pose {type=Curve, loc=<0 0 0>, rot=<0 0 0 1>}
 R_lip_corner_wide_pose {type=Curve, loc=<0 0 0>, rot=<0 0 0 1>}
 L_smile_pose {type=Curve, loc=<0 0 0>, rot=<0 0 0 1>}
 R_smile_pose {type=Curve, loc=<0 0 0>, rot=<0 0 0 1>}
 L_frown_pose {type=Curve, loc=<0 0 0>, rot=<0 0 0 1>}
 R_frown_pose {type=Curve, loc=<0 0 0>, rot=<0 0 0 1>}
 C_nose_wrinkler_pose {type=Curve, loc=<0 0 0>, rot=<0 0 0 1>}
 L_cheek_up_pose {type=Curve, loc=<0 0 0>, rot=<0 0 0 1>}
 R_cheek_up_pose {type=Curve, loc=<0 0 0>, rot=<0 0 0 1>}
 jaw_open_pose {type=Curve, loc=<0 0 0>, rot=<0 0 0 1>}
 jaw_left_pose {type=Curve, loc=<0 0 0>, rot=<0 0 0 1>}
 jaw_right_pose {type=Curve, loc=<0 0 0>, rot=<0 0 0 1>}
 jaw_up_pose {type=Curve, loc=<0 0 0>, rot=<0 0 0 1>}
 jaw_forward_pose {type=Curve, loc=<0 0 0>, rot=<0 0 0 1>}
 jaw_back_pose {type=Curve, loc=<0 0 0>, rot=<0 0 0 1>}
 eye_convergence_pose {type=Curve, loc=<0 0 0>, rot=<0 0 0 1>}
 tongue_forward_pose {type=Curve, loc=<0 0 0>, rot=<0 0 0 1>}
 tongue_back_pose {type=Curve, loc=<0 0 0>, rot=<0 0 0 1>}
 tongue_up_pose {type=Curve, loc=<0 0 0>, rot=<0 0 0 1>}
 tongue_down_pose {type=Curve, loc=<0 0 0>, rot=<0 0 0 1>}
 L_brow_In_Up_pose {type=Curve, loc=<0 0 0>, rot=<0 0 0 1>}
 L_brow_Ot_Up_pose {type=Curve, loc=<0 0 0>, rot=<0 0 0 1>}
 L_brow_In_Dn_pose {type=Curve, loc=<0 0 0>, rot=<0 0 0 1>}
 L_brow_Ot_Dn_pose {type=Curve, loc=<0 0 0>, rot=<0 0 0 1>}
 R_brow_In_Up_pose {type=Curve, loc=<0 0 0>, rot=<0 0 0 1>}
 R_brow_Ot_Up_pose {type=Curve, loc=<0 0 0>, rot=<0 0 0 1>}
 R_brow_In_Dn_pose {type=Curve, loc=<0 0 0>, rot=<0 0 0 1>}
 R_brow_Ot_Dn_pose {type=Curve, loc=<0 0 0>, rot=<0 0 0 1>}
 upper_lip_up_pose {type=Curve, loc=<0 0 0>, rot=<0 0 0 1>}
 upper_lip_dn_pose {type=Curve, loc=<0 0 0>, rot=<0 0 0 1>}
 lower_lip_up_pose {type=Curve, loc=<0 0 0>, rot=<0 0 0 1>}
 lower_lip_dn_pose {type=Curve, loc=<0 0 0>, rot=<0 0 0 1>}
 L_smile_open_correction_pose {type=Curve, loc=<0 0 0>, rot=<0 0 0 1>}
 R_smile_open_correction_pose {type=Curve, loc=<0 0 0>, rot=<0 0 0 1>}
 tongue_left_pose {type=Curve, loc=<0 0 0>, rot=<0 0 0 1>}
 tongue_right_pose {type=Curve, loc=<0 0 0>, rot=<0 0 0 1>}
 reverse_jaw_pose {type=Curve, loc=<0 0 0>, rot=<0 0 0 1>}
 base_pose {type=Curve, loc=<0 0 0>, rot=<0 0 0 1>}
 L_eye_look_up_pose {type=Curve, loc=<0 0 0>, rot=<0 0 0 1>}
 L_eye_look_left_pose {type=Curve, loc=<0 0 0>, rot=<0 0 0 1>}
 L_eye_look_down_pose {type=Curve, loc=<0 0 0>, rot=<0 0 0 1>}
 L_eye_look_right_pose {type=Curve, loc=<0 0 0>, rot=<0 0 0 1>}
 R_eye_look_left_pose {type=Curve, loc=<0 0 0>, rot=<0 0 0 1>}
 R_eye_look_right_pose {type=Curve, loc=<0 0 0>, rot=<0 0 0 1>}
 R_eye_look_up_pose {type=Curve, loc=<0 0 0>, rot=<0 0 0 1>}
 R_eye_look_down_pose {type=Curve, loc=<0 0 0>, rot=<0 0 0 1>}
 L_lower_lid_up_pose {type=Curve, loc=<0 0 0>, rot=<0 0 0 1>}
 R_lower_lid_up_pose {type=Curve, loc=<0 0 0>, rot=<0 0 0 1>}
 Root {loc=<0 0 0>, rot=<0 0 0 1>}
 pelvis {parent=Root, loc=<0 -15.3471 38.2106>, rot=<0.4921 0.5078 -0.4921 -0.5078>}
 spine_01 {parent=pelvis, loc=<0 0 0>, rot=<0 0 -0.0124 -0.9999>}
 spine_02 {parent=spine_01, loc=<13.8833 -0.3457 0>, rot=<0 0 0.0137 -0.9999>}
 spine_03 {parent=spine_02, loc=<15.5639 0 0>, rot=<0 0 -0.011 -0.9999>}
 neck_01 {parent=spine_03, loc=<1.1967 -0.0091 0>, rot=<0 0 0.4992 -0.8665>}
 neck_02 {parent=neck_01, loc=<13.2288 0 0>, rot=<0 0 0.0244 -0.9997>}
 Head {parent=neck_02, loc=<10.6145 0 0>, rot=<-0.1543 -0.6901 0.1543 -0.6901>}
 face_attach {parent=Head, loc=<0 0 0>, rot=<0 0 0 1>}
 jaw {parent=face_attach, loc=<0 1.5414 1.5673>, rot=<0 0.7071 0 -0.7071>}
 lip_lower_03_l {parent=jaw, loc=<-7.7319 9.0929 -3.1083>, rot=<-0.0567 -0.789 -0.0901 -0.605>}
 lip_lower_02_l {parent=jaw, loc=<-7.7793 10.6851 -2.7467>, rot=<-0.0802 -0.728 -0.1069 -0.6725>}
 lip_lower_01_l {parent=jaw, loc=<-7.8737 12.1434 -1.5952>, rot=<-0.1854 -0.7063 -0.1903 -0.6561>}
 lip_lower_mid {parent=jaw, loc=<-7.9235 13.3354 0>, rot=<0.4939 0.506 0.506 0.4939>}
 lip_lower_03_r {parent=jaw, loc=<-7.7318 9.0929 3.1083>, rot=<0.789 -0.0567 -0.605 0.0901>}
 lip_lower_01_r {parent=jaw, loc=<-7.8738 12.1435 1.5952>, rot=<0.7063 -0.1854 -0.6561 0.1903>}
 lip_lower_02_r {parent=jaw, loc=<-7.7793 10.6851 2.7467>, rot=<0.728 -0.0802 -0.6725 0.1069>}
 teeth_lower {parent=jaw, loc=<-7.9764 4.8748 0>, rot=<0 -0.7071 0 -0.7071>}
 tongue_01 {parent=jaw, loc=<-5.8122 -0.4681 0>, rot=<0.5405 -0.4559 0.4559 -0.5405>}
 tongue_02 {parent=tongue_01, loc=<-5.4528 0 0.1308>, rot=<0 -0.04 0 -0.9992>}
 tongue_03 {parent=tongue_02, loc=<-3.7239 0 0.1308>, rot=<0 -0.0699 0 -0.9976>}
 lip_corner_l {parent=face_attach, loc=<3.532 9.0864 -5.4803>, rot=<0.0323 -0.1402 -0.1027 -0.9843>}
 cheek_bt_l {parent=face_attach, loc=<5.2759 4.981 -3.3543>, rot=<0.0975 -0.0174 -0.1751 -0.9796>}
 cheek_bt_r {parent=face_attach, loc=<-5.2759 4.981 -3.3542>, rot=<0.9796 -0.1751 0.0174 0.0975>}
 lip_corner_r {parent=face_attach, loc=<-3.532 9.0864 -5.4803>, rot=<0.9843 -0.1027 0.1402 0.0323>}
 reverse_jaw {parent=face_attach, loc=<0 0.5985 -5.1509>, rot=<0 0.7071 0 -0.7071>}
 lip_upper_03_r {parent=reverse_jaw, loc=<0.129 9.9626 3.7462>, rot=<0.7924 -0.0562 -0.6006 0.0904>}
 lip_upper_02_r {parent=reverse_jaw, loc=<0.3249 12.1056 3.5072>, rot=<0.7311 -0.0539 -0.6742 0.0889>}
 lip_upper_01_r {parent=reverse_jaw, loc=<0.5665 14.2448 2.5854>, rot=<0.707 -0.195 -0.6436 0.219>}
 lip_upper_mid {parent=reverse_jaw, loc=<0.4164 15.4191 0>, rot=<-0.4824 -0.517 -0.517 -0.4824>}
 lip_upper_01_l {parent=reverse_jaw, loc=<0.5666 14.2448 -2.5853>, rot=<-0.195 -0.707 -0.219 -0.6436>}
 lip_upper_02_l {parent=reverse_jaw, loc=<0.3249 12.1056 -3.5072>, rot=<-0.0539 -0.7311 -0.0889 -0.6742>}
 lip_upper_03_l {parent=reverse_jaw, loc=<0.129 9.9627 -3.7462>, rot=<-0.0562 -0.7924 -0.0904 -0.6006>}
 cheek_tp_in_l {parent=reverse_jaw, loc=<5.1286 8.5302 -1.8132>, rot=<-0.2594 -0.6712 -0.2639 -0.6423>}
 cheek_tp_ot_l {parent=reverse_jaw, loc=<4.5625 5.3738 -4.7949>, rot=<-0.1358 -0.6932 -0.1692 -0.6874>}
 cheek_tp_in_r {parent=reverse_jaw, loc=<5.1285 8.5302 1.8132>, rot=<0.6712 -0.2594 -0.6423 0.2639>}
 cheek_tp_ot_r {parent=reverse_jaw, loc=<4.5624 5.3738 4.7949>, rot=<0.6932 -0.1358 -0.6874 0.1692>}
 brow_ot_r {parent=reverse_jaw, loc=<8.808 4.5312 5.3802>, rot=<-0.6792 -0.1966 0.6792 0.1966>}
 brow_md_r {parent=reverse_jaw, loc=<9.9048 6.1058 4.3126>, rot=<-0.6803 -0.1929 0.6803 0.1929>}
 brow_ct {parent=reverse_jaw, loc=<8.3974 8.8869 0>, rot=<0.1061 -0.6991 0.1061 -0.6991>}
 brow_md_l {parent=reverse_jaw, loc=<9.9048 6.1057 -4.3125>, rot=<0.1929 -0.6803 0.1929 -0.6803>}
 brow_in_l {parent=reverse_jaw, loc=<9.7244 6.9681 -1.8436>, rot=<0.1061 -0.6991 0.1061 -0.6991>}
 brow_ot_l {parent=reverse_jaw, loc=<8.808 4.5312 -5.3802>, rot=<0.1966 -0.6792 0.1966 -0.6792>}
 brow_in_r {parent=reverse_jaw, loc=<9.7245 6.9681 1.8436>, rot=<-0.6991 -0.1061 0.6991 0.1061>}
 eye_r {parent=reverse_jaw, loc=<7.2257 6.2822 3.0323>, rot=<0.693 0.1403 -0.6874 -0.1657>}
 ear_01_r {parent=reverse_jaw, loc=<7.5029 -0.0902 0.7241>, rot=<-0.4789 -0.1625 0.842 0.1881>}
 ear_02_r {parent=ear_01_r, loc=<-6.2909 -0.561 0.6142>, rot=<-0.1707 -0.6662 -0.3078 -0.6575>}
 ear_03_r {parent=ear_02_r, loc=<-4.7445 -0.0001 0>, rot=<0.0117 -0.2533 0.0982 -0.9623>}
 ear_04_r {parent=ear_03_r, loc=<-5.0608 0.0001 -0.0001>, rot=<-0.0008 -0.0406 -0.0207 -0.999>}
 ear_01_l {parent=reverse_jaw, loc=<7.5029 -0.0902 -0.7241>, rot=<0.1625 -0.4789 0.1881 -0.842>}
 ear_02_l {parent=ear_01_l, loc=<6.6137 0 0>, rot=<-0.1396 -0.1853 -0.149 -0.9612>}
 ear_03_l {parent=ear_02_l, loc=<3.9918 0 0>, rot=<-0.1664 -0.2168 0.0087 -0.9619>}
 ear_04_l {parent=ear_03_l, loc=<3.9148 0 0>, rot=<0.0672 -0.1399 0.0639 -0.9858>}
 eye_l {parent=reverse_jaw, loc=<7.2257 6.2822 -3.0323>, rot=<0.1403 -0.693 0.1657 -0.6874>}
 lid_bt_r {parent=reverse_jaw, loc=<7.2257 6.2822 3.0323>, rot=<-0.6903 -0.153 0.6903 0.153>}
 lid_tp_r {parent=reverse_jaw, loc=<7.2257 6.2822 3.0323>, rot=<-0.6903 -0.153 0.6903 0.153>}
 nose_mid {parent=reverse_jaw, loc=<3.1478 16.8243 0>, rot=<-0.546 -0.4493 -0.4493 -0.546>}
 nostril_l {parent=nose_mid, loc=<0.2659 -1.23 0.259>, rot=<0 -0.2275 0 -0.9738>}
 nostril_r {parent=nose_mid, loc=<0.2659 1.23 0.259>, rot=<0 -0.9738 0 0.2275>}
 lid_tp_l {parent=reverse_jaw, loc=<7.2257 6.2822 -3.0323>, rot=<0.153 -0.6903 0.153 -0.6903>}
 lid_bt_l {parent=reverse_jaw, loc=<7.2257 6.2822 -3.0323>, rot=<0.153 -0.6903 0.153 -0.6903>}
 teeth_upper {parent=reverse_jaw, loc=<2.2634 6.6577 0>, rot=<0 -0.7071 0 -0.7071>}
 shoulder_fr_l {parent=spine_03, loc=<3.186 4.1292 -7.8083>, rot=<0.0037 0.0034 -0.8214 -0.5703>}
 elbow_fr_l {parent=shoulder_fr_l, loc=<13.7779 0 0>, rot=<-0.0001 -0.0001 0.1796 -0.9837>}
 wrist_fr_l {parent=elbow_fr_l, loc=<15.1156 0 0>, rot=<0.994 -0.0933 0.0203 -0.0527>}
 ball_fr_l {parent=wrist_fr_l, loc=<4.5512 0 0>, rot=<-0.0153 0.0292 -0.3509 -0.9358>}
 toe_fr_l {parent=ball_fr_l, loc=<1.9055 0 0>, rot=<-0.0007 -0.0415 -0.2548 -0.9661>}
 toe_tip_fr_l {parent=toe_fr_l, loc=<4.6825 0 0>, rot=<0.0046 0.0116 -0.0519 -0.9986>}
 scapula_fr_l {parent=shoulder_fr_l, loc=<-6.9459 3.8453 -0.005>, rot=<-0.0048 0.0016 0.2501 -0.9682>}
 shoulder_fr_r {parent=spine_03, loc=<3.186 4.1293 7.8084>, rot=<0.0034 -0.0037 0.5703 -0.8214>}
 elbow_fr_r {parent=shoulder_fr_r, loc=<-13.7779 0 -0.0001>, rot=<-0.0001 -0.0001 0.1796 -0.9837>}
 wrist_fr_r {parent=elbow_fr_r, loc=<-14.4175 -0.1127 0.0153>, rot=<0.994 -0.0933 0.0203 -0.0527>}
 ball_fr_r {parent=wrist_fr_r, loc=<-5.2581 0.0196 -0.0085>, rot=<-0.0153 0.0292 -0.3509 -0.9358>}
 toe_fr_r {parent=ball_fr_r, loc=<-1.9055 0 0>, rot=<-0.0007 -0.0415 -0.2548 -0.9661>}
 toe_tip_fr_r {parent=toe_fr_r, loc=<-4.6824 0 0.0001>, rot=<0.0046 0.0116 -0.0519 -0.9986>}
 scapula_fr_r {parent=shoulder_fr_r, loc=<6.9459 -3.8454 0.005>, rot=<-0.0048 0.0016 0.2501 -0.9682>}
 attach_backpack {parent=spine_02, loc=<14.3456 0 0>, rot=<0 0 0.004 -1>}
 attach_cape {parent=spine_02, loc=<14.3456 0 0>, rot=<0 0 0.004 -1>}
 hips {parent=pelvis, loc=<0 0 0>, rot=<0 0 0 1>}
 tail_01 {parent=hips, loc=<-7.6516 -5.3072 0>, rot=<0.0124 -0.707 0.7071 -0.0069>}
 tail_02 {parent=tail_01, loc=<8.4235 0 0>, rot=<0.0019 -0.049 -0.0364 -0.9981>}
 tail_03 {parent=tail_02, loc=<8.4131 0 0>, rot=<-0.0024 0.0623 0.061 -0.9962>}
 tail_04 {parent=tail_03, loc=<8.4174 0 0>, rot=<0.0004 0.0102 0.0126 -0.9999>}
 thigh_bk_l {parent=hips, loc=<-4.5411 0.7536 -6.4095>, rot=<-0.6756 -0.7373 0 0>}
 knee_bk_l {parent=thigh_bk_l, loc=<15.1791 0 0>, rot=<0 0 0.3983 -0.9172>}
 ankle_bk_l {parent=knee_bk_l, loc=<12.2939 0 0>, rot=<0 0 -0.4247 -0.9053>}
 ball_bk_l {parent=ankle_bk_l, loc=<11.5692 0 0>, rot=<0 0 -0.4268 -0.9043>}
 toe_bk_l {parent=ball_bk_l, loc=<1.4603 0 0>, rot=<0 0 -0.1105 -0.9939>}
 toe_tip_bk_l {parent=toe_bk_l, loc=<3.6201 0 0>, rot=<0 0 -0.2748 -0.9615>}
 thigh_bk_r {parent=hips, loc=<-4.5411 0.7536 6.4095>, rot=<0.7373 -0.6756 0 0>}
 knee_bk_r {parent=thigh_bk_r, loc=<-15.1792 0 0>, rot=<0 0 0.3983 -0.9172>}
 ankle_bk_r {parent=knee_bk_r, loc=<-12.2939 0 0>, rot=<0 0 -0.4247 -0.9053>}
 ball_bk_r {parent=ankle_bk_r, loc=<-11.5691 0 0>, rot=<0 0 -0.4268 -0.9043>}
 toe_bk_r {parent=ball_bk_r, loc=<-1.4603 0 0>, rot=<0 0 -0.1105 -0.9939>}
 toe_tip_bk_r {parent=toe_bk_r, loc=<-3.6201 0 0>, rot=<0 0 -0.2748 -0.9615>}
 IK_Foot_Rear_Root {parent=Root, loc=<0 -30.9193 0>, rot=<0 0 0 1>}
 IK_Foot_Rear_L {parent=IK_Foot_Rear_Root, loc=<6.4093 0.7896 2.7138>, rot=<-0.3024 -0.6392 -0.3024 -0.6392>}
 IK_Foot_Rear_R {parent=IK_Foot_Rear_Root, loc=<-6.4093 0.7896 2.7138>, rot=<0.6392 -0.3024 0.6392 -0.3024>}
 IK_Foot_Front_Root {parent=Root, loc=<0 4.0614 0>, rot=<0.135 0 0 -0.9908>}
 IK_Foot_Front_L {parent=IK_Foot_Front_Root, loc=<7.6694 8.7288 5.0698>, rot=<-0.4464 -0.5742 -0.3431 -0.5944>}
 IK_Foot_Front_R {parent=IK_Foot_Front_Root, loc=<-7.6694 8.7288 5.0698>, rot=<0.5944 -0.3431 0.5742 -0.4464>}
 tongue_curl_tip_up_pose {type=Curve, loc=<0 0 0>, rot=<0 0 0 1>}
 tongue_curl_tip_down_pose {type=Curve, loc=<0 0 0>, rot=<0 0 0 1>}
 tongue_curl_mid_up_pose {type=Curve, loc=<0 0 0>, rot=<0 0 0 1>}
 tongue_curl_mid_down_pose {type=Curve, loc=<0 0 0>, rot=<0 0 0 1>}
 L_tongue_twist_pose {type=Curve, loc=<0 0 0>, rot=<0 0 0 1>}
 R_tongue_twist_pose {type=Curve, loc=<0 0 0>, rot=<0 0 0 1>}
]
variables [
 BK Control {type=FName, default=None}
 Eye Control {type=FName, default=None}
 Face Rig Control {type=FName, default=None}
 Jaw Main Control {type=FName, default=None}
 L Blink Control {type=FName, default=None}
 L Brow Control {type=FName, default=None}
 L Cheek Control {type=FName, default=False}
 L Corner Control {type=FName, default=None}
 L Eye Control {type=FName, default=None}
 L Low Lid {type=FName, default=None}
 L Lower Lip Control {type=FName, default=None}
 L Upper Lip Control {type=FName, default=None}
 Lower Lip Control {type=FName, default=None}
 Nose Control {type=FName, default=None}
 Parent Connector {type=FRigElementKey, default=(Type=Connector,Name="")}
 R Blink Control {type=FName, default=None}
 R Brow Control {type=FName, default=None}
 R Cheek Control {type=FName, default=None}
 R Corner Control {type=FName, default=None}
 R Eye Control {type=FName, default=None}
 R Low Lid {type=FName, default=None}
 R Lower Lip Control {type=FName, default=None}
 R Upper Lip Control {type=FName, default=None}
 Reverse Jaw Control {type=FName, default=None}
 Squeeze Control {type=FName, default=None}
 Tongue Control {type=FName, default=None}
 Tongue Curl Control {type=FName, default=None}
 Tongue Curl Tip Control {type=FName, default=None}
 Tongue Fwd Control {type=FName, default=None}
 Tongue Twist Control {type=FName, default=None}
 Upper Lip Control {type=FName, default=None}
 vector2d {type=FVector2D, default=(X=0,Y=0)}
]
